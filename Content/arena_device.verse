using { /Fortnite.com/Devices }
using { /Fortnite.com/UI }
using { /Verse.org/Simulation }
using { /Verse.org/Random }
using { /Verse.org/Colors }
using { /Verse.org/Assets }
using { /Fortnite.com/Characters }
using { /Fortnite.com/FortPlayerUtilities }
using { /Fortnite.com/Game }
using { /Fortnite.com/Teams }
using { /UnrealEngine.com/Temporary/SpatialMath }

#Pros - Good For Multiple Maps with no modifiers - most cases
#Cons - bad for 3 new minigames in the sceond voting slot (complicated)

# enum_modifiers := enum:
#     FirstPerson
#     FOV
#     SlowMo
#     RandomCrosshair
#     LightOutLogic

# (Texture: enum_modifiers).GetTexture():texture=
#     case(Texture):
#         enum_modifiers.FirstPerson => VerseExposed.SQ_INVISIBLE
#         enum_modifiers.FOV => VerseExposed.SQ_INVISIBLE
#         enum_modifiers.SlowMo => VerseExposed.SQ_INVISIBLE
#         enum_modifiers.RandomCrosshair => VerseExposed.SQ_INVISIBLE
#         enum_modifiers.LightOutLogic => VerseExposed.SQ_INVISIBLE

arena_override_texture := enum:
    Precision

(Texture: arena_override_texture).GetTexture()<transacts>:texture=
    case(Texture):
        arena_override_texture.Precision => VerseExposed.ArenaOverridePrecision

arena_and_modifier_settings := class<unique><concrete>:
    @editable
    Debug : logic = false
    @editable
    ArenaDevice : ?arena_device = false # Optional to fix verse bug
    @editable
    OptBasicModifiers : ?basic_arena_modifier_settings = false
    @editable
    OptVerseModifierDevice : ?verse_functions = false
    # Modifier 3 TODO
    # @editable
    # ENUM_Modifiers : ?[]enum_modifiers  = false
    @editable
    ArenaImageOverride : ?arena_override_texture = false
    @editable
    ModifierImage1 : ?modifier_texture = false
    @editable
    ModifierImage2 : ?modifier_texture = false
    @editable
    ModifierImage3 : ?modifier_texture = false
    # var ModifierImage3 : ?enum_modifiers = false

    # RandomizeEnumModifier():void=
    #     Print("RandomizeEnumModifier")
    #     if(EnumArray := ENUM_Modifiers?):
    #         if(RandomEnumModifier := Shuffle(EnumArray)[0]):
    #             set ModifierImage3 = option{RandomEnumModifier}

arena_props := class<concrete>:
    @editable CreativeProp : creative_prop = creative_prop{}
    @editable NewLocation : vector3 = vector3{}
    @editable OriginLocation : vector3 = vector3{}

arena_device := class<unique>(crashable_device):
    @editable
    Name : string = ""

    @editable
    Storm : basic_storm_controller_device = basic_storm_controller_device{}

    @editable
    OptArenaStartingBarriers : ?[]barrier_device = false

    @editable
    CustomTeleportationMethod : ?minigame_teleport_method = false

    @editable
    BeginningLoadingSequence :?cinematic_sequence_device = false

    @editable
    EndGameClearData : minigame_device_clear_data = minigame_device_clear_data{}

    @editable
    ArenaPropsToTeleport : ?[]arena_props = false

    GetVotingImage()<transacts>: texture = VerseExposed.UI_UNKNOWN
    
    DeviceName<override>:string = "__GENERATED__"
    GetDeviceName<override>()<transacts><decides>:string = Name.Length > 0 and Name or "UNKNOWN ARENA NAME"


# Basic Modifiers
# These modifiers overwrite the basic defaults of the minigame such as loadout, storm, 
# EX: Changing the storm , starting with a different loadout, zero build
basic_arena_modifier_settings := class<concrete>:
    @editable
    ModStorm : ?basic_storm_controller_device = false

    @editable
    ModSelectableAndDefaultLoadout: ?minigame_selectable_default_loadout_items = false

    @editable
    ModDefaultItemGranters : ?[]minigame_device_granter_config = false

    @editable
    ModBottomMessage : ?string = false

    @editable
    ModSafeTimeConfig : ?minigame_device_safetime_config = false

    @editable
    ModRoundMutatorConfig : ?minigame_device_roundwise_mutator_config = false

    @editable
    ModEndTimerConfig : ?minigame_device_generic_endtime_config = false

    @editable
    ModItemTimerConfig : ?minigame_device_generic_itemtimer_config = false

    @editable
    ModClassSelector : ?class_and_team_selector_device = false

    @editable
    ModVitals : ?fort_character_vitals = false

    @editable
    ModBarriers : ?[]barrier_device = false

    # #TODO
    # @editable
    # ModAdditionalSiphon : ?minigame_device_siphon_config = false

    # #TODO
    # @editable
    # ModSiphonGranters : ?[]smart_granter_config = false

verse_functions := class<unique>(crashable_device):
    #Replace Main Functions
    #Start
    _Mod_Start()<suspends>:void=return
    #SafetimeStart
    _Mod_OnSafeTimeEnded()<suspends>:void=return
    #End
    _Mod_OnMinigameEnded(Winners: []player)<suspends>:void=return
    #Reset
    _Mod_Reset()<suspends>:void=return
    #OnEliminated
    _Mod_OnPlayerEliminated(Result: elimination_result)<suspends>:void=return
    #OnGenericEndTimerSuccess
    _Mod_OnGenericEndTimerSuccessReturnArray():[]player=array{}
    #_OnGenericItemTimerSuccessReturnArray
    _Mod_OnGenericItemTimerSuccessReturnArray():[]player=array{}

modifier_texture := enum:
    ReaperSniper
    MammothPistol

(Texture: modifier_texture).GetTexture()<transacts>:texture=
    case(Texture):
        modifier_texture.ReaperSniper => VerseExposed.mod_reaper
        modifier_texture.MammothPistol => VerseExposed.mod_hand_cannon

arena_default_box := class(arena_device):
    Name<override>: string = "Default Box"
    GetVotingImage<override>()<transacts>: texture = VerseExposed.arena_default_box

arena_big_box := class(arena_device):
    Name<override>: string = "BIG BOY BOX"
    GetVotingImage<override>()<transacts>: texture = VerseExposed.arena_big_box

arena_pleasant_park := class(arena_device):
    Name<override>: string = "Pleasant Park"
    GetVotingImage<override>()<transacts>: texture = VerseExposed.arena_pleasant_park

arena_zone_wars := class(arena_device):
    Name<override>: string = "Zone Wars"
    GetVotingImage<override>()<transacts>: texture = VerseExposed.arena_zonewars

arena_gazebo_device := class(arena_device):
    Name<override>: string = "Gazebo"
    GetVotingImage<override>()<transacts>: texture = VerseExposed.arena_gazebo

arena_default_box_zero_build_device := class(arena_device):
    Name<override>: string = "Zero Build Box"
    GetVotingImage<override>()<transacts>: texture = VerseExposed.arena_zero_build_box
    

#verse_function devices
#TBD? - Have toggles so that these functions will add on rather than completely replace every minigame function
verse_functions_test := class(verse_functions):
    _Mod_Start<override>()<suspends>: void =
        Print("Doing verse_functions_test _Mod_Start")

    _Mod_OnSafeTimeEnded<override>()<suspends>:void=
        Print("Doing verse_functions_test _Mod_OnSafeTimeEnded")

    _Mod_OnMinigameEnded<override>(Winners: []player)<suspends>:void=
        Print("Doing verse_functions_test _Mod_OnMinigameEnded")
        
    _Mod_Reset<override>()<suspends>:void=
        Print("Doing verse_functions_test _Mod_Reset")

    _Mod_OnPlayerEliminated<override>(Result: elimination_result)<suspends>:void=
        Print("Doing verse_functions_test _Mod_OnPlayerEliminated")

    _Mod_OnGenericEndTimerSuccessReturnArray<override>():[]player=
        Print("Doing verse_functions_test _Mod_OnGenericEndTimerSuccessReturnArray")
        return array{}

    _Mod_OnGenericItemTimerSuccessReturnArray<override>():[]player=
        Print("Doing verse_functions_test _Mod_OnGenericItemTimerSuccessReturnArray")
        return array{}

    #Use this for additional code in the normal minigame START function
    Mod_AdditionalStart()<suspends>:void=
        Print("Doing verse_functions_test Mod_AdditionalStart")

    Mod_AdditionalOnSafeTimeEnded()<suspends>:void=
        Print("Doing verse_functions_test Mod_AdditionalOnSafeTimeEnded")

    Mod_AdditionalOnMinigameEnded()<suspends>:void=
        Print("Doing verse_functions_test Mod_AdditionalOnMinigameEnded")

    Mod_AdditionalReset()<suspends>:void=
        Print("Doing verse_functions_test Mod_AdditionalReset")
