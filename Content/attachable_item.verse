using { /Fortnite.com/Devices }
using { /Fortnite.com/Game }
using { /Fortnite.com/Characters }
using { /UnrealEngine.com/Temporary/SpatialMath }
using { /UnrealEngine.com/Temporary/Diagnostics }
using { /Verse.org/Simulation }
using { /Verse.org/Concurrency }


attachable_item_type := enum:
    VFXSpawner
    CreativeProp

attachable_item_options := class<concrete>:
    @editable
    Type : attachable_item_type = attachable_item_type.CreativeProp

    @editable
    StartsPaused : logic = false

    @editable
    StartsHidden : logic = false

    @editable
    StartsAtTargetRotation : logic = true

    @editable
    TeleportToTargetWhenAttached : logic = false

    @editable
    FollowsTargetRotation : logic = true

    @editable
    OffsetFromTarget : vector3 = vector3{}

    @editable
    DesiredRefreshRate : type {Value:float where 0.1 <= Value, Value <= 30.0} = 30.0

    # IMPROVE : Convert to speed in m/s
    @editable
    MovablePropsFollowSpeed : type {Value:float where 0.1 <= Value, Value <= 1.0} = 1.0

    @editable
    MoveToDistanceThreshold : float = 0.0


attachable_item := class:
    Item : positional
    Options : attachable_item_options

    var SleepBetweenRefreshes : float = 0.0
    var MoveToDuration : float = 0.0
    var OriginTransform : transform = transform{}

    var TargetScaleTransitions<private> : []tuple(vector3,float,float) = array{}

    # OnFirstMoveInitiated : type {__(:fort_character)<suspends>:void}
    # OnFirstMoveInitiated : my_subscribable_attached_item__character = my_subscribable_attached_item__character{}
    OnFirstMoveInitiated : custom_subscribable_attachable_item = custom_subscribable_attachable_item{}
    OnDetached : custom_subscribable_attachable_item = custom_subscribable_attachable_item{}
    # OnPlayerAttachedObjectDetached : type {__(:fort_character)<suspends>:void}
    # ReleaseObjectEvent : event() = event(){}

    # ObjectAssignedEvent:event(tuple(creative_object, player, int)) = event(tuple(creative_object, player, int)){}
    # OnDetached: my_subscribable_player = my_subscribable_player{}

    block:
        set SleepBetweenRefreshes = if(Options.DesiredRefreshRate >= 30.0) then 0.0 else (1.0 / Options.DesiredRefreshRate)
        set MoveToDuration = Max(SleepBetweenRefreshes * (1.0 / Options.MovablePropsFollowSpeed), SleepBetweenRefreshes + 3.0/30.0)
        set OriginTransform = Item.GetTransform()

    SetTargetScale(Scale: vector3, TransitionDuration: float, ?SkipOldTransitions : logic = false):void=
        if(SkipOldTransitions?):
            set TargetScaleTransitions = array{}
        set TargetScaleTransitions += array{(Scale, TransitionDuration, GetSimulationElapsedTime())}
        # Print("SETTING TARGET SCALE OF {Scale.X}")

    # SetTargetScale(Scale: vector3):void=
    #     set TargetScaleTransitions = array{}
    #     Print("SETTING TARGET SCALE OF {Scale.X}")

    GetItemScale()<transacts>:vector3=
        CurrentScale := Item.GetTransform().Scale

        if(Transition := TargetScaleTransitions[0]):

            TargetScale := Transition(0)
            TransitionDuration := Transition(1)
            TransitionStartTime := Transition(2)

            # PrintT("Found {TargetScale} {TransitionDuration} {TransitionStartTime}")

            LerpAlpha := (TransitionDuration = 0.0 and 1.0) or Clamp((GetSimulationElapsedTime() - TransitionStartTime) / TransitionDuration, 0.0, 1.0)
            # PrintT("LerpAlpha={LerpAlpha}")

            if(
                TransitionDuration = 0.0 or LerpAlpha > 0.9, # Close to end of animation
                TargetScaleTransitions.Length > 1, # Another animation is queued (so we can finish the current animation completly otherwise)
                UpdatedArray := TargetScaleTransitions.RemoveElement[0]
            ):
                set TargetScaleTransitions = UpdatedArray

                # Look for next transition
                return GetItemScale()
            else:
                return Lerp(CurrentScale, TargetScale, LerpAlpha)

        return CurrentScale

    GetFollowTransform(FollowTarget: positional, PlayerRotated: logic)<transacts>:transform =
        # Translate to dynamically computed player feet position (since default position changes when the character crouches)
        DynamicOffset := if(Character := fort_character[FollowTarget])then Character.GetPlayerFeetTranslation() else vector3{}

        transform:
            Translation := DynamicOffset + FollowTarget.GetTransform().Rotation.RotateVector(Options.OffsetFromTarget)
            Rotation := if(PlayerRotated?) then FollowTarget.GetTransform().Rotation else Item.GetTransform().Rotation
            Scale := GetItemScale()

    ApplyMovementMethod(FollowTarget: positional, PlayerRotated: logic, ?Instant:logic = false):void=
        case(Options.Type):
            attachable_item_type.CreativeProp =>
                if(AsProp := creative_prop[Item]):
                    if(Instant?):
                        option{AsProp.TeleportTo[GetFollowTransform(FollowTarget, PlayerRotated)]}
                    else:

                        # (teleport detection) Teleport prop instead of moving if distance threshold condition is met
                        if(Options.MoveToDistanceThreshold > 0.0, Distance(FollowTarget.GetTransform().Translation, Item.GetTransform().Translation) > Options.MoveToDistanceThreshold):
                            option{AsProp.TeleportTo[GetFollowTransform(FollowTarget, PlayerRotated)]}
                        else:
                            spawn{AsProp.MoveTo(GetFollowTransform(FollowTarget, PlayerRotated), MoveToDuration)}

            attachable_item_type.VFXSpawner =>
                if(AsVfxSpawner := vfx_spawner_device[Item]):
                    option{AsVfxSpawner.TeleportTo[GetFollowTransform(FollowTarget, PlayerRotated)]}

    var ShouldDetach : logic = false

    var IsPaused : logic = false
    Pause()<transacts>:void= set IsPaused = true
    Resume()<transacts>:void= set IsPaused = false

    AttachToPositional(FollowTarget: positional):void=
        spawn{_AttachToPositional(FollowTarget)}

    _AttachToPositional(FollowTarget: positional)<suspends>:void=
        if(Options.StartsHidden?):
            if(Prop := creative_prop[Item]):
                Prop.Hide()

        ApplyMovementMethod(FollowTarget, Options.StartsAtTargetRotation, ?Instant := Options.TeleportToTargetWhenAttached)
            
        # IMPROVE: Could do an infinite loop if we call AttachToPlayer again from this Signal()
        # Print("OnFirstMoveInitiated.Signal()")
        OnFirstMoveInitiated.Subscribable.Signal(Self, FollowTarget)
        set ShouldDetach = false
        set IsPaused = false

        if(Options.StartsPaused?):
            Pause()

        loop:
            if(ShouldDetach?):
                break
                
                # if(not Character.IsActive[]):
                #     if(ReleasesOnPlayerDeath?):
                #         break
                # else:
            if(not IsPaused?):
                ApplyMovementMethod(FollowTarget, Options.FollowsTargetRotation)

            Sleep(SleepBetweenRefreshes)

        # OnPlayerAttachedObjectDetached(Character)

        # Print("OnDetached.Signal()")
        OnDetached.Subscribable.Signal(Self, FollowTarget)
        # ReleaseItem(Index)

    # AttachToPlayer(Player: player)<suspends>:void=
    #     if(Character := Player.GetFortCharacter[], Character.IsActive[]):
    #         ApplyMovementMethod(Character, Options.StartsAtPlayerRotation, ?Instant := Options.TeleportToPlayerWhenAttached)
            
    #         # IMPROVE: Could do an infinite loop if we call AttachToPlayer again from this Signal()
    #         Print("OnFirstMoveInitiated.Signal()")
    #         OnFirstMoveInitiated.Signal(Self, Character)
    #         set ShouldDetach = false

    #         loop:
    #             if(ShouldDetach?):
    #                 break
                    
    #                 # if(not Character.IsActive[]):
    #                 #     if(ReleasesOnPlayerDeath?):
    #                 #         break
    #                 # else:
    #             if(not IsPaused?):
    #                 ApplyMovementMethod(Character, Options.FollowsPlayerRotation)

    #             Sleep(SleepBetweenRefreshes)

    #         # OnPlayerAttachedObjectDetached(Character)
    #     else:
    #         Print("Couldn't trigger initial movement method!!!")

    #     Print("OnDetached.Signal()")
    #     OnDetached.Signal(Self, Player)
    #     # ReleaseItem(Index)

    Detach(?ShouldResetTransform:logic = true):void=
        # OnDetached.CancelAll()
        # Print("DETACHING")
        set ShouldDetach = true
        set TargetScaleTransitions = array{}
        if(ShouldResetTransform?):
            ResetTransform()

    ResetTransform():void=
        case(Options.Type):
            attachable_item_type.CreativeProp =>
                if(AsProp := creative_prop[Item]):
                    option{AsProp.TeleportTo[OriginTransform]}
            attachable_item_type.VFXSpawner =>
                if(AsVfxSpawner := vfx_spawner_device[Item]):
                    option{AsVfxSpawner.TeleportTo[OriginTransform]}