
# using { /Fortnite.com/Devices }
# using { /Verse.org/Simulation }
# using { /UnrealEngine.com/Temporary/Diagnostics }
# using { /Verse.org/Assets }
# using { /UnrealEngine.com/Temporary/UI }
# using { /UnrealEngine.com/Temporary/SpatialMath }
# using { /Verse.org/Simulation/Tags }

# # See https://dev.epicgames.com/documentation/en-us/uefn/create-your-own-device-in-verse for how to create a verse device.
# ability_defaultgranter := class(tag) {}
# # A Verse-authored creative device that can be placed in a level
# base_ability_class := class(crashable_device):
#     bMiniGameSuperHero : ?minigame75_SuperHeros = false
#     bPlayer : ?player = false
#     HeroType : Hero_type = Hero_type.WallHack
#     var UltCharge : int = 2
#     var WallHackVFX : vfx_spawner_device = vfx_spawner_device{}
#     var UltUsed : logic = false
#     # Hud Textures
#     var UltimateAbilityTexture : texture = VerseExposed.T_WallHackUltimate
#     var UltimateNoChargeTexture : texture = VerseExposed.T_WallHackUltimate
#     var UltimateAbilityImage : ?texture_block = false
#     UltimateHUDBackgroundTexture: texture = VerseExposed.T_UltHudBackGround
#     var UltimateHUDBackgroundImage : ?texture_block = false
#     UltimateCharge1 : texture = VerseExposed.T_UltCharge1
#     UltimateCharge2 : texture = VerseExposed.T_UltCharge2
#     var GameEnd : logic = false
#     var Canva : canvas = canvas{}
#     var DefaultItemGanter : item_granter_device = item_granter_device{}
#     DefaultGranterTag : ability_defaultgranter = ability_defaultgranter{}
#     var DefaultItemGranter : item_granter_device = item_granter_device{}

#     # Runs when the device is started in a running game
#     OnCrashableDeviceBegin<final><override>()<suspends>:void=
#         ## DONT USE THIS FUNCTION
#         block:

#     OnUltimatePressed(Agent : agent):void=
#         if(Player := player[Agent], UltCharge >= 2, MiniGameSuperHero := bMiniGameSuperHero?, UltUsed = false, ThisPlayer := bPlayer?, Player = ThisPlayer):
#             set UltUsed = true
#             Print("Ultimate")
#             #MiniGameSuperHero.UltimateUseAudio.Play(ThisPlayer)
#             spawn{_AbilityActivate(ThisPlayer, MiniGameSuperHero)}

#     Init<final>(Player : player, MiniGameSuperHero : minigame75_SuperHeros):void=
#         #set bUltimateNoChargeTexture = option{UltimateAbilityTextureNoC}
#         #set bUltimateAbilityTexture = option{UltimateAbilityTexture}
#         TagObjects := GetUniqueCreativeObjectsWithTag(DefaultGranterTag)
#         for (TagObject : TagObjects):
#             if (DefaultGranter := item_granter_device[TagObject]):
#                 set DefaultItemGranter = DefaultGranter
#         if(ThisPlayer := bPlayer?, PlayerUI := GetPlayerUI[ThisPlayer]):
#             DefaultItemGranter.GrantItem(ThisPlayer)
#             set Canva = MakeHeroUI()
#                 PlayerUI.AddWidget(Canva, player_ui_slot{InputMode := ui_input_mode.None})
#         UpdateUltChargeUI()
#         _Init(Player, MiniGameSuperHero)

#     _Init(Player : player, MiniGameSuperHero : minigame75_SuperHeros):void=
#         block:

#     _AbilityActivate(Player : player, MiniGameSuperHero : minigame75_SuperHeros)<suspends>:void=
#         block:

#     ClearUltCharge():void=
#         set UltCharge = 0
#         UpdateUltChargeUI()
#         if(MiniGameSuperHero := bMiniGameSuperHero?, Player := bPlayer?):
#             #MiniGameSuperHero.UltimateEndAudio.Play(Player)

#     UpdateUltChargeUI():void=
#         if (UltCharge = 0, BackgroundImage := UltimateHUDBackgroundImage?, AbilityImage := UltimateAbilityImage?):
#             BackgroundImage.SetImage(UltimateHUDBackgroundTexture)
#             AbilityImage.SetImage(UltimateNoChargeTexture)
#         if (UltCharge = 1, BackgroundImage := UltimateHUDBackgroundImage?, AbilityImage := UltimateAbilityImage?):
#             BackgroundImage.SetImage(UltimateCharge1)
#             AbilityImage.SetImage(UltimateNoChargeTexture)
#         if (UltCharge = 2, BackgroundImage := UltimateHUDBackgroundImage?, AbilityImage := UltimateAbilityImage?):
#             BackgroundImage.SetImage(UltimateCharge2)
#             AbilityImage.SetImage(UltimateAbilityTexture)
#             #if(MiniGameSuperHero := bMiniGameSuperHero?, Player := bPlayer?):
#                 #MiniGameSuperHero.UltimateChargedAudio.Play(Player)

#     AddUltCharge():void=
#         if (UltUsed = false):
#             set UltCharge = Min(UltCharge + 1, 2) 
#             UpdateUltChargeUI()


#     MakeHeroUI():canvas =
#         var NewUI : canvas = canvas{}
#         set UltimateAbilityImage = option{texture_block{DefaultImage := UltimateAbilityTexture, DefaultDesiredSize := vector2{X:= 256.0, Y := 256.0}}}
#         set UltimateHUDBackgroundImage = option{texture_block{DefaultImage := UltimateHUDBackgroundTexture, DefaultDesiredSize := vector2{X:= 256.0, Y := 256.0}}}
#         if (BackgroundImage := UltimateHUDBackgroundImage?, AbilityImage := UltimateAbilityImage?):
#             set NewUI = canvas:
#                 Slots := array:
#                     canvas_slot:
#                         Anchors := anchors{Minimum := vector2{X:= 0.5, Y := 1.0}, Maximum := vector2{X:= 0.5, Y:= 1.0}}
#                         Offsets := margin{Left := 0.0, Top := 0.0, Right := 256.0, Bottom := 256.0}
#                         Alignment := vector2{X:= 0.5, Y := 1.0}
#                         SizeToContent := true
#                         ZOrder := 0
#                         Widget := BackgroundImage
#                     canvas_slot:
#                         Anchors := anchors{Minimum := vector2{X:= 0.5, Y := 1.0}, Maximum := vector2{X:= 0.5, Y:= 1.0}}
#                         Offsets := margin{Left := 0.0, Top := -25.0, Right := 256.0, Bottom := 256.0}
#                         Alignment := vector2{X:= 0.5, Y := 1.0}
#                         SizeToContent := true
#                         ZOrder := 1
#                         Widget := AbilityImage
#         return NewUI

#     EndGame():void=
#         set GameEnd = true
#         if(ThisPlayer := bPlayer?, PlayerUI := GetPlayerUI[ThisPlayer]):
#             PlayerUI.RemoveWidget(Canva)

#     HandleAimIn(Agent : agent):void=
#         if (MiniGameSuperHero := bMiniGameSuperHero?):
#             MiniGameSuperHero.Controls.AddTo(Agent)

#     HandleAimInReleased(TupleReleased : tuple(agent, float)):void=
#         if (MiniGameSuperHero := bMiniGameSuperHero?):
#             MiniGameSuperHero.Controls.RemoveFrom(TupleReleased(0))