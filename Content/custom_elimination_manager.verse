
using { /Fortnite.com/Devices }
using { /Fortnite.com/Characters }
using { /Fortnite.com/Game }
using { /Verse.org/Simulation }
using { /UnrealEngine.com/Temporary/Diagnostics }

custom_elimination_manager := class(creative_device):

    OnBegin<override>()<suspends>:void=
        GetPlayspace().PlayerAddedEvent().SubscribeSuspends(OnPlayerJoined)
        GetPlayspace().PlayerRemovedEvent().Subscribe(OnPlayerLeft)

        for(Player : GetPlayspace().GetPlayers()):
            spawn{OnPlayerJoined(Player)}

    # NOTE: don't add anything else in here atm, it won't work
    var PlayerSubscribedEvents : [player][]cancelable = map{}

    OnPlayerLeft(Player: player):void=
        for(PlayerSubscribedEvent : PlayerSubscribedEvents[Player]):
            PlayerSubscribedEvent.Cancel()

        set PlayerSubscribedEvents = PlayerSubscribedEvents.WithRemovedKey(Player)

    OnPlayerJoined(Agent: agent)<suspends>:void=
        if(Player := player[Agent]):

            CharacterMaybe := AsyncGetFortCharacter(Player, 90.0)
            if(Character := CharacterMaybe?):
                if(not PlayerSubscribedEvents[Player]):
                    option{set PlayerSubscribedEvents[Player] = array{}}

                    EliminatedEvent := Character.EliminatedEvent()
                    option{set PlayerSubscribedEvents[Player] += array{EliminatedEvent.Subscribe(OnCharacterEliminated)}}

    # Custom elimination handler
    OnCharacterEliminated(Result: elimination_result):void=
        # Print("ELIMINATED")
        if(KilledAgent := Result.EliminatedCharacter.GetAgent[]):
            var KillerAgent : ?agent = false
            # Print("KILLED AGENT")
            # Check that the killer is not the same as the killed agent
            if(Killer := Result.EliminatingCharacter?.GetAgent[], KilledAgent <> Killer):
                Print("FOUND KILLER AGENT")
                set KillerAgent = option{Killer}

            EliminationResult := custom_elimination_result:
                Target := KilledAgent
                Source := KillerAgent

            OnEliminated.Subscribable.Signal(EliminationResult)
            if(EliminationResult.Source?):
                OnElimination.Subscribable.Signal(EliminationResult)

    OnEliminated : custom_subscribable_eliminated_player = custom_subscribable_eliminated_player{}
    OnElimination : custom_subscribable_eliminated_player = custom_subscribable_eliminated_player{}