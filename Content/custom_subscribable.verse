using { /Verse.org/Concurrency }
using { /Verse.org/Simulation }
using { /Fortnite.com/Game }

custom_cancelable<public> := class:
    var Canceled : logic = false

    CancelEvent : event() = event(){}

    Cancel():void=
        set Canceled = true
        CancelEvent.Signal()

# NOTE: Doesn't implement 'subscribable' interface because I didn't find a way to convert the Subscribe() method to <transacts>
custom_subscribable<public>(t:type)<computes> := class<public>(awaitable(t)):

    InnerEvent : event(t) = event(t){}
        
    Subscribe(Callback:type {__(:t):void}):custom_cancelable=
        Cancelable := custom_cancelable{}
        spawn{_WaitForEvent(Callback, Cancelable)}

        Cancelable

    CancelAllEvent : event() = event(){}

    CancelAll():void=
        CancelAllEvent.Signal()

    _WaitForEvent(Callback:type {__(:t):void}, Cancelable: custom_cancelable)<suspends>:void=
        race:
            # Wait for payload to be sent
            loop:
                Payload := InnerEvent.Await()

                # Just in case
                if(not Cancelable.Canceled?):
                    Callback(Payload)
                else:
                    break # Just in case

            # Cancelable got canceled
            Cancelable.CancelEvent.Await()

            # Subscribable got canceled
            CancelAllEvent.Await()

    
    Signal<public>(Payload:t):void =
        InnerEvent.Signal(Payload)

    Await<override>()<suspends>:t=
        InnerEvent.Await()

# Fix for "Can’t access a function from a preceding type. (3502)" when referencing parametric class outside of own file
# see https://forums.unrealengine.com/t/parametric-classes-not-usable-outside-of-own-file/1285253
custom_subscribable_player<public> := class:
    Subscribable<public> : custom_subscribable(player) = custom_subscribable(player){}

custom_subscribable_player_int<public> := class:
    Subscribable<public> : custom_subscribable(tuple(player, int)) = custom_subscribable(tuple(player, int)){}

custom_damage_result<public> := struct:
    Target<public> : player
    Source<public> : ?agent
    Amount<public> : float
    Causer<public> : ?game_action_causer
    
custom_subscribable_damaged_player<public> := class:
    Subscribable<public> : custom_subscribable(custom_damage_result) = custom_subscribable(custom_damage_result){}

custom_elimination_result<public> := struct:
    Target<public> : agent # so we can test with wildlife
    Source<public> : ?agent

custom_subscribable_eliminated_player<public> := class:
    Subscribable<public> : custom_subscribable(custom_elimination_result) = custom_subscribable(custom_elimination_result){}
    
custom_subscribable_attachable_item := class:
    Subscribable<public> : custom_subscribable(tuple(attachable_item, positional)) = custom_subscribable(tuple(attachable_item, positional)){}
    
custom_subscribable_empty<public> := class:
    Subscribable<public> : custom_subscribable(tuple()) = custom_subscribable(tuple()){}