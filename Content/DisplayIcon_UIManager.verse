
using { /Fortnite.com/Devices }
using { /Verse.org/Simulation }
using { /UnrealEngine.com/Temporary/Diagnostics }
using { /UnrealEngine.com/Temporary/UI }
using { /Fortnite.com/UI }
using { /UnrealEngine.com/Temporary/SpatialMath}
using { /Verse.org/Colors }
using { /Verse.org/Assets }

CustomIconHUD := class<unique>():
    var CurrentIcon : canvas = canvas{}

    # Create a minigame icon and set CurrentIcon
    CreateMinigameIcon(MinigameIcon : texture, Scale : float) : canvas =
        IconTextureBlock : texture_block = texture_block{DefaultImage := MinigameIcon, DefaultDesiredSize := vector2{X:=Scale,Y:=Scale}}

        IconOverlay := overlay:
            Slots := array:
                overlay_slot:
                    HorizontalAlignment := horizontal_alignment.Fill
                    VerticalAlignment := vertical_alignment.Fill
                    Widget := IconTextureBlock

        Canvas : canvas = canvas:
            Slots := array:
                canvas_slot:
                    Anchors := anchors{Minimum := vector2{X := 0.025, Y := 0.6}, Maximum := vector2{X := 0.225, Y := 0.95}}
                    Alignment := vector2{X := 0.5, Y := 0.5}
                    SizeToContent := true
                    Widget := IconOverlay

        set CurrentIcon = Canvas
        return Canvas

    # Removes current minigame icon from hud
    RemoveCurrentIcon(Player : player):void=
        if(PlayerUI := GetPlayerUI[Player]):
            PlayerUI.RemoveWidget(CurrentIcon)

# A Verse-authored creative device that can be placed in a level
Display_Icon_Manager := class(crashable_device):
    #Image in the bottom left with the minigame icon + description
    var CustomHUDMap : [player]CustomIconHUD = map{}

    # Call this to create CustomHUD
    MakeCustomHUD(Player : player, MinigameIcon : texture, Scale : float):void=
        Print("Make Custom HUD")
        if(set CustomHUDMap[Player] = CustomIconHUD{}, CurrentHUD := CustomHUDMap[Player]):
            IconCanvas := CurrentHUD.CreateMinigameIcon(MinigameIcon, Scale)
            if(PlayerUI := GetPlayerUI[Player]):
                PlayerUI.AddWidget(IconCanvas)

    #If players leave the game and we don't clear the map we'll have a null reference
    ClearHUDMaps():void=
        set CustomHUDMap = map{}
