
using { /Fortnite.com/Devices }
using { /Fortnite.com/AI}
using { /Fortnite.com/Teams}
using { /Fortnite.com/UI }
using { /Verse.org/Simulation }
using { /Verse.org/Random }
using { /Verse.org/Colors }
using { /Verse.org/Assets }
using { /Fortnite.com/Characters }
using { /Fortnite.com/FortPlayerUtilities }
using { /Fortnite.com/Game }
using { /Fortnite.com/Teams }
using { /Fortnite.com/Playspaces }
using { /Verse.org/Simulation/Tags }
using { /UnrealEngine.com/Temporary/Diagnostics }
using { /UnrealEngine.com/Temporary/SpatialMath }
using { /UnrealEngine.com/Temporary/UI }

spawner := class(tag) {}
mutatorSpawnsBox1 := class(tag) {}
barrierSpawnsBox1 := class(tag) {}
camera_device := class(tag) {}
controller_third_person_device := class(tag) {}

#TEAM MINIGAME CODE - ADDITIONAL PARAMETER FOR SWITCHES
Handler1 := class() {
    Device : game_manager   <#YOU MUST CHANGE "main" TO YOUR VERSE DEVICE NAME#>
    Parameter1 : int  <#Change "int" to your parameter type, ex: float, string#>

    HandlerFunction(Agent : agent) : void = {
        Device.SetPlayerLoadout(Agent, Parameter1)
    }
}

(Tracker: tracker_device).Add(Player: player, Amount: int):void=
    OldValue := Tracker.GetValue(Player)
    if(OldValue >= 0):
        Tracker.SetValue(Player, Min(OldValue + Amount, 9999))
        Tracker.UpdatedEvent().Subscribable.Signal(Player)

indexed_team_settings := class<concrete>:
    @editable
    TeamSettings : team_settings_and_inventory_device = team_settings_and_inventory_device{}

    @editable
    TeamIndex : type{Value:int where Value > 1, Value <= 100} = 2 # Don't allow team 0 since we use it to swap loosers

fort_character_vitals := class<concrete>:
    @editable
    MaxHealth : float = 100.0

    @editable
    Health : float = 100.0

    @editable
    MaxShield : float = 100.0

    @editable
    Shield : float = 100.0

    Set(Character: fort_character):void=
        Character.SetMaxHealth(MaxHealth)
        Character.SetMaxShield(MaxShield)
        Character.SetHealth(Health)
        Character.SetShield(Shield)

player_win_sequence_data := class<concrete>:
    @editable
    Sequence : cinematic_sequence_device = cinematic_sequence_device{}

    @editable
    PlayerReference : player_reference_device = player_reference_device{}

podium_sequence_data := class<concrete>:
    @editable
    Sequence : cinematic_sequence_device = cinematic_sequence_device{}

    # Player reference count can easily be changed in the editor
    @editable
    PlayerReferences : []player_reference_device = array:
        player_reference_device{}
        player_reference_device{}
        player_reference_device{}

selectable_loadout_class := class<concrete>:
    @editable DefaultLoadout : item_granter_device = item_granter_device{}
    @editable DefaultLoadoutSecondary : item_granter_device = item_granter_device{}
    @editable SelectableLoadout2Primary : item_granter_device = item_granter_device{}    
    @editable SelectableLoadout2Secondary : item_granter_device = item_granter_device{}      
    @editable SelectableLoadout3Primary : item_granter_device = item_granter_device{} 
    @editable SelectableLoadout3Secondary : item_granter_device = item_granter_device{} 

game_manager := class(crashable_device):
    # Item Shop
    @editable EnableItemShop : logic = true

    @editable SelectableLoadouts : selectable_loadout_class = selectable_loadout_class{}
    @editable MaxMatsItemGranter : item_granter_device = item_granter_device{}
    @editable BoxMatsItemGranter : item_granter_device = item_granter_device{}
    @editable DefaultHealsItemGranter : item_granter_device = item_granter_device{} 

    # Lobby
    @editable BillboardName1 : billboard_device = billboard_device{}
    @editable BillboardWins1 : billboard_device = billboard_device{}
    @editable BillboardElims1 : billboard_device = billboard_device{}
    @editable BillboardName2 : billboard_device = billboard_device{}
    @editable BillboardWins2 : billboard_device = billboard_device{}
    @editable BillboardElims2 : billboard_device = billboard_device{}
    @editable XpLobbyButton : button_device = button_device{}
    @editable XpLobbyAccolade : accolades_device = accolades_device{}

    @editable MinigameCollectibleCoin : collectible_object_device = collectible_object_device{}
    @editable XpWinAccolade : accolades_device = accolades_device{}
    @editable EliminationAccolade : accolades_device = accolades_device{}
    @editable UiMessage : hud_message_device = hud_message_device{}
    @editable UiMessageIndividual : hud_message_device = hud_message_device{}
    # @editable RoundTimer: timer_device = timer_device{}
    # @editable EndGameDevice: end_game_device = end_game_device{}
    @editable WinnersTeamData : indexed_team_settings = indexed_team_settings{TeamIndex := 100}
    @editable WaveRespawnTeamData : indexed_team_settings = indexed_team_settings{TeamIndex := 99}
    @editable ScoreManager: score_manager_device = score_manager_device{}
    @editable ClearLayer1Trigger : trigger_device = trigger_device{}
    @editable ClearLayer2Trigger : trigger_device = trigger_device{}
    @editable Box1Barriers : []barrier_device = array{}
    @editable NoShootingMutatorZone : mutator_zone_device = mutator_zone_device{}
    @editable NoBuildingMutatorZone : mutator_zone_device = mutator_zone_device{}
    @editable NoItemMutatorZone : mutator_zone_device = mutator_zone_device{}
    @editable NoEmotingMutatorZone : mutator_zone_device = mutator_zone_device{}
    @editable HalfGravityMutatorZone : mutator_zone_device = mutator_zone_device{}
    @editable SafeTimeTimerDevice : timer_device = timer_device{}
    @editable GenericMinigameEndTimer : timer_device = timer_device{}
    @editable GenericMinigameItemTimer : timer_device = timer_device{}
    @editable Teleporter : teleporter_device = teleporter_device{}
    @editable EliminationManager : custom_elimination_manager = custom_elimination_manager{}
    @editable GoldTracker : tracker_device = tracker_device{}
    @editable EliminationTracker : tracker_device = tracker_device{}
    @editable MinigameWinTracker : tracker_device = tracker_device{}
    @editable HeadshotTracker : tracker_device = tracker_device{}
    # @editable PlayerHUDPool : hud_message_pool = hud_message_pool{}
    @editable PlayerHUDDevice : hud_message_device = hud_message_device{}
    @editable BottomHUDDevice : hud_message_device = hud_message_device{}
    @editable LobbyMutator : mutator_zone_device = mutator_zone_device{}
    @editable EliminationDamageVolume : damage_volume_device = damage_volume_device{} 
    # @editable Billboard : billboard_device = billboard_device{}
    @editable CustomRespawnAudioMixer : audio_mixer_device = audio_mixer_device{}
    @editable DefaultVitals : fort_character_vitals = fort_character_vitals{MaxHealth := 100.0, MaxShield := 100.0, Health := 100.0, Shield := 100.0}
    @editable LobbySpectateBoxParentProp : creative_prop = creative_prop{}
    @editable LobbySpectateVolume : volume_device = volume_device{}
    @editable LobbySafeStorm : basic_storm_controller_device = basic_storm_controller_device{}

    @editable LoseMinigameHUDMessage : hud_message_device = hud_message_device{}
    @editable WinMinigameHUDMessage : hud_message_device = hud_message_device{}
    @editable RoundInProgressHUDMessage : hud_message_device = hud_message_device{}
    @editable MinigameEndHUDDuration : float = 3.0

    @editable DefaultPlayerWinSequenceData : player_win_sequence_data = player_win_sequence_data{}
    @editable PodiumSequenceData : podium_sequence_data = podium_sequence_data{}
    @editable HUDControllerNoHUD : hud_controller_device = hud_controller_device{}
    @editable BlackScreenHUDMessage : hud_message_device = hud_message_device{}
    @editable HideBlackScreenTrigger : trigger_device = trigger_device{}

    @editable ResetClassSelector : class_and_team_selector_device = class_and_team_selector_device{}
    @editable ResetItemRemover : item_remover_device = item_remover_device{}
    @editable MaximumMinigameDuration : float = 210.0

    @editable RedTeamHUD : hud_message_device = hud_message_device{}
    @editable BlueTeamHUD : hud_message_device = hud_message_device{}
    @editable GreenTeamHUD : hud_message_device = hud_message_device{}
    @editable YellowTeamHUD : hud_message_device = hud_message_device{}

    DeviceName<override>:string="Game Manager"

    @editable
    GoldPerWin : int = 50

    @editable
    GoldPerKill : int = 5

    @editable
    GoldOnMinigameStart : int = 1

    @editable
    GoldPerHeadshot : int = 5

    @editable
    VotingSystemMaybe : ?miniagme_voting_system = false

    @editable
    MaximumRounds : int = 5
    var CurrentRoundIndex : int = 1

    @editable
    ScrapBook : V_ScrapBook = V_ScrapBook{}

    # Force those minigames into the voting UI
    @editable
    DebugMinigames : []minigame_device = array{}

    @editable
    DebugAllMinigamesTwice : logic = false

    @editable
    Minigames : []minigame_device = array{}

    var RunningMinigame : ?minigame_device = false
    var EndGameDisabled : logic = false
    var PlayerRespawnedEvents : [player]event() = map{}
    var ShownSpectatorUIs : [player]logic = map{}
    var PlayerWins : [player]int = map{}
    var PlayerKills : [player]int = map{}
    var AllowSpectatingUI : [player]logic = map{}

    var DefaultSpectateBoxTransform : ?transform = false
    var DefaultLobbyFFATransform : ?transform = false

    @editable
    KeyPadButton : button_device = button_device{}

    @editable
    SubscribeUITrigger : trigger_device = trigger_device{}

    @editable
    ItemShop : V_ItemShop = V_ItemShop{}

    #Team Minigame New Manager - Seperate team manager based on crashable device
    @editable
    TeamSelectionManager : teamselection_manager = teamselection_manager{}

    @editable
    TeleportForTeamSelection : teleporter_device = teleporter_device{}

    @editable
    RespawnHUDMessage : hud_message_device = hud_message_device{}

    @editable
    PlayerMarkerPowerUp1 : visual_effect_powerup_device = visual_effect_powerup_device{}
    @editable
    PlayerMarkerPowerUp2 : visual_effect_powerup_device = visual_effect_powerup_device{}
    @editable
    PlayerMarkerPowerUp3 : visual_effect_powerup_device = visual_effect_powerup_device{}
    @editable
    PlayerMarkerPowerUp4 : visual_effect_powerup_device = visual_effect_powerup_device{}

    @editable
    PlayerLoadout1Switch : switch_device = switch_device{}
    @editable
    PlayerLoadout2Switch : switch_device = switch_device{}
    @editable
    PlayerLoadout3Switch : switch_device = switch_device{}
    @editable
    PlayerLoadout4Switch : switch_device = switch_device{}
    @editable
    PlayerLoadout5Switch : switch_device = switch_device{}
    @editable
    PlayerLoadout6Switch : switch_device = switch_device{}

    @editable
    RespawnEnabledMessage : hud_message_device = hud_message_device{}
    #[Patch 1.2] Added XP per second in lobby
    @editable
    LobbyXPPerSecondAccolade : accolades_device = accolades_device{}

    var DisplayIconManager : Display_Icon_Manager = Display_Icon_Manager{}

    var PlayerDefaultWeaponLoadoutMap : [player]?int = map{}
    var PlayerDefaultWeaponLoadoutSecondaryMap : [player]?int = map{}

    #REPLACES FindCreativeObjectsWithTag Function
    FindObjectsWithTag<public>(Tag:tag)<transacts>:[]creative_object_interface=
        TaggedObjects:=FindCreativeObjectsWithTag(Tag)
        var Objects:[]creative_object_interface=array{}
        for (Obj:TaggedObjects, TaggedObject:=creative_device_base[Obj]):
            set Objects += array{TaggedObject}
        return Objects

    # Runs when the device is started in a running game
    OnCrashableDeviceBegin<override>()<suspends>:void=
        #TEAM MINIGAME CODE [HANDLES LOADOUTS WHEN TURNED ON]
        PlayerLoadout1Switch.TurnedOnEvent.Subscribe(Handler1{Device:=Self, Parameter1:=1}.HandlerFunction)
        PlayerLoadout2Switch.TurnedOnEvent.Subscribe(Handler1{Device:=Self, Parameter1:=2}.HandlerFunction)
        PlayerLoadout3Switch.TurnedOnEvent.Subscribe(Handler1{Device:=Self, Parameter1:=3}.HandlerFunction)
        PlayerLoadout4Switch.TurnedOnEvent.Subscribe(Handler1{Device:=Self, Parameter1:=4}.HandlerFunction)
        PlayerLoadout5Switch.TurnedOnEvent.Subscribe(Handler1{Device:=Self, Parameter1:=5}.HandlerFunction)
        PlayerLoadout6Switch.TurnedOnEvent.Subscribe(Handler1{Device:=Self, Parameter1:=6}.HandlerFunction)

        for(Minigame : Minigames + DebugMinigames):
            set Minigame.Manager = Self

        GetUIFaderInstance().Init(Self)

        if(LobbySpectateBoxParentProp.IsValid[]):
            set DefaultSpectateBoxTransform = option{LobbySpectateBoxParentProp.GetTransform()}

        if(DebugMinigames.Length > 0):
            Print("DEBUG MODE ENABLED", ?Color := NamedColors.Orange, ?Duration := 40.0)
            # Disable end game condition if we're debugging alone
            if(GetPlayspace().GetPlayers().Length <= 1):
                set EndGameDisabled = true
                # # TODO: remove
                # if(Player := GetPlayspace().GetPlayers()[0]):
                #     GoldTracker.Add(Player, 500)

        EliminationManager.OnElimination.Subscribable.Subscribe(OnElimination)
        EliminationManager.OnEliminated.Subscribable.Subscribe(OnEliminated)
        HeadshotTracker.CompleteEvent.Subscribe(OnHeadshot)
        GoldTracker.UpdatedEvent().Subscribable.Subscribe(RefreshPlayerHUD)
        #LobbySpectateVolume.AgentEntersEvent.Subscribe(OnPlayerEnterSpectateBox)
        XpLobbyButton.InteractedWithEvent.Subscribe(OnButtonLobbyTrigger)
        #Fix Bug where people can pickaxe each other in the pre game lobby
        LobbyMutator.AgentEntersEvent.Subscribe(OnPlayerEnterLobbyZone)
        LobbyMutator.AgentExitsEvent.Subscribe(OnPlayerLeavesLobbyZone)

        #TODO:Delete Later
        MinigameCollectibleCoin.CollectedEvent.Subscribe(OnMinigameCollectibleCoinCollected)
        UpdateScoreboard()

        # Init HUD and remove invis glitch
        GetPlayspace().PlayerAddedEvent().SubscribeSuspends(OnPlayerJoined)
        # GetPlayspace().PlayerRemovedEvent().SubscribeSuspends(OnPlayerLeft)
        for(Player : GetPlayspace().GetPlayers()):
            spawn{OnPlayerJoined(Player)}
            CheckWeaponLoadouts(Player)

        spawn{RunCustomRoundsLoop()}
        #[Patch 1.2] Spawns the function for XP while in lobby every second
        spawn{XPWhileInLobby()}
        branch:
            Sleep(0.0) # Just in case the branch crashes if there's no minigame
            WeightSortedMinigames := Minigames.QuickSort(SortMinigameWeight)
            for(Minigame : WeightSortedMinigames):
                Print("Weight[{Minigame.Weight}] {Minigame}")
                Sleep(0.0)

        # Remove lobby player items that are granted during minigames
        loop:
            Sleep(0.0)
            for(Player : GetPlayspace().GetPlayers(), LobbyMutator.IsInVolume[Player]):
                ResetItemRemover.Remove(Player)

    CheckWeaponLoadouts(InPlayer : player):void=
        #TEAM MINIGAME CODE - JOIN IN PROGRESS SWITCH CHECK THEN ASSIGN CORRECT LOADOUT
        if(PlayerLoadout1Switch.GetCurrentState[InPlayer]):
            PlayerLoadout1Switch.StateLoadEvent.Subscribe(Handler1{Device:=Self, Parameter1:=1}.HandlerFunction)
        if(PlayerLoadout2Switch.GetCurrentState[InPlayer]):
            PlayerLoadout2Switch.StateLoadEvent.Subscribe(Handler1{Device:=Self, Parameter1:=2}.HandlerFunction)
        if(PlayerLoadout3Switch.GetCurrentState[InPlayer]):
            PlayerLoadout3Switch.StateLoadEvent.Subscribe(Handler1{Device:=Self, Parameter1:=3}.HandlerFunction)

        if(PlayerLoadout4Switch.GetCurrentState[InPlayer]):
            PlayerLoadout4Switch.StateLoadEvent.Subscribe(Handler1{Device:=Self, Parameter1:=4}.HandlerFunction)
        if(PlayerLoadout5Switch.GetCurrentState[InPlayer]):
            PlayerLoadout5Switch.StateLoadEvent.Subscribe(Handler1{Device:=Self, Parameter1:=5}.HandlerFunction)
        if(PlayerLoadout6Switch.GetCurrentState[InPlayer]):
            PlayerLoadout6Switch.StateLoadEvent.Subscribe(Handler1{Device:=Self, Parameter1:=26}.HandlerFunction)

    SortMinigameWeight(Minigame1: minigame_device, Minigame2: minigame_device)<transacts>: int =
        if (Minigame1.Weight < Minigame2.Weight):
            1
        else if (Minigame1.Weight > Minigame2.Weight):
            -1
        else:
            0

    #################################################
    ################## GAME FLOW ####################
    #################################################

    RoundEndedEvent : event([]player) = event([]player){}
    RoundStartedEvent<public>: event() = event(){}

    RunCustomRoundsLoop()<suspends>:void=
        
        # DEBUG
        var DebugTwiceMinigames : []?minigame_device = array{}
        if(DebugAllMinigamesTwice? and DebugMinigames.Length > 0):
            for(Minigame : DebugMinigames):
                set DebugTwiceMinigames += array{option{Minigame}, option{Minigame}}

        if(VotingSystem := VotingSystemMaybe?):
            loop:
                # Wait for enough players to join the game
                loop:
                    AlivePlayers := for(Player : GetPlayspace().GetPlayers(), Player.GetFortCharacter[].IsActive[]) { Player }
                    
                    # Allow debugging alone
                    if(DebugMinigames.Length > 0, AlivePlayers.Length = 1): #TODO: Fix back to 1
                        break

                    # Normal condition
                    if(AlivePlayers.Length > 1): #TODO: Fix back to 1   
                        break

                    Print("Waiting for 2nd player to join the game")
                    Sleep(1.0)

                # Run the voting system, should call StartMinigame() at some point
                MinigameMaybe := DebugTwiceMinigames[0] or VotingSystem.DecideNextMinigame()

                if(DebugTwiceMinigames.Length > 0):
                    Sleep(5.0) # Would spawn in the air otherwise

                if(Minigame := MinigameMaybe?):

                    # Fix for backwards tac sprint staying after starting a new native round
                    #if(CurrentRoundIndex = 1):
                    #    ResetControls.AddToAll()
                    #    Sleep(0.0)
                    #    ResetControls.Disable()
                    #    Sleep(0.0)

                    #TEAM MINIGAME CODE
                    TeleportForTeamSelection.Disable()

                    Sleep(1.0)
                    # Start minigame
                    spawn{StartMinigame(Minigame)}

                    # Spectating UI can only be shown once per minigame (should not show when taking the spectator box)
                    for(Player : GetPlayspace().GetPlayers()):
                        option{set AllowSpectatingUI[Player] = true}

                    #spawn{ShowSpectateBox()}

                    # Wait for minigame to end
                    RoundWinners := RoundEndedEvent.Await()

                    for(RoundWinner : RoundWinners):
                        OldWinCount := (PlayerWins[RoundWinner] or 0)
                        option{set PlayerWins[RoundWinner] = OldWinCount + 1}

                    # Don't show spectate UI for JIP players
                    set RunningMinigame = false
                    #HideSpectateBox()
                    #Hide spectate box (since round is ending)

                    # Don't do anything if the default fortnite end game callout is about to show (because everyone died)
                    # (fix for HUD elements not being removed if shown at the exact time we end a native round)
                    if(
                        AlivePlayers := for(Player : GetPlayspace().GetPlayers(), Player.GetFortCharacter[].IsActive[]) { Player },
                        AlivePlayers.Length = 0
                    ):
                        break

                    # Show end of minigame HUD messages
                    for(Player : GetPlayspace().GetPlayers()):
                        HUDMessage := RoundWinners.Find[Player] and WinMinigameHUDMessage or LoseMinigameHUDMessage
                        HUDMessage.Show(Player, BuildMessage(""), ?DisplayTime := MinigameEndHUDDuration)
                    Sleep(MinigameEndHUDDuration)
                    Print("B1")
                    # Manage set of minigames ending
                    set CurrentRoundIndex += 1
                    if(DebugTwiceMinigames.Length = 0, CurrentRoundIndex > MaximumRounds):
                        OnGameEnded()
                        break

                    NoItemMutatorZone.Enable()

                    Sleep(1.0)
                    Print("B2")

                    # Respawn people
                    PlayersToRespawn := GetPlayspace().GetPlayers()
                    RespawnEverybody()
                    
                    # Reset minigame after everybody has been teleported to lobby
                    # NOTE: Reset before so we don't take damage from the storm upon being teleported to lobby
                    spawn{Minigame.Reset()}
                    
                    # Fix for backwards tac sprint not going off
                    #Sleep(0.0)
                    #Sleep(0.0)
                    #ResetControls.Enable()
                    Print("B3")
                    race:
                        block:
                            PlayerRespawnedEvent.Await() # Wait for 1 player to be respawned
                            Sleep(0.5) # Everyone should be respawned by now
                        Sleep(8.0)
                    Print("B4")
                    HUDControllerNoHUD.Disable()
                    NoItemMutatorZone.Disable()
                    CustomRespawnAudioMixer.DeactivateMix()
                    UpdateScoreboard()

                    # Fix for backwards tac sprint
                    #Sleep(0.5)
                    #ResetControls.AddToAll()
                    #Sleep(0.0)
                    #ResetControls.Disable()

                    # Fix for black screen staying because player didn't respawn for some reason
                    for(Player : GetPlayspace().GetPlayers()):
                        spawn{FadeOutScreen(Player)}
                        #TEAM MINIGAME CODE - ATTEMPTING TO PUT ALL PLAYERS IN THE SAME TEAM
                        TeamSelectionManager.NoTeamClassSelector.ChangeTeam(Player)
                    
                    #Wait a short period of time and then enable the team selectors [PATCH 1.1]
                    spawn{EnableTeamSelectorsAfterPlayersRespawn()}
                    #[Patch 1.2] Spawns the function for XP while in lobby every second
                    spawn{XPWhileInLobby()}
                    # Fix for players not respawned through the native Wave Respawn system, those players won't be spectatable though
                    spawn{RespawnGlitchedOutPlayers(PlayersToRespawn)}

                    # DEBUG
                    if(DebugTwiceMinigames.Length > 0, NewArray := DebugTwiceMinigames.RemoveElement[0]):
                        set DebugTwiceMinigames = NewArray
                else:
                    # No minigame found by voting system, restarting the game
                    Print("ERR: Couldn't find any minigame to start")
                    WinnersTeamData.TeamSettings.EndRound()
                    break
        else:
            Print("ERR: Couldn't find voting system, restarting the game")
            WinnersTeamData.TeamSettings.EndRound()

    #[PATCH 1.1]
    EnableTeamSelectorsAfterPlayersRespawn()<suspends>:void=
        Sleep(1.0)
        #TEAM MINIGAME CODE - ENABLE TEAM SELECTORS
        TeamSelectionManager.Team1ClassSelector.Enable()
        TeamSelectionManager.Team2ClassSelector.Enable()
        TeamSelectionManager.Team3ClassSelector.Enable()
        TeamSelectionManager.Team4ClassSelector.Enable()
        #TEAM MINIGAME CODE
        TeleportForTeamSelection.Enable()
    
    # Called after voting system ended
    StartMinigame(Minigame: minigame_device)<suspends>:void=
        set RunningMinigame = option{Minigame}

        # Grant gold at each minigame start
        for(Player : GetPlayspace().GetPlayers()):
            GoldTracker.Add(Player, GoldOnMinigameStart)

        # Keep it from showing still for a brieve moment after being teleported to a new minigame
        LobbySafeStorm.DestroyStorm()

        if(Minigame.StartingMessage.Length > 0):
            UiMessage.SetText(StringToMessage(Minigame.StartingMessage))
            UiMessage.Show()
            
        Minigame.SendVoteAnalytics()
        spawn{Minigame.Start()}
        spawn{HandleWinner()}
        #[Patch 1.2] Signal for when a round starts
        RoundStartedEvent.Signal()
        # Keep it from showing still for a brieve moment after being teleported to a new minigame
        Sleep(2.0)
        LobbySafeStorm.GenerateStorm()

    <#
    #TEAM MINIGAME SAFETY FEATURE LOBBY
    MutatorCheckForLeftOverPlayers():void=
        AllPlayersInLobby := GetPlayspace().GetPlayers()
        for(Player : AllPlayersInLobby):
            if(LobbyMutator.IsInVolume[Player]):
                if(Minigame := RunningMinigame?):
                    #Minigame.NewTeleportPlayers(Player,Minigame.CustomTeleportationMethod)
                    Minigame.NewTeleportPlayers(ParticipatingPlayers, ChosenArena?.ArenaDevice?.CustomTeleportationMethod or CustomTeleportationMethod)    
    #>
    HandleWinner()<suspends>: void=
        if(EndGameDisabled?):
            return
            
        Sleep(5.0) # So we don't end the game too soon
        loop:
            # Look for running minigame
            if(Minigame := RunningMinigame?):
                # Wait for winners to be defined
                var Winners : []player = array{}
                race:
                    # Custom winning event call
                    block:
                        set Winners = Minigame.EndedEvent.Await()
                        Print("Set all winners")
                    # Can either use the EndedEvent or the GetWinners[] methods to retrieve the winners
                    loop:
                        Sleep(0.2)
                        # Do we have any winner ?
                        if(_Winners := Minigame.GetWinners[]):
                            #Makes sure we don't get a win on an empty array element
                            if(_Winners.Length >= 1):
                                Print("We have a winner")
                                Minigame.EndedEvent.Signal(_Winners)
                                Sleep(Inf) # Fix for race self cancelling issues
                # Award each winner
                for(Winner : Winners):
                    Minigame.AwardWinner(Winner)

                # End the game
                Sleep(0.0)
                Minigame.OnWinnersFound(Winners)    #This occurs
                RoundEndedEvent.Signal(Winners)     #
                Minigame.OnMinigameEnded(Winners)
                break # Don't look for winners anymore
            Sleep(1.0)

    # Manually respawn glitched out players (stuck in spectating mode after wave respawn)
    RespawnGlitchedOutPlayers(Players: []player)<suspends>:void=
        Sleep(2.0)
        for(Player : Players, Character := Player.GetFortCharacter[], not Character.IsActive[]):
            FakeRespawn(Player)

    # Sort winners by wins/kills
    SortPlayerStats(Data1:tuple(player, int, int), Data2:tuple(player, int, int))<transacts>:int=
        Wins1 := Data1(1)
        Kills1 := Data1(2)

        Wins2 := Data2(1)
        Kills2 := Data2(2)

        if(Wins1 = Wins2 and Kills1 = Kills2):
            return 0
        else:
            if(Wins1 = Wins2):
                return (Kills1 < Kills2 and 1) or -1
            else:
                return (Wins1 < Wins2 and 1) or -1
        
    GetBestPlayers()<transacts>:[]player=
        var PlayerStats : []tuple(player, int, int) = array{}

        for(Player : GetPlayspace().GetPlayers()):
            Wins := PlayerWins[Player] or 0
            Kills := PlayerKills[Player] or 0
            set PlayerStats += array{(Player, Wins, Kills)}

        return for(Stats : PlayerStats.QuickSort(SortPlayerStats)) { Stats(0) }

    GetBestPlayersTotalLifetime()<transacts>:[]tuple(player, int, int)=
        var PlayerStats : []tuple(player, int, int) = array{}

        for(Player : GetPlayspace().GetPlayers()):
            Wins := MinigameWinTracker.GetValue(Player)
            Kills := EliminationTracker.GetValue(Player)
            #EliminationTracker
            set PlayerStats += array{(Player, Wins, Kills)}

        return PlayerStats.QuickSort(SortPlayerStats)

    #[Patch 1.2] Grant XP every second to all players in the lobby
    XPWhileInLobby()<suspends>:void=
        Print("XP Function Called")
        race:
            #Wait for the round to start
            RoundStartedEvent.Await()
            #End once game starts
            block:
                loop:
                    #Get all players and then get a single player. We can then use them to award XP. The device is set to reward all players not a triggering player.
                    #This saves us from looping through all players in the lobby
                    AllPlayers := GetPlayspace().GetPlayers()
                    if(Player := AllPlayers[0]):
                        LobbyXPPerSecondAccolade.Award(Player)
                        Print("Giving XP Per Second")
                    #Give XP
                    Sleep(1.0)

    OnGameEnded()<suspends>:void=
        # Get best players throughout the multiple rounds
        Winners := GetBestPlayers()

        # Fully hide the HUD of everybody
        HUDControllerNoHUD.Enable()

        # Play best player win sequence
        if(BestPlayer := Winners[0]):
            PlayerWinSequenceData := DefaultPlayerWinSequenceData
            PlayerWinSequenceData.PlayerReference.Register(BestPlayer)
            PlayerWinSequenceData.PlayerReference.Enable()
            PlayerWinSequenceData.Sequence.Play()
            race:
                Sleep(20.0) # Just in case
                PlayerWinSequenceData.Sequence.StoppedEvent.Await()
            PlayerWinSequenceData.PlayerReference.Clear()

        # Prepare podium player references
        for(Place -> PlayerReference : PodiumSequenceData.PlayerReferences, AssignedPlayer := Winners[Place]):
            PlayerReference.Enable()
            Print("Enabling PF at {PlayerReference.GetTransform().Translation}")

        # Fix for player references not working when players are dead
        for(Player : GetPlayspace().GetPlayers()):
            FakeRespawn(Player)
        Sleep(0.2)
            
        # Play podium sequence
        for(Place -> PlayerReference : PodiumSequenceData.PlayerReferences, AssignedPlayer := Winners[Place]):
            PlayerReference.Register(AssignedPlayer)
        PodiumSequenceData.Sequence.Play()
        race:
            Sleep(10.0) # Just in case
            PodiumSequenceData.Sequence.StoppedEvent.Await()

        # Switch all players to team 0 and winners to winning team
        TeamCollection := GetPlayspace().GetTeamCollection()
        if(TeamZero := TeamCollection.GetTeams()[0], WinningTeam := TeamCollection.GetTeams()[WinnersTeamData.TeamIndex-1]):
            for(Player : GetPlayspace().GetPlayers(), TeamCollection.AddToTeam[Player, TeamZero]) {}
            for(Winner : Winners, TeamCollection.AddToTeam[Winner, WinningTeam]) {}

        # End round for winning team
        WinnersTeamData.TeamSettings.EndRound()

    OnEliminated(Result : custom_elimination_result):void=
        Print("Elim event")
        if(Player := player[Result.Target]):
            if(CurrentMiniGame := RunningMinigame?, CurrentMiniGame.ShowSubscribeUI = true):
                ShowSubUI(Player)
            if(EnableItemShop = true):
                spawn{ShowSpectatingUI(Player)}
            Print("You Died")
            # AlivePlayers := for(_Player : GetPlayspace().GetPlayers(), _Player.GetFortCharacter[].IsActive[]) { _Player }
            # if(AlivePlayers.Length = 1, LastPlayerAlive := AlivePlayers[0], Character := LastPlayerAlive.GetFortCharacter[]):
            #     Character.SetVulnerability(false)
            #     Character.PutInStasis(stasis_args{AllowEmotes := false})
            #     Print("Success")

    OnElimination(Result : custom_elimination_result):void=
        if(Player := player[Result.Source?]):
            GoldTracker.Add(Player, GoldPerKill)
            #EliminationTracker.Add(Player,1)
            OldPlayerKills := PlayerKills[Player] or 0
            option{set PlayerKills[Player] = OldPlayerKills + 1}

            # Siphon
            if(Minigame := RunningMinigame?):
                if(SiphonData := Minigame.AdditionalSiphon?, Character := Player.GetFortCharacter[]):
                    Character.SetHealth(Character.GetHealth() + SiphonData.HealthRegen)
                    Character.SetShield(Character.GetShield() + SiphonData.ShieldRegen)

                for(SiphonGranter : Minigame.SiphonGranters):
                    SiphonGranter.Grant(array{Player})

    OnHeadshot(Agent: agent):void=
        if(Player := player[Agent]):
            GoldTracker.Add(Player, GoldPerHeadshot)
            Print("Completed Event")
            HeadshotTracker.Reset(Agent)
            Print("Resetting")
            HeadshotTracker.Assign(Agent)
            Print("Assigning")

    # Fix for players able to pickaxe each other if the current minigame allows shooting
    OnPlayerEnterLobbyZone(Agent: agent):void=
        if(Player := player[Agent], Character := Player.GetFortCharacter[]):
            Character.SetVulnerability(false)

    OnPlayerLeavesLobbyZone(Agent: agent):void=
        if(Player := player[Agent], Character := Player.GetFortCharacter[]):
            Character.SetVulnerability(true)

    #################################################
    ############### PLAYER INIT/HUD #################
    #################################################

    OnPlayerJoined(Player : player)<suspends>:void=
        #TEAM MINIGAME CODE - JOIN IN PROGRESS SWITCH CHECK THEN ASSIGN CORRECT LOADOUT
        if(PlayerLoadout1Switch.GetCurrentState[Player]):
            PlayerLoadout1Switch.StateLoadEvent.Subscribe(Handler1{Device:=Self, Parameter1:=1}.HandlerFunction)
        if(PlayerLoadout2Switch.GetCurrentState[Player]):
            PlayerLoadout2Switch.StateLoadEvent.Subscribe(Handler1{Device:=Self, Parameter1:=2}.HandlerFunction)
        if(PlayerLoadout3Switch.GetCurrentState[Player]):
            PlayerLoadout3Switch.StateLoadEvent.Subscribe(Handler1{Device:=Self, Parameter1:=3}.HandlerFunction)
        if(PlayerLoadout4Switch.GetCurrentState[Player]):
            PlayerLoadout4Switch.StateLoadEvent.Subscribe(Handler1{Device:=Self, Parameter1:=4}.HandlerFunction)
        if(PlayerLoadout5Switch.GetCurrentState[Player]):
            PlayerLoadout5Switch.StateLoadEvent.Subscribe(Handler1{Device:=Self, Parameter1:=5}.HandlerFunction)
        if(PlayerLoadout6Switch.GetCurrentState[Player]):
            PlayerLoadout6Switch.StateLoadEvent.Subscribe(Handler1{Device:=Self, Parameter1:=6}.HandlerFunction)


        sync:
            # Apply on spawn fixes
            OnPlayerSpawned(Player)

            # Wait for UI to be available so that hud message can show
            race:
                loop:
                    if(GetPlayerUI[Player]):
                        break
                    Sleep(0.2)
                Sleep(30.0) # Just in case

            # Show minigame hud for JIP players
            # if(Minigame := RunningMinigame?):
            #     Minigame.UiLogoHud.Show()

            RefreshPlayerHUD(Player)

    OnPlayerSpawned(Player: player)<suspends>:void=
        Print("Player has joined the world")
        CustomRespawnAudioMixer.Unregister(Player)
        spawn{FadeOutScreen(Player)}

        # Wait for character to be available
        CharacterMaybe := AsyncGetFortCharacter(Player, 60.0)
        if(Character := CharacterMaybe?):
            Print("Get Character Async")
            # Add to empty team
            SwitchToEmptyTeam(Player)

            # Show to help invisible Glitch
            Character.Show()

            # Just in case
            Character.ReleaseFromStasis()

            # Reset vitals (health/shield)
            DefaultVitals.Set(Character)

            # Other fixes
            ResetClassSelector.ChangeClass(Player)
            ResetItemRemover.Remove(Player)

            #Moved code for eliminate on join to here
            #Players should never be in lobby when a minigame is running, eliminating them here deals with join in progress + any other potential bugs
            if(RunningMinigame?):
                RoundInProgressHUDMessage.Show(Player, BuildMessage(""), ?DisplayTime := MinigameEndHUDDuration)
                Sleep(2.0)
                #Make sure the minigame is still running
                if(RunningMinigame?):
                    Character.SetVulnerability(true)
                    #Attempting to just eliminate player via damage volume instead of this damage method here
                    #if(Character.TeleportTo[EliminationDamageVolume.GetTransform().Translation + vector3{Z := 200.0}, rotation{}]):
                    Character.Damage(9999.0)
                    if(Character.IsDownButNotOut[]):
                        Print("Joining Player DBNO")
                        Sleep(0.0)
                        Character.Damage(9999.0)
                    Player.Respawn(EliminationDamageVolume.GetTransform().Translation + vector3{Z := 300.0}, rotation{})
                    Print("Mini game running, respawn player at the elimination damage volume")

    # var PlayerHUDMessage : [player]tuple(hud_message_device, int) = map{}
        
    RefreshPlayerHUD(Player: player):void=
        spawn{_RefreshPlayerHUD(Player)}

    # var IsRefreshingHUD : logic = false
        
    _RefreshPlayerHUD(Player: player)<suspends>:void=
        # # Queue system, fix for different HUDs getting the same values when updated at the same time
        # race:
        #     Sleep(5.0)
        #     loop:
        #         if(not IsRefreshingHUD?):
        #             break
        #         Sleep(0.0)
        # set IsRefreshingHUD = true

        # if(not PlayerHUDMessage[Player]):
        #     FetchResultMaybe := PlayerHUDPool.FetchAvailableItem()

        #     if(FetchResult := FetchResultMaybe?):
        #         option{set PlayerHUDMessage[Player] = FetchResult}

        # if(HUDMessage := PlayerHUDMessage[Player](0), true?):
        String := GoldTracker.GetValue(Player)
        # Index := if(M := PlayerHUDMessage[Player]) {M(1)} else {-99}
        # Print(BuildPlayerMessage("Gold(",Player,",{Index})={String}"))
        PlayerHUDDevice.Show(Player, BuildMessage("{String}"))

        # Sleep(1.0)
        # set IsRefreshingHUD = false

    # # Spawn Initiators
    # InitSpawners():void=
    #     Spawners := GetCreativeObjectsWithTag(spawner{})
    #     for (Obj: Spawners):
    #         if(Spawner := player_spawner_device[Obj]):
    #             Spawner.SpawnedEvent.Subscribe(OnPlayerSpawned)

    # OnPlayerLeft(Player : player)<suspends>:void=
    #     if(HUDMessageIndex := PlayerHUDMessage[Player](1), true?):
    #         PlayerHUDPool.ReleaseItem(HUDMessageIndex)

    #################################################
    ################## SPECTATING ###################
    #################################################

    # ShowSpectateBox()<suspends>:void=
    #     Sleep(1.0) # Wait for players to be teleported out of lobby
    #     Print("Show Spectating Box")
    #     if(HiddenSpectateBoxTransform := DefaultSpectateBoxTransform?):
    #         Transform := transform:
    #             Translation := vector3{X:= HiddenSpectateBoxTransform.Translation.X, Y:= HiddenSpectateBoxTransform.Translation.Y, Z:= 862.0}
    #             Rotation := HiddenSpectateBoxTransform.Rotation
    #             Scale := HiddenSpectateBoxTransform.Scale
    #         option{LobbySpectateBoxParentProp.TeleportTo[Transform]}

    # HideSpectateBox():void=
    #     Print("Hiding Specating Box")
    #     if(HiddenSpectateBoxTransform := DefaultSpectateBoxTransform?):
    #         option{LobbySpectateBoxParentProp.TeleportTo[HiddenSpectateBoxTransform]}

    # OnPlayerEnterSpectateBox(Agent: agent):void=
    #     if(Player := player[Agent], RunningMinigame?, Character := Player.GetFortCharacter[]):
    #         # If player is fake respawned, he won't be able to be spectatable, but it shouldn't matter here
    #         Character.SetVulnerability(true)
    #         Character.Damage(999.9)

    #         # Fix for JIP players being immune in public servers
    #         Player.Respawn(EliminationDamageVolume.GetTransform().Translation + vector3{Z := 200.0}, rotation{})

    ShowSubUI(Player : player):void=
        if(
            PlayerUI := GetPlayerUI[Player],
            not ShownSpectatorUIs[Player],
            RunningMinigame?,
            AllowSpectatingUI[Player]?
        ):
            option{set ShownSpectatorUIs[Player] = true}
            option{set AllowSpectatingUI[Player] = false}
            SubscribeUITrigger.Trigger(Player)
            option{set AllowSpectatingUI[Player] = true}
        set ShownSpectatorUIs = ShownSpectatorUIs.WithRemovedKey(Player)

    ShowSpectatingUI(Player: player)<suspends>:void=  
        if(
            PlayerUI := GetPlayerUI[Player],
            not ShownSpectatorUIs[Player],
            RunningMinigame?,
            AllowSpectatingUI[Player]?
        ):
            option{set ShownSpectatorUIs[Player] = true}
            option{set AllowSpectatingUI[Player] = false}
            
            UIData := MakeSpectatingUI()
            Canvas := UIData(0)
            LobbyButton := UIData(1)
            SpectateButton := UIData(2)

            Sleep(0.5) # Will get erased by the default spectator UI otherwise
            
            if(Minigame := RunningMinigame?):
                #Create Item shop
                if(EnableItemShop = true):
                    ItemShopUI := ItemShop.CreatePlayerUI(Player, "SpectateMenu")
                    ItemShop.RenderUIOnPlayerScreen(Player, ItemShopUI)
                    #PlayerUI.AddWidget(ItemShopUI.Canvas, player_ui_slot{InputMode := ui_input_mode.All})
                else:
                    PlayerUI.AddWidget(Canvas, player_ui_slot{InputMode := ui_input_mode.All})
                Print("Running game")
                var ShouldRespawn : logic = false
                race:
                    SpectateButton.OnClick().Await()
                    Player.RespawnedEvent().Await()
                    Minigame.EndedEvent.Await()

                    block:
                        LobbyButton.OnClick().Await()
                        if(RunningMinigame?):
                            set ShouldRespawn = true
                        else:
                            Print("Cannot go to lobby when minigame is ending")
                            Sleep(Inf)
                    Sleep(240.0)

                Canvas.SetVisibility(widget_visibility.Hidden)
                if(EnableItemShop = true):
                    ItemShop.RemovePlayerUI(Player)
                else:
                    PlayerUI.RemoveWidget(Canvas)
                option{set AllowSpectatingUI[Player] = true}
                Sleep(0.5)
                # Fix for default spectating HUD staying after respawning (respawn after removing HUD instead)
                if(ShouldRespawn?):
                    FakeRespawn(Player)

            set ShownSpectatorUIs = ShownSpectatorUIs.WithRemovedKey(Player)

    MakeSpectatingUI():tuple(widget, text_button_base, text_button_base)=
        SpectateButton := button_loud:
            DefaultText := BuildMessage("SPECTATE")

        LobbyButton := button_loud:
            DefaultText := BuildMessage("GO TO ARENA")

        Canvas := canvas:
            Slots := array:
                canvas_slot:
                    Anchors := anchors{Minimum := vector2{X:=0.5,Y:=1.0}, Maximum := vector2{X:=0.5,Y:=1.0}}
                    Alignment := vector2{X := 0.5, Y := 1.0}
                    Offsets := margin{Top := -200.0}
                    Widget := stack_box:
                        Orientation := orientation.Horizontal
                        Slots := array:
                            stack_box_slot:
                                Widget := LobbyButton
                                Padding := margin{Right := 25.0}
                            stack_box_slot:
                                Padding := margin{Left := 40.0}
                                Widget := SpectateButton
                            
        return (Canvas, LobbyButton, SpectateButton)

    #################################################
    #################### UTILS ######################
    #################################################

    PlayerRespawnedEvent : event(player) = event(player){}
    (Player: player).RespawnedEvent():event()=
        if(not PlayerRespawnedEvents[Player]):
            option{set PlayerRespawnedEvents[Player] = event(){}}
        return PlayerRespawnedEvents[Player] or event(){} # never fails

    RespawnEverybody()<suspends>:void=
        # Retrieve team with spawn limit at 0 and spawn type set to Wave
        TeamCollection := GetPlayspace().GetTeamCollection()
        if(WaveRespawnTeam := TeamCollection.GetTeams()[WaveRespawnTeamData.TeamIndex-1]):
            
            # Move all players to said team
            PlayersInWaveRespawnTeam := for(Player : GetPlayspace().GetPlayers()):
                spawn{TeamCollection.ForceAddToTeam(Player, WaveRespawnTeam)} # May take up to 1s
                GetUIFaderInstance().FadeIn(Player)
                Player

            # Wait for everyone to be added to the team (hopefully)
            Sleep(1.0)

            CustomRespawnAudioMixer.ActivateMix()
            HUDControllerNoHUD.Enable()
            
            # Kill everybody to force the wave respawn to occur
            for(Player : PlayersInWaveRespawnTeam):
                BlackScreenHUDMessage.Show(Player)
                CustomRespawnAudioMixer.Register(Player)
                Player.Respawn(EliminationDamageVolume.GetTransform().Translation + vector3{Z := 600.0}, rotation{})

            # # Move everybody to wave respawn team
            # if(UnlimitedSpawnsTeam := GetPlayspace().GetTeamCollection().GetTeams()[WaveRespawnTeamData.TeamIndex-1]):
            #     for(Player : PlayersInWaveRespawnTeam):
            #         spawn{GetPlayspace().GetTeamCollection().ForceAddToTeam(Player, UnlimitedSpawnsTeam)} # May take up to 1s
            
            Sleep(1.0) # So we don't trigger this on still alive players
            
            # Watch respawning time (to trigger OnSpawned/Respawned events)
            for(Player : PlayersInWaveRespawnTeam):
                spawn{WatchRespawn(Player)}

    WatchRespawn(Player: player)<suspends>:void=
        if(Character := Player.GetFortCharacter[]):
            # Print("GetFortCharacter[]")
            # NewChar.SetVulnerability(true)
            # NewChar.Damage(999.9)

            # Wait for player to be respawned
            race:
                loop:
                    if(Character.IsActive[]):
                        Print("IsActive[]")
                        OnPlayerSpawned(Player)
                        Player.RespawnedEvent().Signal()
                        PlayerRespawnedEvent.Signal(Player) # another global event
                        break
                    Sleep(0.0)
                Sleep(10.0)

    # Kill player in an organic way so that we don't have to deal with the .Respawn() issues
    KillPlayerLegitely(Player: player)<suspends>:void=
        if(Character := Player.GetFortCharacter[]):
            Print(BuildPlayerMessage("Respawn(", Player, ")"))
            Player.Respawn(EliminationDamageVolume.GetTransform().Translation + vector3{Z := 200.0}, rotation{})

            # Sleep(0.5)
            if(UnlimitedSpawnsTeam := GetPlayspace().GetTeamCollection().GetTeams()[WaveRespawnTeamData.TeamIndex-1]):
                Print("AddToUnlimitedSpawnTeam")
                GetPlayspace().GetTeamCollection().ForceAddToTeam(Player, UnlimitedSpawnsTeam) # May take up to 1s
            
            # Sleep(0.5)
            if(NewChar := Player.GetFortCharacter[]):
                # Print("GetFortCharacter[]")
                # NewChar.SetVulnerability(true)
                # NewChar.Damage(999.9)

                # Wait for player to be respawned
                race:
                    loop:
                        if(Character.IsActive[]):
                            Print("IsActive[]")
                            OnPlayerSpawned(Player)
                            Player.RespawnedEvent().Signal()
                            PlayerRespawnedEvent.Signal(Player) # another global event
                            break
                        Sleep(0.0)
                    Sleep(10.0)
        else:
            Print("No respawnable fort_character found")

    # Called when player presses the "Go to Lobby" button or before showing end game sequence
    # (players will be killed legitely again when round ends)
    FakeRespawn(Player: player):void=
        if(RandomSpawner := Shuffle(FindObjectsWithTag(spawner{}))[0]):
            Player.Respawn(RandomSpawner.GetTransform().Translation, RandomSpawner.GetTransform().Rotation)
            spawn{OnPlayerSpawned(Player)}

    SwitchToEmptyTeam(Player: player):void=
        if(
            Team := FindEmptyTeam[array{WinnersTeamData.TeamIndex, WaveRespawnTeamData.TeamIndex}] # Excluded teams
            TeamCollection := GetPlayspace().GetTeamCollection()
        ):
            spawn{TeamCollection.ForceAddToTeam(Player, Team)}

    FadeOutScreen(Player: player)<suspends>:void=
        # Wait for black screen hud to actually be removed
        HideBlackScreenTrigger.Trigger(Player)
        Sleep(0.5)
        # Helps removing the death feed from respawning method
        GetUIFaderInstance().FadeOut(Player)

    #################################################
    ################## Lobby Stuff ##################
    #################################################
    OnButtonLobbyTrigger(Agent : agent):void=
        #Print("Xp button triggered")
        if:
            Player := player[Agent]
        then:
            if(GoldTracker.GetValue(Player) > 99):
                Print("Player has enough gold")
                #TESTING: HAD -250 before
                GoldTracker.Add(Player,-250)
                XpLobbyAccolade.Award(Player)

    # Sort winners by wins/kills
    SortFunctionMinigameWins(Data1:tuple(player, int), Data2:tuple(player, int))<transacts>: int =
        Wins1 := Data1(1)
        Wins2 := Data2(1)
        if (Wins1 > Wins2):
            1
        else if (Wins1 < Wins2):
            -1
        else:
            0
        
    GetSortedGlobalMinigamePlayers()<transacts>:[]player=
        var PlayerStats : []tuple(player, int) = array{}

        for(Player : GetPlayspace().GetPlayers()):
            MinigameWins := MinigameWinTracker.GetValue(Player)
            set PlayerStats += array{(Player, MinigameWins)}

        return for(Stats : PlayerStats.QuickSort(SortFunctionMinigameWins)) { Stats(0) }
    
    DebugFullSortedPlayersList()<transacts>:[]tuple(player, int, int)=
        AllPlayers := GetPlayspace().GetPlayers()
        if(RandomPlayer := AllPlayers[0]):
            Rando := (RandomPlayer, 7777, 5555)
            var SortedPlayers : []tuple(player, int, int) = array{}
            for(I:=0..17):
                set SortedPlayers += array{Rando}
            return SortedPlayers
        return array{}

    UpdateScoreboard():void=
        Print("Updating Scoreboard ")
        SortedPlayers := GetBestPlayersTotalLifetime()
        #SortedPlayers := DebugFullSortedPlayersList()

        #[]tuple(player, int, int)
        var BillboardNames : message = StringToMessage("")
        var BillboardWins : string = ""
        var BillboardElims : string = ""
        #for(Index := 0..SortedPlayers.Length-1, Index < 8):
        for(Index := 0..7):
            if(Index < SortedPlayers.Length):
                if:
                    PlayerTuple:= SortedPlayers[Index]
                    Player := PlayerTuple(0)
                    PlayerWinsScoreboard := ToString(PlayerTuple(1))
                    PlayerElimsScoreboard:= PlayerTuple(2)
                then:
                    set BillboardNames = BillboardNames + BuildPlayerMessage("{Index+1}.",Player,"\n")
                    set BillboardWins += "{PlayerWinsScoreboard}\n"
                    set BillboardElims += "{PlayerElimsScoreboard}\n"
            else:
                set BillboardNames = BillboardNames + StringToMessage("{Index+1}.\n")
                #set BillboardWins += "\n"
                #set BillboardElims += "\n"

        BillboardName1.SetText(BillboardNames)
        BillboardWins1.SetText(StringToMessage(BillboardWins))
        BillboardElims1.SetText(StringToMessage(BillboardElims))

        set BillboardNames = StringToMessage("")
        set BillboardWins = ""
        set BillboardElims = ""

        for(Index := 8..15):
            if(Index < SortedPlayers.Length):
                #Print("In the array")
                if:
                    PlayerTuple:= SortedPlayers[Index]
                    Player := PlayerTuple(0)
                    PlayerWinsScoreboard := ToString(PlayerTuple(1))
                    PlayerElimsScoreboard:= PlayerTuple(2)
                then:
                    set BillboardNames = BillboardNames + BuildPlayerMessage("{Index+1}.",Player,"\n")
                    set BillboardWins += "{PlayerWinsScoreboard}\n"
                    set BillboardElims += "{PlayerElimsScoreboard}\n"
            else:
                set BillboardNames = BillboardNames + StringToMessage("{Index+1}.\n")
                #set BillboardWins += "\n"
                #set BillboardElims += "\n"

        BillboardName2.SetText(BillboardNames)
        BillboardWins2.SetText(StringToMessage(BillboardWins))
        BillboardElims2.SetText(StringToMessage(BillboardElims))

    RemoveControlThirdPersonDevices()<suspends>:void=
        Print("Removing Gameplay Control Devices")
        #     Spawners := GetCreativeObjectsWithTag(spawner{})
        #     for (Obj: Spawners):
        #         if(Spawner := player_spawner_device[Obj]):
        #if(RandomSpawner := Shuffle(GetCreativeObjectsWithTag(spawner{}))[0]):
        ControlThirdPersonDevices := FindObjectsWithTag(controller_third_person_device{})
        for(Object : ControlThirdPersonDevices):
            if(ControlThirdPersonDevice := gameplay_controls_third_person_device[Object]):
                ControlThirdPersonDevice.Enable()
                Print("Removing Gameplay Control Device")
                ControlThirdPersonDevice.RemoveFromAll()
                ControlThirdPersonDevice.Disable()

    OnMinigameCollectibleCoinCollected(Agent: agent):void=
        if:
            Player := player[Agent]
        then:
            Print("Player collected collectible coin")
                GoldTracker.Add(Player,25)

    OnDefaultWeaponEquip(Player : player, PlayerManager : V_PlayerManager, ButtonGroupName : string, BrandName : string, ItemName : string, bUIData : ?UI_Data, Equip : logic):void={}

    #Subscribe Event from Switch with handler 
    #TEAM MINIGAME ADDED CODE
    SetPlayerLoadout(InAgent : agent, LoadoutIndex : int):void=
        Print("Store Loadout For Player to be used in the minigame, Index = {LoadoutIndex}")
        #Primary Loadouts Weapons 1,2 and 3
        if(LoadoutIndex = 1, Player := player[InAgent]):
            if(set PlayerDefaultWeaponLoadoutMap[Player] = option{0}):
        if(LoadoutIndex = 2, Player := player[InAgent]):
            if(set PlayerDefaultWeaponLoadoutMap[Player] = option{1}):
        if(LoadoutIndex = 3, Player := player[InAgent]):
            if(set PlayerDefaultWeaponLoadoutMap[Player] = option{2}):

        #Secondary Loadout for shotguns Loadouts 1,2 and 3
        if(LoadoutIndex = 4, Player := player[InAgent]):
            if(set PlayerDefaultWeaponLoadoutSecondaryMap[Player] = option{0}):
        if(LoadoutIndex = 5, Player := player[InAgent]):
            if(set PlayerDefaultWeaponLoadoutSecondaryMap[Player] = option{1}):
        if(LoadoutIndex = 6, Player := player[InAgent]):
            if(set PlayerDefaultWeaponLoadoutSecondaryMap[Player] = option{2}):
        

    GetPlayerDefaultWeaponLoadout(Player : player)<transacts><decides>:int=
        PlayerDefaultWeaponLoadoutMap[Player]?

    GetPlayerSecondaryWeaponLoadout(Player : player)<transacts><decides>:int=
        PlayerDefaultWeaponLoadoutSecondaryMap[Player]?
    
    # OnPlayerEliminated(Result : elimination_result):void=
    #     if:
    #         PersonKilled := Result.EliminatedCharacter
    #         PersonKilledAgent := PersonKilled.GetAgent[]
    #         PersonKilledPlayer := player[PersonKilledAgent]
    #         #PersonKilledCustomPlayer := CustomPlayers[PersonKilledPlayer]

    #         Killer := Result.EliminatingCharacter?
    #         KillerAgent := Killer.GetAgent[]
    #         KillerFort := KillerAgent.GetFortCharacter[]
    #         KillerPlayer := player[KillerAgent]
    #         #KillerCustomPlayer := CustomPlayers[KillerPlayer]
    #     then:
    #         Print("Player Eliminated")
    #         #RemovePlayerAlive(PersonKilledPlayer)


    # OnPlayerLeftServer(PlayerLeaving :player):void=
    # RemovePlayerAlive(PlayerLeaving)


    # if(CustomPlayer := CustomPlayers[PlayerLeaving]):
    #     #CustomPlayer.Dispose()
    #     var NewCustomPlayerMap:[player]custom_player = map{}

    #     for(Key -> Value : CustomPlayers, Key <> PlayerLeaving):
    #         set NewCustomPlayerMap = ConcatenateMaps(NewCustomPlayerMap, map{Key => Value})

    #     set CustomPlayers = NewCustomPlayerMap


    # RemovePlayerAlive(PlayerRemoved : player):void=
    #     for(Player->Logic : PlayersAlive):
    #         if(PlayerRemoved = Player):
    #             var NewActivePlayerMap:[player]logic = map{}
    #             for(Key -> Value : PlayersAlive, Key <> PlayerRemoved):
    #                 set NewActivePlayerMap = ConcatenateMaps(NewActivePlayerMap, map{Key => Value})
    #             set PlayersAlive = NewActivePlayerMap
                

    # OnPlayerSaveDataLoaded(AgentMaybe: ?agent):void=
    #     if(Player := player[AgentMaybe?]):
    #         InitBillboard()


    # PlayerName<localizes>(Agent:agent, Score: int):message = "{Agent}, {Score}"


    # InitBillboard():void=
    #     AllPlayers := GetPlayspace().GetPlayers()
    #     Print("AllPlayers Size = {AllPlayers.Length}")
    #     var test: string = ""
    #     if(
    #         Player1 := AllPlayers[0]
    #         Player2 := AllPlayers[1]
    #     ):
    #         if:
    #             Fort := Player1.GetFortCharacter[]
    #             Score := ScoreManager.GetCurrentScore(Player1)
    #         then:
    #             Print("Players current score {Score}")
    #             #AgentName := PlayerName(Agent)
    #             #set test = test + Localize(AgentName) + "\n"
    #             Billboard.SetText(PlayerName(Player1, Score))

#     DeviceIdentifierForDebug : string = ""

#     var DebugEnabledPlayer<private> : [agent]logic = map{}
#     var DebugPlayerWidgets<private> : [agent]tuple([]text_block, stack_box) = map{}

#     var Logs : []string = array{}
#     var LogIDX<private> : int = -1

#     var LogsLimit<private> : int = 10

#     GetLogs():[]string=
#         var LogsAux : []string = array{}
#         if(Logs.Length = LogIDX):
#             set LogsAux = Logs            
#         else:
#             if(set LogsAux += Logs.Slice[LogIDX+1] + Logs.Slice[0,LogIDX+1]):        
#         return LogsAux

#     PrintLog(LogMessage : string):void=
#         if(Logs.Length < LogsLimit):
#             set Logs += array{"{DeviceIdentifierForDebug}-{GetSimulationElapsedTime()}|{LogMessage}"}
#             set LogIDX += 1
#         else:
#             if(LogIDX >= Logs.Length-1):
#                 set LogIDX = 0
#             else:
#                 set LogIDX += 1           
#             if(LogsAux := Logs.ReplaceElement[LogIDX, "{DeviceIdentifierForDebug}-{GetSimulationElapsedTime()}|{LogMessage}"]):
#                 set Logs = LogsAux
                
#         ManageDebug(LogIDX)
                    
#     SetLogsLimit(Limit : type {_X:int where 1 <= _X, _X <= 30}):void=
#         set LogsLimit = Limit

#     GetLogsLimit():int=
#         return LogsLimit

#     TougleDebug(MaybeAgent : ?agent):void=
#         Print("Button Pressed")
#         if(Agent := MaybeAgent?, Player := player[Agent]):
#             if(not DebugEnabledPlayer[Agent]):
#                 Print("Debug On")
#                 if(set DebugEnabledPlayer[Agent] = true):                                    
#                 CrashDebugUI := crash_debug_ui:
#                 StackBoxs := CrashDebugUI.CreateUI(Player, GetLogs())
#                 if(set DebugPlayerWidgets[Agent] = StackBoxs):
#             else if(not DebugEnabledPlayer[Agent]?):
#                 Print("Debug On 2")
#                 if(set DebugEnabledPlayer[Agent] = true):
#                 if(Widgets := DebugPlayerWidgets[Agent], TextBlocks := Widgets(1)):
#                     TextBlocks.SetVisibility(widget_visibility.Visible)
#             else:
#                 Print("Debug Off")
#                 StopDebug(Agent)

#     ManageDebug(IDX : int):void=        
#         for(Agent -> Widgets : DebugPlayerWidgets, DebugEnabledPlayer[Agent]?, TextBlocks := Widgets(0), StackBox := Widgets(1)):
#             if(TextBlock := TextBlocks[IDX], LogMessage := Logs[IDX]):                      
#                 StackBox.RemoveWidget(TextBlock)
                
#                 TextBlockNew : text_block = text_block:                 
#                 TextBlockNew.SetText(MakeCrashableMessageFromString(LogMessage))
#                 TextBlockNew.SetTextColor(MakeColorFromHex("22ffff"))
            
#                 StackBoxSlot := stack_box_slot:
#                     Widget := TextBlockNew
#                     HorizontalAlignment := horizontal_alignment.Left
#                     Padding := margin{Top := 0.0, Left := 10.0, Right := 0.0, Bottom := 5.0}

#                 NewTextBlocks := array{TextBlockNew} + for(TextBlockAux : TextBlocks, TextBlock <> TextBlockAux). TextBlockAux
#                 if(set DebugPlayerWidgets[Agent] = (NewTextBlocks, StackBox)):
                    
#                 StackBox.AddWidget(StackBoxSlot)
#             else if(DebugEnabledPlayer[Agent]?):                
#                 if(LogMessage := Logs[IDX]):     
#                     TextBlockNew : text_block = text_block:                 
#                     TextBlockNew.SetText(MakeCrashableMessageFromString(LogMessage))
#                     TextBlockNew.SetTextColor(MakeColorFromHex("22ffff"))
            
#                     StackBoxSlot := stack_box_slot:
#                         Widget := TextBlockNew
#                         HorizontalAlignment := horizontal_alignment.Left
#                         Padding := margin{Top := 0.0, Left := 10.0, Right := 0.0, Bottom := 5.0}
                    
#                     NewTextBlocks := TextBlocks + array{TextBlockNew}
#                     if(set DebugPlayerWidgets[Agent] = (NewTextBlocks, StackBox)):
                    
#                     StackBox.AddWidget(StackBoxSlot)

#     StopDebug(Agent : agent):void=        
#         if(DebugEnabledPlayer[Agent]?, Widgets := DebugPlayerWidgets[Agent], TextBlocks := Widgets(1)):
#             if(set DebugEnabledPlayer[Agent] = false):            
#             TextBlocks.SetVisibility(widget_visibility.Hidden)

# using { /UnrealEngine.com/Temporary/UI }
# using { /Fortnite.com/UI }
# using { /Verse.org/Colors }
# crash_debug_ui := class():    

#     CreateUI(Player : player, Logs : []string):tuple([]text_block, stack_box)=
#         Print("Creating UI")

#         var TextBlocks : []text_block = array{}
        
#         StackBoxVerticalPosition : stack_box = stack_box:                
#             Orientation := orientation.Vertical        

#         for(LogMessage : Logs):

#             TextBlock : text_block = text_block: 
#             set TextBlocks += array{TextBlock}
#             TextBlock.SetText(MakeCrashableMessageFromString(LogMessage))
#             TextBlock.SetTextColor(MakeColorFromHex("22ffff"))
        
#             StackBoxSlot := stack_box_slot:
#                 Widget := TextBlock
#                 HorizontalAlignment := horizontal_alignment.Left
#                 Padding := margin{Top := 0.0, Left := 10.0, Right := 0.0, Bottom := 5.0}

#             StackBoxVerticalPosition.AddWidget(StackBoxSlot)

#         StackBoxMainPosition : stack_box = stack_box:
#             Slots := array:
#                 stack_box_slot:
#                     Widget := StackBoxVerticalPosition
#                     HorizontalAlignment := horizontal_alignment.Left
#                     Padding := margin{Top := 100.0, Left := 10.0, Right := 0.0, Bottom := 0.0}
#             Orientation := orientation.Vertical
        
#         if(PlayerUI := GetPlayerUI[Player]):
#             PlayerUI.AddWidget(StackBoxMainPosition)                          
        
#         return (TextBlocks, StackBoxVerticalPosition)

# MakeCrashableMessageFromString<localizes>(Text : string) : message = "{Text}"
