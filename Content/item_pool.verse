using { /Fortnite.com/Devices }
using { /Fortnite.com/Game }
using { /Fortnite.com/Characters }
using { /UnrealEngine.com/Temporary/SpatialMath }
using { /UnrealEngine.com/Temporary/Diagnostics }
using { /Verse.org/Simulation }
using { /Verse.org/Random }

var PoolMutableClasses : weak_map(session, [string]item_pool_mutable_class) = map{}

item_pool(item_type:type) := class(crashable_device):

    DeviceName<override>:string="__GENERATED__"
    GetDeviceName<override>()<transacts><decides>:string=GetFullIdentifier()

    Items<private>: []item_type = array{}

    GetItems()<transacts> : []item_type = Items

    GetIdentifier()<transacts> : string = ""

    # UPDATE: Overriding it in child classes will cause project to crash
    OnCrashableDeviceBegin<override><final>()<suspends>:void =
        if(PoolMutableClasses[GetSession()][GetFullIdentifier()]):
            Print("item_pool Identifier {GetFullIdentifier()} is not unique, erasing old class..")
        
        _OnBegin()

        # Make sure mutable classes are all instanced from the begining
        GetInnerClass(?EraseOldClass := true)

    _OnBegin()<suspends>:void=block:

    GetFullIdentifier<final>()<transacts>:string= "POOL_{GetIdentifier()}"

    # Mutable members in parametric classes are not yet implemented
    GetInnerClass(?EraseOldClass:logic = false)<transacts>:item_pool_mutable_class =
        if(not PoolMutableClasses[GetSession()]):
            option{set PoolMutableClasses[GetSession()] = map{}}

        if(not PoolMutableClasses[GetSession()][GetFullIdentifier()] or EraseOldClass?):
            PrintT("Creating mutable class for pool {GetFullIdentifier()}")
            NewMutableClass := item_pool_mutable_class{ItemCount := GetItems().Length}
            NewMutableClass.Init()
            option{set PoolMutableClasses[GetSession()][GetFullIdentifier()] = NewMutableClass}

        if(MutableClass := PoolMutableClasses[GetSession()][GetFullIdentifier()]):
            return MutableClass
            
        item_pool_mutable_class{ItemCount := GetItems().Length}

    # We won't try to spawn the item if there's none available for this duration
    @editable
    FetchTimeout: float = 1.0

    FetchAvailableItems(Count: int)<suspends>:?([]tuple(item_type, int))=
        MutableClass := GetInnerClass()
        AvailableIndexes := MutableClass.GetAvailableIndexes()

        if(Count > AvailableIndexes.Length):
            Print("Tried to fetch {Count} element(s) from pool but only {AvailableIndexes.Length} are available")
            return false

        if(Count < 1):
            return option{array{}}
        
        var FetchedItems : []tuple(item_type, int) = array{}
        var Failed : logic = false

        loop:
            for(I := 1..Count):
                FetchResult := FetchAvailableItem()
                
                if(_FetchResult := FetchResult?):
                    FetchedItem := _FetchResult(0)
                    FetchedItemIndex := _FetchResult(1)

                    set FetchedItems += array{(FetchedItem, FetchedItemIndex)}
                else:
                    set Failed = true
                    # UPDATE: Only way to break out of for loop?
                    break

            # UPDATE: Only way to break out of for loop?
            # if 'for' didn't fail (thus break), break the loop now
            if(true?):
                break

        # Release fetched items
        if(Failed?):
            for(Item : FetchedItems):
                ReleaseItem(Item(1))

        if(Failed?) then false else option{FetchedItems}

    TryFetchItemAtIndex(Index: int)<decides>:item_type=
        var RetItem : ?item_type = false
        MutableClass := GetInnerClass()

        # If available and exists
        if(MutableClass.AvailableIndexesStates[Index]?, Item := GetItems()[Index]):
            OnItemFetched(Item, Index)
            set RetItem = option{Item}
        RetItem?

    # Retrieve one item from the available pool list
    FetchAvailableItem()<suspends>:?tuple(item_type, int)=
        MutableClass := GetInnerClass()

        var AssignedIndex : int = 0

        StartTime := GetSimulationElapsedTime()
        
        loop:
            AvailableIndexes := MutableClass.GetAvailableIndexes()
            if(
                AvailableIndexData := AvailableIndexes.GetRandomElementFromArray[],
                AvailableIndex := AvailableIndexData(0),
                AvailableIndexIndex := AvailableIndexData(1)
            ):
                option{set MutableClass.AvailableIndexesStates[AvailableIndex] = false}
                set AssignedIndex = AvailableIndex
                # Print("Found {AvailableIndexes.Length} elements in {GetFullIdentifier()}")
                break

            # Fetching failed!
            if((GetSimulationElapsedTime() - StartTime) >= FetchTimeout):
                return false

            # Print("Could not find any available item to retrieve, retrying next frame...")
            Sleep(0.0)
    
        # Print("Assigning item at index {AssignedIndex}")

        if(Item := GetItems()[AssignedIndex]):
            OnItemFetched(Item, AssignedIndex)
            return option{(Item, AssignedIndex)}
        false

    # Release the item back into the pool
    ReleaseItem(AtIndex:int):void=
        MutableClass := GetInnerClass()

        # Index exists and didn't got released
        if(MutableClass.AvailableIndexesStates[AtIndex] = false):
            option{set MutableClass.AvailableIndexesStates[AtIndex] = true}

            OnItemReleased(AtIndex)
            # Print("released item({GetFullIdentifier()}) at index {AtIndex}")

    ReleaseAll():void=
        for(I := 0..GetInnerClass().AvailableIndexesStates.Length):
            ReleaseItem(I)

    # OnItemFetched:event(tuple(player, int)) = event(tuple(player, int)){}
    OnItemFetched(Item: item_type, Index: int):void = block:
    # OnItemReleased:event(tuple(player, int)) = event(tuple(player, int)){}
    OnItemReleased(Index: int):void = block:
        
item_pool_mutable_class := class:
    # Pool:item_pool(any)

    ItemCount : int = 0
    # var AvailableIndexes : []int = array{}
    var AvailableIndexesStates : [int]logic = map{}

    Init()<transacts>:void=
        PrintT("INIT POOL WITH {ItemCount} ITEMS")
        if(ItemCount > 0):
            for(Index := 0..(ItemCount-1)):
                # set AvailableIndexes += array{Index}
                option{set AvailableIndexesStates[Index] = true}

    GetAvailableIndexes()<transacts>:[]int=
        var IndexArray : []int = array{}
        for(Index -> State : AvailableIndexesStates, State?):
            set IndexArray += array{Index}

        return IndexArray


hud_message_pool := class(item_pool(hud_message_device)):

    FetchTimeout<override>:float = 0.0

    @editable
    HUDMessageDevices : []hud_message_device = array{}

    @editable
    Identifier : string = ""

    GetItems<override>()<transacts> : []hud_message_device = HUDMessageDevices

    GetIdentifier<override>()<transacts> : string = "HUD_MESSAGE_{Identifier}"

knockback_pool_data := class<concrete>:

    @editable
    ExplosiveDevice : explosive_device = explosive_device{}

    # @editable
    # AudioPlayer : audio_player_device = audio_player_device{}

    # @editable
    # VFXSpawner : vfx_spawner_device = vfx_spawner_device{}

knockback_pool := class(item_pool(knockback_pool_data)):

    @editable
    KnockbackItems : []knockback_pool_data = array{}

    @editable
    Identifier : string = ""

    GetItems<override>()<transacts> : []knockback_pool_data = KnockbackItems

    GetIdentifier<override>()<transacts> : string = "KNOCKBACK_{Identifier}"

    
attachable_item_pool := class(item_pool(attachable_item)):

    @editable
    ItemIdentifier : string = ""

    @editable
    AttachedItemOptions : attachable_item_options = attachable_item_options{}

    # @editable
    # ReleasesOnPlayerDeath : logic = true

    @editable
    Props : []creative_prop = array{}

    var GeneratedPlayerAttachedItems : []attachable_item = array{}

    GetItems<override>()<transacts> : []attachable_item = GeneratedPlayerAttachedItems

    GetIdentifier<override>()<transacts> : string = "AT_IT_{ItemIdentifier}"

    _OnBegin<override>()<suspends>:void =
        for(Prop : Props):
            if(AttachedItemOptions.StartsHidden?):
                Prop.Hide()

            AttachedItem := attachable_item{
                Item := Prop
                Options := AttachedItemOptions
            }

            set GeneratedPlayerAttachedItems += array{AttachedItem}

    TryAttachOneToPositional(Positional: positional)<suspends>:?attachable_item=
        var RetAttachable : ?attachable_item = false
        FetchResultMaybe := FetchAvailableItem()
        if(FetchResult := FetchResultMaybe?):

            AttachedItem := FetchResult(0)
            Index := FetchResult(1)
            
            # Release from pool when item is detached
            spawn{ReleaseOnItemDetached(AttachedItem, Index)}

            AttachedItem.AttachToPositional(Positional)

            set RetAttachable = option{AttachedItem}
        return RetAttachable

    ReleaseOnItemDetached(AttachedItem: attachable_item, Index: int)<suspends>:void=
        AttachedItem.OnDetached.Subscribable.Await()
        ReleaseItem(Index)

dodge_wall := class<concrete>:
    @editable
    Prop : creative_prop = creative_prop{}

    @editable
    CollisionFeetVolumes : []volume_device = array{}

    var OriginPropTransform : ?transform = false

dodge_wall_pool := class(item_pool(dodge_wall)):

    @editable
    DodgeWalls : []dodge_wall = array{}

    GetItems<override>()<transacts> : []dodge_wall = DodgeWalls

    GetIdentifier<override>()<transacts> : string = "DODGE_WALL"

    _OnBegin<override>()<suspends>:void =
        for(DodgeWall : DodgeWalls, DodgeWall.Prop.IsValid[]):
            set DodgeWall.OriginPropTransform = option{DodgeWall.Prop.GetTransform()}

    OnItemReleased<override>(Index: int):void=
        if(DodgeWall := DodgeWalls[Index], OriginTransform := DodgeWall.OriginPropTransform?):
            option{DodgeWall.Prop.TeleportTo[OriginTransform]}

freeze_gun_vfx_pool := class(attachable_item_pool):
    @editable
    AnimationDuration : float = 1.0

    @editable
    VFXSpawners : []vfx_spawner_device = array{}

    _OnBegin<override>()<suspends>:void =
        for(VFX : VFXSpawners, ItemAsPositional := positional[VFX]):
            VFX.Disable()

            AttachedItem := attachable_item{
                Item := ItemAsPositional
                Options := AttachedItemOptions
            }

            AttachedItem.OnFirstMoveInitiated.Subscribable.Subscribe(OnFirstMoveInitiated)
            AttachedItem.OnDetached.Subscribable.Subscribe(OnDetached)

            set GeneratedPlayerAttachedItems += array{AttachedItem}
            
    OnFirstMoveInitiated(Item: attachable_item, : positional):void=
        spawn{_OnFirstMoveInitiated(Item)}

    _OnFirstMoveInitiated(Item: attachable_item)<suspends>:void=
        # Print("Received FirstMoveInitiated Signal")
        race:
            Item.OnDetached.Subscribable.Await()

            sync:
                block:
                    if(VFXSpawner := vfx_spawner_device[Item.Item]):
                        Sleep(0.0)
                        Sleep(0.0)
                        Sleep(0.0)
                        VFXSpawner.Enable()
                block:
                    Sleep(AnimationDuration)
                    Item.Detach()

    OnDetached(Item: attachable_item, :positional):void=
        spawn{_OnDetached(Item)}

    _OnDetached(Item: attachable_item)<suspends>:void=
        if(VFXSpawner := vfx_spawner_device[Item.Item]):
            VFXSpawner.Disable()

player_reference_pool := class(item_pool(player_reference_device)):

    @editable
    ItemIdentifier : string = ""

    @editable
    Devices : []player_reference_device = array{}

    GetItems<override>()<transacts> : []player_reference_device = Devices

    GetIdentifier<override>()<transacts> : string = "PLAYERREF_{ItemIdentifier}"
            
# player_attached_camera_pool_item := class<concrete>:
#     @editable
#     Sequence : cinematic_sequence_device = cinematic_sequence_device{}

#     # @editable
#     # OffSequence : cinematic_sequence_device = cinematic_sequence_device{}

#     @editable
#     ParentProp : creative_prop = creative_prop{}

# player_attached_camera_pool := class(item_pool(player_attached_camera_pool_item)):

#     @editable
#     PFIdentifier : string = ""

#     @editable
#     CamerasAndProps : []player_attached_camera_pool_item = array{}

#     @editable
#     RotatesWithPlayer : logic = false

#     GetIdentifier<override>()<transacts>:string = "PF_CAM_{PFIdentifier}"

#     GetItems<override>()<transacts>:[]player_attached_camera_pool_item = CamerasAndProps

#     GetCameraParentPropPosition(Character: fort_character)<transacts>:vector3 = Character.GetTransform().Translation

#     GetCameraParentPropRotation(Character: fort_character)<transacts>:rotation =
#         if(RotatesWithPlayer?, Yaw := Character.GetTransform().Rotation.GetYawPitchRollDegrees()[0]):
#             return MakeRotationFromYawPitchRollDegrees(Yaw, 0.0, 0.0)
#         return IdentityRotation()

#     # AssignOneToPlayer(Player: player)<suspends>:void=
#     #     FetchResultMaybe := FetchAvailableItem()
#     #     if(FetchResult := FetchResultMaybe?):
#     #         Item := FetchResult(0)
#     #         Index := FetchResult(1)

#     #         Sequence := Item.Sequence
#     #         ParentProp := Item.ParentProp

#     #         if(Character := Player.GetFortCharacter[], ParentProp.TeleportTo[GetCameraParentPropPosition(Character), GetCameraParentPropRotation(Character)]):


#     var IsPaused : logic = false

#     RunAttachLoop(Player: player, Item: player_attached_camera_pool_item)<suspends>:void=
#         Sequence := Item.Sequence
#         ParentProp := Item.ParentProp

#         if(Character := Player.GetFortCharacter[], ParentProp.TeleportTo[GetCameraParentPropPosition(Character), GetCameraParentPropRotation(Character)]):
#             Print("Play SEQ")
#             # if(IsPaused?):
#             #     Sequence.Stop()
#             #     Sleep(0.5)
#             #     Sequence.Play(Player)
#             # else:
#             Sequence.Play(Player)
#             # set IsPaused = false

#             if(Character.IsActive[]):
#                 race:
#                     Character.EliminatedEvent().Await()
#                     block:
#                         loop:
#                             Sleep(0.4)
#                             spawn:
#                                 # ParentProp.MoveTo(GetCameraParentPropPosition(Character), ParentProp.GetTransform().Rotation, 1.0)
#                                 ParentProp.MoveTo(GetCameraParentPropPosition(Character), GetCameraParentPropRotation(Character), 0.6)

#         # Triggered once the player dies
#         Sequence.Pause(Player) # Fix for all characters being locked in a direction after stopping someone else's sequence
#         # set IsPaused = true
#         # Sleep(0.1)
#         # ReleaseItem(Index)
