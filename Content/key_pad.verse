
using { /Fortnite.com/Devices }
using { /Verse.org/Simulation }
using { /UnrealEngine.com/Temporary/Diagnostics }
using { /Fortnite.com/UI }
using { /UnrealEngine.com/Temporary/UI }
using { /UnrealEngine.com/Temporary/SpatialMath }
using { /Verse.org/Colors }

# See https://dev.epicgames.com/documentation/en-us/uefn/create-your-own-device-in-verse for how to create a verse device.
_StringToMessage<localizes>(String : string) : message = "{String}"
# Class contains details of each button
K_Button_Data := class():
    Button : button_quiet = button_quiet{}
    X : int = 0
    Y : int = 0
    ScaleX : int = 85
    ScaleY : int = 75
    Text : string = ""
    ClickText : string = ""
    var bCanva : ?Canva_Data = false

    SetButtonText():void=
        Button.SetText(StringToMessage(Text))

    BindButtonClick():void=
        Button.OnClick().Subscribe(OnClick)

    OnClick(Message : widget_message):void=
        if (Canva := bCanva?):
            if (Text = "✖"):
                Canva.RemovePinEntry()
            else if(Text = "✔"):
                Canva.CheckPlayerCode()
            else:
                Canva.SetPinEntry(Text)



# Class that contains all keypad elements
Canva_Data := class():
    var KeyPadCanvas : canvas = canvas{}
    ExitButton : button_quiet = button_quiet{}
    bLobbyVolume : ?mutator_zone_device = false
    #var ButtonData : []K_Button_Data = array{}
    var CodeEntries : []text_block = array{}
    var bPlayer : ?player = false
    var CurrentEntryIndex : int = 0
    var bKeyPadManager : ?key_pad = false
    var CanvasActive : logic = true

    ButtonData : []K_Button_Data = array:
        K_Button_Data{Text := "1"}
        K_Button_Data{Text := "2"}
        K_Button_Data{Text := "3"}
        K_Button_Data{Text := "4"}
        K_Button_Data{Text := "5"}
        K_Button_Data{Text := "6"}
        K_Button_Data{Text := "7"}
        K_Button_Data{Text := "8"}
        K_Button_Data{Text := "9"}
        K_Button_Data{Text := "✖"}
        K_Button_Data{Text := "0"}
        K_Button_Data{Text := "✔"}

    # called when player hits enter
    CheckPlayerCode():void=
        var Code : string = ""
        for(CodeEntry:CodeEntries):
            Text := CodeEntry.GetText()
            set Code += Text
        if (KeyPadManager := bKeyPadManager?, Player := bPlayer?):
            KeyPadManager.CheckPlayerCode(Code, Player)
        RemoveCanvas()

    SetPinEntry(Num : string):void=
        if(CodeEntry := CodeEntries[CurrentEntryIndex]):
            CodeEntry.SetText(StringToMessage("{Num}"))
            if (CodeEntries[CurrentEntryIndex + 1]):
                set CurrentEntryIndex += 1

    RemovePinEntry():void=
        if(CodeEntry := CodeEntries[CurrentEntryIndex]):
            EntryText := CodeEntry.GetText()
            if(EntryText = "__"):
                set CurrentEntryIndex = Max(CurrentEntryIndex - 1, 0)
                if(PrevCodeEntry := CodeEntries[CurrentEntryIndex]):
                    PrevCodeEntry.SetText(StringToMessage("__"))
            else:
                CodeEntry.SetText(StringToMessage("__"))
    # Create keypad
    CreateKeyPad(Entries : int, OriginX : float, OriginY : float, Player : player, KeyPadManager : key_pad) : canvas =
        set bKeyPadManager = option{KeyPadManager}
        set bPlayer = option{Player}
        #set ButtonData = UIButtonData
        var CanvaSlots : []canvas_slot = array{}
        #IconTextureBlock : texture_block = texture_block{DefaultImage := MinigameIcon, DefaultDesiredSize := vector2{X := Scale, Y := Scale}}
        var X : float = OriginX
        var Y : float = OriginY
        var XCount : int = 0
        # Create button data UI matrix
        for(Data : ButtonData):
            Data.SetButtonText()
            set CanvaSlots += array{MakeCanvasSlot(Data.Button, vector2{X:= X, Y:= Y}, ?Size := vector2{X :=  Data.ScaleX * 1.0, Y :=  Data.ScaleY * 1.0})}
            Data.BindButtonClick()
            set Data.bCanva = option{Self}
            set XCount += 1
            set X += 100.0
            if (XCount >= 3):
                set XCount = 0
                set X = OriginX
                set Y += 100.0
        # Exit button
        set CanvaSlots += array{MakeCanvasSlot(ExitButton, vector2{X:= OriginX + 75.0, Y:= Y}, ?Size := vector2{X :=  150.0, Y :=  75.0})}
        ExitButton.SetText(StringToMessage("Exit"))
        ExitButton.OnClick().Subscribe(RemoveCanva)
        # Create pin code display
        var TextX : float = OriginX + 50.0
        TextY : float = OriginY - 100.0
        for(i := 1..Entries):
            Text := text_block{DefaultText := StringToMessage("__"), DefaultTextColor := color{R := 1.0, G := 1.0, B := 1.0}}
            set CodeEntries += array{Text}
            set CanvaSlots += array{MakeCanvasSlot(Text, vector2{X:= TextX, Y:= TextY})}
            set TextX += 50.0

        Canvas : canvas = canvas:                                                       # Create the canvas that will hold the GoldText -235.0
            Slots := CanvaSlots

        set KeyPadCanvas = Canvas
        spawn{ActivateSafteyCheck()}
        return Canvas

    ActivateSafteyCheck()<suspends>:void=
        loop:
            Sleep(1.0)

            if (Player := bPlayer?, not Player.IsActive[]):
                return

            if (Player := bPlayer?, LobbyVolume := bLobbyVolume?):
                if(LobbyVolume.IsInVolume[Player]):
                else:
                    RemoveCanvas()
            else:
                RemoveCanvas()
                return


    RemoveCanva(Message : widget_message):void=
        RemoveCanvas()

    RemoveCanvas():void=
        set CanvasActive = false
        if(Player := bPlayer?, PlayerUI := GetPlayerUI[Player]):
            KeyPadCanvas.SetVisibility(widget_visibility.Hidden)
            PlayerUI.RemoveWidget(KeyPadCanvas)
            if (KeyPadManager := bKeyPadManager?):
                KeyPadManager.RemoveCanvaFromMap(Player)

Reward_Data := class<concrete>():
    @editable
    Code : string = ""
    @editable
    OptItemGranter : ?item_granter_device = false
    @editable
    OptTrigger : ?trigger_device = false
    @editable
    OptCreativePropsToTeleport : ?[]arena_props = false
    @editable
    OptGold : ?int = false

# A Verse-authored creative device that can be placed in a level
key_pad := class(creative_device):

    var KeyPadHUDMap : [player]?Canva_Data = map{}
    CodeEntries : int = 4
    @editable
    OriginX : float = 750.0
    @editable
    OriginY : float = 350.0
    @editable
    Button : button_device = button_device{}
    # Add more buttons to this array
    #var ButtonData : []K_Button_Data = array{}
    @editable
    RewardData : []Reward_Data = array{}
    @editable
    LobbyVolume : mutator_zone_device = mutator_zone_device{}
    @editable
    GoldTracker : tracker_device = tracker_device{}
    var RewardDataMap : [string]Reward_Data = map{}

    

    # Runs when the device is started in a running game
    OnBegin<override>()<suspends>:void=
        # Button data
        
        # Loop through all reward entry and tie it to a hash map for easy reading of data
        for(Data : RewardData):
            if (set RewardDataMap[Data.Code] = Data){}
        # Bind button interaction of keypad display device      
        Players := GetPlayspace().GetPlayers()
        Button.InteractedWithEvent.Subscribe(OnButtonInteract)

    OnButtonInteract(Agent : agent):void=
        if (Player := player[Agent]):
            MakeCustomHUD(Player)

    # Call this to create CustomHUD for a given player
    MakeCustomHUD(Player : player):void=
        # Only add hud if one is not already active on players screen
        if (KeyPadHUDMap[Player]?):
            return
        if(set KeyPadHUDMap[Player] = option{Canva_Data{bLobbyVolume := option{LobbyVolume}}}, CanvaHUD := KeyPadHUDMap[Player]?):
            Canvas := CanvaHUD.CreateKeyPad(CodeEntries, OriginX, OriginY, Player, Self)
            if(PlayerUI := GetPlayerUI[Player]):
                PlayerUI.AddWidget(Canvas, player_ui_slot{InputMode := ui_input_mode.All})
    
    # Called after removing hud, clears hash map
    RemoveCanvaFromMap(Player : player):void=
        if(set KeyPadHUDMap[Player] = false){}

    # Grant reward based on code
    CheckPlayerCode(Code : string, Player : player):void=
        Print("Work")
        if (Reward := RewardDataMap[Code]):
            if(ItemGranter := Reward.OptItemGranter?):
                ItemGranter.GrantItem(Player)
            if(Trigger := Reward.OptTrigger?):
                Trigger.Trigger(Player)
            if(ArenaProps := Reward.OptCreativePropsToTeleport?):
                for(PropDevice : ArenaProps):
                    Print("Teleporting Prop to {PropDevice.NewLocation}")
                    if(PropDevice.CreativeProp.TeleportTo[PropDevice.NewLocation, rotation{}]):
                        Print("Got here")
            if(Gold := Reward.OptGold?):
                GoldTracker.Add(Player, Gold)