using { /Fortnite.com/Devices }
using { /Verse.org/Simulation }
using { /Verse.org/Random }
using { /Fortnite.com/Characters }
using { /Fortnite.com/Game }
using { /Verse.org/Assets }


lazar_seq := class(creative_device):
    @editable SignalGoldRemote : signal_remote_manager_device = signal_remote_manager_device{}
    @editable CinematicSequence : cinematic_sequence_device = cinematic_sequence_device{}
    @editable AudioPlayer : audio_player_device = audio_player_device{}
    @editable FlyingClass : class_and_team_selector_device = class_and_team_selector_device{}

    # Runs when the device is started in a running game
    OnBegin<override>()<suspends>:void=
        SignalGoldRemote.PrimarySignalEvent.Subscribe(HandleJumpScare)
        SignalGoldRemote.SecondarySignalEvent.Subscribe(HandleHealth)

    HandleJumpScare(Player:agent):void=
        Print("Handle jump scare")
        CinematicSequence.Play()
        AudioPlayer.Play()

    HandleHealth(Player:agent):void=
        if:
            Fort := Player.GetFortCharacter[]
        then:
            FlyingClass.ChangeClass(Player)
            Fort.SetMaxHealth(10000.0)
            Fort.SetHealth(10000.0)
