using { /Fortnite.com/Devices }
using { /Fortnite.com/AI}
using { /Fortnite.com/Teams}
using { /Verse.org/Simulation }
using { /Verse.org/Random }
using { /Fortnite.com/Characters }
using { /Fortnite.com/FortPlayerUtilities }
using { /Fortnite.com/Game }
using { /Fortnite.com/Teams }
using { /Verse.org/Simulation/Tags }
using { /UnrealEngine.com/Temporary/Diagnostics }
using { /UnrealEngine.com/Temporary/SpatialMath } 
using { /Verse.org/Assets }

# See https://dev.epicgames.com/documentation/en-us/uefn/create-your-own-device-in-verse for how to create a verse device.

# A Verse-authored creative device that can be placed in a level
gun_game := class(minigame_device):
    @editable ItemGranterGreenLoadout :item_granter_device = item_granter_device{}
    @editable ItemGranterBlueLoadout :item_granter_device = item_granter_device{}
    @editable ItemGranterPurpleLoadout :item_granter_device = item_granter_device{}
    @editable ItemGranterGoldLoadout :item_granter_device = item_granter_device{}
    @editable LoadoutRemover : item_remover_device = item_remover_device{}

    var PlayerKillMap : [player]int = map{}

    Name<override>: string = "RAGS TO RICHES"
    StartingMessage<override>:string = "Better Loadout Each Elimination!"
    GetVotingImage<override>()<transacts>: texture = VerseExposed.SQ_RAGSTORICHES

    _Start<override>()<suspends>: void = 
        set PlayerKillMap = map{}

        for(Player : GetRemainingPlayers()):
            if(set PlayerKillMap[Player] = 0) {}
                
    _OnPlayerEliminated<override>(Result : elimination_result):void=
        if:
            Killer := Result.EliminatingCharacter?
            KillerAgent := Killer.GetAgent[]
            KillerFort := KillerAgent.GetFortCharacter[]
            KillerPlayer := player[KillerAgent]
            KillerKillCount := PlayerKillMap[KillerPlayer]
            IsInGame[KillerPlayer]
        then:
            Print("Player Eliminated")
            if(KillerKillCount = 0):
                LoadoutRemover.Remove(KillerPlayer)
                ItemGranterGreenLoadout.GrantItem(KillerPlayer)
                if(set PlayerKillMap[KillerPlayer] = 1) {}
            else if(KillerKillCount = 1):
                LoadoutRemover.Remove(KillerPlayer)
                ItemGranterBlueLoadout.GrantItem(KillerPlayer)
                if(set PlayerKillMap[KillerPlayer] = 2) {}
            else if(KillerKillCount = 2):
                LoadoutRemover.Remove(KillerPlayer)
                ItemGranterPurpleLoadout.GrantItem(KillerPlayer)
                if(set PlayerKillMap[KillerPlayer] = 3) {}
            else if(KillerKillCount = 3):
                LoadoutRemover.Remove(KillerPlayer)
                ItemGranterGoldLoadout.GrantItem(KillerPlayer)
                if(set PlayerKillMap[KillerPlayer] = 4) {}


