using { /Fortnite.com/Devices }
using { /Fortnite.com/AI}
using { /Fortnite.com/Teams}
using { /Verse.org/Simulation }
using { /Verse.org/Random }
using { /Verse.org/Assets }
using { /Fortnite.com/Characters }
using { /Fortnite.com/FortPlayerUtilities }
using { /Fortnite.com/Game }
using { /Fortnite.com/Teams }
using { /Verse.org/Simulation/Tags }
using { /UnrealEngine.com/Temporary/Diagnostics }
using { /UnrealEngine.com/Temporary/SpatialMath } 
using { /Verse.org/Assets }

# A Verse-authored creative device that can be placed in a level
alien := class(minigame_device):

    Name<override>: string = "Alien Invastion"
    StartingMessage<override>:string = "Jump Twice to Fly!"
    GetVotingImage<override>()<transacts>: texture = VerseExposed.SQ_ALIEN



