using { /Fortnite.com/Devices }
using { /Fortnite.com/AI}
using { /Fortnite.com/Teams}
using { /Verse.org/Simulation }
using { /Verse.org/Random }
using { /Fortnite.com/Characters }
using { /Fortnite.com/FortPlayerUtilities }
using { /Fortnite.com/Game }
using { /Fortnite.com/Teams }
using { /Verse.org/Simulation/Tags }
using { /UnrealEngine.com/Temporary/Diagnostics }
using { /UnrealEngine.com/Temporary/SpatialMath } 
using { /Verse.org/Assets }

# A Verse-authored creative device that can be placed in a level
edit_wars := class(minigame_device):

    @editable GlassFloorProp : creative_prop = creative_prop{}
    @editable EditWarsProp : creative_prop = creative_prop{}
    # @editable ItemTimer : timer_device = timer_device{}
    # @editable ItemGranterEditWars : item_granter_device = item_granter_device{}

    Name<override>: string = "Edit Wars"
    StartingMessage<override>:string = "Edit Other Player's Down!"
    GetVotingImage<override>()<transacts>: texture = VerseExposed.SQ_EDITWARS
    
    _Start<override>()<suspends>: void = 
        if(EditWarsProp.TeleportTo[vector3{X:= 2292.0, Y:= -12704.96, Z:= 2556.2}, rotation{}]):
        if(GlassFloorProp.TeleportTo[vector3{X:= 512.0, Y:= -14336.00, Z:= -4256.0}, rotation{}]):
        TeamCollection := GetPlayspace().GetTeamCollection()
        if(Team78 := TeamCollection.GetTeams()[77]):
            for(Player : GetRemainingPlayers()):
                option{TeamCollection.AddToTeam[Player, Team78]}
        # RegisterSubscriber(ItemTimer.SuccessEvent.Subscribe(HandleTntTimerSuccess))
        # ItemTimer.Enable()
        # ItemTimer.Reset()
        # ItemTimer.Start()

    _Reset<override>()<suspends>:void=
        # ItemTimer.Disable()
        if(EditWarsProp.TeleportTo[vector3{X:= 2292.0, Y:= -12704.96, Z:= 2556.2}, rotation{}]):
        if(GlassFloorProp.TeleportTo[vector3{X:= 512.0, Y:= -14336.0, Z:= 4256.0}, rotation{}]):

    # HandleTntTimerSuccess(PlayerMaybe : ?agent):void=
    #     for(Player : GetRemainingPlayers()):
    #         ItemGranterEditWars.CycleToRandomItem(Player)
