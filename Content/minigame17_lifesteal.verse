# using { /Fortnite.com/Devices }
# using { /Fortnite.com/AI}
# using { /Fortnite.com/Teams}
# using { /Verse.org/Simulation }
# using { /Verse.org/Random }
# using { /Fortnite.com/Characters }
# using { /Fortnite.com/FortPlayerUtilities }
# using { /Fortnite.com/Game }
# using { /Fortnite.com/Teams }
# using { /Verse.org/Simulation/Tags }
# using { /UnrealEngine.com/Temporary/Diagnostics }
# using { /UnrealEngine.com/Temporary/SpatialMath } 
# using { /Verse.org/Assets }

# # See https://dev.epicgames.com/documentation/en-us/uefn/create-your-own-device-in-verse for how to create a verse device.

# # A Verse-authored creative device that can be placed in a level
# lifesteal := class(minigame_device):

#     # @editable DamageBarrier :damage_volume_device = damage_volume_device{}
#     # @editable ItemGranterDefault :item_granter_device = item_granter_device{}

#     # StartingMessage<override>:string = ""

#     # _Start<override>()<suspends>: void = 
#     #     Print("Starting Damage Minigame")
        
        
#     #     DamageBarrier.Enable()
#     #     AllPlayers := GetRemainingPlayers()
#     #     Print("AllPlayers Size = {AllPlayers.Length}")
#     #     if(Player1 := AllPlayers[0]):
#     #         ItemGranterDefault.GrantItem(Player1)
#     #     for(Player : AllPlayers):
#     #         #ItemGranterDefault.GrantItem(Player)
#     #         if:
#     #             Agent := agent[Player]
#     #             Fort := Agent.GetFortCharacter[]
#     #         then:
#     #             Fort.SetMaxShield(150.0)
#     #             Fort.SetShield(150.0)
#     #             Fort.DamagedEvent().Subscribe(HandlePlayerHit)
#     #             Fort.DamagedShieldEvent().Subscribe(HandlePlayerHitShield)

#     # HandlePlayerHit(DamageResult : damage_result):void=
#     #     HandlePlayerHitCustom(DamageResult)

#     # HandlePlayerHitShield(DamageResult : damage_result ) : void =
#     #     HandlePlayerHitCustom(DamageResult)

#     # HandlePlayerHitCustom( DamageResult : damage_result ) : void =
#     #     if:
#     #         SiphonModifier : float = 0.5
#     #         Target := DamageResult.Target
#     #         Amount := DamageResult.Amount
#     #         var DamageFromAgent : logic = true
#     #         Instigator := DamageResult.Instigator?
#     #         InstigatorAgent := Instigator.GetInstigatorAgent[]
#     #         InstigatorFort := InstigatorAgent.GetFortCharacter[]
#     #         InstigatorPlayer := player[InstigatorAgent]
#     #         InstigatorsTeam := GetPlayspace().GetTeamCollection().GetTeam[InstigatorAgent]

#     #         CharacterWhoWasHitFort := fort_character[Target]
#     #         CharacterWhoWasHitAgent := CharacterWhoWasHitFort.GetAgent[]
#     #         CharacterWhoWasHitPlayer := player[CharacterWhoWasHitAgent]
#     #         CharacterWhoWasHitTeam := GetPlayspace().GetTeamCollection().GetTeam[CharacterWhoWasHitAgent]
#     #     then:
#     #         Print("Lifesteal HIT")
#     #         if(InstigatorAgent = CharacterWhoWasHitAgent):
#     #             Print("Same Agent: Usually from storm damage")
#     #             set DamageFromAgent = false
#     #         else:
#     #             Print("Not the Same Agent")

#     #         #If Damage is from an agent
#     #         if(DamageFromAgent = true):   
#     #             var TotalHealthAndShield : float = InstigatorFort.GetHealth() + InstigatorFort.GetShield()
#     #             #Print("HEALTH: TotalHealthAndShield For Instigator - {TotalHealthAndShield} = GetHealth({InstigatorFort.GetHealth()}) + GetShield{InstigatorFort.GetShield()}")

#     #             var CustomSiphon : float = Amount * SiphonModifier
#     #             #Print("HEALTH: Custom Siphon = {CustomSiphon} = Amount-{Amount} * SiphonModifier-{SiphonModifier}")

#     #             var InstigatorHealth: float = InstigatorFort.GetHealth()
#     #             var InstigatorShield: float = InstigatorFort.GetShield()
#     #             #Print("HEALTH: OldInstigatorHealth={InstigatorHealth}, OldInstigatorShield={InstigatorShield}")

#     #             if(InstigatorHealth < 100.0):
#     #                 if (InstigatorHealth + CustomSiphon > 100.0):
#     #                     InstigatorFort.SetShield(InstigatorHealth + CustomSiphon - 100.0)
#     #                     InstigatorFort.SetHealth(100.0)
#     #                 else:
#     #                     InstigatorFort.SetHealth(InstigatorHealth + CustomSiphon)
#     #             else:
#     #                 InstigatorFort.SetShield(InstigatorShield+CustomSiphon)
#     #             #Vfx_hit_device.Pickup(Agent)

#     #             #Print("HEALTH: NewInstigatorHealth={InstigatorFort.GetHealth()}, NewInstigatorShield={InstigatorFort.GetShield()}")
#     #             #Print("HEALTH: END OF HEALTH PLAYER HIT()")
#     #         else:
#     #             Print("Damage not from an agent bro")

