using { /Fortnite.com/Devices }
using { /Fortnite.com/AI}
using { /Fortnite.com/Teams}
using { /Verse.org/Simulation }
using { /Verse.org/Random }
using { /Fortnite.com/Characters }
using { /Fortnite.com/FortPlayerUtilities }
using { /Fortnite.com/Game }
using { /Fortnite.com/Teams }
using { /Verse.org/Simulation/Tags }
using { /UnrealEngine.com/Temporary/Diagnostics }
using { /UnrealEngine.com/Temporary/SpatialMath } 
using { /Verse.org/Assets }

# A Verse-authored creative device that can be placed in a level
catchGame := class(minigame_device):

    
    @editable ItemGranterCatch :item_granter_device = item_granter_device{}
    @editable ItemGranterCatchWeapons :item_granter_device = item_granter_device{}

    Name<override>: string = "Catch"
    StartingMessage<override>:string = "Random Throwables Every Second!"
    GetVotingImage<override>()<transacts>: texture = VerseExposed.SQ_CATCH

    # _Start<override>()<suspends>: void = 
    #     loop:
    #         for(Player: GetRemainingPlayers()):
    #             ItemGranterCatch.CycleToRandomItem(Player)
    #             ItemGranterCatchWeapons.CycleToRandomItem(Player)
    #         Sleep(5.0)
    
        


            