using { /Fortnite.com/Devices }
using { /Fortnite.com/AI}
using { /Fortnite.com/Teams}
using { /Verse.org/Simulation }
using { /Verse.org/Random }
using { /Fortnite.com/Characters }
using { /Fortnite.com/FortPlayerUtilities }
using { /Fortnite.com/Game }
using { /Fortnite.com/Teams }
using { /Verse.org/Simulation/Tags }
using { /UnrealEngine.com/Temporary/Diagnostics }
using { /UnrealEngine.com/Temporary/SpatialMath } 
using { /Verse.org/Assets }

# A Verse-authored creative device that can be placed in a level
editOtherPlayerBuilds := class(minigame_device):

    Name<override>: string = "Edit Course"
    StartingMessage<override>:string = "Edit Anyone's Builds!"
    GetVotingImage<override>()<transacts>: texture = VerseExposed.SQ_EDITCOURSE
    DisplayIcon<override>: ?texture = option{VerseExposed.SQ_EDITCOURSE_ICON}

    _Start<override>()<suspends>: void = 
        TeamCollection := GetPlayspace().GetTeamCollection()
        if(Team78 := TeamCollection.GetTeams()[77]):
            for(Player : GetRemainingPlayers()):
                option{TeamCollection.AddToTeam[Player, Team78]}