using { /Fortnite.com/Devices }
using { /Verse.org/Simulation }
using { /Verse.org/Random }
using { /Fortnite.com/Characters }
using { /Fortnite.com/Game }
using { /Verse.org/Assets }

pickaxe := class(minigame_device):

    @editable MutatorZoneEmote: mutator_zone_device = mutator_zone_device{}

    StartingMessage<override>:string = "Pickaxe Only and Emote to Heal!"
    Name<override>: string = "Pickaxe Frenzy"
    GetVotingImage<override>()<transacts>: texture = VerseExposed.SQ_PICKAXE
    DisplayIcon<override>: ?texture = option{VerseExposed.SQ_PICKAXE_DISPLAYICON}

    _Start<override>()<suspends>: void =
        MutatorZoneEmote.Enable()

    _Reset<override>()<suspends>:void=
        MutatorZoneEmote.Disable()
            
    