using { /Fortnite.com/Devices }
using { /Fortnite.com/AI}
using { /Fortnite.com/Teams}
using { /Verse.org/Simulation }
using { /Verse.org/Random }
using { /Fortnite.com/Characters }
using { /Fortnite.com/FortPlayerUtilities }
using { /Fortnite.com/Game }
using { /Fortnite.com/Teams }
using { /Verse.org/Simulation/Tags }
using { /UnrealEngine.com/Temporary/Diagnostics }
using { /UnrealEngine.com/Temporary/SpatialMath } 
using { /Verse.org/Assets }

oneInTheChamber := class(minigame_device):

    Name<override>: string = "One In The Chamber"
    StartingMessage<override>:string = "Each Elim Gives 1 Bullet and 10 Wood!"
    GetVotingImage<override>()<transacts>: texture = VerseExposed.SQ_ONEINTHECHAMBER
    DisplayIcon<override>: ?texture = option{VerseExposed.SQ_ONEINCHAMBER_ICON}