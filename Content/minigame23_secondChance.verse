# using { /Fortnite.com/Devices }
# using { /Fortnite.com/Teams}
# using { /Verse.org/Simulation }
# using { /Verse.org/Random }
# using { /Fortnite.com/Characters }
# using { /Fortnite.com/FortPlayerUtilities }
# using { /Fortnite.com/Game }
# using { /Fortnite.com/Teams }
# using { /Verse.org/Simulation/Tags }
# using { /UnrealEngine.com/Temporary/Diagnostics }
# using { /UnrealEngine.com/Temporary/SpatialMath } 
# using { /Verse.org/Assets }

# # See https://dev.epicgames.com/documentation/en-us/uefn/create-your-own-device-in-verse for how to create a verse device.

# # A Verse-authored creative device that can be placed in a level
# secondChance := class(minigame_device):

#     # @editable ItemGranterDefault :item_granter_device = item_granter_device{}
#     # @editable DownButNotOutDevice :down_but_not_out_device = down_but_not_out_device{}
#     # var PlayerDownedLogic : [player]logic = map{}

#     # Name<override>: string = "One In The Chamber"
#     # StartingMessage<override>:string = "1 Extra Life!"

#     # _Start<override>()<suspends>: void = 
        
        
#     #     AllPlayers := GetRemainingPlayers()
        
#     #     DownButNotOutDevice.AgentDownedEvent.Subscribe(OnDowned)
#     #     TeamCollection := GetPlayspace().GetTeamCollection()
#     #     if(Player1 := AllPlayers[0]):
#     #         ItemGranterDefault.GrantItem(Player1)
#     #     for(Player : AllPlayers):
#     #         if:
#     #             Agent := agent[Player]
#     #             Fort := Agent.GetFortCharacter[]
#     #             Team99 := TeamCollection.GetTeams()[98]
#     #             TeamCollection.AddToTeam[Player, Team99]
#     #         then:
#     #             #ItemGranterDefault.GrantItem(Player)
#     #             if(set PlayerDownedLogic[Player] = false) {}
            
#     # OnDowned(Player : agent):void=
#     #     spawn{HandleDownedPlayer(Player)}

#     # HandleDownedPlayer(Agent : agent)<suspends>:void=
#     #     if:
#     #         Player := player[Agent]
#     #     then:
#     #         DownedLogic := GetDownedLogic(Player)
#     #         if(DownedLogic = false):
#     #             Sleep(3.0)
#     #             DownButNotOutDevice.Revive(Agent)
#     #             if(set PlayerDownedLogic[Player] = true) {}
#     #         else:
#     #             Print("PLayer's second death. Let him bleed out")

#     # GetDownedLogic(PlayerDowned : player):logic=
#     #     for(Player -> Logic: PlayerDownedLogic):
#     #         if(PlayerDowned = Player):
#     #             return Logic
#     #     return true
