using { /Fortnite.com/Devices }
using { /Fortnite.com/AI}
using { /Fortnite.com/Teams}
using { /Verse.org/Simulation }
using { /Verse.org/Random }
using { /Fortnite.com/Characters }
using { /Fortnite.com/FortPlayerUtilities }
using { /Fortnite.com/Game }
using { /Fortnite.com/Teams }
using { /Verse.org/Simulation/Tags }
using { /UnrealEngine.com/Temporary/Diagnostics }
using { /UnrealEngine.com/Temporary/SpatialMath } 
using { /Verse.org/Assets }

# See https://dev.epicgames.com/documentation/en-us/uefn/create-your-own-device-in-verse for how to create a verse device.

# A Verse-authored creative device that can be placed in a level
spleef := class(minigame_device):

    @editable HudController : hud_controller_device = hud_controller_device{}
    @editable GlassFloorProp : creative_prop = creative_prop{}
    @editable LavaFloorProp : creative_prop = creative_prop{}

    Name<override>:string = "Spleef"
    BottomMessage<override>: ?string = option{"Tip: You can slide kick players!"}
    StartingMessage<override>:string = "KEEP THE HIGHGROUND!"
    GetVotingImage<override>()<transacts>: texture = VerseExposed.SQ_SPLEEF

    _Start<override>()<suspends>: void = 
        HudController.Enable()
        if(GlassFloorProp.TeleportTo[vector3{X:= 7167.45, Y:= -19470.40, Z:= 3603.069173}, rotation{}]):
        if(LavaFloorProp.TeleportTo[vector3{X:=5042.233594,Y:=-17636.016274,Z:=1447.421718}, rotation{}]):

    _Reset<override>()<suspends>:void=
        HudController.Disable()
        if(GlassFloorProp.TeleportTo[vector3{X:= 7167.45, Y:= -19470.40, Z:= -2230.02}, rotation{}]):
        if(LavaFloorProp.TeleportTo[vector3{X:= 7167.45, Y:= -19470.40, Z:= 4083.11}, rotation{}]):