using { /Fortnite.com/Devices }
using { /Fortnite.com/AI}
using { /Fortnite.com/Teams}
using { /Verse.org/Simulation }
using { /Verse.org/Random }
using { /Fortnite.com/Characters }
using { /Fortnite.com/FortPlayerUtilities }
using { /Fortnite.com/Game }
using { /Fortnite.com/Teams }
using { /Verse.org/Simulation/Tags }
using { /UnrealEngine.com/Temporary/Diagnostics }
using { /UnrealEngine.com/Temporary/SpatialMath } 
using { /Verse.org/Assets }

# A Verse-authored creative device that can be placed in a level
double_pump := class(minigame_device):
    @editable LockShootTime : float = 0.60
    @editable PumpConditionalButton :conditional_button_device = conditional_button_device{}
    @editable InputTriggerShoot : input_trigger_device = input_trigger_device{}
    @editable PumpGranter :item_granter_device = item_granter_device{}
    var PlayerShootLock: [player]logic = map{}

    Name<override>: string = "Double Pump"
    StartingMessage<override>:string = "Pumps Constantly Shoot!"
    GetVotingImage<override>()<transacts>: texture = VerseExposed.SQ_DOUBLEPUMP
    DisplayIcon<override>: ?texture = option{VerseExposed.SQ_DOUBLEPUMP_ICON}
    
    #BottomMessage<override>:?string = option{"Tip: Drop and pickup pump if double pump isn't working!"}

    _Start<override>()<suspends>: void = 
        Print("Lock Time = {LockShootTime}")
        InputTriggerShoot.Enable()
        RegisterSubscriber(InputTriggerShoot.PressedEvent.Subscribe(HandleShoot))
        for(Player : GetRemainingPlayers(), Fort := Player.GetFortCharacter[]):
            if(set PlayerShootLock[Player] = false):

    HandleShoot(Player : agent):void=
        #Print("OnShoot")
        if(PumpConditionalButton.IsHoldingItem[Player]):
            if(PlayerShootLock[Player] = false):
                #Make sure its unlocked
                PumpGranter.GrantItem(Player)
                spawn{LockPlayerShoot(Player)}
            else:
                #Print("Shooting is locked for player")
        else:
            #Print("Not holding shotgun)")

    LockPlayerShoot(Agent : agent)<suspends>:void =
        if(Player := player[Agent]):
            #Print("Locking Shoot Button")
            if(set PlayerShootLock[Player] = true):
            Sleep(LockShootTime)
            #Print("Unocking Shoot Button")
            if(set PlayerShootLock[Player] = false):

    _OnMinigameEnded<override>(Winners: []player)<suspends>:void=
        InputTriggerShoot.UnregisterAll()
        InputTriggerShoot.Disable()
        
    _Reset<override>()<suspends>:void=
        set PlayerShootLock = map{}