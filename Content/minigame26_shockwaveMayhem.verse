using { /Fortnite.com/Devices }
using { /Fortnite.com/AI}
using { /Fortnite.com/Teams}
using { /Verse.org/Simulation }
using { /Verse.org/Random }
using { /Fortnite.com/Characters }
using { /Fortnite.com/FortPlayerUtilities }
using { /Fortnite.com/Game }
using { /Fortnite.com/Teams }
using { /Verse.org/Simulation/Tags }
using { /UnrealEngine.com/Temporary/Diagnostics }
using { /UnrealEngine.com/Temporary/SpatialMath } 
using { /Verse.org/Assets }

# See https://dev.epicgames.com/documentation/en-us/uefn/create-your-own-device-in-verse for how to create a verse device.

# A Verse-authored creative device that can be placed in a level
shockwave_mayhem := class(minigame_device):

    @editable ItemGranterShock :item_granter_device = item_granter_device{}

    Name<override>: string = "Shockwave Mayhem"
    StartingMessage<override>:string = ""
    GetVotingImage<override>()<transacts>: texture = VerseExposed.SQ_SHOCKWAVEMAYHEM

    # _Start<override>()<suspends>: void = 
    #     loop:
    #         for(Player: GetRemainingPlayers()):
    #             ItemGranterShock.GrantItem(Player)
    #         Sleep(4.0)
            