# using { /Fortnite.com/Devices }
# using { /Fortnite.com/AI}
# using { /Fortnite.com/Teams}
# using { /Verse.org/Simulation }
# using { /Verse.org/Random }
# using { /Fortnite.com/Characters }
# using { /Fortnite.com/FortPlayerUtilities }
# using { /Fortnite.com/Game }
# using { /Fortnite.com/Teams }
# using { /Verse.org/Simulation/Tags }
# using { /UnrealEngine.com/Temporary/Diagnostics }
# using { /UnrealEngine.com/Temporary/SpatialMath } 
# using { /Verse.org/Assets }

# # A Verse-authored creative device that can be placed in a level
# stormSurge := class(minigame_device):

#     # @editable ItemGranterStormSurge :item_granter_device = item_granter_device{}

#     # Name<override>: string = "Storm Surge"
#     # StartingMessage<override>:string = "Knock Players Into The Storm!\n Tip: Slide Kick!"
#     # GetVotingImage<override>()<transacts>: texture = VerseExposed.SQ_STORMSURGE

#     # _Start<override>()<suspends>: void = 
        
#     #     # AllPlayers := GetRemainingPlayers()
#     #     # for(Player : AllPlayers):
#     #     #     if:
#     #     #         Fort := Player.GetFortCharacter[]
#     #     #     then:
#     #     #         Fort.SetMaxShield(0.0)
#     #     #         Fort.SetMaxHealth(300.0)
#     #     loop:
#     #         for(Player: GetRemainingPlayers()):
#     #             ItemGranterStormSurge.CycleToRandomItem(Player)
#     #         Sleep(3.0)



