using { /Fortnite.com/Devices }
using { /Fortnite.com/AI}
using { /Fortnite.com/Teams}
using { /Verse.org/Simulation }
using { /Verse.org/Random }
using { /Fortnite.com/Characters }
using { /Fortnite.com/FortPlayerUtilities }
using { /Fortnite.com/Game }
using { /Fortnite.com/Teams }
using { /Verse.org/Simulation/Tags }
using { /UnrealEngine.com/Temporary/Diagnostics }
using { /Verse.org/Assets }
using { /UnrealEngine.com/Temporary/SpatialMath } 

# A Verse-authored creative device that can be placed in a level
invis := class(minigame_device):
    @editable InvisConditionalButton :conditional_button_device = conditional_button_device{}
    @editable DBNO :down_but_not_out_device = down_but_not_out_device{}
    #@editable ItemRemover : item_remover_device = item_remover_device{}

    Name<override>: string = "Invisible"
    StartingMessage<override>:string = "Hold Remote to Turn Invisible!"
    GetVotingImage<override>()<transacts>: texture = VerseExposed.SQ_INVISIBLE
    DisplayIcon<override>: ?texture = option{VerseExposed.SQ_INVIS_DISPLAYICON}

    _Start<override>()<suspends>: void =
        DBNO.Enable()
        AllPlayers := GetRemainingPlayers()
        RegisterSubscriber(DBNO.AgentDownedEvent.Subscribe(OnPlayerDowned))

        for(Player : AllPlayers):
            if:
                Fort := Player.GetFortCharacter[]
            then:
                spawn{InitInvisConditional(Player)}

    _OnMinigameEnded<override>(Winners: []player)<suspends>:void=
        for(Player: GetRemainingPlayers()):
            Manager.ResetItemRemover.Remove(Player)

    _Reset<override>()<suspends>:void=
        DBNO.Disable()

    InitInvisConditional(Player : agent)<suspends>:void =
        Print("Start Invis Loop")
        if:
            Agent := agent[Player]
            Fort := Agent.GetFortCharacter[]
            Fort.IsActive[]
        then:
            race:
                EndedEvent.Await()
                loop:
                    if(InvisConditionalButton.IsHoldingItem[Agent]):
                        #Print("Has Invis")
                        Fort.Hide()
                    else:
                        #Print("Doesn't Have Key Card")
                        Fort.Show()
                    Sleep(0.2)
            Fort.Show()

    OnPlayerDowned(PlayerDowned : agent):void=
        Print("Player downed, removing invis item")
        Manager.ResetItemRemover.Remove(PlayerDowned)

    _OnPlayerEliminated<override>(Result : elimination_result):void=
        if:
            PersonKilled := Result.EliminatedCharacter
            PersonKilledAgent := PersonKilled.GetAgent[]
            PersonKilledFort := PersonKilledAgent.GetFortCharacter[]
        then:
            Print("hopefully removing")
            PersonKilledFort.Show()
            Manager.ResetItemRemover.Remove(PersonKilledAgent)