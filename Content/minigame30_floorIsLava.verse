using { /Fortnite.com/Devices }
using { /Fortnite.com/AI}
using { /Fortnite.com/Teams}
using { /Verse.org/Simulation }
using { /Verse.org/Random }
using { /Fortnite.com/Characters }
using { /Fortnite.com/FortPlayerUtilities }
using { /Fortnite.com/Game }
using { /Fortnite.com/Teams }
using { /Verse.org/Simulation/Tags }
using { /UnrealEngine.com/Temporary/Diagnostics }
using { /UnrealEngine.com/Temporary/SpatialMath } 
using { /Verse.org/Assets }

# See https://dev.epicgames.com/documentation/en-us/uefn/create-your-own-device-in-verse for how to create a verse device.

# A Verse-authored creative device that can be placed in a level
floorIsLava := class(minigame_device):

    @editable ItemGranterLavaIndividual : item_granter_device = item_granter_device{}
    @editable TimerNewShotgun: timer_device = timer_device{}

    Name<override>: string = "Lava Bounce"
    StartingMessage<override>:string = "Shotguns switch every 10s!"
    GetVotingImage<override>()<transacts>: texture = VerseExposed.SQ_LAVABOUNCE

    _Start<override>()<suspends>: void=
        RegisterSubscriber(TimerNewShotgun.SuccessEvent.Subscribe(HandleTimerShotgunSuccess))

    _Reset<override>()<suspends>: void=
        TimerNewShotgun.Reset()
        TimerNewShotgun.Disable()

    OnSafeTimeEnded<override>()<suspends>:void=
        TimerNewShotgun.Enable()
        TimerNewShotgun.Reset()
        TimerNewShotgun.Start()

    HandleTimerShotgunSuccess(PlayerMaybe : ?agent):void=
        if(Player1 := GetRemainingPlayers()[0]):
            ItemGranterLavaIndividual.CycleToRandomItem(Player1)
        spawn{StartTimerAgain()}

    StartTimerAgain()<suspends>:void=
        race:
            EndedEvent.Await()
            block:
                Sleep(0.2)
                TimerNewShotgun.Start()

