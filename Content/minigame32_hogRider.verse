using { /Fortnite.com/Devices }
using { /Fortnite.com/AI}
using { /Fortnite.com/Teams}
using { /Verse.org/Simulation }
using { /Verse.org/Random }
using { /Fortnite.com/Characters }
using { /Fortnite.com/FortPlayerUtilities }
using { /Fortnite.com/Game }
using { /Fortnite.com/Teams }
using { /Verse.org/Simulation/Tags }
using { /UnrealEngine.com/Temporary/Diagnostics }
using { /UnrealEngine.com/Temporary/SpatialMath }
using { /Verse.org/Assets }

hogs := class(tag) {}
hog_rider := class(minigame_device):

    @editable
    HogSpawners : []wildlife_spawner_device = array{}

    @editable HogRiderSound :audio_player_device = audio_player_device{}
    
    Name<override>: string = "Hog Rider"
    StartingMessage<override>:string = "HOG RIDAAAA"
    GetVotingImage<override>()<transacts>: texture = VerseExposed.SQ_HOGRIDER
    BottomMessage<override>:?string = option{"Hammer Tip: Hold aim button to target enemies"}

    # Fails if player is dead or spawner array is empty
    FindClosestHogSpawner(Player: player)<transacts><decides>:wildlife_spawner_device=
        var RetSpawner : ?wildlife_spawner_device = false

        if(Character := Player.GetFortCharacter[], Character.IsActive[]):
            var MinDistance : float = 99999.9
            for(
                Spawner : HogSpawners
                DistanceToSpawner := Distance(Spawner.GetTransform().Translation, Character.GetTransform().Translation)
                DistanceToSpawner < MinDistance
            ):
                set MinDistance = DistanceToSpawner
                set RetSpawner = option{Spawner}

        RetSpawner?

    _Reset<override>()<suspends>:void=
        if(RemainingPlayer:=GetRemainingPlayers()[0]):
            for(Spawner : HogSpawners):
                Spawner.Despawn(RemainingPlayer)
                Spawner.DismountAll()
                Spawner.Disable()

    _Start<override>()<suspends>: void = 
        for(Player : GetRemainingPlayers()):
            spawn{HandleWildlifeSpawners(Player)}

    HandleWildlifeSpawners(Player : player)<suspends>:void=
        race:
            EndedEvent.Await()
            block:
                Sleep(1.0)
                if:
                    Agent := agent[Player]
                    Fort := Agent.GetFortCharacter[]
                    WildlifeSpawner := FindClosestHogSpawner[Player]
                then:
                    Print("NEAR HOG FOUND")
                    WildlifeSpawner.Enable()
                    WildlifeSpawner.Spawn()
                    Sleep(1.5)
                    Print("RIDING")
                    WildlifeSpawner.Ride(Player)
                    HogRiderSound.Play(Player)


    _OnPlayerEliminated<override>(Result : elimination_result):void=
        if:
            PersonKilled := Result.EliminatedCharacter
            PersonKilledAgent := PersonKilled.GetAgent[]
            PersonKilledPlayer := player[PersonKilledAgent]

            Killer := Result.EliminatingCharacter?
            KillerAgent := Killer.GetAgent[]
            KillerFort := KillerAgent.GetFortCharacter[]
            KillerPlayer := player[KillerAgent]
            IsInGame[PersonKilledPlayer]
        then:
            if(not PersonKilledAgent = KillerAgent):
                HogRiderSound.Play(PersonKilledAgent)
                HogRiderSound.Play(KillerAgent)