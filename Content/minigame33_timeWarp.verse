using { /Fortnite.com/Devices }
using { /Fortnite.com/AI}
using { /Fortnite.com/Teams}
using { /Verse.org/Simulation }
using { /Verse.org/Random }
using { /Fortnite.com/Characters }
using { /Fortnite.com/FortPlayerUtilities }
using { /Fortnite.com/Game }
using { /Fortnite.com/Teams }
using { /Verse.org/Simulation/Tags }
using { /UnrealEngine.com/Temporary/Diagnostics }
using { /UnrealEngine.com/Temporary/SpatialMath } 
using { /Verse.org/Assets }

# See https://dev.epicgames.com/documentation/en-us/uefn/create-your-own-device-in-verse for how to create a verse device.

# A Verse-authored creative device that can be placed in a level
time_warp := class(minigame_device):

    @editable SlowmoCinematicSequenceDevice :cinematic_sequence_device = cinematic_sequence_device{}
    @editable SuperSlowmoCinematicSequenceDevice :cinematic_sequence_device = cinematic_sequence_device{}
    @editable FastmoCinematicSequenceDevice :cinematic_sequence_device = cinematic_sequence_device{}
    @editable SuperFastmoCinematicSequenceDevice :cinematic_sequence_device = cinematic_sequence_device{}

    Name<override>: string = "Time Warp"
    StartingMessage<override>:string = "Speed will constantly change!"
    GetVotingImage<override>()<transacts>: texture = VerseExposed.SQUARE_TIMEWARP

    _Start<override>()<suspends>: void = 
        FastmoCinematicSequenceDevice.Play()

        loop:
            StopSequences()
            RandomInt := GetRandomInt(1,5)
            #Normal
            if(RandomInt = 1):
                Sleep(5.0)
            else if(RandomInt = 2):
                SlowmoCinematicSequenceDevice.Play()
                Sleep(4.0)
            else if(RandomInt = 3):
                FastmoCinematicSequenceDevice.Play()
                Sleep(11.0)
            else if(RandomInt = 4):
                SuperFastmoCinematicSequenceDevice.Play()
                Sleep(20.0)
            else if(RandomInt = 5):
                SuperSlowmoCinematicSequenceDevice.Play()
                Sleep(1.5)

    # Ran right before minigame ends
    _OnMinigameEnded<override>(Winners: []player)<suspends>:void=
        StopSequences()

    StopSequences():void=
        SuperSlowmoCinematicSequenceDevice.Stop()
        SlowmoCinematicSequenceDevice.Stop()
        FastmoCinematicSequenceDevice.Stop()
        SuperFastmoCinematicSequenceDevice.Stop()
