using { /Fortnite.com/Devices }
using { /Fortnite.com/AI}
using { /Fortnite.com/Teams}
using { /Verse.org/Simulation }
using { /Verse.org/Random }
using { /Fortnite.com/Characters }
using { /Fortnite.com/FortPlayerUtilities }
using { /Fortnite.com/Game }
using { /Fortnite.com/Teams }
using { /Verse.org/Simulation/Tags }
using { /UnrealEngine.com/Temporary/Diagnostics }
using { /UnrealEngine.com/Temporary/SpatialMath } 
using { /Verse.org/Assets }

# A Verse-authored creative device that can be placed in a level
teleporter := class(minigame_device):

    @editable Teleporter :teleporter_device = teleporter_device{}
    @editable TeleportTimer : timer_device = timer_device{}
    @editable MinimumHeight : float = 0.0

    Name<override>: string = "Teleport Tangle"
    StartingMessage<override>:string = "Teleport to a Random Player Every 10s!"
    GetVotingImage<override>()<transacts>: texture = VerseExposed.SQ_TELEPORT

    _Start<override>()<suspends>: void = 
        RegisterSubscriber(TeleportTimer.SuccessEvent.Subscribe(HandleTeleportTimer))
        TeleportTimer.Enable()
        TeleportTimer.Reset()
        TeleportTimer.Start()

    _Reset<override>()<suspends>: void =
        TeleportTimer.Disable()

    HandleTeleportTimer(Player : ?agent):void=
        spawn{TeleportLoop()}

    TeleportLoop()<suspends>:void=
        # Add Shuffle to randomize
        ActivePlayersArray := for(Player : Shuffle(GetRemainingPlayers()), FortCharacter := Player.GetFortCharacter[], FortCharacter.IsActive[]) { (Player, FortCharacter) }
        # Print("ActivePlayersArray Length = {ActivePlayersArray.Length}")
        # Store First Player Location
        if (FirstPlayerTuple := ActivePlayersArray[0]):
            FirstPlayerFort := GetTupleFort(FirstPlayerTuple)
            FirstPlayerLocation := GetLimitedPlayerLocation(FirstPlayerFort)
            FirstPlayerRotation := FirstPlayerFort.GetTransform().Rotation
            for(index := 0 .. ActivePlayersArray.Length-1):
                # Print("Handle ActivePlayersArrayIndex {index}")
                if(PlayerTeleportingTuple := ActivePlayersArray[index]){
                    PlayerTeleporting := GetTuplePlayer(PlayerTeleportingTuple)
                    if(index = ActivePlayersArray.Length-1):
                        # Print("First PLayer Location:{FirstPlayerLocation}")

                        if(Teleporter.TeleportTo[FirstPlayerLocation, FirstPlayerRotation]):
                            Teleporter.Teleport(PlayerTeleporting)
                    else:
                        if (NextPlayerTuple := ActivePlayersArray[index + 1]):
                            NextPlayerFort := GetTupleFort(NextPlayerTuple)
                            NextPlayerLocation := GetLimitedPlayerLocation(NextPlayerFort)
                            # Print("Next Player Location: {NextPlayerLocation}")
                            NextPlayerRotation := NextPlayerFort.GetTransform().Rotation
                            if(Teleporter.TeleportTo[NextPlayerLocation, NextPlayerRotation]):
                                Teleporter.Teleport(PlayerTeleporting)
                }

    GetLimitedPlayerLocation(Character: fort_character):vector3=
        TargetTranslation := Character.GetTransform().Translation + vector3{Z:= -30.0}
        vector3:
            X := TargetTranslation.X
            Y := TargetTranslation.Y
            Z := Max(TargetTranslation.Z, MinimumHeight)

    GetTuplePlayer(Player : player, Fort : fort_character) : player =
        return Player

    GetTupleFort(Player : player, Fort : fort_character) : fort_character =
        return Fort