using { /Fortnite.com/Devices }
using { /Fortnite.com/AI}
using { /Fortnite.com/Teams}
using { /Verse.org/Simulation }
using { /Verse.org/Random }
using { /Fortnite.com/Characters }
using { /Fortnite.com/FortPlayerUtilities }
using { /Fortnite.com/Game }
using { /Fortnite.com/Teams }
using { /Verse.org/Simulation/Tags }
using { /UnrealEngine.com/Temporary/Diagnostics }
using { /UnrealEngine.com/Temporary/SpatialMath } 
using { /Verse.org/Assets }

# A Verse-authored creative device that can be placed in a level
crosshair_chaos := class(minigame_device):
    @editable HudNoReticle : hud_controller_device = hud_controller_device{}
    @editable Crosshairs : []creative_prop = array{}
    @editable BoxX : float = 500.8
    @editable BoxY : float = 517.1
    @editable BoxZHeight : float = 720.0
    @editable ResetZHeight : float = 4624.8
    var ShuffledCrosshairs : []creative_prop = array{}

    Name<override>: string = "Crosshair Chaos"
    StartingMessage<override>:string = "Random Crosshair Every 10s"
    GetVotingImage<override>()<transacts>: texture = VerseExposed.SQ_CROSSHAIR

    _Start<override>()<suspends>: void = 
        HudNoReticle.Enable()
        set ShuffledCrosshairs = Shuffle(Crosshairs)

        loop:
            for(Index := 0..ShuffledCrosshairs.Length-1):
                if(ShuffledCrosshairs[Index].TeleportTo[vector3{X:= BoxX, Y:= BoxY, Z:= BoxZHeight}, rotation{}]):
                    Print("Teleported Crosshair Box. Index {Index}")
                Sleep(10.0)
                if(ShuffledCrosshairs[Index].TeleportTo[vector3{X:= BoxX, Y:= BoxY, Z:= ResetZHeight}, rotation{}]):
                    Print("Reset Crosshair Box. Index {Index}")
                Sleep(0.65)

    _Reset<override>()<suspends>:void=
        HudNoReticle.Disable()
        for(CrosshairPP : Crosshairs):
            if(CrosshairPP.TeleportTo[vector3{X:= BoxX, Y:= BoxY, Z:= ResetZHeight}, rotation{}]):