# using { /Fortnite.com/Devices }
# using { /Fortnite.com/AI}
# using { /Fortnite.com/Teams}
# using { /Verse.org/Simulation }
# using { /Verse.org/Random }
# using { /Fortnite.com/Characters }
# using { /Fortnite.com/FortPlayerUtilities }
# using { /Fortnite.com/Game }
# using { /Fortnite.com/Teams }
# using { /Verse.org/Simulation/Tags }
# using { /UnrealEngine.com/Temporary/Diagnostics }
# using { /UnrealEngine.com/Temporary/SpatialMath } 
# using { /Verse.org/Assets }

# # See https://dev.epicgames.com/documentation/en-us/uefn/create-your-own-device-in-verse for how to create a verse device.

# # A Verse-authored creative device that can be placed in a level
# tire := class(minigame_device):
#     @editable ItemGranterTire :item_granter_device = item_granter_device{}
#     @editable ItemSpawner : item_spawner_device = item_spawner_device{}

#     Name<override>: string = "Tire Toss"
#     StartingMessage<override>:string = "New Tire Every 30 Seconds!"
#     GetVotingImage<override>()<transacts>: texture = VerseExposed.SQ_TIRETOSS

#     OnSafeTimeEnded<override>()<suspends>:void=
#         loop:
#             Sleep(30.0)
#             ItemSpawner.SpawnItem()
#             ItemSpawner.CycleToNextItem()
#             for(Player : GetRemainingPlayers()):
#                 ItemGranterTire.GrantItem(Player)
            