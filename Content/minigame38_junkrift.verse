using { /Fortnite.com/Devices }
using { /Fortnite.com/AI}
using { /Fortnite.com/Teams}
using { /Verse.org/Simulation }
using { /Verse.org/Random }
using { /Fortnite.com/Characters }
using { /Fortnite.com/FortPlayerUtilities }
using { /Fortnite.com/Game }
using { /Fortnite.com/Teams }
using { /Verse.org/Simulation/Tags }
using { /UnrealEngine.com/Temporary/Diagnostics }
using { /UnrealEngine.com/Temporary/SpatialMath } 
using { /Verse.org/Assets }

# See https://dev.epicgames.com/documentation/en-us/uefn/create-your-own-device-in-verse for how to create a verse device.

# A Verse-authored creative device that can be placed in a level
junk_rift := class(minigame_device):

    @editable ItemGranterJunkRift :item_granter_device = item_granter_device{}
    @editable JunkBarrierZone : barrier_device = barrier_device{}

    Name<override>: string = "Raining Junk"
    StartingMessage<override>:string = "New Junk Rift Every 5 Seconds!"
    GetVotingImage<override>()<transacts>: texture = VerseExposed.SQ_RAININGJUNK

    _Start<override>()<suspends>: void = 
        JunkBarrierZone.Enable()
        # loop:
        #     Sleep(5.0)
        #     AllPlayers := GetRemainingPlayers()
        #     <#if(Player1 := AllPlayers[0]):
        #         ItemGranterJunkRift.GrantItem(Player1)#>
        #     for(Player:AllPlayers):
        #         ItemGranterJunkRift.GrantItem(Player)

    _Reset<override>()<suspends>:void=
        JunkBarrierZone.Disable()