using { /Fortnite.com/Devices }
using { /Verse.org/Simulation }
using { /UnrealEngine.com/Temporary/Diagnostics }
using { /Verse.org/Assets }
using { /Fortnite.com/Game }
using { /Verse.org/Random }

# A Verse-authored creative device that can be placed in a level
slowmo := class(minigame_device):

    @editable SlowmoCinematicSequenceDevice :cinematic_sequence_device = cinematic_sequence_device{}

    Name<override>: string = "Slow-Mo"
    StartingMessage<override>:string = ""
    GetVotingImage<override>()<transacts>: texture = VerseExposed.SQ_SLOWMO

    _Start<override>()<suspends>: void =
        SlowmoCinematicSequenceDevice.Play()

    # Ran right before minigame ends
    _OnMinigameEnded<override>(Winners: []player)<suspends>:void=
        SlowmoCinematicSequenceDevice.Stop()