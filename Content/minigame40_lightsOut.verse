using { /Fortnite.com/Devices }
using { /Fortnite.com/AI}
using { /Fortnite.com/Teams}
using { /Verse.org/Simulation }
using { /Verse.org/Random }
using { /Fortnite.com/Characters }
using { /Fortnite.com/FortPlayerUtilities }
using { /Fortnite.com/Game }
using { /Fortnite.com/Teams }
using { /Verse.org/Simulation/Tags }
using { /UnrealEngine.com/Temporary/Diagnostics }
using { /UnrealEngine.com/Temporary/SpatialMath } 
using { /Verse.org/Assets }

# See https://dev.epicgames.com/documentation/en-us/uefn/create-your-own-device-in-verse for how to create a verse device.

# A Verse-authored creative device that can be placed in a level
lights_out := class(minigame_device):
    @editable DarknessProp : creative_prop = creative_prop{}
    @editable AudioLightsOn : audio_player_device = audio_player_device{}
    @editable AudioLightsOff : audio_player_device = audio_player_device{}

    Name<override>: string = "Lights Out"
    StartingMessage<override>:string = "Lights Flicker On Every 7s"
    GetVotingImage<override>()<transacts>: texture = VerseExposed.SQ_LIGHTSOUT

    _Start<override>()<suspends>: void =
        spawn{TurnLightsOff()}
        
    OnSafeTimeEnded<override>()<suspends>:void=
        spawn{LightsOutSequence()}

    _OnMinigameEnded<override>(Winners: []player)<suspends>:void=
        AudioLightsOff.Disable()
        AudioLightsOn.Disable()
        if(DarknessProp.TeleportTo[vector3{X:= 512.0, Y:= 576.0, Z:= -6500.0}, rotation{}]):

    TurnLightsOff()<suspends>:void=
        Sleep(2.5)
        AudioLightsOff.Play()
        if(DarknessProp.TeleportTo[vector3{X:= 512.0, Y:= 576.0, Z:= 566.0}, rotation{}]):
    
    LightsOutSequence()<suspends>:void=
        race:
            EndedEvent.Await()
            block:
                loop:
                    Print("Waiting")
                    Sleep(7.0)
                    Print("LIGHT")
                    if(DarknessProp.TeleportTo[vector3{X:= 512.0, Y:= 576.0, Z:= -6500.0}, rotation{}]):
                    AudioLightsOn.Play()
                    Sleep(1.7)
                    AudioLightsOff.Play()
                    if(DarknessProp.TeleportTo[vector3{X:= 512.0, Y:= 576.0, Z:= 566.0}, rotation{}]):
