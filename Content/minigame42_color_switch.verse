using { /Fortnite.com/Devices }
using { /Fortnite.com/AI}
using { /Fortnite.com/Teams}
using { /Verse.org/Simulation }
using { /Verse.org/Random }
using { /Fortnite.com/Characters }
using { /Fortnite.com/FortPlayerUtilities }
using { /Fortnite.com/Game }
using { /Fortnite.com/Teams }
using { /Verse.org/Simulation/Tags }
using { /UnrealEngine.com/Temporary/Diagnostics }
using { /UnrealEngine.com/Temporary/SpatialMath } 
using { /Verse.org/Assets }
using { VerseExposed }

color_switch_color := enum:
    Red
    Orange
    Yellow
    Green
    Cyan
    Blue
    Purple
    Pink

color_switch_hud_message := class<concrete>:
    @editable
    Color : color_switch_color = color_switch_color.Red

    @editable
    HUDMessage : hud_message_device = hud_message_device{}

color_switch := class(minigame_device):
    @editable
    Timer: timer_device = timer_device{}
    @editable
    ItemGranters :[]item_granter_device = array{}
    @editable
    NextColorIndicators : []creative_prop = array{}

    @editable
    HUDMessagesData : []color_switch_hud_message = array:
        color_switch_hud_message{Color := color_switch_color.Red}
        color_switch_hud_message{Color := color_switch_color.Orange}
        color_switch_hud_message{Color := color_switch_color.Yellow}
        color_switch_hud_message{Color := color_switch_color.Green}
        color_switch_hud_message{Color := color_switch_color.Cyan}
        color_switch_hud_message{Color := color_switch_color.Blue}
        color_switch_hud_message{Color := color_switch_color.Purple}
        color_switch_hud_message{Color := color_switch_color.Pink}

    var HUDMessages : [color_switch_color]hud_message_device = map{}
    block:
        for(Data : HUDMessagesData):
            option{set HUDMessages[Data.Color] = Data.HUDMessage}

    @editable
    ShowNextColorDuration : float = 5.0

    @editable
    KeepColorBlocksHiddenDuration : float = 2.0

    ColorToMaterial : [color_switch_color]material = map:
        color_switch_color.Blue => M_BlueSwitch
        color_switch_color.Red => M_RedSwitch
        color_switch_color.Purple => M_PurpleSwitch
        color_switch_color.Yellow => M_YellowSwitch
        color_switch_color.Green => M_GreenSwitch
        color_switch_color.Pink => M_PinkSwitch
        color_switch_color.Cyan => M_CyanSwitch
        color_switch_color.Orange => M_OrangeSwitch

    TestColorMap : [color_switch_color]material = map:
        color_switch_color.Blue => M_CSBlue
        color_switch_color.Red => M_CSRed
        color_switch_color.Purple => M_CSPurple
        color_switch_color.Yellow => M_CSYellow
        color_switch_color.Green => M_CSGreen
        color_switch_color.Pink => M_CSPink
        color_switch_color.Cyan => M_CSCyan
        color_switch_color.Orange => M_CSOrange

    @editable
    MapType : map_type = map_type.Normal
    @editable
    BoxCenterParentProp : creative_prop = creative_prop{}
    @editable
    RegularMapParentProp : creative_prop = creative_prop{}

    @editable
    AllBlocks : []creative_prop = array{}

    var AllColors : []color_switch_color = array{}
    var ColoredIndexedBlocks : [color_switch_color][]creative_prop = map{}

    var OutOfBoundsTransform : transform = transform{}
    var InOfBoundsTransform : transform = transform{}

    Name<override>: string = "Color Switch"
    StartingMessage<override>:string = ""
    GetVotingImage<override>()<transacts>: texture = VerseExposed.SQ_COLORSWITCH

    OnMinigameDeviceBegin<override>()<suspends>:void=
        if(MapType = map_type.Normal):
            for(Color -> __ : TestColorMap):
                set AllColors += array{Color}

    _Start<override>(Players: []player)<suspends>: void =
        set InOfBoundsTransform = BoxCenterParentProp.GetTransform()
        set OutOfBoundsTransform = RegularMapParentProp.GetTransform()

        if(RegularMapParentProp.TeleportTo[InOfBoundsTransform]):
        BlockSetup(TestColorMap)
        Timer.Enable()

    #Sets up strating colours for each blocks
    BlockSetup(InMap : [color_switch_color]material) : void=
        # Init random block color and caching
        set ColoredIndexedBlocks = map{}
        for(Block : AllBlocks, RandomColor := AllColors.GetRandomElementFromArray[](0), RandomMaterial := InMap[RandomColor]):
            Block.SetMaterial(RandomMaterial)

            if(not ColoredIndexedBlocks[RandomColor]):
                option{set ColoredIndexedBlocks[RandomColor] = array{}}
            option{set ColoredIndexedBlocks[RandomColor] += array{Block}}

    _Reset<override>()<suspends>:void=
        Timer.Disable()

        for(HUDMessage : HUDMessages):
            HUDMessage.Hide()

        for(Prop : AllBlocks):
            Prop.SetMaterial(M_CSWhite)
            Prop.Show()

        if(BoxCenterParentProp.IsValid[]):
            option{BoxCenterParentProp.TeleportTo[BoxCenterParentProp.GetTransform().Translation, rotation{}]}

        if(RegularMapParentProp.TeleportTo[OutOfBoundsTransform]):

    OnSafeTimeEnded<override>()<suspends>:void=
        var ColorMap : [color_switch_color]material = map{}
        if(MapType = map_type.Normal):
            set ColorMap = TestColorMap

        loop:
            if(
                RandomColorResult := AllColors.GetRandomElementFromArray[],
                NextColor := RandomColorResult(0),
                NextMaterial := ColorMap[NextColor],
                RandomGranter := Shuffle(ItemGranters)[0],
                #RandomGranter := ItemGranters[0],
                HUDMessage := HUDMessages[NextColor]
            ):
                # Show glowing next color widget HUD
                HUDMessage.Show()

                # Grant loadout
                for(Player : GetRemainingPlayers()):
                    RandomGranter.GrantItem(Player)

                # Change indicators color
                for(ColorIndicator : NextColorIndicators):
                    ColorIndicator.SetMaterial(NextMaterial)

                Timer.Reset()
                Timer.Start()

                Sleep(ShowNextColorDuration)

                # Hide wrong props
                for(Color -> Props : ColoredIndexedBlocks, Color <> NextColor):
                    for(Prop : Props):
                        Prop.Hide()

                # Hide HUD now
                HUDMessage.Hide()

                Sleep(KeepColorBlocksHiddenDuration)

                for(Color -> Props : ColoredIndexedBlocks, Color <> NextColor):
                    for(Prop : Props):
                        Prop.Show()
            else:
                Sleep(5.0)