using { /Fortnite.com/Devices }
using { /Fortnite.com/AI}
using { /Fortnite.com/Teams}
using { /Verse.org/Simulation }
using { /Verse.org/Random }
using { /Fortnite.com/Characters }
using { /Fortnite.com/FortPlayerUtilities }
using { /Fortnite.com/Game }
using { /Fortnite.com/Teams }
using { /Verse.org/Simulation/Tags }
using { /UnrealEngine.com/Temporary/Diagnostics }
using { /UnrealEngine.com/Temporary/SpatialMath } 
using { /Verse.org/Assets }

box_rolling_axis := enum:
    Roll
    Pitch

map_type := enum:
    Normal
    HiChew

rolling_box := class(minigame_device):

    @editable
    TimeToDoCompleteTurn : float = 40.0

    @editable
    TimeBeforeStartRotating : float = 3.0

    @editable
    BoxCenterParentProp : creative_prop = creative_prop{}
    @editable
    RegularMapParentProp : creative_prop = creative_prop{}
    @editable
    HiChewMapParentProp : creative_prop = creative_prop{}

    @editable
    MapType : map_type = map_type.Normal

    var OutOfBoundsTransform : transform = transform{}
    var InOfBoundsTransform : transform = transform{}

    Name<override>: string = "Spin Cycle"
    StartingMessage<override>:string = ""
    GetVotingImage<override>()<transacts>: texture = 
        if(MapType = map_type.HiChew):
            VerseExposed.T_PlaceholderIcon_SpinCycle
        else:
            VerseExposed.SQ_SPIN2

    _Start<override>()<suspends>: void =
        set InOfBoundsTransform = BoxCenterParentProp.GetTransform()
        set OutOfBoundsTransform = RegularMapParentProp.GetTransform()

        if(MapType = map_type.HiChew):
            if(HiChewMapParentProp.TeleportTo[InOfBoundsTransform]):
                Print("Teleported HiChewMap")
        else:
            if(RegularMapParentProp.TeleportTo[InOfBoundsTransform]):
                Print("Teleported Regular Map")

        Sleep(TimeBeforeStartRotating)

        RandomDirectionList := array{1.0,-1.0}
        RandomAxisList := array{box_rolling_axis.Roll, box_rolling_axis.Pitch}
        if(RandomAxis := Shuffle(RandomAxisList)[0], RandomDirection := Shuffle(RandomDirectionList)[0]):
            loop:
                if(BoxCenterParentProp.IsValid[]):
                    CurrentRotation := BoxCenterParentProp.GetTransform().Rotation
                    TargetRotation := (RandomAxis = box_rolling_axis.Roll and CurrentRotation.ApplyRoll(DegreesToRadians(90.0 * RandomDirection))) or CurrentRotation.ApplyPitch(DegreesToRadians(90.0 * RandomDirection))
                    BoxCenterParentProp.MoveTo(BoxCenterParentProp.GetTransform().Translation, TargetRotation, TimeToDoCompleteTurn * 0.25)
                else:
                    Sleep(5.0)

    _Reset<override>()<suspends>:void=
        if(BoxCenterParentProp.IsValid[]):
            option{BoxCenterParentProp.TeleportTo[BoxCenterParentProp.GetTransform().Translation, rotation{}]}

        if(HiChewMapParentProp.TeleportTo[OutOfBoundsTransform]):
            Print("Teleported HiChewMap")
        if(RegularMapParentProp.TeleportTo[OutOfBoundsTransform]):
            Print("Teleported Regular Map")