using { /Fortnite.com/Devices }
using { /Fortnite.com/AI}
using { /Fortnite.com/Teams}
using { /Verse.org/Simulation }
using { /Verse.org/Random }
using { /Fortnite.com/Characters }
using { /Fortnite.com/FortPlayerUtilities }
using { /Fortnite.com/Game }
using { /Fortnite.com/Teams }
using { /Verse.org/Simulation/Tags }
using { /UnrealEngine.com/Temporary/Diagnostics }
using { /UnrealEngine.com/Temporary/SpatialMath } 
using { /Verse.org/Assets }



topDown := class(minigame_device):

    @editable NoJumpingMutatorZoneBox5 :mutator_zone_device = mutator_zone_device{}
    @editable FixedCamera : gameplay_camera_fixed_angle_device = gameplay_camera_fixed_angle_device{}
    @editable GameplayControls : gameplay_controls_third_person_device = gameplay_controls_third_person_device{}
    @editable ShipmentProp : creative_prop = creative_prop{}

    Name<override>: string = "Top Down"
    StartingMessage<override>:string = ""
    GetVotingImage<override>()<transacts>: texture = VerseExposed.SQ_TOPDOWN2

    _Start<override>()<suspends>: void =
        NoJumpingMutatorZoneBox5.Enable()
        if(ShipmentProp.TeleportTo[vector3{X:= 7004.85, Y:= -19456.27, Z:= 2996.28}, rotation{}]):
        FixedCamera.Enable()
        FixedCamera.AddToAll()
        GameplayControls.Enable()
        GameplayControls.AddToAll()

    _Reset<override>()<suspends>:void=
        NoJumpingMutatorZoneBox5.Disable()
        GameplayControls.Disable()
        if(ShipmentProp.TeleportTo[vector3{X:= 7004.85, Y:= -19456.27, Z:= -6284.0}, rotation{}]):

    # Ran right before minigame ends
    _OnMinigameEnded<override>(Winners: []player)<suspends>:void=
        GameplayControls.RemoveFromAll()
        FixedCamera.Disable() # Removes the camera from alive players only / FIX for black screen issue