# using { /Fortnite.com/Devices }
# using { /Fortnite.com/AI}
# using { /Fortnite.com/Teams}
# using { /Verse.org/Simulation }
# using { /Verse.org/Random }
# using { /Verse.org/Assets }
# using { /Fortnite.com/Characters }
# using { /Fortnite.com/FortPlayerUtilities }
# using { /Fortnite.com/Game }
# using { /Fortnite.com/Teams }
# using { /Verse.org/Simulation/Tags }
# using { /UnrealEngine.com/Temporary/Diagnostics }
# using { /UnrealEngine.com/Temporary/SpatialMath } 
# using { /Verse.org/Assets }

# # See https://dev.epicgames.com/documentation/en-us/uefn/create-your-own-device-in-verse for how to create a verse device.

# # A Verse-authored creative device that can be placed in a level
# snowball := class(minigame_device):

    
#     # @editable SnowballProp : creative_prop = creative_prop{}
#     # @editable ItemGranterSnowball :item_granter_device = item_granter_device{}
#     # @editable ItemGranterSnowball2 :item_granter_device = item_granter_device{}
#     # @editable NoWeaponDropClassSelector :class_and_team_selector_device = class_and_team_selector_device{}
#     # @editable SnowEffect : vfx_spawner_device = vfx_spawner_device{}

#     # StartingMessage<override>:string = ""
#     # GetVotingImage<override>()<transacts>: texture = VerseExposed.SQ_SNOWSTORM

#     # _Start<override>()<suspends>: void = 
#     #     Print("Starting Snowball Minigame")
        
        
#     #     if(SnowballProp.TeleportTo[vector3{X:= 486.36, Y:= 530.5, Z:= 741.6}, rotation{}]):
#     #         Print("Teleported Bro")
#     #     SnowEffect.Enable()
#     #     AllPlayers := GetRemainingPlayers()
#     #     Print("AllPlayers Size = {AllPlayers.Length}")
#     #     if(Player1 := AllPlayers[0]):
#     #         ItemGranterSnowball.GrantItem(Player1)
#     #     for(Player : AllPlayers):
#     #         NoWeaponDropClassSelector.ChangeClass(Player)
#     #         if:
#     #             Fort := Player.GetFortCharacter[]
#     #         then:
#     #             Fort.SetMaxShield(100.0)
#     #             Fort.SetMaxHealth(200.0)
#     #     spawn{SnowballItems()}

#     # SnowballItems()<suspends>:void=
#     #     Print("New Snowball items")
#     #     loop:
#     #         Sleep(15.0)
#     #         Print("New Snowball items")
#     #         AllPlayers := GetRemainingPlayers()
#     #         if(Player1 := AllPlayers[0]):
#     #             ItemGranterSnowball2.GrantItem(Player1)
#     #         Sleep(10.0)
