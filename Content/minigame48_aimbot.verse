using { /Fortnite.com/Devices }
using { /Fortnite.com/AI}
using { /Fortnite.com/Teams}
using { /Verse.org/Simulation }
using { /Verse.org/Random }
using { /Verse.org/Assets }
using { /Fortnite.com/Characters }
using { /Fortnite.com/FortPlayerUtilities }
using { /Fortnite.com/Game }
using { /Fortnite.com/Teams }
using { /Verse.org/Simulation/Tags }
using { /UnrealEngine.com/Temporary/Diagnostics }
using { /UnrealEngine.com/Temporary/SpatialMath } 
using { /Verse.org/Assets }

# See https://dev.epicgames.com/documentation/en-us/uefn/create-your-own-device-in-verse for how to create a verse device.

# A Verse-authored creative device that can be placed in a level
aimbot := class(minigame_device):

    @editable Controls : gameplay_controls_third_person_device = gameplay_controls_third_person_device{}

    @editable WallhackBoxesPool : attachable_item_pool = attachable_item_pool{}
    @editable PostProcessProp : creative_prop = creative_prop{}

    @editable BoxX : float = 500.8
    @editable BoxY : float = 517.1
    @editable BoxZHeight : float = 720.0

    @editable InputTriggerAimIn : input_trigger_device = input_trigger_device{}

    Name<override>: string = "AUTOAIM"
    StartingMessage<override>:string = "Aim in to activate autoaim!"
    GetVotingImage<override>()<transacts>: texture = VerseExposed.SQ_autoaim
    DisplayIcon<override>: ?texture = option{VerseExposed.SQ_AUTOAIM_ICON}
    BottomMessage<override>:?string = option{"Tip: Aim in to activate autoaim!"}

    _Start<override>()<suspends>: void =
        InputTriggerAimIn.Enable()
        RegisterSubscriber(InputTriggerAimIn.PressedEvent.Subscribe(HandleAimIn))
        RegisterSubscriber(InputTriggerAimIn.ReleasedEvent.Subscribe(HandleAimInReleased))

        for(Player : GetRemainingPlayers(), Fort := Player.GetFortCharacter[]):
            InputTriggerAimIn.Register(Player)
            spawn{TryAttachWallhackBox(Fort)}
            
        Controls.Enable()
        option{PostProcessProp.TeleportTo[vector3{X:= BoxX, Y:= BoxY, Z:= BoxZHeight}, IdentityRotation()]}

    _Reset<override>()<suspends>: void =
        InputTriggerAimIn.UnregisterAll()
        InputTriggerAimIn.Disable()
        Controls.Disable()
        option{PostProcessProp.TeleportTo[vector3{X:= 500.796686, Y:= 517.119712, Z:= 7504.804721}, IdentityRotation()]}

        for(AttachableItem : WallhackBoxesPool.GeneratedPlayerAttachedItems):
            if(Prop := creative_prop[AttachableItem], Prop.IsValid[]):
                Prop.Hide()
            AttachableItem.Detach()

    TryAttachWallhackBox(Character : fort_character)<suspends>:void=
        race:
            EndedEvent.Await()
            block:
                AttachableItemMaybe := WallhackBoxesPool.TryAttachOneToPositional(Character)
                if(AttachableItem := AttachableItemMaybe?, AsProp := creative_prop[AttachableItem.Item]):
                    AsProp.Show()
                    Character.EliminatedEvent().Await()
                    AsProp.Hide()
                    AttachableItem.Detach()

    HandleAimIn(Player : agent):void=
        Controls.AddTo(Player)

    HandleAimInReleased(TupleReleased : tuple(agent, float)):void=
        Controls.RemoveFrom(TupleReleased(0))