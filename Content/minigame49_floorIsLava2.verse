using { /Fortnite.com/Devices }
using { /Fortnite.com/AI}
using { /Fortnite.com/Teams}
using { /Verse.org/Simulation }
using { /Verse.org/Random }
using { /Verse.org/Assets }
using { /Fortnite.com/Characters }
using { /Fortnite.com/FortPlayerUtilities }
using { /Fortnite.com/Game }
using { /Fortnite.com/Teams }
using { /Verse.org/Simulation/Tags }
using { /UnrealEngine.com/Temporary/Diagnostics }
using { /UnrealEngine.com/Temporary/SpatialMath } 
using { /Verse.org/Assets }


# A Verse-authored creative device that can be placed in a level
floorislavaUpdated := class(minigame_device):

    @editable CeilingProp : creative_prop = creative_prop{}
    @editable CinematicSequence :cinematic_sequence_device = cinematic_sequence_device{}

    Name<override>: string = "Rising Lava"
    StartingMessage<override>:string = "GET TO THE HIGHGROUND!"
    GetVotingImage<override>()<transacts>: texture = VerseExposed.SQ_RISINGLAVA

    _Start<override>()<suspends>:void=
        if(CeilingProp.TeleportTo[vector3{X:= 7168.0, Y:= -18958.0, Z:= 3984.0}, rotation{}]):
        CinematicSequence.Play()

    _Reset<override>()<suspends>:void=
        CinematicSequence.Stop()
        if(CeilingProp.TeleportTo[vector3{X:= 11403.0, Y:= -18958.0, Z:= 1919.0}, rotation{}]):