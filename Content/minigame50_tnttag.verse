using { /Fortnite.com/Devices }
using { /Fortnite.com/AI}
using { /Fortnite.com/Teams}
using { /Verse.org/Simulation }
using { /Verse.org/Random }
using { /Verse.org/Assets }
using { /Fortnite.com/Characters }
using { /Fortnite.com/FortPlayerUtilities }
using { /Fortnite.com/Game }
using { /Fortnite.com/Teams }
using { /Verse.org/Simulation/Tags }
using { /UnrealEngine.com/Temporary/Diagnostics }
using { /UnrealEngine.com/Temporary/SpatialMath } 
using { /Verse.org/Assets }

# See https://dev.epicgames.com/documentation/en-us/uefn/create-your-own-device-in-verse for how to create a verse device.
custom_tnt_player := class<unique>():
    Player : player
    TntAsset : creative_prop_asset
    PlayerTntHud: hud_message_device 
    PlayerRunnerHud: hud_message_device
    var TntActive<private> : logic = false
    var TntProp<private> : creative_prop = creative_prop{}

    Xoffset<private> : float = -10.0
    Yoffset<private> : float = 40.0
    Zoffset<private> : float = 20.0
    Scale<private> : float = 1.0

    Init():void=
        if:
            Agent := agent[Player]
            Fort := Agent.GetFortCharacter[]
        then:
            # Spawn prop at player location
            Print("Initializing Custom Tnt Player")
            Tran := transform{
                Rotation := Fort.GetViewRotation(), 
                Translation:= Fort.GetTransform().Translation + vector3{X:=Xoffset,Y:=Yoffset,Z:=Zoffset},
                Scale := vector3{X:=Scale,Y:=Scale,Z:=Scale}}
            if (Prop := SpawnProp(TntAsset, Tran)(0)?):
                set TntProp = Prop
                TntProp.Hide()
                spawn{MoveTnt(Fort)}

    IsTntPlayer():logic=
        return TntActive

    MakeTntPlayer():void=
        if:
            Agent := agent[Player]
        then:
            Print("Making Tnt Player")
            TntProp.Show()
            set TntActive = true
            PlayerTntHud.Show(Agent)

    MakeRunnerPlayer():void=
        if:
            Agent := agent[Player]
        then:
            Print("Making Runner Player")
            set TntActive = false
            TntProp.Hide()
            PlayerRunnerHud.Show(Agent)

    GetPropTransform():transform=
        return TntProp.GetTransform()

    Dispose(): void=
        TntProp.Dispose()
        set TntActive = false
    
    MoveTnt<private>(Fort : fort_character)<suspends>:void=
        loop:
            if(not TntProp.IsValid[]):
                Print("Prop not valid breaking out of loop")
                break
            
            Tran := transform{
                Rotation := Fort.GetViewRotation(), 
                Translation:= Fort.GetTransform().Translation + vector3{X:=Xoffset,Y:=Yoffset,Z:=Zoffset},
                Scale := vector3{X:=Scale,Y:=Scale,Z:=Scale}}
            TntProp.MoveTo(Tran, 0.2)

# A Verse-authored creative device that can be placed in a level
tnttag := class(minigame_device):

    @editable ShipmentProp: creative_prop = creative_prop{}
    @editable UiHud : hud_message_device = hud_message_device{}

    @editable TntAssets : []creative_prop_asset = array{}
    @editable PlayerHudTnt : hud_message_device = hud_message_device{}
    @editable PlayerHudRunner : hud_message_device = hud_message_device{}
    @editable TntExplodeCountdownTimer : timer_device = timer_device{}
    @editable ExplosiveDevice : explosive_device = explosive_device{}

    @editable DeagleRemover :item_remover_device = item_remover_device{}
    @editable DeagleItemGranter : item_granter_device = item_granter_device{}

    @editable AudioTnt : audio_player_device= audio_player_device{}
    @editable AudioRunner : audio_player_device= audio_player_device{}

    Name<override>: string = "TNT Tag"
    StartingMessage<override>:string = "Run from the TNT Tagger(s)!"
    GetVotingImage<override>()<transacts>: texture = VerseExposed.SQ_TNTTAG
    DisplayIcon<override>: ?texture = option{VerseExposed.TransparentTexture}

    var TntPlayers : [player]custom_tnt_player = map{}

    _Start<override>()<suspends>: void =
        AudioTnt.Enable()
        AudioRunner.Enable()
        if(ShipmentProp.TeleportTo[vector3{X:= 7004.85, Y:= -19456.27, Z:= 2996.28}, rotation{}]):
        for(Player : GetRemainingPlayers()):
            if:
                TntAsset := TntAssets[0]
                Fort := Player.GetFortCharacter[]
            then:
                RegisterSubscriber(Fort.DamagedEvent().Subscribe(HandlePlayerHit))
                if (Existing := TntPlayers[Player]){
                    Print("Shouldn't be here bro. This calls an existing player")
                }
                else:
                    CustomTntPlayer := custom_tnt_player{
                        Player := Player, 
                        TntAsset := TntAsset, 
                        PlayerTntHud := PlayerHudTnt,
                        PlayerRunnerHud := PlayerHudRunner}
                    CustomTntPlayer.Init()
                    if(set TntPlayers[Player] = CustomTntPlayer) {}

        RegisterSubscriber(TntExplodeCountdownTimer.SuccessEvent.Subscribe(HandleTntTimerSuccess))
        spawn{TntLoop()}

    _Reset<override>()<suspends>:void=
        for(TntPlayer : TntPlayers):
            TntPlayer.Dispose()
        set TntPlayers = map{}

        AudioTnt.Disable()
        AudioRunner.Disable()
        if(ShipmentProp.TeleportTo[vector3{X:= 7004.85, Y:= -19456.27, Z:= -6284.72}, rotation{}]):

    # Ran right before minigame ends
    _OnMinigameEnded<override>(Winners: []player)<suspends>:void=
        # PlayerHudTnt.Hide()
        # PlayerHudRunner.Hide()
        # for(Player : GetPlayspace().GetPlayers()):
        #     PlayerHudTnt.Show(Player, BuildMessage(""), ?DisplayTime := 0.1)
        #     PlayerHudRunner.Show(Player, BuildMessage(""), ?DisplayTime := 0.1)
        TntExplodeCountdownTimer.Disable()

    TntLoop()<suspends>:void=
        race:
            EndedEvent.Await()
            block:
                Sleep(2.0)
                UiHud.SetText(StringToMessage("Picking TNT Players!"))
                UiHud.Show()
                Sleep(3.0)
                Print("Starting TNT Loop")
                #Start Countdown Timer (10s)
                TntExplodeCountdownTimer.Enable()
                TntExplodeCountdownTimer.Reset()
                TntExplodeCountdownTimer.Start()
                #Figure Out Tnt Amount
                StillAlivePlayers := GetRemainingPlayers()
                var TntAmount : int = 1
                if(StillAlivePlayers.Length > 8):
                    set TntAmount = 3
                else if(StillAlivePlayers.Length > 3):
                    set TntAmount = 2
                Print("Tnt amount = {TntAmount}")
                #Shuffle Players
                ShuffledAlivePlayers := Shuffle(StillAlivePlayers)
                for(Index := 0..ShuffledAlivePlayers.Length-1):
                    Print("Shuffed Loop Index {Index}")
                    if:
                        Player := ShuffledAlivePlayers[Index]
                        CustomTntPlayer := TntPlayers[ShuffledAlivePlayers[Index]]
                    then:
                        if(Index < TntAmount):
                            Print("Should Make TNT Player")
                            CustomTntPlayer.MakeTntPlayer()
                            DeagleItemGranter.GrantItem(Player)
                            AudioTnt.Play(Player)
                        else:
                            Print("Should Make Runner Player")
                            CustomTntPlayer.MakeRunnerPlayer()

    HandleTntTimerSuccess(PlayerMaybe : ?agent):void=
        Print("Tnt Timer Successs Bro")
        spawn{ExplodeTntPlayers()}

    ExplodeTntPlayers()<suspends>:void=
        race:
            EndedEvent.Await()
            block:
                Print("Exploding Players")
                # Put everyone in PutInStasis (Freeze Players)
                # If they move sometimes it's glitchy eliminating the player
                for(Player : GetRemainingPlayers(), Player.GetFortCharacter[].IsActive[]):
                    if:
                        Fort := Player.GetFortCharacter[]
                    then:
                        Args := stasis_args{AllowTurning:=false,AllowFalling:=false,AllowEmotes:=false}
                        Fort.PutInStasis(Args)
                        DeagleRemover.Remove(Player)

                #If tnt player, then explode
                for(Player : GetRemainingPlayers(), Player.GetFortCharacter[].IsActive[]):
                    if:
                        Fort := Player.GetFortCharacter[]
                        CustomTntPlayer := TntPlayers[Player]
                    then:
                        IsTntPlayer := CustomTntPlayer.IsTntPlayer()
                        if(IsTntPlayer = true):
                            Print("TNT PLAYER Found")
                            PropLoc := CustomTntPlayer.GetPropTransform().Translation
                            if(ExplosiveDevice.TeleportTo[PropLoc, rotation{}]):
                                Sleep(0.2)
                                CustomTntPlayer.Dispose()
                                ExplosiveDevice.Reset()
                                ExplosiveDevice.Explode(Player)
                                Sleep(0.5)   
                        else:
                            Print("Not a tnt player")
                
                for(Player : GetRemainingPlayers(), Player.GetFortCharacter[].IsActive[]):
                    if:
                        Fort := Player.GetFortCharacter[]
                    then:
                        Args := stasis_args{AllowTurning:=false,AllowFalling:=false,AllowEmotes:=false}
                        Fort.ReleaseFromStasis()

                spawn{TntLoop()}

    HandlePlayerHit(DamageResult : damage_result):void=
        if:
            Target := DamageResult.Target
            Amount := DamageResult.Amount
            Instigator := DamageResult.Instigator?
            InstigatorAgent := Instigator.GetInstigatorAgent[]
            InstigatorFort := InstigatorAgent.GetFortCharacter[]
            InstigatorPlayer := player[InstigatorAgent]
            InstigatorCustomPlayer := TntPlayers[InstigatorAgent]
            CharacterWhoWasHitFort := fort_character[Target]
            CharacterWhoWasHitAgent := CharacterWhoWasHitFort.GetAgent[]
            CharacterWhoWasHitPlayer := player[CharacterWhoWasHitAgent]
            CharacterWhoWasHitCustomPlayer := TntPlayers[CharacterWhoWasHitPlayer]
            TeamCollection := GetPlayspace().GetTeamCollection()
        then:
            Print("Handling Damage Scenario2")
            if(InstigatorAgent = CharacterWhoWasHitAgent):
                Print("Same Agent: Usually from storm damage")
            else:
                CharacterWhoWasHitFort.SetHealth(200.0)
                IsInstigatorTntPlayer := InstigatorCustomPlayer.IsTntPlayer()
                IsCharacterWhoWasHitTntPlayer := CharacterWhoWasHitCustomPlayer.IsTntPlayer()
                if(IsInstigatorTntPlayer = true):
                    Print("Tnt player hit someone")
                    if(IsCharacterWhoWasHitTntPlayer = false):
                        Print("character who was hit is a runner player...")
                        InstigatorCustomPlayer.MakeRunnerPlayer()
                        DeagleRemover.Remove(InstigatorPlayer)
                        AudioRunner.Play(InstigatorPlayer)
    
                        CharacterWhoWasHitCustomPlayer.MakeTntPlayer()
                        DeagleItemGranter.GrantItem(CharacterWhoWasHitPlayer)
                        AudioTnt.Play(CharacterWhoWasHitPlayer)

                    else if(IsCharacterWhoWasHitTntPlayer = true):
                        Print("character who was hit is a tnt player...")
                else:
                    Print("Runner Player hit someone")

    _OnPlayerEliminated<override>(Result : elimination_result):void=
        if:
            PersonKilled := Result.EliminatedCharacter
            PersonKilledAgent := PersonKilled.GetAgent[]
            PersonKilledFort := PersonKilledAgent.GetFortCharacter[]
            PersonKilledPlayer := player[PersonKilledAgent]
            PersonKilledCustomTntPlayer := TntPlayers[PersonKilledPlayer]
        then:
            PersonKilledCustomTntPlayer.Dispose()
            #PersonKilledFort.DamagedEvent().Subscribe(HandlePlayerHit).Cancel()
