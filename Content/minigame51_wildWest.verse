# using { /Fortnite.com/Devices }
# using { /Fortnite.com/AI}
# using { /Fortnite.com/Teams}
# using { /Verse.org/Simulation }
# using { /Verse.org/Random }
# using { /Fortnite.com/Characters }
# using { /Fortnite.com/FortPlayerUtilities }
# using { /Fortnite.com/Game }
# using { /Fortnite.com/Teams }
# using { /Verse.org/Simulation/Tags }
# using { /UnrealEngine.com/Temporary/Diagnostics }
# using { /UnrealEngine.com/Temporary/SpatialMath } 
# using { /Verse.org/Assets }

# Wild_West := class(minigame_device):

#     @editable ItemGranterWildWest :item_granter_device = item_granter_device{}
#     @editable WildWestProp : creative_prop = creative_prop{}
#     @editable NoBuildNoMatsZone : mutator_zone_device = mutator_zone_device{}

#     Name<override>: string = "Wild West"
#     StartingMessage<override>:string = ""
#     GetVotingImage<override>()<transacts>: texture = VerseExposed.SQ_WILDWEST

#     _Start<override>()<suspends>:void = 
#         if(WildWestProp.TeleportTo[vector3{X:= 7168.0, Y:= -19456.0, Z:= 2048.0}, rotation{}]):

#     _Reset<override>()<suspends>:void=return
#         # TODO: put the default prop location if you wanna use this minigame again
#         # if(WildWestProp.TeleportTo[vector3{}, rotation{}]):
            