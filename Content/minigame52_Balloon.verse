using { /Fortnite.com/Devices }
using { /Verse.org/Simulation }
using { /Verse.org/Random }
using { /Fortnite.com/Characters }
using { /Fortnite.com/FortPlayerUtilities }
using { /Fortnite.com/Game }
using { /Verse.org/Simulation/Tags }
using { /UnrealEngine.com/Temporary/Diagnostics }
using { /UnrealEngine.com/Temporary/SpatialMath } 
using { /Verse.org/Assets }

balloons := class(minigame_device):

    @editable ItemGranterArrowsAndBalloons : item_granter_device = item_granter_device{}
    @editable PlatformProp : creative_prop = creative_prop{}
    @editable BouncerCeilingProp : creative_prop = creative_prop{}
    @editable DamageVolumeProp : creative_prop = creative_prop{}
    @editable LavaProp : creative_prop = creative_prop{}

    Name<override>: string = "Balloon POP"
    StartingMessage<override>:string = "Put on all your balloons!"
    GetVotingImage<override>()<transacts>: texture = VerseExposed.SQ_BALLOONPOP

    _Start<override>()<suspends>: void =
        if(PlatformProp.TeleportTo[vector3{X:= 7424.0, Y:= -19436.0, Z:= 3478.0}, rotation{}]):
        if(BouncerCeilingProp.TeleportTo[vector3{X:= 7296.0, Y:= -19200.0, Z:= 7936.0}, rotation{}]):
        if(DamageVolumeProp.TeleportTo[vector3{X:= 7168.0, Y:= -19308.0, Z:= 1691.0}, rotation{}]):
        if(LavaProp.TeleportTo[vector3{X:= 7040.0, Y:= -18730.0, Z:= 1217.0}, rotation{}]):

    _Reset<override>()<suspends>:void=
        if(PlatformProp.TeleportTo[vector3{X:= 7424.0, Y:= -19436.0, Z:= 20.473955}, rotation{}]):
        if(BouncerCeilingProp.TeleportTo[vector3{X:= 11531.000000, Y:= -19200.0, Z:= -1309.490000}, rotation{}]):
        if(DamageVolumeProp.TeleportTo[vector3{X:= 7168.0, Y:= -19308.0, Z:= -304.0}, rotation{}]):
        if(LavaProp.TeleportTo[vector3{X:= 7040.0, Y:= -18730.0, Z:= -975.994872}, rotation{}]):
            
    OnSafeTimeEnded<override>()<suspends>:void=
        sync:
            MovePlatform()
            AllPlayers := GetRemainingPlayers()
            
            block:
                for(Value : int = 0..7):
                    #if(Player1 := GetRemainingPlayers()[0]):
                        #ItemGranterArrowsAndBalloons.GrantItem(Player1)
                    for(Player : GetRemainingPlayers()):
                        ItemGranterArrowsAndBalloons.GrantItem(Player)
                    Sleep(7.0)
            
    MovePlatform()<suspends>:void=
        PlatformProp.MoveTo(vector3{X:= 7424.0, Y:= -19456.0, Z:= -627.0}, rotation{},5.0)
