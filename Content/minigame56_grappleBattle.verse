# using { /Fortnite.com/Devices }
# using { /Fortnite.com/AI}
# using { /Fortnite.com/Teams}
# using { /Verse.org/Simulation }
# using { /Verse.org/Random }
# using { /Fortnite.com/Characters }
# using { /Fortnite.com/FortPlayerUtilities }
# using { /Fortnite.com/Game }
# using { /Fortnite.com/Teams }
# using { /Verse.org/Simulation/Tags }
# using { /UnrealEngine.com/Temporary/Diagnostics }
# using { /UnrealEngine.com/Temporary/SpatialMath } 
# using { /Verse.org/Assets }

# # A Verse-authored creative device that can be placed in a level
# grapple_battle := class(minigame_device):
#     @editable ItemGranterGrappleBattleItems :item_granter_device = item_granter_device{}
#     @editable DesiredPlatformSize : float = 10.0
#     @editable PlatformProp : creative_prop = creative_prop{}
#     @editable TimerDeviceShrink : timer_device = timer_device{}

#     Name<override>: string = "Grapple Battle"
#     StartingMessage<override>:string = "Arena Shrinks Every 10s!"
#     GetVotingImage<override>()<transacts>: texture = VerseExposed.SQ_GRAPPLEBATTLE2

#     _Start<override>()<suspends>: void =
#         RegisterSubscriber(TimerDeviceShrink.SuccessEvent.Subscribe(HandleTimerShrinkSuccess))

#         # Healing loop
#         loop:
#             for(Player : GetRemainingPlayers(), Player.GetFortCharacter[].IsActive[], Fort:= Player.GetFortCharacter[]):
#                 Fort.SetShield(Fort.GetShield() + 5.0)
#             Sleep(2.0)

#     _Reset<override>()<suspends>: void =
#         TimerDeviceShrink.Disable()

#     OnSafeTimeEnded<override>()<suspends>:void=
#         spawn{ShrinkPlatform()}
        
#     HandleTimerShrinkSuccess(PlayerMaybe : ?agent):void=
#         AllPlayers := GetRemainingPlayers()
#         if(Player1 := AllPlayers[0]):
#             ItemGranterGrappleBattleItems.CycleToRandomItem(Player1)
#         spawn{ShrinkPlatform()}

#     ShrinkPlatform()<suspends>:void=
#         race:
#             EndedEvent.Await()
#             block:
#                 var NewPropScaleSize: float = PlatformProp.GetTransform().Scale.X - 20.0
#                 Print("NewPropScaleSize - {NewPropScaleSize}")
#                 if(NewPropScaleSize < DesiredPlatformSize):
#                     ShrinkScale := vector3{X:=DesiredPlatformSize, Y:=DesiredPlatformSize, Z:=PlatformProp.GetTransform().Scale.Z}
#                     Tran := transform:
#                         Translation := PlatformProp.GetTransform().Translation, 
#                         Rotation := PlatformProp.GetTransform().Rotation, 
#                         Scale := ShrinkScale
#                     PlatformProp.MoveTo(Tran,2.0)
#                 else:
#                     ShrinkScale := vector3{X:=NewPropScaleSize, Y:=NewPropScaleSize, Z:=PlatformProp.GetTransform().Scale.Z}
#                     Tran := transform:
#                         Translation := PlatformProp.GetTransform().Translation, 
#                         Rotation := PlatformProp.GetTransform().Rotation, 
#                         Scale := ShrinkScale
#                     PlatformProp.MoveTo(Tran,2.0)
#                     if(NewPropScaleSize-20.0>=DesiredPlatformSize):
#                         TimerDeviceShrink.Enable()
#                         TimerDeviceShrink.Reset()
#                         TimerDeviceShrink.Start()