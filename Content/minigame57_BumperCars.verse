# using { /Fortnite.com/Devices }
# using { /Fortnite.com/AI}
# using { /Fortnite.com/Teams}
# using { /Verse.org/Simulation }
# using { /Verse.org/Random }
# using { /Fortnite.com/Characters }
# using { /Fortnite.com/FortPlayerUtilities }
# using { /Fortnite.com/Game }
# using { /Fortnite.com/Teams }
# using { /Verse.org/Simulation/Tags }
# using { /UnrealEngine.com/Temporary/Diagnostics }
# using { /UnrealEngine.com/Temporary/SpatialMath } 
# using { /Verse.org/Assets }

# BumperCars := class(minigame_device):
#     @editable BumperSpawners : []npc_spawner_device = array{}
#     @editable BumperMutatorZone : mutator_zone_device = mutator_zone_device{}
#     @editable BumperMusic :audio_player_device = audio_player_device{}
#     @editable BumperSoundEffect :audio_player_device = audio_player_device{}
#     @editable GroundDamageVolume : damage_volume_device = damage_volume_device{}
#     @editable PlatformLayers : []creative_prop = array{}

#     #BumperCar Npc Spawners are visually showing for each round
#     @editable SpawnerGltichProp : creative_prop = creative_prop{}
#     @editable TimerDeviceShrink : timer_device = timer_device{}
    
#     Name<override>: string = "Bumper Cars"
#     StartingMessage<override>:string = "Get in Your Bumper Car!"
#     GetVotingImage<override>()<transacts>: texture = VerseExposed.SQ_BUMPERCARS2
#     BottomMessage<override>:?string = option{"Tac Sprint to Bump Players!"}

#     # Fails if player is dead or spawner array is empty
#     FindClosestBumperSpawner(Player: player)<transacts><decides>:npc_spawner_device=
#         var RetSpawner : ?npc_spawner_device = false

#         if(Character := Player.GetFortCharacter[], Character.IsActive[]):
#             var MinDistance : float = 99999.9
#             for(
#                 Spawner : BumperSpawners
#                 DistanceToSpawner := Distance(Spawner.GetTransform().Translation, Character.GetTransform().Translation)
#                 DistanceToSpawner < MinDistance
#             ):
#                 set MinDistance = DistanceToSpawner
#                 set RetSpawner = option{Spawner}

#         RetSpawner?

#     _Start<override>()<suspends>: void = 
#         BumperMutatorZone.Enable()
#         BumperMusic.Enable()
#         BumperMusic.Play()
#         BumperSoundEffect.Enable()
#         TimerDeviceShrink.Enable()
#         for(Player : GetRemainingPlayers(), Fort := Player.GetFortCharacter[]):
#             RegisterSubscriber(Fort.DamagedEvent().Subscribe(HandlePlayerHit))
#             spawn{HandleNPCSpawners(Player)}
#         spawn{StartShrinkingStuff()}
#         #RegisterSubscriber(TimerDeviceShrink.SuccessEvent.Subscribe(HandleTimerShrinkSuccess))

#     OnSafeTimeEnded<override>()<suspends>:void=
#         GroundDamageVolume.Enable()
#         if(SpawnerGltichProp.TeleportTo[vector3{X:=-13066.58, Y:= -15490.69, Z:= -8000.0}, rotation{}]):

#     _Reset<override>()<suspends>:void=
#         BumperMutatorZone.Disable()
#         BumperMusic.Disable()
#         BumperSoundEffect.Disable()
#         GroundDamageVolume.Disable()
#         # Timer Device was still showing paused time on HUD after reset
#         TimerDeviceShrink.Reset()
#         TimerDeviceShrink.Disable()
#         for(BumperCarUnique : BumperSpawners):
#             BumperCarUnique.DespawnAll(false)
#             BumperCarUnique.Disable()
#         for(PlatformOutsideLayer : PlatformLayers):
#             option{PlatformOutsideLayer.TeleportTo[vector3{X:=-13001.60, Y:= -15550.26, Z:= 1295.39}, rotation{}]}
#         #option{PlatformLayer1.TeleportTo[vector3{X:=-13001.60, Y:= -15550.26, Z:= 1295.39}, rotation{}]}
#         #option{PlatformLayer2.TeleportTo[vector3{X:=-13001.60, Y:= -15550.26, Z:= 1295.39}, rotation{}]}
#         option{SpawnerGltichProp.TeleportTo[vector3{X:=-13066.58, Y:= -15490.69, Z:= 3224.84}, rotation{}]}

#     StartShrinkingStuff()<suspends>:void=
#         race:
#             EndedEvent.Await()
#             block:
#                 for(Index:=0..PlatformLayers.Length-1):
#                     TimerDeviceShrink.Reset()
#                     Sleep(7.0)
#                     TimerDeviceShrink.Start()
#                     Sleep(10.0)
#                     if(PlatformLayer := PlatformLayers[Index]):
#                         spawn{MovePlatform(PlatformLayer)}
#                     Print("End of Loop")
#                 #Sleep(8.0)
#                 #TimerDeviceShrink.Enable()
#                 #TimerDeviceShrink.Reset()
#                 #TimerDeviceShrink.Start()

#     HandlePlayerHit(DamageResult : damage_result):void=
#         if:
#             Target := DamageResult.Target
#             Amount := DamageResult.Amount
#             Instigator := DamageResult.Instigator?
#             InstigatorAgent := Instigator.GetInstigatorAgent[]
#             InstigatorFort := InstigatorAgent.GetFortCharacter[]
            
#             CharacterWhoWasHitFort := fort_character[Target]
#             CharacterWhoWasHitAgent := CharacterWhoWasHitFort.GetAgent[]
#         then:
#             if(InstigatorAgent = CharacterWhoWasHitAgent):
#                 Print("Same Agent: Usually from storm damage")
#             else:
#                 BumperSoundEffect.Play(InstigatorAgent)
#                 Print("Not the Same Agent")

#     HandleNPCSpawners(Player : player)<suspends>:void=
#         Print("Handling Player")
#         Sleep(1.0)
#         if:
#             Agent := agent[Player]
#             Fort := Agent.GetFortCharacter[]
#             BumperSpawner := FindClosestBumperSpawner[Player]
#         then:
#             Print("FOUND NEAR BUMPER{BumperSpawner.GetTransform().Translation}")
#             BumperSpawner.Enable()
#             BumperSpawner.Spawn()

#     MovePlatform(Prop : creative_prop)<suspends>:void=
#         race:
#             EndedEvent.Await()
#             Prop.MoveTo(vector3{X:=-13001.60, Y:= -15550.26, Z:= -8000.0}, rotation{},10.0)