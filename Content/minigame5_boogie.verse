using { /Fortnite.com/Devices }
using { /Fortnite.com/AI}
using { /Fortnite.com/Teams}
using { /Verse.org/Simulation }
using { /Verse.org/Random }
using { /Fortnite.com/Characters }
using { /Fortnite.com/FortPlayerUtilities }
using { /Fortnite.com/Game }
using { /Fortnite.com/Teams }
using { /Verse.org/Simulation/Tags }
using { /UnrealEngine.com/Temporary/Diagnostics }
using { /UnrealEngine.com/Temporary/SpatialMath } 
using { /Verse.org/Assets }

# A Verse-authored creative device that can be placed in a level
boogie := class(minigame_device):

    StartingMessage<override>:string = "5 Boogie Bombs!"
    Name<override>: string = "Boogie Down"
    GetVotingImage<override>()<transacts>: texture = VerseExposed.SQ_BOOGIEDOWN

