using { /Fortnite.com/Devices }
using { /Fortnite.com/Devices/CreativeAnimation }
using { /Fortnite.com/Devices/CreativeAnimation/InterpolationTypes }
using { /Fortnite.com/AI}
using { /Fortnite.com/Teams}
using { /Verse.org/Simulation }
using { /Verse.org/Random }
using { /Fortnite.com/Characters }
using { /Fortnite.com/FortPlayerUtilities }
using { /Fortnite.com/Game }
using { /Fortnite.com/Teams }
using { /Verse.org/Simulation/Tags }
using { /UnrealEngine.com/Temporary/Diagnostics }
using { /UnrealEngine.com/Temporary/SpatialMath } 
using { /Verse.org/Assets }
using { Obstacles }

# A Verse-authored creative device that can be placed in a level
obstacles := class(minigame_device):
    
    Name<override>: string = "Dodge Wall"
    StartingMessage<override>:string = "Dodge the Incoming Walls!"
    GetVotingImage<override>()<transacts>: texture = VerseExposed.SQ_DODGEWALL

    @editable
    ItemSpawners : []item_spawner_device = array{}

    @editable
    NoEmoteMutatorZone : mutator_zone_device = mutator_zone_device{}

    @editable
    ItemWeaponRemover : item_remover_device = item_remover_device{}

    @editable
    SpawnLocationProp : creative_prop = creative_prop{}

    @editable
    EndLocationProp : creative_prop = creative_prop{}

    # @editable
    # ObstacleAssets : []creative_prop_asset = array{}

    @editable
    ObstaclePool : dodge_wall_pool = dodge_wall_pool{}

    @editable
    HitWallMovementModulator : movement_modulator_device = movement_modulator_device{}

    @editable
    var Speed : float = 1.0
    var DefaultSpeedMaybe : ?float = false

    @editable
    SpeedIncrement : float = 0.03

    @editable
    InitialSpawnTime : float = 0.5

    @editable
    InitialMoveTime : float = 8.0

    # @editable
    # MaxMoveSpeed : float = 1600.0 # Players would go through the walls otherwise
    MaxMoveSpeed : float = 1600.0

    @editable
    InitialTimeBetweenObstacles : float = 4.0

    @editable
    InitialWaitTimeAfterSpawned : float = 1.0
    <#
    ColorMaterials : []material = array:
        M_ObstacleWhite
        M_ObstacleRed
        M_ObstacleOrange
        M_ObstacleYellow
        M_ObstacleGreen
        M_ObstacleCyan
        M_ObstacleBlue
        M_ObstaclePurple
        M_ObstaclePink
    #>
    var CurrentlyPushedPlayers : [player]logic = map{}

    _Start<override>()<suspends>: void =
        NoEmoteMutatorZone.Enable()
        ItemWeaponRemover.Enable()
        for(ItemSpawner: ItemSpawners):
            ItemSpawner.Enable()
        spawn{RandomizeItemSpawners()}
        if(DefaultSpeed := DefaultSpeedMaybe?):
            set Speed = DefaultSpeed
        set DefaultSpeedMaybe = option{Speed}

        for(DodgeWall : ObstaclePool.DodgeWalls):
            for(Volume : DodgeWall.CollisionFeetVolumes):
                RegisterSubscriber(Volume.AgentEntersEvent.SubscribeSuspends(OnPlayerFeetHitWall))

    OnSafeTimeEnded<override>()<suspends>:void=
        # spawn{WatchPlayersVelocity()}
        if(
            SpawnLocationProp.IsValid[],
            EndLocationProp.IsValid[],
            SpawnTransform := SpawnLocationProp.GetTransform(),
            EndLocation := EndLocationProp.GetTransform().Translation
        ):
            # sync:
            #     # Game loop, infintely spawn obstacles
            loop:
                SpawnObstacle(SpawnTransform, EndLocation)
                Sleep(InitialTimeBetweenObstacles / Speed)
                set Speed += SpeedIncrement

                # # Fix for characters going through moving collisions when standing still
                # block:
                #     if(FloorParentProp.IsValid[]):
                #         InitialTranslation := FloorParentProp.GetTransform().Translation
                #         InitialRotation := FloorParentProp.GetTransform().Rotation
                #         var Direction : float = 1.0
                #         loop:
                #             FloorParentProp.MoveTo(InitialTranslation + (vector3{Z:=1.0} * Direction), InitialRotation, 1.0)
                #             set Direction *= -1.0
                #     else:
                #         Print("ERR: Obstacle floor prop not valid")
        else:
            Print("ERR: Obstacle spawn/end location prop not valid")

    RandomizeItemSpawners()<suspends>:void=
        race:
            EndedEvent.Await()
            block:
                loop:
                    Sleep(4.0)
                    Print("Spawning Item Bro")
                    if:
                        AllPlayers := GetRemainingPlayers()
                        RandomPlayer := AllPlayers[0]
                        RandomSpawner := ItemSpawners[GetRandomInt(0,ItemSpawners.Length-1)]
                    then:
                        #ItemWeaponRemover.Remove(RandomPlayer)
                        #RandomSpawner.SpawnItem()
                        for(Spawner : Shuffle(ItemSpawners)):
                            Spawner.SpawnItem()
                            Sleep(2.0)
        Print("End of RandomizeItemSpawners()")

    _Reset<override>()<suspends>:void=
        NoEmoteMutatorZone.Disable()
        ItemWeaponRemover.Disable()

        for(ItemSpawner: ItemSpawners):
            ItemSpawner.Disable()

        ObstaclePool.ReleaseAll()
        set CurrentlyPushedPlayers = map{}

    SpawnObstacle(SpawnTransform: transform, EndLocation: vector3)<suspends>:void=
        FetchResultMaybe := ObstaclePool.FetchAvailableItem()
        <#
        if(FetchResult := FetchResultMaybe?, RandomMaterial := Shuffle(ColorMaterials)[0]):
            DodgeWall := FetchResult(0)
            PoolIndex := FetchResult(1)
            Prop := DodgeWall.Prop

            InitialScale := vector3{X := 0.5, Y := 0.5, Z := 0.5}
            InitialTransform := transform:
                Translation := SpawnTransform.Translation
                Rotation := SpawnTransform.Rotation
                Scale := InitialScale

            if(Prop.TeleportTo[InitialTransform]):
                Prop.SetMaterial(RandomMaterial)
                
                # Fix for scale animation stuttering when spawned
                Prop.Hide()
                branch:
                    Sleep(0.0)
                    Prop.Show()

                FinalScale := vector3{X := 1.0, Y := 1.0, Z := 1.0}
                
                # Try to use anim controller to smooth the animation
                if(AnimController := Prop.GetAnimationController[]):
                    Keyframes := array:
                        keyframe_delta:
                            Time := 0.0
                            DeltaRotation := rotation{}
                            DeltaLocation := vector3{}
                        keyframe_delta:
                            Time := InitialSpawnTime / Speed
                            DeltaRotation := rotation{}
                            DeltaLocation := vector3{}
                            DeltaScale := FinalScale / InitialScale
                            Interpolation := EaseOut
                    AnimController.SetAnimation(Keyframes, ?Mode := animation_mode.OneShot)
                    AnimController.Play()
                    AnimController.MovementCompleteEvent.Await()

                # Use fallback linear animation otherwise
                else:
                    InitialTransformScaleOne := transform:
                        Translation := SpawnTransform.Translation
                        Rotation := SpawnTransform.Rotation
                        Scale := FinalScale
                    Prop.MoveTo(InitialTransformScaleOne, InitialSpawnTime / Speed)

                Sleep(InitialWaitTimeAfterSpawned / Speed)
                spawn{MoveObstacle(Prop, EndLocation, PoolIndex)}
            else:
                Print("ERR: Couldn't teleport obstacle prop")
                ObstaclePool.ReleaseItem(PoolIndex)
        #>
    MoveObstacle(Prop: creative_prop, EndLocation: vector3, PoolIndex: int)<suspends>:void=
        Rotation := Prop.GetTransform().Rotation
        InitialLocation := Prop.GetTransform().Translation
        MaxMoveDuration := Distance(InitialLocation, EndLocation) / MaxMoveSpeed

        # if(AnimController := Prop.GetAnimationController[]):
        #     Keyframes := array:
        #         keyframe_delta:
        #             Time := 0.0
        #             DeltaRotation := rotation{}
        #             DeltaLocation := vector3{}
        #         keyframe_delta:
        #             Time := MaxMoveDuration
        #             DeltaRotation := rotation{}
        #             DeltaLocation := EndLocation - InitialLocation
        #     AnimController.SetAnimation(Keyframes, ?Mode := animation_mode.OneShot)
        #     AnimController.Play()
        #     AnimController.MovementCompleteEvent.Await()

        # Prop.MoveTo(EndLocation, Rotation, MaxMoveDuration)
        Prop.MoveTo(EndLocation, Rotation, Max(InitialMoveTime / Speed, MaxMoveDuration))
        DeathTransform := transform:
            Translation := EndLocation
            Rotation := Rotation
            Scale := vector3{X := 0.01, Y := 1.0, Z := 1.0}
        Prop.MoveTo(DeathTransform, InitialSpawnTime / Speed)
        ObstaclePool.ReleaseItem(PoolIndex)


    # WatchPlayersVelocity()<suspends>:void=
    #     var PlayerLastPosition : [player]vector3 = map{}
    #     race:
    #         EndedEvent.Await()
    #         loop:
    #             for(Player : GetRemainingPlayers(), Character := Player.GetFortCharacter[]):
    #                 ActualPosition := Character.GetTransform().Translation

    #                 if(LastPosition := PlayerLastPosition[Player]):
    #                     Velocity := Distance(ActualPosition, LastPosition)
    #                     if(Velocity < 20.0):
    #                         # Print("Grind")
    #                         FixCollisionsGrindDevice.Pickup(Player)
    #                     else:
    #                         # Print("Stop Grind")
    #                         FixCollisionsGrindDeviceStop.Pickup(Player)

    #                 option{set PlayerLastPosition[Player] = ActualPosition}
    #             Sleep(0.2)

    
    OnPlayerFeetHitWall(Agent: agent)<suspends>:void=
        if(Player := player[Agent]):
            if(CurrentlyPushedPlayers[Player]):
                return

            option{set CurrentlyPushedPlayers[Player] = true}
            HitWallMovementModulator.Activate(Agent)

            if(Character := Agent.GetFortCharacter[]):
                Character.Damage(10.0)

                # DOESNT SHOW THE INDICATOR
                # OtherPlayers := for(Player : GetRemainingPlayers(), Player <> Agent) { Player }
                # if(RandomPlayer := Shuffle(OtherPlayers)[0], RandomCharacter := RandomPlayer.GetFortCharacter[]):
                #     Print("HOG FOUND")
                #     DamageArgs := damage_args:
                #         Amount := 10.0
                #         Instigator := option{RandomCharacter}
                #         Source := option{RandomCharacter}
                #     Character.Damage(DamageArgs)
                # else:
                #     Print("DAMAGING")

            Sleep(0.2)
            set CurrentlyPushedPlayers = CurrentlyPushedPlayers.WithRemovedKey(Player)