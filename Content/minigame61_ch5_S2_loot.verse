using { /Fortnite.com/Devices }
using { /Fortnite.com/AI}
using { /Fortnite.com/Teams}
using { /Verse.org/Simulation }
using { /Verse.org/Random }
using { /Fortnite.com/Characters }
using { /Fortnite.com/FortPlayerUtilities }
using { /Fortnite.com/Game }
using { /Fortnite.com/Teams }
using { /Verse.org/Simulation/Tags }
using { /UnrealEngine.com/Temporary/Diagnostics }
using { /UnrealEngine.com/Temporary/SpatialMath } 
using { /Verse.org/Assets }
# A Verse-authored creative device that can be placed in a level
chapter5_season2_loot := class(minigame_device):
    @editable CeilingProp : creative_prop = creative_prop{}

    Name<override>: string = "Myths & Mortals Loot"
    StartingMessage<override>:string = "Myths & Mortals Loadout!"
    GetVotingImage<override>()<transacts>: texture = VerseExposed.SQ_MYTHS

    _Start<override>()<suspends>:void=
        if(CeilingProp.TeleportTo[vector3{X:= 7168.0, Y:= -18958.0, Z:= 4752.0}, rotation{}]):

    _Reset<override>()<suspends>:void=
        if(CeilingProp.TeleportTo[vector3{X:= 11403.0, Y:= -18958.0, Z:= 1919.0}, rotation{}]):