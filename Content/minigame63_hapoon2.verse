using { /Fortnite.com/Devices }
using { /Fortnite.com/AI}
using { /Fortnite.com/Teams}
using { /Verse.org/Simulation }
using { /Verse.org/Random }
using { /Fortnite.com/Characters }
using { /Fortnite.com/FortPlayerUtilities }
using { /Fortnite.com/Game }
using { /Fortnite.com/Teams }
using { /Verse.org/Simulation/Tags }
using { /UnrealEngine.com/Temporary/Diagnostics }
using { /UnrealEngine.com/Temporary/SpatialMath } 
using { /Verse.org/Assets }

harpoon_in_the_hole := class(minigame_device):
    @editable ArenaProp : creative_prop = creative_prop{}
    @editable DamageVolumeProp : creative_prop = creative_prop{}
    @editable LavaProp : creative_prop = creative_prop{}
    @editable Team1Teleporter : teleporter_device = teleporter_device{}
    @editable Team2Teleporter : teleporter_device = teleporter_device{}
    @editable BarrierDevice : barrier_device = barrier_device{}

    Name<override>: string = "Harpoon In The Hole"
    StartingMessage<override>:string = "Harpoon the Other Team Down!"
    GetVotingImage<override>()<transacts>: texture = VerseExposed.SQ_HarpoonHole

    var Team1 : []player = array{}
    var Team2 : []player = array{}

    _Start<override>()<suspends>: void =
        BarrierDevice.Enable()
        if(ArenaProp.TeleportTo[vector3{X:= 7158.3, Y:= -19471.94, Z:= 8671.33}, rotation{}]):
        if(DamageVolumeProp.TeleportTo[vector3{X:= 7168.0, Y:= -19308.0, Z:= 1691.0}, rotation{}]):
        if(LavaProp.TeleportTo[vector3{X:= 7040.0, Y:= -18730.0, Z:= 1217.0}, rotation{}]):
        
    OnSafeTimeEnded<override>()<suspends>:void=
        BarrierDevice.Disable()

    _Reset<override>()<suspends>:void=
        if(DamageVolumeProp.TeleportTo[vector3{X:= 7168.0, Y:= -19308.0, Z:= -304.0}, rotation{}]):
        if(LavaProp.TeleportTo[vector3{X:= 7040.0, Y:= -18730.0, Z:= -975.994872}, rotation{}]):
        if(ArenaProp.TeleportTo[vector3{X:= 11631.209837, Y:= -19471.939978, Z:= 8671.330002}, rotation{}]):
        set Team1 = array{}
        set Team2 = array{}
    <#
    GetWinners<override>()<transacts><decides>:[]player=
        #Print("Getting Winers")
        Team1AlivePlayers := for(Player : GetRemainingPlayers(), IsInTeam(Team1,Player) = true) { Player }
        Team2AlivePlayers := for(Player : GetRemainingPlayers(), IsInTeam(Team2,Player) = true) { Player }
        # Allow for draw to happen
        Team1AlivePlayers.Length = 0 or Team2AlivePlayers.Length = 0
        var RetArray : []player = array{}
        if(Team1AlivePlayers.Length = 0 and Team2AlivePlayers.Length = 0):
            RetArray
        if(Team1AlivePlayers.Length = 0):
            set RetArray = Team2
        if(Team2AlivePlayers.Length = 0):
            set RetArray = Team1
        RetArray
        #>