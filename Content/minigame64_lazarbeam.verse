using { /Fortnite.com/Devices }
using { /Fortnite.com/AI}
using { /Fortnite.com/Teams}
using { /Verse.org/Simulation }
using { /Verse.org/Random }
using { /Fortnite.com/Characters }
using { /Fortnite.com/FortPlayerUtilities }
using { /Fortnite.com/Game }
using { /Fortnite.com/Teams }
using { /Verse.org/Simulation/Tags }
using { /UnrealEngine.com/Temporary/Diagnostics }
using { /UnrealEngine.com/Temporary/SpatialMath } 
using { /Verse.org/Assets }
using { /UnrealEngine.com/Assets }

lazarbeam_attack := class<concrete>():
    @editable LasarProp : creative_prop = creative_prop{}
    @editable OptAttackSequence : ?cinematic_sequence_device = false
    @editable OptTelegraph : ?creative_prop_asset = false
    @editable OptDamageVolume : ?damage_volume_device = false
    @editable OptVFXSpawner : ?vfx_spawner_device = false
    @editable OptBeamProp : ?creative_prop_asset = false
    @editable BeamSpawnPoints : []creative_prop = array{}
    @editable MultiTarget : logic = false

    @editable TimesToRepeat : int = 0
    @editable RotationTime : float = 1.0
    @editable HitDelay : float = 2.0
    @editable AfterAnimationDelay : float = 0.0
    @editable LocationOffset : vector3 = vector3{}
    @editable DamageVolumeLocationOffset : vector3 = vector3{}
    @editable AudioPlayers : []audio_player_device = array{}

    @editable AttackPoints : []creative_prop = array{}

    var StoredDamageVolumeLocation : vector3 = vector3{}
    var LastPlayedAudioIndex : int = -1

    SpawnPropAsset(PropAsset : ?creative_prop_asset, Location : vector3)<suspends>:void=
        if(Prop := PropAsset?):
            # Spawn prop
            SpawnedResult := SpawnProp(Prop, Location, rotation{})
            # Prop lifetime
            Sleep(HitDelay - RotationTime - AfterAnimationDelay)
            # Remove prop
            if(SpawnedProp := SpawnedResult(0)?):
                SpawnedProp.Dispose()

    SpawnBeamAsset(PropAsset : ?creative_prop_asset, Location : vector3, PointLocation : vector3)<suspends>:void=
        if(Prop := PropAsset?):
            # Spawn prop
            SpawnedResult := SpawnProp(Prop, Location, rotation{})
            # Prop lifetime
            if(SpawnedProp := SpawnedResult(0)?):
                LasarLocation := SpawnedProp.GetTransform().Translation
                if(LookDirection := (PointLocation - LasarLocation).MakeUnitVector[]):
                    Yaw := RadiansToDegrees(ArcTan(LookDirection.Y, LookDirection.X))
                    Pitch := RadiansToDegrees(ArcTan(LookDirection.Z, Sqrt((LookDirection.X * LookDirection.X) + (LookDirection.Y * LookDirection.Y))))

                    if(SpawnedProp.TeleportTo[LasarLocation, MakeRotationFromYawPitchRollDegrees(Yaw, Pitch, 0.0)]) {}
                Sleep(0.1)
                # Remove prop
                SpawnedProp.Dispose()

    BasicAttack()<suspends>: void=
        var OldLocation : vector3 = vector3{}
        var LasarLocation : vector3 = vector3{}
        var PointLocation : vector3 = vector3{}
        var AttackLocation : vector3 = vector3{}
        var ShuffledAttackPoints : []creative_prop = array{}
        Print("Basic Attack")
        for(X:=0..TimesToRepeat):
            loop: 
                set ShuffledAttackPoints = Shuffle(AttackPoints)

                if(RandomPoint := ShuffledAttackPoints[0]):
                    set LasarLocation = LasarProp.GetTransform().Translation
                    set PointLocation = RandomPoint.GetTransform().Translation
                    set AttackLocation = PointLocation + LocationOffset

                    if(PointLocation.Y = OldLocation.Y and ShuffledAttackPoints.Length > 1):
                        Print("Same Point Reshuffle")
                    else:
                        break

            set OldLocation = AttackLocation
            if(AttackSequence := OptAttackSequence?):
                AttackSequence.Stop()
                AttackSequence.Play()
            spawn:
                SpawnPropAsset(OptTelegraph, AttackLocation)
            if(LookDirection := (PointLocation - LasarLocation).MakeUnitVector[]):
                Yaw := RadiansToDegrees(ArcTan(LookDirection.Y, LookDirection.X))
                Pitch := RadiansToDegrees(ArcTan(LookDirection.Z, Sqrt((LookDirection.X * LookDirection.X) + (LookDirection.Y * LookDirection.Y))))

                LasarRotaion := MakeRotationFromYawPitchRollDegrees(Yaw - 90.0, 0.0, 0.0)
                LasarProp.MoveTo(LasarLocation, LasarRotaion, RotationTime)
                
            Sleep(HitDelay - RotationTime)
            if(OptVFXSpawner?.TeleportTo[PointLocation, rotation{}]) {}
            # Teleport Damage volume to randomised location
            if(AudioPlayer := AudioPlayers[0]):
                loop:
                    var ShuffledIndex : int = GetRandomInt(0,AudioPlayers.Length-1)
                    if(ShuffledIndex = LastPlayedAudioIndex):
                        Print("Reshuffle Audio")
                    else:
                        if(NextPlayer := AudioPlayers[ShuffledIndex]):
                            NextPlayer.Play()
                            set LastPlayedAudioIndex = ShuffledIndex
                        break
            Sleep(AfterAnimationDelay)
            for(Point : BeamSpawnPoints):
                spawn:
                    SpawnBeamAsset(OptBeamProp, Point.GetTransform().Translation, AttackLocation)
            if(DamageVolume := OptDamageVolume?, Telegraph := OptTelegraph?):
                if(DamageVolume.TeleportTo[AttackLocation + DamageVolumeLocationOffset, rotation{}]) {}
                Sleep(0.05)
                if(VFX := OptVFXSpawner?):
                    VFX.Enable()
                if(DamageVolume.TeleportTo[StoredDamageVolumeLocation, rotation{}]) {}

    MultiTargetAttack()<suspends>:void=
        var LasarLocation : vector3 = vector3{}
        var PointLocation1 : vector3 = vector3{}
        var PointLocation2 : vector3 = vector3{}
        var AttackLocation1 : vector3 = vector3{}
        var AttackLocation2 : vector3 = vector3{}
        var ShuffledAttackPoints : []creative_prop = array{}
        Print("Multi Attack")

        for(X:=0..TimesToRepeat):
            set ShuffledAttackPoints = Shuffle(AttackPoints)

                if(RandomPoint1 := ShuffledAttackPoints[0], RandomPoint2 := ShuffledAttackPoints[1]):
                    set PointLocation1 = RandomPoint1.GetTransform().Translation
                    set AttackLocation1 = PointLocation1 + LocationOffset
                    set PointLocation2 = RandomPoint2.GetTransform().Translation
                    set AttackLocation2 = PointLocation2 + LocationOffset

                spawn:
                    SpawnPropAsset(OptTelegraph, AttackLocation1)
                spawn: 
                    SpawnPropAsset(OptTelegraph, AttackLocation2)

            if(AttackSequence := OptAttackSequence?):
                    AttackSequence.Stop()
                    AttackSequence.Play()
                Sleep(HitDelay)
            for(Point : BeamSpawnPoints):
                spawn:
                    SpawnBeamAsset(OptBeamProp, Point.GetTransform().Translation, AttackLocation1)
                spawn:
                    SpawnBeamAsset(OptBeamProp, Point.GetTransform().Translation, AttackLocation2)
            if(AudioPlayer := AudioPlayers[0]):
                loop:
                    var ShuffledIndex : int = GetRandomInt(0,AudioPlayers.Length-1)
                    if(ShuffledIndex = LastPlayedAudioIndex):
                        Print("Reshuffle Audio")
                    else:
                        if(NextPlayer := AudioPlayers[ShuffledIndex]):
                            NextPlayer.Play()
                            set LastPlayedAudioIndex = ShuffledIndex
                        break
            if(OptVFXSpawner?.TeleportTo[PointLocation1, rotation{}]) {}
            if(DamageVolume := OptDamageVolume?):
                if(DamageVolume.TeleportTo[AttackLocation1 + DamageVolumeLocationOffset, rotation{}]) {}
                    Sleep(0.05)
                    if(VFX := OptVFXSpawner?):
                        VFX.Enable()
                    if(DamageVolume.TeleportTo[StoredDamageVolumeLocation, rotation{}]) {}
                    Sleep(0.01)
                if(OptVFXSpawner?.TeleportTo[PointLocation2, rotation{}]) {}    
                if(DamageVolume.TeleportTo[AttackLocation2 + DamageVolumeLocationOffset, rotation{}]) {}
                    Sleep(0.05)
                    if(VFX := OptVFXSpawner?):
                        VFX.Enable()
                    if(DamageVolume.TeleportTo[StoredDamageVolumeLocation, rotation{}]) {}

lazarbeam_bossfight := class(minigame_device):

    @editable FixedCamera : gameplay_camera_fixed_angle_device = gameplay_camera_fixed_angle_device{}
    @editable GameplayControls : gameplay_controls_third_person_device = gameplay_controls_third_person_device{}
    #@editable TimerDevice : timer_device = timer_device{}
    @editable Subscribe : audio_player_device = audio_player_device{}
    @editable WinAudio : audio_player_device = audio_player_device{}

    @editable IcyVFX : []vfx_spawner_device = array{}
    @editable IcyPowerUp : grind_powerup_device = grind_powerup_device{}
    @editable ClearIcyTrigger : trigger_device = trigger_device{}

    @editable Attacks : []lazarbeam_attack = array{}

    Name<override>: string = "Lazarbeam Boss Fight"
    StartingMessage<override>:string = "Dodge Lazarbeam's Attacks!"
    GetVotingImage<override>()<transacts>: texture = VerseExposed.SQ_LAZARBOSS

    _OnGenericEndTimerSuccessReturnArray<override>():[]player=
        return GetRemainingPlayers()

    _Start<override>()<suspends>: void =
        FixedCamera.Enable()
        FixedCamera.AddToAll()
        GameplayControls.Enable()
        GameplayControls.AddToAll()
        #TimerDevice.Enable()
        #RegisterSubscriber(TimerDevice.SuccessEvent.Subscribe(OnTimerSuccess))

        for(CurrentAttack : Attacks):
            if(DamageVolume := CurrentAttack.OptDamageVolume?):
                set CurrentAttack.StoredDamageVolumeLocation = DamageVolume.GetTransform().Translation
                DamageVolume.Enable()

    _Reset<override>()<suspends>:void=
        GameplayControls.Disable()
        #TimerDevice.Reset()
        #TimerDevice.Disable()

    OnSafeTimeEnded<override>()<suspends>:void=
        #TimerDevice.Reset()
        #TimerDevice.Start()
        Subscribe.Play()
        spawn:
            IcyFeetTimer()
        race:
            EndedEvent.Await()
            block:
                loop:
                    for(CurrentAttack : Attacks):
                        if(CurrentAttack.MultiTarget = true):
                            CurrentAttack.MultiTargetAttack()
                        else:
                            CurrentAttack.BasicAttack()
            #block:
                #TimerDevice.SuccessEvent.Await()

    IcyFeetTimer()<suspends>:void=
        race:
            EndedEvent.Await()
            block:
                Sleep(40.0)
                Print("ICY")
                for(Player : GetRemainingPlayers()):
                    IcyPowerUp.Spawn()
                    IcyPowerUp.Pickup(Player)
                for(VFX : IcyVFX):
                    VFX.Enable()

    _OnMinigameEnded<override>(Winners: []player)<suspends>:void=
        GameplayControls.RemoveFromAll()
        FixedCamera.Disable()
        ClearIcyTrigger.Trigger()
        for(CurrentAttack : Attacks):
            if(DamageVolume := CurrentAttack.OptDamageVolume?, Sequence := CurrentAttack.OptAttackSequence?):
                DamageVolume.Disable()
                Sequence.Stop()
                # Needed if a player dies on an attack to teleport it back to location
                if(DamageVolume.TeleportTo[CurrentAttack.StoredDamageVolumeLocation, rotation{}]) {}
                for(AudioPlayer : CurrentAttack.AudioPlayers):
                    AudioPlayer.Stop()
        for(Player : Winners):
            WinAudio.Play(Player)
            