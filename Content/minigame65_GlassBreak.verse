using { /Fortnite.com/Devices }
using { /Fortnite.com/AI}
using { /Fortnite.com/Teams}
using { /Verse.org/Simulation }
using { /Verse.org/Random }
using { /Fortnite.com/Characters }
using { /Fortnite.com/FortPlayerUtilities }
using { /Fortnite.com/Game }
using { /Fortnite.com/Teams }
using { /Verse.org/Simulation/Tags }
using { /UnrealEngine.com/Temporary/Diagnostics }
using { /UnrealEngine.com/Temporary/SpatialMath } 
using { /Verse.org/Assets }

GlassBreakRow:= class<concrete>:
    @editable
    LeftPlatform : GlassBreakPlatform = GlassBreakPlatform{}

    @editable
    RightPlatform : GlassBreakPlatform = GlassBreakPlatform{}

GlassBreakPlatform := class<concrete>:
    @editable
    PlatformProp : creative_prop = creative_prop{}

    @editable
    CollisionVolume : volume_device = volume_device{}

    var<private> Subscribers : []cancelable = array{}

    RegisterPlatformToFall():void=
        Cancelable := CollisionVolume.AgentEntersEvent.Subscribe(OnAgentEnterVolume)
        set Subscribers += array{Cancelable}

    OnAgentEnterVolume(InAgent : agent):void=
        PlatformProp.Hide()

    CancelSubscribables():void=
        for(Subscriber : Subscribers):
            Subscriber.Cancel()

GlassBreak := class(minigame_device):
    @editable 
    GoldWinValue : int = 450
    @editable
    TimeToResetGlass : float = 10.0

    #@editable DamageVolume :damage_volume_device = damage_volume_device{}
    @editable 
    DamageVolumeBottom :damage_volume_device = damage_volume_device{}
    @editable 
    Collectible : collectible_object_device = collectible_object_device{}
    #Team Minigame Code Addition - Added arrays to support 4 teams
    @editable 
    GlassBreakObstaclesTeam1 : []GlassBreakRow = array{}
    @editable 
    GlassBreakObstaclesTeam2 : []GlassBreakRow = array{}
    @editable 
    GlassBreakObstaclesTeam3 : []GlassBreakRow = array{}
    @editable 
    GlassBreakObstaclesTeam4 : []GlassBreakRow = array{}

    Name<override>: string = "Glass Break"
    StartingMessage<override>:string = "Choose the Correct Path!"
    GetVotingImage<override>()<transacts>: texture = VerseExposed.SQ_GLASSBREAK_Vote
    DisplayIcon<override>: ?texture = option{VerseExposed.SQ_GLASSBREAK}

    _Start<override>()<suspends>: void =
        #DamageVolume.Enable()
        RandomizePlatforms()
        Collectible.Show()
        Collectible.RespawnForAll()
        RegisterSubscriber(Collectible.CollectedEvent.Subscribe(HandleCollectedEvent))
        #TEAM MINIGAME CODE - Reset the glass on an interval
        spawn:
            ResetGlass()

    OnSafeTimeEnded<override>()<suspends>:void=
        Manager.UiMessage.SetText(StringToMessage("First Team to the Crown Wins!"))
        Manager.UiMessage.Show()
        #TEAM MINIGAME - DISABLED FOR NOW DUE TO NEW RESPAWN FUNCTIONALITY
        #spawn{HandleLastPlayer()}
        #DamageVolume.MoveTo(vector3{X:= 29115.0, Y:= -1250.0, Z:= -8850.0}, rotation{},85.0) 

    _Reset<override>()<suspends>:void=
        #if(DamageVolume.TeleportTo[vector3{X:= 14068.0, Y:= -1250.0, Z:= -8850.0}, rotation{}]):
        Collectible.Hide()
        for(Row : GlassBreakObstaclesTeam1):
            Row.LeftPlatform.PlatformProp.Show()
            Row.RightPlatform.PlatformProp.Show()
            Row.LeftPlatform.CancelSubscribables()
            Row.RightPlatform.CancelSubscribables()
        for(Row : GlassBreakObstaclesTeam2):
            Row.LeftPlatform.PlatformProp.Show()
            Row.RightPlatform.PlatformProp.Show()
            Row.LeftPlatform.CancelSubscribables()
            Row.RightPlatform.CancelSubscribables()
        for(Row : GlassBreakObstaclesTeam3):
            Row.LeftPlatform.PlatformProp.Show()
            Row.RightPlatform.PlatformProp.Show()
            Row.LeftPlatform.CancelSubscribables()
            Row.RightPlatform.CancelSubscribables()
        for(Row : GlassBreakObstaclesTeam4):
            Row.LeftPlatform.PlatformProp.Show()
            Row.RightPlatform.PlatformProp.Show()
            Row.LeftPlatform.CancelSubscribables()
            Row.RightPlatform.CancelSubscribables()
        #DamageVolume.SetDamage(80)
        DamageVolumeBottom.SetDamage(500)

    #TEAM MINIGAME CODE - Resets the glass at a certain interval. Pattern remains the same, just the visual so teams have to rembember the pattern
    ResetGlass()<suspends>:void=
        race:
            EndedEvent.Await()
            block:
                loop:
                    for(Row : GlassBreakObstaclesTeam1):
                        Row.LeftPlatform.PlatformProp.Show()
                        Row.RightPlatform.PlatformProp.Show()
                        #Row.LeftPlatform.CancelSubscribables()
                        #Row.RightPlatform.CancelSubscribables()
                    for(Row : GlassBreakObstaclesTeam2):
                        Row.LeftPlatform.PlatformProp.Show()
                        Row.RightPlatform.PlatformProp.Show()
                        #Row.LeftPlatform.CancelSubscribables()
                        #Row.RightPlatform.CancelSubscribables()
                    for(Row : GlassBreakObstaclesTeam3):
                        Row.LeftPlatform.PlatformProp.Show()
                        Row.RightPlatform.PlatformProp.Show()
                        #Row.LeftPlatform.CancelSubscribables()
                        #Row.RightPlatform.CancelSubscribables()
                    for(Row : GlassBreakObstaclesTeam4):
                        Row.LeftPlatform.PlatformProp.Show()
                        Row.RightPlatform.PlatformProp.Show()
                    Sleep(TimeToResetGlass)

    RandomizePlatforms():void=
        for(Row : GlassBreakObstaclesTeam1):
            PlatformInt := GetRandomInt(0,1)
            if(PlatformInt = 0):
                Print("Registering Left Platform")
                Row.LeftPlatform.RegisterPlatformToFall()
            else if(PlatformInt = 1):
                Print("Registering Right Platform")
                Row.RightPlatform.RegisterPlatformToFall()

        for(Row : GlassBreakObstaclesTeam2):
            PlatformInt := GetRandomInt(0,1)
            if(PlatformInt = 0):
                Print("Registering Left Platform")
                Row.LeftPlatform.RegisterPlatformToFall()
            else if(PlatformInt = 1):
                Print("Registering Right Platform")
                Row.RightPlatform.RegisterPlatformToFall()

        for(Row : GlassBreakObstaclesTeam3):
            PlatformInt := GetRandomInt(0,1)
            if(PlatformInt = 0):
                Print("Registering Left Platform")
                Row.LeftPlatform.RegisterPlatformToFall()
            else if(PlatformInt = 1):
                Print("Registering Right Platform")
                Row.RightPlatform.RegisterPlatformToFall()

        for(Row : GlassBreakObstaclesTeam4):
            PlatformInt := GetRandomInt(0,1)
            if(PlatformInt = 0):
                Print("Registering Left Platform")
                Row.LeftPlatform.RegisterPlatformToFall()
            else if(PlatformInt = 1):
                Print("Registering Right Platform")
                Row.RightPlatform.RegisterPlatformToFall()

    _OnMinigameEnded<override>(Winners: []player)<suspends>:void={}
        #DamageVolume.Disable()

    #TEAM MINIGAME MODE - Modified the win condition for glass break
    HandleCollectedEvent(Agent : agent):void=
        if:
            WinningPlayer := player[Agent]
        then:
            #var Winner : []player = array{WinningPlayer}
            var WinningPlayers : []player = array{}
            #Change from winner to the team of winners
            TeamCollection := GetPlayspace().GetTeamCollection()
            if(Team := TeamCollection.GetTeam[Agent]):
                #Get the winning team
                if(WinnerTeamMembers := TeamCollection.GetAgents[Team]):
                    #Loop through and fill the array with winning players
                    for(WinningTeamMember : WinnerTeamMembers):
                        #Convert to a player for the array
                        if(Player := player[WinningTeamMember]):
                            Manager.GoldTracker.Add(Player,GoldWinValue)
                            set WinningPlayers += array{Player}
                    #Handle Winning Players
                    Print("We have a winning team")
                    EndedEvent.Signal(WinningPlayers)

    #Since we're testing the respawn code I removed this for now. But if we don't want the respawn on glass break we can add this back in
    <#
    HandleLastPlayer()<suspends>:void=
        Print("Handle Last Player")
        race:
            EndedEvent.Await()
            block:
                loop:
                    if(GetRemainingPlayers().Length = 1, Player:= GetRemainingPlayers()[0], Fort:=Player.GetFortCharacter[]):
                        Print("One Player left in glass break")
                        Manager.UiMessage.SetText(StringToMessage("Last chance to reach the coin!"))
                        Manager.UiMessage.Show()
                        #DamageVolume.SetDamage(0)
                        DamageVolumeBottom.SetDamage(0)
                        #RegisterSubscriber(DamageVolume.AgentEntersEvent.Subscribe(HandleLastPlayerEliminatedEvent))
                        RegisterSubscriber(DamageVolumeBottom.AgentEntersEvent.Subscribe(HandleLastPlayerEliminatedEvent))
                        Fort.ReleaseFromStasis()
                        break
                    Sleep(0.2)
    
    HandleLastPlayerEliminatedEvent(Agent : agent):void=
        Print("Last player didn't make it to the end")
        if:
            WinningPlayer := player[Agent]
        then:
            var Winner : []player = array{WinningPlayer}
            EndedEvent.Signal(Winner)
    #>

    #TEAM MINIGAME - Overrides so if there are only two players and someone dies the game won't end
    #The actual win condition is in GlassBreak under HandleCollectedEvent
    GetWinners<override>()<transacts><decides>:[]player=
        var RetArray : []player = array{} 
        RetArray