# using { /Fortnite.com/Devices }
# using { /Verse.org/Simulation }
# using { /Verse.org/Random }
# using { /Fortnite.com/Characters }
# using { /Fortnite.com/Game }
# using { /Verse.org/Assets }
# using { /Verse.org/Simulation/Tags }

# hexagon_device := class(tag){}

# hexagone_tile := class(crashable_device):
#     @editable
#     OptVolume : ?volume_device = false
#     @editable
#     OptProp : ?creative_prop = false

#     FlashingMaterial : material = VerseExposed.M_White
#     DefaultMatierial : material = VerseExposed.M_HexagonTile

#     # Runs when the device is started in a running game
#     OnCrashableDeviceBegin<override>()<suspends>:void=
#         if(Volume := OptVolume?):
#             Volume.AgentEntersEvent.Subscribe(OnEntered)

#     # Called on when the trigger is triggered
#     OnEntered(InAgent : agent):void=
#         spawn:
#             TeleportTile()

#     Reset():void=
#         if(Prop := OptProp?):
#             Prop.Show()
#             Prop.SetMaterial(DefaultMatierial)

#     TeleportTile()<suspends>:void=
#         if(Prop := OptProp?):
#             Prop.SetMaterial(FlashingMaterial)
#             Sleep(0.8)
#             Prop.Hide()

# # A Verse-authored creative device that can be placed in a level
# hexagone := class(minigame_device):
#     @editable
#     LowGravityMutatorZone : mutator_zone_device = mutator_zone_device{}

#     @editable
#     Bouncer : movement_modulator_device = movement_modulator_device{}

#     @editable
#     ItemSpawnersLayer1 : item_spawner_device = item_spawner_device{}

#     @editable
#     ItemSpawnersLayer2 : []item_spawner_device = array{}

#     @editable
#     ItemSpawnersLayer3 : []item_spawner_device = array{}

#     @editable
#     StartingPlatforms : []creative_prop = array{}

#     var HexagonTiles : []hexagone_tile = array{}

#     StartingMessage<override>:string = "Picking Normal or Low Gravity Mode!"
#     Name<override>: string = "Hex-A-Gone"
#     GetVotingImage<override>()<transacts>: texture = VerseExposed.SQ_HEX

#     _Start<override>()<suspends>: void =
#         ItemSpawnersLayer1.Enable()
#         ItemSpawnersLayer1.SpawnItem()
#         for(ItemSpawner: ItemSpawnersLayer2):
#             ItemSpawner.Enable()
#         for(ItemSpawner: ItemSpawnersLayer3):
#             ItemSpawner.Enable()
#         spawn{RandomizeItemSpawners(ItemSpawnersLayer2)}
#         spawn{RandomizeItemSpawners(ItemSpawnersLayer3)}

#     OnMinigameDeviceBegin<override>()<suspends>: void =
#         #Bouncer.Disable()
#         TagObjects := GetCreativeObjectsWithTag(hexagon_device{}) #returns creative_object_interface
#         # Loop through TagObjects
#         for(TagObject : TagObjects):
#             # Make sure "TagObject" is actually a V_HexagoneTile type object by casting it to V_HexagoneTile
#             if(TaggedTile := hexagone_tile[TagObject]):
#                 set HexagonTiles += array{TaggedTile}

#     OnSafeTimeEnded<override>()<suspends>:void=
#         Print("{HexagonTiles.Length}")
#         for(Platform : StartingPlatforms):
#             Platform.Hide()
#         Manager.UiMessage.Hide()
#         RandomIntLowGrav := GetRandomInt(0,2)
#         Print("{RandomIntLowGrav}")
#         if(GetRemainingPlayers().Length <= 8 or RandomIntLowGrav = 0):
#             Manager.UiMessage.SetText(StringToMessage("Normal Gravity!"))
#             Manager.UiMessage.Show()
#         else:
#             spawn{ActivateBouncer()}
#             LowGravityMutatorZone.Enable()
#             Manager.UiMessage.SetText(StringToMessage("Low Gravity!"))
#             Manager.UiMessage.Show()

#     ActivateBouncer()<suspends>:void=
#         for(Player : GetRemainingPlayers()):
#             Bouncer.Activate(Player)
        
#     _Reset<override>()<suspends>:void=
#         LowGravityMutatorZone.Disable()
#         ItemSpawnersLayer1.Disable()
#         for(ItemSpawner: ItemSpawnersLayer2):
#             ItemSpawner.Disable()
#         for(ItemSpawner: ItemSpawnersLayer3):
#             ItemSpawner.Disable()
#         for(Platform : StartingPlatforms):
#             Platform.Show()
#         for(Tile : HexagonTiles):
#             Tile.Reset()

#     RandomizeItemSpawners(ItemSpawners : []item_spawner_device)<suspends>:void=
#         Print("Randomizing Item Spawners Bro")
#         race:
#             EndedEvent.Await()
#             loop:
#                 for(ItemSpawner : Shuffle(ItemSpawners)):
#                     ItemSpawner.SpawnItem()
#                     Sleep(15.0)
