using { /Fortnite.com/Devices }
using { /Verse.org/Simulation }
using { /Verse.org/Random }
using { /Fortnite.com/Characters }
using { /Fortnite.com/FortPlayerUtilities }
using { /Fortnite.com/Game }
using { /Verse.org/Simulation/Tags }
using { /UnrealEngine.com/Temporary/Diagnostics }
using { /UnrealEngine.com/Temporary/SpatialMath } 
using { /Verse.org/Assets }
using { /Fortnite.com/Devices/CreativeAnimation }


JumpShowdownWedge:= class<concrete>:
    var Active : logic = true
    @editable Wedge : creative_prop = creative_prop{}
    @editable ItemSpawnerWedge : item_spawner_device = item_spawner_device{}

jumpshowdown := class(minigame_device):
    @editable
    ItemSpawnRate : float = 8.0
    @editable
    JumpBeam : creative_prop = creative_prop{}
    @editable
    CrouchBeam : creative_prop = creative_prop{}
    @editable
    MovementMod : movement_modulator_device = movement_modulator_device{}
    @editable
    DamageVolumeOver : damage_volume_device = damage_volume_device{}
    @editable
    DamageVolumeUnder : damage_volume_device = damage_volume_device{}
    @editable
    JumpShowDownWedges : []JumpShowdownWedge = array{}

    @editable
    SpeedMultiplier : float = 1.5

    var JumpRotateTime : float = 6.0
    var JumpRotateTimeInitial : float = 6.0
    var JumpBeamStartLocation : transform = transform{}

    var CrouchRotateTime : float = 7.0
    var CrouchRotateTimeInitial : float = 7.0
    var CrouchBeamStartLocation : transform = transform{}

    Name<override>: string = "Jump Showdown"
    StartingMessage<override>:string = "Dodge the beams!"
    GetVotingImage<override>()<transacts>: texture = VerseExposed.SQ_JUMPSHOWDOWN

    _Start<override>()<suspends>: void =
        for(Wedge: JumpShowDownWedges):
            set Wedge.Active = true
            Wedge.ItemSpawnerWedge.Enable()
        RegisterSubscriber(DamageVolumeOver.AgentEntersEvent.Subscribe(LaunchPlayerOver))
        RegisterSubscriber(DamageVolumeUnder.AgentEntersEvent.Subscribe(LaunchPlayerUnder))
        set JumpBeamStartLocation = JumpBeam.GetTransform()
        set CrouchBeamStartLocation = CrouchBeam.GetTransform()

    _Reset<override>()<suspends>:void=
        #set ActiveWedges = array{}
        for(Wedge: JumpShowDownWedges):
            set Wedge.Active = true
            Wedge.ItemSpawnerWedge.Disable()
            Wedge.Wedge.SetMaterial(VerseExposed.MI_Blue)
            WedgeOriginalLoc :=  vector3{X := Wedge.Wedge.GetTransform().Translation.X, Y := Wedge.Wedge.GetTransform().Translation.Y, Z := 1016.0}
            if(Wedge.Wedge.TeleportTo[WedgeOriginalLoc, Wedge.Wedge.GetTransform().Rotation]):
        if(JumpBeam.TeleportTo[JumpBeamStartLocation]){}
        if(CrouchBeam.TeleportTo[CrouchBeamStartLocation]){}
        set JumpRotateTime = JumpRotateTimeInitial
        set CrouchRotateTime = CrouchRotateTimeInitial

    OnSafeTimeEnded<override>()<suspends>:void=
        Print("Test")
        spawn{RandomizeItemSpawnersOnActiveWedges()}
        spawn{ArenaFalling()}
        spawn{RotateJumpBeam()}
        spawn{SpeedUpBeams()}

    RotateJumpBeam()<suspends>:void=
        Print("Rotate Beam")
        race:
            EndedEvent.Await()
            block:
                loop:
                    DynamicObstacleMovementAndAnimation(JumpBeam, 0.0,0.0,0.0,90.0,0.0,0.0,JumpRotateTime, animation_mode.OneShot)
                    DynamicObstacleMovementAndAnimation(CrouchBeam, 0.0,0.0,0.0,90.0,0.0,0.0,CrouchRotateTime, animation_mode.OneShot)
                    Sleep(JumpRotateTime)
                    
    # Speed up the beams 4 times a game
    SpeedUpBeams()<suspends>:void=
        race:
            EndedEvent.Await()
            block:
                QuarterOfTime : float = 80.0 / 5.0
                for(X := 0..4):
                    Sleep(QuarterOfTime)
                    set JumpRotateTime = JumpRotateTime / SpeedMultiplier
                    set CrouchRotateTime = CrouchRotateTime / SpeedMultiplier
                    Print("Speed up beams")

    LaunchPlayerOver(InAgent : agent):void=
        LaunchPlayer(InAgent, JumpBeam)

    LaunchPlayerUnder(InAgent : agent):void=
        LaunchPlayer(InAgent, CrouchBeam)

    LaunchPlayer(InAgent : agent, Beam : creative_prop):void=
        # Damaged player is self       #Damagingchar is enemies character
        if(Player := player[InAgent], FortChararacter := Player.GetFortCharacter[]):
            var TargetRotation : rotation = rotation{}
            NewLocation := MovementMod.GetTransform().Translation
            BeamRotation := Beam.GetTransform().Rotation

            BeamForward := BeamRotation.GetLocalForward()
            PlayerLoc := FortChararacter.GetTransform().Translation
            
            Diff := PlayerLoc - Beam.GetTransform().Translation
            if(DirectionVectorToEnemy := Diff.MakeUnitVector[]):
                DotAngle := DotProduct(DirectionVectorToEnemy, BeamForward)
                Print("{DotAngle}")
                BeamYawPitchRoll := BeamRotation.GetYawPitchRollDegrees()
                if(BeamYaw := BeamYawPitchRoll[0], BeamPitch := BeamYawPitchRoll[1], BeamRoll := BeamYawPitchRoll[2]):
                    if (DotAngle <= 0.9):
                        set TargetRotation = MakeRotationFromYawPitchRollDegrees(BeamYaw - 90.0, BeamPitch, BeamRoll)
                    else:
                        set TargetRotation = MakeRotationFromYawPitchRollDegrees(BeamYaw + 90.0, BeamPitch, BeamRoll)

                    
            # Set the player that slapped this player last
            if (MovementMod.TeleportTo[NewLocation, TargetRotation]) :
                MovementMod.Activate(Player)
                Print("Launch")

    #Animation Controller code for defining an end keyframe and then animating that result.
    #This can be used for any type of obstacle for movement and rotation.
    DynamicObstacleMovementAndAnimation(ObstacleToMove : creative_prop, AmountToMoveX : float, AmountToMoveY : float, AmountToMoveZ : float,
    RotationX : float, RotationY : float, RotationZ : float, AnimationTime : float, AnimationModeType : animation_mode)<suspends>:void=
        if:
            AnimationController := ObstacleToMove.GetAnimationController[] 
        then:
            #Create the keyframe
            KeyFrame := keyframe_delta :
                    DeltaLocation := vector3{X:=AmountToMoveX, Y:= AmountToMoveY, Z:= AmountToMoveZ}
                    DeltaRotation := MakeRotationFromYawPitchRollDegrees(RotationX, RotationY, RotationZ)
                    # DeltaScale := vector3{X:=AmountToMoveX, Y:= AmountToMoveY, Z:= AmountToMoveZ}
                    Time := AnimationTime
            #Create keyframe Array
            KeyFrames : []keyframe_delta = array { KeyFrame }
            #Set Animation To Animation Controller
            AnimationController.SetAnimation(KeyFrames, ?Mode:=AnimationModeType)
            #Play Animation
            AnimationController.Play()
        return

    ArenaFalling()<suspends>:void=
        race:
            EndedEvent.Await()
            block:
                ShuffledWedges := Shuffle(JumpShowDownWedges)
                var CurrentArenaIndex : int = 0
                RingAmount := ShuffledWedges.Length
                TimePerRing : float = 80.0 / (RingAmount * 1.0)
                Print("Time Before Arena Ring Falls = {TimePerRing}")
                Sleep(TimePerRing / 2.0)
                #leave last ring
                for (X := 0..RingAmount-2):
                    #Sleep for a short period of time
                    Print("Starting Countdown For Arena To Fall")
                    if(Arena := ShuffledWedges[CurrentArenaIndex], ArenaWedge := Arena.Wedge):
                        #Sleep(TimePerRing / 2.0)
                        ArenaWedge.SetMaterial(VerseExposed.M_JS_FlashingBlue)
                        Sleep(TimePerRing / 2.0)
                        set Arena.Active = false
                        NewArenaLoc :=  vector3{X := ArenaWedge.GetTransform().Translation.X, Y := ArenaWedge.GetTransform().Translation.Y, Z := ArenaWedge.GetTransform().Translation.Z - 2500.0 }
                        ArenaWedge.MoveTo(NewArenaLoc, ArenaWedge.GetTransform().Rotation, 5.5)
                        ArenaWedge.SetMaterial(VerseExposed.MI_Blue)
                        set CurrentArenaIndex += 1
                    else:
                        return

    RandomizeItemSpawnersOnActiveWedges()<suspends>:void=
        race:
            EndedEvent.Await()
            block:
                loop:
                    Print("Finding Wedge")
                    StillAlivePlayers := for(Player : GetRemainingPlayers()) { Player }
                    StillAliveWedges := for(Wedge : Shuffle(JumpShowDownWedges), Wedge.Active = true) { Wedge }
                    if(RandomActiveWedge := StillAliveWedges[0]):
                        RandomActiveWedge.ItemSpawnerWedge.SpawnItem()
                    Sleep(ItemSpawnRate)