#     __________  ___    ____ 
#    / ____/ __ \/   |  / __ \
#   / /_  / / / / /| | / / / /
#  / __/ / /_/ / ___ |/ /_/ / 
# /_/    \____/_/  |_/_____/  
#

using { /Fortnite.com/Devices }
using { /Verse.org/Simulation }
using { /Verse.org/Random }
using { /Fortnite.com/Characters }
using { /Fortnite.com/FortPlayerUtilities }
using { /Fortnite.com/Game }
using { /Verse.org/Simulation/Tags }
using { /UnrealEngine.com/Assets }
using { /UnrealEngine.com/Temporary/Diagnostics }
using { /UnrealEngine.com/Temporary/SpatialMath } 
using { /Verse.org/Assets }
using { /Fortnite.com/Devices/CreativeAnimation }
using { /UnrealEngine.com/Temporary/UI }
using { /Fortnite.com/UI }
using { /Verse.org/Colors }
using {VerseExposed}

foad_infection := class(minigame_device):
    @editable
    AFKTime : int = 8

    @editable
    BackgroundMusic :audio_player_device = audio_player_device{}

    #@editable
    #TimerDevice : timer_device = timer_device{}

    # @editable
    # HealthPowerUp :health_powerup_device = health_powerup_device{}

    @editable
    ItemGranterLimitedMats : item_granter_device = item_granter_device{}

    @editable
    ItemGranterFOAD : item_granter_device = item_granter_device{}

    @editable
    FOADClass : class_and_team_selector_device = class_and_team_selector_device {}

    @editable
    FOADScoreHudDevice : hud_message_device = hud_message_device{}

    @editable
    FOADVFXPowerUp : []visual_effect_powerup_device = array{}

    @editable
    WarningCircleVFXPowerUp : []visual_effect_powerup_device = array{}

    @editable
    WarningCircleResetVFXPowerUp: visual_effect_powerup_device = visual_effect_powerup_device{}

    @editable
    IntroCinematic : cinematic_sequence_device = cinematic_sequence_device{}

    @editable
    InputControls : input_trigger_device = input_trigger_device{}

    @editable
    GameplayCamera : gameplay_camera_orbit_device = gameplay_camera_orbit_device{}

    @editable
    ExplosiveDevice : []explosive_device = array{}

    @editable
    StartingPillarProp : creative_prop = creative_prop{}

    @editable
    CeilingProp : creative_prop = creative_prop{}

    @editable
    StartingSpawnPointFOAD : creative_prop = creative_prop{}

    @editable
    ExplosionProp : creative_prop_asset = DefaultCreativePropAsset

    @editable
    AudioPlayerExplosion :[]audio_player_device = array{}

    @editable
    DBNO : down_but_not_out_device = down_but_not_out_device{}

    @editable
    MovementSpeedMod : movement_modulator_device = movement_modulator_device{}

    var PlayerFOAD : [player]logic = map{}
    var PlayerIsExploding : [player]logic = map{}
    var ExplosionCharged : [player]logic = map{}
    var PowerUpIndex : int = 0
    var PowerUpWarningCircleIndex : int = 0  
    var PowerUpWarningChargeIndex : int = 0  
    var ExplosiveDeviceIndex : int = 0
    var AudioDeviceIndex : int = 0
    var ExplosionReadyCanvas : [player]?canvas = map{}

    Name<override>: string = "FOAD INFECTION"
    StartingMessage<override>:string = "Run away from FOADs!"
    GetVotingImage<override>()<transacts>: texture = VerseExposed.SQ_FOAD


    _Start<override>()<suspends>: void =
        Print("Do Beginning Cinematic")
        BackgroundMusic.Enable()
        BackgroundMusic.Play()
        #TimerDevice.Enable()
        DBNO.Enable()
        InputControls.Enable()
        GameplayCamera.Enable()
        MovementSpeedMod.Enable()

        if(StartingPillarProp.TeleportTo[vector3{X:=7190.0,Y:=-19505.0,Z:=1179.0}, rotation{}]):
        if(CeilingProp.TeleportTo[vector3{X:= 7168.0, Y:= -18958.0, Z:= 2816.0}, rotation{}]):

        RegisterSubscriber(InputControls.PressedEvent.Subscribe(ExplodeFOADPressed))
        RegisterSubscriber(IntroCinematic.StoppedEvent.Subscribe(CinematicEnd))
        RegisterSubscriber(DBNO.AgentDownedEvent.Subscribe(OnPlayerDowned))
        #RegisterSubscriber(TimerDevice.SuccessEvent.Subscribe(OnTimerSuccess))
        RegisterSubscriber(GetPlayspace().PlayerRemovedEvent().Subscribe(OnPlayerLeftFOAD))
        
        TeamCollection := GetPlayspace().GetTeamCollection()
        for(Player : GetRemainingPlayers()):
            if(Team78 := TeamCollection.GetTeams()[77]):
                option{TeamCollection.AddToTeam[Player, Team78]}
            #HealthPowerUp.Pickup()
            if(set PlayerFOAD[Player] = false):
            if(FortChar:= Player.GetFortCharacter[]):
                FortChar.PutInStasis(stasis_args{AllowTurning := true, AllowFalling := true})
        
    OnSafeTimeEnded<override>()<suspends>:void=
        #TimerDevice.Reset()
        #TimerDevice.Start()
        #HealthPowerUp.Pickup()
        spawn{SelectInitialFOAD()}
        spawn{HandleFoadWinners()}
        spawn{ItemGranters()}

    _OnMinigameEnded<override>(Winners: []player)<suspends>:void=
        BackgroundMusic.Disable()
        DBNO.Disable()
        FOADScoreHudDevice.Hide()
        MovementSpeedMod.Disable()
        #TimerDevice.Reset()
        #TimerDevice.Disable()
        InputControls.UnregisterAll()
        InputControls.Disable()
        GameplayCamera.RemoveFromAll()
        GameplayCamera.Disable()
        # for(Player -> Logic : PlayerFOAD, Agent := agent[Player]):
        #     #Print("Removing Explosion UI")
        #     RemoveExplosionReadyUI(Agent)
        for(Player :GetRemainingPlayers(), Agent := agent[Player]):
            RemoveExplosionReadyUI(Agent)

    _Reset<override>()<suspends>:void=
        if(StartingPillarProp.TeleportTo[vector3{X:=7190.0,Y:=-19505.0,Z:=-260.0}, rotation{}]):
        if(CeilingProp.TeleportTo[vector3{X:= 11403.0, Y:= -18958.0, Z:= 1919.0}, rotation{}]):
        for(Player :GetRemainingPlayers(), Agent := agent[Player]):
            RemoveExplosionReadyUI(Agent)
        set PlayerFOAD = map{}

    OnPlayerDowned(Agent : agent):void=
        Print("On Player Downed")
        if(Player := player[Agent]):
            if(set PlayerFOAD[Player] = true): 
            SetScoreHUD()
            DBNO.Revive(Agent)
            FOADClass.ChangeClass(Agent)
            AssignPlayerRandomFOADVFX(Agent)
            InputControls.Register(Agent)
            GameplayCamera.AddTo(Agent)
            MovementSpeedMod.Activate(Agent)
            spawn:
                ChargeExplosion(Agent)

    TurnPlayerIntoFOAD(Agent: agent):void=
        Print("Turning Player into Foad")
        if(Player := player[Agent]):
            if(set PlayerFOAD[Player] = true): 
            SetScoreHUD()
            #DBNO.Revive(Agent)
            FOADClass.ChangeClass(Agent)
            AssignPlayerRandomFOADVFX(Agent)
            InputControls.Register(Agent)
            GameplayCamera.AddTo(Agent)
            MovementSpeedMod.Activate(Agent)
            spawn:
                ChargeExplosion(Agent)

    SelectInitialFOAD()<suspends>:void=
        race:
            EndedEvent.Await()
            block:
                Sleep(0.1)
                # IntroCinematic.Play()
                if(FirstFOADPlayer := Shuffle(GetRemainingPlayers())[0], Agent :=  agent[FirstFOADPlayer], FortChar := FirstFOADPlayer.GetFortCharacter[]):
                    if(set PlayerFOAD[FirstFOADPlayer] = true): 
                    Agent.Respawn(StartingSpawnPointFOAD.GetTransform().Translation, StartingSpawnPointFOAD.GetTransform().Rotation)

                    SetScoreHUD()
                    FortChar.PutInStasis(stasis_args{AllowTurning := false, AllowFalling := true})
                    FOADClass.ChangeClass(Agent)
                    #FortChar.SetVulnerability(false) 
                    AssignPlayerRandomFOADVFX(Agent)
                    Sleep(0.1)
                    Manager.ResetItemRemover.Remove(Agent)
                    IntroCinematic.Play()
                    Sleep(4.0)
                    ChargeExplosion(Agent)
                    InputControls.Register(Agent)
                    GameplayCamera.AddTo(Agent)
                    MovementSpeedMod.Activate(Agent)
                    #AFK Check
                    spawn{WatchInitialFoadStandStill(FirstFOADPlayer,Agent, FortChar)}
                    spawn{HealPlayers()}

    #Check for AFK Players
    WatchInitialFoadStandStill(Player : player,Agent: agent, Fort : fort_character)<suspends>:void=
        var lastPosition : vector3 = Fort.GetTransform().Translation
        var initalTime : int = 0
        race:
            EndedEvent.Await()
            loop:
                ActualPosition := Fort.GetTransform().Translation
                if(Distance(ActualPosition,lastPosition) < 1.0):
                    Print("Same Position bro")
                    set initalTime = initalTime + 1
                    if(initalTime >= AFKTime):
                        for(NewPlayer -> Logic : PlayerFOAD):
                            if(PlayerFOAD[NewPlayer] = false):
                                if:
                                    NewAgent := agent[NewPlayer]
                                then:
                                    Print("Making new random agent FOAD")
                                    if(set PlayerFOAD[NewPlayer] = true): 
                                        SetScoreHUD()
                                        FOADClass.ChangeClass(NewAgent)
                                        AssignPlayerRandomFOADVFX(NewAgent)
                                        InputControls.Register(NewAgent)
                                        GameplayCamera.AddTo(NewAgent)
                                        MovementSpeedMod.Activate(NewAgent)
                                        spawn:
                                            ChargeExplosion(NewAgent)
                            break
                else:
                    Print("First Foad Player Moved. Breaking out of Loop")
                    break
                Sleep(1.0)

    HealPlayers()<suspends>:void=
        race:
            EndedEvent.Await()
            loop:
                Sleep(1.0)
                for(Player: GetRemainingPlayers(), FortChar:= Player.GetFortCharacter[]):
                    FortChar.SetHealth(FortChar.GetHealth() + 1.0)

    SetScoreHUD():void=
        var PlayerScore : int = 0
        var FOADScore : int = 0
        for(Player -> Logic : PlayerFOAD):
            if(PlayerFOAD[Player] = true):
                set FOADScore += 1
            else: 
                set PlayerScore += 1
        FOADScoreHudDevice.SetText(StringToMessage("{PlayerScore}\n{FOADScore}"))
        FOADScoreHudDevice.Show()

    CinematicEnd():void=
        for(Player : GetPlayspace().GetPlayers(), FortChar:= Player.GetFortCharacter[]):
            FortChar.ReleaseFromStasis()
        #Fix Glitch wehre releasing from stasis cancels no emote mutator zone
        Manager.NoEmotingMutatorZone.Disable()
        Manager.NoEmotingMutatorZone.Enable()

    SpawnExplosion(Agent : agent)<suspends>:void=
        race:
            EndedEvent.Await()
            block:
                if(Player := player[Agent]):
                    if(FortChar:= Player.GetFortCharacter[]):
                        
                        WarningCircleResetVFXPowerUp.Pickup(Agent)
                        var NewTransform : transform = FortChar.GetTransform()
                        set NewTransform.Translation = FortChar.GetTransform().Translation  + vector3{Z:=-70.0}
                        NewExplosionProp := SpawnProp(ExplosionProp, NewTransform)

                        PlayExplosionAudio(NewTransform)
                        #ExplosionAudio.Play(Agent)

                        if(Explosive := ExplosiveDevice[ExplosiveDeviceIndex]):
                            if:
                                Explosive.TeleportTo[NewTransform.Translation, NewTransform.Rotation]
                            Explosive.Reset()
                            Explosive.Explode(Agent)

                            if(ExplosiveDeviceIndex >= ExplosiveDevice.Length - 1):
                                set ExplosiveDeviceIndex = 0
                            else:
                                set ExplosiveDeviceIndex += 1

                            defer:
                                Explosive.Reset()

                        if(set PlayerIsExploding[Player] = false):

                        Sleep(0.25)
                        if(RealProp:= NewExplosionProp(0)?):
                            spawn:
                                DespawnExplosionProp(RealProp)
                            spawn:
                                ChargeExplosion(Agent)
                        # VFX.Cancel()

    DespawnExplosionProp(Prop : creative_prop )<suspends>:void=
            Sleep(2.5) 
            Prop.Dispose()
            Print("Canceling VFX")

    ChargeExplosion(Agent : agent)<suspends>:void=
        race:
            EndedEvent.Await()
            block:
                AssignPlayerWarningCircleVFX(Agent)
                Sleep(3.5)
                # InputControls.Register(Agent)
                if(Player := player[Agent]):
                    if(set ExplosionCharged[Player] = true):
                        ActivateExplosionReadyUI(Agent)
                        Print("Explosion Charged")

    ExplodeFOADPressed(Agent:agent):void=

        if(Player := player[Agent], ExplosionCharged[Player] = false):
            Print("Explosion Not Charged.")
            return

        if(Player := player[Agent], PlayerIsExploding[Player] = true):
            Print("Don't fire more - explosion in progress")
            return
        AssignPlayerWarningCircleVFX(Agent)
        if(Player := player[Agent], set PlayerIsExploding[Player] = true):
            if(set  ExplosionCharged[Player] = false):
            RemoveExplosionReadyUI(Agent)
            spawn:
                SpawnExplosion(Agent)

    AssignPlayerFOADVFX(Agent : agent):void=
        if(PowerUp := FOADVFXPowerUp[PowerUpIndex]):
            PowerUp.Pickup(Agent)

        #Increment Powerup Index
        if(PowerUpIndex >= FOADVFXPowerUp.Length - 1):
            set PowerUpIndex = 0
        else:
            set PowerUpIndex += 1

    AssignPlayerRandomFOADVFX(Agent : agent):void=
        DiceRoll := GetRandomInt(0, FOADVFXPowerUp.Length - 1)
        if(PowerUp := FOADVFXPowerUp[DiceRoll]):
            PowerUp.Pickup(Agent)

    AssignPlayerWarningCircleVFX(Agent : agent):void=
        if(PowerUp := WarningCircleVFXPowerUp[PowerUpWarningCircleIndex]):
            PowerUp.Pickup(Agent)
        #Increment Powerup Index
        if(PowerUpWarningCircleIndex >= WarningCircleVFXPowerUp.Length - 1):
            set PowerUpWarningCircleIndex = 0
        else:
            set PowerUpWarningCircleIndex += 1

    PlayExplosionAudio(AudioLocation : transform):void=
        if(AudioDevice := AudioPlayerExplosion[AudioDeviceIndex]):
            if(AudioDevice.TeleportTo[AudioLocation]):
                AudioDevice.Play()
        if(AudioDeviceIndex >= AudioPlayerExplosion.Length - 1):
            set AudioDeviceIndex = 0
        else:
            set AudioDeviceIndex += 1

    #Explosion Charged UI
    ActivateExplosionReadyUI(Agent : agent) : void =
        if (Player := player[Agent], PlayerUI := GetPlayerUI[Player]):
            if (MyUI := ExplosionReadyCanvas[Player]?):
                Print("Removing Explosion Canvas")
                PlayerUI.RemoveWidget(MyUI)
                if (set ExplosionReadyCanvas[Player] = false) {}
            NewUI := CreateExplosionReadyCanvas()
            PlayerUI.AddWidget(NewUI)
            if (set ExplosionReadyCanvas[Player] = option{NewUI}) {}

    RemoveExplosionReadyUI(Agent : agent):void=
        if (InPlayer := player[Agent], PlayerUI := GetPlayerUI[InPlayer]):
            Print("Removing Explosion Canvas Beep")
            if (MyUI := ExplosionReadyCanvas[InPlayer]?):
                PlayerUI.RemoveWidget(MyUI)
                if (set ExplosionReadyCanvas[InPlayer] = false) {}

    CreateExplosionReadyCanvas(): canvas =
        ExplosionReadyTexture  : texture_block = texture_block{ DefaultImage := VerseExposed.T_Press_Fire_To_Explode, DefaultDesiredSize := vector2{ X := 512.0, Y := 73.0}, DefaultTint := color{R:=1.0, G:=0.1, B:=0.0}}
        MyDamageText : text_block = text_block{DefaultText := StringToMessage("Press Fire to Explode"), DefaultTextColor := NamedColors.Gray}
        Print("Creating Explosion Charged UI")
        MyCanvas : canvas = canvas:
            Slots := array:
                canvas_slot: 
                    Anchors := anchors{Minimum := vector2{X := 0.5, Y := 0.85}, Maximum := vector2{X := 0.5, Y :=0.85}}
                    Alignment := vector2{X := 0.5, Y := 0.5}
                    Widget := overlay:
                        Slots := array:  
                            overlay_slot:
                                Widget := ExplosionReadyTexture         
        return MyCanvas

    _OnGenericEndTimerSuccessReturnArray<override>():[]player=
        Print("Timer")
        var Winners : []player = array{}
        for(Player -> Logic : PlayerFOAD):
            if(PlayerFOAD[Player] = false):
                set Winners += array{Player}
        for(Player -> Logic : PlayerFOAD, Agent := agent[Player]):
            RemoveExplosionReadyUI(Agent)
        return Winners

    HandleFoadWinners()<suspends>:void=
        race:
            EndedEvent.Await()
            block:
                loop:
                    Sleep(0.2)
                    var PlayersAlive : int = 0
                    var RetArray : []player = array{}
                    for(Player -> Logic : PlayerFOAD):
                        if(PlayerFOAD[Player] = true):
                            set RetArray += array{Player}
                            #set PlayersAlive = PlayersAlive + 1
                    # 0 NON FOAD Players Alive
                    #Print("RetArr Length {RetArray.Length} --- PlayerFoad Length{PlayerFOAD.Length}")
                    if(RetArray.Length = PlayerFOAD.Length):
                         # Allow for draw to happen
                         for(Player -> Logic : PlayerFOAD, Agent := agent[Player]):
                            RemoveExplosionReadyUI(Agent)
                        EndedEvent.Signal(RetArray)

    ItemGranters()<suspends>:void=
        race:
            EndedEvent.Await()
            block:
                loop:
                    Sleep(10.0)
                    for(Player : GetRemainingPlayers()):
                        ItemGranterLimitedMats.GrantItem(Player)
                        ItemGranterFOAD.CycleToRandomItem(Player)
                        ItemGranterFOAD.CycleToRandomItem(Player)

    OnPlayerLeftFOAD(PlayerLeaving: player):void=
        var NewPlayerFOAD:[player]logic = map{}
        # Concatenate Keys from ExampleMap into NewMap, excluding ElementToRemove
        for (Key -> Value : PlayerFOAD, Key <> PlayerLeaving):
            set NewPlayerFOAD = ConcatenateMaps(NewPlayerFOAD, map{Key => Value})
        Print("Old size = {PlayerFOAD.Length} -- New Size = {NewPlayerFOAD.Length}")
        #If 0 FOADS, make a new Foad
        var FoadAliveLogic :logic = false
        var RetArray : []player = array{}
        for(Player -> Logic : NewPlayerFOAD):
            set RetArray += array{Player}
            if(PlayerFOAD[Player] = true):
                set FoadAliveLogic = true
        if(FoadAliveLogic = false):
            EndedEvent.Signal(RetArray)
        else:
            set PlayerFOAD = NewPlayerFOAD
            SetScoreHUD()