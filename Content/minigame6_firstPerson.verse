using { /Fortnite.com/Devices }
using { /Fortnite.com/AI}
using { /Fortnite.com/Teams}
using { /Verse.org/Simulation }
using { /Verse.org/Random }
using { /Fortnite.com/Characters }
using { /Fortnite.com/FortPlayerUtilities }
using { /Fortnite.com/Game }
using { /Fortnite.com/Teams }
using { /Verse.org/Simulation/Tags }
using { /UnrealEngine.com/Temporary/Diagnostics }
using { /UnrealEngine.com/Temporary/SpatialMath } 
using { /Verse.org/Assets }


# A Verse-authored creative device that can be placed in a level
firstPerson := class(minigame_device):
    @editable FPCamera : gameplay_camera_first_person_device = gameplay_camera_first_person_device{}

    Name<override>: string = "First Person Mode"
    StartingMessage<override>:string = ""
    GetVotingImage<override>()<transacts>: texture = VerseExposed.SQ_FIRSTPERSON

    _Start<override>()<suspends>: void = 
        FPCamera.Enable()
        FPCamera.AddToAll()

    _OnMinigameEnded<override>(Winners: []player)<suspends>:void=
        FPCamera.RemoveFromAll()
        FPCamera.Disable()