using { /Fortnite.com/Devices }
using { /Fortnite.com/AI}
using { /Fortnite.com/Teams}
using { /Verse.org/Simulation }
using { /Verse.org/Random }
using { /Fortnite.com/Characters }
using { /Fortnite.com/FortPlayerUtilities }
using { /Fortnite.com/Game }
using { /Fortnite.com/Teams }
using { /Verse.org/Simulation/Tags }
using { /UnrealEngine.com/Temporary/Diagnostics }
using { /UnrealEngine.com/Temporary/SpatialMath } 
using { /Verse.org/Assets }

# A Verse-authored creative device that can be placed in a level
turtle_wars := class(minigame_device):
    var ScoreToWin : int = 10
    var RespawnCount : int = 0
    var RespawnFloat :float = 400.0
    @editable PumpWarsProp : creative_prop = creative_prop{}
    @editable DBNO : down_but_not_out_device = down_but_not_out_device{}
    @editable BottomHUDDeviceIndividual : hud_message_device = hud_message_device{}
    #@editable TimerDevice : timer_device = timer_device{}

    Name<override>: string = "Turtle Wars"
    StartingMessage<override>:string = "Hit the Elimination Goal!"
    GetVotingImage<override>()<transacts>: texture = VerseExposed.SQ_TURTLEWARS_Vote
    DisplayIcon<override>: ?texture = option{VerseExposed.SQ_TURTLEWARS}

    var PlayerKills : [player]int = map{}
    var PlayerLastHitBy: [player]player = map{}
    
    _Start<override>()<suspends>: void =
        PlayerCount := GetRemainingPlayers().Length
        if(PlayerCount>=10):
            set ScoreToWin = 10
        else if(PlayerCount>=5):
            set ScoreToWin = 7
        else:
            set ScoreToWin = 5

        #TimerDevice.Enable()
        DBNO.Enable()
        RegisterSubscriber(DBNO.AgentDownedEvent.Subscribe(OnPlayerDowned))
        RegisterSubscriber(DBNO.AgentRevivedEvent.Subscribe(OnPlayerRevived))
        #RegisterSubscriber(TimerDevice.SuccessEvent.Subscribe(OnTimerSuccess))
        if(PumpWarsProp.TeleportTo[vector3{X:= 499.820837, Y:= 589.912472, Z:= 937.777006}, rotation{}]):
        TeamCollection := GetPlayspace().GetTeamCollection()
        if(Team78 := TeamCollection.GetTeams()[77]):
            for(Player : GetRemainingPlayers(), Fort := Player.GetFortCharacter[]):
                BottomHUDDeviceIndividual.Show(Player,StringToMessage("0/{ScoreToWin} Elims"))
                RegisterSubscriber(Fort.DamagedEvent().Subscribe(HandlePlayerHit))
                RegisterSubscriber(Fort.DamagedShieldEvent().Subscribe(HandlePlayerHitShield))
                if(set PlayerKills[Player] = 0) {}
                if(set PlayerLastHitBy[Player] = Player) {}
                option{TeamCollection.AddToTeam[Player, Team78]}

    OnSafeTimeEnded<override>()<suspends>:void=
        #TimerDevice.Reset()
        #TimerDevice.Start()
        spawn{HandleWinners()}

    _OnMinigameEnded<override>(Winners: []player)<suspends>:void=
        #TimerDevice.Reset()
        #TimerDevice.Disable()
        for(Player:GetRemainingPlayers()):
            if(Fort:=Player.GetFortCharacter[]):
                Fort.SetVulnerability(true)    

    _Reset<override>()<suspends>:void=
        set RespawnCount = 0
        set RespawnFloat = 400.0
        set PlayerKills = map{}
        set PlayerLastHitBy = map{}
        DBNO.Disable()
        if(PumpWarsProp.TeleportTo[vector3{X:= 499.820837, Y:= 589.912472, Z:= -3158.222994}, rotation{}]):
        Manager.ClearLayer2Trigger.Trigger()

    # OnTimerSuccess(MaybeAgent : ?agent):void=
    #     Print("No winners")
    #     EndedEvent.Signal(array{})

    HandlePlayerHit(DamageResult : damage_result):void=
        HandlePlayerHitCustom(DamageResult)

    HandlePlayerHitShield(DamageResult : damage_result ) : void =
        HandlePlayerHitCustom(DamageResult)

    HandlePlayerHitCustom( DamageResult : damage_result ) : void =
        if:
            Target := DamageResult.Target
            Instigator := DamageResult.Instigator?
            InstigatorAgent := Instigator.GetInstigatorAgent[]
            InstigatorFort := InstigatorAgent.GetFortCharacter[]
            InstigatorPlayer := player[InstigatorAgent]
            CharacterWhoWasHitFort := fort_character[Target]
            CharacterWhoWasHitAgent := CharacterWhoWasHitFort.GetAgent[]
            CharacterWhoWasHitPlayer := player[CharacterWhoWasHitAgent]
        then:
            if(InstigatorAgent = CharacterWhoWasHitAgent):
                #Print("Same Agent: Usually from storm damage")
            else:
                #Print("Setting last player hit by")
                if(set PlayerLastHitBy[CharacterWhoWasHitPlayer] = InstigatorPlayer):

    OnPlayerDowned(Agent : agent):void=
        Print("On Player Downed")
        if(Player := player[Agent]):
            if(LastPlayerToHit := PlayerLastHitBy[Player], not LastPlayerToHit = Player):
                Manager.EliminationAccolade.Award(LastPlayerToHit)
                if(NewKillVal := PlayerKills[LastPlayerToHit]+1):
                    Print("NewKillVal= {NewKillVal}")
                    BottomHUDDeviceIndividual.Show(LastPlayerToHit,StringToMessage("{NewKillVal}/{ScoreToWin} Elims"))
                    if(set PlayerKills[LastPlayerToHit] = NewKillVal):
            spawn{TeleportPlayer(Player)}

    OnPlayerRevived(Agent : agent):void=
        Print("On Player Revived")
        if:
            Fort := Agent.GetFortCharacter[]
        then:
            spawn{DelayHealing(Fort)}
    
    DelayHealing(Fort : fort_character)<suspends>:void=
        Sleep(0.3)
        Fort.SetHealth(100.0)
        Fort.SetShield(100.0)

    TeleportPlayer(Player : player)<suspends>:void=
        Print("Teleport and heal bro")
        if(RespawnCount >= GetRemainingPlayers().Length):
            #Swap up and down spawn positions
            if(RespawnFloat=400.0):
                set RespawnFloat = 200.0
            else:
                set RespawnFloat = 400.0
            set RespawnCount = 0
        else:
            set RespawnCount += 1
        Print("RespawnFloat={RespawnFloat}")
        FurthestBarrierFromPlayers := array{FindFurthestBarrierFromPlayers(Manager.Box1Barriers, false)}
        #FurthestBarrierFromPlayers.BulkTeleport(array{Player}, Manager.Teleporter, ?TranslationOffset := vector3{Z := RespawnFloat}, ?YawOffset := -90.0)
        spawn{ReviveAndSpawnProtection(Player)}

    ReviveAndSpawnProtection(Player: player)<suspends>:void=
        Sleep(0.1)
        DBNO.Revive(Player)
        if(Fort:=Player.GetFortCharacter[]):
            Fort.SetVulnerability(false)
            Print("Invincible")
            Sleep(1.0)
            Print("Not Invincible")
            Fort.SetVulnerability(true)

    HandleWinners()<suspends>:void=
        race:
            EndedEvent.Await()
            block:
                loop:
                    Sleep(0.2)
                    var RetArray : []player = array{}
                    for(Player -> Integer : PlayerKills):
                        if(Integer >= ScoreToWin):
                            Print("Someone got to {ScoreToWin} Kills")
                            Manager.ClearLayer2Trigger.Trigger()
                            Manager.BottomHUDDevice.Show(BuildPlayerMessage("", Player, " WINS!"))
                            for(StasisPlayer : GetRemainingPlayers()):
                                if:
                                    Fort:= StasisPlayer.GetFortCharacter[]
                                then:
                                    Fort.PutInStasis(stasis_args{AllowEmotes := false})
                            set RetArray += array{Player}
                            EndedEvent.Signal(RetArray)
                            
                        