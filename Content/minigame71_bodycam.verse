using { /Fortnite.com/Devices }
using { /Fortnite.com/AI}
using { /Fortnite.com/Teams}
using { /Verse.org/Simulation }
using { /Verse.org/Random }
using { /Fortnite.com/Characters }
using { /Fortnite.com/FortPlayerUtilities }
using { /Fortnite.com/Game }
using { /Fortnite.com/Teams }
using { /Verse.org/Simulation/Tags }
using { /UnrealEngine.com/Temporary/Diagnostics }
using { /UnrealEngine.com/Temporary/SpatialMath } 
using { /Verse.org/Colors }
using { /Verse.org/Assets }
using { /UnrealEngine.com/Temporary/Diagnostics }



bodycam := class(minigame_device):
    @editable BackgroundMusic :audio_player_device = audio_player_device{}
    @editable HudNoReticle : hud_controller_device = hud_controller_device{}
    @editable ItemGranterDefaultWeapon : item_granter_device = item_granter_device{}
    @editable ItemGrantersInOrder : []item_granter_device = array{}
    @editable DayPostProcessProp : creative_prop = creative_prop{}
    @editable DBNO : down_but_not_out_device = down_but_not_out_device{}
    @editable DefaultOrbitCamera : gameplay_camera_orbit_device = gameplay_camera_orbit_device{}
    @editable InputTriggerAimIn : input_trigger_device = input_trigger_device{}
    @editable ZoomInOrbitCamera : gameplay_camera_orbit_device = gameplay_camera_orbit_device{}
    @editable IntroCinematic : cinematic_sequence_device = cinematic_sequence_device{}
    @editable Collectible : collectible_object_device = collectible_object_device{}

    Name<override>: string = "BODYCAM Gun Game"
    StartingMessage<override>:string = "Get A Kill With All 10 Weapons To Win!"
    GetVotingImage<override>()<transacts>: texture = VerseExposed.SQ_BODYCAM
    DisplayIcon<override>: ?texture = option{VerseExposed.TransparentTexture}

    var PlayerKills : [player]int = map{}
    var PlayerLastHitBy: [player]player = map{}
    

    _Start<override>()<suspends>: void =
        if(DayPostProcessProp.TeleportTo[vector3{X:=29297.380442,Y:=24195.435289, Z:= 3255.797211}, rotation{}]):
        HudNoReticle.Enable()
        IntroCinematic.Play()
        BackgroundMusic.Enable()
        BackgroundMusic.Play()
        DefaultOrbitCamera.Enable()
        DefaultOrbitCamera.AddToAll()
        InputTriggerAimIn.Enable()
        ZoomInOrbitCamera.Enable()
        DBNO.Enable()
        RegisterSubscriber(InputTriggerAimIn.PressedEvent.Subscribe(HandleAimIn))
        RegisterSubscriber(InputTriggerAimIn.ReleasedEvent.Subscribe(HandleAimInReleased))
        RegisterSubscriber(DBNO.AgentDownedEvent.Subscribe(OnPlayerDowned))
        RegisterSubscriber(DBNO.AgentRevivedEvent.Subscribe(OnPlayerRevived))
        Collectible.Show()
        Collectible.RespawnForAll()
        RegisterSubscriber(Collectible.CollectedEvent.Subscribe(HandleCollectedEvent))
        TeamCollection := GetPlayspace().GetTeamCollection()
        if(Team78 := TeamCollection.GetTeams()[77]):
            for(Player : GetRemainingPlayers(), Fort := Player.GetFortCharacter[]):
                Fort.PutInStasis(stasis_args{})
                RegisterSubscriber(Fort.DamagedEvent().Subscribe(HandlePlayerHit))
                InputTriggerAimIn.Register(Player)
                RegisterSubscriber(Fort.DamagedShieldEvent().Subscribe(HandlePlayerHitShield))
                if(set PlayerKills[Player] = 0) {}
                if(set PlayerLastHitBy[Player] = Player) {}
                option{TeamCollection.AddToTeam[Player, Team78]}
        
        spawn{GrantItemsAndRelease()}
        

    GrantItemsAndRelease()<suspends>:void=
        race:
            EndedEvent.Await()
            block:
                Sleep(6.1)
                Print("Releasing From Stasis")
                for(Player : GetPlayspace().GetPlayers(), FortChar:= Player.GetFortCharacter[]):
                    ItemGranterDefaultWeapon.GrantItem(Player)
                    FortChar.ReleaseFromStasis()
                
    _OnMinigameEnded<override>(Winners: []player)<suspends>:void=
        BackgroundMusic.Disable()
        HudNoReticle.Disable()

        DefaultOrbitCamera.RemoveFromAll()
        DefaultOrbitCamera.Disable()

        ZoomInOrbitCamera.RemoveFromAll()
        ZoomInOrbitCamera.Disable()
        
        InputTriggerAimIn.UnregisterAll()
        InputTriggerAimIn.Disable()

        for(Player:GetRemainingPlayers()):
            if(Fort:=Player.GetFortCharacter[]):
                DBNO.Revive(Player)
                Fort.PutInStasis(stasis_args{AllowEmotes := false})

    _Reset<override>()<suspends>:void=
        set PlayerKills = map{}
        set PlayerLastHitBy = map{}
        DBNO.Disable()
        if(DayPostProcessProp.TeleportTo[vector3{X:=29297.380442,Y:=24195.435289,Z:=11633.867539}, rotation{}]):
            
    HandleAimIn(Player : agent):void=
        ZoomInOrbitCamera.AddTo(Player)

    HandleAimInReleased(TupleReleased : tuple(agent, float)):void=
        if(Player := player[TupleReleased(0)]):
            ZoomInOrbitCamera.RemoveFrom(Player)

    HandlePlayerHit(DamageResult : damage_result):void=
        HandlePlayerHitCustom(DamageResult)

    HandlePlayerHitShield(DamageResult : damage_result ) : void =
        HandlePlayerHitCustom(DamageResult)

    HandlePlayerHitCustom( DamageResult : damage_result ) : void =
        if:
            Instigator := DamageResult.Instigator?
            InstigatorAgent := Instigator.GetInstigatorAgent[]
            InstigatorPlayer := player[InstigatorAgent]
            CharacterWhoWasHitFort := fort_character[DamageResult.Target]
            CharacterWhoWasHitAgent := CharacterWhoWasHitFort.GetAgent[]
            CharacterWhoWasHitPlayer := player[CharacterWhoWasHitAgent]
        then:
            if(InstigatorAgent = CharacterWhoWasHitAgent):
                #Print("Same Agent: Usually from storm damage")
            else:
                #Print("Setting last player hit by")
                if(set PlayerLastHitBy[CharacterWhoWasHitPlayer] = InstigatorPlayer):

    OnPlayerDowned(Agent : agent):void=
        Print("On Player Downed")
        if(Player := player[Agent]):
            if(LastPlayerToHit := PlayerLastHitBy[Player], not LastPlayerToHit = Player):
                Manager.EliminationAccolade.Award(LastPlayerToHit)
                if(NewKillVal := PlayerKills[LastPlayerToHit]+1):
                    if(NewKillVal = ItemGrantersInOrder.Length):
                        Print("Player hit end of array")
                        Manager.BottomHUDDevice.Show(BuildPlayerMessage("", LastPlayerToHit, " WINS!"))
                        EndedEvent.Signal(array{LastPlayerToHit})
                    else:
                        Print("Incrementing kill")
                        if(ItemGranter := ItemGrantersInOrder[NewKillVal]):
                            ItemGranter.GrantItem(LastPlayerToHit)
                        if(set PlayerKills[LastPlayerToHit] = NewKillVal):
            spawn{TeleportPlayer(Player)}

    OnPlayerRevived(Agent : agent):void=
        Print("On Player Revived")
        if:
            Fort := Agent.GetFortCharacter[]
        then:
            spawn{DelayHealing(Fort)}
    
    DelayHealing(Fort : fort_character)<suspends>:void=
        Sleep(0.3)
        Fort.SetHealth(150.0)

    TeleportPlayer(Player : player)<suspends>:void=
        Print("Teleport and heal bro")
        #Manager.Box1Barriers.BulkTeleport(array{Player}, Manager.Teleporter, ?TranslationOffset := vector3{Z := 200.0}, ?YawOffset := -90.0)
        if(Method := CustomTeleportationMethod?):
            case(Method.Type):
                # Use a single teleporter
                minigame_teleport_method_type.UseTeleporter => 
                    if(Teleporter := Method.OptTeleporter?):
                        Print("ERR: BodyCam Minigame should not get here. Teleport method is on UseTeleporter")
                # Use barriers locations
                minigame_teleport_method_type.UseBarrierTransforms => 
                    Print("ERR: BodyCam Minigame should not get here. Teleport method is on barriers")
                # Use creative prop locations
                minigame_teleport_method_type.UseCreativePropTransforms =>
                    if(CreativeProps := Method.OptCreativeProps?, CreativeProps.Length > 0):
                        
                        #CreativeProps.BulkTeleport(array{Player}, Manager.Teleporter)# Old Code
                        ## New Code
                        FurthestPropFromPlayers := array{FindFurthestPropFromPlayers(CreativeProps, false)}
                        #FurthestPropFromPlayers.BulkTeleport(array{Player}, Manager.Teleporter)
                    else:
                        Print("ERR: Minigame {Name} teleport method is set to UseCreativePropTransforms but props are missing")
        # Use default method (TP to box1)
        else:
            Print("ERR: BodyCam Minigame should not get here. Teleport method is unset")

        spawn{ReviveAndSpawnProtection(Player)}

    ReviveAndSpawnProtection(Player: player)<suspends>:void=
        Sleep(0.1)
        DBNO.Revive(Player)
        if(Fort:=Player.GetFortCharacter[]):
            Fort.SetVulnerability(false)
            Print("Invincible")
            Sleep(1.0)
            Print("Not Invincible")
            Fort.SetVulnerability(true)

    <#FindFurthestPropFromPlayers(CreativeProps : []creative_prop, bPlayers : ?[]player):creative_prop=
        var FurthestPropFromPlayers : creative_prop = creative_prop{}
        var FurthestDistance : float = 0.0
        var Players : []player = GetRemainingPlayers()
        if (ValidPlayers := bPlayers?):
            set Players = ValidPlayers
        if (Prop := CreativeProps[0]):
            set FurthestPropFromPlayers = Prop
        for(CreativeProp:CreativeProps):
            var ClosestPlayerDist : float = 99999999999.0
            for(Player:Players, FortChar := Player.GetFortCharacter[]):
                Translation := FortChar.GetTransform().Translation
                Dist := Distance(Translation, CreativeProp.GetTransform().Translation)
                if (ClosestPlayerDist > Dist):
                    set ClosestPlayerDist = Dist
            if (ClosestPlayerDist > FurthestDistance):
                set FurthestDistance = ClosestPlayerDist
                set FurthestPropFromPlayers = CreativeProp
        Debug.DrawSphere(FurthestPropFromPlayers.GetTransform().Translation + vector3{X:= 0.0, Y:= 0.0, Z:=2000.0}, ?Radius := 200.0, ?Thickness := 50.0, ?Duration := 4.0, ?Color:=color{R:=1.0,G:=0.0,B:=0.0})
        return FurthestPropFromPlayers#>


    HandleCollectedEvent(Agent : agent):void=
        Print("Collected Bro")
        Collectible.Hide()
        if(Player := player[Agent]):
            Manager.GoldTracker.Add(Player, 100)

    TestSpawnLocation()<suspends>:void=
        loop:
            if(Method := CustomTeleportationMethod?):
                case(Method.Type):
                    minigame_teleport_method_type.UseTeleporter => 
                    minigame_teleport_method_type.UseBarrierTransforms => 
                    minigame_teleport_method_type.UseCreativePropTransforms =>
                        if(CreativeProps := Method.OptCreativeProps?, CreativeProps.Length > 0):
                            FindFurthestPropFromPlayers(CreativeProps, false)
                            Print("Finding Spawn Loc")
            Sleep(5.0)
