using { /Fortnite.com/Devices }
using { /Fortnite.com/AI}
using { /Fortnite.com/Teams}
using { /Verse.org/Simulation }
using { /Verse.org/Random }
using { /Fortnite.com/Characters }
using { /Fortnite.com/FortPlayerUtilities }
using { /Fortnite.com/Game }
using { /Fortnite.com/Teams }
using { /Verse.org/Simulation/Tags }
using { /UnrealEngine.com/Temporary/Diagnostics }
using { /UnrealEngine.com/Temporary/SpatialMath } 
using { /Verse.org/Assets }

bodycam_lights_out := class(minigame_device):
    @editable BackgroundMusic :audio_player_device = audio_player_device{}
    @editable ItemGranterDefaultWeapon : item_granter_device = item_granter_device{}
    @editable HudNoReticle : hud_controller_device = hud_controller_device{}
    @editable DefaultOrbitCamera : gameplay_camera_orbit_device = gameplay_camera_orbit_device{}
    @editable InputTriggerAimIn : input_trigger_device = input_trigger_device{}
    @editable ZoomInOrbitCamera : gameplay_camera_orbit_device = gameplay_camera_orbit_device{}
    @editable IntroCinematic : cinematic_sequence_device = cinematic_sequence_device{}
    @editable NightPostProcessProp : creative_prop = creative_prop{}

    Name<override>: string = "BODYCAM Lights Out!"
    #StartingMessage<override>:string = ""
    GetVotingImage<override>()<transacts>: texture = VerseExposed.SQ_BCLIGHTSOUT
    DisplayIcon<override>: ?texture = option{VerseExposed.TransparentTexture}

    _Start<override>()<suspends>: void =
        #IntroCinematic.Play()
        HudNoReticle.Enable()
        BackgroundMusic.Enable()
        BackgroundMusic.Play()

        RegisterSubscriber(InputTriggerAimIn.PressedEvent.Subscribe(HandleAimIn))
        RegisterSubscriber(InputTriggerAimIn.ReleasedEvent.Subscribe(HandleAimInReleased))
        for(Player : GetRemainingPlayers(), Fort := Player.GetFortCharacter[]):
            Fort.PutInStasis(stasis_args{AllowFalling := true, AllowTurning := true})
            InputTriggerAimIn.Register(Player)
        #spawn{GrantItemsAndRelease()}

    OnSafeTimeEnded<override>()<suspends>:void=
        DefaultOrbitCamera.Enable()
        DefaultOrbitCamera.AddToAll()
        InputTriggerAimIn.Enable()
        ZoomInOrbitCamera.Enable()
        if(NightPostProcessProp.TeleportTo[vector3{X:=29297.380442,Y:=24195.435289, Z:= 3255.797211}, rotation{}]):
        for(Player : GetPlayspace().GetPlayers(), FortChar:= Player.GetFortCharacter[]):
            ItemGranterDefaultWeapon.GrantItem(Player)
            FortChar.ReleaseFromStasis() 

    _Reset<override>()<suspends>:void=
        if(NightPostProcessProp.TeleportTo[vector3{X:=29297.380442,Y:=24195.435289,Z:=11633.867539}, rotation{}]):
        
    # GrantItemsAndRelease()<suspends>:void=
    #     race:
    #         EndedEvent.Await()
    #         block:
    #             Sleep(2.1)
    #             Print("Releasing From Stasis")
    #             for(Player : GetPlayspace().GetPlayers(), FortChar:= Player.GetFortCharacter[]):
    #                 ItemGranterDefaultWeapon.GrantItem(Player)
    #                 FortChar.ReleaseFromStasis()
                
    _OnMinigameEnded<override>(Winners: []player)<suspends>:void=
        BackgroundMusic.Disable()
        HudNoReticle.Disable()

        DefaultOrbitCamera.RemoveFromAll()
        DefaultOrbitCamera.Disable()

        ZoomInOrbitCamera.RemoveFromAll()
        ZoomInOrbitCamera.Disable()
        
        InputTriggerAimIn.UnregisterAll()
        InputTriggerAimIn.Disable()

        for(Player:GetRemainingPlayers()):
            if(Fort:=Player.GetFortCharacter[]):
                Fort.PutInStasis(stasis_args{AllowEmotes := false})

    HandleAimIn(Player : agent):void=
        ZoomInOrbitCamera.AddTo(Player)

    HandleAimInReleased(TupleReleased : tuple(agent, float)):void=
        if(Player := player[TupleReleased(0)]):
            ZoomInOrbitCamera.RemoveFrom(Player)