
using { /Fortnite.com/Devices }
using { /Verse.org/Simulation }
using { /UnrealEngine.com/Temporary/Diagnostics }
using { /Fortnite.com/Characters }
using { /UnrealEngine.com/Temporary/UI }
using { /UnrealEngine.com/Temporary/SpatialMath }
using { /Fortnite.com/UI }
using { /Verse.org/Assets }
using { /Fortnite.com/Game }

# A Verse-authored creative device that can be placed in a level
Gulag := class(minigame_device):
    @editable GulagTime : float = 45.0
    @editable DefaultStorm : basic_storm_controller_device = basic_storm_controller_device{}
    @editable DBNO : down_but_not_out_device = down_but_not_out_device{}
    @editable GulagDamageVolume : damage_volume_device = damage_volume_device{}
    @editable GulagWaitingTeleporter : teleporter_device = teleporter_device{}
    @editable Player1Tele : teleporter_device = teleporter_device{}
    @editable Player2Tele : teleporter_device = teleporter_device{}
    @editable MainTele : teleporter_device = teleporter_device{}
    @editable GulagWaitingProps :  []creative_prop = array{}
    @editable GulagWeaponGranter : item_granter_device = item_granter_device{}
    @editable GulagPvpProps : []creative_prop = array{}
    @editable GulagSafeZone : basic_storm_controller_device = basic_storm_controller_device{}
    #@editable GulagMainWeaponGranter : item_granter_device = item_granter_device{}
    @editable LastLifeHUDMessage : hud_message_device = hud_message_device{}
    @editable GulagClosingHUDMessage : hud_message_device = hud_message_device{}
    @editable GulagWaitingMessage : hud_message_device = hud_message_device{}
    @editable GulagPVPMessage : hud_message_device = hud_message_device{}
    @editable ReleaseMessage : hud_message_device = hud_message_device{}
    var bGulagActive : logic = true
    var GulagWaitingPlayers : []player = array{}
    var GulagPvpPlayers : []player = array{}
    var LastLifePlayers : []player = array{}

    Name<override>: string = "Gulag Reload"
    StartingMessage<override>:string = "You have one more chance in the GULAG!"
    GetVotingImage<override>()<transacts>: texture = VerseExposed.SQ_GULAG
    
    _Start<override>()<suspends>: void =
        set bGulagActive = true
        RegisterSubscriber(DBNO.AgentDownedEvent.Subscribe(OnPlayerDowned))
        RegisterSubscriber(DBNO.AgentRevivedEvent.Subscribe(OnPlayerRevived))
        set GulagWaitingPlayers = array{}
        set GulagPvpPlayers = array{}
        set LastLifePlayers = array{}
        DBNO.Enable()
        spawn{ManageGulagPVP()}
        #GulagSafeZone.GenerateStorm()
        spawn{CloseGulag()}
        #Storm.DestroyStorm()

    OnSafeTimeEnded<override>()<suspends>: void =
        GulagSafeZone.GenerateStorm()

    CloseGulag()<suspends>:void=
        Print("Closing Gulag")
        race:
            EndedEvent.Await()
            block:
                Sleep(GulagTime)
                set bGulagActive = false
                Print("Gulag closed")
                MainPlayers := GetRemainingNonGulagPlayers()
                for(MainPlayer:MainPlayers):
                    GulagClosingHUDMessage.Show(MainPlayer)
            
    _OnMinigameEnded<override>(Winners: []player)<suspends>:void=
        DBNO.Disable()
        for(Player:GulagPvpPlayers, FortChar := Player.GetFortCharacter[]):
            FortChar.ReleaseFromStasis()

    _Reset<override>()<suspends>:void=
        DefaultStorm.DestroyStorm()
        GulagSafeZone.DestroyStorm()

    OnPlayerDowned(Agent : agent):void=
        if(GetRemainingPlayers().Length < 3):
            if(FortChar := Agent.GetFortCharacter[]):
                FortChar.Damage(5000.0)
                return

        if (Player := player[Agent]):
            if(LPlayer := LastLifePlayers.Find[Player], FortChar := Player.GetFortCharacter[]):
                FortChar.Damage(5000.0)      

        if (Player := player[Agent]):
            if(GulagPvpPlayers.Find[Player]):
                set GulagPvpPlayers = GulagPvpPlayers.RemoveAllElements(Player)
                if(WinningPlayer := GulagPvpPlayers[0]):
                    spawn{MovePlayerBackToMainPVP(WinningPlayer)}
                set GulagPvpPlayers = array{}
                KillAllPlayersIfLastOneStanding()
                return

        if (Player := player[Agent], not LastLifePlayers.Find[Player]):
            Print("Function body")
            if(bGulagActive = true):
                TeleportPlayerToGulag(Agent)
                spawn{ReviveAndSpawnProtection(Player)}
            else:
                if(FortChar := Player.GetFortCharacter[]):
                    FortChar.Damage(5000.0)

    _OnPlayerEliminated<override>(Result: elimination_result)<suspends>: void =
        if(Agent := Result.EliminatedCharacter.GetAgent[], Player := player[Agent]):
            Print("Player Eliminated")
            if(GulagWaitingPlayers.Find[Player]):
                set GulagWaitingPlayers = GulagWaitingPlayers.RemoveAllElements(Player)
            if(GulagPvpPlayers.Find[Player]):
                set GulagPvpPlayers = GulagPvpPlayers.RemoveAllElements(Player)
                if (WinningPlayer := GulagPvpPlayers[0]):
                    spawn{MovePlayerBackToMainPVP(WinningPlayer)}
                set GulagPvpPlayers = array{}

    OnPlayerRevived(Agent : agent):void=
        Print("On Player Revived")

    KillAllPlayersIfLastOneStanding():void=
        if (LastPlayers := GetRemainingNonGulagPlayers(), LastPlayers.Length < 2, LastPlayer := LastPlayers[0], GetTotalGulagPlayers().Length <= 1):
            RemainingPlayers := GetRemainingPlayers()
            for(RemainingPlayer : RemainingPlayers, not RemainingPlayer = LastPlayer, FortChar := RemainingPlayer.GetFortCharacter[]):
                FortChar.Damage(5000.0)
            Print("Last one standing")

    MovePlayerBackToMainPVP(Player : player)<suspends>:void=
        ReleaseMessage.Show(Player)
        Sleep(1.0)
        TeleportToMainPvp(Player)
        LastLifeHUDMessage.Show(Player)
        # Set Health and shield
        if (FortChar := Player.GetFortCharacter[]):
            FortChar.SetHealth(100.0)
            FortChar.SetShield(100.0)
        spawn{GrantDefaultWeapons(Player)}

    TeleportPlayerToGulag(Agent : agent):void=
        FurthestPropFromPlayers := array{FindFurthestPropFromPlayers(GulagWaitingProps, option{GulagWaitingPlayers})}
        #FurthestPropFromPlayers.BulkTeleport(array{Agent}, GulagWaitingTeleporter)
        GulagWaitingMessage.Show(Agent)

    TeleportToMainPvp(Player : player):void=
        if(GulagPvpProps.Length > 0):
            AllPlayersInMainPvp := GetRemainingNonGulagPlayers()
            FurthestPropFromPlayers := array{FindFurthestPropFromPlayers(GulagPvpProps, option{AllPlayersInMainPvp})}
            #FurthestPropFromPlayers.BulkTeleport(array{Player}, MainTele)
            for(DefaultGranter : DefaultItemGranters, DefaultGranter.GrantAfterSafeTime?):
                DefaultGranter.Grant(array{Player})
            #GulagMainWeaponGranter.GrantItem(Player)
            #ResetItemRemover.Remove(Player)

    GrantDefaultWeapons(Player : player)<suspends>:void=
        Sleep(0.1)
        for(DefaultItemGranter : DefaultItemGranters):
            DefaultItemGranter.ItemGranter.GrantItem(Player)

    GetRemainingNonGulagPlayers()<transacts>:[]player=
        var AllPlayersInMainPvp : []player = GetRemainingPlayers()
        # Remove gulag players from calculating spawn for main pvp
        for(WaitingPlayer : GulagWaitingPlayers):
            set AllPlayersInMainPvp = AllPlayersInMainPvp.RemoveAllElements(WaitingPlayer)
        for(GulagPvpPlayer : GulagPvpPlayers):
            set AllPlayersInMainPvp = AllPlayersInMainPvp.RemoveAllElements(GulagPvpPlayer)
        return AllPlayersInMainPvp

    GetTotalGulagPlayers()<transacts>:[]player=
        var Players : []player = array{}
        set Players = GulagPvpPlayers + GulagWaitingPlayers
        return Players
      
    ReviveAndSpawnProtection(Player: player)<suspends>:void=
        Sleep(0.1)
        DBNO.Revive(Player)
        if(Fort:=Player.GetFortCharacter[]):
            Fort.SetVulnerability(false)
            Sleep(1.0)
            Fort.SetVulnerability(true)
        Sleep(1.0)
        set GulagWaitingPlayers += array{Player}
        set LastLifePlayers += array{Player}
        KillAllPlayersIfLastOneStanding()

    CountDownUI(Player1 : player, Player2 : player)<suspends>:void=
        GulagPVPMessage.SetText(StringToMessage("3"))
        GulagPVPMessage.SetDisplayTime(0.95)
        GulagPVPMessage.Show(Player1)
        GulagPVPMessage.Show(Player2)
        Sleep(1.0)
        GulagPVPMessage.SetText(StringToMessage("2"))
        GulagPVPMessage.Show(Player1)
        GulagPVPMessage.Show(Player2)
        Sleep(1.0)
        GulagPVPMessage.SetText(StringToMessage("1"))
        GulagPVPMessage.Show(Player1)
        GulagPVPMessage.Show(Player2)
        Sleep(1.0)
        GulagPVPMessage.SetDisplayTime(2.8)
        GulagPVPMessage.SetText(StringToMessage("Fight"))
        GulagPVPMessage.Show(Player1)
        GulagPVPMessage.Show(Player2)
        
    ManageGulagPVP()<suspends>:void=
        race:
            EndedEvent.Await()
            loop:
                Sleep(0.0)
                if (GetRemainingPlayers().Length < 2, GetRemainingNonGulagPlayers().Length < 2):
                    break      
            loop:
                Sleep(1.0)
                if (GulagPvpPlayers.Length <= 0, GulagWaitingPlayers.Length > 1, GetRemainingPlayers().Length > 2):
                    if(Player1 := GulagWaitingPlayers[0], Player2 := GulagWaitingPlayers[1]):
                        Sleep(0.5)
                        Player1Tele.Teleport(Player1)
                        Player2Tele.Teleport(Player2)
                        set GulagPvpPlayers += array{Player1}
                        set GulagPvpPlayers += array{Player2}
                        set GulagWaitingPlayers = GulagWaitingPlayers.RemoveAllElements(Player1)
                        set GulagWaitingPlayers = GulagWaitingPlayers.RemoveAllElements(Player2)
                        Manager.ResetItemRemover.Remove(Player1)
                        Manager.ResetItemRemover.Remove(Player2)
                        GulagWeaponGranter.CycleToRandomItem(Player1)
                        GulagWeaponGranter.GrantItem(Player1)
                        GulagWeaponGranter.GrantItem(Player2)
                        if (FortChar := Player1.GetFortCharacter[]):
                            FortChar.SetHealth(100.0)
                            FortChar.SetShield(100.0)
                            FortChar.PutInStasis(stasis_args{AllowTurning := true, AllowFalling := true})
                        if (FortChar := Player2.GetFortCharacter[]):
                            FortChar.SetHealth(100.0)
                            FortChar.SetShield(100.0)
                            FortChar.PutInStasis(stasis_args{AllowTurning := true, AllowFalling := true})
                        CountDownUI(Player1, Player2)
                        Sleep(0.7)
                        if (FortChar := Player1.GetFortCharacter[], FortChar2 := Player2.GetFortCharacter[]):
                            FortChar.ReleaseFromStasis()
                            FortChar2.ReleaseFromStasis()
                if (bGulagActive = false, GulagWaitingPlayers.Length = 1, LastWaitingPlayer := GulagWaitingPlayers[0]):
                    spawn{MovePlayerBackToMainPVP(LastWaitingPlayer)}
                    DefaultStorm.GenerateStorm()
                    Sleep(2.0)
                    Storm.DestroyStorm()
                    break
            loop:
                Sleep(1.2)
                if (bGulagActive = false, GulagWaitingPlayers.Length <= 0, GulagPvpPlayers.Length <= 0):
                    DefaultStorm.GenerateStorm()
                    Sleep(2.0)
                    Storm.DestroyStorm()
                    break

