# using { /Fortnite.com/Devices }
# using { /Fortnite.com/AI}
# using { /Fortnite.com/Teams}
# using { /Verse.org/Simulation }
# using { /Verse.org/Random }
# using { /Fortnite.com/Characters }
# using { /Fortnite.com/FortPlayerUtilities }
# using { /Fortnite.com/Game }
# using { /Fortnite.com/Teams }
# using { /Verse.org/Simulation/Tags }
# using { /UnrealEngine.com/Assets }
# using { /UnrealEngine.com/Temporary/Diagnostics }
# using { /UnrealEngine.com/Temporary/SpatialMath } 
# using { /Verse.org/Assets }

# HitEffect : particle_system = CoreGame.NS_HitImpact
# DefflectEffect : particle_system = CoreGame.NS_HitImpact1
# BarrierEffect : particle_system = CoreGame.NS_Barrier

# custom_player := class<unique>():
#     Player : player
#     DeflectTime: float
#     PlayerID : int
#     var CanDeflect:logic = false
#     var bIsPlayerInBarrier : logic = false

#     # Methods
#     # GetPlayerReference()<transacts>:player=
#     #     return Player

#     CanDeflectActivate()<suspends>:void =
#         set CanDeflect = true
#         Sleep(DeflectTime)
#         set CanDeflect = false

#     ## Function to allow barrier 
#     PlayerBarrierBehavior(TimetoDissapearEffect : int, VFX:visual_effect_powerup_device, HideVFX : visual_effect_powerup_device)<suspends>:void =
#         if(bIsPlayerInBarrier = true):
#             return
#         set bIsPlayerInBarrier = true
#         CoverPlayer := Player
#         if:
#             FortCharacter := CoverPlayer.GetFortCharacter[]
#             bIsPlayerInBarrier = true
#         then:
#             VFX.Pickup(CoverPlayer)
#             FortCharacter.SetVulnerability(false)
#             Sleep (TimetoDissapearEffect * 1.0) # Time to dissapeareffect
#             set bIsPlayerInBarrier = false
#             FortCharacter.SetVulnerability(true)
#             HideVFX.Pickup(CoverPlayer)

#     ## Get forward position to deflect
#     GetForwardPosition(ForwardOffset : float):transform =
#         if:
#             FortChar:= Player.GetFortCharacter[]
#         then:
#             PlayerPos := FortChar.GetTransform().Translation
#             PlayerRot := FortChar.GetTransform().Rotation
#             ForwardVec := PlayerRot.GetLocalForward()
#             Offset := ForwardVec * ForwardOffset
    
#             # Define the crosshair offset in the player's local space
#             LocalCrosshairOffset := vector3{ X:= 32.0, Y:= 32.0, Z:= 50.0}
            
#             # Transform the crosshair offset from local space to world space
#             WorldCrosshairOffset := PlayerRot.RotateVector(LocalCrosshairOffset)
    
#             # Add the world space crosshair offset to the final position
#             FinalPos := PlayerPos + Offset + WorldCrosshairOffset
    
#             return transform:
#                 Translation:= FinalPos
#                 Rotation:= PlayerRot
#                 Scale:= vector3{ X:= 1.0, Y:= 1.0, Z:= 1.0}
#         else:
#             return transform{}

# USOPEN_TennisStorm := class(minigame_device):
#     @editable
#     var DistancetoBallCollision:float = 1000000000.0
#     @editable
#     var BallLifetime :float = 7.0
#     @editable
#     var BallDamageAmount :float= 40.0
#     @editable
#     var DeflectLivetime :float = 1.0
#     @editable
#     var BallSpeed :float = 1.0
#     @editable
#     var SlowBallSpeed :float = 1.0
#     @editable
#     var TimeUntilNextRoundLaunchersAppear:int= 10
#     @editable
#     var TimeToIncreaseFireRate:float = 1.0
#     @editable
#     BackgroundMusic :audio_player_device = audio_player_device{}
#     @editable
#     PlayerHitBallSound :audio_player_device = audio_player_device{}
#     @editable
#     PlayerGotHitSound :audio_player_device = audio_player_device{}
#     @editable
#     BallSpawnedSound :audio_player_device = audio_player_device{}
#     @editable
#     BombFlowers : []wilds_plant_device = array{}
#     @editable
#     IntroCinematic : cinematic_sequence_device = cinematic_sequence_device{}
#     @editable 
#     ShootTrigger: trigger_device = trigger_device{}
#     @editable
#     DeflectTrigger: trigger_device = trigger_device{}
#     @editable 
#     GreenBallProp : creative_prop_asset = DefaultCreativePropAsset
#     @editable 
#     RedBallProp : creative_prop_asset = DefaultCreativePropAsset
#     @editable 
#     MarkerLauncher : []creative_prop = array{}
#     @editable 
#     BallSpawners : []creative_prop =  array{}
#     @editable
#     SoundSpawners :[]audio_player_device = array{}
#     @editable
#     InputTrigger : input_trigger_device = input_trigger_device{}
#     @editable
#     VFXTennisRacketTrigger: trigger_device = trigger_device{}
#     @editable
#     ExplosiveDevices :[]movement_modulator_device= array{}
#     @editable
#     KnockbackMovementModulator :movement_modulator_device = movement_modulator_device{}
#     @editable
#     KnockbackTeleporterProp :creative_prop = creative_prop{}

#     @editable
#     BallLaunchersCinematic :cinematic_sequence_device = cinematic_sequence_device{}
#     @editable
#     ShakeCinematics :cinematic_sequence_device = cinematic_sequence_device{}
#     @editable
#     GraceTimeHud :hud_message_device = hud_message_device{}
#     @editable
#     SecondGraceTimeHud :hud_message_device = hud_message_device{}
#     @editable
#     HudController : hud_controller_device = hud_controller_device{}
#     @editable
#     VFXSpawnerPosition :[]vfx_spawner_device= array{}
    
#     ### Health Timers
#     @editable
#     var MinTimeToAppearPowerUp :float = 11.0
#     @editable
#     var MaxTimeToAppearPowerUp :float = 15.0
#     @editable
#     @editable
#     var MinTimeToAppearBall :float = 7.0
#     @editable
#     var MaxTimeToAppearBall :float = 13.0
#     @editable
#     var BarrierActiveTime :int= 3
#     @editable
#     HealthPowerUps : []health_powerup_device=array{}
#     @editable
#     ShieldPowerUps :[]health_powerup_device=array{}
#     @editable
#     ShieldBarrierVisibleVFX :visual_effect_powerup_device = visual_effect_powerup_device{}
#     @editable
#     ShieldBarrierInvisibleVFX :visual_effect_powerup_device = visual_effect_powerup_device{}

#     #var CurrentPlayRate :float = 0.0
#     #var SecondCurrentPlayRate :float = 0.0

#     #Non serializable
#     var CustomPlayers<private> : [player]custom_player = map{}
#     var DefaultVector:vector3 = vector3{X:=-75.0,Y:=0.0,Z:=0.0} 
#     var CurrentLauncher :creative_prop = creative_prop{}
#     var SecondCurrentLauncher:creative_prop = creative_prop{}
#     var CurrentIndexLauncher :int= 0
#     var SecondCurrentIndexLauncher :int= 0

#     var TimerTime :int = 1
#     var IDCounter :int = 0 ## ID to identify the current player position on the array
#     var IndexBall :int= 0
#     var AnotherIndexBall :int= 0

#     #### GracePeriodTimer
#     var GraceTime :int = 5
#     var BreakLoop :logic = false
#     ####

#     ### Fire Rate
#     var InitialFireRate:float = 1.0;  # Tasa de disparo inicial en disparos por segundo
#     @editable var ScalingFactor :float= 0.1;    # Factor de escalado por segundo
#     var CurrentFireRate:float = 0.0;        # Tasa de disparo actual
#     var NextFireTime: float= 0.0; # Tiempo hasta el siguiente disparo permitido
#     var MarkerTimeBeforeFire:float= 1.0
#     var MarkerDisplayed :logic= false
#     var bIsUltimateSpeed:logic = false

#     ## Explosion Stuff
#     var ExplosiveInitialPosition :vector3 = vector3{}
#     var ExplosiveInitialRotation : rotation= rotation{}

#     # Runs when the device is started in a running game
#     _Start<override>()<suspends>:void=
#         BackgroundMusic.Enable()
#         BackgroundMusic.Play()
#         HudController.Enable()
#         RegisterSubscriber(IntroCinematic.StoppedEvent.Subscribe(CinematicEnd))
#         IntroCinematic.Play()
#         for(Player : GetRemainingPlayers(), FortChar:= Player.GetFortCharacter[]):
#             FortChar.PutInStasis(stasis_args{AllowTurning := false, AllowFalling := true})
#             VFXTennisRacketTrigger.Trigger(Player)
#             spawn{InitPlayer(Player)}

#         for(BombFlower : BombFlowers):
#             Print("registering")
#             RegisterSubscriber(BombFlower.LaunchEvent.SubscribeQuestionAgent(OnBombFlowerLaunched,BombFlower))
#             RegisterSubscriber(BombFlower.GrowEvent.SubscribeEmptyTuple(OnBombFlowerGrow,BombFlower))      

#         for(HealthPowerUp : HealthPowerUps):
#             RegisterSubscriber(HealthPowerUp.ItemPickedUpEvent.SubscribeAgent(OnPickUpHealthPowerUpEvent,HealthPowerUp))
#             HealthPowerUp.Despawn()

#         for(ShieldPowerUp : ShieldPowerUps):
#             RegisterSubscriber(ShieldPowerUp.ItemPickedUpEvent.SubscribeAgent(OnPickUpShieldPowerUpEvent,ShieldPowerUp))
#             ShieldPowerUp.Despawn()

#         HideAllMarkers()
#         HideSecondRoundLaunchers()

#         spawn{OnCountdownTimer()}
#         set CurrentFireRate = InitialFireRate;
           
#     OnSafeTimeEnded<override>()<suspends>:void=
#         InputTrigger.Enable()
#         InputTrigger.PressedEvent.Subscribe(HandleInput)
#         Print("Spawning Middle Bomb Flower First")
#         if(MiddleBomb := BombFlowers[0]):
#             MiddleBomb.Grow()
#         if(ShuffledBombFlowers := Shuffle(BombFlowers.RemoveElement[0])):
#             race:
#                 EndedEvent.Await()
#                 block:
#                     sync:
#                         SpawnBombFlowers(ShuffledBombFlowers)
#                         SpawnPowerUps(Shuffle(HealthPowerUps))
#                         SpawnPowerUps(Shuffle(ShieldPowerUps))
        
#     _OnMinigameEnded<override>(Winners: []player)<suspends>:void=
#         BackgroundMusic.Disable()
#         InputTrigger.UnregisterAll()
#         InputTrigger.Disable()
#         BallLaunchersCinematic.Stop()

#     _Reset<override>()<suspends>:void=
#         HudController.Disable()
#         set CustomPlayers = map{}
#         set TimerTime = 1
#         set IDCounter = 0
            
#     CinematicEnd():void=
#         for(Player : GetPlayspace().GetPlayers(), FortChar:= Player.GetFortCharacter[]):
#             FortChar.ReleaseFromStasis()
#         #Fix Glitch wehre releasing from stasis cancels no emote mutator zone
#         # Manager.NoEmotingMutatorZone.Disable()
#         # Manager.NoEmotingMutatorZone.Enable()

#     OnBombFlowerGrow(BombFlower:wilds_plant_device):void =
#         Print("Plant Grew")
#         if(BallSpawnedSound.TeleportTo[BombFlower.GetTransform()]):
#             Print("teleporting")
#             BallSpawnedSound.Play()

#     SpawnBombFlowers(ShuffledBombFlowers : []wilds_plant_device)<suspends>:void =
#         Print("Spawn Bombflowers")
#         for(Index := 0..ShuffledBombFlowers.Length, BombFlower := ShuffledBombFlowers[Index]):
#             if(Index <= 4):
#                 Sleep(0.5)
#                 BombFlower.Grow()
#             else:
#                 #Sleep(GetRandomFloat(MinTimeToAppearPowerUp,MaxTimeToAppearPowerUp))
#                 Sleep(5.0)
#                 BombFlower.Grow()
#             Print("New Bomb Flower")

#     SpawnPowerUps(PowerUps: []health_powerup_device)<suspends>:void =
#         Print("Spawn PowerUps")
#         for(Index := 0..PowerUps.Length, PowerUp := PowerUps[Index]):
#             if(Index = 0):
#                 Sleep(4.0)
#                 PowerUp.Spawn()
#             else:
#                 Sleep(8.0)
#                 #Sleep(GetRandomFloat(MinTimeToAppearPowerUp,MaxTimeToAppearPowerUp))
#                 PowerUp.Spawn()
#         Print("New Power Up")

#     # Show time until launchers appear
#     GraceTimer()<suspends>:void =
#         var Seconds :int= GraceTime
#         set BreakLoop = false
#         race:
#             EndedEvent.Await()
#             block:
#                 Sleep(15.0)
#                 GraceTimeHud.Show()
#                 ShakeCinematics.Play()
#                 loop:
#                     Sleep(1.0)
#                     if(BreakLoop?):
#                         break
#                     Hudmessage :string = "{Seconds}"
#                     SecondGraceTimeHud.SetText(StringToMessage(Hudmessage))
#                     if:
#                         Seconds >= 1
#                     then:
#                         set Seconds -= 1
#                     else:
#                         GraceTimeHud.Hide()
#                         SecondGraceTimeHud.Hide()
#                         BallLaunchersCinematic.Play()
#                         set BreakLoop = true
#                         Tick()

#     #### Bomb Flower Stuff
#     OnBombFlowerLaunched(AgentMaybe:?agent, BombFlower:wilds_plant_device):void =
#         Print("Bomb Flower Launched")
#         if(Player := player[AgentMaybe?]):
#             Print("Found Player")
#             PlayerHitBallSound.Play(Player)
#         spawn{SpawnBombFlowerRandomTime(BombFlower)}

#     SpawnBombFlowerRandomTime(BombFlower:wilds_plant_device)<suspends>:void =
#         # TODO: Could add random spots here
#         Time := GetRandomFloat(MinTimeToAppearBall,MaxTimeToAppearBall)
#         Print("Time={Time}")
#         Sleep(Time)
#         BombFlower.Grow()

#     #### PowerUp Stuff
#     OnPickUpHealthPowerUpEvent(Agent:agent, Health:health_powerup_device):void =
#         Print("Picked up health")
#         spawn{PowerUpPickUpRandomTimeSpawn(Health)}

#     OnPickUpShieldPowerUpEvent(Agent:agent, ShieldPower : health_powerup_device):void =
#         Print("Picked up shield")
#         spawn{PowerUpPickUpRandomTimeSpawn(ShieldPower)}
#         ActivateBarrier(Agent)
        
#     PowerUpPickUpRandomTimeSpawn(PowerUp:health_powerup_device)<suspends>:void =
#         Time := GetRandomFloat(MinTimeToAppearPowerUp,MaxTimeToAppearPowerUp)
#         Print("Random Time Spawn={Time}")
#         Sleep(Time)
#         PowerUp.Spawn()

#     #### Barrier 
#     ActivateBarrier(Agent:agent):void =
#         if:
#             Player:= player[Agent]
#             Player.IsActive[]
#             CustomPlayer := CustomPlayers[Player]
#         then:
#             spawn{CustomPlayer.PlayerBarrierBehavior(BarrierActiveTime,ShieldBarrierVisibleVFX, ShieldBarrierInvisibleVFX)}

#     OnCountdownTimer()<suspends>:void =
#         race:
#             EndedEvent.Await()
#             loop:
#                 Print("{TimerTime}", ?Duration:=1.0)
#                 if (TimerTime >= TimeUntilNextRoundLaunchersAppear):
#                     ShowSecondRoundLaunchers()
#                     GraceTimer()
#                     break
#                 Sleep(1.0)
#                 if:
#                     TimerTime >= 1
#                 then:
#                     set TimerTime += 1

#     HideSecondRoundLaunchers():void =
#         for(Index := 0..BallSpawners.Length - 1):
#             if(SecondRoundLauncher := BallSpawners[Index]):
#                 SecondRoundLauncher.Hide()

#     ShowSecondRoundLaunchers()<suspends>:void =
#         for(VFX : VFXSpawnerPosition):
#             VFX.Enable()
#         for(SecondRoundLauncher : BallSpawners):
#             SecondRoundLauncher.Show()

#     HideAllMarkers():void =
#         for(Marker : MarkerLauncher):
#             Marker.Hide()

#     RandomizeLauncherIndex():tuple(int,int)=
#         value1:= GetRandomInt(0,3)
#         value2 := GetRandomInt(0,3)
#         return (value1, value2)

#     SetRandomLauncher():tuple(creative_prop, creative_prop) =
#         PickedLauncher:= RandomizeLauncherIndex()
#         set CurrentIndexLauncher = PickedLauncher(0)
#         set SecondCurrentIndexLauncher = PickedLauncher(1)
#         spawn {ShowLauncherMarker(CurrentIndexLauncher)}
#         spawn {ShowLauncherMarker(SecondCurrentIndexLauncher)}
#         PlayAttentionSound(CurrentIndexLauncher)
#         if:
#             Launcher:= BallSpawners[PickedLauncher(0)]
#             OtherLauncher:= BallSpawners[PickedLauncher(1)]
#         then:
#             return (Launcher, OtherLauncher)
#         else:
#             return (creative_prop{}, creative_prop{})

#     PlayAttentionSound(Index: int):void =
#         if:
#             bIsUltimateSpeed = false
#         then:
#             if(Sound:=SoundSpawners[Index]):
#                 Sound.Enable()
#                 Sound.Play()
#         else:
#             return

#     DisableAttentionSound(TimetoDisable:float):void =
#         if(bIsUltimateSpeed = true):
#             return
#         if:
#             TimetoDisable <= 0.6
#         then:
#             set bIsUltimateSpeed = true
#             for(Index := 0..SoundSpawners.Length - 1):
#                 if(SecondRoundLauncher := SoundSpawners[Index]):
#                     SecondRoundLauncher.Disable()

#     ShowLauncherMarker(Index:int)<suspends>:void =
#         if(Marker:= MarkerLauncher[Index]): 
#             Marker.Show()
#             Sleep(MarkerTimeBeforeFire)
#             Marker.Hide()

#     # Function to select the Launcher
#     CreateBallLauncher():void =
#         BallLauncher := SetRandomLauncher()
#         set CurrentLauncher = BallLauncher(0)
#         set SecondCurrentLauncher = BallLauncher(1)

#     BallPropReference(Index : int):creative_prop_asset =
#         if:
#             Index = 0
#         then:
#             return GreenBallProp
#         else:
#             return RedBallProp

#     SpeedBallPreference(Index: int):float= ## This function return certain speed float if X = 1 or X = 0
#         if:
#             Index = 0
#         then:
#             return BallSpeed
#         else:
#             return SlowBallSpeed

#     BallDamagePreference(Index:int):float =
#         if:
#             Index = 0
#         then:
#             return BallDamageAmount #* 10000.0)
#         else:
#             return 10000.0

#     Shoot():void=
#         set IndexBall = 0
#         set AnotherIndexBall = 0
#         BallLauncher:= CurrentLauncher
#         OtherBallLauncher:= SecondCurrentLauncher
#         ShootSettings(BallLauncher)
#         ShootSettings(SecondCurrentLauncher)

#     ShootSettings(Ball:creative_prop):void = ### Settings , position, rotation , and spawn ball
#         BallLauncherPosition := Ball.GetTransform()
#         var RotatedVector:vector3 = BallLauncherPosition.Rotation.RotateVector(DefaultVector)
#         test := vector3:
#             X:= RotatedVector.X
#             Y:= RotatedVector.Y
#             Z:= RotatedVector.Z

#         Yaw:= RadiansToDegrees(ArcTan(test.Y, test.X))
#         Pitch := RadiansToDegrees(ArcTan(test.Z, Sqrt(test.X * test.X) + (test.Y * test.Y)))
#         Roll := 0.0

#         NewRotation := MakeRotationFromYawPitchRollDegrees(Yaw, Pitch, Roll)
#         SpawnTransform := transform:
#             Translation:= BallLauncherPosition.Translation + RotatedVector 
#             Rotation:=  NewRotation
#             Scale := vector3:
#                 X:= SpeedBallPreference(IndexBall)
#                 Y:= 0.0 # Custom Data for Control Rig
#                 Z:= 0.0 # Custom Data for Control Rig
#         #ShootTrigger.Trigger()
#         SpawnedProp:= SpawnProp(BallPropReference(IndexBall), SpawnTransform)
#         if (ValidProp := SpawnedProp(0)?):
#                 spawn:
#                     # Start waiting for the ball to collide with something
#                     AwaitBallCollision(ValidProp, IndexBall)

#     InitPlayer(Player : player)<suspends>:void =
#         if:
#             Fort := Player.GetFortCharacter[]
#         then:
#             #Fort.SetMaxHealth(MaxHhealth)
#            # Print("{Fort.GetMaxHealth()}", ?Duration:=6.0)
#             if (Existing := CustomPlayers[Player]){}
#             else:
#                 CustomPlayer := custom_player{
#                     Player := Player,
#                     DeflectTime:=DeflectLivetime,
#                     PlayerID := IDCounter
#                 }
#                 # if(VFX := VFXRacket[IDCounter]):
#                 #     VFX.Pickup(Player)
#                 if:
#                     set IDCounter += 1
#                     set CustomPlayers[Player] = CustomPlayer
            
#             # Create our spawn transform, here we control the scale to pass 3 custom parameters to Control rig if needed
#             # (We read the scale of the prop in control rig to find these values)
            
#             # # Spawn our ball at the position (we use GetProjectileProp because of our cosmetics system, you don't need that - just a direct reference to the prop asset)

#         # Ball Collision function that wait for our ball prop to move after spawning
#         # We only move via control rig once we have collided with something via our sphere trace
#         # So if we move we know we need to perform some kind of action (check if we are close to a player, or not etc...)
        
#     AwaitBallCollision(InProp : creative_prop, IndexDamageAmount : int)<suspends>:void=
#         CachedTranslation := vector3:
#             X:= InProp.GetTransform().Translation.X
#             Y:= InProp.GetTransform().Translation.Y
#             Z:= InProp.GetTransform().Translation.Z
        
#         race:
#             EndedEvent.Await()
#             block:
#                 loop:
#                     Dist := Distance(InProp.GetTransform().Translation, CachedTranslation)
#                     if (Dist > 0.2): ### lees value 
#                         #Print("{Dist}", ?Duration:=6.0)
#                         # This is a custom function we wrote, likely overkill for Tennis but I can provide.
#                         ContextualBallCollision(InProp, IndexDamageAmount)
#                         break
#                     Sleep(0.01)
#                 # Stop searching after this amount of time.
#             block:
#                 # Simulated life time
#                 Sleep(BallLifetime)
#                 InProp.Dispose()
#                 #return

#     # This function takes a collision point and checks the agents within a set radius over time. Likely overkill for the Tennis game but a useful reference.
#     ContextualBallCollision(InProp : creative_prop, IndexDamageAmount : int)<suspends>:void=
#         HitPosition := InProp.GetTransform().Translation
#         HitRotation := InProp.GetTransform().Rotation
#         # Find closest agent to that position

#         var DistanceToBall :float= DistancetoBallCollision
#         #var FoundPlayer:logic = false
#         var ClosestPlayer:?agent = false

#         PlayerCount :int = CustomPlayers.Length
#         for(Player:CustomPlayers):
#             if:
#                 FortCharacters:fort_character = Player.Player.GetFortCharacter[]
#             then:
#                 FortCharacterPosition :vector3= FortCharacters.GetTransform().Translation
#                 TotalDistance :float = Distance(FortCharacterPosition, HitPosition)
#                # Print("Total Distance {TotalDistance}", ?Duration:=6.0)
#                 if(TotalDistance <  DistanceToBall ):
#                 #    Print("FOUND", ?Duration:=6.0)
#                    # set FoundPlayer = true
#                     set DistanceToBall = TotalDistance
#                     set ClosestPlayer = option{Player.Player}
#                 else:
#                  #   Print("HIDE", ?Duration:=6.0)
#                     Sleep(0.1)
#                     InProp.Hide()

#         if:
#             Player := ClosestPlayer?
#             CustomPlayer := CustomPlayers[Player]
#             TargetPlayerAgent := agent[ClosestPlayer?]
#             FortCharacter := TargetPlayerAgent.GetFortCharacter[]
#         then:
#             if:
#                 CustomPlayer.CanDeflect = true
#             then:
#                 InProp.Hide()
#                 Sleep(0.0)
#                 Deflect(CustomPlayer)
#                 #InProp.Dispose()
#             else if:
#                 CustomPlayer.bIsPlayerInBarrier = true
#             then:
#                 InProp.Hide()
#             else:
#                 FortCharacter.Damage(BallDamagePreference(IndexDamageAmount))
#                 spawn {ActivateHitEffect(HitPosition, HitRotation)}
#                 spawn{SetPlayerIndexForExplossion(CustomPlayer.PlayerID,FortCharacter, HitRotation)}
#                 #spawn {SetKnockBackBasedOnDamage(Player,FortCharacter, HitRotation)}
#                 PlayerGotHitSound.Play(Player)
#                 #ShowCollisionDebugData(HitPosition, 10.0, 50.0)
#                 Sleep(0.1)
#                 InProp.Dispose()

#     # SetKnockBackBasedOnDamage(DamagedPlayer : agent,FortCharacter:fort_character, HitRotation : rotation)<suspends> : void = 
#     #     Print("ermm")
#     #     KnockBackLocation : vector3 = vector3{X:=-17762.976666,Y:=-17465.360579,Z:=-775.259213}
#     #     if(KnockbackTeleporterProp.TeleportTo[KnockBackLocation, HitRotation]):
#     #         Print("Changed Direction: {HitRotation}")
#     #     KnockbackMovementModulator.Activate(DamagedPlayer)

#     # Knockback We select a movement modulator , to throw the player that was hited
#     SetPlayerIndexForExplossion(PlayerIndex:int, FortCharacter:fort_character, test:rotation)<suspends>:void =
#         if:
#             Explosive := ExplosiveDevices[PlayerIndex]
#             Explosive.TeleportTo[FortCharacter.GetTransform().Translation - vector3{Z := 150.0}, test]
#         then:
#             Sleep(0.5)
#             if(Explosive.TeleportTo[ExplosiveInitialPosition ,ExplosiveInitialRotation]):

#     HandleInput(Agent: agent):void = 
#         if:
#             Player:= player[Agent]
#             Player.IsActive[]
#         then:
#             if:
#                 CustomPlayer := CustomPlayers[Player]
#             then:
#                 spawn{CustomPlayer.CanDeflectActivate()}

#     Deflect(Player : custom_player):void =
#         if:
#             ValidPlayer := Player.Player.IsActive[]
#         then:
#             DeflectTrigger.Trigger()
#             SpawnPosition := Player.GetForwardPosition(255.0)
#             SpawnTransform := transform:
#                 Translation:= SpawnPosition.Translation
#                 Rotation:= SpawnPosition.Rotation
#                 Scale := vector3:
#                     X:= SlowBallSpeed
#                     Y:= 1.0 # Custom Data for Control Rig
#                     Z:= 0.0 # Custom Data for Control Rig

#             SpawnedProp:= SpawnProp(RedBallProp, SpawnTransform)
#             spawn {ActivateDefflectEffect(SpawnPosition.Translation, SpawnPosition.Rotation)}
#             if (ValidProp := SpawnedProp(0)?):
#                 spawn:
#                     # Start waiting for the ball to collide with something
#                     AwaitBallCollision(ValidProp, 1)

#     ActivateHitEffect(BallPosition:vector3, BallRotation:rotation)<suspends>:void =
#         ## Ajustar position de HIT
#         # Recheckear si existe en respawn
#         MyVFX :cancelable= SpawnParticleSystem(HitEffect,BallPosition,?Rotation := BallRotation)
#         Sleep(0.3)
#         MyVFX.Cancel()
        
#     ActivateDefflectEffect(BallPosition:vector3, BallRotation:rotation)<suspends>:void =
#         ## Ajustar position de HIT
#         # Recheckear si existe en respawn
#         MyVFX :cancelable= SpawnParticleSystem(DefflectEffect,BallPosition,?Rotation := BallRotation)
#         Sleep(0.3)
#         MyVFX.Cancel()

#     IncreaseFirstRoundPlayRate(Playrate : float) :void =
#         #set CurrentPlayRate += (Playrate * 10)
#         BallLaunchersCinematic.SetPlayRate(3.0)

#     IncreaseBallSpeedBasedOnTime(SpeedScalingFactor:float):void =
#         var MaxSpeed :float= 3.0
#             if(BallSpeed >= MaxSpeed):
#                 set BallSpeed = MaxSpeed
#             else:
#                 set BallSpeed += (SpeedScalingFactor * 10)

#     # Fire rate system, increase rate based on tick function
#     Tick()<suspends>:void=
#         var Time :float= 1.0
#         var SleepFire :float = 1.0
#         var HandleTime :float=0.0
#         race:
#             EndedEvent.Await()
#             loop:
#                 set Time += 0.1
#                 set HandleTime += 1.0
#                 DisableAttentionSound(SleepFire)
#                 if(HandleTime >= TimeToIncreaseFireRate): 
#                     if(SleepFire <= 0.3):
#                         set SleepFire = 0.3
#                     else:
#                         set SleepFire -= ScalingFactor
#                     #IncreaseBallSpeedBasedOnTime(1.0)
#                     IncreaseFirstRoundPlayRate(ScalingFactor)
#                     set HandleTime = 0.0
#                 if(MarkerDisplayed = false and Time >= (NextFireTime - MarkerTimeBeforeFire)):
#                     CreateBallLauncher()
#                     set MarkerDisplayed = true
#                 if((Time) >= NextFireTime):
#                     Sleep(SleepFire)
#                     Shoot()
#                     set NextFireTime = (Time  + 1.0) / InitialFireRate
#                     set MarkerDisplayed = false
#                 Sleep(0.1)

#     Name<override>: string = "US Open Tennis Storm"
#     StartingMessage<override>:string = ""
#     GetVotingImage<override>()<transacts>: texture = VerseExposed.SQ_TENNIS