
using { /Fortnite.com/Devices }
using { /Verse.org/Simulation }
using { /Verse.org/Simulation/Tags }
using { /UnrealEngine.com/Temporary/Diagnostics }
using { /UnrealEngine.com/Temporary/UI }
using { /Fortnite.com/UI }
using { /UnrealEngine.com/Temporary/SpatialMath }
using { /Verse.org/Colors }
using { /Verse.org/Assets }
using { /Fortnite.com/Characters }
using { /Fortnite.com/Game }
using { /Verse.org/Random }

custom_hero_player := class<unique>():
    Player : player
    HeroDevice : base_hero_device
    #EndedEvent : event([]player)
    var UltCharge : int  = 0
    var UltimateAbilityImage : ?texture_block = false
    var UltimateHUDBackgroundImage : ?texture_block = false
    UltimateHUDBackgroundTexture: texture = VerseExposed.T_UltHudBackGround
    UltimateCharge1 : texture = VerseExposed.T_UltCharge1
    UltimateCharge2 : texture = VerseExposed.T_UltCharge2
    #This will be set later
    var UltimateAbilityTexture : texture = VerseExposed.TransparentTexture
    var UltimateNoChargeTexture : texture = VerseExposed.TransparentTexture
    var Canvas : canvas = canvas{}
    var WallHackVFX : vfx_spawner_device = vfx_spawner_device{}

    #Maybe
    var UltUsed : logic = false

    InitPlayer():void=
        if:
            Agent := agent[Player]
            Fort := Agent.GetFortCharacter[]
        then:
            # Spawn prop at player location
            Print("Initializing Custom Hero for Player")
            HeroDevice.InitCustomHero(Player)
            #HeroDevice.RegisterHeroSubscriber(UltimateInputTrigger.PressedEvent.Subscribe(CustomHeroPlayer.OnUltimatePressed))
            set UltimateAbilityTexture = HeroDevice.UltimateAbilityTexture
            set UltimateNoChargeTexture = HeroDevice.UltimateNoChargeTexture
            set Canvas = MakeHeroUI()
            UpdateUltChargeUI()
            if(PlayerUI := GetPlayerUI[Player]):
                PlayerUI.AddWidget(Canvas, player_ui_slot{InputMode := ui_input_mode.None})
            HeroDevice.MiniGameSuperHero.Manager.UiMessageIndividual.SetText(StringToMessage(HeroDevice.InitialUIMessage))
            HeroDevice.MiniGameSuperHero.Manager.UiMessageIndividual.Show(Player)

    MakeHeroUI():canvas =
        var NewUI : canvas = canvas{}
        set UltimateAbilityImage = option{texture_block{DefaultImage := UltimateAbilityTexture, DefaultDesiredSize := vector2{X:= 256.0, Y := 256.0}}}
        set UltimateHUDBackgroundImage = option{texture_block{DefaultImage := UltimateHUDBackgroundTexture, DefaultDesiredSize := vector2{X:= 256.0, Y := 256.0}}}
        if (BackgroundImage := UltimateHUDBackgroundImage?, AbilityImage := UltimateAbilityImage?):
            set NewUI = canvas:
                Slots := array:
                    canvas_slot:
                        Anchors := anchors{Minimum := vector2{X:= 0.5, Y := 1.0}, Maximum := vector2{X:= 0.5, Y:= 1.0}}
                        Offsets := margin{Left := 0.0, Top := 0.0, Right := 256.0, Bottom := 256.0}
                        Alignment := vector2{X:= 0.5, Y := 1.0}
                        SizeToContent := true
                        ZOrder := 0
                        Widget := BackgroundImage
                    canvas_slot:
                        Anchors := anchors{Minimum := vector2{X:= 0.5, Y := 1.0}, Maximum := vector2{X:= 0.5, Y:= 1.0}}
                        Offsets := margin{Left := 0.0, Top := -25.0, Right := 256.0, Bottom := 256.0}
                        Alignment := vector2{X:= 0.5, Y := 1.0}
                        SizeToContent := true
                        ZOrder := 1
                        Widget := AbilityImage
        return NewUI     

    RemoveWidget():void=
        if(PlayerUI := GetPlayerUI[Player]):
            PlayerUI.RemoveWidget(Canvas)

    AddUltCharge():void=
        set UltCharge = Min(UltCharge + 1, 2)
        Print("New Ult = {UltCharge}")
        UpdateUltChargeUI()

    # OnUltimatePressed(Agent : agent):void=
    #     if(P := player[Agent] and UltCharge >= 2 and UltUsed = false):
    #         Print("Ultimate")
    #         set UltUsed = true
    #         spawn{HeroDevice.ActivateUltimate(P)}
    #     else:
    #         Print("Not enough for ult")

    # OnUltimatePressed(Agent : agent, Agent2 :agent):void=
    #     if(UltCharge >= 2 and UltUsed = false):
    #         Print("Ultimate")
    #         set UltUsed = true
    #         spawn{HeroDevice.ActivateUltimate(Player)}
    #     else:
    #         Print("Not enough for ult")

    UpdateUltChargeUI():void=
        if(UltUsed = false):
            if (UltCharge = 0, BackgroundImage := UltimateHUDBackgroundImage?, AbilityImage := UltimateAbilityImage?):
                BackgroundImage.SetImage(UltimateHUDBackgroundTexture)
                AbilityImage.SetImage(UltimateNoChargeTexture)
            if (UltCharge = 1, BackgroundImage := UltimateHUDBackgroundImage?, AbilityImage := UltimateAbilityImage?):
                BackgroundImage.SetImage(UltimateCharge1)
                AbilityImage.SetImage(UltimateNoChargeTexture)
            if (UltCharge = 2, BackgroundImage := UltimateHUDBackgroundImage?, AbilityImage := UltimateAbilityImage?):
                BackgroundImage.SetImage(UltimateCharge2)
                AbilityImage.SetImage(UltimateAbilityTexture)
                HeroDevice.MiniGameSuperHero.UltimateChargedAudio.Play(Player)
                HeroDevice.MiniGameSuperHero.Manager.UiMessageIndividual.SetText(StringToMessage("Ultimate is Ready [Crouch]"))
                HeroDevice.MiniGameSuperHero.Manager.UiMessageIndividual.Show(Player)

    # ClearUltCharge():void=
    #     set UltCharge = 0
    #     UpdateUltChargeUI()
    #     if(MiniGameSuperHero := bMiniGameSuperHero?, Player := bPlayer?):
    #         MiniGameSuperHero.UltimateEndAudio.Play(Player)

    # WatchEndedEvent<final>()<suspends>:void=
    #     Print("EndedEvent Spawning")
    #     EndedEvent.Await()
    #     Print("EndedEvent happened!")
    #     if(PlayerUI := GetPlayerUI[Player]):
    #         PlayerUI.RemoveWidget(Canvas)

UI_Hero_Select2 := class():
    Texture : texture = VerseExposed.T_HeroUI
    var Button1 : button_loud = button_loud{}
    var Button2 : button_loud = button_loud{}
    var Button3 : button_loud = button_loud{}
    var Button4 : button_loud = button_loud{}
    var Button5 : button_loud = button_loud{}
    var Button6 : button_loud = button_loud{}
    var Button7 : button_loud = button_loud{}
    var Button8 : button_loud = button_loud{}

    var Canva : canvas = canvas{}
    
    RemoveUI(Player : player):void=
        if(PlayerUI := GetPlayerUI[Player]):
            PlayerUI.RemoveWidget(Canva)

    CreateUI(Player : player, MiniGame : minigame75_SuperHeroesBirdo):canvas=
        Button1.SetText(_StringToMessage("Invisible"))
        Button2.SetText(_StringToMessage("Double Pump"))
        Button3.SetText(_StringToMessage("Second Chance"))
        Button4.SetText(_StringToMessage("Flight"))
        Button5.SetText(_StringToMessage("FOAD"))
        Button6.SetText(_StringToMessage("ONE SHOT"))
        Button7.SetText(_StringToMessage("Aimbot"))
        Button8.SetText(_StringToMessage("Boxing Gloves"))


        Image : texture_block = texture_block{DefaultImage := Texture, DefaultDesiredSize := vector2{X:= 1920.0, Y := 1080.0}}
        NewUI : canvas = canvas:
            Slots := array:
                canvas_slot:
                    Anchors := anchors{Minimum := vector2{X:= 0.5, Y := 0.5}, Maximum := vector2{X:= 0.5, Y:= 0.5}}
                    Offsets := margin{Left := 0.0, Top := 0.0, Right := 1920.0, Bottom := 1080.0}
                    Alignment := vector2{X:= 0.5, Y := 0.5}
                    SizeToContent := true
                    Widget := Image

                canvas_slot:
                    Anchors := anchors{Minimum := vector2{X:= 0.5, Y := 0.5}, Maximum := vector2{X:= 0.5, Y:= 0.5}}
                    Offsets := margin{Left := -660.0, Top := -10.0, Right := 360.0, Bottom := 82.0}
                    Alignment := vector2{X:= 0.5, Y := 0.5}
                    SizeToContent := true
                    Widget := Button1

                canvas_slot:
                    Anchors := anchors{Minimum := vector2{X:= 0.5, Y := 0.5}, Maximum := vector2{X:= 0.5, Y:= 0.5}}
                    Offsets := margin{Left := -667.86853, Top := -478.0, Right := 360.0, Bottom := 82.0}
                    Alignment := vector2{X:= 0.5, Y := 0.5}
                    SizeToContent := true
                    Widget := text_block{DefaultText := _StringToMessage("")}

                canvas_slot:
                    Anchors := anchors{Minimum := vector2{X:= 0.5, Y := 0.5}, Maximum := vector2{X:= 0.5, Y:= 0.5}}
                    Offsets := margin{Left := -210.0, Top := -10.0, Right := 360.0, Bottom := 83.0}
                    Alignment := vector2{X:= 0.5, Y := 0.5}
                    SizeToContent := true
                    Widget := Button2

                canvas_slot:
                    Anchors := anchors{Minimum := vector2{X:= 0.5, Y := 0.5}, Maximum := vector2{X:= 0.5, Y:= 0.5}}
                    Offsets := margin{Left := -232.0, Top := -478.0, Right := 360.0, Bottom := 83.0}
                    Alignment := vector2{X:= 0.5, Y := 0.5}
                    SizeToContent := true
                    Widget := text_block{DefaultText := _StringToMessage("")}

                canvas_slot:
                    Anchors := anchors{Minimum := vector2{X:= 0.5, Y := 0.5}, Maximum := vector2{X:= 0.5, Y:= 0.5}}
                    Offsets := margin{Left := 220.0, Top := -10.0, Right := 360.0, Bottom := 83.0}
                    Alignment := vector2{X:= 0.5, Y := 0.5}
                    SizeToContent := true
                    Widget := Button3

                canvas_slot:
                    Anchors := anchors{Minimum := vector2{X:= 0.5, Y := 0.5}, Maximum := vector2{X:= 0.5, Y:= 0.5}}
                    Offsets := margin{Left := 209.0, Top := -478.0, Right := 360.0, Bottom := 83.0}
                    Alignment := vector2{X:= 0.5, Y := 0.5}
                    SizeToContent := true
                    Widget := text_block{DefaultText := _StringToMessage("")}
                
                canvas_slot:
                    Anchors := anchors{Minimum := vector2{X:= 0.5, Y := 0.5}, Maximum := vector2{X:= 0.5, Y:= 0.5}}
                    Offsets := margin{Left := 660.0, Top := -10.0, Right := 360.0, Bottom := 83.0}
                    Alignment := vector2{X:= 0.5, Y := 0.5}
                    SizeToContent := true
                    Widget := Button4

                canvas_slot:
                    Anchors := anchors{Minimum := vector2{X:= 0.5, Y := 0.5}, Maximum := vector2{X:= 0.5, Y:= 0.5}}
                    Offsets := margin{Left := 644.0, Top := -478.0, Right := 360.0, Bottom := 83.0}
                    Alignment := vector2{X:= 0.5, Y := 0.5}
                    SizeToContent := true
                    Widget := text_block{DefaultText := _StringToMessage("")}

                canvas_slot:
                    Anchors := anchors{Minimum := vector2{X:= 0.5, Y := 0.5}, Maximum := vector2{X:= 0.5, Y:= 0.5}}
                    Offsets := margin{Left := -640.86853, Top := 450.0, Right := 360.0, Bottom := 83.0}
                    Alignment := vector2{X:= 0.5, Y := 0.5}
                    SizeToContent := true
                    Widget := Button5

                canvas_slot:
                    Anchors := anchors{Minimum := vector2{X:= 0.5, Y := 0.5}, Maximum := vector2{X:= 0.5, Y:= 0.5}}
                    Offsets := margin{Left := -667.86853, Top := 59.0, Right := 360.0, Bottom := 83.0}
                    Alignment := vector2{X:= 0.5, Y := 0.5}
                    SizeToContent := true
                    Widget := text_block{DefaultText := _StringToMessage("")}

                canvas_slot:
                    Anchors := anchors{Minimum := vector2{X:= 0.5, Y := 0.5}, Maximum := vector2{X:= 0.5, Y:= 0.5}}
                    Offsets := margin{Left := -240.0, Top := 450.0, Right := 360.0, Bottom := 83.0}
                    Alignment := vector2{X:= 0.5, Y := 0.5}
                    SizeToContent := true
                    Widget := Button6

                canvas_slot:
                    Anchors := anchors{Minimum := vector2{X:= 0.5, Y := 0.5}, Maximum := vector2{X:= 0.5, Y:= 0.5}}
                    Offsets := margin{Left := -232.0, Top := 59.0, Right := 360.0, Bottom := 83.0}
                    Alignment := vector2{X:= 0.5, Y := 0.5}
                    SizeToContent := true
                    Widget := text_block{DefaultText := _StringToMessage("")}

                canvas_slot:
                    Anchors := anchors{Minimum := vector2{X:= 0.5, Y := 0.5}, Maximum := vector2{X:= 0.5, Y:= 0.5}}
                    Offsets := margin{Left := 220.0, Top := 450.0, Right := 360.0, Bottom := 83.0}
                    Alignment := vector2{X:= 0.5, Y := 0.5}
                    SizeToContent := true
                    Widget := Button7

                canvas_slot:
                    Anchors := anchors{Minimum := vector2{X:= 0.5, Y := 0.5}, Maximum := vector2{X:= 0.5, Y:= 0.5}}
                    Offsets := margin{Left := 209.0, Top := 59.0, Right := 360.0, Bottom := 83.0}
                    Alignment := vector2{X:= 0.5, Y := 0.5}
                    SizeToContent := true
                    Widget := text_block{DefaultText := _StringToMessage("")}
                
                canvas_slot:
                    Anchors := anchors{Minimum := vector2{X:= 0.5, Y := 0.5}, Maximum := vector2{X:= 0.5, Y:= 0.5}}
                    Offsets := margin{Left := 644.0, Top := 450.0, Right := 360.0, Bottom := 83.0}
                    Alignment := vector2{X:= 0.5, Y := 0.5}
                    SizeToContent := true
                    Widget := Button8

                canvas_slot:
                    Anchors := anchors{Minimum := vector2{X:= 0.5, Y := 0.5}, Maximum := vector2{X:= 0.5, Y:= 0.5}}
                    Offsets := margin{Left := 644.0, Top := 59.0, Right := 360.0, Bottom := 83.0}
                    Alignment := vector2{X:= 0.5, Y := 0.5}
                    SizeToContent := true
                    Widget := text_block{DefaultText := _StringToMessage("")}
        #Bindings

        MiniGame.RegisterSubscriber(Button1.OnClick().Subscribe(MiniGame.OnHeroSelect_Any))
        MiniGame.RegisterSubscriber(Button1.OnClick().Subscribe(MiniGame.OnHeroSelect_1))

        MiniGame.RegisterSubscriber(Button2.OnClick().Subscribe(MiniGame.OnHeroSelect_Any))
        MiniGame.RegisterSubscriber(Button2.OnClick().Subscribe(MiniGame.OnHeroSelect_2))

        MiniGame.RegisterSubscriber(Button3.OnClick().Subscribe(MiniGame.OnHeroSelect_Any))
        MiniGame.RegisterSubscriber(Button3.OnClick().Subscribe(MiniGame.OnHeroSelect_3))

        MiniGame.RegisterSubscriber(Button4.OnClick().Subscribe(MiniGame.OnHeroSelect_Any))
        MiniGame.RegisterSubscriber(Button4.OnClick().Subscribe(MiniGame.OnHeroSelect_4))

        MiniGame.RegisterSubscriber(Button5.OnClick().Subscribe(MiniGame.OnHeroSelect_Any))
        MiniGame.RegisterSubscriber(Button5.OnClick().Subscribe(MiniGame.OnHeroSelect_5))

        MiniGame.RegisterSubscriber(Button6.OnClick().Subscribe(MiniGame.OnHeroSelect_Any))
        MiniGame.RegisterSubscriber(Button6.OnClick().Subscribe(MiniGame.OnHeroSelect_6))

        MiniGame.RegisterSubscriber(Button7.OnClick().Subscribe(MiniGame.OnHeroSelect_Any))
        MiniGame.RegisterSubscriber(Button7.OnClick().Subscribe(MiniGame.OnHeroSelect_7))

        MiniGame.RegisterSubscriber(Button8.OnClick().Subscribe(MiniGame.OnHeroSelect_Any))
        MiniGame.RegisterSubscriber(Button8.OnClick().Subscribe(MiniGame.OnHeroSelect_8))

        set Canva = NewUI

        return NewUI

# A Verse-authored creative device that can be placed in a level
minigame75_SuperHeroesBirdo := class(minigame_device):
    @editable CeilingProp : creative_prop = creative_prop{}
    @editable InputDebugSprint : input_trigger_device = input_trigger_device{}
    var PlayerUIMap : [player]UI_Hero_Select2 = map{}
    var PlayerHeroMap : [player]custom_hero_player = map{}
    @editable DebugMode : logic = false

    #Audio Devices
    @editable UltimateChargedAudio : audio_player_device = audio_player_device{}
    @editable UltimateUseAudio : audio_player_device = audio_player_device{}

    @editable UltimateInputTrigger : input_trigger_device = input_trigger_device{}
    @editable WallHackVFXDevices : []vfx_spawner_device = array{}

    @editable HeroDevice1 : ?base_hero_device = false
    @editable HeroDevice2 : ?base_hero_device = false
    @editable HeroDevice3 : ?base_hero_device = false
    @editable HeroDevice4 : ?base_hero_device = false
    @editable HeroDevice5 : ?base_hero_device = false
    @editable HeroDevice6 : ?base_hero_device = false
    @editable HeroDevice7 : ?base_hero_device = false
    @editable HeroDevice8 : ?base_hero_device = false

    var AllHeroDevices : []base_hero_device = array{}

    GetVotingImage<override>()<transacts>: texture = VerseExposed.SQ_ULTHERO
    DisplayIcon<override>: ?texture = option{VerseExposed.TransparentTexture}

    #REPLACES FindCreativeObjectsWithTag Function
    FindObjectsWithTag<public>(Tag:tag)<transacts>:[]creative_object_interface=
        TaggedObjects:=FindCreativeObjectsWithTag(Tag)
        var Objects:[]creative_object_interface=array{}
        for (Obj:TaggedObjects, TaggedObject:=creative_device_base[Obj]):
            set Objects += array{TaggedObject}
        return Objects

    OnMinigameDeviceBegin<override>()<suspends>:void=
        Print("SETTING HERO CLASS STUFF")
        HeroDevices := FindObjectsWithTag(hero_device{})
        for (Obj: HeroDevices):
            if(HeroDevice := base_hero_device[Obj]):
                set HeroDevice.MiniGameSuperHero = Self
                set AllHeroDevices += array{HeroDevice}
        Print("AllHeroDevices Length = {AllHeroDevices.Length}")

    _Start<override>()<suspends>: void =
        if(CeilingProp.TeleportTo[vector3{X:= 7168.0, Y:= -18958.0, Z:= 4752.0}, rotation{}]):
        if(DebugMode = true):
            InputDebugSprint.Enable()
            InputDebugSprint.PressedEvent.Subscribe(OnDebugSprint)

        UltimateInputTrigger.Enable()

        for(HeroDevice : AllHeroDevices):
            HeroDevice.HeroDeviceInit()

        for(Player : GetRemainingPlayers(), Fort := Player.GetFortCharacter[]):
            NewUI := UI_Hero_Select2{}
            if(PlayerUI := GetPlayerUI[Player]):
                if(set PlayerUIMap[Player] = NewUI):
                    PlayerUI.AddWidget(NewUI.CreateUI(Player, Self), player_ui_slot{InputMode := ui_input_mode.All})

        # On screen hero selection timer
        Sleep(8.0)
        # Remove Hero selection UI
        for(Player -> PlayerUI : PlayerUIMap):
            PlayerUI.RemoveUI(Player)
        set PlayerUIMap = map{}

        # If they didn't pick
        Sleep(1.0)
        for(Player : GetRemainingPlayers()):
            if (PlayerHeroMap[Player]):
            else:
                if(HeroDevice := RandomHeroDevice := Shuffle(AllHeroDevices)[0]):
                    HeroObj := HeroDevice.CreateHeroObj(Player)
                    if(set PlayerHeroMap[Player] = HeroObj):
                    HeroObj.InitPlayer()

        var i : int = 0
        for(Player -> HeroClass : PlayerHeroMap):
            if(WallHackVFX := WallHackVFXDevices[i]):
                WallHackVFX.Enable()
                set HeroClass.WallHackVFX = WallHackVFX
                set i += 1
        spawn{MoveWallHackVFXToPlayers()}

        Manager.UiMessage.SetText(StringToMessage("Get 2 Eliminations to get your ULTIMATE!"))
        Manager.UiMessage.Show()

    OnSafeTimeEnded<override>():void=
        # Subscribe to their ults
        for(Player : GetRemainingPlayers()):
            if:
                Agent := agent[Player]
                CustomHeroPlayer := PlayerHeroMap[Player]
            then:
                UltimateInputTrigger.Register(Agent)
                RegisterSubscriber(UltimateInputTrigger.PressedEvent.Subscribe(OnUltimatePressed))

    OnUltimatePressed(Agent : agent):void=
        if:
            Player := player[Agent]
            CustomHeroPlayer := PlayerHeroMap[Player]
            CustomHeroPlayer.UltCharge >= 2
            CustomHeroPlayer.UltUsed = false
        then:
            Print("Ultimate")
            set CustomHeroPlayer.UltUsed = true
            spawn{CustomHeroPlayer.HeroDevice.ActivateUltimate(Player)}
        else:
            Print("Not enough for ult")

    OnDebugSprint(Agent:agent):void=
        Print("Debug Sprint")
        if(Player := player[Agent], HeroClass := PlayerHeroMap[Player]):
            HeroClass.AddUltCharge()

    _OnMinigameEnded<override>(Winners: []player)<suspends>:void=
        UltimateInputTrigger.UnregisterAll()
        UltimateInputTrigger.Disable()

        for (HeroDevice: AllHeroDevices):
            Print("Resetting Hero Device")
            spawn{HeroDevice.HeroDeviceReset()}

        for(WallHackVFX : WallHackVFXDevices):
            WallHackVFX.Disable()
            if(WallHackVFX.TeleportTo[vector3{}, rotation{}]):

        for(K -> V :PlayerHeroMap):
            Print("Removing Widget")
            V.RemoveWidget()
        set PlayerHeroMap = map{}

    _Reset<override>()<suspends>:void=
        #backup for lasers
        for(Player:GetRemainingPlayers()):
            if(Fort:=Player.GetFortCharacter[]):
                Fort.SetVulnerability(true)    
        if(CeilingProp.TeleportTo[vector3{X:= 11403.0, Y:= -18958.0, Z:= 1919.0}, rotation{}]):

    _OnPlayerEliminated<override>(Result: elimination_result)<suspends>:void =
        if(Agent := Result.EliminatingCharacter?.GetAgent[], Player := player[Agent]):
            # Increase players ult for getting an elim
            if(CustomHeroClass := PlayerHeroMap[Player]):
                CustomHeroClass.AddUltCharge()
        if(Agent := Result.EliminatedCharacter.GetAgent[], Player := player[Agent]):
            # Remove player from hero class list
            if(CustomHeroClass := PlayerHeroMap[Player]):
                CustomHeroClass.RemoveWidget()

    OnHeroSelect_Any(Widget: widget_message):void=
        if (PlayerUI := PlayerUIMap[Widget.Player]):
            PlayerUI.RemoveUI(Widget.Player)

    # wall hack
    OnHeroSelect_1(Widget: widget_message):void=
        Print("On Hero Select1")
        if(HeroDevice := HeroDevice1?):
            HeroObj := HeroDevice.CreateHeroObj(Widget.Player)
            if(set PlayerHeroMap[Widget.Player] = HeroObj):
            HeroObj.InitPlayer()

    OnHeroSelect_2(Widget: widget_message):void=
        Print("On Hero Select2")
        if(HeroDevice := HeroDevice2?):
            HeroObj := HeroDevice.CreateHeroObj(Widget.Player)
            if(set PlayerHeroMap[Widget.Player] = HeroObj):
            HeroObj.InitPlayer()

    OnHeroSelect_3(Widget: widget_message):void=
        Print("On Hero Select3")
        if(HeroDevice := HeroDevice3?):
            HeroObj := HeroDevice.CreateHeroObj(Widget.Player)
            if(set PlayerHeroMap[Widget.Player] = HeroObj):
            HeroObj.InitPlayer()

    OnHeroSelect_4(Widget: widget_message):void=
        Print("On Hero Select4")
        if(HeroDevice := HeroDevice4?):
            HeroObj := HeroDevice.CreateHeroObj(Widget.Player)
            if(set PlayerHeroMap[Widget.Player] = HeroObj):
            HeroObj.InitPlayer()

    OnHeroSelect_5(Widget: widget_message):void=
        Print("On Hero Select5")
        if(HeroDevice := HeroDevice5?):
            HeroObj := HeroDevice.CreateHeroObj(Widget.Player)
            if(set PlayerHeroMap[Widget.Player] = HeroObj):
            HeroObj.InitPlayer()

    OnHeroSelect_6(Widget: widget_message):void=
        Print("On Hero Select6")
        if(HeroDevice := HeroDevice6?):
            HeroObj := HeroDevice.CreateHeroObj(Widget.Player)
            if(set PlayerHeroMap[Widget.Player] = HeroObj):
            HeroObj.InitPlayer()

    OnHeroSelect_7(Widget: widget_message):void=
        Print("On Hero Select7")
        if(HeroDevice := HeroDevice7?):
            HeroObj := HeroDevice.CreateHeroObj(Widget.Player)
            if(set PlayerHeroMap[Widget.Player] = HeroObj):
            HeroObj.InitPlayer()
            
    OnHeroSelect_8(Widget: widget_message):void=
        Print("On Hero Select8")
        if(HeroDevice := HeroDevice8?):
            HeroObj := HeroDevice.CreateHeroObj(Widget.Player)
            if(set PlayerHeroMap[Widget.Player] = HeroObj):
            HeroObj.InitPlayer()

    MoveWallHackVFXToPlayers()<suspends>:void=
        # First teleport all VFX to players before moving them, we do this because players are far away when this function is first called
        for(Player -> HeroClass : PlayerHeroMap):
            if (FortChar := Player.GetFortCharacter[]):
                if(HeroClass.WallHackVFX.TeleportTo[FortChar.GetTransform().Translation, FortChar.GetTransform().Rotation]):
            else:
                if(HeroClass.WallHackVFX.TeleportTo[vector3{X := 0.0, Y := 0.0, Z := 0.0}, MakeRotationFromYawPitchRollDegrees(0.0, 0.0, 0.0)]):
        # VFX will follow players
        race:
            block:
                loop:
                    Sleep(0.1)
                    for(Player -> HeroClass : PlayerHeroMap):
                        if (FortChar := Player.GetFortCharacter[]):
                            spawn{HeroClass.WallHackVFX.MoveTo(FortChar.GetTransform().Translation, FortChar.GetTransform().Rotation, 0.1)}
                        else:
                            if(HeroClass.WallHackVFX.TeleportTo[vector3{X := 0.0, Y := 0.0, Z := 0.0}, MakeRotationFromYawPitchRollDegrees(0.0, 0.0, 0.0)]):
            EndedEvent.Await()