using { /Fortnite.com/Devices }
using { /Verse.org/Simulation }
using { /Verse.org/Simulation/Tags }
using { /UnrealEngine.com/Temporary/Diagnostics }
using { /UnrealEngine.com/Temporary/UI }
using { /Fortnite.com/UI }
using { /UnrealEngine.com/Temporary/SpatialMath }
using { /Verse.org/Colors }
using { /Verse.org/Assets }
using { /Fortnite.com/Characters }
using { /Fortnite.com/Game }
using { /Verse.org/Random }

hero_device := class(tag) {}
base_hero_device := class(crashable_device):
    @editable InitialUIMessage : string = ""
    @editable DefaultItemGranters : []minigame_device_granter_config = array{}
    @editable UltimateItemGranters : []minigame_device_granter_config = array{}
    @editable OptClassSelector : ?class_and_team_selector_device = false
    @editable UltimateClassSelector : ?class_and_team_selector_device = false
    @editable OptHeroImpressionAnalytics : ?analytics_device = false

    var MiniGameSuperHero : minigame75_SuperHeroesBirdo = minigame75_SuperHeroesBirdo{}
    # Hud Textures
    var UltimateAbilityTexture : texture = VerseExposed.TransparentTexture
    var UltimateNoChargeTexture : texture = VerseExposed.TransparentTexture
    var DisplayHudTexture : texture = VerseExposed.TransparentTexture

    var<private> HeroSubscribables : []cancelable = array{}

    CreateHeroObj(Player : player):custom_hero_player = 
        return custom_hero_player{Player := Player, HeroDevice := Self}

    RegisterHeroSubscriber<final>(Cancelable: cancelable):void=
        set HeroSubscribables += array{Cancelable}

    #Hud Stuff
    HeroDeviceInit():void= return

    #Hud Stuff
    InitCustomHero<final>(Player : player):void=
        Print("Base Init Stuff")
        #Hud Stuff
        for(DefaultGranter : DefaultItemGranters, not DefaultGranter.GrantAfterSafeTime?):
            DefaultGranter.Grant(array{Player})

        if(ClassSelector := OptClassSelector?):
            ClassSelector.ChangeClass(Player)

        MiniGameSuperHero.Manager.DisplayIconManager.MakeCustomHUD(Player, DisplayHudTexture, 512.0)

        if(AnalyticDevice := OptHeroImpressionAnalytics?):
            Print("HeroDevice - {DeviceName} Analytic Impression")
            AnalyticDevice.Submit(Player)

        _InitCustomHero(Player)

    _InitCustomHero(Player : player):void=return

    ResetCustomHero(Player : player):void=return

    ActivateUltimate(Player : player)<suspends>:void=
        MiniGameSuperHero.UltimateUseAudio.Play(Player)
        for(UltimateGranter : UltimateItemGranters, not UltimateGranter.GrantAfterSafeTime?):
            UltimateGranter.Grant(array{Player})

        if(ClassSelector := UltimateClassSelector?):
            ClassSelector.ChangeClass(Player)

        spawn{_ActivateUltimate(Player)}

    _ActivateUltimate(Player : player)<suspends>:void= return

    HeroDeviceReset<final>()<suspends>:void =
        for(Subscriber : HeroSubscribables):
            Subscriber.Cancel()
        set HeroSubscribables = array{}
        spawn{_HeroDeviceReset()}

    _HeroDeviceReset()<suspends>:void=return

hero1_invis_device := class(base_hero_device):
    var UltimateAbilityTexture <override>: texture = VerseExposed.T_WallHackUltimate
    var UltimateNoChargeTexture <override>: texture = VerseExposed.T_WallHackUltimateNoCharge
    var DisplayHudTexture <override>: texture = VerseExposed.SQ_INVISIBLE
        
hero2_double_pump_device := class(base_hero_device):
    @editable LockShootTime : float = 0.66
    @editable DoublePumpGranter : item_granter_device = item_granter_device{}
    @editable DoubleHavocGranter : item_granter_device = item_granter_device{}
    @editable PumpConditionalButton :conditional_button_device = conditional_button_device{}
    @editable HavocConditionalButton :conditional_button_device = conditional_button_device{}
    @editable InputTriggerShoot : input_trigger_device = input_trigger_device{}
    var PlayerShootLock: [player]logic = map{}

    var UltimateAbilityTexture <override>: texture = VerseExposed.T_DoublePumpUltimate
    var UltimateNoChargeTexture <override>: texture = VerseExposed.T_DoublePumpUltimateNoCharge
    var DisplayHudTexture <override>: texture = VerseExposed.SQ_DOUBLEPUMP_ICON

    HeroDeviceInit<override>():void=
        Print("Initializing Hero Device")
        InputTriggerShoot.Enable()

    _InitCustomHero<override>(Player : player):void=
        InputTriggerShoot.Register(Player)
        if(set PlayerShootLock[Player] = false):
        RegisterHeroSubscriber(InputTriggerShoot.PressedEvent.Subscribe(HandleShoot))

    HandleShoot(Agent : agent):void=
        if(PumpConditionalButton.IsHoldingItem[Agent]):
            if(PlayerShootLock[Agent] = false):
                #Make sure its unlocked
                DoublePumpGranter.GrantItem(Agent)
                spawn{LockPlayerShoot(Agent)}
        if(HavocConditionalButton.IsHoldingItem[Agent]):
            if(PlayerShootLock[Agent] = false):
                #Make sure its unlocked
                DoubleHavocGranter.GrantItem(Agent)
                spawn{LockPlayerShoot(Agent)}

    LockPlayerShoot(Agent : agent)<suspends>:void =
        if(Player := player[Agent]):
            #Print("Locking Shoot Button")
            if(set PlayerShootLock[Player] = true):
            Sleep(LockShootTime)
            #Print("Unocking Shoot Button")
            if(set PlayerShootLock[Player] = false):
        
    _HeroDeviceReset<override>():void=
        InputTriggerShoot.UnregisterAll()
        InputTriggerShoot.Disable()
        set PlayerShootLock = map{}



hero3_secondChance_device := class(base_hero_device):
    var UltimateAbilityTexture <override>: texture = VerseExposed.T_BobSummon
    var UltimateNoChargeTexture <override>: texture = VerseExposed.T_BobSummonNoCharge
    var DisplayHudTexture <override>: texture = VerseExposed.SQ_SECONDCHANCE
    var PlayerLife: [player]logic = map{}

    @editable DBNO :down_but_not_out_device = down_but_not_out_device{}
    @editable BobTrigger : trigger_device = trigger_device{}
    @editable NPCToSpawn : npc_spawner_device = npc_spawner_device{}
    var SecondPlayers : []player = array{}

    HeroDeviceInit<override>():void=
        Print("Initializing {DeviceName} Hero Device")
        DBNO.Enable()
        NPCToSpawn.Enable()

    _InitCustomHero<override>(Player : player):void=
        set SecondPlayers += array{Player}
        RegisterHeroSubscriber(DBNO.AgentDownedEvent.Subscribe(OnPlayerDowned))
        if(set PlayerLife[Player] = true):

    OnPlayerDowned(PlayerDowned : agent):void=
        Print("Player downed second chance")
        if(SecondPlayers.Find[PlayerDowned]):
            Print("in second CHANCE array")
            if(PlayerLife[PlayerDowned] = true):
                if(FortPlayerDowned := PlayerDowned.GetFortCharacter[]):
                    spawn{ReviveSequence(PlayerDowned, FortPlayerDowned)}
            else:
                if(Fort := PlayerDowned.GetFortCharacter[]):
                    Fort.SetHealth(1.0)

    ReviveSequence(Agent: agent, Fort : fort_character)<suspends>:void=
        #race to check if in middle of finisher
        Print("Reviving sequence")
        #Fort.SetMaxHealth(5000.0)
        #Fort.SetHealth(200.0)
        Sleep(1.0)
        if(Fort.IsActive[] and Fort.IsDownButNotOut[]):
            DBNO.Revive(Agent)
            if(Player := player[Agent], set PlayerLife[Player] = false):
                #MiniGameSuperHero.Manager.ResetClassSelector.ChangeClass(Player)
            Fort.SetVulnerability(false)
            Print("Invincible")
            Sleep(1.0)
            Print("Not Invincible")
            
            Fort.SetVulnerability(true)

    _HeroDeviceReset<override>():void=
        DBNO.Disable()
        NPCToSpawn.Disable()
        set SecondPlayers = array{}

    _ActivateUltimate<override>(Player : player)<suspends>:void=
        BobTrigger.Trigger(Player)

hero4_flight_device := class(base_hero_device):
    var UltimateAbilityTexture <override>: texture = VerseExposed.T_AutoReloadUltimate
    var UltimateNoChargeTexture <override>: texture = VerseExposed.T_AutoReloadUltimateNoCharge
    var DisplayHudTexture <override>: texture = VerseExposed.SQ_FLIGHT

    @editable LaserBeamManager : V_LaserBeamUltimate = V_LaserBeamUltimate{}
    @editable DamageAmp : damage_amplifier_powerup_device = damage_amplifier_powerup_device{}

    _InitCustomHero<override>(Player : player):void=
        LaserBeamManager.RegisterPlayer(Player)
    
    # HeroDeviceInit<override>():void=
    #     LaserBeamManager.Reset()

    _ActivateUltimate<override>(Player : player)<suspends>:void=
        #LaserBeamManager.RegisterPlayer(Player)
        DamageAmp.Pickup(Player)
        Sleep(0.0)
        spawn:
            LaserBeamManager.LaserBeamUltimate(Player)

hero5_foad_device := class(base_hero_device):
    var UltimateAbilityTexture <override>: texture = VerseExposed.T_FoadUltimate
    var UltimateNoChargeTexture <override>: texture = VerseExposed.T_FoadUltimateNoCharge
    var DisplayHudTexture <override>: texture = VerseExposed.SQ_EXPLOSIVES


    @editable ExplosiveBoxDevice : explosive_device = explosive_device{}
    @editable ItemGranterExplosives : item_granter_device = item_granter_device{}
    @editable InputControls : input_trigger_device = input_trigger_device{}
    @editable GameplayCamera : gameplay_camera_orbit_device = gameplay_camera_orbit_device{}
    @editable FOADVFXPowerUp : visual_effect_powerup_device = visual_effect_powerup_device{}
    @editable MovementSpeedMod : movement_modulator_device = movement_modulator_device{}

    @editable WarningCircleVFXPowerUp : []visual_effect_powerup_device = array{}
    @editable WarningCircleResetVFXPowerUp: visual_effect_powerup_device = visual_effect_powerup_device{}
    @editable ExplosionProp : creative_prop_asset = DefaultCreativePropAsset
    @editable AudioPlayerExplosion :[]audio_player_device = array{}
    @editable ExplosiveDevice : []explosive_device = array{}
    

    var ExplosionCharged : [player]logic = map{}
    var ExplosionReadyCanvas : [player]?canvas = map{}
    var PlayerIsExploding : [player]logic = map{}
    var PowerUpWarningCircleIndex : int = 0
    var AudioDeviceIndex : int = 0
    var ExplosiveDeviceIndex : int = 0

    UltimateEvent : event() = event(){}

    HeroDeviceInit<override>():void=
        InputControls.Enable()
        GameplayCamera.Enable()

    _InitCustomHero<override>(Player : player):void=
        RegisterHeroSubscriber(InputControls.PressedEvent.Subscribe(ExplodeFOADPressed))
        spawn{InitExplosiveItemGranters(Player)}

    InitExplosiveItemGranters(Player : agent)<suspends>:void =
        Print("Start Granting explosives Loop")
        if:
            Agent := agent[Player]
        then:
            race:
                MiniGameSuperHero.EndedEvent.Await()
                UltimateEvent.Await()
                block:
                    loop:
                        ItemGranterExplosives.CycleToRandomItem(Agent)
                        Sleep(5.0)
            Print("Signal Worked!!!")

    ResetCustomHero<override>(Player : player):void=
        if(Agent := agent[Player]):
            if(set  ExplosionCharged[Player] = false):
            RemoveExplosionReadyUI(Agent)

    _HeroDeviceReset<override>():void=
        InputControls.UnregisterAll()
        InputControls.Disable()
        GameplayCamera.RemoveFromAll()
        GameplayCamera.Disable()

    _ActivateUltimate<override>(Player : player)<suspends>:void=
        UltimateEvent.Signal()
        TeamCollection := GetPlayspace().GetTeamCollection()
        # if(Team10 := GetPlayspace().GetTeamCollection().GetTeams()[9]):
        #     option{TeamCollection.AddToTeam[Player, Team10]}
        if(AnyAgent := GetPlayspace().GetPlayers()[0]):
            ExplosiveBoxDevice.Reset()
            ExplosiveBoxDevice.Explode(AnyAgent)
        if(Agent := agent[Player],Fort := Player.GetFortCharacter[]):
            Fort.SetMaxShield(150.0)
            Fort.SetMaxHealth(150.0)
            FOADVFXPowerUp.Pickup(Agent)
            InputControls.Register(Agent)
            GameplayCamera.AddTo(Agent)
            MovementSpeedMod.Activate(Agent)
            spawn:
                ChargeExplosion(Agent)

    ExplodeFOADPressed(Agent:agent):void=
        if(Player := player[Agent], ExplosionCharged[Player] = false):
            Print("Explosion Not Charged.")
            return

        if(Player := player[Agent], PlayerIsExploding[Player] = true):
            Print("Don't fire more - explosion in progress")
            return
        AssignPlayerWarningCircleVFX(Agent)
        if(Player := player[Agent], set PlayerIsExploding[Player] = true):
            if(set  ExplosionCharged[Player] = false):
            RemoveExplosionReadyUI(Agent)
            spawn:
                SpawnExplosion(Agent)

    SpawnExplosion(Agent : agent)<suspends>:void=
        race:
            MiniGameSuperHero.EndedEvent.Await()
            block:
                if(Player := player[Agent]):
                    if(FortChar:= Player.GetFortCharacter[]):
                        
                        WarningCircleResetVFXPowerUp.Pickup(Agent)
                        var NewTransform : transform = FortChar.GetTransform()
                        set NewTransform.Translation = FortChar.GetTransform().Translation  + vector3{Z:=-70.0}
                        NewExplosionProp := SpawnProp(ExplosionProp, NewTransform)

                        PlayExplosionAudio(NewTransform)
                        #ExplosionAudio.Play(Agent)
                        FortChar.SetShield(FortChar.GetShield()+60.0)
                        FortChar.SetHealth(FortChar.GetHealth()+60.0)

                        if(Explosive := ExplosiveDevice[ExplosiveDeviceIndex]):
                            if:
                                Explosive.TeleportTo[NewTransform.Translation, NewTransform.Rotation]
                            Explosive.Explode(Agent)


                            if(ExplosiveDeviceIndex >= ExplosiveDevice.Length - 1):
                                set ExplosiveDeviceIndex = 0
                            else:
                                set ExplosiveDeviceIndex += 1

                            defer:
                                Explosive.Reset()

                        if(set PlayerIsExploding[Player] = false):

                        Sleep(0.25)
                        if(RealProp:= NewExplosionProp(0)?):
                            spawn:
                                DespawnExplosionProp(RealProp)
                            spawn:
                                ChargeExplosion(Agent)
                        # VFX.Cancel()

    PlayExplosionAudio(AudioLocation : transform):void=
        if(AudioDevice := AudioPlayerExplosion[AudioDeviceIndex]):
            if(AudioDevice.TeleportTo[AudioLocation]):
                AudioDevice.Play()
        if(AudioDeviceIndex >= AudioPlayerExplosion.Length - 1):
            set AudioDeviceIndex = 0
        else:
            set AudioDeviceIndex += 1

    DespawnExplosionProp(Prop : creative_prop )<suspends>:void=
            Sleep(2.5) 
            Prop.Dispose()
            Print("Canceling VFX")
            
    ChargeExplosion(Agent : agent)<suspends>:void=
        race:
            MiniGameSuperHero.EndedEvent.Await()
            block:
                AssignPlayerWarningCircleVFX(Agent)
                Sleep(3.5)
                # InputControls.Register(Agent)
                if(Player := player[Agent]):
                    if(set ExplosionCharged[Player] = true):
                        ActivateExplosionReadyUI(Agent)
                        Print("Explosion Charged")

    AssignPlayerWarningCircleVFX(Agent : agent):void=
        if(PowerUp := WarningCircleVFXPowerUp[PowerUpWarningCircleIndex]):
            PowerUp.Pickup(Agent)
        #Increment Powerup Index
        if(PowerUpWarningCircleIndex >= WarningCircleVFXPowerUp.Length - 1):
            set PowerUpWarningCircleIndex = 0
        else:
            set PowerUpWarningCircleIndex += 1

    ActivateExplosionReadyUI(Agent : agent) : void =
        if (Player := player[Agent], PlayerUI := GetPlayerUI[Player]):
            if (MyUI := ExplosionReadyCanvas[Player]?):
                Print("Removing Explosion Canvas")
                PlayerUI.RemoveWidget(MyUI)
                if (set ExplosionReadyCanvas[Player] = false) {}
            NewUI := CreateExplosionReadyCanvas()
            PlayerUI.AddWidget(NewUI)
            if (set ExplosionReadyCanvas[Player] = option{NewUI}) {}

    RemoveExplosionReadyUI(Agent : agent):void=
        if (InPlayer := player[Agent], PlayerUI := GetPlayerUI[InPlayer]):
            Print("Removing Explosion Canvas Beep")
            if (MyUI := ExplosionReadyCanvas[InPlayer]?):
                PlayerUI.RemoveWidget(MyUI)
                if (set ExplosionReadyCanvas[InPlayer] = false) {}

    CreateExplosionReadyCanvas(): canvas =
        ExplosionReadyTexture  : texture_block = texture_block{ DefaultImage := VerseExposed.T_Press_Fire_To_Explode, DefaultDesiredSize := vector2{ X := 512.0, Y := 73.0}, DefaultTint := color{R:=1.0, G:=0.1, B:=0.0}}
        MyDamageText : text_block = text_block{DefaultText := StringToMessage("Press Fire to Explode"), DefaultTextColor := NamedColors.Gray}
        Print("Creating Explosion Charged UI")
        MyCanvas : canvas = canvas:
            Slots := array:
                canvas_slot: 
                    Anchors := anchors{Minimum := vector2{X := 0.5, Y := 0.85}, Maximum := vector2{X := 0.5, Y :=0.85}}
                    Alignment := vector2{X := 0.5, Y := 0.5}
                    Widget := overlay:
                        Slots := array:  
                            overlay_slot:
                                Widget := ExplosionReadyTexture         
        return MyCanvas

hero6_oneshot_device := class(base_hero_device):
    var UltimateAbilityTexture <override>: texture = VerseExposed.T_PickaxeUltimate
    var UltimateNoChargeTexture <override>: texture = VerseExposed.T_PickaxeUltimateNoCharge
    var DisplayHudTexture <override>: texture = VerseExposed.SQ_HERO_ONESHOT
    
    @editable DamageAmp : damage_amplifier_powerup_device = damage_amplifier_powerup_device{}

    @editable
    SlowMo : movement_modulator_device = movement_modulator_device{}
    @editable
    Speed : movement_modulator_device = movement_modulator_device{}

    HeroDeviceInit<override>():void=
        SlowMo.Enable()
        Speed.Enable()

    _InitCustomHero<override>(Player : player):void=
        DamageAmp.Pickup(Player)
        SlowMo.Activate(Player)
        if(Fort := Player.GetFortCharacter[]):
            Fort.SetMaxShield(200.0)
            Fort.SetMaxHealth(200.0)
        
    _HeroDeviceReset<override>():void=
        SlowMo.Disable()
        Speed.Disable()

    _ActivateUltimate<override>(Player : player)<suspends>:void=
        Speed.Activate(Player)


hero7_autoaim_device := class(base_hero_device):
    var UltimateAbilityTexture <override>: texture = VerseExposed.T_AutoAim
    var UltimateNoChargeTexture <override>: texture = VerseExposed.T_AutoAimNoCharge
    var DisplayHudTexture <override>: texture = VerseExposed.SQ_AUTOAIM_ICON

    @editable Controls : gameplay_controls_third_person_device = gameplay_controls_third_person_device{}
    @editable InputTriggerAimIn : input_trigger_device = input_trigger_device{}

    HeroDeviceInit<override>():void=
        Print("Initializing Hero Device")
        InputTriggerAimIn.Enable()
        Controls.Enable()

    _InitCustomHero<override>(Player : player):void=
        InputTriggerAimIn.Register(Player)
        RegisterHeroSubscriber(InputTriggerAimIn.PressedEvent.Subscribe(HandleAimIn))
        RegisterHeroSubscriber(InputTriggerAimIn.ReleasedEvent.Subscribe(HandleAimInReleased))

    HandleAimIn(Player : agent):void=
        Controls.AddTo(Player)

    HandleAimInReleased(TupleReleased : tuple(agent, float)):void=
        Controls.RemoveFrom(TupleReleased(0))
        
    _HeroDeviceReset<override>():void=
        InputTriggerAimIn.UnregisterAll()
        InputTriggerAimIn.Disable()
        Controls.Disable()
    
hero8_gloves_device := class(base_hero_device):
    var UltimateAbilityTexture <override>: texture = VerseExposed.T_NitroGloveUltimate
    var UltimateNoChargeTexture <override>: texture = VerseExposed.T_NitroGloveUltimateNoCharge
    @editable DamageAmpNitro : damage_amplifier_powerup_device = damage_amplifier_powerup_device{}
    @editable Speed : movement_modulator_device = movement_modulator_device{}

    _ActivateUltimate<override>(Player : player)<suspends>:void=
        DamageAmpNitro.Pickup(Player)
        Speed.Activate(Player)

    HeroDeviceInit<override>():void=
        Speed.Enable()
        
    _HeroDeviceReset<override>():void=
        Speed.Disable()