using { /Fortnite.com/Devices }
using { /Verse.org/Simulation }
using { /Verse.org/Assets }
using { /Verse.org/Random }
using { /Verse.org/Simulation/Tags }
using { /Fortnite.com/Characters }
using { /Fortnite.com/Devices/CreativeAnimation }
using { /UnrealEngine.com/Temporary/SpatialMath }
using { /UnrealEngine.com/Temporary/Diagnostics }

############################################################################################################################
################################################## Hi-Chew Tags and Stats ##################################################
############################################################################################################################
HiChewTag := class(tag){}
LowGravBouncer := class(HiChewTag){}

Collectable := class(HiChewTag){}
RandItem := class(Collectable){}
Coin := class(Collectable){}

Powerup := class(Collectable){}
LowGrav := class(Powerup){}
SpeedBoost := class(Powerup){}
############################################################################################################################
################################################## Hi-Chew Custom Classes ##################################################
############################################################################################################################
CustomFallFruitData := class<unique><concrete>:
    @editable
    MovePoint : creative_prop = creative_prop{}
    @editable
    PropManip : prop_manipulator_device = prop_manipulator_device{}

    @editable
    FallDistance : float = -5460.0
    @editable
    DelayBeforeFall : float = 0.0
    @editable
    FallTime : float = 2.0
    @editable
    FallResetTime : float = 2.0
    
    var StartPos : vector3 = vector3{}

    #Loops fruit falling
    FruitFalling()<suspends> : void=
        Sleep(DelayBeforeFall)
        PropManip.ShowProps()

        if :
            AnimationController := MovePoint.GetAnimationController[]
        then:
            KeyFrame1 := keyframe_delta :
                DeltaLocation := vector3{Z:= FallDistance }
                DeltaRotation := MakeRotationFromYawPitchRollDegrees(0.0, 0.0, 0.0)
                Time := FallTime

            KeyFrames : []keyframe_delta = array {KeyFrame1}
            AnimationController.SetAnimation(KeyFrames, ?Mode:=animation_mode.OneShot)
            AnimationController.Play()

    #Teleports and Explodes on fall position
    Explode(InPlayer : player, Explosive : explosive_device)<suspends> : void=
        if:
            Explosive.TeleportTo[MovePoint.GetTransform().Translation, MakeRotationFromYawPitchRollDegrees(0.0, 0.0, 0.0)]
        then:
            Explosive.Reset()
            Sleep(0.1)
            Explosive.Explode(InPlayer)
            PropManip.HideProps()
            Reset()

    #Starts reset timer for falling fruit
    Reset()<suspends> : void=
        if(MovePoint.TeleportTo[StartPos, MakeRotationFromYawPitchRollDegrees(0.0, 0.0, 0.0)]):
        Sleep(FallResetTime)

############################################################################################################################
##################################################### Hi-Chew Minigame #####################################################
############################################################################################################################
minigame76_LilyLeapers := class(minigame_device):
    @editable
    CinematicMapLoad : cinematic_sequence_device = cinematic_sequence_device{}
    @editable
    CinematicMapUnLoad : cinematic_sequence_device = cinematic_sequence_device{}

    @editable
    DefaultClass : class_and_team_selector_device = class_and_team_selector_device{}
    @editable
    LowGravityClass : class_and_team_selector_device = class_and_team_selector_device{}

    @editable 
    WinCollectible : collectible_object_device = collectible_object_device{}
    @editable
    MaxTotalWinners : int = 3
    @editable
    EndBarrier : barrier_device = barrier_device{}

    @editable
    RandItemGranter : item_granter_device = item_granter_device{}
    @editable
    CollectResetTime : float = 5.0
    @editable
    LowGravDuration : float = 7.0
    @editable
    CoinPickupAccolade : accolades_device = accolades_device{}

    @editable
    GloveMovePoint : creative_prop = creative_prop{}
    @editable
    GloveMovingDistance : float = 722.0
    @editable
    GloveMovingTime : float = 2.0

    @editable
    SpinnerRotateTime : float = 2.0
    @editable
    RotatingSpinner1 : creative_prop = creative_prop{}
    @editable
    RotatingSpinner2 : creative_prop = creative_prop{}
    @editable
    RotatingSpinner3 : creative_prop = creative_prop{}
    @editable
    RotatingSpinner4 : creative_prop = creative_prop{}
    @editable
    RotatingSpinner5 : creative_prop = creative_prop{}
    @editable
    RotatingSpinner6 : creative_prop = creative_prop{}
    @editable
    RotatingSpinner7 : creative_prop = creative_prop{}
    @editable
    RotatingSpinner8 : creative_prop = creative_prop{}
    @editable
    RotatingSpinner9 : creative_prop = creative_prop{}
    @editable
    RotatingSpinner10 : creative_prop = creative_prop{}
    @editable
    RotatingSpinner11 : creative_prop = creative_prop{}
    @editable
    RotatingSpinner12 : creative_prop = creative_prop{}
    @editable
    RotatingSpinner13 : creative_prop = creative_prop{}
    @editable
    RotatingSpinner14 : creative_prop = creative_prop{}

    @editable
    FallFruitExplosive1 : explosive_device = explosive_device{}
    @editable
    FallFruitExplosive2 : explosive_device = explosive_device{}
    @editable
    FallFruitExplosive3 : explosive_device = explosive_device{}
    @editable
    FallFruits : []CustomFallFruitData = array{}

    #Checks if the player has been granted a harpoon gun then equips the skin
    @editable
    ConditionalButtonCheckHarpoonGun : conditional_button_device = conditional_button_device{}

    @editable
    HarpoonGuns : []visual_effect_powerup_device = array{}
    @editable
    EmptyHandVFX : visual_effect_powerup_device = visual_effect_powerup_device{}

    var CurrentHarpoonGunVFX : visual_effect_powerup_device = visual_effect_powerup_device{}

    var CurrentWinners : []player = array{}
    var GloveMovePointStartPos : vector3 = vector3{}

    var RandCollectActors : []creative_object_interface = array{}
    var CollectCoinActors : []creative_object_interface = array{}
    var LowGravBounceActors : []creative_object_interface = array{}

    var EndGame : logic = false

    Name<override>: string = "Hi-Chew Hoppers"
    StartingMessage<override>:string = "Get to the end!"
    GetVotingImage<override>()<transacts>: texture = VerseExposed.T_PlaceholderIcon_BouncyCourse

    #REPLACES FindCreativeObjectsWithTag Function
    FindObjectsWithTag<public>(Tag:tag)<transacts>:[]creative_object_interface=
        TaggedObjects:=FindCreativeObjectsWithTag(Tag)
        var Objects:[]creative_object_interface=array{}
        for (Obj:TaggedObjects, TaggedObject:=creative_device_base[Obj]):
            set Objects += array{TaggedObject}
        return Objects

    #Called when the minigame starts
    _Start<override>()<suspends>: void=
        CinematicMapLoad.Play()

        set CurrentWinners = array{}
        set EndGame = false
        set GloveMovePointStartPos = GloveMovePoint.GetTransform().Translation

        for(Fruit : FallFruits):
            set Fruit.StartPos = Fruit.MovePoint.GetTransform().Translation
            Fruit.PropManip.HideProps()

        #Sets up Collectibles
        set RandCollectActors = FindObjectsWithTag(RandItem{})
        set CollectCoinActors = FindObjectsWithTag(Coin{})
        set LowGravBounceActors = FindObjectsWithTag(LowGravBouncer{})

        #RandCollect
        for(TaggedActor : RandCollectActors, Collectible := collectible_object_device[TaggedActor]):
            Collectible.CollectedEvent.Subscribe(RandomItemCollected)

        #Low Gravity Bouncers
        for(TaggedActor : LowGravBounceActors, Bouncer := movement_modulator_device[TaggedActor]):
            Bouncer.ActivationEvent.Subscribe(LowGravCollected)

        #Coin Collectibles
        for(TaggedActor : CollectCoinActors, CollectibleCoin := collectible_object_device[TaggedActor]):
            CollectibleCoin.CollectedEvent.Subscribe(Manager.OnMinigameCollectibleCoinCollected)
            CollectibleCoin.CollectedEvent.Subscribe(AwardAccolade)

        #Set Win Collectible
        WinCollectible.Show()
        WinCollectible.RespawnForAll()
        RegisterSubscriber(WinCollectible.CollectedEvent.Subscribe(HandlePlayerWinEvent))

        for(Player : GetPlayspace().GetPlayers()):
            if(Agent := agent[Player]):
                EndBarrier.AddToIgnoreList(Agent)

        #Assigns random harpoon skin
        RandomHarpoonNum := GetRandomInt(0, HarpoonGuns.Length - 1)
        if(HarpoonGun := HarpoonGuns[RandomHarpoonNum]):
            set CurrentHarpoonGunVFX = HarpoonGun

        #Start Collectible Reset Loop, Spinners and HiChew Stick Movements
        sync:
            block:
                PrepareSpinnerRotation()
            block:
                PrepareHiChewStickMovement()
            block:
                PrepareFallingFruit()
            block:
                CheckHarpoonGunEquipped()
            block:
                loop:
                    Sleep(CollectResetTime)
                    CollectibleReset(RandCollectActors, 0)
                    CollectibleReset(CollectCoinActors, 1)
            block:
                loop:
                    Sleep(1.0)
                    if(EndGame = true):
                        break

    #Assigns harpoon gun skin
    CheckHarpoonGunEquipped()<suspends>:void=
        loop:
            for(Player : GetPlayspace().GetPlayers()):
                if(ConditionalButtonCheckHarpoonGun.IsHoldingItem[Player]):
                    CurrentHarpoonGunVFX.Pickup(Player)
                else:
                    EmptyHandVFX.Pickup(Player)
            Sleep(0.0)

    #Called on true start game
    OnSafeTimeEnded<override>()<suspends>:void=
        Manager.UiMessage.SetText(StringToMessage("Get to the End!"))
        Manager.UiMessage.Show()
        spawn{HandleLastPlayer()}

    #Called when the minigame ends
    _OnMinigameEnded<override>(Winners: []player)<suspends> : void=
        set EndGame = true
        CinematicMapUnLoad.Play()
        WinCollectible.Hide()

############################################################################################################################
################################################## Hi-Chew Win Conditions ##################################################
############################################################################################################################
    #BIRDO ADDED THIS                  
    _OnGenericEndTimerSuccessReturnArray<override>():[]player=
        return CurrentWinners

    HandleLastPlayer()<suspends>:void=
        Print("Handle Last Player")
        race:
            EndedEvent.Await()
            block:
                loop:
                    if(GetRemainingPlayers().Length = 1, Player:= GetRemainingPlayers()[0]):
                        Print("One Player left in Hi-Chew Hoppers")
                        Manager.UiMessage.SetText(StringToMessage("Last chance to reach the crown!"))
                        Manager.UiMessage.Show()
                        break
                    Sleep(0.2)

    HandlePlayerWinEvent(InAgent : agent):void=
        EndBarrier.RemoveFromIgnoreList(InAgent)

        if(WinningPlayer := player[InAgent]):
            if(CurrentWinners.Length = 0):
                set CurrentWinners = array{WinningPlayer}
            else:
                set CurrentWinners += array{WinningPlayer}

            if(CurrentWinners.Length > MaxTotalWinners - 1):
                Manager.GoldTracker.Add(WinningPlayer, 450)
                EndedEvent.Signal(CurrentWinners)

            if(Manager.GenericMinigameEndTimer.GetActiveDuration() > 30.0):
                Manager.GenericMinigameEndTimer.SetMaxDuration(30.0)
                Manager.GenericMinigameEndTimer.Reset()
                Manager.GenericMinigameEndTimer.Start()

############################################################################################################################
##################################################### Hi-Chew Pick-ups #####################################################
############################################################################################################################
    #Called when Hi-Chew Candy is collected
    CollectibleReset(InObjects : []creative_object_interface, CollectType : int)<suspends>: void=
        for(Object : InObjects, Collectible := collectible_object_device[Object]):
            Collectible.RespawnForAll()
            if(CollectType = 0):
                Collectible.CollectedEvent.Subscribe(RandomItemCollected)
            if(CollectType = 1):
                Collectible.CollectedEvent.Subscribe(Manager.OnMinigameCollectibleCoinCollected)
                Collectible.CollectedEvent.Subscribe(AwardAccolade)

    #Resets player variables after delay
    VariableDelaytimer(DelayTime : float, InAgent : agent)<suspends> : void=
        Sleep(DelayTime)
        DefaultClass.ChangeClass(InAgent)

        if(FortChar := InAgent.GetFortCharacter[]):
            FortChar.SetMaxHealth(100.0)

    #Called when Item is Collected
    RandomItemCollected(InAgent : agent): void=
        RandItemGranter.CycleToRandomItem(InAgent)

    #Called when LowGrav Collect/Bounce Event is triggered
    LowGravCollected(InAgent : agent) : void=
        LowGravityClass.ChangeClass(InAgent)
        spawn:
            VariableDelaytimer(LowGravDuration, InAgent)

    #Provide XP for Coin Pickup
    AwardAccolade(InAgent : agent) : void=
        CoinPickupAccolade.Award(InAgent)

############################################################################################################################
#################################################### Hi-Chew Obsticales ####################################################
############################################################################################################################
    #Moves the Punching Gloves back and fourth
    PrepareHiChewStickMovement()<suspends> : void=
        loop:
            MoveHiChewSticks()
            Sleep(GloveMovingTime * 2)

    MoveHiChewSticks() : void=
        if :
            AnimationController := GloveMovePoint.GetAnimationController[]
        then:
            KeyFrame1 := keyframe_delta :
                DeltaLocation := vector3{Y:= GloveMovingDistance }
                DeltaRotation := MakeRotationFromYawPitchRollDegrees(0.0, 0.0, 0.0)
                Time := GloveMovingTime
            KeyFrame2 := keyframe_delta :
                DeltaLocation := vector3{Y:= -GloveMovingDistance }
                DeltaRotation := MakeRotationFromYawPitchRollDegrees(0.0, 0.0, 0.0)
                Time := GloveMovingTime

            KeyFrames : []keyframe_delta = array {KeyFrame1, KeyFrame2}
            AnimationController.SetAnimation(KeyFrames, ?Mode:=animation_mode.OneShot)
            AnimationController.Play()

    #Starts rotating spinner
    PrepareSpinnerRotation()<suspends> : void=
        loop:
            RotateSpinner(RotatingSpinner1)
            RotateSpinner(RotatingSpinner2)
            RotateSpinner(RotatingSpinner3)
            RotateSpinner(RotatingSpinner4)
            RotateSpinner(RotatingSpinner5)
            RotateSpinner(RotatingSpinner6)
            RotateSpinner(RotatingSpinner7)
            RotateSpinner(RotatingSpinner8)
            RotateSpinner(RotatingSpinner9)
            RotateSpinner(RotatingSpinner10)
            RotateSpinner(RotatingSpinner11)
            RotateSpinner(RotatingSpinner12)
            RotateSpinner(RotatingSpinner13)
            RotateSpinner(RotatingSpinner14)
            Sleep(SpinnerRotateTime)

    #Rotates spinner around itself
    RotateSpinner(InSpinningObject : creative_prop) : void=
        if :
            AnimationController := InSpinningObject.GetAnimationController[]
        then:
            KeyFrame1 := keyframe_delta :
                DeltaLocation := vector3{X:=0.0, Y:=0.0, Z:=0.0}
                DeltaRotation := MakeRotationFromYawPitchRollDegrees(-90.0, 0.0, 0.0)
                Time := SpinnerRotateTime

            KeyFrames : []keyframe_delta = array {KeyFrame1}
            AnimationController.SetAnimation(KeyFrames, ?Mode:=animation_mode.OneShot)
            AnimationController.Play()

    #Starts fruit falling
    PrepareFallingFruit()<suspends> : void=
        loop:
            sync:
                block:
                    if(Fruit1 := FallFruits[0]):
                        FallFruitLooper(Fruit1, FallFruitExplosive1)
                block:
                    if(Fruit2 := FallFruits[1]):
                        FallFruitLooper(Fruit2, FallFruitExplosive1)
                block:
                    if(Fruit3 := FallFruits[2]):
                        FallFruitLooper(Fruit3, FallFruitExplosive1)
                block:
                    if(Fruit4 := FallFruits[3]):
                        FallFruitLooper(Fruit4, FallFruitExplosive2)
                block:
                    if(Fruit5 := FallFruits[4]):
                        FallFruitLooper(Fruit5, FallFruitExplosive2)
                block:
                    if(Fruit6 := FallFruits[5]):
                        FallFruitLooper(Fruit6, FallFruitExplosive2)
                block:
                    if(Fruit7 := FallFruits[6]):
                        FallFruitLooper(Fruit7, FallFruitExplosive3)
                block:
                    if(Fruit8 := FallFruits[7]):
                        FallFruitLooper(Fruit8, FallFruitExplosive3)
                block:
                    if(Fruit9 := FallFruits[8]):
                        FallFruitLooper(Fruit9, FallFruitExplosive3)

    FallFruitLooper(Fruit : CustomFallFruitData, InExplosive : explosive_device)<suspends> : void=
        Fruit.FruitFalling()
        Sleep(Fruit.FallTime - 0.5)
        for(Player : GetPlayspace().GetPlayers()):
            Fruit.Explode(Player, InExplosive)