
using { /Fortnite.com/Devices }
using { /Verse.org/Simulation }
using { /UnrealEngine.com/Temporary/Diagnostics }
using { /UnrealEngine.com/Temporary/UI }
using { /Fortnite.com/UI }
using { /UnrealEngine.com/Temporary/SpatialMath }
using { /Verse.org/Colors }
using { /Verse.org/Assets }
using { /Fortnite.com/Characters }
using { /Fortnite.com/Game }
using { /Verse.org/Random }
using { /UnrealEngine.com/Assets }

my_debug_draw := class(debug_draw_channel) {}

Pack_Effects_1 := module:
    # Air_Magic<public> := module:
    #     VFX_Niagara<public> := module {}
    Fire_Magic<public> := module:
        VFX_Niagara<public> := module {}

Electric_VFX := module:
    Niagara<public> := module:
        Electric<public> := module {}

SkillsVFX := module:
    Particles<public> := module {}

vfx_type := enum:
    Flamethrower
    Electricity
    Portal

finisher_vfx := class<concrete>:
    @editable
    Type : vfx_type = vfx_type.Flamethrower

finisher_devices := class <concrete>():
    DebugDraw:debug_draw = debug_draw{Channel := my_debug_draw}

    HIDE_LOCATION<private>: vector3 = vector3 {Z := -700.0}
    EXPLOSIVE_OFFSET : vector3 = vector3{X:=72.678460,Y:=-23.826618,Z:=223.318434} 
    DAMAGE_DISTANCE: float = 1500.0
    DAMAGE:float = 300.0
    var BeingUsed : logic = false
    var _Instigator : ?agent = false
    var _Target : ?agent = false

    @editable
    InstigatorReferenceTrigger: trigger_device = trigger_device {}
    @editable
    TargetReferenceTrigger: trigger_device = trigger_device {}
    @editable
    ReferenceMover: creative_prop = creative_prop {}
    @editable
    FinisherCinema: cinematic_sequence_device = cinematic_sequence_device {}
    @editable
    AudioPlayer : audio_player_device = audio_player_device{}
    #VFX
    @editable
    VFX : ?finisher_vfx = false
    @editable
    Duration : float = 2.0

    # FINISHER_DURATION: []float = array {
    #     2.2, # 2
    #     1.9, # 4
    #     2.4, # 8
    #     1.66, # 9
    #     3.70, # Ninja 3
    #     4.43, # Ninja 4
    #     2.90, # Ninja 5
    #     3.17 # Hero 3
    # }

    Activate(Instigator:agent, Target:agent, FinisherId:int, Offset:vector3, Explosive: explosive_device)<suspends>:void=
        set BeingUsed = true
        set _Instigator = option {Instigator}
        set _Target = option {Target}

        InstigatorReferenceTrigger.Trigger(Instigator)
        TargetReferenceTrigger.Trigger(Target)
        if:
            FC := Instigator.GetFortCharacter[],
            TargetFC := Target.GetFortCharacter[]
        then:
            T := FC.GetTransform()
            var Loc :vector3 = T.Translation - vector3 {Z := 76.0}

            if (FC.IsCrouching[]):
                set Loc = Loc + vector3 {Z := 25.0}

            set Loc = Loc + Offset

            Rot := T.Rotation
            Sleep(0.0)
            HideFortCharacter(FC)
            HideFortCharacter(TargetFC)

            MoverTransform := transform {Translation := Loc, Rotation := Rot}

            if (ReferenceMover.TeleportTo[Loc, Rot]):
                ExpLoc := ReferenceMover.GetTransform().Translation + EXPLOSIVE_OFFSET
                if(Explosive.TeleportTo[ExpLoc, rotation{}]):
                Explosive.Reset()
                Explosive.Explode(Instigator)

                FinisherCinema.Play()
                AudioPlayer.Play(Instigator)
                if (HasVFX := VFX?):
                    case(HasVFX.Type):
                        vfx_type.Flamethrower =>
                            Print("Flamethrower VFX")
                            spawn:
                                FireVfx(T, Offset)   
                        vfx_type.Electricity =>
                            Print("Electricity VFX")
                            spawn:
                                ElectricityVfx(T, Offset) 
                        vfx_type.Portal =>
                            Print("Portal VFX")
                            spawn:
                                PortalVfx(T, Offset) 
                Sleep(Duration)

            Sleep(0.0)
            ReferenceMover.MoveTo(HIDE_LOCATION, Rot, 0.1)

            Sleep(0.0)
            ShowFortCharacter(FC)
            spawn{GiveInvinicibility(FC)}
            DmgArgs : damage_args = damage_args {Instigator := option{FC}, 
                                                            Amount:=9999.0,
                                                            Source:=option{FC}}
            ShowFortCharacter(TargetFC)
            TargetFC.Damage(DmgArgs)
            

        set BeingUsed = false
        set _Instigator = false
        set _Target = false

    FireVfx(InstigatorT : transform, TestOffset:vector3)<suspends>:void=
        Print("Spawning VFX")
        Forward := InstigatorT.Rotation.GetLocalForward()
        NewLoc := InstigatorT.Translation + vector3 {Z := 32.0} + Forward * 70.0
        #DebugDraw.DrawSphere(NewLoc, ?DrawDurationPolicy := debug_draw_duration_policy.FiniteDuration, ?Duration := 10.0, ?Radius := 70.0, ?Color := NamedColors.Red)
        Sleep(2.83)
        # Spawns a VFX particle system at the given Translation
        MyVFX:cancelable = SpawnParticleSystem(Asset := Pack_Effects_1.Fire_Magic.VFX_Niagara.NS_Fire_Magic_Flamethrower, Position := NewLoc + TestOffset, ?Rotation := InstigatorT.Rotation, ?StartDelay := 0.0)
        Sleep(0.8)
        MyVFX.Cancel()

    ElectricityVfx(MoverT : transform, TestOffset:vector3)<suspends>:void=
        Forward := MoverT.Rotation.GetLocalForward()
        NewLoc := MoverT.Translation + vector3 {Z := 60.0} + Forward * 80.0
        # DebugDraw.DrawSphere(NewLoc, ?DrawDurationPolicy := debug_draw_duration_policy.FiniteDuration, ?Duration := 10.0, ?Radius := 70.0, ?Color := NamedColors.Red)
        Sleep(1.3)
        # Spawns a VFX particle system at the given Translation
        MyVFX:cancelable = SpawnParticleSystem(Asset := Electric_VFX.Niagara.Electric.NS_Electric_Orb, Position := NewLoc + TestOffset, ?Rotation := MoverT.Rotation, ?StartDelay := 0.0)
        Sleep(2.2)
        MyVFX.Cancel()

    PortalVfx(MoverT : transform, TestOffset:vector3)<suspends>:void=
        Forward := MoverT.Rotation.GetLocalForward()
        NewLoc := MoverT.Translation + vector3 {Z := -60.0} + Forward * 78.0
        Sleep(0.3)
        # Spawns a VFX particle system at the given Translation
        MyVFX:cancelable = SpawnParticleSystem(Asset := SkillsVFX.Particles.NS_Skill_Portal, Position := NewLoc + TestOffset, ?Rotation := MoverT.Rotation, ?StartDelay := 0.0)
        Sleep(2.6)
        MyVFX.Cancel()

    HideFortCharacter(FC: fort_character):void=
        FC.Hide()
        FC.PutInStasis(stasis_args { AllowTurning:= false, AllowFalling := false, AllowEmotes := false})
        FC.SetVulnerability(false)

    ShowFortCharacter(FC: fort_character):void=
        FC.ReleaseFromStasis()
        FC.Show()
        FC.SetVulnerability(true)

    GiveInvinicibility(FC: fort_character)<suspends>:void=
        FC.SetVulnerability(false)
        Print("Invincible")
        Sleep(1.2)
        Print("Not Invincible")
        FC.SetVulnerability(true)

finishers_minigame := class(minigame_device):
    @editable FinisherDevices :[]finisher_devices = array{}
    var ShuffledFinisherDevices : []finisher_devices = array{}
    @editable FinisherExplosiveDevice : explosive_device = explosive_device {}
    @editable DefaultFinisher : finisher_devices = finisher_devices{}
    @editable FinisherSelfie : finisher_devices = finisher_devices{}
    @editable DBNO :down_but_not_out_device = down_but_not_out_device{}
    @editable RunFinisherInputDebug : input_trigger_device = input_trigger_device{}
    @editable FinisherTipMessage : hud_message_device = hud_message_device{}
    var FinisherIndex : int = 0 

    var PlayerInFinisher: [player]logic = map{}

    Name<override>: string = "Finishers Box PVP"
    StartingMessage<override>:string = "Pickaxe Knocked Players to Activate Finishers"
    GetVotingImage<override>()<transacts>: texture = VerseExposed.SQ_FINISHERS
    #DisplayIcon<override>: ?texture = option{VerseExposed.SQ_FINISHERS_ICON}

    _Start<override>()<suspends>: void =  
        set ShuffledFinisherDevices = Shuffle(FinisherDevices)
        #set ShuffledFinisherDevices = FinisherDevices
        DBNO.Enable()
        RunFinisherInputDebug.PressedEvent.Subscribe(OnTestPressed)
        RegisterSubscriber(DBNO.AgentDownedEvent.Subscribe(OnPlayerDowned))
        RegisterSubscriber(DBNO.AgentRevivedEvent.Subscribe(OnPlayerRevived))
        for(Player : GetRemainingPlayers(), Fort := Player.GetFortCharacter[]):
            if(set PlayerInFinisher[Player] = false) {}
            RegisterSubscriber(Fort.DamagedEvent().Subscribe(FinisherSequence))

    OnSafeTimeEnded<override>():void=
        return

    _Reset<override>()<suspends>:void=
        DBNO.Disable()
        set ShuffledFinisherDevices = array{}
        set FinisherIndex = 0
        set PlayerInFinisher = map{}

    OnPlayerDowned(PlayerDowned : agent):void=
        Print("Player downed")
        if(FortPlayerDowned := PlayerDowned.GetFortCharacter[]):
            spawn{ReviveSequence(PlayerDowned, FortPlayerDowned)}

    OnPlayerRevived(Agent : agent):void=
        Print("On Player Revived")
        if:
            Fort := Agent.GetFortCharacter[]
        then:
            spawn{SpawnProtection(Fort)}

    SpawnProtection(Fort: fort_character)<suspends>:void=
        Fort.SetVulnerability(false)
        Print("Invincible")
        Sleep(1.5)
        Print("Not Invincible")
        Fort.SetVulnerability(true)

    ReviveSequence(Player: agent, Fort : fort_character)<suspends>:void=
        #race to check if in middle of finisher
        race:
            loop:
                if(PlayerInFinisher[Player] = true):
                    Print("Downed Player is in middle of finisher. Breaking out of race and not reviving")
                    break
                Sleep(0.2)
            block:
                Sleep(10.0)
                if(Fort.IsActive[] and Fort.IsDownButNotOut[]):
                    DBNO.Revive(Player)

    FinisherSequence(DamageResult : damage_result) : void =
        PICKAXE_DAMAGE := 20.0
        if:
            Instigator := DamageResult.Instigator?
            InstigatorAgent := Instigator.GetInstigatorAgent[]
            InstigatorPlayer := player[InstigatorAgent]
            InstigatorFort := fort_character[DamageResult.Instigator?]
            CharacterWhoWasHitFort := fort_character[DamageResult.Target]
            CharacterWhoWasHitAgent := CharacterWhoWasHitFort.GetAgent[]
            CharacterWhoWasHitPlayer := player[CharacterWhoWasHitAgent]

            #Distance Check
            InstigatorTranslation := InstigatorFort.GetTransform().Translation
            TargetTranslation := CharacterWhoWasHitFort.GetTransform().Translation
            Dist := Distance(InstigatorTranslation, TargetTranslation)
            Print("Distance = {Dist}")

            # Same Agent: Usually from storm damage
            not InstigatorAgent = CharacterWhoWasHitAgent
            # IsDownButNotOut
            CharacterWhoWasHitFort.IsDownButNotOut[]
        then:
            if (DamageResult.Amount = PICKAXE_DAMAGE):
                Print("Running Finisher")
                #Get Unused Finisher
                if(set PlayerInFinisher[CharacterWhoWasHitPlayer] = true):
                #If there are only 2 players remaining do the final finisher
                if(GetRemainingPlayers().Length <= 2):
                    Print("Only 2 players alive - Final Finisher Attack")
                    spawn:
                        FinisherSelfie.Activate(InstigatorAgent, CharacterWhoWasHitAgent, 0, vector3 {}, FinisherExplosiveDevice)
                    #RunFinisher(InstigatorAgent, CharacterWhoWasHitAgent, 8)
                else:
                    Print("Performing Unused Finisher")
                    spawn:
                        GetUnusedFinisher().Activate(InstigatorAgent, CharacterWhoWasHitAgent, 0, vector3 {}, FinisherExplosiveDevice)
                        #RunFinisher(InstigatorAgent, CharacterWhoWasHitAgent, 0)
            else:
                Print("Not Pickaxing")
                FinisherTipMessage.Show(InstigatorPlayer)

    GetUnusedFinisher():finisher_devices=
        Print("Get Unused Finisher")
        if(FinisherToUse := ShuffledFinisherDevices[FinisherIndex]):
            #Extra Check
            if(FinisherToUse.BeingUsed = true):
                PrintError("Finisher is being used???? SHould Never get here but returning DefaultFinisher")
                #Iterate through all to find unused finisher
                return DefaultFinisher
            else:
                if(FinisherIndex = ShuffledFinisherDevices.Length-1):
                    set FinisherIndex = 0
                else:
                    set FinisherIndex += 1
                return FinisherToUse

        #Should Never get here but is a backup default finisher
        return DefaultFinisher

    RunFinisher(Instigator:agent, Target:agent, FinisherId:int)<suspends>:void=
        # Add logic to iterate
        if (FinisherToUse := ShuffledFinisherDevices[0]):
            spawn:
                #Use(FinisherToUse, Instigator, Target, FinisherId)
                FinisherToUse.Activate(Instigator, Target, FinisherId, vector3 {}, FinisherExplosiveDevice)

    # Use(FinisherToUse:finisher_devices, Instigator:agent, Target:agent, FinisherId:int)<suspends>:void=
    #     NO_OFFSET:vector3 = vector3 {}
    #     FinisherToUse.Activate(Instigator, Target, FinisherId, NO_OFFSET)

    OnTestPressed(Agent:agent):void=
        spawn:
            RunFinisherTest(Agent, Agent, 0)

    RunFinisherTest(Instigator:agent, Target:agent, FinisherId:int)<suspends>:void=
        # Add logic to iterate
        if (FinisherToUse := FinisherDevices[0]):
            spawn:
                #Use(FinisherToUse, Instigator, Target, FinisherId)
                FinisherToUse.Activate(Instigator, Target, FinisherId, vector3 {}, FinisherExplosiveDevice)