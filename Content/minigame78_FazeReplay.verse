using { /Fortnite.com/Devices }
using { /Fortnite.com/AI}
using { /Fortnite.com/Teams}
using { /Fortnite.com/UI }
using { /Verse.org/Simulation }
using { /Verse.org/Random }
using { /Verse.org/Colors }
using { /Verse.org/Assets }
using { /Fortnite.com/Characters }
using { /Fortnite.com/FortPlayerUtilities }
using { /Fortnite.com/Game }
using { /Fortnite.com/Teams }
using { /Fortnite.com/Playspaces }
using { /Verse.org/Simulation/Tags }
using { /UnrealEngine.com/Temporary/Diagnostics }
using { /UnrealEngine.com/Temporary/SpatialMath }
using { /UnrealEngine.com/Temporary/UI }

minigame78_FazeReplay := class(minigame_device):
    Name<override>: string = "Faze Replay"
    StartingMessage<override>:string = "Jump in the air to Shoot!"
    GetVotingImage<override>()<transacts>: texture = VerseExposed.SQ_REPLAYS
    @editable DistanceToGetVfx : float = 75.0

    @editable InputTriggerJump : input_trigger_device = input_trigger_device{}
    @editable Bouncer :bouncer_device = bouncer_device{}
    var BouncerOrigin :vector3 = vector3{}

    @editable BouncerFloorProp : creative_prop = creative_prop{}

    @editable InputTriggeShoot : input_trigger_device = input_trigger_device{}
    @editable FinisherTipMessage : hud_message_device = hud_message_device{}
    @editable Teleporter : teleporter_device = teleporter_device{}
    @editable PleasantMutatorZone : mutator_zone_device = mutator_zone_device{}

    #Todo: Take out when arena voting is in
    @editable DataLayerSeq :cinematic_sequence_device = cinematic_sequence_device{}
    @editable DataLayerSeqUnload :cinematic_sequence_device = cinematic_sequence_device{}

    @editable AudioTrickshot : audio_player_device = audio_player_device{}
    var PlayerLocation :[player]vector3 = map{}
    @editable VfxTrigger : trigger_device = trigger_device{}

    @editable
    LowGravityMutatorZone : mutator_zone_device = mutator_zone_device{}

    @editable
    DefaultPleasantBarrier : barrier_device = barrier_device{}

    @editable
    ReplaysPleasantBarrier : barrier_device = barrier_device{}
    

    #Called when the minigame starts
    _Start<override>()<suspends>: void=
        DefaultPleasantBarrier.Disable()
        ReplaysPleasantBarrier.Enable()
        DataLayerSeq.Play()
        LowGravityMutatorZone.Enable()
        set BouncerOrigin = Bouncer.GetTransform().Translation
        InputTriggerJump.Enable()
        InputTriggeShoot.Enable()
        Teleporter.Enable()
        RegisterSubscriber(PleasantMutatorZone.AgentExitsEvent.Subscribe(OnPlayerEnterLobbyZone))
        RegisterSubscriber(InputTriggerJump.PressedEvent.Subscribe(HandleJump))
        if(BouncerFloorProp.TeleportTo[vector3{X:= 1401.480793, Y:= -37375.456288, Z:= -586.653942}, rotation{}]):
        for(Player : GetRemainingPlayers(), Fort := Player.GetFortCharacter[]):
            InputTriggerJump.Register(Player)
            RegisterSubscriber(InputTriggeShoot.PressedEvent.Subscribe(HandleShootSubscribe))
            spawn{InitShootRegister(Player)}

    OnPlayerEnterLobbyZone(Agent: agent):void=
        Teleporter.Teleport(Agent)

    HandleShootSubscribe(Agent : agent):void=
        if(Fort := Agent.GetFortCharacter[]):
            if(Fort.IsOnGround[]):
                FinisherTipMessage.Show(Agent)
            else:
                InputTriggeShoot.Unregister(Agent)

    HandleJump(Agent : agent):void=
        if:
            Fort := Agent.GetFortCharacter[]
            Fort.IsOnGround[]
        then:
            InputTriggeShoot.Unregister(Agent)
            spawn{BouncerEffect(Fort)}

    BouncerEffect(Fort: fort_character)<suspends>:void=
        if(Bouncer.TeleportTo[Fort.GetTransform().Translation + vector3{Y:=-20.0, Z := -90.0}, Fort.GetTransform().Rotation]):
            if(Bouncer.TeleportTo[BouncerOrigin, Fort.GetTransform().Rotation]): 

    _OnMinigameEnded<override>(Winners: []player)<suspends> : void=
        LowGravityMutatorZone.Disable()
        if(Bouncer.TeleportTo[BouncerOrigin, rotation{}]): 
        InputTriggerJump.UnregisterAll()
        InputTriggerJump.Disable()

        InputTriggeShoot.UnregisterAll()
        InputTriggeShoot.Disable()

        Teleporter.Disable()

        if(BouncerFloorProp.TeleportTo[vector3{X:= 1401.480741,Y:= -37375.456345,Z:= -16748.448820}, rotation{}]):


    _Reset<override>()<suspends>:void=
        DefaultPleasantBarrier.Enable()
        ReplaysPleasantBarrier.Disable()
        #Unload the data layers
        DataLayerSeqUnload.Play()

    InitShootRegister(Agent : agent)<suspends>:void =
        if:
            Fort := Agent.GetFortCharacter[]
            Player := player[Agent]
        then:
            race:
                EndedEvent.Await()
                loop:
                    #Update Location for death
                    if(Fort.IsActive[]):
                        if(set PlayerLocation[Player] = Fort.GetTransform().Translation):
                        if(Fort.IsOnGround[]):
                            InputTriggeShoot.Register(Player)
                        else:
                            InputTriggeShoot.Unregister(Player)
                    else:
                        Print("not active")
                        break
                    Sleep(0.2)
                    

    _OnPlayerEliminated<override>(Result : elimination_result):void=
        Print("On Elimination")
        if:
            PersonKilled := Result.EliminatedCharacter
            PersonKilledAgent := PersonKilled.GetAgent[]
            PersonKilledFort := PersonKilledAgent.GetFortCharacter[]
            PersonKilledPlayer := player[PersonKilledAgent]
            PersonKilledTranslation := PlayerLocation[PersonKilledPlayer]

            Killer := Result.EliminatingCharacter?
            KillerAgent := Killer.GetAgent[]
            KillerFort := KillerAgent.GetFortCharacter[]
            KillerPlayer := player[KillerAgent]
            KillerTranslation := KillerFort.GetTransform().Translation
        then:
            DistanceFromChar := Distance(PersonKilledTranslation, KillerTranslation)
            DivNumber := DistanceFromChar/100.0
            Print("Distance = {DistanceFromChar} or DivNumber = {DivNumber}m")
            if(DivNumber >= DistanceToGetVfx):
                Print("Far enough!")
                if(not PersonKilledAgent = KillerAgent):
                    Print("granting vfx")
                    VfxTrigger.Trigger(KillerPlayer)
                    AudioTrickshot.Play(KillerPlayer)
            
        
