using { /Fortnite.com/Devices }
using { /Fortnite.com/AI}
using { /Fortnite.com/Teams}
using { /Verse.org/Simulation }
using { /Verse.org/Random }
using { /Fortnite.com/Characters }
using { /Fortnite.com/FortPlayerUtilities }
using { /Fortnite.com/Game }
using { /Fortnite.com/Teams }
using { /Verse.org/Simulation/Tags }
using { /UnrealEngine.com/Temporary/Diagnostics }
using { /UnrealEngine.com/Temporary/SpatialMath } 
using { /Verse.org/Assets }

# A Verse-authored creative device that can be placed in a level
close_encounters := class(minigame_device):
    Name<override>: string = "Close Encounters"
    #StartingMessage<override>:string = "Random Chapter 5 Loadout!"
    GetVotingImage<override>()<transacts>: texture = VerseExposed.SQ_CLOSE

    #Todo: Take out when arena voting is in
    @editable DataLayerSeq :cinematic_sequence_device = cinematic_sequence_device{}
    @editable DataLayerSeqUnload :cinematic_sequence_device = cinematic_sequence_device{}

    _Start<override>()<suspends>: void=
        DataLayerSeq.Play()

    _Reset<override>()<suspends>:void=
        #Unload the data layers
        DataLayerSeqUnload.Play()