# using { /Fortnite.com/Devices }
# using { /Fortnite.com/AI}
# using { /Fortnite.com/Teams}
# using { /Verse.org/Simulation }
# using { /Verse.org/Random }
# using { /Fortnite.com/Characters }
# using { /Fortnite.com/FortPlayerUtilities }
# using { /Fortnite.com/Game }
# using { /Fortnite.com/Teams }
# using { /Verse.org/Simulation/Tags }
# using { /UnrealEngine.com/Temporary/Diagnostics }
# using { /UnrealEngine.com/Temporary/SpatialMath } 
# using { /Verse.org/Assets }

# chickensMinigame := class(tag) {}
# chicken := class(minigame_device):

#     # @editable ChickenClassSelector :class_and_team_selector_device = class_and_team_selector_device{}
#     # @editable UiChickenTip : hud_message_device = hud_message_device{}
#     # @editable MutatorZoneChickenNoJumping: mutator_zone_device = mutator_zone_device{}

#     # StartingMessage<override>:string = "Grab a chicken to damage enemies!"
#     # Name<override>: string = "Chicken Coop"
#     # GetVotingImage<override>()<transacts>: texture = VerseExposed.SQ_CHICKEN

#     # _Start<override>()<suspends>: void = 
#     #     Print("Starting Chicken Minigame")
        
#     #     UiChickenTip.Show()
#     #     AllPlayers := GetRemainingPlayers()
#     #     Print("AllPlayers Size = {AllPlayers.Length}")
        
#     #     MutatorZoneChickenNoJumping.Enable()

#     #     for(Player : AllPlayers):
#     #         ChickenClassSelector.ChangeClass(Player)
#     #         if:
#     #             Fort := Player.GetFortCharacter[]
#     #         then:
#     #             Fort.SetMaxShield(0.0)
#     #             Fort.SetMaxHealth(5.0)

#     #     Chickens := GetCreativeObjectsWithTag(chickensMinigame{})
#     #     for (Chicken : Chickens):
#     #         if(WildLifeSpawner := wildlife_spawner_device[Chicken]):
#     #             Print("Spawning Wildlife bro")
#     #             WildLifeSpawner.Enable()
#     #             WildLifeSpawner.Spawn()
