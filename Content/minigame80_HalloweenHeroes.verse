
using { /Fortnite.com/Devices }
using { /Verse.org/Simulation }
using { /Verse.org/Simulation/Tags }
using { /UnrealEngine.com/Temporary/Diagnostics }
using { /UnrealEngine.com/Temporary/UI }
using { /Fortnite.com/UI }
using { /UnrealEngine.com/Temporary/SpatialMath }
using { /Verse.org/Colors }
using { /Verse.org/Assets }
using { /Fortnite.com/Characters }
using { /Fortnite.com/Game }
using { /Verse.org/Random }
using { /Fortnite.com/Devices/CreativeAnimation }

halloween_hero_device_tag := class(tag) {}
halloween_hero_device := class(crashable_device):
    @editable InitialUIMessage : string = ""
    @editable DefaultItemGranters : []minigame_device_granter_config = array{}
    @editable OptClassSelector : ?class_and_team_selector_device = false
    @editable OptHeroImpressionAnalytics : ?analytics_device = false
    var HalloweenHeroesManager : ?minigame80_HalloweenHeroes = false
    var GameEnded : logic = true
    var Frozen : logic = false

    var MiniGameHalloweenDevice : minigame80_HalloweenHeroes = minigame80_HalloweenHeroes{}
    # Hud Textures
    var DisplayHudTexture : texture = VerseExposed.TransparentTexture

    var<private> HeroSubscribables : []cancelable = array{}

    CreateHeroObj(Player : player):custom_halloween_player = 
        return custom_halloween_player{Player := Player, HalloweenHeroDevice := Self}

    RegisterHeroSubscriber<final>(Cancelable: cancelable):void=
        set HeroSubscribables += array{Cancelable}

    #Hud Stuff
    HeroDeviceInit():void= return

    GetPlayerHeroDevice(Player : player):void=
        if(HeroObj := HalloweenHeroesManager?.PlayerHeroMap[Player].HalloweenHeroDevice):
            HeroObj

    Freeze():void=
        set Frozen = true
        
    UnFreeze():void=
        set Frozen = false


    #Hud Stuff
    InitCustomHero<final>(Player : player):void=
        Print("Initializing {DeviceName} Hero Device")
        set GameEnded = false
        #Hud Stuff
        for(DefaultGranter : DefaultItemGranters, not DefaultGranter.GrantAfterSafeTime?):
            DefaultGranter.Grant(array{Player})

        if(ClassSelector := OptClassSelector?):
            ClassSelector.ChangeClass(Player)

        MiniGameHalloweenDevice.Manager.DisplayIconManager.MakeCustomHUD(Player, DisplayHudTexture, 512.0)

        if(AnalyticDevice := OptHeroImpressionAnalytics?):
            Print("HeroDevice - {DeviceName} Analytic Impression")
            AnalyticDevice.Submit(Player)
        spawn:
            _InitCustomHero(Player)

    _InitCustomHero(Player : player)<suspends>:void=return

    ResetCustomHero(Player : player):void=return

    HalloweenHeroDeviceReset<final>()<suspends>:void =
        set GameEnded = true
        for(Subscriber : HeroSubscribables):
            Subscriber.Cancel()
        set HeroSubscribables = array{}
        spawn{_HalloweenHeroDeviceReset()}

    _HalloweenHeroDeviceReset()<suspends>:void=return

halloweenHero1_freeze_device := class(halloween_hero_device):
    @editable TimeToGrantAmmo : float = 15.0
    @editable Ammo1Shot : item_granter_device = item_granter_device{}
    _InitCustomHero<override>(Player : player):void=
        spawn{LoseHealthFunction(Player)}

    LoseHealthFunction(Player : player)<suspends>:void =
        race:
            MiniGameHalloweenDevice.EndedEvent.Await()
            loop:
                Sleep(TimeToGrantAmmo)
                Ammo1Shot.GrantItemIndex(Player, 1)


HalloweenHero2_slenderman := class(halloween_hero_device):
        
BroomClass := class<concrete>():
    @editable
    Prop : creative_prop = creative_prop{}
    @editable
    VFXDevice : visual_effect_powerup_device = visual_effect_powerup_device{}

    var OriginalPosition : vector3 = vector3{}

HalloweenHero3_witch := class(halloween_hero_device):

    @editable
    var Brooms : []BroomClass = array{}

    var AllBrooms : []BroomClass = array{}

    @editable
    NoClass : class_and_team_selector_device = class_and_team_selector_device{}

    @editable
    WitchClass : class_and_team_selector_device = class_and_team_selector_device{}

    @editable
    VFXClearDevice : visual_effect_powerup_device = visual_effect_powerup_device{}

    var PlayerBroomMap : [player]?BroomClass = map{}

    # Runs when the device is started in a running game
    OnCrashableDeviceBegin<override>()<suspends>:void=
        set AllBrooms = Brooms
        for(Broom : Brooms):
            set Broom.OriginalPosition = Broom.Prop.GetTransform().Translation

    _InitCustomHero<override>(Player : player)<suspends>:void=
        SpawnBroom(Player)

    SpawnBroom(InAgent : agent):void=
        Print("Spawn Broom")
        if(Player := player[InAgent], FortCharacter := Player.GetFortCharacter[]):
            if(Broom := Brooms[0], set PlayerBroomMap[Player] = option{Broom}):
                if(NewBrooms := Brooms.RemoveElement[0]):
                    set Brooms = NewBrooms

                TargetPos := vector3{X:=FortCharacter.GetTransform().Translation.X, Y:=FortCharacter.GetTransform().Translation.Y, Z:=Broom.Prop.GetTransform().Translation.Z}
                if(Broom.Prop.TeleportTo[TargetPos, rotation{}]):
                    Broom.VFXDevice.Pickup(InAgent)
                    WitchClass.ChangeClass(InAgent)
                    spawn:
                        MoveBroomLogic(InAgent, Broom)

    DespawnBroom(InAgent : agent):void=
        Print("Spawn Broom")
        if(Player := player[InAgent], FortCharacter := Player.GetFortCharacter[]):
            if(Broom := PlayerBroomMap[Player]?):
                if(Broom.Prop.TeleportTo[Broom.OriginalPosition, rotation{}]):
                
                set Brooms += array{Broom}
                VFXClearDevice.Pickup(InAgent)
                NoClass.ChangeClass(InAgent)
                if(set PlayerBroomMap[Player] = false):

    _HalloweenHeroDeviceReset<override>()<suspends>:void=
        set Brooms = AllBrooms
        for(Player -> Tricycle : PlayerBroomMap):
            DespawnBroom(Player)

    MoveBroomLogic(InAgent : agent, InBroom : BroomClass)<suspends>:void=
        Print("Moving Logic")
        if(Player := player[InAgent], FortCharacter := Player.GetFortCharacter[]):
            race:
                loop:
                    Sleep(0.0)
                    if (GameEnded = true):
                        DespawnBroom(InAgent)
                        Print("Despawn bbbbbbbbbbbbbbbb")
                        break
                        
                block:
                    #FortCharacter.Hide()
                    loop:
                        PlayerLocationRef : vector3 = FortCharacter.GetTransform().Translation              # Get the players location
                        var FollowerPropLocation : vector3 = InBroom.Prop.GetTransform().Translation   # Get the prop
                        #Only move once the player is at a certain distance away
                        #Slots from where the props can follow you
            
                        if:
                            AnimationController := InBroom.Prop.GetAnimationController[]    # Get reference to animation controller
                        then:
            
                            NoJumpPlayerPos := vector3{X:= PlayerLocationRef.X, Y:= PlayerLocationRef.Y, Z := PlayerLocationRef.Z}
                            TargetPosition := NoJumpPlayerPos - FollowerPropLocation                        # Get the target position for the pet to move to
            
                            Rotation := FortCharacter.GetViewRotation()
                            YawPitchRoll := Rotation.GetYawPitchRollDegrees()
                            if(NewRotation := MakeRotationFromYawPitchRollDegrees(YawPitchRoll[0], 0.0, 0.0)):
                                TargetRotation := MakeShortestRotationBetween(InBroom.Prop.GetTransform().Rotation, NewRotation)        # Get the rotation for the pet to face the player
            
                                KeyFrame := keyframe_delta :                                                                            # Create a keyframe for the pet to move to the player
                                    DeltaLocation := TargetPosition
                                    DeltaRotation := TargetRotation
                                    Time := 0.1
                                KeyFrames : []keyframe_delta = array { KeyFrame }                                                       # Create an array of keyframes
                                AnimationController.SetAnimation(KeyFrames, ?Mode:=animation_mode.OneShot)                              # Set the animation controller to move the pet to the player
                                AnimationController.Play()                                                                              # Play the animation controller
                        Sleep(0.1)

HalloweenHero4_saw := class(halloween_hero_device):
HalloweenHero5_reaper := class(halloween_hero_device):
    @editable IndividualRedWaterVolume : water_device = water_device{}
    var IndividualRedWaterVolumeOrigin :vector3 = vector3{}

    _InitCustomHero<override>(Player : player):void=
        set IndividualRedWaterVolumeOrigin = IndividualRedWaterVolume.GetTransform().Translation
        if(Fort := Player.GetFortCharacter[]):
            spawn{DashLoop(Player, Fort)}

    DashLoop(Player : player, Fort : fort_character)<suspends>:void =
        race:
            MiniGameHalloweenDevice.EndedEvent.Await()
            loop:
                if(IndividualRedWaterVolume.TeleportTo[Fort.GetTransform().Translation + vector3{Y:=-20.0, Z := -40.0}, Fort.GetTransform().Rotation]):
                    Sleep(0.1)
                    if(IndividualRedWaterVolume.TeleportTo[IndividualRedWaterVolumeOrigin, Fort.GetTransform().Rotation]):
                Sleep(20.0)


HalloweenHero6_thing := class(halloween_hero_device):
    var BouncerOrigin :vector3 = vector3{}
    @editable Bouncer :bouncer_device = bouncer_device{}

    _InitCustomHero<override>(Player : player):void=
        set BouncerOrigin = Bouncer.GetTransform().Translation
        if(Fort := Player.GetFortCharacter[]):
            RegisterHeroSubscriber(Fort.JumpedEvent().Subscribe(HandleJump))

    HandleJump(Fort : fort_character):void=
        Print("Jumping Logic")
        if(Bouncer.TeleportTo[Fort.GetTransform().Translation + vector3{Y:=-20.0, Z := -90.0}, Fort.GetTransform().Rotation]):
            if(Bouncer.TeleportTo[BouncerOrigin, Fort.GetTransform().Rotation]): 

    _HalloweenHeroDeviceReset<override>():void=
        if(Bouncer.TeleportTo[BouncerOrigin, rotation{}]): 

HalloweenHero7_jason := class(halloween_hero_device):
HalloweenHero8_vampire := class(halloween_hero_device):
    #var DisplayHudTexture <override>: texture = VerseExposed.SQ_DOUBLEPUMP_ICON
    _InitCustomHero<override>(Player : player):void=
        if(Fort := Player.GetFortCharacter[]):
            Fort.SetMaxHealth(200.0)
            Fort.SetHealth(200.0)
            Fort.SetMaxShield(0.0)
            spawn{LoseHealthFunction(Player, Fort)}

    LoseHealthFunction(Player : player, Fort : fort_character)<suspends>:void =
        race:
            MiniGameHalloweenDevice.EndedEvent.Await()
            loop:
                Fort.SetHealth(Fort.GetHealth() - 1.0)
                Sleep(1.0)


TricycleClass := class<concrete>():
    @editable
    ChairDevice : chair_device = chair_device{}
    @editable
    Explosive : explosive_device = explosive_device{}
    @editable
    Volume : volume_device = volume_device{}

    var AgentRef : []agent = array{}

    var OriginalPosition : vector3 = vector3{}

    OnVolumeEntered(InAgent : agent):void=
        Print("Volume Entered")
        if(AgentRef = array{InAgent}):
            Print("Owned Player Enter")
            return
            
        if(Player := player[InAgent], FortCharacter := Player.GetFortCharacter[]):
            FortCharacter.Damage(1000.0)

HalloweenHero9_tricycle := class(halloween_hero_device):
    var PlayerTricycleMap : [player]?TricycleClass = map{}
    @editable
    var Tricycles : []TricycleClass = array{}
    @editable
    InputTrigger_Forward : input_trigger_device = input_trigger_device{}
    @editable
    MaxVectors : vector2 = vector2{}
    @editable
    MinVectors : vector2 = vector2{}
    @editable
    SignalRemote : signal_remote_manager_device = signal_remote_manager_device{}

    var PlayerInputMap : [player]float = map{}

    # Runs when the device is started in a running game
    OnCrashableDeviceBegin<override>()<suspends>:void=
        for(Tricycle : Tricycles):
            Tricycle.ChairDevice.SeatedEvent.Subscribe(OnSeated)
            Tricycle.ChairDevice.ExitedEvent.Subscribe(OnExit)
            set Tricycle.OriginalPosition = Tricycle.ChairDevice.GetTransform().Translation
            Tricycle.Volume.AgentEntersEvent.Subscribe(Tricycle.OnVolumeEntered)
            InputTrigger_Forward.PressedEvent.Subscribe(OnForward)
            SignalRemote.PrimarySignalEvent.Subscribe(SpawnTricycle)

    HeroDeviceInit<override>():void=
        InputTrigger_Forward.Enable()


    _InitCustomHero<override>(Player : player)<suspends>:void=
        block:


    SpawnTricycle(InAgent : agent):void=
        Print("Spawn Tricycle")
        if(Player := player[InAgent], FortCharacter := Player.GetFortCharacter[]):
            if(Tricycle := Tricycles[0], set PlayerTricycleMap[Player] = option{Tricycle}):
                set Tricycle.AgentRef = array{InAgent}
                if(NewTricycles := Tricycles.RemoveElement[0]):
                    set Tricycles = NewTricycles

                TargetPos := vector3{X:=FortCharacter.GetTransform().Translation.X, Y:=FortCharacter.GetTransform().Translation.Y, Z:=Tricycle.ChairDevice.GetTransform().Translation.Z}
                if(Tricycle.ChairDevice.TeleportTo[TargetPos, rotation{}]):
                    Tricycle.ChairDevice.Seat(InAgent)

    DespawnTricycle(InAgent : agent):void=
        Print("Spawn Tricycle")
        if(Player := player[InAgent], FortCharacter := Player.GetFortCharacter[]):
            if(Tricycle := PlayerTricycleMap[Player]?):
                Tricycle.ChairDevice.Eject(InAgent)
                if(Tricycle.ChairDevice.TeleportTo[Tricycle.OriginalPosition, rotation{}]):

                set Tricycles += array{Tricycle}
                if(set PlayerTricycleMap[Player] = false):

    OnSeated(InAgent : agent):void=
        #Added to player Input Map
        if(Player := player[InAgent]):
            if(set PlayerInputMap[Player] = 0.0):
        #Register to inputs
        InputTrigger_Forward.Register(InAgent)
        if(Player := player[InAgent]):
            if(Tricycle := PlayerTricycleMap[Player]?):
                spawn:
                    MoveTricycleLogic(InAgent, Tricycle)
        Print("Seated")

    OnForward(InAgent : agent):void=
        if(Player := player[InAgent]):
            if(set PlayerInputMap[Player] = 1.0):
        Print("Moving Forwards")

    OnExit(InAgent : agent):void=
        #Register to inputs
        InputTrigger_Forward.Unregister(InAgent)
        Print("Exit")

    _HalloweenHeroDeviceReset<override>()<suspends>:void=
        InputTrigger_Forward.UnregisterAll()
        InputTrigger_Forward.Disable()
        for(Player -> Tricycle : PlayerTricycleMap):
            DespawnTricycle(Player)

    MoveTricycleLogic(InAgent : agent, InTricycle : TricycleClass)<suspends>:void=
        Print("Moving Logic")
        if(Player := player[InAgent], FortCharacter := Player.GetFortCharacter[]):
            Z := FortCharacter.GetTransform().Translation.Z
            race:
                MiniGameHalloweenDevice.EndedEvent.Await()
                block:
                    loop:
                        Sleep(0.0)
                        if (GameEnded = true):
                            break
                block:
                    Sleep(5.0)

                block:
                    loop:
                        if(not InTricycle.ChairDevice.IsSeated[Player]):
                            Print("Player Not Seated")
                            break

                        if(PlayerInput := PlayerInputMap[Player]):
                            TargetPos := InTricycle.ChairDevice.GetTransform().Translation + FortCharacter.GetViewRotation().GetLocalForward() * (PlayerInput * 100)
                            
                            
                            var TargetRotation : rotation = rotation{}
                            if(YawPitchRoll := FortCharacter.GetViewRotation().GetYawPitchRollDegrees(), Yaw := YawPitchRoll[0]):
                                set TargetRotation = MakeRotationFromYawPitchRollDegrees(Yaw, 0.0, 0.0)

                            #if(TargetPos.X < MaxVectors.X, TargetPos.Y < MaxVectors.Y, TargetPos.X > MinVectors.X, TargetPos.Y > MinVectors.Y):
                            if(Frozen = false):
                                InTricycle.ChairDevice.MoveTo(vector3{X:=TargetPos.X, Y:=TargetPos.Y, Z:=Z}, TargetRotation, 0.25)
                                InTricycle.Explosive.Explode(InAgent)
                                InTricycle.Explosive.Reset()
                            else:
                                Sleep(0.25)
            DespawnTricycle(InAgent)


custom_halloween_player := class<unique>():
    Player : player
    HalloweenHeroDevice : halloween_hero_device
    var UltimateAbilityImage : ?texture_block = false
    var Canvas : canvas = canvas{}
    var VampireLogic : logic = false

    InitPlayer():void=
        if:
            Agent := agent[Player]
            Fort := Agent.GetFortCharacter[]
        then:
            # Spawn prop at player location
            Print("Initializing Custom Hero for Player")
            HalloweenHeroDevice.InitCustomHero(Player)
            if(PlayerUI := GetPlayerUI[Player]):
                PlayerUI.AddWidget(Canvas, player_ui_slot{InputMode := ui_input_mode.None})
            HalloweenHeroDevice.MiniGameHalloweenDevice.Manager.UiMessageIndividual.SetText(StringToMessage(HalloweenHeroDevice.InitialUIMessage))
            HalloweenHeroDevice.MiniGameHalloweenDevice.Manager.UiMessageIndividual.Show(Player)

    RemoveWidget():void=
        if(PlayerUI := GetPlayerUI[Player]):
            PlayerUI.RemoveWidget(Canvas)

UI_Hero_Select_Halloween := class():
    Texture : texture = VerseExposed.T_HalloweenUI
    var Button1 : button_loud = button_loud{}
    var Button2 : button_loud = button_loud{}
    var Button3 : button_loud = button_loud{}
    var Button4 : button_loud = button_loud{}
    var Button5 : button_loud = button_loud{}
    var Button6 : button_loud = button_loud{}
    var Button7 : button_loud = button_loud{}
    var Button8 : button_loud = button_loud{}

    var Canva : canvas = canvas{}
    
    RemoveUI(Player : player):void=
        if(PlayerUI := GetPlayerUI[Player]):
            PlayerUI.RemoveWidget(Canva)

    CreateUI(Player : player, MiniGame : minigame80_HalloweenHeroes):canvas=
        Button1.SetText(_StringToMessage("Freeze"))
        Button2.SetText(_StringToMessage("Reaper"))
        Button3.SetText(_StringToMessage("Slenderman"))
        #Button4.SetText(_StringToMessage("Coming Soon"))
        Button5.SetText(_StringToMessage("Vampire"))
        Button6.SetText(_StringToMessage("Jason"))
        Button7.SetText(_StringToMessage("Thing"))
        Button8.SetText(_StringToMessage("Billy"))


        Image : texture_block = texture_block{DefaultImage := Texture, DefaultDesiredSize := vector2{X:= 1920.0, Y := 1080.0}}
        NewUI : canvas = canvas:
            Slots := array:
                canvas_slot:
                    Anchors := anchors{Minimum := vector2{X:= 0.5, Y := 0.5}, Maximum := vector2{X:= 0.5, Y:= 0.5}}
                    Offsets := margin{Left := 0.0, Top := 0.0, Right := 1920.0, Bottom := 1080.0}
                    Alignment := vector2{X:= 0.5, Y := 0.5}
                    SizeToContent := true
                    Widget := Image

                canvas_slot:
                    Anchors := anchors{Minimum := vector2{X:= 0.5, Y := 0.5}, Maximum := vector2{X:= 0.5, Y:= 0.5}}
                    Offsets := margin{Left := -660.0, Top := -10.0, Right := 360.0, Bottom := 82.0}
                    Alignment := vector2{X:= 0.5, Y := 0.5}
                    SizeToContent := true
                    Widget := Button1

                canvas_slot:
                    Anchors := anchors{Minimum := vector2{X:= 0.5, Y := 0.5}, Maximum := vector2{X:= 0.5, Y:= 0.5}}
                    Offsets := margin{Left := -667.86853, Top := -478.0, Right := 360.0, Bottom := 82.0}
                    Alignment := vector2{X:= 0.5, Y := 0.5}
                    SizeToContent := true
                    Widget := text_block{DefaultText := _StringToMessage("")}

                canvas_slot:
                    Anchors := anchors{Minimum := vector2{X:= 0.5, Y := 0.5}, Maximum := vector2{X:= 0.5, Y:= 0.5}}
                    Offsets := margin{Left := -210.0, Top := -10.0, Right := 360.0, Bottom := 83.0}
                    Alignment := vector2{X:= 0.5, Y := 0.5}
                    SizeToContent := true
                    Widget := Button2

                canvas_slot:
                    Anchors := anchors{Minimum := vector2{X:= 0.5, Y := 0.5}, Maximum := vector2{X:= 0.5, Y:= 0.5}}
                    Offsets := margin{Left := -232.0, Top := -478.0, Right := 360.0, Bottom := 83.0}
                    Alignment := vector2{X:= 0.5, Y := 0.5}
                    SizeToContent := true
                    Widget := text_block{DefaultText := _StringToMessage("")}

                canvas_slot:
                    Anchors := anchors{Minimum := vector2{X:= 0.5, Y := 0.5}, Maximum := vector2{X:= 0.5, Y:= 0.5}}
                    Offsets := margin{Left := 220.0, Top := -10.0, Right := 360.0, Bottom := 83.0}
                    Alignment := vector2{X:= 0.5, Y := 0.5}
                    SizeToContent := true
                    Widget := Button3

                canvas_slot:
                    Anchors := anchors{Minimum := vector2{X:= 0.5, Y := 0.5}, Maximum := vector2{X:= 0.5, Y:= 0.5}}
                    Offsets := margin{Left := 209.0, Top := -478.0, Right := 360.0, Bottom := 83.0}
                    Alignment := vector2{X:= 0.5, Y := 0.5}
                    SizeToContent := true
                    Widget := text_block{DefaultText := _StringToMessage("")}
                
                # canvas_slot:
                #     Anchors := anchors{Minimum := vector2{X:= 0.5, Y := 0.5}, Maximum := vector2{X:= 0.5, Y:= 0.5}}
                #     Offsets := margin{Left := 660.0, Top := -10.0, Right := 360.0, Bottom := 83.0}
                #     Alignment := vector2{X:= 0.5, Y := 0.5}
                #     SizeToContent := true
                #     Widget := Button4

                canvas_slot:
                    Anchors := anchors{Minimum := vector2{X:= 0.5, Y := 0.5}, Maximum := vector2{X:= 0.5, Y:= 0.5}}
                    Offsets := margin{Left := 644.0, Top := -478.0, Right := 360.0, Bottom := 83.0}
                    Alignment := vector2{X:= 0.5, Y := 0.5}
                    SizeToContent := true
                    Widget := text_block{DefaultText := _StringToMessage("")}

                canvas_slot:
                    Anchors := anchors{Minimum := vector2{X:= 0.5, Y := 0.5}, Maximum := vector2{X:= 0.5, Y:= 0.5}}
                    Offsets := margin{Left := -640.86853, Top := 450.0, Right := 360.0, Bottom := 83.0}
                    Alignment := vector2{X:= 0.5, Y := 0.5}
                    SizeToContent := true
                    Widget := Button5

                canvas_slot:
                    Anchors := anchors{Minimum := vector2{X:= 0.5, Y := 0.5}, Maximum := vector2{X:= 0.5, Y:= 0.5}}
                    Offsets := margin{Left := -667.86853, Top := 59.0, Right := 360.0, Bottom := 83.0}
                    Alignment := vector2{X:= 0.5, Y := 0.5}
                    SizeToContent := true
                    Widget := text_block{DefaultText := _StringToMessage("")}

                canvas_slot:
                    Anchors := anchors{Minimum := vector2{X:= 0.5, Y := 0.5}, Maximum := vector2{X:= 0.5, Y:= 0.5}}
                    Offsets := margin{Left := -240.0, Top := 450.0, Right := 360.0, Bottom := 83.0}
                    Alignment := vector2{X:= 0.5, Y := 0.5}
                    SizeToContent := true
                    Widget := Button6

                canvas_slot:
                    Anchors := anchors{Minimum := vector2{X:= 0.5, Y := 0.5}, Maximum := vector2{X:= 0.5, Y:= 0.5}}
                    Offsets := margin{Left := -232.0, Top := 59.0, Right := 360.0, Bottom := 83.0}
                    Alignment := vector2{X:= 0.5, Y := 0.5}
                    SizeToContent := true
                    Widget := text_block{DefaultText := _StringToMessage("")}

                canvas_slot:
                    Anchors := anchors{Minimum := vector2{X:= 0.5, Y := 0.5}, Maximum := vector2{X:= 0.5, Y:= 0.5}}
                    Offsets := margin{Left := 220.0, Top := 450.0, Right := 360.0, Bottom := 83.0}
                    Alignment := vector2{X:= 0.5, Y := 0.5}
                    SizeToContent := true
                    Widget := Button7

                canvas_slot:
                    Anchors := anchors{Minimum := vector2{X:= 0.5, Y := 0.5}, Maximum := vector2{X:= 0.5, Y:= 0.5}}
                    Offsets := margin{Left := 209.0, Top := 59.0, Right := 360.0, Bottom := 83.0}
                    Alignment := vector2{X:= 0.5, Y := 0.5}
                    SizeToContent := true
                    Widget := text_block{DefaultText := _StringToMessage("")}
                
                canvas_slot:
                    Anchors := anchors{Minimum := vector2{X:= 0.5, Y := 0.5}, Maximum := vector2{X:= 0.5, Y:= 0.5}}
                    Offsets := margin{Left := 644.0, Top := 450.0, Right := 360.0, Bottom := 83.0}
                    Alignment := vector2{X:= 0.5, Y := 0.5}
                    SizeToContent := true
                    Widget := Button8

                canvas_slot:
                    Anchors := anchors{Minimum := vector2{X:= 0.5, Y := 0.5}, Maximum := vector2{X:= 0.5, Y:= 0.5}}
                    Offsets := margin{Left := 644.0, Top := 59.0, Right := 360.0, Bottom := 83.0}
                    Alignment := vector2{X:= 0.5, Y := 0.5}
                    SizeToContent := true
                    Widget := text_block{DefaultText := _StringToMessage("")}
        #Bindings

        MiniGame.RegisterSubscriber(Button1.OnClick().Subscribe(MiniGame.OnHeroSelect_Any))
        MiniGame.RegisterSubscriber(Button1.OnClick().Subscribe(MiniGame.OnHeroSelect_1))

        MiniGame.RegisterSubscriber(Button2.OnClick().Subscribe(MiniGame.OnHeroSelect_Any))
        MiniGame.RegisterSubscriber(Button2.OnClick().Subscribe(MiniGame.OnHeroSelect_2))

        MiniGame.RegisterSubscriber(Button3.OnClick().Subscribe(MiniGame.OnHeroSelect_Any))
        MiniGame.RegisterSubscriber(Button3.OnClick().Subscribe(MiniGame.OnHeroSelect_3))

        # MiniGame.RegisterSubscriber(Button4.OnClick().Subscribe(MiniGame.OnHeroSelect_Any))
        # MiniGame.RegisterSubscriber(Button4.OnClick().Subscribe(MiniGame.OnHeroSelect_4))

        MiniGame.RegisterSubscriber(Button5.OnClick().Subscribe(MiniGame.OnHeroSelect_Any))
        MiniGame.RegisterSubscriber(Button5.OnClick().Subscribe(MiniGame.OnHeroSelect_5))

        MiniGame.RegisterSubscriber(Button6.OnClick().Subscribe(MiniGame.OnHeroSelect_Any))
        MiniGame.RegisterSubscriber(Button6.OnClick().Subscribe(MiniGame.OnHeroSelect_6))

        MiniGame.RegisterSubscriber(Button7.OnClick().Subscribe(MiniGame.OnHeroSelect_Any))
        MiniGame.RegisterSubscriber(Button7.OnClick().Subscribe(MiniGame.OnHeroSelect_7))

        MiniGame.RegisterSubscriber(Button8.OnClick().Subscribe(MiniGame.OnHeroSelect_Any))
        MiniGame.RegisterSubscriber(Button8.OnClick().Subscribe(MiniGame.OnHeroSelect_8))

        set Canva = NewUI

        return NewUI

# A Verse-authored creative device that can be placed in a level
halloween_hero := class(tag) {}
minigame80_HalloweenHeroes := class(minigame_device):
    @editable CeilingProp : creative_prop = creative_prop{}
    @editable FreezeVFX : visual_effect_powerup_device = visual_effect_powerup_device{}
    @editable VFXClearDevice : visual_effect_powerup_device = visual_effect_powerup_device{}
    @editable MammothConditionalButton :conditional_button_device = conditional_button_device{}
    @editable SignalGrayRemoteSlenderman : signal_remote_manager_device = signal_remote_manager_device{}
    @editable RipsawConditionalButton :conditional_button_device = conditional_button_device{}
    @editable AudioFreeze : audio_player_device = audio_player_device{}
    @editable AudioEvilLaugh : audio_player_device = audio_player_device{}
    @editable AudioJumpScare: audio_player_device = audio_player_device{}
    var BeginningPlayerUIMap : [player]UI_Hero_Select_Halloween = map{}
    var PlayerHeroMap : [player]custom_halloween_player = map{}

    @editable HeroDevice1 : ?halloween_hero_device = false
    @editable HeroDevice2 : ?halloween_hero_device = false
    @editable HeroDevice3 : ?halloween_hero_device = false
    @editable HeroDevice4 : ?halloween_hero_device = false
    @editable HeroDevice5 : ?halloween_hero_device = false
    @editable HeroDevice6 : ?halloween_hero_device = false
    @editable HeroDevice7 : ?halloween_hero_device = false
    @editable HeroDevice8 : ?halloween_hero_device = false

    var AllHeroDevices : []halloween_hero_device = array{}

    GetVotingImage<override>()<transacts>: texture = VerseExposed.SQ_MonsterMash
    DisplayIcon<override>: ?texture = option{VerseExposed.TransparentTexture}

    #REPLACES FindCreativeObjectsWithTag Function
    FindObjectsWithTag<public>(Tag:tag)<transacts>:[]creative_object_interface=
        TaggedObjects:=FindCreativeObjectsWithTag(Tag)
        var Objects:[]creative_object_interface=array{}
        for (Obj:TaggedObjects, TaggedObject:=creative_device_base[Obj]):
            set Objects += array{TaggedObject}
        return Objects

    OnMinigameDeviceBegin<override>()<suspends>:void=
        Print("SETTING HERO CLASS STUFF")
        HeroDevices := FindObjectsWithTag(halloween_hero{})
        for (Obj: HeroDevices):
            if(HeroDevice := halloween_hero_device[Obj]):
                set HeroDevice.MiniGameHalloweenDevice = Self
                set AllHeroDevices += array{HeroDevice}
        Print("AllHeroDevices Length = {AllHeroDevices.Length}")

    _Start<override>()<suspends>: void =
        if(CeilingProp.TeleportTo[vector3{X:= 7168.0, Y:= -18958.0, Z:= 4752.0}, rotation{}]):

        for(HeroDevice : AllHeroDevices):
            HeroDevice.HeroDeviceInit()

        RegisterSubscriber(SignalGrayRemoteSlenderman.PrimarySignalEvent.Subscribe(HandleJumpScare))

        for(Player : GetRemainingPlayers(), Fort := Player.GetFortCharacter[]):
            RegisterSubscriber(Fort.DamagedEvent().Subscribe(OnCharacterDamaged))
            RegisterSubscriber(Fort.DamagedShieldEvent().Subscribe(OnCharacterDamaged))
            NewUI := UI_Hero_Select_Halloween{}
            if(PlayerUI := GetPlayerUI[Player]):
                if(set BeginningPlayerUIMap[Player] = NewUI):
                    PlayerUI.AddWidget(NewUI.CreateUI(Player, Self), player_ui_slot{InputMode := ui_input_mode.All})

        # On screen hero selection timer
        Sleep(8.0)
        # Remove Hero selection UI
        for(Player -> PlayerUI : BeginningPlayerUIMap):
            PlayerUI.RemoveUI(Player)
        set BeginningPlayerUIMap = map{}

        # If they didn't pick
        Sleep(1.0)
        for(Player : GetRemainingPlayers()):
            if (HeroObj := PlayerHeroMap[Player]):
            else:
                if(HeroDevice := HeroDevice3?):
                    HeroObj := HeroDevice.CreateHeroObj(Player)
                    set HeroObj.HalloweenHeroDevice.HalloweenHeroesManager = option{Self}
                    if(set PlayerHeroMap[Player] = HeroObj):
                    HeroObj.InitPlayer()

    OnSafeTimeEnded<override>():void=
        # Subscribe to their ults
        return

    _OnMinigameEnded<override>(Winners: []player)<suspends>:void=
        for (HeroDevice: AllHeroDevices):
            Print("Resetting Hero Device")
            spawn{HeroDevice.HalloweenHeroDeviceReset()}

        for(K -> V :PlayerHeroMap):
            Print("Removing Widget")
            V.RemoveWidget()
        set PlayerHeroMap = map{}

    _Reset<override>()<suspends>:void= 
        if(CeilingProp.TeleportTo[vector3{X:= 11403.0, Y:= -18958.0, Z:= 1919.0}, rotation{}]):

    _OnPlayerEliminated<override>(Result: elimination_result)<suspends>:void =
        if(Agent := Result.EliminatingCharacter?.GetAgent[], Player := player[Agent]):
            # Increase players ult for getting an elim
            if(CustomHeroClass := PlayerHeroMap[Player]):
        if(Agent := Result.EliminatedCharacter.GetAgent[], Player := player[Agent]):
            # Remove player from hero class list
            if(CustomHeroClass := PlayerHeroMap[Player]):
                CustomHeroClass.RemoveWidget()

    OnHeroSelect_Any(Widget: widget_message):void=
        if (PlayerUI := BeginningPlayerUIMap[Widget.Player]):
            PlayerUI.RemoveUI(Widget.Player)

    OnCharacterDamaged(DamageResult : damage_result):void=
        HandlePlayerHitCustom(DamageResult)

    HandlePlayerHitCustom( DamageResult : damage_result ) : void =
        if:
            SiphonModifier : float = 0.5
            Target := DamageResult.Target
            Amount := DamageResult.Amount
            Instigator := DamageResult.Instigator?
            InstigatorAgent := Instigator.GetInstigatorAgent[]
            InstigatorFort := InstigatorAgent.GetFortCharacter[]
            InstigatorPlayer := player[InstigatorAgent]
            #Instigator is VAMPIRE
            CustomHalloweenHero := PlayerHeroMap[InstigatorPlayer]


            CharacterWhoWasHitFort := fort_character[Target]
            CharacterWhoWasHitAgent := CharacterWhoWasHitFort.GetAgent[]
            CharacterWhoWasHitPlayer := player[CharacterWhoWasHitAgent]
        then:
            
            if(InstigatorAgent = CharacterWhoWasHitAgent):
                Print("Same Agent: Usually from storm damage")
            else:
                if(CustomHalloweenHero.VampireLogic = true):
                    Print("Lifesteal HIT")
                    var CustomSiphon : float = Amount * SiphonModifier
                    var InstigatorHealth: float = InstigatorFort.GetHealth()
                    InstigatorFort.SetHealth(InstigatorHealth + CustomSiphon)
                if(MammothConditionalButton.IsHoldingItem[InstigatorPlayer]):
                    Print("Holding Mammoth")
                    spawn{HandleFreezingCharacter(CharacterWhoWasHitAgent)}
                if(RipsawConditionalButton.IsHoldingItem[InstigatorPlayer]):
                    Print("Holding RIPSAW")
                    #Args := damage_args{Instigator:=DamageResult.Instigator,Source:=DamageResult.Source,Amount:=500.0}
                    CharacterWhoWasHitFort.Damage(5000.0)
                    AudioEvilLaugh.Play(InstigatorAgent)
                else:
                    Print("not special")

    HandleFreezingCharacter(Agent : agent)<suspends>:void=
        Print("Handling Freeze Fort Character")
        if:
            Fort := Agent.GetFortCharacter[]
            HeroObj := PlayerHeroMap[Agent].HalloweenHeroDevice
        then:
            Args := stasis_args{AllowTurning:=false,AllowFalling:=false,AllowEmotes:=false}
            Fort.PutInStasis(Args)
            Fort.SetVulnerability(false)
            AudioFreeze.Play(Agent)
            FreezeVFX.Pickup(Agent)
            HeroObj.Freeze()
            Sleep(3.5)
            HeroObj.UnFreeze()
            VFXClearDevice.Pickup(Agent)
            Fort.ReleaseFromStasis()
            Fort.SetVulnerability(true)
                  
    # wall hack
    OnHeroSelect_1(Widget: widget_message):void=
        Print("On Hero Select1")
        if(HeroDevice := HeroDevice1?):
            HeroObj := HeroDevice.CreateHeroObj(Widget.Player)
            set HeroObj.HalloweenHeroDevice.HalloweenHeroesManager = option{Self}
            if(set PlayerHeroMap[Widget.Player] = HeroObj):
            HeroObj.InitPlayer()

    OnHeroSelect_2(Widget: widget_message):void=
        Print("On Hero Select2")
        if(HeroDevice := HeroDevice2?):
            HeroObj := HeroDevice.CreateHeroObj(Widget.Player)
            set HeroObj.HalloweenHeroDevice.HalloweenHeroesManager = option{Self}
            if(set PlayerHeroMap[Widget.Player] = HeroObj):
            HeroObj.InitPlayer()

    OnHeroSelect_3(Widget: widget_message):void=
        Print("On Hero Select3")
        if(HeroDevice := HeroDevice3?):
            HeroObj := HeroDevice.CreateHeroObj(Widget.Player)
            set HeroObj.HalloweenHeroDevice.HalloweenHeroesManager = option{Self}
            if(set PlayerHeroMap[Widget.Player] = HeroObj):
            HeroObj.InitPlayer()

    OnHeroSelect_4(Widget: widget_message):void=
        Print("On Hero Select4")
        if(HeroDevice := HeroDevice4?):
            HeroObj := HeroDevice.CreateHeroObj(Widget.Player)
            set HeroObj.HalloweenHeroDevice.HalloweenHeroesManager = option{Self}
            if(set PlayerHeroMap[Widget.Player] = HeroObj):
            HeroObj.InitPlayer()

    OnHeroSelect_5(Widget: widget_message):void=
        Print("On Hero Select5")
        if(HeroDevice := HeroDevice5?):
            HeroObj := HeroDevice.CreateHeroObj(Widget.Player)
            set HeroObj.HalloweenHeroDevice.HalloweenHeroesManager = option{Self}
            if(set PlayerHeroMap[Widget.Player] = HeroObj):
            HeroObj.InitPlayer()
            set HeroObj.VampireLogic = true

    OnHeroSelect_6(Widget: widget_message):void=
        Print("On Hero Select6")
        if(HeroDevice := HeroDevice6?):
            HeroObj := HeroDevice.CreateHeroObj(Widget.Player)
            set HeroObj.HalloweenHeroDevice.HalloweenHeroesManager = option{Self}
            if(set PlayerHeroMap[Widget.Player] = HeroObj):
            HeroObj.InitPlayer()

    OnHeroSelect_7(Widget: widget_message):void=
        Print("On Hero Select7")
        if(HeroDevice := HeroDevice7?):
            HeroObj := HeroDevice.CreateHeroObj(Widget.Player)
            set HeroObj.HalloweenHeroDevice.HalloweenHeroesManager = option{Self}
            if(set PlayerHeroMap[Widget.Player] = HeroObj):
            HeroObj.InitPlayer()
            
    OnHeroSelect_8(Widget: widget_message):void=
        Print("On Hero Select8")
        if(HeroDevice := HeroDevice8?):
            HeroObj := HeroDevice.CreateHeroObj(Widget.Player)
            set HeroObj.HalloweenHeroDevice.HalloweenHeroesManager = option{Self}
            if(set PlayerHeroMap[Widget.Player] = HeroObj):
            HeroObj.InitPlayer()

    HandleJumpScare(Player:agent):void=
        Print("Handle jump scare")
        var MaybeTarget : ?fort_character = false
        var NearestLocation : float = 1000000.0

        for (EnemyPlayer : GetRemainingPlayers()):
            if(not Player = EnemyPlayer, TargetFort := EnemyPlayer.GetFortCharacter[]):
                if (DistancePlayerToTarget := Distance(TargetFort.GetTransform().Translation, Player.GetFortCharacter[].GetTransform().Translation) < NearestLocation):
                    set MaybeTarget = option{TargetFort}
                    set NearestLocation = DistancePlayerToTarget
        if(Target := MaybeTarget?):
            if(Player.GetFortCharacter[].TeleportTo[Target.GetViewLocation(), Target.GetViewRotation()]):
                AudioJumpScare.Play(Player)