using { /Fortnite.com/Devices }
using { /Fortnite.com/AI}
using { /Fortnite.com/Teams}
using { /Verse.org/Simulation }
using { /Verse.org/Random }
using { /Verse.org/Assets }
using { /Fortnite.com/Characters }
using { /Fortnite.com/FortPlayerUtilities }
using { /Fortnite.com/Game }
using { /Fortnite.com/Teams }
using { /Verse.org/Simulation/Tags }
using { /UnrealEngine.com/Temporary/Diagnostics }
using { /UnrealEngine.com/Temporary/SpatialMath } 
using { /Verse.org/Assets }

# A Verse-authored creative device that can be placed in a level
dash_clash := class(minigame_device):
    @editable CeilingProp : creative_prop = creative_prop{}
    @editable BigWaterVolume : water_device = water_device{}
    @editable IndividualRedWaterVolume : water_device = water_device{}

    var BigWaterVolumeOrigin :vector3 = vector3{}
    var IndividualRedWaterVolumeOrigin :vector3 = vector3{}

    Name<override>: string = "Dash Clash"
    StartingMessage<override>:string = "Get more dashes on eliminations!"
    GetVotingImage<override>()<transacts>: texture = VerseExposed.SQ_DASHCLASH

    _Start<override>()<suspends>:void=
        set BigWaterVolumeOrigin = BigWaterVolume.GetTransform().Translation
        set IndividualRedWaterVolumeOrigin = IndividualRedWaterVolume.GetTransform().Translation
        if(CeilingProp.TeleportTo[vector3{X:= 7168.0, Y:= -18958.0, Z:= 4752.0}, rotation{}]):
            #(X=7245.570882,Y=-19530.280654,Z=1604.875992)
        if(BigWaterVolume.TeleportTo[vector3{X:= 7245.5, Y:= Y:=-19530.2, Z:= 1604.8}, rotation{}]):
        Sleep(4.0)
        if(BigWaterVolume.TeleportTo[BigWaterVolumeOrigin, rotation{}]):

    OnSafeTimeEnded<override>()<suspends>:void=
        Print("Disabling input trigger")
        

    _OnMinigameEnded<override>(Winners: []player)<suspends> : void=
        #if(BigWaterVolume.TeleportTo[BigWaterVolumeOrigin, rotation{}]):
        if(IndividualRedWaterVolume.TeleportTo[IndividualRedWaterVolumeOrigin, rotation{}]):
        
    _Reset<override>()<suspends>:void=
        if(CeilingProp.TeleportTo[vector3{X:= 11403.0, Y:= -18958.0, Z:= 1919.0}, rotation{}]):

    _OnPlayerEliminated<override>(Result: elimination_result)<suspends>:void =
        if(Agent := Result.EliminatingCharacter?.GetAgent[], Player := player[Agent], Fort := Player.GetFortCharacter[]):
            if(IndividualRedWaterVolume.TeleportTo[Fort.GetTransform().Translation + vector3{Y:=-20.0, Z := -40.0}, Fort.GetTransform().Rotation]):
                Sleep(0.1)
                if(IndividualRedWaterVolume.TeleportTo[IndividualRedWaterVolumeOrigin, Fort.GetTransform().Rotation]):
