using { /Fortnite.com/Devices }
using { /Fortnite.com/AI}
using { /Fortnite.com/Teams}
using { /Verse.org/Simulation }
using { /Verse.org/Random }
using { /Verse.org/Assets }
using { /Fortnite.com/Characters }
using { /Fortnite.com/FortPlayerUtilities }
using { /Fortnite.com/Game }
using { /Fortnite.com/Teams }
using { /Verse.org/Simulation/Tags }
using { /UnrealEngine.com/Temporary/Diagnostics }
using { /UnrealEngine.com/Temporary/SpatialMath } 
using { /Verse.org/Assets }

# A Verse-authored creative device that can be placed in a level
dash_attack := class(minigame_device):
    @editable GreenWaterVolume : water_device = water_device{}
    @editable ItemGranterPort : item_granter_device = item_granter_device{}

    var GreenWaterVolumeOrigin :vector3 = vector3{}
    var PlayerKilledMap : [player]logic = map{}

    Name<override>: string = "Dash Attack"
    #StartingMessage<override>:string = "Get more dashes on eliminations!"
    GetVotingImage<override>()<transacts>: texture = VerseExposed.SQ_DASHES

    _Start<override>()<suspends>:void=
        set GreenWaterVolumeOrigin = GreenWaterVolume.GetTransform().Translation
        #(X=595.385439,Y=442.426530,Z=380.965001)
        if(GreenWaterVolume.TeleportTo[vector3{X:= 595.0, Y:=442.0, Z:= 380.0}, rotation{}]):
        for(Player : GetRemainingPlayers(), Fort := Player.GetFortCharacter[]):
            if(set PlayerKilledMap[Player] = false) {}

    _Reset<override>()<suspends>:void=
        set PlayerKilledMap = map{}
        if(GreenWaterVolume.TeleportTo[GreenWaterVolumeOrigin, rotation{}]):

    _OnPlayerEliminated<override>(Result: elimination_result)<suspends>:void =
        if(Agent := Result.EliminatingCharacter?.GetAgent[], Player := player[Agent], Fort := Player.GetFortCharacter[]):
            if(PlayerKilledMap[Player] = false):
                Print("First Kill")
                if(set PlayerKilledMap[Player] = true) {}
                ItemGranterPort.GrantItem(Player)
            else:
                Print("Not FIrst Kill")