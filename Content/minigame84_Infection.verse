using { /Fortnite.com/Devices }
using { /Verse.org/Simulation }
using { /Verse.org/Random }
using { /Fortnite.com/Characters }
using { /Fortnite.com/FortPlayerUtilities }
using { /Fortnite.com/Game }
using { /Verse.org/Simulation/Tags }
using { /UnrealEngine.com/Assets }
using { /UnrealEngine.com/Temporary/Diagnostics }
using { /UnrealEngine.com/Temporary/SpatialMath } 
using { /Verse.org/Assets }
using { /Fortnite.com/Devices/CreativeAnimation }
using { /UnrealEngine.com/Temporary/UI }
using { /Fortnite.com/UI }
using { /Verse.org/Colors }
using {VerseExposed}

minigame_Infection := class(minigame_device):
    @editable
    AFKTime : int = 8

    @editable
    DoubleInfectPlayerCount : int = 8

    @editable
    BackgroundMusic :audio_player_device = audio_player_device{}

    #@editable
    #TimerDevice : timer_device = timer_device{}

    # @editable
    # HealthPowerUp :health_powerup_device = health_powerup_device{}

    @editable
    ItemGranterLimitedMats : item_granter_device = item_granter_device{}

    @editable
    ItemGranterInfectedItems : item_granter_device = item_granter_device{}

    @editable
    ItemGranterInfectedWeapon : item_granter_device = item_granter_device{}

    @editable
    InfectedClass : class_and_team_selector_device = class_and_team_selector_device {}

    @editable
    InfectedScoreHudDevice : hud_message_device = hud_message_device{}

    #@editable
    #InfectedVFXPowerUp : []visual_effect_powerup_device = array{}

    #@editable
    #WarningCircleVFXPowerUp : []visual_effect_powerup_device = array{}

    #@editable
    #WarningCircleResetVFXPowerUp: visual_effect_powerup_device = visual_effect_powerup_device{}

    @editable
    IntroCinematic : cinematic_sequence_device = cinematic_sequence_device{}

    #@editable
    #InputControls : input_trigger_device = input_trigger_device{}

    @editable
    GameplayCamera : gameplay_camera_orbit_device = gameplay_camera_orbit_device{}

    #@editable
    #ExplosiveDevice : []explosive_device = array{}

    @editable
    StartingPillarProp : creative_prop = creative_prop{}

    @editable
    CeilingProp : creative_prop = creative_prop{}

    @editable
    StartingSpawnPointInfected1 : creative_prop = creative_prop{}

    @editable
    StartingSpawnPointInfected2 : creative_prop = creative_prop{}

    #@editable
    #ExplosionProp : creative_prop_asset = DefaultCreativePropAsset

    #@editable
    #AudioPlayerExplosion :[]audio_player_device = array{}

    @editable
    DBNO : down_but_not_out_device = down_but_not_out_device{}

    @editable
    MovementSpeedMod : movement_modulator_device = movement_modulator_device{}

    var PlayerInfected : [player]logic = map{}
    var PowerUpIndex : int = 0
    var AudioDeviceIndex : int = 0
    #var PlayerIsExploding : [player]logic = map{}
    #var ExplosionCharged : [player]logic = map{}
    #var PowerUpWarningCircleIndex : int = 0  
    #var PowerUpWarningChargeIndex : int = 0  
    #var ExplosiveDeviceIndex : int = 0
    #var ExplosionReadyCanvas : [player]?canvas = map{}

    Name<override>: string = "INFECTION"
    StartingMessage<override>:string = "Run away from Infected!"
    GetVotingImage<override>()<transacts>: texture = VerseExposed.UI_UNKNOWN

    _Start<override>()<suspends>: void =
        Print("Do Beginning Cinematic")
        BackgroundMusic.Enable()
        BackgroundMusic.Play()
        #TimerDevice.Enable()
        #InputControls.Enable()
        DBNO.Enable()
        GameplayCamera.Enable()
        MovementSpeedMod.Enable()

        if(StartingPillarProp.TeleportTo[vector3{X:=7190.0,Y:=-19505.0,Z:=1179.0}, rotation{}]):
        if(CeilingProp.TeleportTo[vector3{X:= 7168.0, Y:= -18958.0, Z:= 2816.0}, rotation{}]):

        #RegisterSubscriber(InputControls.PressedEvent.Subscribe(ExplodeInfectedPressed))
        #RegisterSubscriber(TimerDevice.SuccessEvent.Subscribe(OnTimerSuccess))
        RegisterSubscriber(IntroCinematic.StoppedEvent.Subscribe(CinematicEnd))
        RegisterSubscriber(DBNO.AgentDownedEvent.Subscribe(OnPlayerDowned))
        RegisterSubscriber(GetPlayspace().PlayerRemovedEvent().Subscribe(OnPlayerLeftInfected))

        TeamCollection := GetPlayspace().GetTeamCollection()
        for(Player : GetRemainingPlayers()):
            if(Team78 := TeamCollection.GetTeams()[77]):
                option{TeamCollection.AddToTeam[Player, Team78]}
            #HealthPowerUp.Pickup()
            if(set PlayerInfected[Player] = false):
            if(FortChar:= Player.GetFortCharacter[]):
                FortChar.PutInStasis(stasis_args{AllowTurning := true, AllowFalling := true})
        
    OnSafeTimeEnded<override>()<suspends>:void=
        #TimerDevice.Reset()
        #TimerDevice.Start()
        #HealthPowerUp.Pickup()
        spawn{SelectInitialInfected()}
        spawn{HandleInfectedWinners()}
        spawn{ItemGranters()}

    _OnMinigameEnded<override>(Winners: []player)<suspends>:void=
        BackgroundMusic.Disable()
        DBNO.Disable()
        InfectedScoreHudDevice.Hide()
        MovementSpeedMod.Disable()
        #TimerDevice.Reset()
        #TimerDevice.Disable()
        #InputControls.UnregisterAll()
        #InputControls.Disable()
        GameplayCamera.RemoveFromAll()
        GameplayCamera.Disable()
        # for(Player -> Logic : PlayerInfected, Agent := agent[Player]):
        #     #Print("Removing Explosion UI")
        #     RemoveExplosionReadyUI(Agent)
        #for(Player :GetRemainingPlayers(), Agent := agent[Player]):
        #    RemoveExplosionReadyUI(Agent)

    _Reset<override>()<suspends>:void=
        if(StartingPillarProp.TeleportTo[vector3{X:=7190.0,Y:=-19505.0,Z:=-260.0}, rotation{}]):
        if(CeilingProp.TeleportTo[vector3{X:= 11403.0, Y:= -18958.0, Z:= 1919.0}, rotation{}]):
        #for(Player :GetRemainingPlayers(), Agent := agent[Player]):
        #    RemoveExplosionReadyUI(Agent)
        set PlayerInfected = map{}

    OnPlayerDowned(Agent : agent):void=
        Print("On Player Downed")
        if(Player := player[Agent]):
            if(set PlayerInfected[Player] = true): 
            SetScoreHUD()
            DBNO.Revive(Agent)
            InfectedClass.ChangeClass(Agent)
            #AssignPlayerRandomInfectedVFX(Agent)
            #InputControls.Register(Agent)
            GameplayCamera.AddTo(Agent)
            MovementSpeedMod.Activate(Agent)
            Print("Give Sword to Infected 1")
            ItemGranterInfectedWeapon.GrantItem(Player)
            #spawn:
            #    ChargeExplosion(Agent)

    TurnPlayerIntoInfected(Agent: agent):void=
        Print("Turning Player into Infected 2")
        if(Player := player[Agent]):
            if(set PlayerInfected[Player] = true): 
            SetScoreHUD()
            #DBNO.Revive(Agent)
            InfectedClass.ChangeClass(Agent)
            #AssignPlayerRandomInfectedVFX(Agent)
            #InputControls.Register(Agent)
            GameplayCamera.AddTo(Agent)
            MovementSpeedMod.Activate(Agent)
            Print("Give Sword to Infected 3")
            ItemGranterInfectedWeapon.GrantItem(Player)
            #spawn:
            #    ChargeExplosion(Agent)

    SelectInitialInfected()<suspends>:void=
        race:
            EndedEvent.Await()
            block:
                Sleep(0.1)
                ShuffledPlayers := Shuffle(GetRemainingPlayers())
                if(FirstInfectedPlayer := ShuffledPlayers[0], Agent1 :=  agent[FirstInfectedPlayer], FortChar1 := FirstInfectedPlayer.GetFortCharacter[], Player1 := player[Agent1]):
                    if(set PlayerInfected[FirstInfectedPlayer] = true): 
                        Agent1.Respawn(StartingSpawnPointInfected1.GetTransform().Translation, StartingSpawnPointInfected1.GetTransform().Rotation)

                    SetScoreHUD()
                    FortChar1.PutInStasis(stasis_args{AllowTurning := false, AllowFalling := true})
                    InfectedClass.ChangeClass(Agent1)
                    #FortChar.SetVulnerability(false) 
                    #AssignPlayerRandomInfectedVFX(Agent)
                    Sleep(0.1)
                    Manager.ResetItemRemover.Remove(Agent1)
                    IntroCinematic.Play()
                    Sleep(4.0)
                    #ChargeExplosion(Agent)
                    #InputControls.Register(Agent)
                    GameplayCamera.AddTo(Agent1)
                    MovementSpeedMod.Activate(Agent1)
                    Print("Give Sword to Infected 4")
                    ItemGranterInfectedWeapon.GrantItem(Player1)
                    #AFK Check
                    spawn{WatchInitialInfectedStandStill(FirstInfectedPlayer,Agent1, FortChar1)}
                    spawn{HealPlayers()}

                #Checks for adding 2 starting infected
                if(GetRemainingPlayers().Length >= DoubleInfectPlayerCount):
                    if(SecondInfectedPlayer := ShuffledPlayers[1], Agent2 :=  agent[SecondInfectedPlayer], FortChar2 := SecondInfectedPlayer.GetFortCharacter[], Player2 := player[Agent2]):
                        if(set PlayerInfected[SecondInfectedPlayer] = true): 
                            Agent2.Respawn(StartingSpawnPointInfected2.GetTransform().Translation, StartingSpawnPointInfected2.GetTransform().Rotation)

                        SetScoreHUD()
                        FortChar2.PutInStasis(stasis_args{AllowTurning := false, AllowFalling := true})
                        InfectedClass.ChangeClass(Agent2)
                        #FortChar.SetVulnerability(false) 
                        #AssignPlayerRandomInfectedVFX(Agent)
                        Sleep(0.1)
                        Manager.ResetItemRemover.Remove(Agent2)
                        IntroCinematic.Play()
                        Sleep(4.0)
                        #ChargeExplosion(Agent)
                        #InputControls.Register(Agent)
                        GameplayCamera.AddTo(Agent2)
                        MovementSpeedMod.Activate(Agent2)
                        Print("Give Sword to Infected 5")
                        ItemGranterInfectedWeapon.GrantItem(Player2)
                        #AFK Check
                        spawn{WatchInitialInfectedStandStill(SecondInfectedPlayer, Agent2, FortChar2)}
                        spawn{HealPlayers()}

    #Check for AFK Players
    WatchInitialInfectedStandStill(Player : player,Agent: agent, Fort : fort_character)<suspends>:void=
        var lastPosition : vector3 = Fort.GetTransform().Translation
        var initalTime : int = 0
        race:
            EndedEvent.Await()
            loop:
                ActualPosition := Fort.GetTransform().Translation
                if(Distance(ActualPosition,lastPosition) < 1.0):
                    Print("Same Position bro")
                    set initalTime = initalTime + 1
                    if(initalTime >= AFKTime):
                        for(NewPlayer -> Logic : PlayerInfected):
                            if(PlayerInfected[NewPlayer] = false):
                                if:
                                    NewAgent := agent[NewPlayer]
                                then:
                                    Print("Making new random agent Infected")
                                    if(set PlayerInfected[NewPlayer] = true): 
                                        SetScoreHUD()
                                        InfectedClass.ChangeClass(NewAgent)
                                        #AssignPlayerRandomInfectedVFX(NewAgent)
                                        #InputControls.Register(NewAgent)
                                        GameplayCamera.AddTo(NewAgent)
                                        MovementSpeedMod.Activate(NewAgent)
                                        Print("Give Sword to Infected 6")
                                        ItemGranterInfectedWeapon.GrantItem(Player)
                                        #spawn:
                                        #    ChargeExplosion(NewAgent)
                            break
                else:
                    Print("First Infected Player Moved. Breaking out of Loop")
                    break
                Sleep(1.0)

    HealPlayers()<suspends>:void=
        race:
            EndedEvent.Await()
            loop:
                Sleep(1.0)
                for(Player: GetRemainingPlayers(), FortChar:= Player.GetFortCharacter[]):
                    FortChar.SetHealth(FortChar.GetHealth() + 1.0)

    SetScoreHUD():void=
        var PlayerScore : int = 0
        var InfectedScore : int = 0
        for(Player -> Logic : PlayerInfected):
            if(PlayerInfected[Player] = true):
                set InfectedScore += 1
            else: 
                set PlayerScore += 1
        InfectedScoreHudDevice.SetText(StringToMessage("{PlayerScore}\n{InfectedScore}"))
        InfectedScoreHudDevice.Show()

    CinematicEnd():void=
        for(Player : GetPlayspace().GetPlayers(), FortChar:= Player.GetFortCharacter[]):
            FortChar.ReleaseFromStasis()
        #Fix Glitch wehre releasing from stasis cancels no emote mutator zone
        Manager.NoEmotingMutatorZone.Disable()
        Manager.NoEmotingMutatorZone.Enable()

<#
    AssignPlayerInfectedVFX(Agent : agent):void=
        if(PowerUp := InfectedVFXPowerUp[PowerUpIndex]):
            PowerUp.Pickup(Agent)

        #Increment Powerup Index
        if(PowerUpIndex >= InfectedVFXPowerUp.Length - 1):
            set PowerUpIndex = 0
        else:
            set PowerUpIndex += 1

    AssignPlayerRandomInfectedVFX(Agent : agent):void=
        DiceRoll := GetRandomInt(0, InfectedVFXPowerUp.Length - 1)
        if(PowerUp := InfectedVFXPowerUp[DiceRoll]):
            PowerUp.Pickup(Agent)
#>

    _OnGenericEndTimerSuccessReturnArray<override>():[]player=
        Print("Timer")
        var Winners : []player = array{}
        for(Player -> Logic : PlayerInfected):
            if(PlayerInfected[Player] = false):
                set Winners += array{Player}
        #for(Player -> Logic : PlayerInfected, Agent := agent[Player]):
        #    RemoveExplosionReadyUI(Agent)
        return Winners
        
    HandleInfectedWinners()<suspends>:void=
        race:
            EndedEvent.Await()
            block:
                loop:
                    Sleep(0.2)
                    var PlayersAlive : int = 0
                    var RetArray : []player = array{}
                    for(Player -> Logic : PlayerInfected):
                        if(PlayerInfected[Player] = true):
                            set RetArray += array{Player}
                            #set PlayersAlive = PlayersAlive + 1
                    # 0 NON Infected Players Alive
                    #Print("RetArr Length {RetArray.Length} --- PlayerInfected Length{PlayerInfected.Length}")
                    if(RetArray.Length = PlayerInfected.Length):
                        # Allow for draw to happen
                        #for(Player -> Logic : PlayerInfected, Agent := agent[Player]):
                        #    RemoveExplosionReadyUI(Agent)
                        EndedEvent.Signal(RetArray)

    ItemGranters()<suspends>:void=
        race:
            EndedEvent.Await()
            block:
                loop:
                    Sleep(10.0)
                    for(Player : GetRemainingPlayers()):
                        ItemGranterLimitedMats.GrantItem(Player)
                        ItemGranterInfectedItems.CycleToRandomItem(Player)
                        ItemGranterInfectedItems.CycleToRandomItem(Player)

    OnPlayerLeftInfected(PlayerLeaving: player):void=
        var NewPlayerInfected:[player]logic = map{}
        # Concatenate Keys from ExampleMap into NewMap, excluding ElementToRemove
        for (Key -> Value : PlayerInfected, Key <> PlayerLeaving):
            set NewPlayerInfected = ConcatenateMaps(NewPlayerInfected, map{Key => Value})
        Print("Old size = {PlayerInfected.Length} -- New Size = {NewPlayerInfected.Length}")
        #If 0 Infected, make a new Infected
        var InfectedAliveLogic :logic = false
        var RetArray : []player = array{}
        for(Player -> Logic : NewPlayerInfected):
            set RetArray += array{Player}
            if(PlayerInfected[Player] = true):
                set InfectedAliveLogic = true
        if(InfectedAliveLogic = false):
            EndedEvent.Signal(RetArray)
        else:
            set PlayerInfected = NewPlayerInfected
            SetScoreHUD()
<#
    AssignPlayerWarningCircleVFX(Agent : agent):void=
        if(PowerUp := WarningCircleVFXPowerUp[PowerUpWarningCircleIndex]):
            PowerUp.Pickup(Agent)
        #Increment Powerup Index
        if(PowerUpWarningCircleIndex >= WarningCircleVFXPowerUp.Length - 1):
            set PowerUpWarningCircleIndex = 0
        else:
            set PowerUpWarningCircleIndex += 1


    PlayExplosionAudio(AudioLocation : transform):void=
        if(AudioDevice := AudioPlayerExplosion[AudioDeviceIndex]):
            if(AudioDevice.TeleportTo[AudioLocation]):
                AudioDevice.Play()
        if(AudioDeviceIndex >= AudioPlayerExplosion.Length - 1):
            set AudioDeviceIndex = 0
        else:
            set AudioDeviceIndex += 1

    #Explosion Charged UI
    ActivateExplosionReadyUI(Agent : agent) : void =
        if (Player := player[Agent], PlayerUI := GetPlayerUI[Player]):
            if (MyUI := ExplosionReadyCanvas[Player]?):
                Print("Removing Explosion Canvas")
                PlayerUI.RemoveWidget(MyUI)
                if (set ExplosionReadyCanvas[Player] = false) {}
            NewUI := CreateExplosionReadyCanvas()
            PlayerUI.AddWidget(NewUI)
            if (set ExplosionReadyCanvas[Player] = option{NewUI}) {}

    RemoveExplosionReadyUI(Agent : agent):void=
        if (InPlayer := player[Agent], PlayerUI := GetPlayerUI[InPlayer]):
            Print("Removing Explosion Canvas Beep")
            if (MyUI := ExplosionReadyCanvas[InPlayer]?):
                PlayerUI.RemoveWidget(MyUI)
                if (set ExplosionReadyCanvas[InPlayer] = false) {}

    CreateExplosionReadyCanvas(): canvas =
        ExplosionReadyTexture  : texture_block = texture_block{ DefaultImage := VerseExposed.T_Press_Fire_To_Explode, DefaultDesiredSize := vector2{ X := 512.0, Y := 73.0}, DefaultTint := color{R:=1.0, G:=0.1, B:=0.0}}
        MyDamageText : text_block = text_block{DefaultText := StringToMessage("Press Fire to Explode"), DefaultTextColor := NamedColors.Gray}
        Print("Creating Explosion Charged UI")
        MyCanvas : canvas = canvas:
            Slots := array:
                canvas_slot: 
                    Anchors := anchors{Minimum := vector2{X := 0.5, Y := 0.85}, Maximum := vector2{X := 0.5, Y :=0.85}}
                    Alignment := vector2{X := 0.5, Y := 0.5}
                    Widget := overlay:
                        Slots := array:  
                            overlay_slot:
                                Widget := ExplosionReadyTexture         
        return MyCanvas

    SpawnExplosion(Agent : agent)<suspends>:void=
        race:
            EndedEvent.Await()
            block:
                if(Player := player[Agent]):
                    if(FortChar:= Player.GetFortCharacter[]):
                        
                        WarningCircleResetVFXPowerUp.Pickup(Agent)
                        var NewTransform : transform = FortChar.GetTransform()
                        set NewTransform.Translation = FortChar.GetTransform().Translation  + vector3{Z:=-70.0}
                        NewExplosionProp := SpawnProp(ExplosionProp, NewTransform)

                        PlayExplosionAudio(NewTransform)
                        #ExplosionAudio.Play(Agent)

                        if(Explosive := ExplosiveDevice[ExplosiveDeviceIndex]):
                            if:
                                Explosive.TeleportTo[NewTransform.Translation, NewTransform.Rotation]
                            Explosive.Reset()
                            Explosive.Explode(Agent)

                            if(ExplosiveDeviceIndex >= ExplosiveDevice.Length - 1):
                                set ExplosiveDeviceIndex = 0
                            else:
                                set ExplosiveDeviceIndex += 1

                            defer:
                                Explosive.Reset()

                        if(set PlayerIsExploding[Player] = false):

                        Sleep(0.25)
                        if(RealProp:= NewExplosionProp(0)?):
                            spawn:
                                DespawnExplosionProp(RealProp)
                            spawn:
                                ChargeExplosion(Agent)
                        # VFX.Cancel()

    DespawnExplosionProp(Prop : creative_prop )<suspends>:void=
            Sleep(2.5) 
            Prop.Dispose()
            Print("Canceling VFX")

    ChargeExplosion(Agent : agent)<suspends>:void=
        race:
            EndedEvent.Await()
            block:
                AssignPlayerWarningCircleVFX(Agent)
                Sleep(3.5)
                # InputControls.Register(Agent)
                if(Player := player[Agent]):
                    if(set ExplosionCharged[Player] = true):
                        ActivateExplosionReadyUI(Agent)
                        Print("Explosion Charged")

    ExplodeInfectedPressed(Agent:agent):void=
        if(Player := player[Agent], ExplosionCharged[Player] = false):
            Print("Explosion Not Charged.")
            return

        if(Player := player[Agent], PlayerIsExploding[Player] = true):
            Print("Don't fire more - explosion in progress")
            return

        #AssignPlayerWarningCircleVFX(Agent)
        if(Player := player[Agent], set PlayerIsExploding[Player] = true):
            if(set  ExplosionCharged[Player] = false):
            RemoveExplosionReadyUI(Agent)
            spawn:
                SpawnExplosion(Agent)
#>