using { /Fortnite.com/Devices }
using { /Verse.org/Simulation }
using { /Verse.org/Random }
using { /Fortnite.com/Characters }
using { /Fortnite.com/FortPlayerUtilities }
using { /Fortnite.com/Game }
using { /Verse.org/Simulation/Tags }
using { /UnrealEngine.com/Assets }
using { /UnrealEngine.com/Temporary/Diagnostics }
using { /UnrealEngine.com/Temporary/SpatialMath } 
using { /Verse.org/Assets }
using { /Fortnite.com/Devices/CreativeAnimation }
using { /UnrealEngine.com/Temporary/UI }
using { /Fortnite.com/UI }
using { /Verse.org/Colors }
using {VerseExposed}

minigame_VIP := class(minigame_device):
    @editable
    VipVfx : visual_effect_powerup_device = visual_effect_powerup_device{}

    Name<override>: string = "Protect the KING"
    StartingMessage<override>:string = "No more respawns if King is eliminated"
    GetVotingImage<override>()<transacts>: texture = VerseExposed.SQ_KING_Vote
    DisplayIcon<override>: ?texture = option{VerseExposed.SQ_KING}
    #GetVotingImage<override>()<transacts>: texture = VerseExposed.SQ_DOUBLEPUMP
    #DisplayIcon<override>: ?texture = option{VerseExposed.SQ_DOUBLEPUMP_ICON}
    var VipPlayers : []player = array{}

    _Start<override>()<suspends>: void =
        set VipPlayers = array{}
        for(i -> Team : GetRemainingTeams()):
            Print("{i}")
            if(RandomPlayerFromTeam := Shuffle(Team.Players)[0]):
                VipVfx.Pickup(RandomPlayerFromTeam)
                set VipPlayers += array{RandomPlayerFromTeam}
                Print("Found VIP {i}")
                #spawn{CheckIfVipIsAlive(RandomPlayerFromTeam, Team)}

        #Call this whenever you want to show crowns in a minigame
        spawn:
            GetTeamLivesUIManager().ShowCrownUI(Self)

    _OnPrePlayerEliminated<override>(Result: elimination_result)<suspends>:void=
        if(EliminatedPlayer := player[Result.EliminatedCharacter.GetAgent[]]):
            if(VipPlayers.Find[EliminatedPlayer], Team := GetTeam[EliminatedPlayer]):
                set Team.AllowSpawning = false
                VipPlayers.RemoveAllElements(EliminatedPlayer)
                Print("Vip Died")

    CheckIfVipIsAlive(Player : player, Team : custom_team)<suspends>:void =
        Print("Start VIP Loop")
        race:
            EndedEvent.Await()
            loop:
                if(Player.GetFortCharacter[].IsActive[]):
                    #Print("Vip is alive bro")
                else:
                    Print("Vip Died")
                    set Team.AllowSpawning = false
                    break
                Sleep(0.0)

    _Reset<override>()<suspends>:void=
        GetTeamLivesUIManager().ResetCrownUI()