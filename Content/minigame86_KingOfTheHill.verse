using { /Fortnite.com/Devices }
using { /Verse.org/Simulation }
using { /Verse.org/Random }
using { /Fortnite.com/Characters }
using { /Fortnite.com/FortPlayerUtilities }
using { /Fortnite.com/Game }
using { /Verse.org/Simulation/Tags }
using { /UnrealEngine.com/Assets }
using { /UnrealEngine.com/Temporary/Diagnostics }
using { /UnrealEngine.com/Temporary/SpatialMath } 
using { /Verse.org/Assets }
using { /Fortnite.com/Devices/CreativeAnimation }
using { /UnrealEngine.com/Temporary/UI }
using { /Fortnite.com/UI }
using { /Verse.org/Colors }
using {VerseExposed}

minigame_kingOfTheHill := class(minigame_device):
    @editable
    CaptureArea : capture_area_device = capture_area_device{}

    @editable 
    GoldWinValue : int = 450

    Name<override>: string = "King of the Hill"
    StartingMessage<override>:string = "Capture the hill as a team to win!"
    GetVotingImage<override>()<transacts>: texture = VerseExposed.SQ_KINGOFHILL_Vote
    DisplayIcon<override>: ?texture = option{VerseExposed.SQ_KINGOFHILL}

    OnMinigameDeviceBegin<override>()<suspends>:void=
        block:
            CaptureArea.ControlChangeEvent.Subscribe(OnControlChanged)

    OnControlChanged(Agent : agent):void=
        if:
            WinningPlayer := player[Agent]
        then:
            #var Winner : []player = array{WinningPlayer}
            var WinningPlayers : []player = array{WinningPlayer}
            #Change from winner to the team of winners
            TeamCollection := GetPlayspace().GetTeamCollection()
            if(Team := TeamCollection.GetTeam[Agent]):
                #Get the winning team
                if(WinnerTeamMembers := TeamCollection.GetAgents[Team]):
                    #Loop through and fill the array with winning players
                    for(WinningTeamMember : WinnerTeamMembers):
                        #Convert to a player for the array
                        if(Player := player[WinningTeamMember]):
                            Manager.GoldTracker.Add(Player,GoldWinValue)
                            set WinningPlayers += array{Player}
                    #Handle Winning Players
                    Print("We have a winning team")
                    EndedEvent.Signal(WinningPlayers)

    _Start<override>()<suspends>: void =
        CaptureArea.Enable()

    _Reset<override>()<suspends>:void=
        CaptureArea.Reset()
        CaptureArea.Disable()

    GetWinners<override>()<transacts><decides>:[]player=
        var RetArray : []player = array{} 
        RetArray