using { /Fortnite.com/Devices }
using { /Verse.org/Simulation }
using { /Verse.org/Assets }
using { /UnrealEngine.com/Temporary/Diagnostics }
using { /Fortnite.com/Characters }

minigame86_sabotage := class(minigame_device):
    Name<override>: string = "Sabotage"
    StartingMessage<override>:string = "Plant The Bomb In The Middle"
    GetVotingImage<override>()<transacts>: texture = VerseExposed.SQ_PLANTDEFUSE_Vote
    DisplayIcon<override>: ?texture = option{VerseExposed.T_SQ_PLANTDEFUSE}

    #Simple method to determine who has the bomb and can place
    #@editable
    #BombPlantConditionalButton : conditional_button_device = conditional_button_device{}

    #ItemSpawner - For now just spawning an item in the middle players can pickup
    #@editable
    #BombSpawner : item_spawner_device = item_spawner_device{}

    @editable
    TimeForRespawnDisable : float = 90.0

    @editable
    BombButton : switch_device = switch_device{}

    @editable
    DefuseButton : switch_device = switch_device{}

    @editable
    BombTimerDevice : timer_device = timer_device{}

    @editable
    BombTime : float = 5.0

    @editable
    SabotageBeacon : beacon_device = beacon_device{}

    @editable
    ArmedBombHUD : hud_message_device = hud_message_device{}
    @editable
    DisArmedBombHUD : hud_message_device = hud_message_device{}

    @editable
    DefuseBombHUD : hud_message_device = hud_message_device{}
    @editable
    DefendBombHUD : hud_message_device = hud_message_device{}

    @editable
    ExplosionVFX : vfx_spawner_device = vfx_spawner_device{}
    #We use team 1 and team 2 for these 8v8 minigames. So I can break it down to an int and then assign winner based on this
    var PlantingTeam : int = 0
    #Team who defuses the bomb
    var DefusingTeam : int = 0
    #I can determine the winning team based on the objective
    var WinningTeam : int = 0

    SabotageEnded<public>:event() = event(){}

    _Start<override>()<suspends>: void =
        # When the conditional button has been interacted with the bomb is planted
        BombButton.Enable()
        BombButton.SetState(true)
        DefuseButton.SetState(false)

        RegisterSubscriber(BombButton.TurnedOffEvent.Subscribe(BombPlanted))
        RegisterSubscriber(DefuseButton.TurnedOffEvent.Subscribe(BombDefused))
        BombTimerDevice.ResetForAll()

        DefuseButton.Disable()
        SabotageBeacon.Enable()

        RespawnStopCheck()

    BombPlanted(Agent : agent):void=
        DefuseButton.Enable()
        DefuseButton.SetState(true)
        BombButton.Disable()

        ArmedBombHUD.Show()
        BombTimerDevice.SetMaxDuration(BombTime)
        BombTimerDevice.StartForAll()

        if(Player := player[Agent], Team := GetTeam[Player]):
            set PlantingTeam = Team.TeamIndex
        spawn:
            BombTimer(Agent)

        #Adds HUD to players not on planting team
        for(CurrentTeam : GetRemainingTeams()):
            if(CurrentTeam.TeamIndex = PlantingTeam):
                for(Player : CurrentTeam.Players):
                    DefendBombHUD.Show(Player)
            else:
                for(Player : CurrentTeam.Players):
                    DefuseBombHUD.Show(Player)

    #This is for when the bomb is defused during the game
    #States are set for visuals
    ResetBombDuringGame():void=
        SabotageEnded.Signal()
        #Disable the Defuse Button
        DefuseButton.SetState(false)
        DefuseButton.Disable()
        #Enable the Bomb Button
        BombButton.Enable()
        BombButton.SetState(true)
        BombTimerDevice.ResetForAll()

    #Commented out win condition. Added Bomb Reset Function so players can try again
    BombDefused(Agent : agent):void=
        if(Player := player[Agent], Team := GetTeam[Player]):
            if(Team.TeamIndex = PlantingTeam):
                DefuseButton.SetState(true)
                return

            ResetBombDuringGame()
            DisArmedBombHUD.Show()
            #EndedEvent.Signal(Team.Players)    

    BombTimer(Agent : agent)<suspends>:void=
        if(Player := player[Agent], Team := GetTeam[Player]):
            race:
                SabotageEnded.Await()
                block:
                    Sleep(BombTime)
                    Print("End The Game")
                    #Explosion VFX will play
                    ExplosionVFX.Restart()
                    EndedEvent.Signal(Team.Players)
                    SabotageEnded.Signal()

    GetWinners<override>()<transacts><decides>:[]player=
        array{}

    _Reset<override>():void=
        BombButton.SetState(false)
        DefuseButton.SetState(false)
        BombButton.Disable()
        DefuseButton.Disable()
        SabotageBeacon.Disable()

    RespawnStopCheck()<suspends> : void=
        race:
            EndedEvent.Await()
            block:
                Sleep(TimeForRespawnDisable)
                StopRespawns()

    #Called when game goes on for to long
    StopRespawns() : void=
        RemainingTeamsInMinigame := GetRemainingTeams()

        for(Team : RemainingTeamsInMinigame):
            set Team.AllowSpawning = false