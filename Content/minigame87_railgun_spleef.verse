using { /Fortnite.com/Devices }
using { /Fortnite.com/AI}
using { /Fortnite.com/Teams}
using { /Verse.org/Simulation }
using { /Verse.org/Random }
using { /Verse.org/Assets }
using { /Fortnite.com/Characters }
using { /Fortnite.com/FortPlayerUtilities }
using { /Fortnite.com/Game }
using { /Fortnite.com/Teams }
using { /Verse.org/Simulation/Tags }
using { /UnrealEngine.com/Temporary/Diagnostics }
using { /UnrealEngine.com/Temporary/SpatialMath } 
using { /Verse.org/Assets }

#One Vs Three Teams
# - Need the largest possible team to be the lone team (if all teams are full/random then equal)
# for this scenario we need to be able to get the solo team and 3Team combined


railgun_spleef := class(minigame_device):
    @editable IntroCinematic : cinematic_sequence_device = cinematic_sequence_device{}
    @editable ItemGranterRailgun : item_granter_device = item_granter_device{}
    @editable RailgunClass : class_and_team_selector_device = class_and_team_selector_device {}
    @editable TeleporterBottomBox : teleporter_device = teleporter_device{}
    @editable InfectedScoreHudDevice : hud_message_device = hud_message_device{}
    @editable RailgunSpleefBlankProp : creative_prop = creative_prop{}
    @editable ItemSpawnerDeviceItems : item_spawner_device = item_spawner_device{}
    @editable ItemSpawnerDeviceItemsWithMats : item_spawner_device = item_spawner_device{}
    @editable Barriers : []barrier_device = array{}

    @editable DamageVolume : damage_volume_device = damage_volume_device{}
    @editable LavaProp : creative_prop = creative_prop{}
    @editable PlatformProp : creative_prop = creative_prop{}

    var Railgunplayers : int = 1
    var RailgunPlayersArr : []player = array{}
    var RunnersPlayersArr : []player = array{}

    Name<override>: string = "Railgun Spleef"
    #GetVotingImage<override>()<transacts>: texture = VerseExposed.SQ_RAILGUNSPLEEF2

    _Start<override>()<suspends>:void=
        ItemSpawnerDeviceItems.Enable()
        ItemSpawnerDeviceItemsWithMats.Enable()
        RegisterSubscriber(IntroCinematic.StoppedEvent.Subscribe(CinematicEnd))  
        RandomPlayersArr := Shuffle(GetRemainingPlayers())
        if(Railgunplayers = Ceil(16/4)):
            Print("{Railgunplayers}")
        if(Railgunplayers = Ceil(GetRemainingPlayers().Length/4)):
            Print("Railgunplayers = {Railgunplayers}")
            for(Index := 0..Railgunplayers-1,RandomPlayer := RandomPlayersArr[Index], Agent :=  agent[RandomPlayer]):
                set RailgunPlayersArr += array{RandomPlayer}
                ItemGranterRailgun.GrantItem(Agent)
                RailgunClass.ChangeClass(Agent)
                Manager.UiMessageIndividual.SetText(StringToMessage("Shoot down the players above!"))
                Manager.UiMessageIndividual.Show(RandomPlayer)
            for(Index := Railgunplayers..RandomPlayersArr.Length-1, Player := RandomPlayersArr[Index], Agent :=  agent[Player]):
                set RunnersPlayersArr += array{Player}
                Manager.UiMessageIndividual.SetText(StringToMessage("SURVIVE"))
                Manager.UiMessageIndividual.Show(Player)
        SetScoreHUD()
        if(RailgunSpleefBlankProp.TeleportTo[vector3{X:= 7184.0, Y:= -19392.0, Z:= 8224.0}, rotation{}]):
        if(PlatformProp.TeleportTo[vector3{X:= 7424.0, Y:= -19436.0, Z:= 3478.0}, rotation{}]):
        if(LavaProp.TeleportTo[vector3{X:= 7040.0, Y:= -18730.0, Z:= 2828.0}, rotation{}]):     
        Sleep(1.0)
        IntroCinematic.Play()
        
        for(Player : RailgunPlayersArr, Fort := Player.GetFortCharacter[]):
            TeleporterBottomBox.Teleport(Player)
            Fort.PutInStasis(stasis_args{AllowEmotes := false, AllowTurning := false, AllowFalling := true})
        
    OnSafeTimeEnded<override>()<suspends>:void=
        DamageVolume.Enable()
        for(Player : RailgunPlayersArr, Fort := Player.GetFortCharacter[]):
            Fort.SetMaxHealth(10000.0)
            Fort.SetHealth(10000.0)
            Fort.SetMaxShield(10000.0)
            Fort.SetShield(10000.0)
        spawn{ItemSpawnerLoop()}

    CinematicEnd():void=
        for(Player : GetPlayspace().GetPlayers(), FortChar:= Player.GetFortCharacter[]):
            FortChar.ReleaseFromStasis()
        #Fix Glitch wehre releasing from stasis cancels no emote mutator zone
        Manager.NoEmotingMutatorZone.Disable()
        Manager.NoEmotingMutatorZone.Enable()

    _OnMinigameEnded<override>(Winners: []player)<suspends>:void=
        ItemSpawnerDeviceItems.Disable()
        ItemSpawnerDeviceItemsWithMats.Disable()
        InfectedScoreHudDevice.Hide()
        set RailgunPlayersArr = array{}
        set RunnersPlayersArr = array{}

    _Reset<override>()<suspends>:void=
        DamageVolume.Disable()
        if(PlatformProp.TeleportTo[vector3{X:= 7424.0, Y:= -19436.0, Z:= 20.473955}, rotation{}]):
        if(LavaProp.TeleportTo[vector3{X:= 7040.0, Y:= -18730.0, Z:= -975.994872}, rotation{}]):
        if(RailgunSpleefBlankProp.TeleportTo[vector3{X:= 1405.0, Y:= -19392.0, Z:= 8224.0}, rotation{}]):

    _OnPlayerEliminated<override>(Result: elimination_result)<suspends>:void =
        SetScoreHUD()

    SetScoreHUD():void=
        InfectedScoreHudDevice.SetText(StringToMessage("{GetRemainingPlayers().Length-Railgunplayers}\n{Railgunplayers}"))

    <#GetWinners<override>()<transacts><decides>:[]player=
        #Print("Getting Winers")
        StillAlivePlayers := for(Player : RunnersPlayersArr, Player.GetFortCharacter[].IsActive[]) { Player }
        # Allow for draw to happen
        StillAlivePlayers.Length = 0
        RailgunPlayersArr#>

    _OnGenericEndTimerSuccessReturnArray<override>():[]player=
        RunnersPlayersArr

    ItemSpawnerLoop()<suspends>:void =
        var index : int = 0
        RandBarriers := Shuffle(Barriers)
        ItemSpawnerZ := ItemSpawnerDeviceItems.GetTransform().Translation.Z
        race:
            EndedEvent.Await()
            block:
                loop:
                    if(Barrier := RandBarriers[index]):
                        Print("teleporting")
                        Position := Barrier.GetTransform().Translation
                        if(ItemSpawnerDeviceItems.TeleportTo[vector3{X:= Position.X, Y:= Position.Y, Z:=ItemSpawnerZ}, rotation{}]):
                    ItemSpawnerDeviceItems.SpawnItem()
                    if(index >= Barriers.Length) then set index = 0 else set index+=1
                    Sleep(10.0)
            block:
                #Teleport Mats + items
                Sleep(45.0)
                loop:
                    if(Barrier := RandBarriers[index]):
                        Print("teleporting 2")
                        Position := Barrier.GetTransform().Translation
                        if(ItemSpawnerDeviceItemsWithMats.TeleportTo[vector3{X:= Position.X, Y:= Position.Y, Z:=ItemSpawnerZ}, rotation{}]):
                    ItemSpawnerDeviceItemsWithMats.SpawnItem()
                    if(index >= Barriers.Length) then set index = 0 else set index+=1
                    Sleep(10.0)