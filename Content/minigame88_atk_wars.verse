using { /Fortnite.com/Devices }
using { /Fortnite.com/AI}
using { /Fortnite.com/Teams}
using { /Verse.org/Simulation }
using { /Verse.org/Random }
using { /Verse.org/Assets }
using { /Fortnite.com/Characters }
using { /Fortnite.com/FortPlayerUtilities }
using { /Fortnite.com/Game }
using { /Fortnite.com/Teams }
using { /Verse.org/Simulation/Tags }
using { /UnrealEngine.com/Temporary/Diagnostics }
using { /UnrealEngine.com/Temporary/SpatialMath } 
using { /Verse.org/Assets }

#boogie bomb in center

atk_wars := class(minigame_device):
    @editable AtkBlue : vehicle_spawner_atk_device = vehicle_spawner_atk_device{}
    @editable AtkRed : vehicle_spawner_atk_device = vehicle_spawner_atk_device{}
    @editable AtkGreen : vehicle_spawner_atk_device = vehicle_spawner_atk_device{}
    @editable AtkYellow : vehicle_spawner_atk_device = vehicle_spawner_atk_device{}

    var PlayerInCorrectVehicleMap : [player]logic = map{}
    var AtkArr : []vehicle_spawner_atk_device = array{}

    Name<override>: string = "Golf Cart Wars!"
    #GetVotingImage<override>()<transacts>: texture = VerseExposed.SQ_RAILGUNSPLEEF2
    BottomMessage<override>: ?string = option{"You must stay in YOUR team's cart to not take damage!"}

    _Start<override>()<suspends>:void=
        set AtkArr = array{AtkBlue,AtkRed,AtkGreen,AtkYellow}
        for(ATK : AtkArr):
            ATK.Enable()
            #RegisterSubscriber(ATK.AgentEntersVehicleEvent.Subscribe(OnAgentEntersEvent))
            RegisterSubscriber(ATK.AgentExitsVehicleEvent.Subscribe(OnAgentExitsEvent))
        #Possibly Set it per team
        # for(Player : RedTeam):
        #     RegisterSubscriber(AtkRed.AgentEntersVehicleEvent.SubscribeAgent(OnAgentEntersEvent)
        # for(Player : BlueTeam):
        #     RegisterSubscriber(AtkRed.AgentEntersVehicleEvent.SubscribeAgent(OnAgentEntersEvent)
        
        for(Player : GetRemainingPlayers()):
            if(set PlayerInCorrectVehicleMap[Player] = false) {}

    OnSafeTimeEnded<override>()<suspends>:void=
        spawn{DamagePlayersNotInVehicleLoop()}

    #_OnMinigameEnded<override>(Winners: []player)<suspends>:void=

    _Reset<override>()<suspends>:void=
        for(ATK : AtkArr):
            ATK.Disable()
        set AtkArr = array{}
        set PlayerInCorrectVehicleMap = map{}

    OnAgentEntersEvent(Agent : agent):void=
        Print("Player entered vehicle")
        if (Player := player[Agent], set PlayerInCorrectVehicleMap[Player] = true):
            
    OnAgentExitsEvent(Agent : agent):void=
        Print("Player exiting vehicle")
        if (Player := player[Agent], set PlayerInCorrectVehicleMap[Player] = false):

    DamagePlayersNotInVehicleLoop()<suspends>:void=
        race:
            EndedEvent.Await()
            block:
                loop:
                    for(Player : GetRemainingPlayers()):
                        if(PlayerInCorrectVehicleMap[Player]=true, Fort := Player.GetFortCharacter[]):
                            Fort.SetHealth(Fort.GetHealth() - 10.0)
                    Sleep(0.2)