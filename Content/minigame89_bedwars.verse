using { /Fortnite.com/Devices }
using { /Verse.org/Simulation }
using { /Verse.org/Assets }
using { /UnrealEngine.com/Temporary/Diagnostics }
using { /UnrealEngine.com/Temporary/SpatialMath }

using { /Fortnite.com/Characters }

ExplosiveWithIndexHandler := class() {
    Device : minigame89_bedwars   
    DeviceID : int  

    HandlerFunction(Agent : agent) : void = {
        Device.BedDestroyed(Agent,DeviceID)
    }
}

#FIXED EXPLOSIVE DEVICE VISIBILITY - THEY HAVE TO BE NEAR SPAWN, SOMETHING WEIRD HAPPENS IF PLAYERS AREN"T NEAR THEM WHEN THE GAME IS RUNNING
minigame89_bedwars := class(minigame_device):
    GetVotingImage<override>()<transacts>: texture = VerseExposed.SQ_BEDWARS_Vote
    DisplayIcon<override>: ?texture = option{VerseExposed.SQ_BEDWARS}

    #For teleporting the bed into the correct position
    @editable
    BedStartingPosition : []creative_prop = array{}
    #Tracks the beds alive for the win condition
    var BedsAlive : int = 0

    #Allows us to setup which team owns the bed
    var TeamOwningBed : []int = array{}
    #The time before the beds will automatically break because the storm has reached them
    @editable
    StormBreakBedTimer : float = 80.0

    @editable
    BedBreakWarningTIme : float = 15.0

    @editable
    HUDWarning : hud_message_device = hud_message_device{}
    @editable
    HUDDestroy : hud_message_device = hud_message_device{}

    @editable
    ItemSpawners : []item_spawner_device = array{}

    #Minigames can't be selected twice in a row. 5 might work? #[PATCH 1.2]
    @editable
    Round1TeamObjectiveBeds : []objective_device = array{}
    @editable
    Round2TeamObjectiveBeds : []objective_device = array{}
    @editable
    Round3TeamObjectiveBeds : []objective_device = array{}
    @editable
    Round4TeamObjectiveBeds : []objective_device = array{}
    @editable
    Round5TeamObjectiveBeds : []objective_device = array{}
    
    #Tracks how many rounds of bed wars have been played #[PATCH 1.2]
    var BedwarsRoundTracker : int = 1

    #Initial setup, subscribe to the beds exploded event and save locations
    OnMinigameDeviceBegin<override>()<suspends>:void=
        #Loop through the beds and setup the subscribe events #[PATCH 1.2]
        for(Index -> Round1TeamBeds : Round1TeamObjectiveBeds):
            Round1TeamBeds.DestroyedEvent.Subscribe(ExplosiveWithIndexHandler{Device:= Self, DeviceID := Index}.HandlerFunction)
        for(Index -> Round2TeamBeds : Round2TeamObjectiveBeds):
            Round2TeamBeds.DestroyedEvent.Subscribe(ExplosiveWithIndexHandler{Device:= Self, DeviceID := Index}.HandlerFunction)
        for(Index -> Round3TeamBeds : Round3TeamObjectiveBeds):
            Round3TeamBeds.DestroyedEvent.Subscribe(ExplosiveWithIndexHandler{Device:= Self, DeviceID := Index}.HandlerFunction)
        for(Index -> Round4TeamBeds : Round4TeamObjectiveBeds):
            Round4TeamBeds.DestroyedEvent.Subscribe(ExplosiveWithIndexHandler{Device:= Self, DeviceID := Index}.HandlerFunction)
        for(Index -> Round5TeamBeds : Round5TeamObjectiveBeds):
            Round5TeamBeds.DestroyedEvent.Subscribe(ExplosiveWithIndexHandler{Device:= Self, DeviceID := Index}.HandlerFunction)

    #Here we are setting up the beds. We teleport them into place as there are instances where there might only be two or three teams.
    #Beds alive gets set dynamically based on the amount of beds required
    _Start<override>()<suspends>:void=
        spawn:
            GetTeamLivesUIManager().ShowBedUI(Self)

        for(ItemSpawner : ItemSpawners):
            Print("Bed Wars Minigame - Item Spawner")
            ItemSpawner.SpawnItem()

        #DEBUG TESTING - 
        Sleep(1.0)
        #I get the teams that are currently in the game
        RemainingTeamsInMinigame := GetRemainingTeams()
        #[PATCH 1.2]
        BedsToTeleport()

        spawn:
            DestroyBedsAfterTime()
    
    #Depending on the round we teleport certain beds in #[PATCH 1.2]
    BedsToTeleport():void=
        case(BedwarsRoundTracker):
            5 => TeleportBedsIn(Round5TeamObjectiveBeds)
            4 => TeleportBedsIn(Round4TeamObjectiveBeds)
            3 => TeleportBedsIn(Round3TeamObjectiveBeds)
            2 => TeleportBedsIn(Round2TeamObjectiveBeds)
            1 => TeleportBedsIn(Round1TeamObjectiveBeds)
            _=> Print("Something has gone wrong, default setting was called")

    #Depending on the round we teleport certain beds in #[PATCH 1.2]
    BedsToDestroy():void=
        case(BedwarsRoundTracker):
            5 => DestroyBedsInRound(Round5TeamObjectiveBeds)
            4 => DestroyBedsInRound(Round4TeamObjectiveBeds)
            3 => DestroyBedsInRound(Round3TeamObjectiveBeds)
            2 => DestroyBedsInRound(Round2TeamObjectiveBeds)
            1 => DestroyBedsInRound(Round1TeamObjectiveBeds)
            _=> Print("Something has gone wrong, default setting was called")

    #Destroy the beds in this specific round #[PATCH 1.2]
    DestroyBedsInRound(RoundBeds : []objective_device):void=
        AllPlayers := GetPlayspace().GetPlayers()
        for(BedToDestroy : RoundBeds):
            if(AgentRef := AllPlayers[0]):
                BedToDestroy.Destroy(AgentRef)

    #Teleport the bed into the map
    TeleportBedsIn(RoundBeds : []objective_device):void=
        #I get the teams that are currently in the game
        RemainingTeamsInMinigame := GetRemainingTeams()
        for(Team : RemainingTeamsInMinigame):
            set BedsAlive += 1
            #Add the specific team index of the team owning this bed
            set TeamOwningBed += array{Team.TeamIndex}
            #Teleport bed into place [I think the TeamBeds using Index was not teleporting them to the correct team location]
            if(RoundBeds[Team.TeamIndex].TeleportTo[BedStartingPosition[Team.TeamIndex].GetTransform().Translation, BedStartingPosition[Team.TeamIndex].GetTransform().Rotation]):

    #When a bed is destoyed we remove a bed from beds alive and then stop players from respawning
    BedDestroyed(InAgent : agent, BedID : int):void=
        #I'm getting the player who destroyed the bed. I need to get the bed destroyed if possible
        RemainingTeamsInMinigame := GetRemainingTeams()
        #Now I go through all the teams and then check if the Bed ID is = to that TeamIndex. If it is we can disable that teams spawning
        for(Team : RemainingTeamsInMinigame):
            if(Team.TeamIndex = BedID):
                set BedsAlive -= 1
                set Team.AllowSpawning = false
                Print("Bed Destroyed Team {Team.TeamIndex}")
                #Teleport back after being destroyed [Moves the barriers as well! So no collision will occur anymore]
                #if(TeamBeds[BedID].TeleportTo[OriginalBedLocation[BedID], BedStartingPosition[BedID].GetTransform().Rotation]):
        Print("Team {BedID} Bed Destroyed")

    #Reset the beds and locations
    _Reset<override>()<suspends>:void=
        set BedsAlive = 0
        #[PATCH 1.2]
        BedsToDestroy()

        #Need to make the array empty
        set TeamOwningBed = array{}

        #Increment round tracker
        set BedwarsRoundTracker += 1

        GetTeamLivesUIManager().ResetBedUI()

    #Bed Wars Win Condition. All other players dead and 1 bed remaining
    GetWinners<override>()<transacts><decides>:[]player=
        var RetArray : []player = array{}
        RemainingTeamsInMinigame := GetRemainingTeams()
        if(ScoreMinigame := ScoreLogic?):
            for(Team : RemainingTeamsInMinigame):
                if(Team.Kills = ScoreMinigame.ScoreToWin):
                    Team.Players + Team.EliminatedPlayers
        else if(BedsAlive <= 1):
            if(RemainingTeamsInMinigame.Length = 1, WinningTeam := RemainingTeamsInMinigame[0]):
                set RetArray = WinningTeam.Players + WinningTeam.EliminatedPlayers
        RetArray

    DestroyBedsAfterTime()<suspends>:void=
        race:
            EndedEvent.Await()

            block:
                Sleep(StormBreakBedTimer)

                HUDWarning.Show(BuildMessage("Storm in 15 Seconds!"))
                Sleep(BedBreakWarningTIme)

                Players := GetPlayspace().GetPlayers()
                RemainingTeamsInMinigame := GetRemainingTeams()

                for(Team : RemainingTeamsInMinigame):
                    if(Team.AllowSpawning = true):
                        # Add 50 shields and 200 health
                        for(Player : Team.Players):
                            if(FortChar := Player.GetFortCharacter[]):
                                FortChar.SetMaxShield(150.0)
                                FortChar.SetShield(FortChar.GetShield() + 50.0)

                                FortChar.SetMaxHealth(300.0)
                                FortChar.SetHealth(FortChar.GetHealth() + 200.0)

                HUDDestroy.Show()
                #[PATCH 1.2]
                BedsToDestroy()
