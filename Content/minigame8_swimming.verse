# using { /Fortnite.com/Devices }
# using { /Fortnite.com/AI}
# using { /Fortnite.com/Teams}
# using { /Verse.org/Simulation }
# using { /Verse.org/Random }
# using { /Fortnite.com/Characters }
# using { /Fortnite.com/FortPlayerUtilities }
# using { /Fortnite.com/Game }
# using { /Fortnite.com/Teams }
# using { /Verse.org/Simulation/Tags }
# using { /UnrealEngine.com/Temporary/Diagnostics }
# using { /UnrealEngine.com/Temporary/SpatialMath } 
# using { /Verse.org/Assets }

# fishing_hole := class(tag) {}
# swimming := class(minigame_device):

#     @editable WaterDevice : water_device = water_device{}
#     @editable Boat : vehicle_spawner_boat_device = vehicle_spawner_boat_device{}

#     Name<override>: string = "Flooded"
#     StartingMessage<override>:string = ""
#     GetVotingImage<override>()<transacts>: texture = VerseExposed.SQ_FLOODED

#     _Start<override>()<suspends>: void = 
#         WaterDevice.Enable()
#         WaterDevice.BeginVerticalFilling()

#         sync:
#             block:
#                 Print("Handle Boat bro")
#                 Sleep(25.0)
#                 Boat.Enable()
#                 Boat.RespawnVehicle()

#             block:
#                 Print("HandleRise")
#                 Sleep(35.0)
#                 WaterDevice.ResumeVerticalMovement()

#     _Reset<override>()<suspends>: void =
#         Boat.DestroyVehicle()
#         Boat.Disable()

#         WaterDevice.Reset()
#         WaterDevice.Disable()
