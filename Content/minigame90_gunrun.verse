
using { /Fortnite.com/Devices }
using { /Verse.org/Simulation }
using { /Verse.org/Assets }
using { /UnrealEngine.com/Temporary/Diagnostics }
using { /Fortnite.com/Game }
using { /UnrealEngine.com/Temporary/SpatialMath }

minigame90_gunrun := class(minigame_device):
    GetVotingImage<override>()<transacts>: texture = VerseExposed.SQ_GUNRUN_Vote
    DisplayIcon<override>: ?texture = option{VerseExposed.SQ_GUNRUN}

    #Has all of the weapons you can get
    @editable
    GunGameGuns : item_granter_device = item_granter_device{}

    @editable
    ZeroBuildBox : creative_prop = creative_prop{}

    @editable
    TeleportLocation : vector3 = vector3{X:= 544.0, Y:= 567.0, Z:= 415.0}

    OriginalLocation : vector3 = vector3{X:= 544.0, Y:= 567.0, Z:= -1000.0}

    @editable
    Team1NextGunHUD : hud_message_device = hud_message_device{}
    @editable
    Team2NextGunHUD : hud_message_device = hud_message_device{}
    @editable
    Team3NextGunHUD : hud_message_device = hud_message_device{}
    @editable
    Team4NextGunHUD : hud_message_device = hud_message_device{}

    @editable
    NextGunTextArray : []string = array{}

    _Start<override>()<suspends>:void=
        if(ZeroBuildBox.TeleportTo[TeleportLocation, ZeroBuildBox.GetTransform().Rotation]):

        for(Player : GetPlayspace().GetPlayers()):
            if(Team := GetTeam[Player]):
                if(Message := NextGunTextArray[0]):
                    case(Team.TeamIndex):
                        0 => Team1NextGunHUD.Show(Player, BuildMessage(Message))
                        1 => Team2NextGunHUD.Show(Player, BuildMessage(Message))
                        2 => Team3NextGunHUD.Show(Player, BuildMessage(Message))
                        3 => Team4NextGunHUD.Show(Player, BuildMessage(Message))
                        _=>
    
    _Reset<override>()<suspends>:void=
        if(ZeroBuildBox.TeleportTo[OriginalLocation, ZeroBuildBox.GetTransform().Rotation]):

        for(Player : GetPlayspace().GetPlayers()):
            Team1NextGunHUD.Hide(Player)
            Team2NextGunHUD.Hide(Player)
            Team3NextGunHUD.Hide(Player)
            Team4NextGunHUD.Hide(Player)

    #When a player is eliminated we get the player, team and then grant an item at the Team.Kill index
    _OnPlayerEliminated<override>(Result : elimination_result):void=
        Print("Player Eliminated")
        if(FortChar := Result.EliminatingCharacter?, Player := player[FortChar.GetAgent[]], Team := GetTeam[Player]):
            #Grants a weapon at the current kill index
            Print("Grant new item to player")
            GunGameGuns.GrantItemIndex(Player, Team.Kills)

            for(CurrentPlayer : Team.Players):
                if(Message := NextGunTextArray[Team.Kills]):
                    case(Team.TeamIndex):
                        0 => Team1NextGunHUD.Show(Player, BuildMessage(Message))
                        1 => Team2NextGunHUD.Show(Player, BuildMessage(Message))
                        2 => Team3NextGunHUD.Show(Player, BuildMessage(Message))
                        3 => Team4NextGunHUD.Show(Player, BuildMessage(Message))
                        _=>