using { /Fortnite.com/Devices }
using { /Fortnite.com/AI}
using { /Fortnite.com/Teams}
using { /Verse.org/Simulation }
using { /Verse.org/Random }
using { /Verse.org/Assets }
using { /Fortnite.com/Characters }
using { /Fortnite.com/FortPlayerUtilities }
using { /Fortnite.com/Game }
using { /Fortnite.com/Teams }
using { /Verse.org/Simulation/Tags }
using { /UnrealEngine.com/Temporary/Diagnostics }
using { /UnrealEngine.com/Temporary/SpatialMath } 
using { /Verse.org/Assets }

minigame91_pillars := class(minigame_device):
    Name<override>: string = "Pillars"
    StartingMessage<override>:string = "Pickaxe Has Knockback!"
    GetVotingImage<override>()<transacts>: texture = VerseExposed.SQ_PILLARS

    @editable KnockBackDevicePillar : KnockbackDevice = KnockbackDevice{}

    KnockBackLocation : vector3 = vector3{X:= -1498.470165, Y:= 2347.905954, Z:= 0.0}

    @editable
    ItemSpawners : []item_spawner_device = array{}

    _Start<override>()<suspends>: void =
        for(ItemSpawner : ItemSpawners):
            Print("Pillars Minigame - Item Spawner")
            ItemSpawner.SpawnItem()

        for(Player : GetRemainingPlayers(), Fort := Player.GetFortCharacter[]):
            RegisterSubscriber(Fort.DamagedEvent().Subscribe(InitializePlayerKnockBack))
            RegisterSubscriber(Fort.DamagedShieldEvent().Subscribe(InitializePlayerKnockBack))

    _Reset<override>()<suspends>: void = return

    InitializePlayerKnockBack(DamageInflicted : damage_result) : void =
        spawn {KnockPlayerBack(DamageInflicted)}

    KnockPlayerBack(DamageInflicted : damage_result)<suspends> : void =
        Print("Knock Back Player")
        if (CharacterDamaged := fort_character[DamageInflicted.Target]):
            if (CharacterShoots := fort_character[DamageInflicted.Instigator?]):
                if (AgentFromChar := CharacterDamaged.GetAgent[]):
                    if (DamagedAgent := CharacterDamaged.GetAgent[]):
                        if (DamagedPlayer := player[DamagedAgent]):
                            Print("Got Damaged Agent")
                            #Pickaxe Only
                            if (DamageInflicted.Amount = 20.0): 
                                Print("Knocking back now") 
                                spawn {SetKnockBackBasedOnDamage( DamagedPlayer, CharacterShoots)}

    SetKnockBackBasedOnDamage(DamagedPlayer : player, DamageInstigatorchar : fort_character)<suspends> : void = 
        TeleProp := KnockBackDevicePillar.TeleporterProp
        KnockbackMovementMod := KnockBackDevicePillar.KnockbackMovementModulator
        TeleProp.MoveTo(KnockBackLocation, DamageInstigatorchar.GetViewRotation(), 0.1)
        KnockbackMovementMod.Activate(DamagedPlayer)