using { /Fortnite.com/Devices }
using { /Verse.org/Simulation }
using { /Verse.org/Assets }
using { /Fortnite.com/Game }
using { /Fortnite.com/Characters }
using { /UnrealEngine.com/Temporary/SpatialMath }
using { /UnrealEngine.com/Temporary/Diagnostics }

minigame92_oddball := class(minigame_device):
    GetVotingImage<override>()<transacts>: texture = VerseExposed.SQ_GUARDBALL_Vote
    DisplayIcon<override>: ?texture = option{VerseExposed.SQ_GUARDBALL}

    #The object to hold can be a collectible that respawns in the middle once dropped
    @editable
    HoldingCollectible : collectible_object_device = collectible_object_device{}

    @editable
    HoldingVisualPowerup : visual_effect_powerup_device = visual_effect_powerup_device{}    

    @editable
    HoldingDamagePowerup : damage_amplifier_powerup_device = damage_amplifier_powerup_device{}

    @editable
    HoldingObjectTeleportSpot : creative_prop = creative_prop{}

    @editable
    HolderItemRemover : item_remover_device = item_remover_device{}

    @editable
    TimePerScore : float = 1.0

    @editable
    RedPickedUpBallHUD : hud_message_device = hud_message_device{}
    @editable
    BluePickedUpBallHUD : hud_message_device = hud_message_device{}
    @editable
    GreenPickedUpBallHUD : hud_message_device = hud_message_device{}
    @editable
    YellowPickedUpBallHUD : hud_message_device = hud_message_device{}

    @editable
    DroppedBallHUD : hud_message_device = hud_message_device{}

    @editable
    NoBuildClass : class_and_team_selector_device = class_and_team_selector_device{}

    @editable
    NoClass : class_and_team_selector_device = class_and_team_selector_device{}

    var StartingCollectiblePosition : vector3 = vector3{}
    var PlayerHoldingMap : [player]logic = map{}

    OddBallDropped<public>:event() = event(){}

    var PickupCollectedCancelable : []cancelable = array{}

    _Start<override>()<suspends>: void =
        #Show and Respawn The Collectible
        HoldingCollectible.Show()
        HoldingCollectible.RespawnForAll()

        #Subscribe for players to pickup the ball
        set PickupCollectedCancelable = array{HoldingCollectible.CollectedEvent.Subscribe(PickedUpTheBall)}

        set PlayerHoldingMap = map{}

        #Set position for starting location of holding collectible
        set StartingCollectiblePosition = HoldingCollectible.GetTransform().Translation

        #Teleport Collectible into starting area
        if(HoldingCollectible.TeleportTo[HoldingObjectTeleportSpot.GetTransform().Translation, HoldingObjectTeleportSpot.GetTransform().Rotation]):

    _Reset<override>()<suspends>:void=
        HoldingCollectible.Hide()
        OddBallDropped.Signal()

        for(Cancelable : PickupCollectedCancelable):
            Cancelable.Cancel()
        set PickupCollectedCancelable = array{}

        #Teleport Holding Collectable back to original spot
        if(HoldingCollectible.TeleportTo[StartingCollectiblePosition, HoldingCollectible.GetTransform().Rotation]):

    _OnPlayerEliminated<override>(Result: elimination_result)<suspends>:void=
        if(Agent := Result.EliminatedCharacter.GetAgent[]):
            Print("OddBall - Found Agent")
            if(Player := player[Agent]):
                Print("OddBall - Found Player")
                if(PlayerHoldingMap[Player] = true):
                    Print("OddBall - Player Holding Ball")
                    BallSpawnLocation := Result.EliminatedCharacter.GetTransform().Translation
                    Sleep(0.5)
                    DroppedBall(Result.EliminatedCharacter, BallSpawnLocation)

    #Picked up ball event
    PickedUpTheBall(InAgent : agent):void=
        #set PlayerHoldingMap = map{}
        if(Player := player[InAgent]):
            if(set PlayerHoldingMap[Player] = true):
            HoldingVisualPowerup.Pickup(Player)
            HoldingDamagePowerup.Pickup(Player)

            HolderItemRemover.Remove(Player)
            NoBuildClass.ChangeClass(Player)

            if(TeamPlayerIsOn := PlayersTeamMap[Player]):
                if(TeamPlayerIsOn.TeamIndex = 0):
                    RedPickedUpBallHUD.Show()
                if(TeamPlayerIsOn.TeamIndex = 1):
                    BluePickedUpBallHUD.Show()
                if(TeamPlayerIsOn.TeamIndex = 2):
                    GreenPickedUpBallHUD.Show()
                if(TeamPlayerIsOn.TeamIndex = 3):
                    YellowPickedUpBallHUD.Show()

        spawn:
            HoldingBall(InAgent)

    #Grant team points for holding the ball
    HoldingBall(InAgent : agent)<suspends>:void=
        if(HoldingPlayer := player[InAgent]):
            race:
                #Break the loop once the oddball is dropped
                OddBallDropped.Await()
                block:
                    loop:
                        IncreaseTeamScore(HoldingPlayer)
                        CustomScoreEvent.Signal()
                        Sleep(TimePerScore)

    #Stop granting points for that team, player will respawn [Only way to drop is to die so no need for reset]
    DroppedBall(InFortChar : fort_character, BallSpawnLocation : vector3):void=
        #Stop the ball score loop and respawn the collectible
        #Teleport on players spot
        if(HoldingCollectible.TeleportTo[BallSpawnLocation, InFortChar.GetTransform().Rotation]):

        OddBallDropped.Signal()
        HoldingCollectible.RespawnForAll()
        DroppedBallHUD.Show()

        if(Agent := InFortChar.GetAgent[]):
            if(Player := player[Agent]):
                NoClass.ChangeClass(Player)
                GivePlayerSelectableLoadout(Player)
                if(set PlayerHoldingMap[Player] = false):

    