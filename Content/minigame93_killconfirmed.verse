
using { /Fortnite.com/Devices }
using { /Verse.org/Simulation }
using { /Verse.org/Assets }
using { /Fortnite.com/Game }
using { /Fortnite.com/Characters }
using { /UnrealEngine.com/Temporary/SpatialMath }
using { /UnrealEngine.com/Temporary/Diagnostics }

minigame93_killconfirmed := class(minigame_device):
    Name<override>: string = "Dog Tags"
    StartingMessage<override>:string = "Pick up Dog tags for points"
    GetVotingImage<override>()<transacts>: texture = VerseExposed.SQ_DOGTAGS_Vote
    DisplayIcon<override>: ?texture = option{VerseExposed.SQ_DOGTAGS}

    @editable
    RedTeamTags : []collectible_object_device = array{}
    @editable
    BlueTeamTags : []collectible_object_device = array{}
    @editable
    GreenTeamTags : []collectible_object_device = array{}
    @editable
    YellowTeamTags : []collectible_object_device = array{}

    var RedCurrentTagIndex : int = -1
    var BlueCurrentTagIndex : int = -1
    var GreenCurrentTagIndex : int = -1
    var YellowCurrentTagIndex : int = -1

    _Start<override>()<suspends>: void =
        for(Tag : RedTeamTags):
            RegisterSubscriber(Tag.CollectedEvent.Subscribe(RedCollected))
        for(Tag : BlueTeamTags):
            RegisterSubscriber(Tag.CollectedEvent.Subscribe(BlueCollected))
        for(Tag : GreenTeamTags):
            RegisterSubscriber(Tag.CollectedEvent.Subscribe(GreenCollected))
        for(Tag : YellowTeamTags):
            RegisterSubscriber(Tag.CollectedEvent.Subscribe(YellowCollected))

    RedCollected(Agent : agent):void=
        if(Player := player[Agent]):
            if(Team := GetTeam[Player]):
                if(Team.TeamIndex = 0):
                    # Collecting same team tag
                    IncreaseTeamScore(Player, 1)
                else:
                    # Collecting Other team tag
                    IncreaseTeamScore(Player, 3)
                

    BlueCollected(Agent : agent):void=
        if(Player := player[Agent]):
            if(Team := GetTeam[Player]):
                if(Team.TeamIndex = 1):
                    # Collecting same team tag
                    IncreaseTeamScore(Player, 1)
                else:
                    # Collecting Other team tag
                    IncreaseTeamScore(Player, 3)

    GreenCollected(Agent : agent):void=
        if(Player := player[Agent]):
            if(Team := GetTeam[Player]):
                if(Team.TeamIndex = 2):
                    # Collecting same team tag
                    IncreaseTeamScore(Player, 1)
                else:
                    # Collecting Other team tag
                    IncreaseTeamScore(Player, 3)

    YellowCollected(Agent : agent):void=
        if(Player := player[Agent]):
            if(Team := GetTeam[Player]):
                if(Team.TeamIndex = 3):
                    # Collecting same team tag
                    IncreaseTeamScore(Player, 1)
                else:
                    # Collecting Other team tag
                    IncreaseTeamScore(Player, 3)

    _Reset<override>()<suspends>:void=
        for(Tag : RedTeamTags):
            Tag.Hide()
        for(Tag : BlueTeamTags):
            Tag.Hide()
        for(Tag : GreenTeamTags):
            Tag.Hide()
        for(Tag : YellowTeamTags):
            Tag.Hide()

    _OnPlayerEliminated<override>(Result: elimination_result)<suspends>:void=
        
        if(Player := player[Result.EliminatedCharacter.GetAgent[]]):
            if(Team := GetTeam[Player]):

                if(EnemyPlayer := player[Result.EliminatingCharacter?.GetAgent[]]):
                    if(EnemyTeam := GetTeam[EnemyPlayer]):
                        # Reward player who got the kill with 1 point
                        if(not Team.TeamIndex = EnemyTeam.TeamIndex):
                            IncreaseTeamScore(EnemyPlayer, 1)
                # Red
                if(Team.TeamIndex = 0):
                    set RedCurrentTagIndex += 1
                    if(RedCurrentTagIndex > RedTeamTags.Length - 1):
                        set RedCurrentTagIndex = 0
                    if(Tag := RedTeamTags[RedCurrentTagIndex]):
                        DroppedTag(Result.EliminatedCharacter, Tag)
                    
                # Blue
                if(Team.TeamIndex = 1):
                    set BlueCurrentTagIndex += 1
                    if(BlueCurrentTagIndex > BlueTeamTags.Length - 1):
                        set BlueCurrentTagIndex = 0
                    if(Tag := BlueTeamTags[BlueCurrentTagIndex]):
                        DroppedTag(Result.EliminatedCharacter, Tag)

                # Green
                if(Team.TeamIndex = 2):
                    set GreenCurrentTagIndex += 1
                    if(GreenCurrentTagIndex > GreenTeamTags.Length - 1):
                        set GreenCurrentTagIndex = 0
                    if(Tag := GreenTeamTags[GreenCurrentTagIndex]):
                        DroppedTag(Result.EliminatedCharacter, Tag)

                # Yellow
                if(Team.TeamIndex = 3):
                    set YellowCurrentTagIndex += 1
                    if(YellowCurrentTagIndex > YellowTeamTags.Length - 1):
                        set YellowCurrentTagIndex = 0
                    if(Tag := YellowTeamTags[YellowCurrentTagIndex]):
                        DroppedTag(Result.EliminatedCharacter, Tag)

                

    #Picked up ball event
    PickedUpTheBall(InAgent : agent):void=
        block:

    #Stop granting points for that team, player will respawn [Only way to drop is to die so no need for reset]
    DroppedTag(InFortChar : fort_character, Tag : collectible_object_device):void=
        #Teleport on players spot
        if(Tag.TeleportTo[InFortChar.GetTransform().Translation, InFortChar.GetTransform().Rotation]):
        Tag.RespawnForAll()
        Tag.Show()