using { /Fortnite.com/Devices }
using { /Verse.org/Simulation }
using { /UnrealEngine.com/Temporary/Diagnostics }
using { /Fortnite.com/Game }
using { /Verse.org/Assets }
using { /Fortnite.com/Characters }

<#RELOAD which each kill will revive one person on your team instantly#>
minigame_94RELOAD := class(minigame_device):

    Name<override>: string = "Reload"
    StartingMessage<override>:string = "Eliminate Players to revive your Squad"
    GetVotingImage<override>()<transacts>: texture = VerseExposed.SQ_RELOAD_Vote
    DisplayIcon<override>: ?texture = option{VerseExposed.SQ_RELOAD}

    #Set Respawn Logic Respawns Enabled to be false. We are using a custom respawn method
    _Start<override>()<suspends>: void=
        if(ReloadRespawnLogic := RespawnLogic?):
            #set ReloadRespawnLogic.RespawnsEnabled = false

    #On Elimination respawn a random Team Member
    _OnPrePlayerEliminated<override>(Result: elimination_result)<suspends>:void=
        if(EliminatingPlayer := player[Result.EliminatingCharacter?.GetAgent[]]):
            Print("Player Eliminated")
            RespawnRandomTeamMember(EliminatingPlayer)

    #Custom Respawn method for Reload. Respawn team member on elimination of enemy player.
    RespawnRandomTeamMember(Player : player):void=
        Print("Respawning Random Team Member")
        #We get the player who is eliminating the other player. Then get a reference to their team. GetTeam is from minigame_device
        if(Team := GetTeam[Player]):
            Print("Successfully accessed Team")
            #We begin a loop through the players in that team
            for(Index -> TeamPlayer : Team.Players):
                #We access a FortCharacter and check if they are active. If they aren't active we respawn them
                if(FortChar := TeamPlayer.GetFortCharacter[]):
                    if(FortChar.IsActive[]):
                        Print("Character is alive")
                    else:
                        Print("Character is not alive, respawn them and break function")
                        spawn{RespawnAndTeleportPlayer(TeamPlayer,ChosenArena?.ArenaDevice?.CustomTeleportationMethod or CustomTeleportationMethod, 1.0, RespawnLogic)}
                        return

    GetWinners<override>()<transacts><decides>:[]player=
        var RetArray : []player = array{}
        RemainingTeamsInMinigame := RemainingTeams
        var TeamsWithFortChar : int = 4
        var WinningTeam : custom_team = custom_team{}

        #Set the teams with fort character to be the remaining teams in minigame length
        set TeamsWithFortChar = RemainingTeamsInMinigame.Length
        Print("Teams In Game Length = {TeamsWithFortChar}")
        for(Index -> Team : RemainingTeams):
            #Within the teams we are checking if any players are alive
            if(Team.IsAFortCharacterAliveOnTeam() = false):
                Print("No Character on this team is alive")
                set TeamsWithFortChar -= 1
            else:
                set WinningTeam = Team
            if(TeamsWithFortChar = 1):
                Print("Team has won {WinningTeam.TeamIndex}")
                set RetArray = WinningTeam.Players + WinningTeam.EliminatedPlayers
        RetArray