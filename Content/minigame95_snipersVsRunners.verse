using { /Fortnite.com/Devices }
using { /Fortnite.com/AI}
using { /Fortnite.com/Teams}
using { /Verse.org/Simulation }
using { /Verse.org/Random }
using { /Fortnite.com/Characters }
using { /Fortnite.com/FortPlayerUtilities }
using { /Fortnite.com/Game }
using { /Fortnite.com/Teams }
using { /Verse.org/Simulation/Tags }
using { /UnrealEngine.com/Temporary/Diagnostics }
using { /UnrealEngine.com/Temporary/SpatialMath } 
using { /Verse.org/Assets }

SnipersVsRunners := class(minigame_device):
    @editable 
    GoldWinValue : int = 450

    @editable 
    Collectible : collectible_object_device = collectible_object_device{}

    @editable
    ItemRemover : item_remover_device = item_remover_device{}
    @editable
    ItemSpawners : []item_spawner_device = array{}

    @editable
    IntroSequence : cinematic_sequence_device = cinematic_sequence_device{}
    @editable
    Camera : gameplay_camera_fixed_point_device = gameplay_camera_fixed_point_device{}

    @editable
    AFKTime : float = 10.0
    @editable
    AFKElimZone : creative_prop = creative_prop{}
    @editable
    AFKIn : creative_prop = creative_prop{}
    @editable
    AFKOut : creative_prop = creative_prop{}

    Name<override>: string = "Snipers Vs Runners"
    StartingMessage<override>:string = ""
    GetVotingImage<override>()<transacts>: texture = VerseExposed.SQ_SNIPERSRUNNERS

    _Start<override>()<suspends>: void =
        Collectible.Show()
        Collectible.RespawnForAll()
        RegisterSubscriber(Collectible.CollectedEvent.Subscribe(HandleCollectedEvent))
        for(ItemSpawner : ItemSpawners):
            Print("Snipers Vs Runners - Item Spawner")
            ItemSpawner.SpawnItem()

        IntroSequence.Play()
        Camera.AddToAll()
        Sleep(3.5)
        Camera.RemoveFromAll()

    OnSafeTimeEnded<override>()<suspends>:void=
        Manager.UiMessage.Show()

        Sleep(0.5)

        for(Player : RemainingTeams[1].Players):
            ItemRemover.Remove(Player)

        Sleep(AFKTime)
        if(AFKElimZone.TeleportTo[AFKIn.GetTransform()]):

    _Reset<override>()<suspends>:void=
        Collectible.Hide()
        if(AFKElimZone.TeleportTo[AFKOut.GetTransform()]):

    _OnGenericEndTimerSuccessReturnArray<override>():[]player=
        TeamCollection := GetPlayspace().GetTeamCollection()
        if(WinnerTeam := RemainingTeams[0]):
            return WinnerTeam.Players
        return array{}

    #TEAM MINIGAME MODE - Modified the win condition for Snipers Vs Runners
    HandleCollectedEvent(Agent : agent):void=
        if:
            WinningPlayer := player[Agent]
        then:
            #var Winner : []player = array{WinningPlayer}
            var WinningPlayers : []player = array{}
            #Change from winner to the team of winners
            TeamCollection := GetPlayspace().GetTeamCollection()
            if(Team := TeamCollection.GetTeam[Agent]):
                #Get the winning team
                if(WinnerTeamMembers := TeamCollection.GetAgents[Team]):
                    #Loop through and fill the array with winning players
                    for(WinningTeamMember : WinnerTeamMembers):
                        #Convert to a player for the array
                        if(Player := player[WinningTeamMember]):
                            Manager.GoldTracker.Add(Player,GoldWinValue)
                            set WinningPlayers += array{Player}
                    #Handle Winning Players
                    Print("We have a winning team")
                    EndedEvent.Signal(WinningPlayers)