using { /Fortnite.com/Devices }
using { /Verse.org/Simulation }
using { /Verse.org/Random }
using { /Fortnite.com/Characters }
using { /UnrealEngine.com/Temporary/SpatialMath }
using { /Fortnite.com/Devices/CreativeAnimation }
using { /UnrealEngine.com/Temporary/Diagnostics }
using { /Fortnite.com/Devices/CreativeAnimation/InterpolationTypes }
using { /Verse.org/Assets }

############################################################################################################################
###################################################### Custom Classes ######################################################
############################################################################################################################
CustomRoomConfigClass := class<concrete><unique>():
    @editable
    PlayerCount : int = 0

    @editable
    PossibleAvailableRooms : []int = array{}
    @editable
    PossiblePlayersPerRoom : []int = array{}

CustomRoomClass := class<concrete><unique>():
    @editable
    RoomIndex : int = 0
    @editable
    DoorLockingSwitch : switch_device = switch_device{}
    @editable
    PlayerCounter : player_counter_device = player_counter_device{}

    var MinigameDevice : minigame96_Mingle = minigame96_Mingle{}
    var PlayersInRoom : []player = array{}

    # Sets Room Stats on Game Start
    Setup(InMinigameDevice : minigame96_Mingle) : void=
        set MinigameDevice = InMinigameDevice
        set PlayersInRoom = array{}

        MinigameDevice.RegisterSubscriber(PlayerCounter.CountedEvent.Subscribe(PlayerEnteredRoom))
        MinigameDevice.RegisterSubscriber(PlayerCounter.RemovedEvent.Subscribe(PlayerLeftRoom))

        Reset()

    # Resets settings adjusted during gameplay
    Reset() : void=
        set PlayersInRoom = array{}
        SetRoomNotInUse()

    # Called when the room will be in use
    SetRoomInUse(InPlayerPerRoom : int) : void=
        set PlayersInRoom = array{}
        PlayerCounter.SetTargetCount(InPlayerPerRoom)

        for(Player : MinigameDevice.GetPlayspace().GetPlayers()):
            DoorLockingSwitch.TurnOff(Player)

    # Called when the room will not be in use
    SetRoomNotInUse() : void=
        set PlayersInRoom = array{}
        PlayerCounter.SetTargetCount(0)

        for(Player : MinigameDevice.GetPlayspace().GetPlayers()):
            DoorLockingSwitch.TurnOn(Player)

    # Adds player to list once in a room
    PlayerEnteredRoom(InAgent : agent) : void=
        if(Player := player[InAgent]): 
            if(PlayersInRoom.Length = 0):
                set PlayersInRoom = array{Player}
            else:
                set PlayersInRoom += array{Player}

    # Removes player to list once in a room
    PlayerLeftRoom(InAgent : agent) : void=
        if(Player := player[InAgent]):
            if(FindPlayerIndex := PlayersInRoom.Find[Player]):
                if(TempPlayers := PlayersInRoom.RemoveElement[FindPlayerIndex]):
                    set PlayersInRoom = TempPlayers

############################################################################################################################
###################################################### Mingle Minigame #####################################################
############################################################################################################################

minigame96_Mingle := class(minigame_device):
    @editable
    PlaformSpinSpeed : float = 1.2

    @editable
    MinSpinTime : float = 5.0
    @editable
    MaxSpinTime : float = 10.0
    @editable
    RunTime : float = 20.0
    @editable
    DelayAfterElimination : float = 3.0

    @editable
    CenterPlatformProp : creative_prop = creative_prop{}
    @editable
    PlaformBarrierProp : creative_prop = creative_prop{}

    @editable
    RunTimer : timer_device = timer_device{}

    @editable
    MinglePlayerCountHUD : hud_message_device = hud_message_device{}
    @editable
    MingleSurvivedHUD : hud_message_device = hud_message_device{}

    @editable
    EliminationAudio : audio_player_device = audio_player_device{}
    @editable
    SpinningMusic : audio_player_device = audio_player_device{}

    @editable
    MingleIntro : cinematic_sequence_device = cinematic_sequence_device{}

    @editable
    ItemSpawners : []item_spawner_device = array{}

    @editable
    AllRooms : []CustomRoomClass = array{}
    @editable
    RoomConfigs : []CustomRoomConfigClass = array{}
    @editable
    TeleportSpots : []creative_prop = array{}
    @editable
    DoorLights : []customizable_light_device = array{}

    var CurrentConfig : CustomRoomConfigClass = CustomRoomConfigClass{}
    var PlayersToEliminate : []player = array{}

    var CurrentPlayersPerRoom : int = 0
    var CurrentAvailableRooms : int = 0

    GameLoopEvent : event() = event(){}

    GetVotingImage<override>()<transacts>: texture = VerseExposed.SQ_MingleIcon
    #GetVotingImage<override>()<transacts>: texture = VerseExposed.SQ_FINISHERS

    _Start<override>()<suspends>: void=
        set PlayersToEliminate = array{}
        set CurrentPlayersPerRoom = 0
        set CurrentAvailableRooms = 0

        if(DefaultConfig := RoomConfigs[0]):
            set CurrentConfig = DefaultConfig

        MingleIntro.Play()
        TeleportBarrier(true)

        # Sets up Rooms
        for(Room : AllRooms):
            Room.Setup(Self)

        for(ItemSpawner: ItemSpawners):
            ItemSpawner.Enable()
            ItemSpawner.SpawnItem()

    OnSafeTimeEnded<override>()<suspends>:void=
        spawn{GameLoop()}

    _Reset<override>()<suspends>:void=
        RunTimer.ResetForAll()
        TeleportBarrier(true)
        MinglePlayerCountHUD.Hide()

        for(Room : AllRooms):
            Room.Reset()

    # The loop that defines when each stage of the game occurs
    GameLoop()<suspends> : void=
        race:
            EndedEvent.Await()
            block:
                loop:
                    ConfigSetPhase()
                    spawn{SpinningPhase()}
                    Print("Loop Await")
                    GameLoopEvent.Await()
                    Sleep(0.0)

    # This is the phase in which the room config is refreshed based on the player cuont
    ConfigSetPhase() : void=
        Print("Mingle Minigame - Config Phase")
        CurrentPlayerCount := GetRemainingPlayers().Length
        for(Config : RoomConfigs):
            if(CurrentPlayerCount >= Config.PlayerCount):
                set CurrentConfig = Config
                return

    # This is the phase in which players are teleported to the center and revolve around
    SpinningPhase()<suspends> : void=
        Print("Mingle Minigame - Spinning Phase")

        ToggleLights(true)
        SpinningMusic.Play()

        # Teleport all players to center
        var CurrentIndex : int = 0
        for(Player : GetRemainingPlayers()):
            if(FortChar := Player.GetFortCharacter[]):
                if(Spot := TeleportSpots[CurrentIndex].GetTransform()):
                    if(FortChar.TeleportTo[Spot.Translation, Spot.Rotation]):
                        set CurrentIndex += 1

        #Resets Rooms
        for(CurrentRoom : AllRooms):
            CurrentRoom.SetRoomNotInUse()

        TeleportBarrier(true)

        # Start Spinning Platform
        RandSpinTime := GetRandomFloat(MinSpinTime, MaxSpinTime)
        spawn{PlatformRotateAnimation(RandSpinTime, animation_mode.OneShot)}
        Sleep(RandSpinTime)

        ToggleLights(false)
        SpinningMusic.Stop()

        RunningPhase()

    # This is the phase in which players have to run to a room
    RunningPhase()<suspends> : void=
        Print("Mingle Minigame - Running Phase")

        # Sets up which rooms are in use
        var CurrentRoomIndex : int = 0
        RandRoomOrder := Shuffle(AllRooms)

        RandUseCountNum := GetRandomInt(0, CurrentConfig.PossibleAvailableRooms.Length - 1)
        RandChosenPlayerPerRoomNum := GetRandomInt(0, CurrentConfig.PossiblePlayersPerRoom.Length - 1)

        if(ChosenRoomUse := CurrentConfig.PossibleAvailableRooms[RandUseCountNum]):
            if(ChosenPlayerPerRoom := CurrentConfig.PossiblePlayersPerRoom[RandChosenPlayerPerRoomNum]):

                set CurrentPlayersPerRoom = ChosenPlayerPerRoom
                set CurrentAvailableRooms = ChosenRoomUse

                #Sets HUD Device
                if(ChosenPlayerPerRoom = 1):
                    MinglePlayerCountHUD.Show(BuildMessage("{CurrentPlayersPerRoom} Player Per Room"), ?DisplayTime:=RunTime)
                else:
                    MinglePlayerCountHUD.Show(BuildMessage("{CurrentPlayersPerRoom} Players Per Room"), ?DisplayTime:=RunTime)

                #Sets up Rooms
                for(CurrentRoom : RandRoomOrder):
                    set CurrentRoomIndex += 1

                    if(ChosenRoomUse >= CurrentRoomIndex):
                        CurrentRoom.SetRoomInUse(CurrentPlayersPerRoom)
                    else:
                        CurrentRoom.SetRoomNotInUse()

        # Sets up timer and barrier
        TeleportBarrier(false)
        RunTimer.SetMaxDuration(RunTime)
        RunTimer.StartForAll()

        #SetPlayerMovability(true)
        Sleep(RunTime)
        EliminationPhase()

    # This is the phase in which players are eliminated if they do not meet the win requirement
    EliminationPhase() : void=
        Print("Mingle Minigame - Elimination Phase")

        AllPlayers := GetRemainingPlayers()
        set PlayersToEliminate = AllPlayers

        # Checks for players who need to be eliminated
        for(CurrentRoom : AllRooms):
            # Checks if players in room meet requirement
            if(CurrentRoom.PlayersInRoom.Length = CurrentPlayersPerRoom):
                for(RemovePlayer : CurrentRoom.PlayersInRoom):
                    # Show Survived HUD
                    MingleSurvivedHUD.Show(RemovePlayer)
                    # Remove player from Elimination List
                    if(FindPlayerToRemove := PlayersToEliminate.Find[RemovePlayer]):
                        if(TempPlayers := PlayersToEliminate.RemoveElement[FindPlayerToRemove]):
                            set PlayersToEliminate = TempPlayers

        # Eliminates all players on Eliminate List
        for(ElimPlayer : PlayersToEliminate):
            # Stops entire game session from restarting
            if(GetRemainingPlayers().Length <= 1):
                EndedEvent.Signal()
            else:
                if(FortChar := ElimPlayer.GetFortCharacter[]):
                    EliminationAudio.Play()
                    FortChar.Damage(1000.0)

        spawn{DelayBetweenRounds()}

    # Called to rotate Statue
    PlatformRotateAnimation(TimeToSpin : float, AnimationModeType : animation_mode)<suspends>:void=
        if:
            AnimationController := CenterPlatformProp.GetAnimationController[] 
        then:
            XRot := 180.0 * TimeToSpin * PlaformSpinSpeed

            #Create the keyframe
            KeyFrame1 := keyframe_delta :
                DeltaLocation := vector3{X:=0.0, Y:= 0.0, Z:= 0.0}
                DeltaRotation := MakeRotationFromYawPitchRollDegrees(XRot, 0.0, 0.0)
                Time := TimeToSpin
                Interpolation := Linear
            #Create keyframe Array
            KeyFrames : []keyframe_delta = array { KeyFrame1 }
            #Set Animation To Animation Controller
            AnimationController.SetAnimation(KeyFrames, ?Mode:=AnimationModeType)
            #Play Animation
            AnimationController.Play()
        return

    # Called to teleport Barrier over the spinning platform
    TeleportBarrier(IsOn : logic) : void=
        PlatTransform := CenterPlatformProp.GetTransform()
        PlatPos := PlatTransform.Translation
        TeleRot := PlatTransform.Rotation

        if(IsOn = true):
            TelePos := PlatPos
            if(PlaformBarrierProp.TeleportTo[TelePos, TeleRot]):
        else:
            TelePos := vector3{X:=PlatPos.X, Y:=PlatPos.Y, Z:=1000.0}
            if(PlaformBarrierProp.TeleportTo[TelePos, TeleRot]):

    # Called after elimination to add delay between each round
    DelayBetweenRounds()<suspends> : void=
        Sleep(DelayAfterElimination)

        # Resets Elimination Array
        set PlayersToEliminate = array{}

        # Restarts the Game Loop
        RunTimer.ResetForAll()
        GameLoopEvent.Signal()

    # Turns Lights On and Off
    ToggleLights(IsOn : logic) : void=
        for(Light : DoorLights):
            if(IsOn = true):
                Light.TurnOn()
            else:
                Light.TurnOff()