using { /Fortnite.com/Devices }
using { /Fortnite.com/AI}
using { /Fortnite.com/Teams}
using { /Verse.org/Simulation }
using { /Verse.org/Random }
using { /Fortnite.com/Characters }
using { /Fortnite.com/FortPlayerUtilities }
using { /Fortnite.com/Game }
using { /Fortnite.com/Teams }
using { /Verse.org/Simulation/Tags }
using { /UnrealEngine.com/Temporary/Diagnostics }
using { /UnrealEngine.com/Temporary/SpatialMath } 
using { /Verse.org/Assets }

ColorChangingTile := class(tag) {}

CaptureHandler := class():
    Device : TurfWars
    Tile : color_changing_tiles_device

    var OwnedTeam : ?custom_team = false

    HandlerFunction(InAgent : agent):void=
        if(Player := player[InAgent], Team := Device.GetTeam[Player]):
            Device.IncreaseTeamScore(Player)

            if(FoundTeam := OwnedTeam?):
                set FoundTeam.CurrentScore -= 1

            set OwnedTeam = option{Team}

TurfWars := class(minigame_device):
    var StoredTiles : []CaptureHandler = array{}

    var GameInitialized : logic = false

    @editable
    ItemSpawners : []item_spawner_device = array{}

    #@editable
    #IntroSequence : cinematic_sequence_device = cinematic_sequence_device{}
    #@editable
    #Camera : gameplay_camera_fixed_point_device = gameplay_camera_fixed_point_device{}

    Name<override>: string = "Turf Wars"
    StartingMessage<override>:string = ""
    GetVotingImage<override>()<transacts>: texture = VerseExposed.SQ_TURFWARS

    _Start<override>()<suspends>: void =
        #IntroSequence.Play()
        #Camera.AddToAll()
        #Camera.RemoveFromAll()
        for(ItemSpawner : ItemSpawners):
            Print("Turf Wars Minigame - Item Spawner")
            ItemSpawner.SpawnItem()

        if(GameInitialized = false):
            TaggedActors := FindCreativeObjectsWithTag(ColorChangingTile{})
            for(Actor : TaggedActors, CastedActor := color_changing_tiles_device[Actor]):
                CreatedCaptureHandler := CaptureHandler{Device := Self, Tile := CastedActor}
                CastedActor.ActivatedEvent.Subscribe(CreatedCaptureHandler.HandlerFunction)
                set StoredTiles += array{CreatedCaptureHandler}
            set GameInitialized = true

    OnSafeTimeEnded<override>()<suspends>:void=
        return

    _Reset<override>()<suspends>:void=
        for(Tile : StoredTiles):
            Tile.Tile.Reset()

    _OnGenericEndTimerSuccessReturnArray<override>():[]player=
        var CurrentHighScore : int = 0
        var Winners : []player = array{}
        for(Team : GetRemainingTeams()):
            if(Team.CurrentScore > CurrentHighScore):
                set CurrentHighScore = Team.CurrentScore
                set Winners = Team.Players + Team.EliminatedPlayers
            else if(Team.CurrentScore = CurrentHighScore):
                set Winners += Team.Players + Team.EliminatedPlayers
        Winners

    GetWinners<override>()<transacts><decides>:[]player=
        var RetArray : []player = array{} 
        RetArray