using { /Fortnite.com/Devices }
using { /Fortnite.com/AI}
using { /Fortnite.com/Teams}
using { /Verse.org/Simulation }
using { /Verse.org/Random }
using { /Fortnite.com/Characters }
using { /Fortnite.com/FortPlayerUtilities }
using { /Fortnite.com/Game }
using { /Fortnite.com/Teams }
using { /Verse.org/Simulation/Tags }
using { /UnrealEngine.com/Temporary/Diagnostics }
using { /UnrealEngine.com/Temporary/SpatialMath } 
using { /Verse.org/Assets }

headshot := class(minigame_device):

    @editable HealthPowerUp :health_powerup_device = health_powerup_device{}

    Name<override>: string = "FULL PIECE 185"
    StartingMessage<override>:string = "One Pumps Only + INSTANT ELIMINATION Storm!"
    GetVotingImage<override>()<transacts>: texture = VerseExposed.SQ_FULLPIECE185
    DisplayIcon<override>: ?texture = option{VerseExposed.SQ_FULLPIECE185_ICON}

    _Start<override>()<suspends>: void = 
        loop:
            HealthPowerUp.Pickup()
            Sleep(0.2)

