using { /Fortnite.com/Devices }
using { /Verse.org/Simulation }
using { /UnrealEngine.com/Temporary/Diagnostics }
using { /UnrealEngine.com/Temporary/UI }
using { /Fortnite.com/UI }
using { /UnrealEngine.com/Temporary/SpatialMath }
using { /Verse.org/Colors }
using { /Verse.org/Assets }
using { /Fortnite.com/Characters }
using { /Fortnite.com/Game }
using { /Verse.org/Random }
using { /UnrealEngine.com/Assets }

finishersFIXED_minigame := class(minigame_device):
    @editable FinisherDevices :[]finisher_devices = array{}
    var ShuffledFinisherDevices : []finisher_devices = array{}
    @editable FinisherExplosiveDevice : explosive_device = explosive_device {}
    @editable DefaultFinisher : finisher_devices = finisher_devices{}
    @editable FinisherSelfie : finisher_devices = finisher_devices{}
    @editable DBNO :down_but_not_out_device = down_but_not_out_device{}
    @editable RunFinisherInputDebug : input_trigger_device = input_trigger_device{}
    @editable FinisherTipMessage : hud_message_device = hud_message_device{}
    var FinisherIndex : int = 0 

    var PlayerInFinisher: [player]logic = map{}

    Name<override>: string = "Finishers Box PVP"
    StartingMessage<override>:string = "Pickaxe Knocked Players to Activate Finishers"
    GetVotingImage<override>()<transacts>: texture = VerseExposed.SQ_FINISHERS
    #DisplayIcon<override>: ?texture = option{VerseExposed.SQ_FINISHERS_ICON}

    _Start<override>()<suspends>: void =  
        set ShuffledFinisherDevices = Shuffle(FinisherDevices)
        #set ShuffledFinisherDevices = FinisherDevices
        DBNO.Enable()
        RunFinisherInputDebug.PressedEvent.Subscribe(OnTestPressed)
        RegisterSubscriber(DBNO.AgentDownedEvent.Subscribe(OnPlayerDowned))
        RegisterSubscriber(DBNO.AgentRevivedEvent.Subscribe(OnPlayerRevived))
        for(Player : GetRemainingPlayers(), Fort := Player.GetFortCharacter[]):
            if(set PlayerInFinisher[Player] = false) {}
            RegisterSubscriber(Fort.DamagedEvent().Subscribe(FinisherSequence))

    OnSafeTimeEnded<override>():void=
        return

    _Reset<override>()<suspends>:void=
        DBNO.Disable()
        set ShuffledFinisherDevices = array{}
        set FinisherIndex = 0
        set PlayerInFinisher = map{}

    OnPlayerDowned(PlayerDowned : agent):void=
        Print("Player downed")
        if(FortPlayerDowned := PlayerDowned.GetFortCharacter[]):
            spawn{ReviveSequence(PlayerDowned, FortPlayerDowned)}

    OnPlayerRevived(Agent : agent):void=
        Print("On Player Revived")
        if:
            Fort := Agent.GetFortCharacter[]
        then:
            spawn{SpawnProtection(Fort)}

    SpawnProtection(Fort: fort_character)<suspends>:void=
        Fort.SetVulnerability(false)
        Print("Invincible")
        Sleep(1.5)
        Print("Not Invincible")
        Fort.SetVulnerability(true)

    ReviveSequence(Player: agent, Fort : fort_character)<suspends>:void=
        #race to check if in middle of finisher
        race:
            loop:
                if(PlayerInFinisher[Player] = true):
                    Print("Downed Player is in middle of finisher. Breaking out of race and not reviving")
                    break
                Sleep(0.2)
            block:
                Sleep(10.0)
                if(Fort.IsActive[] and Fort.IsDownButNotOut[]):
                    DBNO.Revive(Player)

    FinisherSequence(DamageResult : damage_result) : void =
        PICKAXE_DAMAGE := 20.0
        if:
            Instigator := DamageResult.Instigator?
            InstigatorAgent := Instigator.GetInstigatorAgent[]
            InstigatorPlayer := player[InstigatorAgent]
            InstigatorFort := fort_character[DamageResult.Instigator?]
            CharacterWhoWasHitFort := fort_character[DamageResult.Target]
            CharacterWhoWasHitAgent := CharacterWhoWasHitFort.GetAgent[]
            CharacterWhoWasHitPlayer := player[CharacterWhoWasHitAgent]

            #Distance Check
            InstigatorTranslation := InstigatorFort.GetTransform().Translation
            TargetTranslation := CharacterWhoWasHitFort.GetTransform().Translation
            Dist := Distance(InstigatorTranslation, TargetTranslation)
            Print("Distance = {Dist}")

            # Same Agent: Usually from storm damage
            not InstigatorAgent = CharacterWhoWasHitAgent
            # IsDownButNotOut
            CharacterWhoWasHitFort.IsDownButNotOut[]
        then:
            if (DamageResult.Amount = PICKAXE_DAMAGE):
                Print("Running Finisher")
                #Get Unused Finisher
                if(set PlayerInFinisher[CharacterWhoWasHitPlayer] = true):
                #If there are only 2 players remaining do the final finisher
                if(GetRemainingPlayers().Length <= 2):
                    Print("Only 2 players alive - Final Finisher Attack")
                    spawn:
                        FinisherSelfie.Activate(InstigatorAgent, CharacterWhoWasHitAgent, 0, vector3 {}, FinisherExplosiveDevice)
                    #RunFinisher(InstigatorAgent, CharacterWhoWasHitAgent, 8)
                else:
                    Print("Performing Unused Finisher")
                    spawn:
                        GetUnusedFinisher().Activate(InstigatorAgent, CharacterWhoWasHitAgent, 0, vector3 {}, FinisherExplosiveDevice)
                        #RunFinisher(InstigatorAgent, CharacterWhoWasHitAgent, 0)
            else:
                Print("Not Pickaxing")
                FinisherTipMessage.Show(InstigatorPlayer)

    GetUnusedFinisher():finisher_devices=
        Print("Get Unused Finisher")
        if(FinisherToUse := ShuffledFinisherDevices[FinisherIndex]):
            #Extra Check
            if(FinisherToUse.BeingUsed = true):
                PrintError("Finisher is being used???? SHould Never get here but returning DefaultFinisher")
                #Iterate through all to find unused finisher
                return DefaultFinisher
            else:
                if(FinisherIndex = ShuffledFinisherDevices.Length-1):
                    set FinisherIndex = 0
                else:
                    set FinisherIndex += 1
                return FinisherToUse

        #Should Never get here but is a backup default finisher
        return DefaultFinisher

    RunFinisher(Instigator:agent, Target:agent, FinisherId:int)<suspends>:void=
        # Add logic to iterate
        if (FinisherToUse := ShuffledFinisherDevices[0]):
            spawn:
                #Use(FinisherToUse, Instigator, Target, FinisherId)
                FinisherToUse.Activate(Instigator, Target, FinisherId, vector3 {}, FinisherExplosiveDevice)

    # Use(FinisherToUse:finisher_devices, Instigator:agent, Target:agent, FinisherId:int)<suspends>:void=
    #     NO_OFFSET:vector3 = vector3 {}
    #     FinisherToUse.Activate(Instigator, Target, FinisherId, NO_OFFSET)

    OnTestPressed(Agent:agent):void=
        spawn:
            RunFinisherTest(Agent, Agent, 0)

    RunFinisherTest(Instigator:agent, Target:agent, FinisherId:int)<suspends>:void=
        # Add logic to iterate
        if (FinisherToUse := FinisherDevices[0]):
            spawn:
                #Use(FinisherToUse, Instigator, Target, FinisherId)
                FinisherToUse.Activate(Instigator, Target, FinisherId, vector3 {}, FinisherExplosiveDevice)