##################################################
################## INSTRUCTIONS ##################
##################################################
# Careful:
# - Add Name (for supervote)
# - Functions that are spawned{} from the _Start() function
# - GetPlayspace().GetPlayers() will also retrieve lobby players, use GetRemainingPlayers() instead
# - Subscribed functions should be registered with the RegisterSubscriber() method
# - Timers should never just "Start()", but "Enable+Reset+Start" instead and "Disable" on _Reset()
# - Use IsInGame[Player] method to check if player is still in the minigame (not in the lobby or dead)

# What to do in the _Reset() method:
# - Reset your variables
# - Disable some devices
# - Enable some reset devices (powerups?)

# Notes:
# - Can use OnSafeTimeEnded to plug something when safe time ends
# - Can use _OnPlayerEliminated to plug something when a participating player dies
# - Watch "OVERRIDE ME" methods to learn more about how to implement a new game
# - Call order is _Start() -> OnSafeTimeEnded() -> _OnMinigameEnded() -> _Reset()

# Team Colors
# Red = 1
# Blue = 2
# Green = 3
# Yellow = 4

# TwoVsTwo
# Red = 1
# Blue = 2

# One Vs Three
# Red = solo team (Snipers/less players)
# Blue = three teams (Runners/ More players)

using { /Fortnite.com/Devices }
using { /Fortnite.com/AI}
using { /Fortnite.com/Teams}
using { /Fortnite.com/UI }
using { /Verse.org/Simulation }
using { /Verse.org/Random }
using { /Verse.org/Colors }
using { /Verse.org/Assets }
using { /Fortnite.com/Characters }
using { /Fortnite.com/FortPlayerUtilities }
using { /Fortnite.com/Game }
using { /Fortnite.com/Teams }
using { /Fortnite.com/Playspaces }
using { /Verse.org/Simulation/Tags }
using { /UnrealEngine.com/Temporary/Diagnostics }
using { /UnrealEngine.com/Temporary/SpatialMath }
using { /UnrealEngine.com/Temporary/UI }

#TEAM MINIGAME CODE - ADDED
custom_team := class<unique><concrete>():
    var Players : []player = array{}
    var EliminatedPlayers : []player = array{}
    var Kills : int = 0
    var CurrentScore : int = 0
    var AllowSpawning : logic = true
    var TeamIndex : int = 0

    #Use this if you're using Respawn_Logic AutomaticRespawn = true to check if players are alive
    IsSomeoneOnTeamAlive()<transacts>:logic=
        if(Players.Length > 0):
            return true
        return false

    #For modes where we don't clear the Players array such as RELOAD. Use this if you're using AutomaticRespawn = False
    IsAFortCharacterAliveOnTeam()<transacts>:logic=
        for(Player : Players, Fort := Player.GetFortCharacter[], Fort.IsActive[]):
            return true
        return false

team_type_enum := enum:
    FourTeams
    TwoTeams
    OneVsThreeTeams

team_type := class<concrete>:
    @editable
    TwoTeams : logic = false

    @editable
    OneVsThree : logic = false

#TEAM MINIGAME CODE - ADDED
score_logic := class<concrete>:
    @editable
    ScoreToWin : type{Value:int where Value > 0} = 15

    @editable
    DynamicScoring : logic = false

    @editable
    CustomScoring : logic = false

    @editable
    ShowScoreToWin : logic = true

#TEAM MINIGAME CODE - ADDED
respawn_logic := class<concrete>:
    @editable
    RespawnTime : type{Value:float where Value >= 0.0} = 5.0
    @editable
    AutomaticRespawn : logic = true
    #Variable we can set to false during minigames to stop respawning from occuring
    var RespawnsEnabled : logic = true

    @editable
    RespawnAtPointCloseToTeam : logic = false
    @editable
    RespawnPoints : []creative_prop = array{}
    @editable
    GrantDefaultItemsOnRespawn : logic = false
    @editable
    GrantDefaultArenaItemsOnRespawn : logic = false
    @editable
    GrantDefaultLoadoutsOnRespawn : logic = true
    @editable
    ItemsToGrantOnRespawn : []item_granter_device = array{}
    @editable
    ClearMaterials : ?item_remover_device = false

minigame_teleport_method_type := enum:
    UseTeleporter
    UseBarrierTransforms
    UseCreativePropTransforms

minigame_teleport_method := class<concrete>:
    @editable
    Type : minigame_teleport_method_type = minigame_teleport_method_type.UseTeleporter

    @editable
    OptTeleporter : ?teleporter_device = false

    @editable
    OptBarriers : ?[]barrier_device = false

    @editable
    OptCreativeProps : ?[]creative_prop = false

minigame_device_clear_data := class<concrete>:
    @editable
    OptExplosiveDevice : ?explosive_device = false

    @editable
    OptCinematicSequenceDataLayerReload : ?cinematic_sequence_device = false

minigame_device_mutator_config := class<concrete>:
    @editable
    DisableShooting : logic = false

    @editable
    DisableBuilding : logic = false

minigame_device_safetime_config := class(minigame_device_mutator_config):
    @editable
    Duration : float = 7.0

minigame_device_generic_endtime_config := class<concrete>:
    @editable
    Duration : float = 90.0

minigame_device_generic_itemtimer_config := class<concrete>:
    @editable
    Duration : float = 15.0

    @editable
    RemoveItemsBeforeGrant : ?item_timer_device_remover_config = false

    @editable
    ItemGrantersOnTimer : []item_timer_device_granter_config = array{}

item_timer_device_granter_config := class(smart_granter_config):
    @editable
    CycleToRandomItemBeforeGranting : logic = false

    @editable
    GrantEachPlayerIndividually : logic = false

    ShouldGrantPlayersIndividually<override>()<transacts><decides>:void = GrantEachPlayerIndividually?

item_timer_device_remover_config := class<concrete>:
    # This is used if we need to use a specific item remover (Ex: Knockback)
    # If not checked, we will use default item remover to remove everything
    @editable
    CustomItemRemover : ?item_remover_device = false
    
minigame_device_roundwise_mutator_config := class(minigame_device_mutator_config):
    @editable
    DisablePickupLifespan : logic = false

    @editable
    DisableEmoting : logic = false

    @editable
    HalfGravity : logic = false

smart_granter_config := class<concrete>:
    @editable
    ItemGranter : item_granter_device = item_granter_device{}

    @editable
    RandomItem : logic = false
    
    @editable
    GrantCount : int = 1

    ShouldGrantPlayersIndividually()<transacts><decides>:void = false?

    # Grant(Player: player):void=
    #     for(__ := 1..GrantCount):
    #         if(RandomItem?) then ItemGranter.CycleToRandomItem(Player) else ItemGranter.GrantItem(Player)

    Grant(Players: []player):void=
        for(__ := 1..GrantCount):
            if(ShouldGrantPlayersIndividually[]):
                for(Player : Players):
                    if(RandomItem?) then ItemGranter.CycleToRandomItem(Player) else ItemGranter.GrantItem(Player)
            else if(AnyAgent := Players[0]):
                if(RandomItem?) then ItemGranter.CycleToRandomItem(AnyAgent) else ItemGranter.GrantItem(AnyAgent)

minigame_device_granter_config := class(smart_granter_config):
    @editable
    GrantAfterSafeTime : logic = false

    @editable
    GrantEachPlayerIndividually : logic = false

    ShouldGrantPlayersIndividually<override>()<transacts><decides>:void = GrantEachPlayerIndividually?

minigame_device_siphon_config := class<concrete>:
    @editable
    HealthRegen : float = 50.0

    @editable
    ShieldRegen : float = 0.0

minigame_device_build_reset_config := class<concrete>:
    @editable
    IsResetPerTimer : logic = false

    @editable
    TimePerReset : float = 30.0

    @editable
    OptWarningHUD : ?hud_message_device = false

    @editable
    OptResetHUD : ?hud_message_device = false

mat_type_enum := enum:
    MaxMats
    BoxMats

minigame_selectable_default_loadout_items := class<concrete>:
    @editable
    DefaultSelectableWeapons : logic = true

    @editable
    StartingMats : ?mat_type_enum = false

    @editable
    DefaultHeals : logic = true

minigame_device_tooltip_config := class<concrete>:  
    @editable
    ToolTipHUD : hud_message_device = hud_message_device{}

    @editable
    ToolTipMessage : string = ""

    @editable
    TimeToShow : float = 0.0

minigame_device := class<unique>(crashable_device):

    # Might not be set yet if retrieved too early
    var Manager : game_manager = game_manager{}

    @editable
    TeamType : team_type_enum = team_type_enum.FourTeams

    @editable
    Name : string = ""

    @editable
    ScoreLogic : ?score_logic = false

    @editable
    RespawnLogic : ?respawn_logic = false

    var Team1Kills : int = 0
    var Team2Kills : int = 0
    var Team3Kills : int = 0
    var Team4Kills : int = 0

    var ChosenArena : ?arena_and_modifier_settings = false
    var Arena1 : logic = false
    var Arena2 : logic = false
    var Arena3 : logic = false

    @editable
    PossibleArenas : ?[]arena_and_modifier_settings = false

    @editable
    BottomMessage : ?string = false

    @editable
    ToolTipConfig : ?minigame_device_tooltip_config = false

    @editable
    Storm : basic_storm_controller_device = basic_storm_controller_device{}

    @editable
    CustomTeleportationMethod : ?minigame_teleport_method = false

    @editable
    SafeTimeConfig : minigame_device_safetime_config = minigame_device_safetime_config{}

    @editable
    EndTimerConfig : ?minigame_device_generic_endtime_config = false

    @editable
    ItemGranterTimerConfig : ?minigame_device_generic_itemtimer_config = false

    @editable
    RoundMutatorConfig : minigame_device_roundwise_mutator_config = minigame_device_roundwise_mutator_config{}

    @editable
    GrantSelectableAndDefaultLoadout : ?minigame_selectable_default_loadout_items = false

    @editable
    DefaultItemGranters : []minigame_device_granter_config = array{}

    @editable
    GrantDefaultItemsAndArenaDefaultItems : logic = false

    @editable
    OptClassSelector : ?class_and_team_selector_device = false

    @editable
    OptVitals : ?fort_character_vitals = false

    # Disabled after safe time
    @editable
    OptBarriers : ?[]barrier_device = false

    @editable
    AdditionalSiphon : ?minigame_device_siphon_config = false

    @editable
    SiphonGranters : []smart_granter_config = array{}

    @editable
    OptBuildResetConfig : ?minigame_device_build_reset_config = false

    @editable
    EndGameClearData : minigame_device_clear_data = minigame_device_clear_data{}

    # How many times the game has been shown in the voting system (1 per player)
    @editable
    OptLogoImpressionAnalytics : ?analytics_device = false

    # How many times the game has been voted (send only 1 when game wins the vote)
    @editable
    OptVoteAnalytics : ?analytics_device = false

    # How many times the game has been voted (1 per teleported player)
    # (JIP Spectating players will not count)
    @editable
    OptGameImpressionAnalytics : ?analytics_device = false

    @editable
    Weight : int = 1

    @editable
    CollaboratorSuperVoteButton : ?button_device = false

    @editable
    DisplayTeamLives : logic = true

    @editable
    DisplayTeamScores : logic = true

    @editable
    ShowSubscribeUI : logic = false

    GetVotingImage()<transacts>: texture = VerseExposed.UI_UNKNOWN
    DisplayIcon: ?texture = false

    StartingMessage : string = ""

    # Call me to end minigame and set winners
    EndedEvent : event([]player) = event([]player){}
    # Call me to increase a custsom score
    CustomScoreEvent : event([]player) = event([]player){}

    var<private> Subscribers : []cancelable = array{}

    var RemainingTeams : []custom_team = array{} # Participating players that are not registered as eliminated
    var RemainingPlayers<private> : [player]logic = map{} # Participating players that are not registered as eliminated
    var PlayersTeamMap : [player]custom_team = map{}

    var PlayerRespawning<private> : [player]logic = map{}

    OnCrashableDeviceBegin<override><final>()<suspends>:void=
        spawn{OnMinigameDeviceBegin()}
        GetPlayspace().PlayerRemovedEvent().Subscribe(OnPlayerLeaveGame)
        if(SuperVoteButton := CollaboratorSuperVoteButton?):
            race:
                # Enable/Disable button dynamically (could be better if we could use events instead)
                loop:
                    Sleep(5.0)
                    if(Manager.RunningMinigame?):
                        SuperVoteButton.Disable()
                    else:
                        SuperVoteButton.Enable()

                # Send supervote signal, which will only get caught if voting system is currently running
                loop:
                    Agent := SuperVoteButton.InteractedWithEvent.Await()
                    if(Player := player[Agent], VotingSystem := Manager.VotingSystemMaybe?):
                        VotingSystem.SuperVoteEvent.Signal(Player, option{Self})
                        SuperVoteButton.Disable()

    OnMinigameDeviceBegin()<suspends>:void=return

    GetRemainingPlayers<final>()<transacts>:[]player=
        for(Player -> __ : RemainingPlayers, IsInGame[Player]) { Player }

    GetRemainingTeams<final>()<transacts>:[]custom_team=
        for(Team : RemainingTeams, Team.IsSomeoneOnTeamAlive() = true) { Team }

    IsInGame(Player: player)<transacts><decides>:void=
        Player.GetFortCharacter[].IsActive[]
        not Manager.LobbyMutator.IsInVolume[Player] # Change -  not Manager.LobbyMutator.IsInVolume[Player]

    # OVERRIDE ME : If you want to change how winners are defined 
    #TEAM MINIGAME CODE - UPDATED
    #This code is for a generic team who is alive win condition.
    #We will need other specific win conditions for unique game modes.
    GetWinners()<transacts><decides>:[]player=
        var RetArray : []player = array{}
        #Minigame based on score logic
        if(ScoreMinigame := ScoreLogic?):
            #Print("Score Logic GetWInners()")
                for(Team : RemainingTeams):
                    if(ScoreMinigame.CustomScoring = false):
                        if(Team.Kills = ScoreMinigame.ScoreToWin):
                            var TeamPlayers : []player = array{}
                            set TeamPlayers = Team.Players + Team.EliminatedPlayers
                            TeamPlayers
                    else:
                        if(Team.CurrentScore = ScoreMinigame.ScoreToWin):
                            var TeamPlayers : []player = array{}
                            set TeamPlayers = Team.Players + Team.EliminatedPlayers
                            TeamPlayers
        else:
            #Print("Current Remaining Teams {RemainingTeams.Length}")
            if(RemainingTeams.Length <= 1, WinningTeam := RemainingTeams[0]):
                set RetArray = WinningTeam.Players + WinningTeam.EliminatedPlayers
        RetArray

    # OVERRIDE ME : If you want to change how winners are awarded
    # Default awarding called on each winners
    AwardWinner(Winner: player):void=
        Manager.XpWinAccolade.Award(Winner)
        Manager.ScoreManager.Activate(Winner)
        Manager.GoldTracker.Add(Winner, Manager.GoldPerWin)
        Manager.MinigameWinTracker.Add(Winner, 1)

    # OVERRIDE ME : If you want to change how we act on winners upon minigame ending (called at the same time as OnMinigameEnded())
    OnWinnersFound<final>(Winners: []player)<suspends>:void=
        for(Winner : Winners, Character := Winner.GetFortCharacter[]):
            Character.SetVulnerability(false)
            Character.PutInStasis(stasis_args{AllowEmotes := false})

    # Instigate random agent for vote
    SendVoteAnalytics():void=
        if(VoteAnalytics := OptVoteAnalytics?):
            if(AnyAgent := Shuffle(GetPlayspace().GetPlayers())[0]):
                VoteAnalytics.Submit(AnyAgent)

    # Make sure that players did get teleported correctly before sending analytics (they have 3 frames to tp which should be done in 1 but we never know)
    # Calling GetRemainingPlayers() again will check that the players are alive and not in the lobby mutator (thus teleported correctly)
    SendGameImpressionAnalytics()<suspends>:void=
        if(GameImpressionAnalytics := OptGameImpressionAnalytics?):
            Sleep(0.0)
            Sleep(0.0)
            Sleep(0.0)
            for(Player : GetRemainingPlayers()):
                Print("Game Impression")
                GameImpressionAnalytics.Submit(Player)

    GetTeamLivesUIManager():team_lives_ui_manager =
        var TeamUI : team_lives_ui_manager = team_lives_ui_manager{}
        TaggedActors := FindCreativeObjectsWithTag(team_lives{})
        for(TaggedActor : TaggedActors):
            if (TeamLivesUI := team_lives_ui_manager[TaggedActor]):
                set TeamUI = TeamLivesUI
        return TeamUI

    GetTeamScoresUIManager():team_score_ui_manager =
        var TeamUI : team_score_ui_manager = team_score_ui_manager{}
        TaggedActors := FindCreativeObjectsWithTag(team_scores{})
        for(TaggedActor : TaggedActors):
            if (TeamScoreUI := team_score_ui_manager[TaggedActor]):
                set TeamUI = TeamScoreUI
        return TeamUI

    Start<final>()<suspends>:void=
        ParticipatingPlayers := GetPlayspace().GetPlayers()
        set RemainingPlayers = ParticipatingPlayers.ToMapIndexes()

        var Icon : texture = VerseExposed.UI_UNKNOWN
        if(TempIcon := DisplayIcon?):
            Print("Adding display icon")
            set Icon = TempIcon
        else:
            Print("No display Icon. Using Voting Image")
            set Icon = GetVotingImage()

        for(Player : ParticipatingPlayers, Character := Player.GetFortCharacter[]):
            RegisterSubscriber(Character.EliminatedEvent().Subscribe(OnPlayerEliminatedInMinigame))
            Manager.DisplayIconManager.MakeCustomHUD(Player, Icon, 512.0)

        set PlayerRespawning = map{}

        spawn{WatchMaxRoundDuration()}

        # Hard Code Testing
        # Delete later
        # if(X := PossibleArenas?, X.Length>0):
        #     if(Value := X[1]):
        #         set ChosenArena = option{Value}
        if(Arenas := PossibleArenas?):
            for(Arena : Arenas):
                if(Arena.Debug = true):
                    Print("Setting ChosenArena? = {Arena.ArenaDevice?.Name or "Debug arena has no name"}")
                    set ChosenArena = option{Arena}

        if(BeginningLoadingSeq := ChosenArena?.ArenaDevice?.BeginningLoadingSequence?):
            Print("Has Beginning Loading Sequence!")
            BeginningLoadingSeq.Play()

        (ChosenArena?.OptBasicModifiers?.ModStorm? or ChosenArena?.ArenaDevice?.Storm or Storm).GenerateStorm()

        #Teleport Creative Props
        if(ArenaProps := ChosenArena?.ArenaDevice?.ArenaPropsToTeleport?):
            for(PropDevice : ArenaProps):
                Print("Teleporting Prop")
                if(PropDevice.CreativeProp.TeleportTo[PropDevice.NewLocation, rotation{}]):

        #TEAM MINIGAME CODE
        Print("Call Auto Sort Team Code")
        Manager.TeamSelectionManager.AutomaticallySortTeams(TeamType)

        #TEAM MINIGAME CODE - This replaces the code below. We setup our custom team class which can be used by any minigame
        SetupCustomTeams()

        NewTeleportPlayers(ParticipatingPlayers, ChosenArena?.ArenaDevice?.CustomTeleportationMethod or CustomTeleportationMethod) 
           
        #[PATCH 1.1] will attempt to catch anyone stuck in the lobby if new teleport players does not work for some reason
        spawn{TeleportPlayersStuckInLobby()}

        spawn{SendGameImpressionAnalytics()}

        spawn{CreateToolTip()}

        if(Message := ChosenArena?.OptBasicModifiers?.ModBottomMessage? or BottomMessage?):
            Manager.BottomHUDDevice.Show(BuildMessage(Message))

        if(ClassSelector := ChosenArena?.OptBasicModifiers?.ModClassSelector? or OptClassSelector?):
            for(Player : ParticipatingPlayers):
                ClassSelector.ChangeClass(Player)

        GrantersBeforeConfig := ChosenArena?.OptBasicModifiers?.ModDefaultItemGranters? or DefaultItemGranters
        for(DefaultGranter : GrantersBeforeConfig, not DefaultGranter.GrantAfterSafeTime?):
            DefaultGranter.Grant(ParticipatingPlayers)

        #Gives all players loadouts
        if(SelectableLoadout := ChosenArena?.OptBasicModifiers?.ModSelectableAndDefaultLoadout? or GrantSelectableAndDefaultLoadout?):
            if(SelectableLoadout.DefaultSelectableWeapons = true):
                for(Player : ParticipatingPlayers):
                    GivePlayerSelectableLoadout(Player)
            else:
                Print("GrantSelectableItemShopLoadout = false")

            GiveRandPlayerMatsAndHeal(SelectableLoadout)

        Vitals := ChosenArena?.OptBasicModifiers?.ModVitals? or OptVitals? or Manager.DefaultVitals
        for(Player : ParticipatingPlayers, Character := Player.GetFortCharacter[]):
            Vitals.Set(Character)
            Character.SetVulnerability(true)
            # GivePlayerMarker(Player)

        # New change to set player team nametag visuals after sleep so it wont conflict with player changing teams on the same frame
        spawn:
            GivePlayerMarkerToParticipatingPlayersDelayed()

        Print("Starting Minigame {Name}")

        ##### AFTER SAFE TIME #####
        # Show UI for teams and scores
        if(DisplayTeamLives = true):
            spawn:
                GetTeamLivesUIManager().ShowTeamUI(Self)

        if(Score := ScoreLogic?, DisplayTeamScores = true):
            spawn:
                GetTeamScoresUIManager().ShowTeamUI(Self)

        spawn{RaceStart()}

        #FIXME:
        <# Show respawn UI if enabled
        if(MinigameWithRespawn := RespawnLogic?):
            Players := GetRemainingPlayers()
            for(Player:Players):
                CheckSettingsAndShowOrHideRespawnEnableUI(Player)#>

        # Manage safe time before minigame actually starts
        Barriers := ChosenArena?.OptBasicModifiers?.ModBarriers? or ChosenArena?.ArenaDevice?.OptArenaStartingBarriers? or OptBarriers? or Manager.Box1Barriers

        ##### BEFORE SAFE TIME #####
        if(ChosenArena?.OptBasicModifiers?.ModRoundMutatorConfig?.DisableBuilding? or ChosenArena?.OptBasicModifiers?.ModSafeTimeConfig?.DisableBuilding?):
            Manager.NoBuildingMutatorZone.Enable()
        else:
            if(RoundMutatorConfig.DisableBuilding? or SafeTimeConfig.DisableBuilding?):
                Manager.NoBuildingMutatorZone.Enable()

        if(ChosenArena?.OptBasicModifiers?.ModRoundMutatorConfig?.DisableShooting? or ChosenArena?.OptBasicModifiers?.ModSafeTimeConfig?.DisableShooting?):
            Manager.NoShootingMutatorZone.Enable()
        else:
            if(RoundMutatorConfig.DisableShooting? or SafeTimeConfig.DisableShooting?):
                Manager.NoShootingMutatorZone.Enable()

        if(ChosenArena?.OptBasicModifiers?.ModRoundMutatorConfig?.DisablePickupLifespan? or RoundMutatorConfig.DisablePickupLifespan?):
            Manager.NoItemMutatorZone.Enable()

        if(ChosenArena?.OptBasicModifiers?.ModRoundMutatorConfig?.DisableEmoting? or RoundMutatorConfig.DisableEmoting?):
            Manager.NoEmotingMutatorZone.Enable()

        if(ChosenArena?.OptBasicModifiers?.ModRoundMutatorConfig?.HalfGravity? or RoundMutatorConfig.HalfGravity?):
            Manager.HalfGravityMutatorZone.Enable()

        for(Barrier : Barriers):
            Print("Barrier Enable")
            Barrier.Enable()

        Manager.SafeTimeTimerDevice.SetMaxDuration(ChosenArena?.OptBasicModifiers?.ModSafeTimeConfig?.Duration or SafeTimeConfig.Duration)
        Manager.SafeTimeTimerDevice.Reset()
        Manager.SafeTimeTimerDevice.Start()
        Manager.SafeTimeTimerDevice.SuccessEvent.Await()

        for(Barrier : Barriers):
            Barrier.Disable()

        if(not ChosenArena?.OptBasicModifiers?.ModRoundMutatorConfig?.DisableBuilding? or RoundMutatorConfig.DisableBuilding?):
            Manager.NoBuildingMutatorZone.Disable()

        if(not ChosenArena?.OptBasicModifiers?.ModRoundMutatorConfig?.DisableShooting? or RoundMutatorConfig.DisableShooting?):
            Manager.NoShootingMutatorZone.Disable()

        if(GrantDefaultItemsAndArenaDefaultItems = true):
            if(GrantersAfterConfig := ChosenArena?.OptBasicModifiers?.ModDefaultItemGranters?):
                for(DefaultGranter : GrantersAfterConfig, DefaultGranter.GrantAfterSafeTime?):
                    DefaultGranter.Grant(ParticipatingPlayers)
            for(Granters : DefaultItemGranters):
                Granters.Grant(ParticipatingPlayers)
        else:
            GrantersAfterConfig := ChosenArena?.OptBasicModifiers?.ModDefaultItemGranters? or DefaultItemGranters
            for(DefaultGranter : GrantersAfterConfig, DefaultGranter.GrantAfterSafeTime?):
                DefaultGranter.Grant(ParticipatingPlayers)

        spawn{BuildResetLoop()}
        spawn{RaceOnSafeTimeEnded()}

    #[PATCH 1.1] Attempt to teleport anyone who is stuck in the lobby
    TeleportPlayersStuckInLobby()<suspends>:void=
        Sleep(1.0)
        var LeftOverPlayers : []player = array{}
        for(StuckInLobbyPlayer : GetPlayspace().GetPlayers()):
            #Check for any players that are in the lobby and didn't get teleported
            if(Manager.LobbyMutator.IsInVolume[StuckInLobbyPlayer]):
                set LeftOverPlayers += array{StuckInLobbyPlayer}
        NewTeleportPlayers(LeftOverPlayers, ChosenArena?.ArenaDevice?.CustomTeleportationMethod or CustomTeleportationMethod)

    # Show respawn UI if respawn is enabled
    CheckSettingsAndShowOrHideRespawnEnableUI(Player : player):void=
        if(MinigameWithRespawn := RespawnLogic?):
            if(MinigameWithRespawn.RespawnsEnabled = true, Team := PlayersTeamMap[Player], Team.AllowSpawning = true): # Add any respawn checks related to individual players
                Manager.RespawnEnabledMessage.Show(Player)
            else:
                Manager.RespawnEnabledMessage.Hide(Player)
        else:
            Manager.RespawnEnabledMessage.Hide(Player)

    #Gives Players their selectable loadout
    GivePlayerSelectableLoadout(InPlayer : player) : void=
        if(WpIndex := Manager.GetPlayerDefaultWeaponLoadout[InPlayer]):
            #Print("Players WP Index = {WpIndex}")
            #Primary Selections
            if(WpIndex = 0):
                Manager.SelectableLoadouts.DefaultLoadout.GrantItem(InPlayer)
            if(WpIndex = 1):
                Manager.SelectableLoadouts.SelectableLoadout2Primary.GrantItem(InPlayer)
            if(WpIndex = 2):
                Manager.SelectableLoadouts.SelectableLoadout3Primary.GrantItem(InPlayer)
        else:
            #PrintError("Player has not selected a loadout or something wrong dude")
            Manager.SelectableLoadouts.DefaultLoadout.GrantItem(InPlayer)

        if(SecondaryWeaponIndex := Manager.GetPlayerSecondaryWeaponLoadout[InPlayer]):
            #Print("Secondary Weapon Selected = {SecondaryWeaponIndex}")
            #Secondary Selections
            if(SecondaryWeaponIndex = 0):
                Manager.SelectableLoadouts.DefaultLoadoutSecondary.GrantItem(InPlayer)
            if(SecondaryWeaponIndex = 1):
                Manager.SelectableLoadouts.SelectableLoadout2Secondary.GrantItem(InPlayer)
            if(SecondaryWeaponIndex = 2):
                Manager.SelectableLoadouts.SelectableLoadout3Secondary.GrantItem(InPlayer)
        else:
            #PrintError("Player has not selected a loadout or something wrong dude")
            Manager.SelectableLoadouts.DefaultLoadoutSecondary.GrantItem(InPlayer)
            
    #Gives Players random mats and heals
    GiveRandPlayerMatsAndHeal(InLoadout : minigame_selectable_default_loadout_items) : void=
        #MATS AND HEALS SET TO GIVE TO EVERYONE
        if(Mats := InLoadout.StartingMats?):
            case(Mats):
                mat_type_enum.MaxMats =>
                    if(RandomPlayer := GetRemainingPlayers()[0]) then Manager.MaxMatsItemGranter.GrantItem(RandomPlayer)
                mat_type_enum.BoxMats =>
                    if(RandomPlayer := GetRemainingPlayers()[0]) then Manager.BoxMatsItemGranter.GrantItem(RandomPlayer)

        if(Heals := InLoadout.DefaultHeals?):
            if(RandomPlayer := GetRemainingPlayers()[0]) then Manager.DefaultHealsItemGranter.GrantItem(RandomPlayer)

    #Adds players to the custom team
    AddPlayerToCustomTeam(FotTeamIndex : int, CustomTeam : custom_team):void=
        TeamsCollection := GetPlayspace().GetTeamCollection()
        Teams := TeamsCollection.GetTeams()
        if(Team := Teams[FotTeamIndex], TeamAgents := TeamsCollection.GetAgents[Team]):
            if(TeamAgents.Length > 0):
                for(Agent : TeamAgents):
                    #Update - We are adding IsActive as there are cases where Team Collection Agents is adding an extra Player that does not exist
                    if(Player := player[Agent], Player.GetFortCharacter[].IsActive[]):
                        set CustomTeam.Players += array{Player}
                        if(set PlayersTeamMap[Player] = CustomTeam): #Add team player is on to the map
                Print("Team {FotTeamIndex}, Number of Players in that team {TeamAgents.Length }")

    #Teams get sorted before this is called.
    SetupCustomTeams():void=
        TeamsCollection := GetPlayspace().GetTeamCollection()
        FortTeams := TeamsCollection.GetTeams()
        NumOfTeams := FortTeams.Length - 4
        # Create array and map data for all players and teams they belong to
        for(Index := 0..3):
            #Update - We now check if the player is active as we were previously getting non valid agents registered and creating false players in teams
            if(FortTeam := FortTeams[Index], Agents := TeamsCollection.GetAgents[FortTeam], Agents.Length > 0):
                var OneValidAgentFound : logic = false
                for(Agent : Agents):
                    if(Agent.GetFortCharacter[].IsActive[]):
                        set OneValidAgentFound = true
                        #Print("Active Players {PlayersActiveOnTeam} On Team {Index + 1}")
                    else:
                        Print("Player Inactive on Team but counting in Team Collection {Index + 1}")
                if(OneValidAgentFound = true):
                    NewTeam := custom_team{}
                    set NewTeam.TeamIndex = Index
                    set RemainingTeams += array{NewTeam}
                    AddPlayerToCustomTeam(Index, NewTeam)
                
        #Print("Numbe of teams to stat with {RemainingTeams.Length}")

    OnPlayerLeaveGame<final>(Player : player):void=
        ClearPlayerMiniGameTeamData(Player)

    #TEAM MINIGAME CODE - REPLACED WITH HASHMAP and Optimised code
    OnPlayerEliminatedInMinigame<final>(Result: elimination_result):void=
        if(EliminatingPlayer := player[Result.EliminatingCharacter?.GetAgent[]], ScoreMinigame := ScoreLogic?):
            Print("Eliminating Player + 1")
            Manager.EliminationTracker.Add(EliminatingPlayer,1)
            Manager.EliminationAccolade.Award(EliminatingPlayer)
            if(TeamPlayerIsOn := PlayersTeamMap[EliminatingPlayer], set TeamPlayerIsOn.Kills += 1):
                if(ScoreMinigame.CustomScoring = false):
                    if(TeamPlayerIsOn.Kills >= ScoreMinigame.ScoreToWin):
                        EndedEvent.Signal(TeamPlayerIsOn.Players)

        if(EliminatedPlayer := player[Result.EliminatedCharacter.GetAgent[]]):
            spawn{_OnPrePlayerEliminated(Result)}
            #TEAM MINIGAME CODE - Setup so minigames can have respawn that aren't score based
            if(MinigameWithRespawn := RespawnLogic?):
                if(MinigameWithRespawn.RespawnsEnabled = true, Team := PlayersTeamMap[EliminatedPlayer], Team.AllowSpawning = true): # Add any espawn checks related to individual players
                    Print("Minigame Allowing Respawn")
                    #For game modes such as Reload where we want to create a custom respawn. This stops players from automatically respawning which
                    #Should be the default way respawn_logic works
                    if(MinigameWithRespawn.AutomaticRespawn = true):
                        Print("Automatically Respawn Player")
                        spawn{RespawnAndTeleportPlayer(EliminatedPlayer, ChosenArena?.ArenaDevice?.CustomTeleportationMethod or CustomTeleportationMethod, MinigameWithRespawn.RespawnTime, RespawnLogic)}
                #else if(TeamPlayerIsOn := PlayersTeamMap[EliminatedPlayer]):
                else:
                    ClearPlayerMiniGameTeamData(EliminatedPlayer)
            else:
                ClearPlayerMiniGameTeamData(EliminatedPlayer)
                # Handle playe leaving sepeatly

        spawn{_OnPlayerEliminated(Result)}

    #Increase Custom Score for the players team
    IncreaseTeamScore(ScoringPlayer : player) : void=
        if(ScoreMinigame := ScoreLogic?):
            if(TeamPlayerIsOn := PlayersTeamMap[ScoringPlayer], set TeamPlayerIsOn.CurrentScore += 1):
                if(ScoreMinigame.CustomScoring = true):
                    if(TeamPlayerIsOn.CurrentScore >= ScoreMinigame.ScoreToWin):
                        EndedEvent.Signal(TeamPlayerIsOn.Players)

    #Increase Custom Score by X for the players team
    IncreaseTeamScore(ScoringPlayer : player, ScoreToIncreaseBy : int) : void=
        if(ScoreMinigame := ScoreLogic?):
            if(TeamPlayerIsOn := PlayersTeamMap[ScoringPlayer], set TeamPlayerIsOn.CurrentScore += ScoreToIncreaseBy):
                if(ScoreMinigame.CustomScoring = true):
                    if(TeamPlayerIsOn.CurrentScore >= ScoreMinigame.ScoreToWin):
                        EndedEvent.Signal(TeamPlayerIsOn.Players)

    #Decrease Custom Score for the players team (If players had option to lose score)
    DecreaseTeamScore(PunishPlayer : player) : void=
        if(ScoreMinigame := ScoreLogic?):
            if(TeamPlayerIsOn := PlayersTeamMap[PunishPlayer], set TeamPlayerIsOn.CurrentScore -= 1):

    ClearPlayerMiniGameTeamData(Player : player):void=
        Manager.RespawnEnabledMessage.Hide(Player)
        if(RemainingPlayers[Player]):
            # Remove player from remaining players in minigame map
            set RemainingPlayers = RemainingPlayers.WithRemovedKey(Player)
        if(Team := PlayersTeamMap[Player]):
            # Remove player from map that uses player as key and custom_team as value
            set PlayersTeamMap = PlayersTeamMap.WithRemovedKey(Player)
            # Removing player from Team object
            set Team.Players = Team.Players.RemoveAllElements(Player)
            # Add player to the EliminatedPlayers array
            set Team.EliminatedPlayers += array{Player}
            # If this team has no players on it remove the team from array
            if(Team.Players.Length <= 0):
                RemoveTeamData(Team)

    RemoveTeamData(Team : custom_team):void=
        if (RemainingTeams.Find[Team]):
            set RemainingTeams = RemainingTeams.RemoveAllElements(Team)

    # OVERRIDE ME: If you want to do something when a participating player gets eliminated
    # NOTE: If you're using a suspending call, it will not be canceled at the end of the minigame by default
    _OnPlayerEliminated(Result: elimination_result)<suspends>:void=return

    # OVERRIDE ME: If you want to do something when a participating player gets eliminated
    # NOTE: If you're using a suspending call, it will not be canceled at the end of the minigame by default
    # Called before player eliminated logic / respawn logic occurs 
    _OnPrePlayerEliminated(Result: elimination_result)<suspends>:void=return


    # Generic timer to end the game with a visual HUD
    # Note: Only works with multiple players in debug mode...
    OnGenericEndTimerSuccess(Agent: ?agent):void=
        Print("Generic Minigame EndTimer Success")
        EndedEvent.Signal(_OnGenericEndTimerSuccessReturnArray())

    # OVERRIDE ME: If you want to return a select player array
    _OnGenericEndTimerSuccessReturnArray():[]player=array{}

    OnGenericItemTimerSuccess(Agent: ?agent):void=
        Print("Generic Minigame ItemTimer Success")
        if(ItemTimer := ChosenArena?.OptBasicModifiers?.ModItemTimerConfig? or ItemGranterTimerConfig?):
            if(RemoveItems := ItemTimer.RemoveItemsBeforeGrant?):
                Print("Removing Items Optional On")
                if(CustomItemRemover := RemoveItems.CustomItemRemover?):
                    for(Player : GetRemainingPlayers()):
                        CustomItemRemover.Remove(Player)
                else:
                    for(Player : GetRemainingPlayers()):
                        Manager.ResetItemRemover.Remove(Player)
            for(Granter : ItemTimer.ItemGrantersOnTimer):
                Print("Granting Item Granters")
                if(Granter.CycleToRandomItemBeforeGranting?, RandomPlayer := GetRemainingPlayers()[0]):
                    Print("Cycle to random item FIRST")
                    Granter.ItemGranter.CycleToRandomItem(RandomPlayer)
                Granter.Grant(GetRemainingPlayers())

    # Failsafe to restart the round in case the server is stuck or something
    WatchMaxRoundDuration<final>()<suspends>:void=
        race:
            EndedEvent.Await()
            block:
                Sleep(Manager.MaximumMinigameDuration)
                EndedEvent.Signal(array{})
                Sleep(Inf) # Avoid self canceling issues

    RaceStart<final>()<suspends>:void=
        race:
            EndedEvent.Await()
            block:
                _Start()
                Sleep(Inf)

    # OVERRIDE ME : Start the minigame
    # Be careful when using spawn{} here because those functions won't be cancelled when the minigame ends
    # Register any Subscriber with the RegisterSubscriber() function
    _Start()<suspends>:void=return

    RaceOnSafeTimeEnded<final>()<suspends>:void=
        if(Manager.RunningMinigame? = Self): # Check if we're still running the minigame
            race:
                EndedEvent.Await()
                block:
                    if(EndTimer := ChosenArena?.OptBasicModifiers?.ModEndTimerConfig? or EndTimerConfig?):
                        Print("EndTimerConfig is true")
                        Manager.GenericMinigameEndTimer.Enable()
                        Manager.GenericMinigameEndTimer.SetMaxDuration(EndTimer.Duration)
                        Manager.GenericMinigameEndTimer.Reset()
                        Manager.GenericMinigameEndTimer.Start()
                        RegisterSubscriber(Manager.GenericMinigameEndTimer.SuccessEvent.Subscribe(OnGenericEndTimerSuccess))
                    if(ItemTimer := ChosenArena?.OptBasicModifiers?.ModItemTimerConfig? or ItemGranterTimerConfig?):
                        Print("ItemTimerConfig is true")
                        Manager.GenericMinigameItemTimer.Enable()
                        Manager.GenericMinigameItemTimer.SetMaxDuration(ItemTimer.Duration)
                        Manager.GenericMinigameItemTimer.Reset()
                        Manager.GenericMinigameItemTimer.Start()
                        RegisterSubscriber(Manager.GenericMinigameItemTimer.SuccessEvent.Subscribe(OnGenericItemTimerSuccess))
                    OnSafeTimeEnded()
                    Sleep(Inf)

    # OVERRIDE ME : If you want to achieve something after safe time timer ends
    OnSafeTimeEnded()<suspends>:void=return

    OnMinigameEnded<final>(Winners: []player)<suspends>:void=
        for(Player: GetPlayspace().GetPlayers()):
            # Remove respawn UI
            Manager.RespawnEnabledMessage.Hide(Player)
            if(CurrentIconHud := Manager.DisplayIconManager.CustomHUDMap[Player]):
                CurrentIconHud.RemoveCurrentIcon(Player)

        for(Subscriber : Subscribers):
            Subscriber.Cancel()
        set Subscribers = array{}
        Manager.MinigameCollectibleCoin.Hide()
        if(ChosenArena?.OptBasicModifiers?.ModEndTimerConfig? or EndTimerConfig?):
            Print("EndTimerConfig - OnMinigameEnded")
            Manager.GenericMinigameEndTimer.Reset()
            Manager.GenericMinigameEndTimer.Disable()
        if(ChosenArena?.OptBasicModifiers?.ModItemTimerConfig? or ItemGranterTimerConfig?):
            Print("ItemTimerConfig - OnMinigameEnded")
            #Manager.GenericMinigameItemTimer.Complete()
            #Manager.GenericMinigameItemTimer.Reset()
            Manager.GenericMinigameItemTimer.Disable()
        Print("Mingame_Device - Minigame has ended!")
        OnScrapBookUpdate(Winners)
        _OnMinigameEnded(Winners)

    # OVERRIDE ME : If you want to achieve something after the minigame ends (right before the victory/defeat callouts)
    # NOTE: This function is called right before the victory/defeat callouts, use _Reset() to trigger things while players have a black loading screen
    _OnMinigameEnded(Winners: []player)<suspends>:void=return

    OnScrapBookUpdate<final>(Winners: []player)<suspends>:void=
        #Print("Minigame_Device - Called On Scrapbook Update")
        for(Player : Winners):
            #Print("Minigame_Device - Checking Player")
            if(Agent := agent[Player]):
                #Print("Minigame_Device - Gets Agent")
                Manager.ScrapBook.IncreasePlayerLevel(Agent, Name)
        _OnScrapBookUpdate(Winners)

    # OVERRIDE ME : If you want to achieve something else for updating scrap book
    _OnScrapBookUpdate(Winners: []player)<suspends>:void=return

    Reset<final>()<suspends>:void=
        #If players leave the game and we don't clear the icon map we'll have a null reference
        Print("Reset The Map")
        #set Team1Players = array{}
        #set Team2Players = array{}
        #set Team3Players = array{}
        #set Team4Players = array{}
        #Reset the teams, players and map 
        set RemainingTeams = array{}
        #set RemainingPlayers = array
        #set PlayersTeamMap = array{}
        set Team1Kills = 0
        set Team2Kills = 0
        set Team3Kills = 0
        set Team4Kills = 0

        if(DisplayTeamLives = true):
            GetTeamLivesUIManager().HideTeamUI()

        if(Score := ScoreLogic?, DisplayTeamScores = true):
            GetTeamScoresUIManager().HideTeamUI()

        Manager.DisplayIconManager.ClearHUDMaps()

        Manager.NoShootingMutatorZone.Disable()
        Manager.NoBuildingMutatorZone.Disable()
        Manager.NoEmotingMutatorZone.Disable()
        Manager.HalfGravityMutatorZone.Disable()
        # NOTE: We don't disable the "no item" mutator here since it's gonna be disabled soon by the game manager

        #UiLogoHud.Hide()
        Manager.BottomHUDDevice.Hide()
        Manager.ClearLayer1Trigger.Trigger()
        Manager.ClearLayer2Trigger.Trigger()

        for(Player : GetPlayspace().GetPlayers()):
            Manager.RedTeamHUD.Hide(Player)
            Manager.BlueTeamHUD.Hide(Player)
            Manager.GreenTeamHUD.Hide(Player)
            Manager.YellowTeamHUD.Hide(Player)

        if(ArenaProps := ChosenArena?.ArenaDevice?.ArenaPropsToTeleport?):
            for(PropDevice : ArenaProps):
                if(PropDevice.CreativeProp.TeleportTo[PropDevice.OriginLocation, rotation{}]):

        (ChosenArena?.OptBasicModifiers?.ModStorm? or ChosenArena?.ArenaDevice?.Storm or Storm).DestroyStorm()
        if(ExplosiveDevice := ChosenArena?.ArenaDevice?.EndGameClearData.OptExplosiveDevice? or EndGameClearData.OptExplosiveDevice?):
            # FIX 24/01/25 - attempting to fix build destruction
            var bAnyAgent : ?player = false
            for(AnyPlayer : GetPlayspace().GetPlayers(), AnyPlayer.IsActive[]):
                set bAnyAgent = option{AnyPlayer}
            if(AnyAgent := bAnyAgent?):
                ExplosiveDevice.Reset()
                ExplosiveDevice.Explode(AnyAgent)
            else:
                Print("No valid agent found so cant destroy builds", ?Duration := 10.0)

        if(CinematicDataLayerReload := ChosenArena?.ArenaDevice?.EndGameClearData.OptCinematicSequenceDataLayerReload? or EndGameClearData.OptCinematicSequenceDataLayerReload?):
            CinematicDataLayerReload.Play() 

        Sleep(0.0)
        spawn{_Reset()}

    # OVERRIDE ME : Reset the state of the minigame
    # - Reset variables
    # - Disable devices
    # - Enable other devices that cancel certain devices (powerups/class selectors...)
    # NOTE: This function is called when everybody is being teleported back to lobby (and have a black loading screen)
    # Use OnMinigameEnded() function if you want to trigger things right before the victory/defeat callouts
    _Reset()<suspends>:void=return

    # Always register subscribers by using this function
    RegisterSubscriber<final>(Cancelable: cancelable):void=
        set Subscribers += array{Cancelable}
    
    # OVERRIDE ME : If you want to change how we teleport players at the start of the minigame
    NewTeleportPlayers(Players: []player, MinigameTeleportMethod : ?minigame_teleport_method):void=
        #TEAM MINIGAME CODE
        TeamsCollection := GetPlayspace().GetTeamCollection()
        if(Method := MinigameTeleportMethod?):
            case(Method.Type):
                # Use a single teleporter
                minigame_teleport_method_type.UseTeleporter => 
                    if(Teleporter := Method.OptTeleporter?):
                        for(Player: Players):
                            Teleporter.Teleport(Player)
                    else:
                        Print("ERR: Minigame {Name} teleport method is set to UseTeleporter but teleporter is missing")
                # Use barriers locations
                minigame_teleport_method_type.UseBarrierTransforms => 
                    if(TeleportBarriers := Method.OptBarriers?, TeleportBarriers.Length > 0):
                        #TeleportBarriers.BulkTeleport(TeamsCollection, Players, Manager.Teleporter, ?TranslationOffset := vector3{Z := 200.0}, ?YawOffset := -90.0)
                        for(Player: Players):
                            if(PlayerTeam := PlayersTeamMap[Player]):
                                TeleportBarriers.SingleTeleport(PlayerTeam, Player, Manager.Teleporter, ?TranslationOffset := vector3{Z := 200.0}, ?YawOffset := -90.0)
                    else:
                        Print("ERR: Minigame {Name} teleport method is set to UseBarrierTransforms but barriers are missing")
                # Use creative prop locations
                minigame_teleport_method_type.UseCreativePropTransforms =>
                    if(CreativeProps := Method.OptCreativeProps?, CreativeProps.Length > 0):
                        #CreativeProps.BulkTeleport(TeamsCollection, Players, Manager.Teleporter)
                        for(Player: Players):
                            if(PlayerTeam := PlayersTeamMap[Player]):
                                CreativeProps.SingleTeleport(PlayerTeam, Player, Manager.Teleporter, ?TranslationOffset := vector3{Z := 200.0}, ?YawOffset := -90.0)
                    else:
                        Print("ERR: Minigame {Name} teleport method is set to UseCreativePropTransforms but props are missing")
        # Use default method (TP to box1)
        else:
            #TEAM MINIGAME CODE
            #Team 1,2,3 and 4
            #Manager.Box1Barriers.BulkTeleport(TeamsCollection,Players, Manager.Teleporter, ?TranslationOffset := vector3{Z := 200.0}, ?YawOffset := -90.0) 
            for(Player: Players):
                if(PlayerTeam := PlayersTeamMap[Player]):
                    Manager.Box1Barriers.SingleTeleport(PlayerTeam, Player, Manager.Teleporter, ?TranslationOffset := vector3{Z := 200.0}, ?YawOffset := -90.0)      

    DeviceName<override>:string = "__GENERATED__"
    GetDeviceName<override>()<transacts><decides>:string = Name.Length > 0 and Name or "UNKNOWN MINIGAME"

    #Resets Builds after X amount of seconds
    BuildResetLoop()<suspends> : void=
        if(BuildResetConfig := OptBuildResetConfig?):
            if(BuildResetConfig.IsResetPerTimer = true):
                if(Explosive := EndGameClearData.OptExplosiveDevice?):
                    race:
                        block:
                            EndedEvent.Await()
                        block:
                            loop:
                                if(WarningHUD := BuildResetConfig.OptWarningHUD?):
                                    WarningHUD.Show()
                                
                                Sleep(BuildResetConfig.TimePerReset)

                                if(AnyAgent := GetPlayspace().GetPlayers()[0]):
                                    Explosive.Reset()
                                    Explosive.Explode(AnyAgent)

                                if(ResetHUD := BuildResetConfig.OptResetHUD?):
                                    ResetHUD.Show()

    FindFurthestPropFromPlayers(CreativeProps : []creative_prop, bPlayers : ?[]player)<transacts>:creative_prop=
        var FurthestPropFromPlayers : creative_prop = creative_prop{}
        var FurthestDistance : float = 0.0
        var Players : []player = GetRemainingPlayers()
        if (ValidPlayers := bPlayers?):
            set Players = ValidPlayers
        if (Prop := CreativeProps[0]):
            set FurthestPropFromPlayers = Prop
        for(CreativeProp:CreativeProps):
            var ClosestPlayerDist : float = 99999999999.0
            for(Player:Players, FortChar := Player.GetFortCharacter[]):
                Translation := FortChar.GetTransform().Translation
                Dist := Distance(Translation, CreativeProp.GetTransform().Translation)
                if (ClosestPlayerDist > Dist):
                    set ClosestPlayerDist = Dist
            if (ClosestPlayerDist > FurthestDistance):
                set FurthestDistance = ClosestPlayerDist
                set FurthestPropFromPlayers = CreativeProp
        #Debug.DrawSphere(FurthestPropFromPlayers.GetTransform().Translation + vector3{X := 0.0, Y := 0.0, Z := 950.0}, ?Radius:= 200.0, ?Color:=color{R:= 1.0, G:= 0.0, B := 0.0},  ?Thickness:=5.0, ?Duration:=2.0)
        return FurthestPropFromPlayers

    FindFurthestBarrierFromPlayers(CreativeProps : []barrier_device, bPlayers : ?[]player)<transacts>:barrier_device=
        var FurthestPropFromPlayers : barrier_device = barrier_device{}
        var FurthestDistance : float = 0.0
        var Players : []player = GetRemainingPlayers()
        if (ValidPlayers := bPlayers?):
            set Players = ValidPlayers
        if (Prop := CreativeProps[0]):
            set FurthestPropFromPlayers = Prop
        for(CreativeProp:CreativeProps):
            var ClosestPlayerDist : float = 99999999999.0
            for(Player:Players, FortChar := Player.GetFortCharacter[]):
                Translation := FortChar.GetTransform().Translation
                Dist := Distance(Translation, CreativeProp.GetTransform().Translation)
                if (ClosestPlayerDist > Dist):
                    set ClosestPlayerDist = Dist
            if (ClosestPlayerDist > FurthestDistance):
                set FurthestDistance = ClosestPlayerDist
                set FurthestPropFromPlayers = CreativeProp
        #Debug.DrawSphere(FurthestPropFromPlayers.GetTransform().Translation + vector3{X := 0.0, Y := 0.0, Z := 950.0}, ?Radius:= 200.0, ?Color:=color{R:= 1.0, G:= 0.0, B := 0.0},  ?Thickness:=5.0, ?Duration:=2.0)
        return FurthestPropFromPlayers

    RespawnAndTeleportPlayer(Player : player, MinigameTeleportMethod : ?minigame_teleport_method, RespawnTime : float, bRespawnLogic : ?respawn_logic)<suspends>: void=
        TeamsCollection := GetPlayspace().GetTeamCollection()
        #TEAM MINIGAME CODE
        if(PlayerRespawning[Player] = true):
            return
        
        if(set PlayerRespawning[Player] = true){}
        Print("Reset and respawn player. Respawn Time = {RespawnTime}")
        #UPDATE [We Removed the -1.0 from Sleep]
        Sleep(RespawnTime)
        #We've added a check here because this code should not run if the minigame is running
        #You run into a scenario where a player could respawn at a specific frame and void the respawn + teleport to lobby
        if(Manager.RunningMinigame?):
            # Make sure team respawn is allowed before respawning player
            if(PlayerTeam := PlayersTeamMap[Player], PlayerTeam.AllowSpawning = false):
                Print("respawn cancel")
                ClearPlayerMiniGameTeamData(Player)
                return
            Print("respawn")
            #Manager.RespawnHUDMessage.Show(Player)
            Player.Respawn(vector3{},IdentityRotation())
            #Add If Statement checking if current minigame is running

            # Respawn player away from all enemies and closer to teamates
            if(MiniGameRespawn := bRespawnLogic?, MiniGameRespawn.RespawnAtPointCloseToTeam = true):
                if(MiniGameRespawn.RespawnPoints[0], Prop := FindClosestSpawnPointToTeam(MiniGameRespawn.RespawnPoints, Player)?):
                    Print("Found prop location to spawn at")
                    NewSpawnLoc := Prop.GetTransform().Translation
                    NewSpawnRot := Prop.GetTransform().Rotation
                    Manager.Teleporter.CustomTeleport(Player, NewSpawnLoc, NewSpawnRot)

                else if(Method := MinigameTeleportMethod?):
                    case(Method.Type):
                        # Use a single teleporter
                        minigame_teleport_method_type.UseTeleporter => 
                            if(Teleporter := Method.OptTeleporter?):
                                #Player.Respawn(vector3{},IdentityRotation())
                                Teleporter.Teleport(Player)

                        minigame_teleport_method_type.UseBarrierTransforms => 
                            Print("Using Barrier Teleportation For Respawning")
                            if(TeleportBarriers := Method.OptBarriers?, TeleportBarriers.Length > 0, Barrier := FindClosestBarrierToTeam(TeleportBarriers, Player)?, PlayerTeam := PlayersTeamMap[Player]):
                                TeleportBarrier := array{Barrier}
                                Print("Teleport barrier Length = {TeleportBarriers.Length}")
                                TeleportBarrier.SingleTeleport(PlayerTeam, Player, Manager.Teleporter, ?TranslationOffset := vector3{Z := 200.0}, ?YawOffset := -90.0)

                        minigame_teleport_method_type.UseCreativePropTransforms =>
                            if(CreativeProps := Method.OptCreativeProps?, CreativeProps.Length > 0, Prop := FindClosestSpawnPointToTeam(CreativeProps, Player)?):
                                Transform := Prop.GetTransform()
                                FinalRotation := Transform.Rotation.ApplyYaw(DegreesToRadians(0.0))

                                Manager.Teleporter.CustomTeleport(Player, Transform.Translation, FinalRotation)
            else if(Method := MinigameTeleportMethod?):
                Print("Minigame has a teleport method")
                case(Method.Type):
                    # Use a single teleporter
                    minigame_teleport_method_type.UseTeleporter => 
                        if(Teleporter := Method.OptTeleporter?):
                            #Player.Respawn(vector3{},IdentityRotation())
                            Teleporter.Teleport(Player)
                        else:
                            Print("ERR: Minigame {Name} teleport method is set to UseTeleporter but teleporter is missing")
                    # Use barriers locations 
                    #TEAM MINIGAME CODE - UPDATED METHOD FOR RESPAWNING AT CORRECT BARRIER LOCATION OF TEAM
                    minigame_teleport_method_type.UseBarrierTransforms => 
                    
                        if(TeleportBarriers := Method.OptBarriers?, TeleportBarriers.Length > 0):
                            if(PlayerTeam := PlayersTeamMap[Player]):
                                Print("Teleport barrier Length = {TeleportBarriers.Length}")
                                TeleportBarriers.SingleTeleport(PlayerTeam, Player, Manager.Teleporter, ?TranslationOffset := vector3{Z := 200.0}, ?YawOffset := -90.0)
                        else:
                            Print("ERR: Minigame {Name} teleport method is set to UseBarrierTransforms but barriers are missing")
                    # Use creative prop locations
                    minigame_teleport_method_type.UseCreativePropTransforms =>
                        if(CreativeProps := Method.OptCreativeProps?, CreativeProps.Length > 0):
                            if:
                                RandomBarrier := Shuffle(CreativeProps)[0]
                                Transform := RandomBarrier.GetTransform()
                                FinalRotation := Transform.Rotation.ApplyYaw(DegreesToRadians(0.0))
                            then:
                                Manager.Teleporter.CustomTeleport(Player, Transform.Translation, FinalRotation)
                        else:
                            Print("ERR: Minigame {Name} teleport method is set to UseCreativePropTransforms but props are missing")
            if(MiniGameRespawn := bRespawnLogic?):
                spawn:
                    RegrantItemsAfterRespawn(Player, MiniGameRespawn)

        #Manager.RespawnHUDMessage.Hide(Player)
            #Manager.RespawnHUDMessage.Hide(Player)
        #Left this out of the if statement so PlayerRespawning map can be set to false
        if(set PlayerRespawning[Player] = false){}
        Sleep(0.5)
        if(Manager.RunningMinigame?):
            GivePlayerMarker(Player)
            #FIXME:
            #CheckSettingsAndShowOrHideRespawnEnableUI(Player)

    GivePlayerMarkerToParticipatingPlayersDelayed()<suspends>:void=
        Sleep(0.8)
        race:
            EndedEvent.Await()
            loop:
                #Print("Give Players Markers")
                for(Player : GetPlayspace().GetPlayers(), Character := Player.GetFortCharacter[], Character.IsActive[]):
                    GivePlayerMarker(Player)
                Sleep(1.0)


    GivePlayerMarker(InPlayer : player):void=
        if(PlayerTeam := PlayersTeamMap[InPlayer]):
            #Print("Assigning Player Marker")
            case(PlayerTeam.TeamIndex):
                0 =>
                    Manager.RedTeamHUD.Show(InPlayer)
                    Manager.PlayerMarkerPowerUp1.Pickup(InPlayer)
                    #Print("Assigning Red Marker")
                1 =>
                    Manager.BlueTeamHUD.Show(InPlayer)
                    Manager.PlayerMarkerPowerUp2.Pickup(InPlayer)
                    #Print("Assigning Blue Marker")
                2 =>
                    Manager.GreenTeamHUD.Show(InPlayer)
                    Manager.PlayerMarkerPowerUp3.Pickup(InPlayer)
                    #Print("Assigning Green Marker")
                3 =>
                    Manager.YellowTeamHUD.Show(InPlayer)
                    Manager.PlayerMarkerPowerUp4.Pickup(InPlayer)
                    #Print("Assigning Yellow Marker")
                _ =>
                    Print("ERR: No team assigned to player")

    RegrantItemsAfterRespawn(Player:player, MiniGameRespawn : respawn_logic)<suspends>:void=
        Sleep(0.2)
        if(ClearMatt := MiniGameRespawn.ClearMaterials?):
            ClearMatt.Remove(Player)
        Sleep(0.0)
        if( MiniGameRespawn.GrantDefaultItemsOnRespawn = true):
            for(Granter : DefaultItemGranters):
                Granter.ItemGranter.GrantItem(Player)
        if(MiniGameRespawn.GrantDefaultArenaItemsOnRespawn = true, GrantersAfterConfig := ChosenArena?.OptBasicModifiers?.ModDefaultItemGranters?):
            for(DefaultGranter : GrantersAfterConfig):
                DefaultGranter.Grant(array{Player})
        
        if(MiniGameRespawn.ItemsToGrantOnRespawn[0]):
            for(Granter : MiniGameRespawn.ItemsToGrantOnRespawn):
                Granter.GrantItem(Player)
        if(MiniGameRespawn.GrantDefaultLoadoutsOnRespawn = true):
            #Gives all players loadouts
            if(SelectableLoadout := ChosenArena?.OptBasicModifiers?.ModSelectableAndDefaultLoadout? or GrantSelectableAndDefaultLoadout?):
                if(SelectableLoadout.DefaultSelectableWeapons = true):
                    GivePlayerSelectableLoadout(Player)
                else:
                    Print("GrantSelectableItemShopLoadout = false")

    FindClosestSpawnPointToTeam(Props : []creative_prop, Player : player)<transacts>:?creative_prop=
        if(Prop1 := Props[0]):
            Teams := GetRemainingTeams()
            var EnemyTeamPlayers : []player = GetEnemyPlayers(Player)
            var PlayerTeamArray : []player = GetTeamPlayers(Player)

            var LowestPropWeight : tuple(float, creative_prop) = (999999.0, Prop1)
            for(Prop : Props):
                var PropWeight : float = 0.0
                PropLocation := Prop.GetTransform().Translation
                for(EnemyPlayer : EnemyTeamPlayers, EnemyChar := EnemyPlayer.GetFortCharacter[]):
                    EnemyLoc := EnemyChar.GetTransform().Translation
                    Dist := (PropLocation - EnemyLoc).Length()
                    if(Dist < 2000.0):
                        WeightToAdd := 2000.0 - Dist
                        set PropWeight +=  WeightToAdd

                for(TeamPlayer : PlayerTeamArray, TeamChar := TeamPlayer.GetFortCharacter[], not TeamPlayer = Player):
                    FriendLoc := TeamChar.GetTransform().Translation
                    Dist := (PropLocation - FriendLoc).Length()
                    if(Dist < 2000.0):
                        WeightToSubTract := 2000.0 - Dist
                        set PropWeight -=  WeightToSubTract
                    
                if(PropWeight < LowestPropWeight(0)):
                    set LowestPropWeight = (PropWeight, Prop)
            return option{LowestPropWeight(1)}
        return false

    FindClosestBarrierToTeam(Barriers : []barrier_device, Player : player)<transacts>:?barrier_device=
        if(Barrier1 := Barriers[0]):
            Print("Found Barriers")
            Teams := GetRemainingTeams()
            var EnemyTeamPlayers : []player = GetEnemyPlayers(Player)
            var PlayerTeamArray : []player = GetTeamPlayers(Player)
            
            var LowestBarrierWeight : tuple(float, barrier_device) = (999999.0, Barrier1)
            for(Barrier : Barriers):
                var BarrierWeight : float = 0.0
                BarrierLocation := Barrier.GetTransform().Translation
                for(EnemyPlayer : EnemyTeamPlayers, EnemyChar := EnemyPlayer.GetFortCharacter[]):
                    EnemyLoc := EnemyChar.GetTransform().Translation
                    Dist := (BarrierLocation - EnemyLoc).Length()
                    if(Dist < 2000.0):
                        WeightToAdd := 2000.0 - Dist
                        set BarrierWeight +=  WeightToAdd

                for(TeamPlayer : PlayerTeamArray, TeamChar := TeamPlayer.GetFortCharacter[], not TeamPlayer = Player):
                    FriendLoc := TeamChar.GetTransform().Translation
                    Dist := (BarrierLocation - FriendLoc).Length()
                    if(Dist < 2000.0):
                        WeightToSubTract := 2000.0 - Dist
                        set BarrierWeight -=  WeightToSubTract
                    
                if(BarrierWeight < LowestBarrierWeight(0)):
                    set LowestBarrierWeight = (BarrierWeight, Barrier)
            return option{LowestBarrierWeight(1)}
        Print("No Barriers Found, returning false")
        return false

    IsInTeam(Team : []player, Player : player)<transacts>:logic=
        var isRemainingPlayerLogic : logic = false
        if(Team.Find[Player]):
            set isRemainingPlayerLogic = true
        isRemainingPlayerLogic

    IsInTeam(Team : custom_team, Player : player)<transacts><decides>:void=
        Team.Players.Find[Player]

    GetEnemyPlayers(Player : player)<transacts>:[]player=
        var EnemyPlayers : []player = array{}
        AllPlayes := GetRemainingPlayers()
        if (PlayerTeam := PlayersTeamMap[Player]):
            for(ThisPlayer : AllPlayes):
                if (not ThisPlayer = Player):
                    if (IsInTeam[PlayerTeam, ThisPlayer]):
                    else:
                        set EnemyPlayers += array{ThisPlayer}
        else:
            for(ThisPlayer : AllPlayes):
                if (not ThisPlayer = Player):
                    set EnemyPlayers += array{ThisPlayer}

        return EnemyPlayers

    GetTeamPlayers(Player : player)<transacts>:[]player=
        var TeamPlayers : []player = array{}
        if (Team := PlayersTeamMap[Player]):
            set TeamPlayers = Team.Players
        return TeamPlayers

    GetTeam(Player : player)<transacts><decides>:custom_team=
        PlayersTeamMap[Player]

    CreateToolTip()<suspends> : void=
        if(Config := ToolTipConfig?):
            Sleep(Config.TimeToShow)
            Config.ToolTipHUD.SetText(BuildMessage(Config.ToolTipMessage))
            Config.ToolTipHUD.Show()
