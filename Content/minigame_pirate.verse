# using { /Fortnite.com/Devices }
# using { /Verse.org/Simulation }
# using { /UnrealEngine.com/Temporary/Diagnostics }
# using { /Fortnite.com/Characters }
# using { /Fortnite.com/Playspaces }
# using { /UnrealEngine.com/Temporary/SpatialMath }
# using { /Fortnite.com/Devices/CreativeAnimation/InterpolationTypes}
# using { /Fortnite.com/Devices/CreativeAnimation}
# using { /Verse.org/Verse }
# using {/Verse.org/Colors}
# using {/Verse.org/Colors/NamedColors}
# using { /Fortnite.com/Playspaces}
# using { /Verse.org/Random}
# using { /Fortnite.com/Game}
# using { /Fortnite.com/UI}
# using { /UnrealEngine.com/Temporary/UI }
# using { /Verse.org/Assets }

# knockback_pirate := class(minigame_device):
#     Logger:log = log{Channel := KnockBackLog}
#     @editable KnockbackElimSiphon : float = 200.0
#     @editable KnockBackDevices : []KnockbackDevice = array{}
#     @editable PirateMutatorZone : mutator_zone_device = mutator_zone_device{}

#     #Damage that has already been inflected to player
#     var TotalDamagePerPlayer : [player]float = map{}
#     var PlayerLastHitBy : [player]player = map{}

#     KnockBackLocation : vector3 = vector3{X:= -1498.470165, Y:= 2347.905954, Z:= 0.0}

#     var UIPerPlayer : [player]?canvas = map{}

#     @editable ItemGranterRocketRam:item_granter_device = item_granter_device{}

#     Name<override>: string = "Knockback"
#     StartingMessage<override>:string = "More Damage = More Knockback"
#     GetVotingImage<override>()<transacts>: texture = VerseExposed.SQ_PIRATE

#     _Start<override>()<suspends>: void =
#         AllPlayers := GetRemainingPlayers()
#         CreateWidgetForPlayers(AllPlayers)
#         RegisterSubscriber(PirateMutatorZone.AgentExitsEvent.Subscribe(MapMutatorZoneAgentExitsEliminationEvent))
#         for(Player : AllPlayers, Fort := Player.GetFortCharacter[]):
#             option{set TotalDamagePerPlayer[Player] = 0.0}
#             if(set PlayerLastHitBy[Player] = Player):
#             RegisterSubscriber(Fort.DamagedShieldEvent().Subscribe(InitializePlayerKnockBack))

#     _Reset<override>()<suspends>: void =
#         for(Player -> WidgetMaybe : UIPerPlayer, Widget := WidgetMaybe?, PlayerUI := GetPlayerUI[Player]):
#             PlayerUI.RemoveWidget(Widget)
#         set UIPerPlayer = map{}
#         set TotalDamagePerPlayer = map{}
#         set PlayerLastHitBy = map{}

#     StartingDamageText<localizes> : message = "0%"

#     DamageText<localizes>(TotalDamage : int) : message = "{TotalDamage}%"

#     InitializePlayerKnockBack(DamageResult : damage_result) : void =
#         if:
#             Instigator := DamageResult.Instigator?
#             InstigatorAgent := Instigator.GetInstigatorAgent[]
#             InstigatorPlayer := player[InstigatorAgent]
#             CharacterWhoWasHitFort := fort_character[DamageResult.Target]
#             CharacterWhoWasHitAgent := CharacterWhoWasHitFort.GetAgent[]
#             CharacterWhoWasHitPlayer := player[CharacterWhoWasHitAgent]
#         then:
#             #Update map
#             if(InstigatorAgent = CharacterWhoWasHitAgent):
#                 Print("Same Agent: Usually from storm damage")
#             else:
#                 Print("Updating LastPlayerMap")
#                 if(set PlayerLastHitBy[CharacterWhoWasHitPlayer] = InstigatorPlayer):
#         spawn{KnockPlayerBack(DamageResult)}

#     KnockPlayerBack(DamageInflicted : damage_result)<suspends> : void =
#         if (CharacterDamaged := fort_character[DamageInflicted.Target]):
#             if (CharacterShoots := fort_character[DamageInflicted.Instigator?]):
#                 if (AgentFromChar := CharacterDamaged.GetAgent[]):
#                     if (DamagedAgent := CharacterDamaged.GetAgent[]):
#                         if (DamagedPlayer := player[DamagedAgent]):
#                             DamageAmount := DamageInflicted.Amount
#                             Print("DamageAmount = {DamageAmount}")
#                             if (set TotalDamagePerPlayer[DamagedPlayer] += DamageAmount):         
#                                 if (TotalDamageOnPlayer := TotalDamagePerPlayer[DamagedPlayer]):
#                                     #Print("Total Damage on Player : {TotalDamageOnPlayer}")
#                                     if (TotalDamageAsInt := Ceil[TotalDamageOnPlayer]):
#                                         spawn {SetKnockBackBasedOnDamage(TotalDamageAsInt, DamagedPlayer, CharacterShoots)}
#                                     else {Logger.Print("Unable to ceil total damage", ?Level := log_level.Error)}
#                                 else {Logger.Print("Can't get total damage", ?Level := log_level.Error)}
#                             else {Logger.Print("Can't set damage", ?Level := log_level.Error)}
#                         else {Logger.Print("Can't get damaged player", ?Level := log_level.Error)}
#                     else {Logger.Print("Can't get damaged agent", ?Level := log_level.Error)}
#                 else {Logger.Print("Can't Get character from agent", ?Level := log_level.Error)}
#             else {Logger.Print("can't get Character from instigator", ?Level := log_level.Error)}
#         else {Logger.Print("Can't get Character from Damaged", ?Level := log_level.Error)}


#     SetKnockBackBasedOnDamage(Damage : int, DamagedPlayer : player, DamageInstigatorchar : fort_character)<suspends> : void = 
#         if (RightCanvas := UIPerPlayer[DamagedPlayer]?):
#             if (Slot1OfCanvas := RightCanvas.Slots[0]):
#                 if (TypeWidget := text_base[Slot1OfCanvas.Widget]):
#                     TypeWidget.SetText(DamageText(Damage))
#                     for(KnockbackDev : KnockBackDevices):
#                         if(Damage >= KnockbackDev.MinDamage and Damage <= KnockbackDev.MaxDamage):
#                             Print("Found Knockback Device!")
#                             TeleProp := KnockbackDev.TeleporterProp
#                             KnockbackMovementMod := KnockbackDev.KnockbackMovementModulator
#                             TypeWidget.SetTextColor(KnockbackDev.Color)
#                             TeleProp.MoveTo(KnockBackLocation, DamageInstigatorchar.GetViewRotation(), 0.1)
#                             KnockbackMovementMod.Activate(DamagedPlayer)

#                 else {Logger.Print("Can't get type widget in knockback", ?Level := log_level.Error)}
#             else {Logger.Print("Can't get right canvas slot in knockback", ?Level := log_level.Error)}
#         else {Logger.Print("Can't get right canvas in knockback", ?Level := log_level.Error)}
    

#     CreateWidgetForPlayers(AllPlayers : []player) : void=
#         for (Player : AllPlayers):
#             if (PlayerUI := GetPlayerUI[Player]):
#                 DamageUI := CreateDamageUI()
#                 PlayerUI.AddWidget(DamageUI)
#                 if (set UIPerPlayer[Player] = option{DamageUI}) {}
#                 else {Logger.Print("can't set UIPERPLAYER for player", ?Level := log_level.Error)}
    
#     CreateDamageUI() : canvas = 
#         MyDamageText : text_block = text_block{DefaultText := StartingDamageText, DefaultTextColor := DarkGreen}
#         MyCanvas : canvas = canvas:
#             Slots := array:
#                 canvas_slot:
#                     Widget := MyDamageText
#                     Anchors := anchors{Minimum := vector2{X := 0.49, Y := 0.11}, Maximum := vector2{X := 0.50, Y := 0.50}}

#         return MyCanvas

#     _OnPlayerEliminated<override>(Result: elimination_result)<suspends>:void=
#         if:
#             Killer := Result.EliminatingCharacter?
#             KillerAgent := Killer.GetAgent[]
#             #KillerFort := KillerAgent.GetFortCharacter[]
#             KillerPlayer := player[KillerAgent]
#             KillerHealth := TotalDamagePerPlayer[KillerPlayer]
#             NewHealthFloat := Max(KillerHealth - KnockbackElimSiphon,0.0)
#             NewHealthInt := Ceil[NewHealthFloat]
#             IsInGame[KillerPlayer]
#         then:
#             if (set TotalDamagePerPlayer[KillerPlayer] = NewHealthFloat):
#             if (RightCanvas := UIPerPlayer[KillerPlayer]?):
#                 if (Slot1OfCanvas := RightCanvas.Slots[0]):
#                     if (TypeWidget := text_base[Slot1OfCanvas.Widget]):
#                         TypeWidget.SetText(DamageText(NewHealthInt))
#                         for(KnockbackDev : KnockBackDevices):
#                             if(NewHealthInt >= KnockbackDev.MinDamage and NewHealthInt <= KnockbackDev.MaxDamage):
#                                 TypeWidget.SetTextColor(KnockbackDev.Color)

#     MapMutatorZoneAgentExitsEliminationEvent(Agent : agent) : void =
#         Print("Player Exiting Mutator Zone")
#         if(PlayerExiting := player[Agent], PlayerExitingFort := PlayerExiting.GetFortCharacter[]):
#             if(LastPlayerToHit := PlayerLastHitBy[PlayerExiting], LastPlayerToHitFort := LastPlayerToHit.GetFortCharacter[]):
#                 if(not LastPlayerToHit = PlayerExiting):
#                     Print("Should Damage")
#                     DamageArgs := damage_args:
#                         Amount := 99999.9
#                         Instigator := option{LastPlayerToHitFort}
#                         Source := option{LastPlayerToHitFort}
#                     PlayerExitingFort.SetVulnerability(true)
#                     PlayerExitingFort.Damage(DamageArgs)
#                 else:
#                     Print("Should Damage2")
#                     PlayerExitingFort.SetVulnerability(true)
#                     PlayerExitingFort.Damage(99999.9)