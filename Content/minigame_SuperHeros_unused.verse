
using { /Fortnite.com/Devices }
using { /Verse.org/Simulation }
using { /UnrealEngine.com/Temporary/Diagnostics }
using { /UnrealEngine.com/Temporary/UI }
using { /Fortnite.com/UI }
using { /UnrealEngine.com/Temporary/SpatialMath }
using { /Verse.org/Colors }
using { /Verse.org/Assets }
using { /Fortnite.com/Characters }
using { /Fortnite.com/Game }
using { /Verse.org/Random }

# See https://dev.epicgames.com/documentation/en-us/uefn/create-your-own-device-in-verse for how to create a verse device.

# UI_Hero_Select := class():
#     Texture : texture = VerseExposed.T_HeroUI
#     var Button1 : button_loud = button_loud{}
#     var Button2 : button_loud = button_loud{}
#     var Button3 : button_loud = button_loud{}
#     var Button4 : button_loud = button_loud{}
#     var Button5 : button_loud = button_loud{}
#     var Button6 : button_loud = button_loud{}
#     var Button7 : button_loud = button_loud{}
#     var Button8 : button_loud = button_loud{}

#     var Canva : canvas = canvas{}
    
#     RemoveUI(Player : player):void=
#         if(PlayerUI := GetPlayerUI[Player]):
#             PlayerUI.RemoveWidget(Canva)

#     CreateUI(Player : player, MiniGame : minigame75_SuperHeros):canvas=
#         Button1.SetText(_StringToMessage("Invisible"))
#         Button2.SetText(_StringToMessage("FOAD"))
#         Button3.SetText(_StringToMessage("Bob"))
#         Button4.SetText(_StringToMessage("One Shot"))
#         Button5.SetText(_StringToMessage("Auto Aim"))
#         Button6.SetText(_StringToMessage("Flight"))
#         Button7.SetText(_StringToMessage("Double Pump"))
#         Button8.SetText(_StringToMessage("Nitro Gloves"))


#         Image : texture_block = texture_block{DefaultImage := Texture, DefaultDesiredSize := vector2{X:= 1920.0, Y := 1080.0}}
#         NewUI : canvas = canvas:
#             Slots := array:
#                 canvas_slot:
#                     Anchors := anchors{Minimum := vector2{X:= 0.5, Y := 0.5}, Maximum := vector2{X:= 0.5, Y:= 0.5}}
#                     Offsets := margin{Left := 0.0, Top := 0.0, Right := 1920.0, Bottom := 1080.0}
#                     Alignment := vector2{X:= 0.5, Y := 0.5}
#                     SizeToContent := true
#                     Widget := Image

#                 canvas_slot:
#                     Anchors := anchors{Minimum := vector2{X:= 0.5, Y := 0.5}, Maximum := vector2{X:= 0.5, Y:= 0.5}}
#                     Offsets := margin{Left := -660.0, Top := -10.0, Right := 360.0, Bottom := 82.0}
#                     Alignment := vector2{X:= 0.5, Y := 0.5}
#                     SizeToContent := true
#                     Widget := Button1

#                 canvas_slot:
#                     Anchors := anchors{Minimum := vector2{X:= 0.5, Y := 0.5}, Maximum := vector2{X:= 0.5, Y:= 0.5}}
#                     Offsets := margin{Left := -667.86853, Top := -478.0, Right := 360.0, Bottom := 82.0}
#                     Alignment := vector2{X:= 0.5, Y := 0.5}
#                     SizeToContent := true
#                     Widget := text_block{DefaultText := _StringToMessage("")}

#                 canvas_slot:
#                     Anchors := anchors{Minimum := vector2{X:= 0.5, Y := 0.5}, Maximum := vector2{X:= 0.5, Y:= 0.5}}
#                     Offsets := margin{Left := -210.0, Top := -10.0, Right := 360.0, Bottom := 83.0}
#                     Alignment := vector2{X:= 0.5, Y := 0.5}
#                     SizeToContent := true
#                     Widget := Button2

#                 canvas_slot:
#                     Anchors := anchors{Minimum := vector2{X:= 0.5, Y := 0.5}, Maximum := vector2{X:= 0.5, Y:= 0.5}}
#                     Offsets := margin{Left := -232.0, Top := -478.0, Right := 360.0, Bottom := 83.0}
#                     Alignment := vector2{X:= 0.5, Y := 0.5}
#                     SizeToContent := true
#                     Widget := text_block{DefaultText := _StringToMessage("")}

#                 canvas_slot:
#                     Anchors := anchors{Minimum := vector2{X:= 0.5, Y := 0.5}, Maximum := vector2{X:= 0.5, Y:= 0.5}}
#                     Offsets := margin{Left := 220.0, Top := -10.0, Right := 360.0, Bottom := 83.0}
#                     Alignment := vector2{X:= 0.5, Y := 0.5}
#                     SizeToContent := true
#                     Widget := Button3

#                 canvas_slot:
#                     Anchors := anchors{Minimum := vector2{X:= 0.5, Y := 0.5}, Maximum := vector2{X:= 0.5, Y:= 0.5}}
#                     Offsets := margin{Left := 209.0, Top := -478.0, Right := 360.0, Bottom := 83.0}
#                     Alignment := vector2{X:= 0.5, Y := 0.5}
#                     SizeToContent := true
#                     Widget := text_block{DefaultText := _StringToMessage("")}
                
#                 canvas_slot:
#                     Anchors := anchors{Minimum := vector2{X:= 0.5, Y := 0.5}, Maximum := vector2{X:= 0.5, Y:= 0.5}}
#                     Offsets := margin{Left := 660.0, Top := -10.0, Right := 360.0, Bottom := 83.0}
#                     Alignment := vector2{X:= 0.5, Y := 0.5}
#                     SizeToContent := true
#                     Widget := Button4

#                 canvas_slot:
#                     Anchors := anchors{Minimum := vector2{X:= 0.5, Y := 0.5}, Maximum := vector2{X:= 0.5, Y:= 0.5}}
#                     Offsets := margin{Left := 644.0, Top := -478.0, Right := 360.0, Bottom := 83.0}
#                     Alignment := vector2{X:= 0.5, Y := 0.5}
#                     SizeToContent := true
#                     Widget := text_block{DefaultText := _StringToMessage("")}

#                 canvas_slot:
#                     Anchors := anchors{Minimum := vector2{X:= 0.5, Y := 0.5}, Maximum := vector2{X:= 0.5, Y:= 0.5}}
#                     Offsets := margin{Left := -640.86853, Top := 450.0, Right := 360.0, Bottom := 83.0}
#                     Alignment := vector2{X:= 0.5, Y := 0.5}
#                     SizeToContent := true
#                     Widget := Button5

#                 canvas_slot:
#                     Anchors := anchors{Minimum := vector2{X:= 0.5, Y := 0.5}, Maximum := vector2{X:= 0.5, Y:= 0.5}}
#                     Offsets := margin{Left := -667.86853, Top := 59.0, Right := 360.0, Bottom := 83.0}
#                     Alignment := vector2{X:= 0.5, Y := 0.5}
#                     SizeToContent := true
#                     Widget := text_block{DefaultText := _StringToMessage("")}

#                 canvas_slot:
#                     Anchors := anchors{Minimum := vector2{X:= 0.5, Y := 0.5}, Maximum := vector2{X:= 0.5, Y:= 0.5}}
#                     Offsets := margin{Left := -240.0, Top := 450.0, Right := 360.0, Bottom := 83.0}
#                     Alignment := vector2{X:= 0.5, Y := 0.5}
#                     SizeToContent := true
#                     Widget := Button6

#                 canvas_slot:
#                     Anchors := anchors{Minimum := vector2{X:= 0.5, Y := 0.5}, Maximum := vector2{X:= 0.5, Y:= 0.5}}
#                     Offsets := margin{Left := -232.0, Top := 59.0, Right := 360.0, Bottom := 83.0}
#                     Alignment := vector2{X:= 0.5, Y := 0.5}
#                     SizeToContent := true
#                     Widget := text_block{DefaultText := _StringToMessage("")}

#                 canvas_slot:
#                     Anchors := anchors{Minimum := vector2{X:= 0.5, Y := 0.5}, Maximum := vector2{X:= 0.5, Y:= 0.5}}
#                     Offsets := margin{Left := 220.0, Top := 450.0, Right := 360.0, Bottom := 83.0}
#                     Alignment := vector2{X:= 0.5, Y := 0.5}
#                     SizeToContent := true
#                     Widget := Button7

#                 canvas_slot:
#                     Anchors := anchors{Minimum := vector2{X:= 0.5, Y := 0.5}, Maximum := vector2{X:= 0.5, Y:= 0.5}}
#                     Offsets := margin{Left := 209.0, Top := 59.0, Right := 360.0, Bottom := 83.0}
#                     Alignment := vector2{X:= 0.5, Y := 0.5}
#                     SizeToContent := true
#                     Widget := text_block{DefaultText := _StringToMessage("")}
                
#                 canvas_slot:
#                     Anchors := anchors{Minimum := vector2{X:= 0.5, Y := 0.5}, Maximum := vector2{X:= 0.5, Y:= 0.5}}
#                     Offsets := margin{Left := 644.0, Top := 450.0, Right := 360.0, Bottom := 83.0}
#                     Alignment := vector2{X:= 0.5, Y := 0.5}
#                     SizeToContent := true
#                     Widget := Button8

#                 canvas_slot:
#                     Anchors := anchors{Minimum := vector2{X:= 0.5, Y := 0.5}, Maximum := vector2{X:= 0.5, Y:= 0.5}}
#                     Offsets := margin{Left := 644.0, Top := 59.0, Right := 360.0, Bottom := 83.0}
#                     Alignment := vector2{X:= 0.5, Y := 0.5}
#                     SizeToContent := true
#                     Widget := text_block{DefaultText := _StringToMessage("")}
#         #Bindings

#         MiniGame.RegisterSubscriber(Button1.OnClick().Subscribe(MiniGame.OnHeroSelect_Any))
#         MiniGame.RegisterSubscriber(Button1.OnClick().Subscribe(MiniGame.OnHeroSelect_1))

#         MiniGame.RegisterSubscriber(Button2.OnClick().Subscribe(MiniGame.OnHeroSelect_Any))
#         MiniGame.RegisterSubscriber(Button2.OnClick().Subscribe(MiniGame.OnHeroSelect_2))

#         MiniGame.RegisterSubscriber(Button3.OnClick().Subscribe(MiniGame.OnHeroSelect_Any))
#         MiniGame.RegisterSubscriber(Button3.OnClick().Subscribe(MiniGame.OnHeroSelect_3))

#         MiniGame.RegisterSubscriber(Button4.OnClick().Subscribe(MiniGame.OnHeroSelect_Any))
#         MiniGame.RegisterSubscriber(Button4.OnClick().Subscribe(MiniGame.OnHeroSelect_4))

#         MiniGame.RegisterSubscriber(Button5.OnClick().Subscribe(MiniGame.OnHeroSelect_Any))
#         MiniGame.RegisterSubscriber(Button5.OnClick().Subscribe(MiniGame.OnHeroSelect_5))

#         MiniGame.RegisterSubscriber(Button6.OnClick().Subscribe(MiniGame.OnHeroSelect_Any))
#         MiniGame.RegisterSubscriber(Button6.OnClick().Subscribe(MiniGame.OnHeroSelect_6))

#         MiniGame.RegisterSubscriber(Button7.OnClick().Subscribe(MiniGame.OnHeroSelect_Any))
#         MiniGame.RegisterSubscriber(Button7.OnClick().Subscribe(MiniGame.OnHeroSelect_7))

#         MiniGame.RegisterSubscriber(Button8.OnClick().Subscribe(MiniGame.OnHeroSelect_Any))
#         MiniGame.RegisterSubscriber(Button8.OnClick().Subscribe(MiniGame.OnHeroSelect_8))

#         set Canva = NewUI

#         return NewUI

# Hero_type := enum:
#     WallHack
#     FOAD
#     Bob 
#     PickaxeOneShot
#     AutoAim
#     InstantReload
#     DoublePump
#     Nitro 

# Hero_Class := class():
#     bMiniGameSuperHero : ?minigame75_SuperHeros = false
#     bPlayer : ?player = false
#     HeroType : Hero_type = Hero_type.WallHack
#     var UltCharge : int = 2
#     var WallHackVFX : vfx_spawner_device = vfx_spawner_device{}
#     var UltUsed : logic = false
#     # Hud Textures
#     var bUltimateAbilityTexture : ?texture = false
#     var bUltimateNoChargeTexture : ?texture = false
#     var UltimateAbilityImage : ?texture_block = false
#     UltimateHUDBackgroundTexture: texture = VerseExposed.T_UltHudBackGround
#     var UltimateHUDBackgroundImage : ?texture_block = false
#     UltimateCharge1 : texture = VerseExposed.T_UltCharge1
#     UltimateCharge2 : texture = VerseExposed.T_UltCharge2
#     var GameEnd : logic = false
#     var Canva : canvas = canvas{}
    

#     # OnUltimatePressed(Agent : agent):void=
#     #     if(Player := player[Agent], UltCharge >= 2, MiniGameSuperHero := bMiniGameSuperHero?, UltUsed = false, ThisPlayer := bPlayer?, Player = ThisPlayer):
#     #         set UltUsed = true
#     #         Print("Ultimate")
#     #         #MiniGameSuperHero.UltimateUseAudio.Play(ThisPlayer)
#     #         case(HeroType):

#     #             Hero_type.WallHack => 
#     #                 #spawn{InvisAbilityActivate(ThisPlayer, MiniGameSuperHero)}
                    
                    
#     #             Hero_type.FOAD => 
#     #                 spawn{FOADAbilityActivate(ThisPlayer, MiniGameSuperHero)}
                 
   
#     #             Hero_type.Bob => 
#     #                 spawn{BobAbilityActivate(ThisPlayer, MiniGameSuperHero)}
                   

#     #             Hero_type.PickaxeOneShot => 
#     #                 spawn{PickaxeOneShot(ThisPlayer, MiniGameSuperHero)}
                   

#     #             Hero_type.AutoAim => 
#     #                 spawn{AutoAimAbilityActivate(ThisPlayer, MiniGameSuperHero)}
                   

#     #             Hero_type.InstantReload => 
#     #                 spawn{InstantReloadAbilityActivate(ThisPlayer, MiniGameSuperHero)}

                    
#     #             Hero_type.DoublePump => 
#     #                 # spawn{DoublePumpAbilityActivate(ThisPlayer, MiniGameSuperHero)}
              

#     #             Hero_type.Nitro => 
#     #                 spawn{NitroAbilityActivate(ThisPlayer, MiniGameSuperHero)}
                  

#     AddUltCharge():void=
#         if (UltUsed = false):
#             set UltCharge = Min(UltCharge + 1, 2) 
#             UpdateUltChargeUI()

#     ClearUltCharge():void=
#         set UltCharge = 0
#         UpdateUltChargeUI()
#         #if(MiniGameSuperHero := bMiniGameSuperHero?, Player := bPlayer?):
#             #MiniGameSuperHero.UltimateEndAudio.Play(Player)

#     UpdateUltChargeUI():void=
#         if (UltCharge = 0, BackgroundImage := UltimateHUDBackgroundImage?, AbilityImage := UltimateAbilityImage?, NoChargeTexture := bUltimateNoChargeTexture?):
#             BackgroundImage.SetImage(UltimateHUDBackgroundTexture)
#             AbilityImage.SetImage(NoChargeTexture)
#         if (UltCharge = 1, BackgroundImage := UltimateHUDBackgroundImage?, AbilityImage := UltimateAbilityImage?, NoChargeTexture := bUltimateNoChargeTexture?):
#             BackgroundImage.SetImage(UltimateCharge1)
#             AbilityImage.SetImage(NoChargeTexture)
#         if (UltCharge = 2, BackgroundImage := UltimateHUDBackgroundImage?, AbilityImage := UltimateAbilityImage?, ChargeTexture := bUltimateAbilityTexture?):
#             BackgroundImage.SetImage(UltimateCharge2)
#             AbilityImage.SetImage(ChargeTexture)
#             #if(MiniGameSuperHero := bMiniGameSuperHero?, Player := bPlayer?):
#                 #MiniGameSuperHero.UltimateChargedAudio.Play(Player)

#     Init(UltimateAbilityTexture : texture, UltimateAbilityTextureNoC : texture):void=
#         set bUltimateNoChargeTexture = option{UltimateAbilityTextureNoC}
#         set bUltimateAbilityTexture = option{UltimateAbilityTexture}
#         if(ThisPlayer := bPlayer?, PlayerUI := GetPlayerUI[ThisPlayer]):
#             set Canva = MakeHeroUI()
#                 PlayerUI.AddWidget(Canva, player_ui_slot{InputMode := ui_input_mode.None})
#         UpdateUltChargeUI()

#     MakeHeroUI():canvas =
#         var NewUI : canvas = canvas{}
#         if (UltimateAbilityTexture := bUltimateAbilityTexture?):
#             set UltimateAbilityImage = option{texture_block{DefaultImage := UltimateAbilityTexture, DefaultDesiredSize := vector2{X:= 256.0, Y := 256.0}}}
#             set UltimateHUDBackgroundImage = option{texture_block{DefaultImage := UltimateHUDBackgroundTexture, DefaultDesiredSize := vector2{X:= 256.0, Y := 256.0}}}
#             if (BackgroundImage := UltimateHUDBackgroundImage?, AbilityImage := UltimateAbilityImage?):
#                 set NewUI = canvas:
#                     Slots := array:
#                         canvas_slot:
#                             Anchors := anchors{Minimum := vector2{X:= 0.5, Y := 1.0}, Maximum := vector2{X:= 0.5, Y:= 1.0}}
#                             Offsets := margin{Left := 0.0, Top := 0.0, Right := 256.0, Bottom := 256.0}
#                             Alignment := vector2{X:= 0.5, Y := 1.0}
#                             SizeToContent := true
#                             ZOrder := 0
#                             Widget := BackgroundImage
#                         canvas_slot:
#                             Anchors := anchors{Minimum := vector2{X:= 0.5, Y := 1.0}, Maximum := vector2{X:= 0.5, Y:= 1.0}}
#                             Offsets := margin{Left := 0.0, Top := -25.0, Right := 256.0, Bottom := 256.0}
#                             Alignment := vector2{X:= 0.5, Y := 1.0}
#                             SizeToContent := true
#                             ZOrder := 1
#                             Widget := AbilityImage
#         return NewUI

#     BobAbilityActivate(Player : player, MiniGameSuperHero : minigame75_SuperHeros)<suspends>:void=
#         race:
#             block:
#                 MiniGameSuperHero.BobGranter.GrantItem(Player)
#                 MiniGameSuperHero.BobTrigger.Trigger(Player)
#                 MiniGameSuperHero.DamageAmpUltimateGeneral.Pickup(Player)
#                 Sleep(20.0)
#                 #MiniGameSuperHero.DefaultItemGante.GrantItem(Player)
#                 set UltUsed = false
#                 ClearUltCharge()
#             block:
#                 loop:
#                     if (GameEnd = true):
#                         #MiniGameSuperHero.DefaultItemGante.GrantItem(Player)
#                         set UltUsed = false
#                         ClearUltCharge()
#                         break
#                     Sleep(0.0)

    # DoublePumpAbilityActivate(Player : player, MiniGameSuperHero : minigame75_SuperHeros)<suspends>:void=
    #     MiniGameSuperHero.DoubleHavocGranter.GrantItem(Player)
        # race:
        #     block:
        #         #MiniGameSuperHero.DoublePumpGranter.GrantItem(Player)
        #         MiniGameSuperHero.DamageAmpUltimateGeneral.Pickup(Player)
        #         Sleep(0.0)
        #         # Sleep(20.0)
        #         # #MiniGameSuperHero.DefaultItemGante.GrantItem(Player)
        #         # set UltUsed = false
        #         # ClearUltCharge()
        #     block:
        #         loop:
        #             if (GameEnd = true):
        #                 #MiniGameSuperHero.DefaultItemGante.GrantItem(Player)
        #                 set UltUsed = false
        #                 ClearUltCharge()
        #                 break
        #             Sleep(0.0)

    # InvisAbilityActivate(Player : player, MiniGameSuperHero : minigame75_SuperHeros)<suspends>:void=
    #     race:
    #         block:
    #             MiniGameSuperHero.WallHackGranter.GrantItem(Player)
    #             MiniGameSuperHero.WallHackClass.ChangeClass(Player)
    #             MiniGameSuperHero.DamageAmpUltimateGeneral.Pickup(Player)
    #             Print("Wall Hack")
    #             Sleep(20.0)
    #             MiniGameSuperHero.NoClass.ChangeClass(Player)
    #             #MiniGameSuperHero.DefaultItemGante.GrantItem(Player)
    #             set UltUsed = false
    #             ClearUltCharge()
    #         block:
    #             loop:
    #                 if (GameEnd = true):
    #                     MiniGameSuperHero.NoClass.ChangeClass(Player)
    #                     #MiniGameSuperHero.DefaultItemGante.GrantItem(Player)
    #                     set UltUsed = false
    #                     ClearUltCharge()
    #                     break
    #                 Sleep(0.0)

    # FOADAbilityActivate(Player : player, MiniGameSuperHero : minigame75_SuperHeros)<suspends>:void=
    #     race:
    #         block:
    #             #MiniGameSuperHero.Foad.ChangeClass(Player)
    #             MiniGameSuperHero.FoadSpeedGranter.GrantItem(Player)
    #             MiniGameSuperHero.Speed.Activate(Player)
    #             MiniGameSuperHero.DamageAmpUltimateGeneral.Pickup(Player)
    #             Sleep(20.0)
    #             #MiniGameSuperHero.NoClass.ChangeClass(Player)
    #             #MiniGameSuperHero.DefaultItemGante.GrantItem(Player)
    #             set UltUsed = false
    #             ClearUltCharge()
    #         block:
    #             loop:
    #                 if (GameEnd = true):
    #                     #MiniGameSuperHero.NoClass.ChangeClass(Player)
    #                     #MiniGameSuperHero.DefaultItemGante.GrantItem(Player)
    #                     set UltUsed = false
    #                     ClearUltCharge()
    #                     break
    #                 Sleep(0.0)

    # PickaxeOneShot(Player : player, MiniGameSuperHero : minigame75_SuperHeros)<suspends>:void=
    #     race:
    #         block:
    #             MiniGameSuperHero.PickAxeGranter.GrantItem(Player)
    #             MiniGameSuperHero.ClearInv.GrantItem(Player)
    #             MiniGameSuperHero.SlowMo.Activate(Player)
    #             MiniGameSuperHero.DamageAmpPickaxe.Pickup(Player)
    #             Sleep(20.0)
    #             MiniGameSuperHero.NoClass.ChangeClass(Player)
    #             #MiniGameSuperHero.DefaultItemGante.GrantItem(Player)
    #             set UltUsed = false
    #             ClearUltCharge()
    #         block:
    #             loop:
    #                 if (GameEnd = true):
    #                     MiniGameSuperHero.NoClass.ChangeClass(Player)
    #                     #MiniGameSuperHero.DefaultItemGante.GrantItem(Player)
    #                     set UltUsed = false
    #                     ClearUltCharge()
    #                     break
    #                 Sleep(0.0)

    # AutoAimAbilityActivate(Player : player, MiniGameSuperHero : minigame75_SuperHeros)<suspends>:void=
    #     race:
    #         block:
    #             MiniGameSuperHero.AutoAimGranter.GrantItem(Player)
    #             MiniGameSuperHero.DamageAmpUltimateGeneral.Pickup(Player)
    #             Sleep(20.0)
    #             MiniGameSuperHero.NoClass.ChangeClass(Player)
    #             #MiniGameSuperHero.DefaultItemGante.GrantItem(Player)
    #             set UltUsed = false
    #             ClearUltCharge()
    #         block:
    #             loop:
    #                 if (GameEnd = true):
    #                     MiniGameSuperHero.NoClass.ChangeClass(Player)
    #                     #MiniGameSuperHero.DefaultItemGante.GrantItem(Player)
    #                     set UltUsed = false
    #                     ClearUltCharge()
    #                     break
    #                 Sleep(0.0)

    # InstantReloadAbilityActivate(Player : player, MiniGameSuperHero : minigame75_SuperHeros)<suspends>:void=
    #     race:
    #         block:
    #             MiniGameSuperHero.InstantReload.ChangeClass(Player)
    #             MiniGameSuperHero.InstantReloadGranter.GrantItemIndex(Player, GetRandomInt(0,8))
    #             MiniGameSuperHero.DamageAmpUltimateGeneral.Pickup(Player)
    #             Sleep(20.0)
    #             MiniGameSuperHero.NoClass.ChangeClass(Player)
    #             #MiniGameSuperHero.DefaultItemGante.GrantItem(Player)
    #             set UltUsed = false
    #             ClearUltCharge()
    #         block:
    #             loop:
    #                 if (GameEnd = true):
    #                     MiniGameSuperHero.NoClass.ChangeClass(Player)
    #                     #MiniGameSuperHero.DefaultItemGante.GrantItem(Player)
    #                     set UltUsed = false
    #                     ClearUltCharge()
    #                     break
    #                 Sleep(0.0)

    # NitroAbilityActivate(Player : player, MiniGameSuperHero : minigame75_SuperHeros)<suspends>:void=
    #     race:
    #         block:
    #             MiniGameSuperHero.NitroGoldVFXR.Pickup(Player)
    #             MiniGameSuperHero.NitroGoldVFXL.Pickup(Player)
    #             MiniGameSuperHero.NitroGranter.GrantItem(Player)
    #             MiniGameSuperHero.DamageAmpNitro.Pickup(Player)
    #             Sleep(20.0)
    #             #MiniGameSuperHero.DefaultItemGante.GrantItem(Player)
    #             set UltUsed = false
    #             ClearUltCharge()
    #         block:
    #             loop:
    #                 if (GameEnd = true):
    #                     #MiniGameSuperHero.NitroRemover.Remove(Player)
    #                     #MiniGameSuperHero.DefaultItemGante.GrantItem(Player)
    #                     set UltUsed = false
    #                     ClearUltCharge()
    #                     break
    #                 Sleep(0.0)

    # EndGame():void=
    #     set GameEnd = true
    #     if(ThisPlayer := bPlayer?, PlayerUI := GetPlayerUI[ThisPlayer]):
    #         PlayerUI.RemoveWidget(Canva)

    # HandleAimIn(Agent : agent):void=
    #     if (MiniGameSuperHero := bMiniGameSuperHero?):
    #         MiniGameSuperHero.Controls.AddTo(Agent)

    # HandleAimInReleased(TupleReleased : tuple(agent, float)):void=
    #     if (MiniGameSuperHero := bMiniGameSuperHero?):
    #         MiniGameSuperHero.Controls.RemoveFrom(TupleReleased(0))

# A Verse-authored creative device that can be placed in a level
minigame75_SuperHeros := class(minigame_device):
    # UltimateIcons
    # AbilityPickaxe : texture = VerseExposed.T_PickaxeUltimate
    # AbilityNitro : texture = VerseExposed.T_BoxingGloveUltimate
    # AbilityFoad : texture = VerseExposed.T_FoadUltimate
    # AbilityDoublePump : texture = VerseExposed.T_DoublePumpUltimate
    # AbilityBob : texture = VerseExposed.T_BobSummon
    # AbilityAutoAim : texture = VerseExposed.T_AutoAim
    # AbilityWallHack : texture = VerseExposed.T_WallHackUltimate
    # AbilityInstantReload : texture = VerseExposed.T_AutoReloadUltimate

    # AbilityPickaxeNoC : texture = VerseExposed.T_PickaxeUltimateNoCharge
    # AbilityNitroNoC : texture = VerseExposed.T_BoxingGloveUltimateNoCharge
    # AbilityFoadNoC : texture = VerseExposed.T_FoadUltimateNoCharge
    # AbilityDoublePumpNoC : texture = VerseExposed.T_DoublePumpUltimateNoCharge
    # AbilityBobNoC : texture = VerseExposed.T_BobSummonNoCharge
    # AbilityAutoAimNoC : texture = VerseExposed.T_AutoAimNoCharge
    # AbilityWallHackNoC : texture = VerseExposed.T_WallHackUltimateNoCharge
    # AbilityInstantReloadNoC : texture = VerseExposed.T_AutoReloadUltimateNoCharge
    #
    # Auto Aim var
    # @editable Controls : gameplay_controls_third_person_device = gameplay_controls_third_person_device{}

    # @editable PumpConditionalButton :conditional_button_device = conditional_button_device{}
    # @editable HavocConditionalButton :conditional_button_device = conditional_button_device{}

    # @editable LockShootTime : float = 0.66

    # @editable
    # BobTrigger : trigger_device = trigger_device{}

    # @editable BoxX : float = 500.8
    # @editable BoxY : float = 517.1
    # @editable BoxZHeight : float = 720.0
    # @editable InputTriggerAimIn : input_trigger_device = input_trigger_device{}
    # @editable
    # InputTriggerShoot : input_trigger_device = input_trigger_device{}
    #
    # var PlayerUIMap : [player]UI_Hero_Select = map{}
    #var PlayerHeroMap : [player]base_ability_class = map{}
    # var PlayerShootLock: [player]logic = map{}
    # @editable
    # ClearInv : item_granter_device = item_granter_device{}
    # @editable
    # DefaultItemGante : item_granter_device = item_granter_device{}
    # @editable
    # NitroGranter : item_granter_device = item_granter_device{}
    # @editable
    # Speed : movement_modulator_device = movement_modulator_device{}

    # @editable
    # DoublePumpGranter : item_granter_device = item_granter_device{}
    # @editable
    # DoubleHavocGranter : item_granter_device = item_granter_device{}

    # @editable
    # InstantReloadGranter : item_granter_device = item_granter_device{}
    # @editable
    # AutoAimGranter : item_granter_device = item_granter_device{}
    # @editable
    # PickAxeGranter : item_granter_device = item_granter_device{}
    # @editable
    # FoadSpeedGranter : item_granter_device = item_granter_device{}
    # @editable
    # WallHackGranter : item_granter_device = item_granter_device{}
    # @editable
    # BobGranter : item_granter_device = item_granter_device{}

    # @editable
    # SlowMo : movement_modulator_device = movement_modulator_device{}

    # @editable
    # NitroRemover : item_remover_device = item_remover_device{}

    # @editable
    # DamageAmpPickaxe : damage_amplifier_powerup_device = damage_amplifier_powerup_device{}
    # # @editable
    # # DamageAmpNitro : damage_amplifier_powerup_device = damage_amplifier_powerup_device{}
    # @editable
    # DamageAmpUltimateGeneral : damage_amplifier_powerup_device = damage_amplifier_powerup_device{}

    # @editable
    # WallHackVFXDevices : []vfx_spawner_device = array{}
    # @editable
    # UltimateInputTrigger : input_trigger_device = input_trigger_device{}
    # @editable
    # WallHackClass : class_and_team_selector_device = class_and_team_selector_device{}
    # @editable
    # NoClass : class_and_team_selector_device = class_and_team_selector_device{}
    # @editable
    # Foad : class_and_team_selector_device = class_and_team_selector_device{}

    # @editable
    # InstantReload : class_and_team_selector_device = class_and_team_selector_device{}

    # @editable
    # UltimateChargedAudio : audio_player_device = audio_player_device{}
    # @editable
    # UltimateUseAudio : audio_player_device = audio_player_device{}
    # @editable
    # UltimateEndAudio : audio_player_device = audio_player_device{}

    # @editable
    # NitroRedVFXL : visual_effect_powerup_device = visual_effect_powerup_device{}
    # @editable
    # NitroRedVFXR : visual_effect_powerup_device = visual_effect_powerup_device{}

    # @editable
    # NitroGoldVFXL : visual_effect_powerup_device = visual_effect_powerup_device{}
    # @editable
    # NitroGoldVFXR : visual_effect_powerup_device = visual_effect_powerup_device{}

    # @editable
    # HudDescriptionWallHack : hud_message_device = hud_message_device{}
    # @editable
    # HudDescriptionSpeed : hud_message_device = hud_message_device{}
    # @editable
    # HudDescriptionBob : hud_message_device = hud_message_device{}
    # @editable
    # HudDescriptionOneShot : hud_message_device = hud_message_device{}
    # @editable
    # HudDescriptionAutoAim : hud_message_device = hud_message_device{}
    # @editable
    # HudDescriptionAutoReload : hud_message_device = hud_message_device{}
    # @editable
    # HudDescriptionDoublePump : hud_message_device = hud_message_device{}
    # @editable
    # HudDescriptionNitro : hud_message_device = hud_message_device{}

    # @editable
    # HeroAbilities : []Hero_type = array{}
    # @editable
    # LaserBeamManager : V_LaserBeamUltimate = V_LaserBeamUltimate{}

    # @editable
    # WeaponLifeSpanVolume : mutator_zone_device = mutator_zone_device{}
    # Add more buttons to this array
    #var ButtonData : []Button_Data = array{}
    # GetVotingImage<override>()<transacts>: texture = VerseExposed.SQ_ULTHERO
    # _Start<override>()<suspends>: void =
        # WeaponLifeSpanVolume.Enable()
        #LaserBeamManager.Reset()
        # InputTriggerShoot.Enable()
        #RegisterSubscriber(InputTriggerShoot.PressedEvent.Subscribe(HandleShoot))
        # for(Player : GetRemainingPlayers(), Fort := Player.GetFortCharacter[]):
        #     if(set PlayerShootLock[Player] = false):
        # SlowMo.Enable()
        # Speed.Enable()
        # InputTriggerAimIn.Enable()
        # UltimateInputTrigger.Enable()
        # for(Player : GetRemainingPlayers(), Fort := Player.GetFortCharacter[]):
        #     NewUI := UI_Hero_Select{}
        #     if(PlayerUI := GetPlayerUI[Player]):
        #         if(set PlayerUIMap[Player] = NewUI):
        #             PlayerUI.AddWidget(NewUI.CreateUI(Player, Self), player_ui_slot{InputMode := ui_input_mode.All})

        # On screen hero selection timer
        # Sleep(8.0)
        # # Remove Hero selection UI
        # for(Player -> PlayerUI : PlayerUIMap):
        #     PlayerUI.RemoveUI(Player)
        
        # set PlayerUIMap = map{}
        # Assign Hero selection for players who didnt yet choose
        # for(Player : GetRemainingPlayers()):
        #     if (PlayerHeroMap[Player]):
        #     else:
        #         HeroObj := ability_wallhack{bPlayer := option{Player}, HeroType := Hero_type.WallHack, bMiniGameSuperHero := option{Self}}
        #         if (set PlayerHeroMap[Player] = HeroObj):
        #             HeroObj.Init(Player, Self)
        #             HudDescriptionWallHack.Show(Player)


        # Pass in each players wall hack material that follows them
        # var i : int = 0
        # for(Player -> HeroClass : PlayerHeroMap):
        #     if(WallHackVFX := WallHackVFXDevices[i]):
        #         set HeroClass.WallHackVFX = WallHackVFX
        #         set i += 1
            
        # Controls.Enable()
        #option{PostProcessProp.TeleportTo[vector3{X:= BoxX, Y:= BoxY, Z:= BoxZHeight}, IdentityRotation()]}
        #
        # Player input detection registeration
        # RegisterPlayers(GetRemainingPlayers())
        # # Wall hack material follows each player
        # spawn{MoveWallHackVFXToPlayers()}

    #_OnSafeTimeEnded<override>():void=
        #block

    # HandleShoot(Agent : agent):void=
    #     if(PumpConditionalButton.IsHoldingItem[Agent]):
    #         if(PlayerShootLock[Agent] = false):
    #             #Make sure its unlocked
    #             DoublePumpGranter.GrantItem(Agent)
    #             spawn{LockPlayerShoot(Agent)}
    #     if(HavocConditionalButton.IsHoldingItem[Agent]):
    #         if(PlayerShootLock[Agent] = false):
    #             #Make sure its unlocked
    #             DoubleHavocGranter.GrantItem(Agent)
    #             spawn{LockPlayerShoot(Agent)}

    # LockPlayerShoot(Agent : agent)<suspends>:void =
    #     if(Player := player[Agent]):
    #         #Print("Locking Shoot Button")
    #         if(set PlayerShootLock[Player] = true):
    #         Sleep(LockShootTime)
    #         #Print("Unocking Shoot Button")
    #         if(set PlayerShootLock[Player] = false):

    # _Reset<override>()<suspends>:void=
    #     set PlayerShootLock = map{}

    # _OnMinigameEnded<override>(Winners: []player)<suspends>:void=
        # WeaponLifeSpanVolume.Disable()
        # SlowMo.Disable()
        # Speed.Disable()
        # InputTriggerAimIn.Disable()
        # InputTriggerAimIn.UnregisterAll()
        # UltimateInputTrigger.Disable()
        # UltimateInputTrigger.UnregisterAll()
        # Controls.Disable()
        # for(K -> V :PlayerHeroMap):
        #     V.EndGame()
        # set PlayerHeroMap = map{}

#     OnHeroSelect_Any(Widget: widget_message):void=
#         if (PlayerUI := PlayerUIMap[Widget.Player]):
#             PlayerUI.RemoveUI(Widget.Player)
#     # wall hack
#     OnHeroSelect_1(Widget: widget_message):void=
#         if (HeroChoiceEnum := HeroAbilities[0]):
#             HeroObj := HeroChoiceEnum.CreateHeroObject(Widget.Player, Self)
#             if (set PlayerHeroMap[Widget.Player] = HeroObj):
#                 HeroObj.Init(Widget.Player, Self)
#                 # HudDescriptionWallHack.Show(Widget.Player)
#     # foad
#     OnHeroSelect_2(Widget: widget_message):void=
#         if (HeroChoiceEnum := HeroAbilities[1]):
#             HeroObj := HeroChoiceEnum.CreateHeroObject(Widget.Player, Self)
#             if (set PlayerHeroMap[Widget.Player] = HeroObj):
#                 HeroObj.Init(Widget.Player, Self)
#             # HudDescriptionSpeed.Show(Widget.Player)
#     # bob
#     OnHeroSelect_3(Widget: widget_message):void=
#         if (HeroChoiceEnum := HeroAbilities[2]):
#             HeroObj := HeroChoiceEnum.CreateHeroObject(Widget.Player, Self)
#             if (set PlayerHeroMap[Widget.Player] = HeroObj):
#                 HeroObj.Init(Widget.Player, Self)
#             # HudDescriptionBob.Show(Widget.Player)

#     # pickaxe
#     OnHeroSelect_4(Widget: widget_message):void=
#         if (HeroChoiceEnum := HeroAbilities[3]):
#             HeroObj := HeroChoiceEnum.CreateHeroObject(Widget.Player, Self)
#             if (set PlayerHeroMap[Widget.Player] = HeroObj):
#                 HeroObj.Init(Widget.Player, Self)
#             # HudDescriptionOneShot.Show(Widget.Player)
#     # Auto Aim reload
#     OnHeroSelect_5(Widget: widget_message):void=
#         if (HeroChoiceEnum := HeroAbilities[4]):
#             HeroObj := HeroChoiceEnum.CreateHeroObject(Widget.Player, Self)
#             if (set PlayerHeroMap[Widget.Player] = HeroObj):
#                 HeroObj.Init(Widget.Player, Self)
#             InputTriggerAimIn.Register(Widget.Player)
#             RegisterSubscriber(InputTriggerAimIn.PressedEvent.Subscribe(HeroObj.HandleAimIn))
#             RegisterSubscriber(InputTriggerAimIn.ReleasedEvent.Subscribe(HeroObj.HandleAimInReleased))
#             # HudDescriptionAutoAim.Show(Widget.Player)
#     # Flight
#     OnHeroSelect_6(Widget: widget_message):void=
#         if (HeroChoiceEnum := HeroAbilities[5]):
#             HeroObj := HeroChoiceEnum.CreateHeroObject(Widget.Player, Self)
#             if (set PlayerHeroMap[Widget.Player] = HeroObj):
#                 HeroObj.Init(Widget.Player, Self)
#             # HudDescriptionAutoReload.Show(Widget.Player)
#     # Double pump
#     OnHeroSelect_7(Widget: widget_message):void=
#         if (HeroChoiceEnum := HeroAbilities[6]):
#             HeroObj := HeroChoiceEnum.CreateHeroObject(Widget.Player, Self)
#             if (set PlayerHeroMap[Widget.Player] = HeroObj):
#                 HeroObj.Init(Widget.Player, Self)
#             # HudDescriptionDoublePump.Show(Widget.Player)
#             # DoublePumpGranter.GrantItem(Widget.Player)

#     # Nitro
#     OnHeroSelect_8(Widget: widget_message):void=
#         if (HeroChoiceEnum := HeroAbilities[7]):
#             HeroObj := HeroChoiceEnum.CreateHeroObject(Widget.Player, Self)
#             if (set PlayerHeroMap[Widget.Player] = HeroObj):
#                 HeroObj.Init(Widget.Player, Self)
#             # HudDescriptionNitro.Show(Widget.Player)

#     # register player with input device so we can detect players input   
#     RegisterPlayers(Agents : []agent):void=
#         for(Agent:Agents):
#             UltimateInputTrigger.Register(Agent)
#             if(Player := player[Agent], HeroClass := PlayerHeroMap[Player]):
#                 RegisterSubscriber(UltimateInputTrigger.PressedEvent.Subscribe(HeroClass.OnUltimatePressed))

#      <override>(Result: elimination_result)<suspends>:void =
#         if(Agent := Result.EliminatingCharacter?.GetAgent[], Player := player[Agent]):
#             # Increase players ult for getting an elim
#             if(HeroClass := PlayerHeroMap[Player]):
#                 HeroClass.AddUltCharge()
#         if(Agent := Result.EliminatedCharacter.GetAgent[], Player := player[Agent]):
#             # Remove player from hero class list
#             if(HeroClass := PlayerHeroMap[Player]):
#                 HeroClass.EndGame()
#             set PlayerHeroMap = RemoveKeyFromMap(Player, PlayerHeroMap)

#     MoveWallHackVFXToPlayers()<suspends>:void=
#         # First teleport all VFX to players before moving them, we do this because players are far away when this function is first called
#         for(Player -> HeroClass : PlayerHeroMap):
#             if (FortChar := Player.GetFortCharacter[]):
#                 if(HeroClass.WallHackVFX.TeleportTo[FortChar.GetTransform().Translation, FortChar.GetTransform().Rotation]):
#             else:
#                 if(HeroClass.WallHackVFX.TeleportTo[vector3{X := 0.0, Y := 0.0, Z := 0.0}, MakeRotationFromYawPitchRollDegrees(0.0, 0.0, 0.0)]):
#         # VFX will follow players
#         race:
#             loop:
#                 Sleep(0.1)
#                 for(Player -> HeroClass : PlayerHeroMap):
#                     if (FortChar := Player.GetFortCharacter[]):
#                         spawn{HeroClass.WallHackVFX.MoveTo(FortChar.GetTransform().Translation, FortChar.GetTransform().Rotation, 0.1)}
#                     else:
#                         if(HeroClass.WallHackVFX.TeleportTo[vector3{X := 0.0, Y := 0.0, Z := 0.0}, MakeRotationFromYawPitchRollDegrees(0.0, 0.0, 0.0)]):
#             block:
#                 EndedEvent.Await()
    
#     # This is for removing a single player from map
#     RemoveKeyFromMap(Player : player, Map : [player]base_ability_class):[player]base_ability_class=
#         var NewMap : [player]base_ability_class = map{}
#         for(K -> V : Map):
#             if (not K = Player):
#                 if (set NewMap[K] = V):
#         return NewMap

#     <#OnUltimatePressed(Agent : agent):void=
#     if(Player := player[Agent], HeroClass := PlayerHeroMap[Player], HeroClass.UltCharge >= 2):
        
#         case(HeroClass.HeroType):

#             Hero_type.Invisible => 
#                 WallHackClass.ChangeClass(Player)
                
#             Hero_type.FOAD => 
#                 Print("Ultimate")

#             Hero_type.Bob => 
#                 Print("Ultimate")

#             Hero_type.PickaxeOneShot => 
#                 Print("Ultimate")

#             Hero_type.AutoAim => 
#                 Print("Ultimate")

#             Hero_type.Flight => 
#                 Print("Ultimate")
                
#             Hero_type.DoublePump => 
#                 Print("Ultimate")

#             Hero_type.Nitro => 
#                 Print("Ultimate")#>

# (HeroType : Hero_type).CreateHeroObject(Player : player, MiniGameSuperHero : minigame75_SuperHeros):base_ability_class=
#         case(HeroType):
#             Hero_type.WallHack => 
#                 return ability_foad{bPlayer := option{Player}, HeroType := Hero_type.AutoAim, bMiniGameSuperHero := option{MiniGameSuperHero}}
            
#             Hero_type.FOAD => 
#                 return ability_foad{bPlayer := option{Player}, HeroType := Hero_type.AutoAim, bMiniGameSuperHero := option{MiniGameSuperHero}}
            

#             Hero_type.Bob => 
#                 return ability_bob{bPlayer := option{Player}, HeroType := Hero_type.AutoAim, bMiniGameSuperHero := option{MiniGameSuperHero}}
            

#             Hero_type.PickaxeOneShot => 
#                 return ability_pickaxe{bPlayer := option{Player}, HeroType := Hero_type.AutoAim, bMiniGameSuperHero := option{MiniGameSuperHero}}
            

#             Hero_type.AutoAim => 
#                 return ability_autoaim{bPlayer := option{Player}, HeroType := Hero_type.AutoAim, bMiniGameSuperHero := option{MiniGameSuperHero}}
            

#             Hero_type.InstantReload => 
#                 return ability_flight{bPlayer := option{Player}, HeroType := Hero_type.AutoAim, bMiniGameSuperHero := option{MiniGameSuperHero}}

                
#             Hero_type.DoublePump => 
#                 return ability_doublepump{bPlayer := option{Player}, HeroType := Hero_type.AutoAim, bMiniGameSuperHero := option{MiniGameSuperHero}}
        

#             Hero_type.Nitro => 
#                 return ability_nitro{bPlayer := option{Player}, HeroType := Hero_type.AutoAim, bMiniGameSuperHero := option{MiniGameSuperHero}}





