using { /Fortnite.com/UI }
using { /Fortnite.com/Devices }
using { /Fortnite.com/Characters }
using { /Verse.org/Simulation }
using { /Verse.org/Random }
using { /Verse.org/Colors }
using { /Verse.org/Assets }
using { /UnrealEngine.com/Temporary/Diagnostics }
using { /UnrealEngine.com/Temporary/UI }
using { /UnrealEngine.com/Temporary/SpatialMath }
using { VerseExposed }

FormattedDuration(Duration: float, ?HideMinutesIfZero:logic = true)<transacts>:string=
    if(Minutes := Int[Duration / 60.0]):
        var Seconds : int = 0

        if(_Seconds := Mod[Floor[Duration], 60]):
            option{set Seconds = _Seconds}

        if(Minutes = 0, HideMinutesIfZero?):
            return "{Seconds}"
        else:
            ZeroLeadedSeconds := if(Seconds < 10) then "0{Seconds}" else "{Seconds}"
            return "{Minutes}:{ZeroLeadedSeconds}"
    return ""

_BuildMessage<localizes>(String : string) : message = "{String}"
_EmptyMessage<localizes>() : message = ""

BuildMessage(String : string): message = if(String.Length > 0) then _BuildMessage(String) else _EmptyMessage()

voting_slot_data := class:
    Widget : widget
    Button : text_button_base
    SuperVoteGoldIcon : texture_block
    SuperVoteButton : text_button_base
    VoteCountTextBlock : text_block
    BackgroundColorBlock : color_block
    BackgroundColorBlock2 : color_block
    SelectedBordersColorBlocks : []color_block

voting_canvas_data := class:
    Canvas: canvas = canvas{}
    VotingSlots : [?minigame_device]voting_slot_data = map{}
    TitleTextBlock : text_block = text_block{}
    ExitButton : text_button_base = button_quiet{}

vote_button_handler := class:
    MinigameMaybe : ?minigame_device
    Callback : tuple(player, ?minigame_device) -> void

    HandleOnClickEvent(WidgetMessage:widget_message):void=
        Callback(WidgetMessage.Player, MinigameMaybe)

minigame_vote_slot_count := enum:
    Three
    Four

(Count: minigame_vote_slot_count).ToInt()<transacts>:int=
    case(Count):
        minigame_vote_slot_count.Three => 3
        minigame_vote_slot_count.Four => 4

# A Verse-authored creative device that can be placed in a level
miniagme_voting_system := class(crashable_device):

    DeviceName<override>:string="Voting System"

    @editable
    GameManager : game_manager = game_manager{}

    @editable
    TimeBeforeVoting : float = 1.5

    @editable
    VoteDuration : float = 10.0

    @editable
    VoteSlots : minigame_vote_slot_count = minigame_vote_slot_count.Three

    @editable
    EnableSuperVotes : logic = true

    @editable
    SuperVoteGoldPrice : int = 500

    @editable
    SuperVoteHUDMessage : hud_message_device = hud_message_device{}

    @editable
    SuperVoteHUDDuration : float = 3.0

    SuperVoteEvent : event(tuple(player, ?minigame_device)) = event(tuple(player, ?minigame_device)){}

    var PickableMinigames : []minigame_device = array{}
    var AlreadyPickedMinigames : []minigame_device = array{}

    var TotalMinigameWeight : int = 0
    var MinigamesWeightThresholds : [int]minigame_device = map{}

    var PlayerVotes : [player]?minigame_device = map{}
    # var MinigameVotes : [?minigame_device]int = map{}

    var ProposedMinigames : []minigame_device = array{}
    var VoteEndTimestamp : float = 0.0

    # OnCrashableDeviceBegin<override>()<suspends>:void=return

    DecideNextMinigame()<suspends>:?minigame_device=
        Sleep(Max(0.5, TimeBeforeVoting)) # Sleep enough to make sure that game_manager is initialized before

        # Show vote UI
        set PickableMinigames = GameManager.DebugMinigames.Length > 0 and GameManager.DebugMinigames or GameManager.Minigames
        set PickableMinigames = for(Minigame : PickableMinigames, not AlreadyPickedMinigames.Find[Minigame]) { Minigame }
        set VoteEndTimestamp  = GetSimulationElapsedTime() + VoteDuration
        set TotalMinigameWeight = 0
        set MinigamesWeightThresholds = map{}
        set PlayerVotes = map{}
        set ProposedMinigames = array{}
        set BuiltCanvas = map{}

        ChooseAvailableMinigames()
        DisplayVoteUI()

        # Wait for vote to end somehow
        var SuperVoteData : ?tuple(player, ?minigame_device) = false
        race:
            Sleep(VoteDuration)

            block:
                Data := SuperVoteEvent.Await()
                set SuperVoteData = option{Data}

            # Refresh remaining time text
            loop:
                Sleep(0.9)
                RefreshUIs()

        # Hide UIs
        HideUIs()

        # Handle vote winner (supervote / most voted / fallback function)
        if(Data := SuperVoteData?, Player := Data(0), Minigame := Data(1)? or Shuffle(PickableMinigames)[0]):
            SuperVoteHUDMessage.Show(BuildPlayerMessage("🕹 <Argon>", Player, "</> <RespawnAlertB>SUPERVOTED</> <RespawnAlertBLrg>{Minigame.Name}</>"), ?DisplayTime := SuperVoteHUDDuration)
            Sleep(SuperVoteHUDDuration)
            set AlreadyPickedMinigames += array{Minigame}
            return option{Minigame}
        else if(NextGame := (GetBestVotedMinigame[] or Shuffle(PickableMinigames)[0])):
            set AlreadyPickedMinigames += array{NextGame}
            return option{NextGame}
        else:
            return false
            
    GetMinigameVotes()<transacts>:[?minigame_device]int=
        var Votes : [?minigame_device]int = map{}
        for(MinigameMaybe : PlayerVotes):
            if(not Votes[MinigameMaybe]):
                option{set Votes[MinigameMaybe] = 0}

            option{set Votes[MinigameMaybe] += 1}
        return Votes

    GetBestVotedMinigame()<transacts><decides>:minigame_device=
        var RetGame : ?minigame_device = false
        var HighestVoteCount : int = 0
        var SameVoteGamesMaybe : []?minigame_device = array{}

        for(GameMaybe -> Votes : GetMinigameVotes()):
            if(Votes >= HighestVoteCount):

                if(Votes > HighestVoteCount):
                    set SameVoteGamesMaybe = array{}
                    set HighestVoteCount = Votes

                set SameVoteGamesMaybe += array{GameMaybe}

        RandomGame := Shuffle(SameVoteGamesMaybe)[0]
        set RetGame = RandomGame

        # Will fail if random vote has been voted
        return RetGame?


    ChooseAvailableMinigames():void=
        # Add every debug minigames once when debug is enabled and minigame count is low
        if(GameManager.DebugMinigames.Length <= VoteSlots.ToInt()):
            set ProposedMinigames = GameManager.DebugMinigames

        for(Minigame : PickableMinigames, Minigame.Weight > 0):
            set TotalMinigameWeight += Minigame.Weight
            option{set MinigamesWeightThresholds[TotalMinigameWeight] = Minigame}

        loop:
            if(ProposedMinigames.Length >= VoteSlots.ToInt()):
                break

            if(
                RandomMinigame := GetRandomWeightedElement[MinigamesWeightThresholds, TotalMinigameWeight],
                not ProposedMinigames.Find[RandomMinigame] or GameManager.DebugMinigames.Length > 0 # Allow duplicates when debugging
            ):
                # Print("Adding {RandomMinigame.Name} to the voting list")
                set ProposedMinigames += array{RandomMinigame}

    DisplayVoteUI():void=
        for(Player : GetPlayspace().GetPlayers(), Player.GetFortCharacter[].IsActive[]):
            AddVotingCanvas(Player)

    var BuiltCanvas : [player]voting_canvas_data = map{}

    GetRemainingVotingTime()<transacts><decides>:float=
        GetSimulationElapsedTime() <= VoteEndTimestamp and Ceil[VoteEndTimestamp - GetSimulationElapsedTime()] * 1.0

    RefreshUIs():void=
        MinigameVotes := GetMinigameVotes()
        for(CanvasData : BuiltCanvas):

            if(RemainingTime := GetRemainingVotingTime[]):
                CanvasData.TitleTextBlock.SetText(BuildMessage("🕹 Vote for Minigame : {FormattedDuration(RemainingTime, ?HideMinutesIfZero := false)} 🕹"))
            
            for(MinigameMaybe -> SlotData : CanvasData.VotingSlots, SlotVotes := (MinigameVotes[MinigameMaybe] or 0)):
                SlotData.VoteCountTextBlock.SetText(BuildMessage("{SlotVotes}"))

    RefreshPlayerUI(Player: player):void=
        if(CanvasData := BuiltCanvas[Player]):
            PlayerCanSuperVote := logic{CanSuperVote[Player]}
            for(MinigameMaybe -> SlotData : CanvasData.VotingSlots):

                # Selected card borders
                if(MinigameMaybe = PlayerVotes[Player]):
                    for(BorderColorBlock : SlotData.SelectedBordersColorBlocks) { BorderColorBlock.SetVisibility(widget_visibility.Visible) }
                else:
                    for(BorderColorBlock : SlotData.SelectedBordersColorBlocks) { BorderColorBlock.SetVisibility(widget_visibility.Hidden) }
    
                SlotData.SuperVoteButton.SetEnabled(PlayerCanSuperVote)
                SlotData.SuperVoteGoldIcon.SetImage(PlayerCanSuperVote? and T_CustomCoin or T_CustomCoinDisabled)
                #     SlotData.SuperVoteButton.SetEnabled(false)
                #     SlotData.SuperVoteGoldIcon.SetImage(T_CustomCoinDisabled)

    # Called once at start
    InitUI(Player: player):void=
        MinigameVotes := GetMinigameVotes()
        if(CanvasData := BuiltCanvas[Player]):

            if(RemainingTime := GetRemainingVotingTime[]):
                CanvasData.TitleTextBlock.SetText(BuildMessage("🕹 Vote for Minigame : {FormattedDuration(RemainingTime, ?HideMinutesIfZero := false)} 🕹"))

            for(MinigameMaybe -> SlotData : CanvasData.VotingSlots, SlotVotes := (MinigameVotes[MinigameMaybe] or 0)):
                SlotData.VoteCountTextBlock.SetText(BuildMessage("{SlotVotes}"))

            RefreshPlayerUI(Player)

    HideUIs():void=
        for(Player -> CanvasData : BuiltCanvas):
            spawn{HideUI(Player)}

    HideUI(Player: player)<suspends>:void=
        Sleep(0.0) # So we're sure that the canvas has been added to BuiltCanvas map
        if(PlayerUI := GetPlayerUI[Player], CanvasData := BuiltCanvas[Player]):
            CanvasData.Canvas.SetVisibility(widget_visibility.Hidden)
            PlayerUI.RemoveWidget(CanvasData.Canvas)

    AddVotingCanvas(Player: player):void=
        if(PlayerUI := GetPlayerUI[Player]):
            CanvasData := MakeVotingCanvas()

            # Subscribe to button click
            for(MinigameMaybe -> VotingSlot : CanvasData.VotingSlots):
                Handler := vote_button_handler{MinigameMaybe := MinigameMaybe, Callback := HandlePlayerVote}
                SuperVoteHandler := vote_button_handler{MinigameMaybe := MinigameMaybe, Callback := HandlePlayerSuperVote}

                VotingSlot.Button.OnClick().Subscribe(Handler.HandleOnClickEvent)
                VotingSlot.SuperVoteButton.OnClick().Subscribe(SuperVoteHandler.HandleOnClickEvent)

                # Send impression analytics
                if(Minigame := MinigameMaybe?, LogoImpressionAnalytics := Minigame.OptLogoImpressionAnalytics?):
                    LogoImpressionAnalytics.Submit(Player)
                    Print("Logo Impression")

            CanvasData.ExitButton.OnClick().Subscribe(HandlePlayerExit)
            
            PlayerUI.AddWidget(CanvasData.Canvas, player_ui_slot{InputMode := ui_input_mode.All})
            option{set BuiltCanvas[Player] = CanvasData}

            InitUI(Player) # For join-in-progress players mostly

    HandlePlayerVote(Player: player, MinigameMaybe: ?minigame_device):void=
        option{set PlayerVotes[Player] = MinigameMaybe}
        RefreshUIs()
        RefreshPlayerUI(Player)

    CanSuperVote(Player: player)<transacts><decides>:void=
        GameManager.GoldTracker.GetValue(Player) >= SuperVoteGoldPrice

    HandlePlayerSuperVote(Player: player, MinigameMaybe: ?minigame_device):void=
        if(CanSuperVote[Player]):
            GameManager.GoldTracker.Add(Player, -SuperVoteGoldPrice)
            SuperVoteEvent.Signal(Player, MinigameMaybe)

    HandlePlayerExit(WidgetMessage: widget_message):void=
        spawn{HideUI(WidgetMessage.Player)}

    MakeVotingCanvas():voting_canvas_data=
        var RetCanvas : ?voting_canvas_data = false

        ExitButton := button_quiet:
            DefaultText := BuildMessage("Exit")

        TitleTextBlock := text_block:
            DefaultText := BuildMessage("🕹 Vote for Minigame 🕹")
            DefaultTextColor := NamedColors.White

        CenteredAlignment := vector2{X := 0.5, Y := 0.5}

        TitleCanvasSlot := canvas_slot:
            Anchors := anchors{Minimum := vector2{X:=0.5, Y:=0.15}, Maximum := vector2{X:=0.5, Y:=0.15}}
            Alignment := CenteredAlignment
            SizeToContent := true
            Widget := overlay:
                Slots := array:
                    overlay_slot:
                        VerticalAlignment := vertical_alignment.Fill
                        HorizontalAlignment := horizontal_alignment.Fill
                        Widget := color_block:
                            DefaultColor := NamedColors.Black
                            DefaultOpacity := 0.3
                    overlay_slot:
                        Widget := TitleTextBlock
                        Padding := margin{Left := 25.0, Right := 25.0, Top := 25.0, Bottom := 25.0}

        if(
            VoteSlots = minigame_vote_slot_count.Three,
            Minigame1 := ProposedMinigames[0],
            Minigame2 := ProposedMinigames[1],
            Minigame3 := ProposedMinigames[2]
        ):
            IconMargin := margin{Bottom := 155.0, Left := 25.0, Right := 25.0, Top := 10.0}
            VotingSlotData1 := MakeVotingSlot(option{Minigame1}, IconMargin)
            VotingSlotData2 := MakeVotingSlot(option{Minigame2}, IconMargin)
            VotingSlotData3 := MakeVotingSlot(option{Minigame3}, IconMargin)
            
            Canvas := canvas:
                Slots := array:
                    TitleCanvasSlot
                    canvas_slot:
                        Anchors := anchors{Minimum := vector2{X:=0.075, Y:=0.25}, Maximum := vector2{X:=0.325, Y:=0.75}}
                        Alignment := CenteredAlignment
                        Widget := VotingSlotData1.Widget
                        SizeToContent := true
                    canvas_slot:
                        Anchors := anchors{Minimum := vector2{X:=0.375, Y:=0.25}, Maximum := vector2{X:=0.625, Y:=0.75}}
                        Alignment := CenteredAlignment
                        Widget := VotingSlotData2.Widget
                        SizeToContent := true
                    canvas_slot:
                        Anchors := anchors{Minimum := vector2{X:=0.675, Y:=0.25}, Maximum := vector2{X:=0.925, Y:=0.75}}
                        Alignment := CenteredAlignment
                        Widget := VotingSlotData3.Widget
                        SizeToContent := true
                    canvas_slot:
                        Anchors := anchors{Minimum := vector2{X:=0.5, Y:=0.85}, Maximum := vector2{X:=0.5, Y:=0.85}}
                        Alignment := CenteredAlignment
                        Widget := ExitButton
                    

            return voting_canvas_data:
                Canvas := Canvas
                ExitButton := ExitButton
                TitleTextBlock := TitleTextBlock
                VotingSlots := map:
                    option{Minigame1} => VotingSlotData1
                    option{Minigame2} => VotingSlotData2
                    option{Minigame3} => VotingSlotData3
                    # false             => (VotingSlotDataRandom(2), VotingSlotDataRandom(1))

        # 4 VOTE SLOTS
        else if(
            VoteSlots = minigame_vote_slot_count.Four,
            Minigame1 := ProposedMinigames[0],
            Minigame2 := ProposedMinigames[1],
            Minigame3 := ProposedMinigames[2],
            Minigame4 := ProposedMinigames[3]
        ):
            IconMargin := margin{Bottom := 175.0, Left := 10.0, Right := 10.0, Top := 30.0}
            VotingSlotData1 := MakeVotingSlot(option{Minigame1}, IconMargin)
            VotingSlotData2 := MakeVotingSlot(option{Minigame2}, IconMargin)
            VotingSlotData3 := MakeVotingSlot(option{Minigame3}, IconMargin)
            VotingSlotData4 := MakeVotingSlot(option{Minigame4}, IconMargin)
            
            Canvas := canvas:
                Slots := array:
                    TitleCanvasSlot
                    canvas_slot:
                        Anchors := anchors{Minimum := vector2{X:=0.04, Y:=0.25}, Maximum := vector2{X:=0.24, Y:=0.75}}
                        Alignment := CenteredAlignment
                        Widget := VotingSlotData1.Widget
                        SizeToContent := true
                    canvas_slot:
                        Anchors := anchors{Minimum := vector2{X:=0.28, Y:=0.25}, Maximum := vector2{X:=0.48, Y:=0.75}}
                        Alignment := CenteredAlignment
                        Widget := VotingSlotData2.Widget
                        SizeToContent := true
                    canvas_slot:
                        Anchors := anchors{Minimum := vector2{X:=0.52, Y:=0.25}, Maximum := vector2{X:=0.72, Y:=0.75}}
                        Alignment := CenteredAlignment
                        Widget := VotingSlotData3.Widget
                        SizeToContent := true
                    canvas_slot:
                        Anchors := anchors{Minimum := vector2{X:=0.76, Y:=0.25}, Maximum := vector2{X:=0.96, Y:=0.75}}
                        Alignment := CenteredAlignment
                        Widget := VotingSlotData4.Widget
                        SizeToContent := true
                    canvas_slot:
                        # Fix for controllers not being able to find buttons that are not perpendicular to the selected button possition
                        Anchors := anchors{Minimum := vector2{X:=0.4, Y:=0.85}, Maximum := vector2{X:=0.6, Y:=0.85}}
                        Alignment := CenteredAlignment
                        Widget := ExitButton

            return voting_canvas_data:
                Canvas := Canvas
                ExitButton := ExitButton
                TitleTextBlock := TitleTextBlock
                VotingSlots := map:
                    option{Minigame1} => VotingSlotData1
                    option{Minigame2} => VotingSlotData2
                    option{Minigame3} => VotingSlotData3
                    option{Minigame4} => VotingSlotData4

        
        # We can't use <transacts> here because of messages
        return voting_canvas_data{}
    
    MakeVotingSlot(MinigameMaybe: ?minigame_device, IconMargin: margin):voting_slot_data=
        VoteButton := button_quiet:
            DefaultText := BuildMessage("Vote")

        SuperVoteButton := button_quiet:
            DefaultText := BuildMessage("Super Vote\n{SuperVoteGoldPrice}   ")
        SuperVoteButton.SetEnabled(false)

        SuperVoteGoldIcon := texture_block:
            DefaultImage := T_CustomCoinDisabled
            DefaultDesiredSize := vector2{X:= 30.0, Y:=30.0}
        
        SuperVoteOverlay := overlay:
            Slots := array:
                overlay_slot:
                    VerticalAlignment := vertical_alignment.Fill
                    HorizontalAlignment := horizontal_alignment.Fill
                    Widget := SuperVoteButton
                overlay_slot:
                    VerticalAlignment := vertical_alignment.Bottom
                    HorizontalAlignment := horizontal_alignment.Center
                    Padding := margin{Left := 40.0, Bottom := 4.0}
                    Widget := SuperVoteGoldIcon

        VoteCountTextBlock := text_block:
            DefaultText := BuildMessage("0")
            DefaultTextColor := NamedColors.White
            DefaultShadowColor := NamedColors.Black
            DefaultShadowOffset := option{vector2{X := 2.0, Y := 3.0}}

        VoteCountTextBlock.SetShadowOpacity(0.9)

        VoteCountSquareSize := 128.0
        VoteCountSquareTextureBlock := texture_block:
            DefaultImage := UI_BOX9
            DefaultDesiredSize := vector2{X := VoteCountSquareSize, Y := VoteCountSquareSize}
            
        VoteCountOverlay := overlay:
            Slots := array:
                overlay_slot:
                    VerticalAlignment := vertical_alignment.Center
                    HorizontalAlignment := horizontal_alignment.Center
                    Widget := VoteCountSquareTextureBlock
                overlay_slot:
                    VerticalAlignment := vertical_alignment.Center
                    HorizontalAlignment := horizontal_alignment.Center
                    Widget := VoteCountTextBlock

        # SlotNameTextBlock := text_block:
        #     DefaultText := BuildMessage(MinigameMaybe?.Name or "Random")
        #     DefaultTextColor := NamedColors.White

        BackgroundColorBlock := color_block:
            DefaultColor := NamedColors.Black
            DefaultOpacity := 0.4

        BackgroundColorBlock2 := color_block:
            DefaultColor := NamedColors.Black
            DefaultOpacity := 0.4

        BorderSize := 4.0
        TopBorderColorBlock := color_block{DefaultColor := NamedColors.White, DefaultDesiredSize := vector2{X := BorderSize, Y := BorderSize}}
        RightBorderColorBlock := color_block{DefaultColor := NamedColors.White, DefaultDesiredSize := vector2{X := BorderSize, Y := BorderSize}}
        BottomBorderColorBlock := color_block{DefaultColor := NamedColors.White, DefaultDesiredSize := vector2{X := BorderSize, Y := BorderSize}}
        LeftBorderColorBlock := color_block{DefaultColor := NamedColors.White, DefaultDesiredSize := vector2{X := BorderSize, Y := BorderSize}}

        SelectedBordersColorBlocks := array{TopBorderColorBlock, RightBorderColorBlock, BottomBorderColorBlock, LeftBorderColorBlock}
        for(ColorBlock : SelectedBordersColorBlocks) { ColorBlock.SetVisibility(widget_visibility.Hidden) }

        VoteStackBoxSlots := if(EnableSuperVotes?):
            array:
                stack_box_slot:
                    Widget := VoteButton
                stack_box_slot:
                    Padding := margin{Top := 10.0}
                    Widget := SuperVoteOverlay
        else:
            array:
                stack_box_slot:
                    Widget := VoteButton

        Overlay := overlay:
            Slots := array:
                overlay_slot:
                    VerticalAlignment := vertical_alignment.Fill
                    HorizontalAlignment := horizontal_alignment.Fill
                    Widget := BackgroundColorBlock
                overlay_slot:
                    VerticalAlignment := vertical_alignment.Bottom
                    HorizontalAlignment := horizontal_alignment.Fill
                    Padding := margin{Top := 50.0}
                    Widget := overlay:
                        Slots := array:
                            overlay_slot:
                                VerticalAlignment := vertical_alignment.Fill
                                HorizontalAlignment := horizontal_alignment.Fill
                                # Padding := margin{Bottom := 70.0}
                                Widget := BackgroundColorBlock2
                            overlay_slot:
                                Padding := margin{Top := 20.0, Bottom := 20.0}
                                VerticalAlignment := vertical_alignment.Center
                                HorizontalAlignment := horizontal_alignment.Center
                                Widget := stack_box:
                                    Orientation := orientation.Vertical
                                    Slots := VoteStackBoxSlots
                overlay_slot:
                    VerticalAlignment := vertical_alignment.Fill
                    HorizontalAlignment := horizontal_alignment.Fill
                    Padding := IconMargin
                    Widget := texture_block:
                        DefaultImage := MinigameMaybe?.GetVotingImage() or UI_UNKNOWN
                overlay_slot:
                    VerticalAlignment := vertical_alignment.Top
                    HorizontalAlignment := horizontal_alignment.Left
                    Padding := margin{Left := -25.0, Top := -25.0}
                    #Padding := margin{Bottom := -50.0} #-50 originally
                    #Padding := margin{Left := -VoteCountSquareSize*0.5, Top := -VoteCountSquareSize*0.5}
                    Widget := VoteCountOverlay
                overlay_slot:
                    VerticalAlignment := vertical_alignment.Top
                    HorizontalAlignment := horizontal_alignment.Fill
                    Widget := TopBorderColorBlock
                overlay_slot:
                    VerticalAlignment := vertical_alignment.Bottom
                    HorizontalAlignment := horizontal_alignment.Fill
                    Widget := BottomBorderColorBlock
                overlay_slot:
                    VerticalAlignment := vertical_alignment.Fill
                    HorizontalAlignment := horizontal_alignment.Right
                    Widget := RightBorderColorBlock
                overlay_slot:
                    VerticalAlignment := vertical_alignment.Fill
                    HorizontalAlignment := horizontal_alignment.Left
                    Widget := LeftBorderColorBlock

        return voting_slot_data:
            Widget := Overlay
            Button := VoteButton
            SuperVoteGoldIcon := SuperVoteGoldIcon
            SuperVoteButton := SuperVoteButton
            VoteCountTextBlock := VoteCountTextBlock
            BackgroundColorBlock := BackgroundColorBlock
            BackgroundColorBlock2 := BackgroundColorBlock2
            SelectedBordersColorBlocks := SelectedBordersColorBlocks