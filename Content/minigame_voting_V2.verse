
using { /Fortnite.com/Devices }
using { /Verse.org/Simulation }
using { /UnrealEngine.com/Temporary/Diagnostics }
using { /Fortnite.com/Characters }
using { /Verse.org/Random }
using { /Verse.org/Colors }
using { /Verse.org/Assets }
using { /UnrealEngine.com/Temporary/UI }
using { /UnrealEngine.com/Temporary/SpatialMath }
using { /Fortnite.com/UI }
using { VerseExposed }
# See https://dev.epicgames.com/documentation/en-us/uefn/create-your-own-device-in-verse for how to create a verse device.

# A Verse-authored creative device that can be placed in a level
minigame_voting_V2 := class(miniagme_voting_system):

    DeviceName<override>:string="Voting System"

    @editable
    TimeBeforeVoting <override>: float = 1.5

    @editable
    VoteDuration <override>: float = 10.0

    @editable
    VoteSlots <override>: minigame_vote_slot_count = minigame_vote_slot_count.Three

    @editable
    EnableSuperVotes <override>: logic = true

    @editable
    SuperVoteGoldPrice <override>: int = 500


    @editable
    SuperVoteHUDDuration <override>: float = 3.0

    @editable
    bArenaSelectUI : ?UI_Voting = false

    var PlayerArenaVotes : [player]?arena_and_modifier_settings = map{}
    var ProposedArenas : [arena_and_modifier_settings]int = map{}
    var bChosenMinigame : ?minigame_device = false
    # OnCrashableDeviceBegin<override>()<suspends>:void=return

    DecideNextArena(bMiniGame : ?minigame_device)<suspends>:void=
        if(ArenaSelectUI := bArenaSelectUI?, MiniGame := bMiniGame?, Arenas := MiniGame.PossibleArenas?, Arenas.Length > 1):
            # Set vote count for all valid proposed arenas
            for(Arena : Arenas, ArenaDevice := Arena.ArenaDevice?){if(set ProposedArenas[Arena] = 0){}}
            set ArenaSelectUI.ValidArenas = Arenas.Length
            Sleep(0.2)
            for(Player : GetPlayspace().GetPlayers(), Player.GetFortCharacter[].IsActive[]):
                UIData := ArenaSelectUI.CreatePlayerUI(Player, ArenaSelectUI.DefaultGroup)
                spawn:
                    ArenaSelectUI.RenderUIOnPlayerScreen(Player, UIData)
            Sleep(ArenaSelectUI.Time * 1.0)
            ArenaSelectUI.RemoveAllUI()
            Sleep(0.0)
            
            var bCurrentVotedArena : ?tuple(arena_and_modifier_settings, int) = false
            var ArenaIndex : int = 1
            var ArenaLogic : int = 1
            for(ProposedArena -> Votes : ProposedArenas, ArenaDevice := ProposedArena.ArenaDevice?):
                #Print("ARENAS to votes {ArenaDevice.Name} {Votes}", ?Duration := 15.0)
                if (CurrentVotedArena := bCurrentVotedArena?):
                    if(CurrentVotedArena(1) < Votes):
                        set bCurrentVotedArena = option{(ProposedArena, Votes)}
                        set ArenaLogic = ArenaIndex
                else:
                    set bCurrentVotedArena = option{(ProposedArena, Votes)}
                    set ArenaLogic = ArenaIndex
                set ArenaIndex += 1
            set MiniGame.ChosenArena = option{bCurrentVotedArena?(0)}
            Print("ArenaLogic = {ArenaLogic}")
            if(ArenaLogic = 1):
                set MiniGame.Arena1 = true
            else if(ArenaLogic = 2):
                set MiniGame.Arena2 = true
            else if(ArenaLogic = 3):
                set MiniGame.Arena3 = true
            else:
                PrintError("ArenaLogic: Should not have got here bro...")
            if(Name := bCurrentVotedArena?(0).ArenaDevice?.Name):
                Print("CHOSEN {Name}", ?Duration := 15.0)

        else if(ArenaSelectUI := bArenaSelectUI?, MiniGame := bMiniGame?, Arenas := MiniGame.PossibleArenas?, Arenas.Length = 1, Arena_1 := Arenas[0]):
            set MiniGame.ChosenArena = option{Arena_1}

    DecideNextMinigame<override>()<suspends>:?minigame_device=
        Sleep(Max(0.5, TimeBeforeVoting)) # Sleep enough to make sure that game_manager is initialized before

        # Show vote UI
        set PickableMinigames = GameManager.DebugMinigames.Length > 0 and GameManager.DebugMinigames or GameManager.Minigames
        set PickableMinigames = for(Minigame : PickableMinigames, not AlreadyPickedMinigames.Find[Minigame]) { Minigame }
        set VoteEndTimestamp  = GetSimulationElapsedTime() + VoteDuration
        set TotalMinigameWeight = 0
        set MinigamesWeightThresholds = map{}
        set PlayerVotes = map{}
        set ProposedMinigames = array{}
        set BuiltCanvas = map{}
        set PlayerArenaVotes = map{}
        set ProposedArenas = map{}
        set bChosenMinigame = false

        ChooseAvailableMinigames()
        DisplayVoteUI()

        # Wait for vote to end somehow
        var SuperVoteData : ?tuple(player, ?minigame_device) = false
        race:
            Sleep(VoteDuration)

            block:
                Data := SuperVoteEvent.Await()
                set SuperVoteData = option{Data}

            # Refresh remaining time text
            loop:
                Sleep(0.9)
                RefreshUIs()

        # Hide UIs
        HideUIs()
        for(Player : GetPlayspace().GetPlayers()):
            HideUI(Player)

        # Handle vote winner (supervote / most voted / fallback function)
        if(Data := SuperVoteData?, Player := Data(0), Minigame := Data(1)? or Shuffle(PickableMinigames)[0]):
            SuperVoteHUDMessage.Show(BuildPlayerMessage("🕹 <Argon>", Player, "</> <RespawnAlertB>SUPERVOTED</> <RespawnAlertBLrg>{Minigame.Name}</>"), ?DisplayTime := SuperVoteHUDDuration)
            Sleep(SuperVoteHUDDuration)
            set AlreadyPickedMinigames += array{Minigame}
            set bChosenMinigame = option{Minigame}
            DecideNextArena(bChosenMinigame)
            return option{Minigame}
        else if(NextGame := (GetBestVotedMinigame[] or Shuffle(PickableMinigames)[0])):
            set AlreadyPickedMinigames += array{NextGame}
            set bChosenMinigame = option{NextGame}
            DecideNextArena(bChosenMinigame)
            return option{NextGame}
        else:
            return false

    GetArenaVotes()<transacts>:[?arena_and_modifier_settings]int=
        var Votes : [?arena_and_modifier_settings]int = map{}
        for(ArenaMaybe : PlayerArenaVotes):
            if(not Votes[ArenaMaybe]):
                option{set Votes[ArenaMaybe] = 0}

            option{set Votes[ArenaMaybe] += 1}
        return Votes
    

    HandlePlayerVote<override>(Player: player, MinigameMaybe: ?minigame_device):void=
        option{set PlayerVotes[Player] = MinigameMaybe}
        RefreshUIs()
        RefreshPlayerUI(Player)

    CanSuperVote<override>(Player: player)<transacts><decides>:void=
        GameManager.GoldTracker.GetValue(Player) >= SuperVoteGoldPrice

    HandlePlayerSuperVote<override>(Player: player, MinigameMaybe: ?minigame_device):void=
        if(CanSuperVote[Player]):
            GameManager.GoldTracker.Add(Player, -SuperVoteGoldPrice)
            SuperVoteEvent.Signal(Player, MinigameMaybe)

    HandlePlayerExit<override>(WidgetMessage: widget_message):void=
        spawn{HideUI(WidgetMessage.Player)}

    HandleArenaVote_1(WidgetMessage: widget_message):void=
        Player := WidgetMessage.Player
        Print("Voted For 1", ?Duration := 15.0)
        # Remove players vote on the previous chosen map
        if (bArena := PlayerArenaVotes[Player], Arena := bArena?):
            if(set ProposedArenas[Arena] -= 1):
        if (ChosenMinigame := bChosenMinigame?, Arenas := ChosenMinigame.PossibleArenas?, VotedArena := Arenas[0]):
            if(set PlayerArenaVotes[Player] = option{VotedArena}):
                # Add players vote to the arena map
                if(set ProposedArenas[VotedArena] += 1):
        # Highlight chosen vote screen for player
        if (ArenaSelectUI := bArenaSelectUI?):
            ArenaSelectUI.HideGroupUI(Player, E_VoteMenu_Group.ArenaHighlights.GetName(), true)
            # Highlight screen 1
            ArenaSelectUI.ShowGroupUI(Player, E_VoteMenu_Group.ArenaHighlight1.GetName(), false)
        UpdatePlayerVoteScreen()

    HandleArenaVote_2(WidgetMessage: widget_message):void=
        Player := WidgetMessage.Player
        # Remove players vote on the previous chosen map
        Print("Voted For 2", ?Duration := 15.0)
        if (bArena := PlayerArenaVotes[Player], Arena := bArena?):
            if(set ProposedArenas[Arena] -= 1):
        if (ChosenMinigame := bChosenMinigame?, Arenas := ChosenMinigame.PossibleArenas?, VotedArena := Arenas[1]):
            if(set PlayerArenaVotes[Player] = option{VotedArena}):
                # Add players vote to the arena map
                if(set ProposedArenas[VotedArena] += 1):
        # Highlight chosen vote screen for player
        if (ArenaSelectUI := bArenaSelectUI?):
            ArenaSelectUI.HideGroupUI(Player, E_VoteMenu_Group.ArenaHighlights.GetName(), true)
            # Highlight screen 2
            ArenaSelectUI.ShowGroupUI(Player, E_VoteMenu_Group.ArenaHighlight2.GetName(), false)
        UpdatePlayerVoteScreen()

    HandleArenaVote_3(WidgetMessage: widget_message):void=
        Player := WidgetMessage.Player
        # Remove players vote on the previous chosen map
        Print("Voted For 3", ?Duration := 15.0)
        if (bArena := PlayerArenaVotes[Player], Arena := bArena?):
            if(set ProposedArenas[Arena] -= 1):
        # Add players vote to the chosen arena map
        if (ChosenMinigame := bChosenMinigame?, Arenas := ChosenMinigame.PossibleArenas?, VotedArena := Arenas[2]):
            if(set PlayerArenaVotes[Player] = option{VotedArena}):
                # Add players vote to the arena map
                if(set ProposedArenas[VotedArena] += 1):
        # Highlight chosen vote screen for player
        if (ArenaSelectUI := bArenaSelectUI?):
            ArenaSelectUI.HideGroupUI(Player, E_VoteMenu_Group.ArenaHighlights.GetName(), true)
            # Highlight screen 3
            ArenaSelectUI.ShowGroupUI(Player, E_VoteMenu_Group.ArenaHighlight3.GetName(), false)

        UpdatePlayerVoteScreen()

    UpdatePlayerVoteScreen():void=
        if (ArenaSelectUI := bArenaSelectUI?, ChosenMinigame := bChosenMinigame?, Arenas := ChosenMinigame.PossibleArenas?):
            if(Arena_1 := Arenas[0], Votes := ProposedArenas[Arena_1]):
                ArenaSelectUI.SetTextAllPlayers(E_VoteMenu_Widget.ArenaVoteScreen1_Txt_VoteAmmount.GetName(), "{Votes}")
            if(Arena_2 := Arenas[1], Votes := ProposedArenas[Arena_2]):
                ArenaSelectUI.SetTextAllPlayers(E_VoteMenu_Widget.ArenaVoteScreen2_Txt_VoteAmmount.GetName(), "{Votes}")
            if(Arena_3 := Arenas[2], Votes := ProposedArenas[Arena_3]):
                ArenaSelectUI.SetTextAllPlayers(E_VoteMenu_Widget.ArenaVoteScreen3_Txt_VoteAmmount.GetName(), "{Votes}")

     
    
    