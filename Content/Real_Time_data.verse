
using { /Fortnite.com/Devices }
using { /Verse.org/Simulation }
using { /UnrealEngine.com/Temporary/Diagnostics }
using { /Verse.org/Random }

Time_And_Date := class<final><persistable><concrete>:
    @editable
    Years : int = -1
    @editable
    Months : int = -1
    @editable
    Days : int = -1
    @editable
    Hours : int = -1
    @editable
    Minutes : int = -1
    @editable
    Seconds : int = -1


SECONDS_PER_MINUTE: int = 60
SECONDS_PER_HOUR: int = 3600
SECONDS_PER_DAY: int = 86400
SECONDS_PER_MONTH: int = 30 * SECONDS_PER_DAY  # Assuming 30 days per month
SECONDS_PER_YEAR: int = 365 * SECONDS_PER_DAY  # Assuming 365 days per year

MakeTimeAndDate<constructor>(OldTime : Time_And_Date)<transacts>:= 
    Time_And_Date:
        Years := OldTime.Years
        Months := OldTime.Months
        Days := OldTime.Days
        Hours := OldTime.Hours
        Minutes := OldTime.Minutes
        Seconds := OldTime.Seconds

MakeTimeAndDateFromProp(Prop : creative_prop):?Time_And_Date=
    Transform := Prop.GetTransform()
    Translation := Transform.Translation
    Scale := Transform.Scale
    
    if:
        TimeDate := Time_And_Date:
            Years := Floor[Translation.X]
            Months := Floor[Translation.Y]
            Days := Floor[Translation.Z]
            Hours := Floor[Scale.X] - 1
            Minutes := Floor[Scale.Y] - 1
            Seconds := Floor[Scale.Z] - 1
    then:
        return option{MakeTimeAndDate(TimeDate)}

    return false

# Function to convert Time_And_Date instance to total seconds
TotalSeconds(Time: Time_And_Date)<transacts>: int=
    var total_seconds : int = 0
    set total_seconds += Time.Seconds
    set total_seconds += Time.Minutes * SECONDS_PER_MINUTE
    set total_seconds += Time.Hours * SECONDS_PER_HOUR
    set total_seconds += Time.Days * SECONDS_PER_DAY
    set total_seconds += Time.Months * SECONDS_PER_MONTH
    set total_seconds += Time.Years * SECONDS_PER_YEAR
    return total_seconds

# Function to calculate the difference in hours between two Time_And_Date instances
GetHoursDifference(Time1 : Time_And_Date, Time2 : Time_And_Date) : ?int=
    total_seconds1 := TotalSeconds(Time1)
    total_seconds2 := TotalSeconds(Time2)
    difference_in_seconds := total_seconds2 - total_seconds1
    difference_in_hours_float := difference_in_seconds * 1.0 / SECONDS_PER_HOUR * 1.0
    if(difference_in_hours := Floor[difference_in_hours_float]):
        return option{difference_in_hours}
    return false

# Function to calculate the difference in hours between two Time_And_Date instances
GetHoursDifference(total_seconds1 : int, total_seconds2 : int) : int=
    difference_in_seconds := total_seconds2 - total_seconds1
    difference_in_hours_float := (difference_in_seconds * 1.0) / (SECONDS_PER_HOUR * 1.0)
    if(difference_in_hours := Floor[difference_in_hours_float]):
        return difference_in_hours
    return -1




# A Verse-authored creative device that can be placed in a level
V_Real_Time_data := class(creative_device):

    var bCurrentTimeDate : ?Time_And_Date = false
    @editable
    bRealTimeProp : ?creative_prop = false

    S2M<localizes>(Value:string):message =  "{Value}"
    #@editable Date : billboard_device = billboard_device{}
    #@editable Time : billboard_device = billboard_device{}
    DateTimeProvider : Date_Time_Provider = Date_Time_Provider{}
    var CurrentDate : date_time = date_time{year := 0, month := 0, day := 0, hour := 0, minute := 0, second := 0}

    # Runs when the device is started in a running game
    OnBegin<override>()<suspends>:void=
        # TODO: Replace this with your code
        spawn:
            TimerTick()

        loop:
            set CurrentDate = DateTimeProvider.GetDateAndTime()
            MonthName := DateTimeProvider.GetMonthName(CurrentDate.month) 
            #Date.SetText(S2M("{MonthName} {DateTime.day}, {DateTime.year}"))
            #Time.SetText(S2M("UTC: {DateTime.hour}:{DateTime.minute}:{DateTime.second}"))
            Sleep(1.0)

    TimerTick()<suspends>:void=
        loop:
            Sleep(1.0)
            if (RealTimeProp := bRealTimeProp?):
                set bCurrentTimeDate = MakeTimeAndDateFromProp(RealTimeProp)





## PURE VERRSE
date_time := struct:
    year<public>: int
    month<public>: int
    day<public>: int
    hour<public>: int
    minute<public>: int
    second<public>: int

Date_Time_Provider := class:  
    #Made By Himez 
    #Discord : Himez
    #Twitter : @HimezUEFN
    #This class is used to get the current date and time from the seconds since epoch

    # Constants
    SecondsInMinute: int= 60
    SecondsInHour: int= 3600
    SecondsInDay: int= 86400
    SecondsInYear: int= 31536000
    SecondsInLeapYear: int= 31622400

    # Init date and time variables
    var Year: int= 1970
    var Month: int= 1
    var Day: int= 1
    var Hour: int= 0
    var Minute: int= 0
    var Second: int= 0
    var EpochSeconds: int= 0
    
    # Get the current seconds since epoch
    GetMonthName<public>(MonthInt:int):string=
        Months : []string = array{"January", "February", "March", "April", "May", "June", "July", "August", "September", "October", "November", "December"}
        if. MonthName := Months[MonthInt - 1] then. return MonthName  else. return "Invalid Month"

    #Get Date and Time method
    GetDateAndTime<public>():date_time=
        ResetDateAndTime()
        CurrentEpochsSeconds  :=  GetSecondsSinceEpoch()
        if.  Temp := Floor[CurrentEpochsSeconds] then. set EpochSeconds = Temp
        CalculateDateAndTime()
        return date_time{year:= Year, month:= Month, day:= Day, hour:= Hour, minute:= Minute, second:= Second}

    #Get Date and Time method
    GetDateAndTime<public>(Seconds : float):date_time=
        ResetDateAndTime()
        CurrentEpochsSeconds  :=  Seconds
        if.  Temp := Floor[CurrentEpochsSeconds] then. set EpochSeconds = Temp
        CalculateDateAndTime()
        return date_time{year:= Year, month:= Month, day:= Day, hour:= Hour, minute:= Minute, second:= Second}

    # caculate day and time
    CalculateDateAndTime<private>():void=
        CaculateYear()
        CalculateDay()
        CalculateMonth()
        CalculateHour()
        CalculateMinute()
        CalculateSecond()

    # Reset date and time
    ResetDateAndTime<private>():void=
        set Year = 1970
        set Month = 1
        set Day = 1
        set Hour = 0
        set Minute = 0
        set Second = 0
        set EpochSeconds = 0

    # Determine the year
    CaculateYear<private>():void =
        loop:
            LeapYear := IsLeapYear(Year)
            SecondsInCurrentYear := if. LeapYear? then. SecondsInLeapYear else. SecondsInYear
            if (EpochSeconds >= SecondsInCurrentYear):
                set EpochSeconds -= SecondsInCurrentYear
                set Year += 1
            else:
                break

    #Helper function to determine if a year is a leap year
    IsLeapYear<private>(YearInt: int):logic=
        if. Mod[YearInt , 4 ] = 0 and Mod[YearInt ,100 ] <> 0 or Mod[ YearInt , 400] = 0 then. return true else. return false
    
    # Calculate day
    CalculateDay<private>(): void =
        set Day = if. Temp := Floor(EpochSeconds / SecondsInDay) then. Temp + 1 else. 1
        if. Temp := Mod[EpochSeconds, SecondsInDay] then. set EpochSeconds = Temp  # Calculate  remaining seconds for time of day 

    # Calculate month
    CalculateMonth<private>(): void = 
        var DaysInMonth: []int = array{31, 28, 31, 30, 31, 30, 31, 31, 30, 31, 30, 31}
        LeapYear := IsLeapYear(Year) # Check if it is a leap year
        if. LeapYear? then. if. set DaysInMonth[1] = 29  # Adjust February for leap year
        loop:
            if (Day > DaysInMonth[Month - 1]):
                if. set Day -= DaysInMonth[Month - 1]
                set Month += 1       
            else:
                break
            
    # Calculate hour
    CalculateHour<private>():void=
        set Hour = if. Temp := Floor(EpochSeconds / SecondsInHour) then. Temp else. 0
        set EpochSeconds = if . Temp := Mod[EpochSeconds, SecondsInHour] then. Temp else. 0 # Calculate remaining secondsq

    # Calculate minute
    CalculateMinute<private>():void=
        set Minute = if. Temp := Floor(EpochSeconds / SecondsInMinute) then. Temp else. 0
        set EpochSeconds = if. Temp := Mod[EpochSeconds, SecondsInMinute] then. Temp else. 0 # Calculate remaining seconds

    # Remaining seconds
    CalculateSecond<private>():void=
        set Second += EpochSeconds
        set EpochSeconds = 0



