
using { /Fortnite.com/Devices }
using { /Verse.org/Simulation }
using { /UnrealEngine.com/Temporary/Diagnostics }
using { /Fortnite.com/Characters }
using { /Fortnite.com/Teams }
using { /Verse.org/Random }
# See https://dev.epicgames.com/documentation/en-us/uefn/create-your-own-device-in-verse for how to create a verse device.

# A Verse-authored creative device that can be placed in a level
teamselection_manager := class(crashable_device):
    @editable
    Team1ClassSelector : class_and_team_selector_device = class_and_team_selector_device{}
    @editable
    Team2ClassSelector : class_and_team_selector_device = class_and_team_selector_device{}
    @editable
    Team3ClassSelector : class_and_team_selector_device = class_and_team_selector_device{}
    @editable
    Team4ClassSelector : class_and_team_selector_device = class_and_team_selector_device{}
    @editable
    NoTeamClassSelector : class_and_team_selector_device = class_and_team_selector_device{}

    @editable
    Team1Volume : volume_device = volume_device{}
    @editable
    Team2Volume : volume_device = volume_device{}
    @editable
    Team3Volume : volume_device = volume_device{}
    @editable
    Team4Volume : volume_device = volume_device{}

    @editable
    LobbyTeleporter : teleporter_device = teleporter_device{}

    @editable
    PlayerCounterTeam1 : player_counter_device = player_counter_device{}
    @editable
    PlayerCounterTeam2 : player_counter_device = player_counter_device{}
    @editable
    PlayerCounterTeam3 : player_counter_device = player_counter_device{}
    @editable
    PlayerCounterTeam4 : player_counter_device = player_counter_device{}
    @editable
    TeamSizeLimit : int = 4

    @editable
    Team1ClassSelectorVerseOnly : class_and_team_selector_device = class_and_team_selector_device{}
    @editable
    Team2ClassSelectorVerseOnly : class_and_team_selector_device = class_and_team_selector_device{}
    @editable
    Team3ClassSelectorVerseOnly : class_and_team_selector_device = class_and_team_selector_device{}
    @editable
    Team4ClassSelectorVerseOnly : class_and_team_selector_device = class_and_team_selector_device{}

    var TeamTypeForGameMode : team_type_enum = team_type_enum.FourTeams

    OnCrashableDeviceBegin<override><final>()<suspends>:void=
        Team1Volume.AgentEntersEvent.Subscribe(Team1VolumeEvent)
        Team2Volume.AgentEntersEvent.Subscribe(Team2VolumeEvent)
        Team3Volume.AgentEntersEvent.Subscribe(Team3VolumeEvent)
        Team4Volume.AgentEntersEvent.Subscribe(Team4VolumeEvent)

    AutomaticallySortTeams(TeamTypeEnum : team_type_enum):void=
        AllPlayersInLobby := Shuffle(GetPlayspace().GetPlayers())
        TeamsCollection := GetPlayspace().GetTeamCollection()
        Teams := TeamsCollection.GetTeams()

        #Set the TeamTypeForGameMode so we can use it in other functions
        set TeamTypeForGameMode = TeamTypeEnum

        #We always want to sort the teams and then split it up if there is only 1 team in the game
        PlaceUnsortedPlayersIntoTeams()
        #4V4V4V4 - Four Teams
        if(TeamTypeEnum = team_type_enum.FourTeams):
            SplitSingleTeam()

        #8V8 - Two Teams
        if(TeamTypeEnum = team_type_enum.TwoTeams):
            CombineBothTeams()
            SplitSingleTeam()

        #4V12 - OneVsThreeTeams [Work in progress, this is for later so we have something to work off]
        if(TeamTypeEnum = team_type_enum.OneVsThreeTeams):
            CombineBothTeams()
            SplitSingleTeam()
            #Store Team 1 and Team 2 before we use them for the class selector
            if(Team1 := Teams[0], Team2 := Teams[1]):       
                if(TeamAgents1 := TeamsCollection.GetAgents[Team1], TeamAgents2 := TeamsCollection.GetAgents[Team2]):
                    #After moving all players then we check which team is bigger
                    if(TeamAgents1.Length > TeamAgents2.Length):
                        Print("Team 1 is bigger, swap them because the runner team should be bigger")
                        #Swap Team        
                        for(AgentTeam1 : TeamAgents1):
                            Team2ClassSelectorVerseOnly.ChangeTeam(AgentTeam1)
                        for(AgentTeam2 : TeamAgents2):
                            Team1ClassSelectorVerseOnly.ChangeTeam(AgentTeam2)
                    else:
                        Print("Team 2 is bigger, doing nothing as team 2 are runners")

    #This code splits a single team up in the instance there is only 1 team in the game
    SplitSingleTeam():void=
        TeamsCollection := GetPlayspace().GetTeamCollection()
        Teams := TeamsCollection.GetTeams()
        #Split the teams
        TeamsInGameResult := GetOnlyOneTeamInGame()
        #Only split teams if there is one team in the game
        if(TeamsInGameResult(0) = true):
            Print("Split The Teams")
            if(Team := Teams[TeamsInGameResult(1)], TeamAgents := TeamsCollection.GetAgents[Team]):
                var TeamSize : int = 0
                #Agents in the team that are valid
                var AgentsInTeam : []agent = array{}
                #Actual agents to move into a different team
                var AgentsToMove : []agent = array{}
                #Store active fort characters
                for(TeamAgent : TeamAgents):
                    if(TeamAgent.GetFortCharacter[].IsActive[]):
                        set TeamSize +=1
                        set AgentsInTeam += array{TeamAgent}
                Print("Players In Game = {TeamSize}")

                #If there are only two players, move the first player from blue to red
                if(TeamSize = 2):
                    if(set AgentsToMove += array{AgentsInTeam[0]}):
                else:
                    #If there are more then two players in the team we add those players to the Agents To Move array
                    if(Group := Ceil(AgentsInTeam.Length / 2)):
                        Print("Split Size = {Group}")
                    for(index -> Agent : AgentsInTeam):
                        if(index > Ceil(AgentsInTeam.Length / 2)-1):
                            #Print("Moving Players into another team {Floor(AgentsInTeam.Length / 2)}")
                            set AgentsToMove += array{Agent}
                    
                #Now we Move the agents in AgentsToMove into a new team
                for(AgentToMove : AgentsToMove):
                    #If the players are on Team 2 move then to team 1
                    if(TeamsInGameResult(1) = 1):
                        Team1ClassSelector.ChangeTeam(AgentToMove)
                    #Move them to team 2 if they are on any other team
                    else:
                        Team2ClassSelector.ChangeTeam(AgentToMove)
    
    PlaceUnsortedPlayersIntoTeams():void=
        TeamsCollection := GetPlayspace().GetTeamCollection()
        Teams := TeamsCollection.GetTeams()
        # Assign players with no team to teams
        var LobbyTeamPlayers : []player = array{}

        AllPlayersInLobby := GetPlayspace().GetPlayers()
        Print("Players in lobby = {AllPlayersInLobby.Length}")
        for(Player : AllPlayersInLobby):
            if(Player.GetFortCharacter[].IsActive):
                if (PlayersTeam := TeamsCollection.GetTeam[Player]):
                    if(TeamsCollection.IsOnTeam[Player, Teams[4]] or TeamsCollection.IsOnTeam[Player, Teams[5]] or TeamsCollection.IsOnTeam[Player, Teams[6]]):
                        Print("Found player on team 5,6 or 7")
                        set LobbyTeamPlayers += array{Player}

        #The player team counter is disabling the team selectors!!
        #This is why only 4 of us can go through.
        # Iterate through the agents with no team and assign them to a team
        for(Player : LobbyTeamPlayers):
            RandomAssignOrder := GetRandomInt(0,3)
            Print("Adding Player to Team {RandomAssignOrder}")
            if(Player.GetFortCharacter[].IsActive):
                case(RandomAssignOrder):
                    3 =>
                        if(PlayerCounterTeam4.GetCount() < 5):
                            Team4ClassSelectorVerseOnly.ChangeTeam(Player)       
                        else if(PlayerCounterTeam3.GetCount() < 5):
                                Team3ClassSelectorVerseOnly.ChangeTeam(Player)
                        else if(PlayerCounterTeam2.GetCount() < 5):
                                Team2ClassSelectorVerseOnly.ChangeTeam(Player)
                        else if(PlayerCounterTeam1.GetCount() < 5):
                                Team1ClassSelectorVerseOnly.ChangeTeam(Player)
                    2 =>
                        if(PlayerCounterTeam2.GetCount() < 5):
                                Team2ClassSelectorVerseOnly.ChangeTeam(Player)
                        else if(PlayerCounterTeam3.GetCount() < 5):
                                Team3ClassSelectorVerseOnly.ChangeTeam(Player)
                        else if(PlayerCounterTeam1.GetCount() < 5):
                                Team1ClassSelectorVerseOnly.ChangeTeam(Player)
                        else if(PlayerCounterTeam4.GetCount() < 5):
                                Team4ClassSelectorVerseOnly.ChangeTeam(Player)
                    1 =>
                        if(PlayerCounterTeam1.GetCount() < 5):
                            Team1ClassSelectorVerseOnly.ChangeTeam(Player)
                        else if(PlayerCounterTeam2.GetCount() < 5):
                                Team2ClassSelectorVerseOnly.ChangeTeam(Player)
                        else if(PlayerCounterTeam3.GetCount() < 5):
                                Team3ClassSelectorVerseOnly.ChangeTeam(Player)
                        else if(PlayerCounterTeam4.GetCount() < 5):
                                Team4ClassSelectorVerseOnly.ChangeTeam(Player)
                    0 =>
                        if(PlayerCounterTeam3.GetCount() < 5):
                            Team3ClassSelectorVerseOnly.ChangeTeam(Player)
                        else if(PlayerCounterTeam2.GetCount() < 5):
                                Team2ClassSelectorVerseOnly.ChangeTeam(Player)
                        else if(PlayerCounterTeam1.GetCount() < 5):
                                Team1ClassSelectorVerseOnly.ChangeTeam(Player)
                        else if(PlayerCounterTeam4.GetCount() < 5):
                                Team4ClassSelectorVerseOnly.ChangeTeam(Player)
                    _=>
                        if(PlayerCounterTeam1.GetCount() < 5):
                            Team1ClassSelectorVerseOnly.ChangeTeam(Player)
                        if(PlayerCounterTeam2.GetCount() < 5):
                            Team2ClassSelectorVerseOnly.ChangeTeam(Player)
                        if(PlayerCounterTeam3.GetCount() < 5):
                            Team3ClassSelectorVerseOnly.ChangeTeam(Player)
                        if(PlayerCounterTeam4.GetCount() < 5):
                            Team4ClassSelectorVerseOnly.ChangeTeam(Player)
                
                    
    #Function that can combine both teams. Moved into a function as we re-use this for different team layouts
    CombineBothTeams():void=
        TeamsCollection := GetPlayspace().GetTeamCollection()
        Teams := TeamsCollection.GetTeams()
        CheckForMoreThenTwoTeams := GetOnlyOneTeamInGame()
        #We check if you are on Team 3 or Team 4 then add you to Team 1 or Team 2. If everyone for some reason ends up on
        #The same team the split code will fix that.
        #Because balance is not important we don't need to have any balancing functions
        AllPlayersInLobby := GetPlayspace().GetPlayers()
        for(Player : AllPlayersInLobby):
            if(Player.GetFortCharacter[].IsActive):
                if(TeamsCollection.IsOnTeam[Player, Teams[2]]):
                    Team1ClassSelectorVerseOnly.ChangeTeam(Player)
                if(TeamsCollection.IsOnTeam[Player, Teams[3]]):
                    Team2ClassSelectorVerseOnly.ChangeTeam(Player)

    #Checks to try and give you a team that is not maxed out by sorting through all of the teams. If everything fails you go onto team 1.
    GetTeamNotMaxed():int=
        Teams := GetPlayspace().GetTeamCollection().GetTeams()
        # Get first team with 0 players, if that fails get first team that isn't full
        for(Index -> Team : Teams):
            var TeamSize : int = 0
            #Index 0,1,2 and 3 are teams 1,2,3 and 4. We don't want to return an index for 4,5 or 6
            if(Index <> 4 and Index <> 5 and Index <> 6):
                Print("Player is not on team 4,5 or 6")
                if(TeamAgents := GetPlayspace().GetTeamCollection().GetAgents[Team]):
                    #Loop through the team and for each active member increase team size by 1
                    for(Agent : TeamAgents, Agent.GetFortCharacter[].IsActive[]):
                        Print("Increment team size")
                        set TeamSize +=1
                #If the team is empty return the index
                if(TeamSize = 0):  
                    Print("Found empty team to assign")
                    return Index
                #If the team size is less then or equal to 4 then return index
                else if(TeamSize <= 4):
                    Print("Found team with players but it's less then 4")
                    return Index
        #If everything fails we just return 0 which is team 1. We know somethings messed up if most players go into team 1
        return 0

    #This function checks if there is only one team in the game with players.
    GetOnlyOneTeamInGame():tuple(logic, int, []logic, int)=
        var Result : logic = false
        var TeamResults : []logic = array{false,false,false,false}
        var TeamsWithMembers : int = 0
        var TeamWithAllPlayersIndex : int = 0
        Teams := GetPlayspace().GetTeamCollection().GetTeams()

        #Go through all teams and replace the array element at that index with false instead of true if an agent exists on that team
        for(Index -> Team : Teams, TeamAgents := GetPlayspace().GetTeamCollection().GetAgents[Team]):
            for(TeamAgent : TeamAgents):
                if(TeamAgent.GetFortCharacter[].IsActive):
                    if(ResultToSet := TeamResults.ReplaceElement[Index, true]):
                        set TeamResults = ResultToSet

        #For each Team Result being true we increment by 1
        for(index -> TeamResult : TeamResults):
            if(TeamResult = true):
                set TeamsWithMembers +=1

        #We then check if Team with members is more then 1, if it isn't then only one team exists and the result is true
        if(TeamsWithMembers > 1):
            set Result = false
        else:
            set Result = true
            #This function is only useful when there is one team
            if(TeamWithPlayers := TeamResults.Find[true]):
                set TeamWithAllPlayersIndex = TeamWithPlayers
                Print("Team {TeamWithAllPlayersIndex + 1} has players")

        return (Result, TeamWithAllPlayersIndex, TeamResults, TeamsWithMembers)

    #Checks if player is already on team
    CheckIfOnTeamAlready(SelectedTeamIdnex : int, InAgent : agent) : void=
        Teams := GetPlayspace().GetTeamCollection().GetTeams()
        Print("Team Count: {Teams.Length}")

        if(GetPlayspace().GetTeamCollection().IsOnTeam[InAgent, Teams[SelectedTeamIdnex]]):
            Print("Is On Team Already")
            LobbyTeleporter.Teleport(InAgent)

    <#
    ##### MUTATOR ZONE ENTERED EVENTS #####
    Team1SwitchEvent(InAgent : agent):void=
        CheckTeamCount(1, InAgent)
        Print("Change to Team 1")
    
    Team2SwitchEvent(InAgent : agent):void=
        CheckTeamCount(2, InAgent)
        Print("Change to Team 2")

    Team3SwitchEvent(InAgent : agent):void=
        CheckTeamCount(3, InAgent)
        Print("Change to Team 3")

    Team4SwitchEvent(InAgent : agent):void=
        CheckTeamCount(4, InAgent)
        Print("Change to Team 4")
    #>
    ##### VOLUME ZONE ENTERED EVENTS #####
    Team1VolumeEvent(InAgent : agent):void=
        CheckIfOnTeamAlready(0, InAgent)
        Print("Entered Team 1")

    Team2VolumeEvent(InAgent : agent):void=
        CheckIfOnTeamAlready(1, InAgent)
        Print("Entered Team 2")

    Team3VolumeEvent(InAgent : agent):void=
        CheckIfOnTeamAlready(2, InAgent)
        Print("Entered Team 3")

    Team4VolumeEvent(InAgent : agent):void=
        CheckIfOnTeamAlready(3, InAgent)
        Print("Entered Team 4")