using { /Fortnite.com/Devices }
using { /Verse.org/Simulation }
using { /UnrealEngine.com/Temporary/Diagnostics }
using { /Verse.org/Simulation/Tags }

team_lives := class(tag){}

team_lives_ui_manager := class(creative_device):
    @editable
    HUDTeams: []hud_message_device = array{}

    @editable
    HUDTeamCrowns : []hud_message_device = array{}

    @editable
    HUDTeamBeds : []hud_message_device = array{}

    @editable
    var bMinigame : ?minigame_device = false

    ShowCrownUI(MiniGameDevice : minigame_device)<suspends>:void=
        set bMinigame = option{MiniGameDevice}
        loop:
            Sleep(0.2)
            if (Minigame := bMinigame?):
                Teams := Minigame.GetRemainingTeams()
                var ValidTeam : [int]custom_team = map{}
                for(Team:Teams):
                    if(set ValidTeam [Team.TeamIndex] = Team):

                for(i -> HUDCrown : HUDTeamCrowns):
                    if (Team := ValidTeam[i]):
                        if(Team.AllowSpawning = true):
                            PlayersLength := Team.Players.Length
                            HUDCrown.Show()
                            #Print("You are on team {i}")
                        else:
                            HUDCrown.Hide()
                    else:
                        HUDCrown.Hide()
            else:
                break

    ShowBedUI(MiniGameDevice : minigame_device)<suspends>:void=
        set bMinigame = option{MiniGameDevice}
        loop:
            Sleep(0.2)
            if (Minigame := bMinigame?):
                Teams := Minigame.GetRemainingTeams()
                var ValidTeam : [int]custom_team = map{}
                for(Team:Teams):
                    if(set ValidTeam [Team.TeamIndex] = Team):

                for(i -> HUDBed : HUDTeamBeds):
                    if (Team := ValidTeam[i]):
                        if(Team.AllowSpawning = true):
                            PlayersLength := Team.Players.Length
                            HUDBed.Show()
                            #Print("You are on team {i}")
                        else:
                            HUDBed.Hide()
                    else:
                        HUDBed.Hide()
            else:
                break

    ResetCrownUI():void=
        for(i -> HUDCrown : HUDTeamCrowns):
            HUDCrown.Hide()

    ResetBedUI():void=
        for(i -> HUDBed : HUDTeamBeds):
            HUDBed.Hide()

    ShowTeamUI(MiniGameDevice : minigame_device)<suspends>:void=
        set bMinigame = option{MiniGameDevice}
        loop:
            Sleep(0.2)
            if (Minigame := bMinigame?):
                Teams := Minigame.GetRemainingTeams()
                var ValidTeam : [int]custom_team = map{}
                for(Team:Teams):
                    if(set ValidTeam [Team.TeamIndex] = Team):

                for(i -> HUDTeam : HUDTeams):
                    if (Team := ValidTeam[i]):
                        PlayersLength := Team.Players.Length
                        HUDTeam.SetText(StringToMessage("{PlayersLength}"))
                        HUDTeam.Show()
                        #Print("You are on team {i}")
                    else:
                        HUDTeam.Hide()
            else:
                break

    HideTeamUI():void=
        set bMinigame = false
        for(i -> HUDTeam : HUDTeams):
            HUDTeam.Hide()