
using { /Fortnite.com/Devices }
using { /Verse.org/Simulation }
using { /UnrealEngine.com/Temporary/Diagnostics }
using { /Verse.org/Simulation/Tags }

team_scores := class(tag){}

team_score_ui_manager := class(creative_device):

    @editable
    HUDTeams: []hud_message_device = array{}

    @editable
    var bMinigame : ?minigame_device = false

    # Runs when the device is started in a running game
    OnBegin<override>()<suspends>:void=
        Sleep(1.0)

    ShowTeamUI(MiniGameDevice : minigame_device)<suspends>:void=
        var TeamScore : int = 0

        set bMinigame = option{MiniGameDevice}
        loop:
            Sleep(0.2)
            if (Minigame := bMinigame?, ScoreMinigame := Minigame.ScoreLogic?):
                Teams := Minigame.GetRemainingTeams()
                var ValidTeam : [int]custom_team = map{}
                for(Team:Teams):
                    if(set ValidTeam [Team.TeamIndex] = Team):

                for(i -> HUDTeam : HUDTeams):
                    if (Team := ValidTeam[i]):
                        ScoreToWin := ScoreMinigame.ScoreToWin

                        if(ScoreMinigame.CustomScoring = true):
                            set TeamScore = Team.CurrentScore
                        else:
                            set TeamScore = Team.Kills

                        if(ScoreMinigame.ShowScoreToWin = true):
                            HUDTeam.SetText(StringToMessage("{TeamScore} / {ScoreToWin}"))
                        else:
                            HUDTeam.SetText(StringToMessage("{TeamScore}"))
                        HUDTeam.Show()
                    else:
                        HUDTeam.Hide()
            else:
                break

    HideTeamUI():void=
        set bMinigame = false
        for(i -> HUDTeam : HUDTeams):
            HUDTeam.Hide()