using { /Fortnite.com/Devices }
using { /Fortnite.com/Characters }
using { /Fortnite.com/FortPlayerUtilities }
using { /Verse.org/Simulation }
using { /Verse.org/Random }
using { /Verse.org/Assets }
using { /Verse.org/Colors }
using { /UnrealEngine.com/Temporary/Diagnostics }
using { /UnrealEngine.com/Temporary/UI }
using { /UnrealEngine.com/Temporary/SpatialMath }

var UIFaderInstances : weak_map(session, ui_fader) = map{}
GetUIFaderInstance()<transacts>:ui_fader=
    Session := GetSession()
    if(not UIFaderInstances[Session]):
        option{set UIFaderInstances[Session] = ui_fader{}}
    UIFaderInstances[Session] or ui_fader{} # Can't fail

var PlayerFaderStoppedEvents : weak_map(session, [player]event()) = map{}
(Player: player).FaderStoppedEvent():event() =
    Session := GetSession()
    if(not PlayerFaderStoppedEvents[Session]):
        option{set PlayerFaderStoppedEvents[Session] = map{}}

    if(not PlayerFaderStoppedEvents[Session][Player]):
        option{set PlayerFaderStoppedEvents[Session][Player] = event(){}}

    if(Event := PlayerFaderStoppedEvents[Session][Player]):
        return Event
    return event(){}

ui_fader_fade_data := class:
    Widget : widget
    ColorBlock : color_block

    # true = fade in, false = fade out
    var Direction : logic = true

    var Animating : logic = false
    var AnimationDuration : float = 1.0

ui_fader := class:

    var PlayerFadeData : [player]ui_fader_fade_data = map{}

    var Initialized : logic = false

    Init(CreativeDevice: creative_device):void=
        if(Initialized?):
            return
        set Initialized = true

        CreativeDevice.GetPlayspace().PlayerAddedEvent().SubscribeSuspends(OnPlayerJoined)
        CreativeDevice.GetPlayspace().PlayerRemovedEvent().SubscribeSuspends(OnPlayerLeft)
        for(Player : CreativeDevice.GetPlayspace().GetPlayers()):
            spawn{OnPlayerJoined(Player)}

    # Add the canvas right away so that fading will be more responsive
    OnPlayerJoined(Player: player)<suspends>:void=
        if(PlayerFadeData[Player]):
            return

        FadeData := GetOrMakePlayerFadeData(Player)
        race:
            Sleep(60.0)
            loop:
                if(PlayerUI := GetPlayerUI[Player]):
                    PlayerUI.AddWidget(FadeData.Widget, player_ui_slot{ZOrder := 1000000})
                    break
                Sleep(0.2)

    # Clean-up
    OnPlayerLeft(Player: player)<suspends>:void=
        Sleep(5.0)
        set PlayerFadeData = PlayerFadeData.WithRemovedKey(Player)

    GetOrMakePlayerFadeData(Player: player):ui_fader_fade_data=
        if(not PlayerFadeData[Player]):
            option{set PlayerFadeData[Player] = MakeFadeData()}
        PlayerFadeData[Player] or ui_fader_fade_data{Widget := canvas{}, ColorBlock := color_block{}}

    FadeIn(Player: player, ?Duration:float = 1.0, ?ForceAnimation:logic = false, ?FailsafeRevertDuration:float = 20.0):void=
        # Print(BuildPlayerMessage("FadeIn(", Player, ")={Duration}"))
        FadeData := GetOrMakePlayerFadeData(Player)
        set FadeData.Direction = true
        set FadeData.AnimationDuration = Duration
        spawn{Animate(Player, ForceAnimation)}
        spawn{RunFadeInFailsafe(Player, FailsafeRevertDuration)}

    # NOTE : could be improved and watch for fade completion to stop the thread earlier
    # Try to avoid being stuck on a black screen
    RunFadeInFailsafe(Player: player, Duration : float)<suspends>:void=
        Sleep(Duration)
        if(FadeData := PlayerFadeData[Player]):
            CurrentOpacity := FadeData.ColorBlock.GetOpacity()
            if(
                CurrentOpacity > 0.01,
                (not FadeData.Animating?) or (not FadeData.Direction?) # Not currently fading out
            ):
                FadeOut(Player)


    FadeOut(Player: player, ?Duration : float = 1.0, ?ForceAnimation:logic = false):void=
        # Print(BuildPlayerMessage("FadeOut(", Player, ")={Duration}"))
        FadeData := GetOrMakePlayerFadeData(Player)
        set FadeData.Direction = false
        set FadeData.AnimationDuration = Duration
        spawn{Animate(Player, ForceAnimation)}

    MakeFadeData()<transacts>:ui_fader_fade_data=
        ColorBlock := color_block:
            DefaultColor := NamedColors.Black
            DefaultOpacity := 0.0

        Canvas := canvas:
            Slots := array:
                canvas_slot:
                    Anchors := anchors{Minimum := vector2{X:=0.0,Y:=0.0}, Maximum := vector2{X:=1.0,Y:=1.0}}
                    Alignment := vector2{X:=0.5,Y:=0.5}
                    Widget := ColorBlock

        return ui_fader_fade_data:
            Widget := Canvas
            ColorBlock := ColorBlock

    Animate(Player: player, ForceAnimation: logic)<suspends>:void=
        if(
            PlayerUI := GetPlayerUI[Player],
            FadeData := PlayerFadeData[Player],
            ForceAnimation? or not FadeData.Animating?
        ):
            if(ForceAnimation?):
                FadeData.ColorBlock.SetOpacity(FadeData.Direction? and 0.0 or 1.0)
                if(FadeData.Animating?):
                    Player.FaderStoppedEvent().Signal()

            set FadeData.Animating = true

            race:
                Player.FaderStoppedEvent().Await()
                loop:
                    TargetOpacity := FadeData.Direction? and 1.0 or 0.0

                    if(FadeData.AnimationDuration = 0.0):
                        FadeData.ColorBlock.SetOpacity(TargetOpacity)
                        break

                    OpacityDelta := (1.0 / 30.0) / FadeData.AnimationDuration
                    DirectionSign := FadeData.Direction? and 1.0 or -1.0

                    Opacity := FadeData.ColorBlock.GetOpacity()

                    NewOpacity := Clamp(Opacity + (OpacityDelta * DirectionSign), 0.0, 1.0)
                    NewTypedOpacity := type{_X:float where 0.000000 <= _X, _X <= 1.000000}[NewOpacity] or 0.0
                    FadeData.ColorBlock.SetOpacity(NewTypedOpacity)

                    # Print(BuildPlayerMessage("Opacity(", Player, ") = {NewTypedOpacity}"))

                    # Break when any target is reached
                    if(Abs(NewOpacity - 0.5) = 0.5):
                        break
                    Sleep(0.0)
                
            # Print(BuildPlayerMessage("Stop(", Player, ")"))
            set FadeData.Animating = false