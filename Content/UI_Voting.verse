
using { /Fortnite.com/Devices }
using { /Verse.org/Simulation }
using { /UnrealEngine.com/Temporary/Diagnostics }
using { /UnrealEngine.com/Temporary/UI }
using { /Fortnite.com/UI }
using { /UnrealEngine.com/Temporary/SpatialMath}
using { /Verse.org/Colors }
using { /Verse.org/Assets }

# See https://dev.epicgames.com/documentation/en-us/uefn/create-your-own-device-in-verse for how to create a verse device.

# A Verse-authored creative device that can be placed in a level
UI_Voting := class(V_Krita):
    DefaultGroup <override>: string = "ArenaSelect"
    @editable
    bVoteManager : ?minigame_voting_V2 = false
    var ValidArenas : int = 2
    @editable
    var Time : int = 5

    _OnDeviceBegin<override>():void=
        block:


    _InjectUIData<override>(Player : player, UIData : UI_Data):void=
        # Setting arena textures
        SetArenaVotingChosenTextures(Player, UIData)
        if (ValidArenas = 2):
            var OrderedWidgets : []Parent_Widget_Data = array{}
            set OrderedWidgets += UIData.GetGroupWidgets(E_VoteMenu_Group.ArenaVoteScreens.GetName(), true)
            set OrderedWidgets += UIData.GetGroupWidgets(E_VoteMenu_Group.ArenaHighlights.GetName(), true)
            set OrderedWidgets += UIData.GetGroupWidgets(E_VoteMenu_Group.ArenaVoteButtons.GetName(), true)
            for(OrderedWidget : OrderedWidgets):
                set OrderedWidget.X += 245.0
        # Yellow block that slides from left is offset
        if(ColorBlocks := UIData.GetColorBlocks[E_VoteMenu_Widget.ArenaSelect_T_SlidingWidget.GetName()]):
            for(ColorBlock : ColorBlocks):
                set ColorBlock.X -= 500.0
        # Selected game mode texture that slides from left is offset
        if(ImageBlocks := UIData.GetImageBlocks[E_VoteMenu_Widget.ArenaSelect_T_ChosenGame.GetName()]):
            for(ImageBlock : ImageBlocks):
                set ImageBlock.X -= 500.0
        # Selected game mode texture text that slides from left is offset
        if(ImageBlocks := UIData.GetImageBlocks[E_VoteMenu_Widget.ArenaSelect_T_ChosenMiniGame.GetName()]):
            for(ImageBlock : ImageBlocks):
                set ImageBlock.X -= 500.0
        
        <#
        # All the voting screen widgets moved to top so they can animate back in
        OrderedWidgets1 := UIData.GetGroupWidgets(E_VoteMenu_Group.ArenaVoteScreen1.GetName(), true)
            for(Widget:OrderedWidgets1, not Button_Data[Widget]):
                set Widget.Y -= 900.0
        # All the voting screen widgets moved to botom so they can animate back in
        OrderedWidgets2 := UIData.GetGroupWidgets(E_VoteMenu_Group.ArenaVoteScreen2.GetName(), true)
            for(Widget:OrderedWidgets2, not Button_Data[Widget]):
                set Widget.Y += 900.0
        # All the voting screen widgets moved to top so they can animate back in
        OrderedWidgets3 := UIData.GetGroupWidgets(E_VoteMenu_Group.ArenaVoteScreen3.GetName(), true)
            for(Widget:OrderedWidgets3, not Button_Data[Widget]):
                set Widget.Y -= 900.0#>

    # Animates all voting screens from top to kritas default position, then it shows all vote buttons
    DelayVoteScreenAnimation(Player : player, UIData : UI_Data)<suspends>:void=
        UIData.ShowGroup(E_VoteMenu_Group.ArenaVoteScreens.GetName(), true)
        Sleep(0.5)
        UIData.MoveGroupAdditive(E_VoteMenu_Group.ArenaVoteScreen1.GetName(), true, 0.0, +900.0, 0.2, ?Seperate := false)
        UIData.MoveGroupAdditive(E_VoteMenu_Group.ArenaVoteScreen2.GetName(), true, 0.0, -900.0, 0.2, ?Seperate := false)
        UIData.MoveGroupAdditive(E_VoteMenu_Group.ArenaVoteScreen3.GetName(), true, 0.0, +900.0, 0.2, ?Seperate := false)
        UIData.ShowGroup(E_VoteMenu_Group.ArenaVoteButtons.GetName(), false)
                

    _OnUICreated<override>(Player : player, UIData : UI_Data):void=
        #Hide Highlights
        UIData.HideGroup(E_VoteMenu_Group.ArenaHighlights.GetName(), true)
        if (ValidArenas = 2):
            UIData.HideGroup(E_VoteMenu_Group.ArenaVoteScreen3.GetName(), true)
            UIData.HideWidget(E_VoteMenu_Widget.ArenaVoteButtons_B_Vote_3.GetName())

    _OnUIRendered<override>(Player : player, UIData : UI_Data):void=
        
        # Animate yellow block from left
        if(ColorBlocks := UIData.GetColorBlocks[E_VoteMenu_Widget.ArenaSelect_T_SlidingWidget.GetName()]):
            for(ColorBlock : ColorBlocks):
                spawn:
                    UIData.MoveWidget(ColorBlock, ColorBlock.X + 500.0, ColorBlock.Y, 0.05)
        # Animate selected game mode texture from left   
        if(ImageBlocks := UIData.GetImageBlocks[E_VoteMenu_Widget.ArenaSelect_T_ChosenGame.GetName()]):
            for(ImageBlock : ImageBlocks):
                spawn:
                    UIData.MoveWidget(ImageBlock, ImageBlock.X + 500.0, ImageBlock.Y, 0.05)

        # Animate selected game mode texture text from left   
        if(ImageBlocks := UIData.GetImageBlocks[E_VoteMenu_Widget.ArenaSelect_T_ChosenMiniGame.GetName()]):
            for(ImageBlock : ImageBlocks):
                spawn:
                    UIData.MoveWidget(ImageBlock, ImageBlock.X + 500.0, ImageBlock.Y, 0.05)
        # Show voting screens
        #UIData.ShowGroup(E_VoteMenu_Group.ArenaVoteScreens.GetName(), true)
        spawn:
            StartVoteTimer(Player, UIData)

        # Animate all voting screen widgets in order
        <#spawn:
            DelayVoteScreenAnimation(Player, UIData)#>


    _OnAllButtonSlotsAdded<override>(Player : player, UIData : UI_Data):void=
        # Bind all voting buttons
        if (VoteManager := bVoteManager?):
            if(VoteButton_1 := UIData.GetButtons[E_VoteMenu_Button.ArenaVoteButtons_B_Vote_1.GetName()][0]):
                VoteButton_1.Button.OnClick().Subscribe(VoteManager.HandleArenaVote_1)
            if(VoteButton_2 := UIData.GetButtons[E_VoteMenu_Button.ArenaVoteButtons_B_Vote_2.GetName()][0]):
                VoteButton_2.Button.OnClick().Subscribe(VoteManager.HandleArenaVote_2)
            if(VoteButton_3 := UIData.GetButtons[E_VoteMenu_Button.ArenaVoteButtons_B_Vote_3.GetName()][0]):
                VoteButton_3.Button.OnClick().Subscribe(VoteManager.HandleArenaVote_3)
    
    SetArenaVotingChosenTextures(Player : player, UIData : UI_Data):void=
        if(VoteManager := bVoteManager?, ChosenMiniGame := VoteManager.bChosenMinigame?):
            ChosenGameTexture := ChosenMiniGame.GetVotingImage()
            UIData.SetTexture(E_VoteMenu_Widget.ArenaSelect_T_ChosenGame.GetName(), option{ChosenGameTexture})
            if (Arenas := ChosenMiniGame.PossibleArenas?, VoteArena := Arenas[0]):
                if(Texture := VoteArena.ArenaImageOverride?.GetTexture() or VoteArena.ArenaDevice?.GetVotingImage()):
                    UIData.SetTexture(E_VoteMenu_Widget.ArenaVoteScreen1_T_Arena.GetName(), option{Texture})
                    var ModTextures : []texture = array{}
                    if(ModTexture := VoteArena.ModifierImage1?.GetTexture()):
                        set ModTextures += array{ModTexture}
                    if(ModTexture := VoteArena.ModifierImage2?.GetTexture()):
                        set ModTextures += array{ModTexture}
                    if(ModTexture := VoteArena.ModifierImage3?.GetTexture()):
                        set ModTextures += array{ModTexture}
                    #Print("LEEENNNNNNNGTH {ModTextures.Length}")
                    for(X := 0..3):
                        UIData.SetTexture("ArenaVoteScreen1", "T_Mod{X + 1}", option{ModTextures[X]})

            if (Arenas := ChosenMiniGame.PossibleArenas?, VoteArena := Arenas[1]):
                if(Texture := VoteArena.ArenaImageOverride?.GetTexture() or VoteArena.ArenaDevice?.GetVotingImage()):
                    UIData.SetTexture(E_VoteMenu_Widget.ArenaVoteScreen2_T_Arena.GetName(), option{Texture})
                    var ModTextures : []texture = array{}
                    if(ModTexture := VoteArena.ModifierImage1?.GetTexture()):
                        set ModTextures += array{ModTexture}
                    if(ModTexture := VoteArena.ModifierImage2?.GetTexture()):
                        set ModTextures += array{ModTexture}
                    if(ModTexture := VoteArena.ModifierImage3?.GetTexture()):
                        set ModTextures += array{ModTexture}
                    for(X := 0..3):
                        UIData.SetTexture("ArenaVoteScreen2", "T_Mod{X + 1}", option{ModTextures[X]})

            if (Arenas := ChosenMiniGame.PossibleArenas?, VoteArena := Arenas[2]):
                if(Texture := VoteArena.ArenaImageOverride?.GetTexture() or VoteArena.ArenaDevice?.GetVotingImage()):
                    UIData.SetTexture(E_VoteMenu_Widget.ArenaVoteScreen3_T_Arena.GetName(), option{Texture})
                    var ModTextures : []texture = array{}
                    if(ModTexture := VoteArena.ModifierImage1?.GetTexture()):
                        set ModTextures += array{ModTexture}
                    if(ModTexture := VoteArena.ModifierImage2?.GetTexture()):
                        set ModTextures += array{ModTexture}
                    if(ModTexture := VoteArena.ModifierImage3?.GetTexture()):
                        set ModTextures += array{ModTexture}
                    for(X := 0..3):
                        UIData.SetTexture("ArenaVoteScreen3", "T_Mod{X + 1}", option{ModTextures[X]})

    StartVoteTimer(Player : player, UIData : UI_Data)<suspends>:void=
        var Timer : int = Time
        race:
            Sleep(Time * 1.0)
            loop:
                if (Timer < 10):
                    UIData.SetText(E_VoteMenu_Widget.ArenaSelect_Txt_Timer.GetName(), " 0{Timer}")
                else:
                    UIData.SetText(E_VoteMenu_Widget.ArenaSelect_Txt_Timer.GetName(), " {Timer}")

                if(Timer <= 0):
                    Sleep(0.2)
                    break
                set Timer -= 1
                Sleep(1.0)
        UIData.ExitUI()
                
    @editable
    ButtonTriggerBinds : []VoteMenu_BindToTriggers = array{}
    @editable
    ShowGroups : []VoteMenu_ShowGroups = array{}
    @editable
    ExitButtons : []VoteMenu_ExitButton = array{}
    _MakeUIDataFromKrita<override>():UI_Data=
        var ExitButtonData : []VoteMenu_ExitButton = array{}
        for(ExitButton:ExitButtons):
            set ExitButtonData += array{MakeVoteMenu_ExitButton(ExitButton)}
        var ShowGroupData : []VoteMenu_ShowGroups = array{}
        for(ShowGroup:ShowGroups):
            set ShowGroupData += array{MakeVoteMenu_ShowGroups(ShowGroup)}
        var ButtonTriggerData : []VoteMenu_BindToTriggers = array{}
        for(ButtonTriggerBind:ButtonTriggerBinds):
            set ButtonTriggerData += array{MakeVoteMenu_BindToTriggers(ButtonTriggerBind)}
        TempUIDATA :=UI_Data:
            GroupUIMap := map:
                "root" => Group_UI{Name := "root", ParentGroupName := "", ChildGroups := array{"GameSelect", "ArenaSelect"}, AnchorMinX := 0.0, AnchorMinY := 0.0, AnchorMaxX := 0.0, AnchorMaxY := 0.0 }
                "GameSelect" => Group_UI{Name := "GameSelect", ParentGroupName := "root", ChildGroups := array{}, AnchorMinX := 0.5, AnchorMinY := 0.5, AnchorMaxX := 0.5, AnchorMaxY := 0.5 }
                "ArenaSelect" => Group_UI{Name := "ArenaSelect", ParentGroupName := "root", ChildGroups := array{"ArenaVoteScreens", "ArenaHighlights", "ArenaVoteButtons"}, AnchorMinX := 0.5, AnchorMinY := 0.5, AnchorMaxX := 0.5, AnchorMaxY := 0.5 }
                "ArenaVoteScreens" => Group_UI{Name := "ArenaVoteScreens", ParentGroupName := "ArenaSelect", ChildGroups := array{"ArenaVoteScreen1", "ArenaVoteScreen2", "ArenaVoteScreen3"}, AnchorMinX := 0.5, AnchorMinY := 0.5, AnchorMaxX := 0.5, AnchorMaxY := 0.5 }
                "ArenaVoteScreen1" => Group_UI{Name := "ArenaVoteScreen1", ParentGroupName := "ArenaVoteScreens", ChildGroups := array{}, AnchorMinX := 0.5, AnchorMinY := 0.5, AnchorMaxX := 0.5, AnchorMaxY := 0.5 }
                "ArenaVoteScreen2" => Group_UI{Name := "ArenaVoteScreen2", ParentGroupName := "ArenaVoteScreens", ChildGroups := array{}, AnchorMinX := 0.5, AnchorMinY := 0.5, AnchorMaxX := 0.5, AnchorMaxY := 0.5 }
                "ArenaVoteScreen3" => Group_UI{Name := "ArenaVoteScreen3", ParentGroupName := "ArenaVoteScreens", ChildGroups := array{}, AnchorMinX := 0.5, AnchorMinY := 0.5, AnchorMaxX := 0.5, AnchorMaxY := 0.5 }
                "ArenaHighlights" => Group_UI{Name := "ArenaHighlights", ParentGroupName := "ArenaSelect", ChildGroups := array{"ArenaHighlight1", "ArenaHighlight2", "ArenaHighlight3"}, AnchorMinX := 0.5, AnchorMinY := 0.5, AnchorMaxX := 0.5, AnchorMaxY := 0.5 }
                "ArenaHighlight1" => Group_UI{Name := "ArenaHighlight1", ParentGroupName := "ArenaHighlights", ChildGroups := array{}, AnchorMinX := 0.5, AnchorMinY := 0.5, AnchorMaxX := 0.5, AnchorMaxY := 0.5 }
                "ArenaHighlight2" => Group_UI{Name := "ArenaHighlight2", ParentGroupName := "ArenaHighlights", ChildGroups := array{}, AnchorMinX := 0.5, AnchorMinY := 0.5, AnchorMaxX := 0.5, AnchorMaxY := 0.5 }
                "ArenaHighlight3" => Group_UI{Name := "ArenaHighlight3", ParentGroupName := "ArenaHighlights", ChildGroups := array{}, AnchorMinX := 0.5, AnchorMinY := 0.5, AnchorMaxX := 0.5, AnchorMaxY := 0.5 }
                "ArenaVoteButtons" => Group_UI{Name := "ArenaVoteButtons", ParentGroupName := "ArenaSelect", ChildGroups := array{}, AnchorMinX := 0.5, AnchorMinY := 0.5, AnchorMaxX := 0.5, AnchorMaxY := 0.5 }
            OrderedWidgetData := array:
                Color_Block{X:=315.0, Y := -287.0, ScaleX := 493, ScaleY := 569, Name := "C_ColorBlock", GroupName := "GameSelect", Color := color_block{DefaultColor := color{R := 0.4019777798321958, G := 0.4019777798321958, B := 0.4019777798321958}, DefaultOpacity := 0.30196078431372547}, AnchorMinX := 0.5, AnchorMinY := 0.5, AnchorMaxX := 0.5, AnchorMaxY := 0.5 }
                Color_Block{X:=315.0, Y := 91.0, ScaleX := 493, ScaleY := 191, Name := "C_ColorBlock", GroupName := "GameSelect", Color := color_block{DefaultColor := color{R := 0.0822827071298148, G := 0.0822827071298148, B := 0.0822827071298148}, DefaultOpacity := 0.30196078431372547}, AnchorMinX := 0.5, AnchorMinY := 0.5, AnchorMaxX := 0.5, AnchorMaxY := 0.5 }
                Button_Quiet_Data{X:=490.0, Y := 152.0, ScaleX := 150, ScaleY := 68, Color := option{color_block{DefaultColor := color{R := 0.9686274509803922, G := 0.9607843137254902, B := 0.44313725490196076}, DefaultOpacity := 1.0} }, Text := "Vote", Name := "B_Vote_1", GroupName := "GameSelect", ButtonType := E_ButtonType.Quiet, AnchorMinX := 0.5, AnchorMinY := 0.5, AnchorMaxX := 0.5, AnchorMaxY := 0.5}
                Color_Block{X:=-247.0, Y := -287.0, ScaleX := 493, ScaleY := 569, Name := "C_ColorBlock", GroupName := "GameSelect", Color := color_block{DefaultColor := color{R := 0.4019777798321958, G := 0.4019777798321958, B := 0.4019777798321958}, DefaultOpacity := 0.30196078431372547}, AnchorMinX := 0.5, AnchorMinY := 0.5, AnchorMaxX := 0.5, AnchorMaxY := 0.5 }
                Color_Block{X:=-247.0, Y := 91.0, ScaleX := 493, ScaleY := 191, Name := "C_ColorBlock", GroupName := "GameSelect", Color := color_block{DefaultColor := color{R := 0.0822827071298148, G := 0.0822827071298148, B := 0.0822827071298148}, DefaultOpacity := 0.30196078431372547}, AnchorMinX := 0.5, AnchorMinY := 0.5, AnchorMaxX := 0.5, AnchorMaxY := 0.5 }
                Button_Quiet_Data{X:=-72.0, Y := 152.0, ScaleX := 150, ScaleY := 68, Color := option{color_block{DefaultColor := color{R := 0.9686274509803922, G := 0.9607843137254902, B := 0.44313725490196076}, DefaultOpacity := 1.0} }, Text := "Vote", Name := "B_Vote_2", GroupName := "GameSelect", ButtonType := E_ButtonType.Quiet, AnchorMinX := 0.5, AnchorMinY := 0.5, AnchorMaxX := 0.5, AnchorMaxY := 0.5}
                Color_Block{X:=-811.0, Y := -287.0, ScaleX := 493, ScaleY := 569, Name := "C_ColorBlock", GroupName := "GameSelect", Color := color_block{DefaultColor := color{R := 0.4019777798321958, G := 0.4019777798321958, B := 0.4019777798321958}, DefaultOpacity := 0.30196078431372547}, AnchorMinX := 0.5, AnchorMinY := 0.5, AnchorMaxX := 0.5, AnchorMaxY := 0.5 }
                Color_Block{X:=-811.0, Y := 91.0, ScaleX := 493, ScaleY := 191, Name := "C_ColorBlock", GroupName := "GameSelect", Color := color_block{DefaultColor := color{R := 0.0822827071298148, G := 0.0822827071298148, B := 0.0822827071298148}, DefaultOpacity := 0.30196078431372547}, AnchorMinX := 0.5, AnchorMinY := 0.5, AnchorMaxX := 0.5, AnchorMaxY := 0.5 }
                Button_Quiet_Data{X:=-636.0, Y := 152.0, ScaleX := 150, ScaleY := 68, Color := option{color_block{DefaultColor := color{R := 0.9686274509803922, G := 0.9607843137254902, B := 0.44313725490196076}, DefaultOpacity := 1.0} }, Text := "Vote", Name := "B_Vote_3", GroupName := "GameSelect", ButtonType := E_ButtonType.Quiet, AnchorMinX := 0.5, AnchorMinY := 0.5, AnchorMaxX := 0.5, AnchorMaxY := 0.5}
                Image_Block{X:=0.0, Y := -540.0, ScaleX := 508, ScaleY := 1080, Texture := option{VerseExposed.T_SlidingWidget}, Name := "T_SlidingWidget", GroupName := "ArenaSelect", AnchorMinX := 0.0, AnchorMinY := 0.5, AnchorMaxX := 0.0, AnchorMaxY := 0.5}
                Image_Block{X:=-4.0, Y := -218.0, ScaleX := 512, ScaleY := 512, Texture := option{VerseExposed.T_ChosenGame}, Name := "T_ChosenGame", GroupName := "ArenaSelect", AnchorMinX := 0.0, AnchorMinY := 0.5, AnchorMaxX := 0.0, AnchorMaxY := 0.5}
                Image_Block{X:=54.0, Y := -315.0, ScaleX := 396, ScaleY := 169, Texture := option{VerseExposed.T_ChosenMiniGame}, Name := "T_ChosenMiniGame", GroupName := "ArenaSelect", AnchorMinX := 0.0, AnchorMinY := 0.5, AnchorMaxX := 0.0, AnchorMaxY := 0.5}
                Button_Quiet_Data{X:=75.0, Y := 362.0, ScaleX := 350, ScaleY := 68, Color := option{color_block{DefaultColor := color{R := 0.9686274509803922, G := 0.9607843137254902, B := 0.44313725490196076}, DefaultOpacity := 1.0} }, Text := "Exit", Name := "B_Exit", GroupName := "ArenaSelect", ButtonType := E_ButtonType.Quiet, AnchorMinX := 0.5, AnchorMinY := 0.5, AnchorMaxX := 0.5, AnchorMaxY := 0.5}
                Color_Block{X:=-446.0, Y := -303.0, ScaleX := 447, ScaleY := 585, Name := "C_ColorBlock", GroupName := "ArenaVoteScreen1", Color := color_block{DefaultColor := color{R := 0.0, G := 0.0, B := 0.0}, DefaultOpacity := 0.30196078431372547}, AnchorMinX := 0.5, AnchorMinY := 0.5, AnchorMaxX := 0.5, AnchorMaxY := 0.5 }
                Color_Block{X:=-446.0, Y := 91.0, ScaleX := 447, ScaleY := 191, Name := "C_ColorBlock", GroupName := "ArenaVoteScreen1", Color := color_block{DefaultColor := color{R := 0.0, G := 0.0, B := 0.0}, DefaultOpacity := 0.30196078431372547}, AnchorMinX := 0.5, AnchorMinY := 0.5, AnchorMaxX := 0.5, AnchorMaxY := 0.5 }
                Image_Block{X:=-446.0, Y := -302.0, ScaleX := 447, ScaleY := 393, Texture := option{VerseExposed.T_Arena}, Name := "T_Arena", GroupName := "ArenaVoteScreen1", AnchorMinX := 0.5, AnchorMinY := 0.5, AnchorMaxX := 0.5, AnchorMaxY := 0.5}
                Image_Block{X:=-70.0, Y := -301.0, ScaleX := 69, ScaleY := 76, Texture := option{VerseExposed.T_Mod}, Name := "T_Mod1", GroupName := "ArenaVoteScreen1", AnchorMinX := 0.5, AnchorMinY := 0.5, AnchorMaxX := 0.5, AnchorMaxY := 0.5}
                Image_Block{X:=-70.0, Y := -232.0, ScaleX := 69, ScaleY := 76, Texture := option{VerseExposed.T_Mod}, Name := "T_Mod2", GroupName := "ArenaVoteScreen1", AnchorMinX := 0.5, AnchorMinY := 0.5, AnchorMaxX := 0.5, AnchorMaxY := 0.5}
                Image_Block{X:=-70.0, Y := -163.0, ScaleX := 69, ScaleY := 77, Texture := option{VerseExposed.T_Mod}, Name := "T_Mod3", GroupName := "ArenaVoteScreen1", AnchorMinX := 0.5, AnchorMinY := 0.5, AnchorMaxX := 0.5, AnchorMaxY := 0.5}
                Color_Block{X:=-439.0, Y := -295.0, ScaleX := 47, ScaleY := 51, Name := "C_ColorBlock", GroupName := "ArenaVoteScreen1", Color := color_block{DefaultColor := color{R := 1.0, G := 0.7605245046752924, B := 0.0}, DefaultOpacity := 1.0}, AnchorMinX := 0.5, AnchorMinY := 0.5, AnchorMaxX := 0.5, AnchorMaxY := 0.5 }
                Color_Block{X:=-435.0, Y := -291.0, ScaleX := 39, ScaleY := 43, Name := "C_ColorBlock", GroupName := "ArenaVoteScreen1", Color := color_block{DefaultColor := color{R := 0.006995410187265387, G := 0.11953842798834562, B := 1.0}, DefaultOpacity := 1.0}, AnchorMinX := 0.5, AnchorMinY := 0.5, AnchorMaxX := 0.5, AnchorMaxY := 0.5 }
                Text_Block{X:=-422.0, Y := -283.0, ScaleX := 20, ScaleY := 24, Text := "0", Name := "Txt_VoteAmmount", GroupName := "ArenaVoteScreen1", Color := color{R := 1.0, G := 1.0, B := 1.0}, Opacity := 1.0, AnchorMinX := 0.5, AnchorMinY := 0.5, AnchorMaxX := 0.5, AnchorMaxY := 0.5}
                Color_Block{X:=30.0, Y := -303.0, ScaleX := 447, ScaleY := 585, Name := "C_ColorBlock", GroupName := "ArenaVoteScreen2", Color := color_block{DefaultColor := color{R := 0.0, G := 0.0, B := 0.0}, DefaultOpacity := 0.30196078431372547}, AnchorMinX := 0.5, AnchorMinY := 0.5, AnchorMaxX := 0.5, AnchorMaxY := 0.5 }
                Color_Block{X:=30.0, Y := 91.0, ScaleX := 447, ScaleY := 191, Name := "C_ColorBlock", GroupName := "ArenaVoteScreen2", Color := color_block{DefaultColor := color{R := 0.0, G := 0.0, B := 0.0}, DefaultOpacity := 0.30196078431372547}, AnchorMinX := 0.5, AnchorMinY := 0.5, AnchorMaxX := 0.5, AnchorMaxY := 0.5 }
                Image_Block{X:=30.0, Y := -302.0, ScaleX := 447, ScaleY := 394, Texture := option{VerseExposed.T_Arena}, Name := "T_Arena", GroupName := "ArenaVoteScreen2", AnchorMinX := 0.5, AnchorMinY := 0.5, AnchorMaxX := 0.5, AnchorMaxY := 0.5}
                Image_Block{X:=405.0, Y := -301.0, ScaleX := 69, ScaleY := 76, Texture := option{VerseExposed.T_Mod}, Name := "T_Mod1", GroupName := "ArenaVoteScreen2", AnchorMinX := 0.5, AnchorMinY := 0.5, AnchorMaxX := 0.5, AnchorMaxY := 0.5}
                Image_Block{X:=405.0, Y := -232.0, ScaleX := 69, ScaleY := 76, Texture := option{VerseExposed.T_Mod}, Name := "T_Mod2", GroupName := "ArenaVoteScreen2", AnchorMinX := 0.5, AnchorMinY := 0.5, AnchorMaxX := 0.5, AnchorMaxY := 0.5}
                Image_Block{X:=405.0, Y := -163.0, ScaleX := 69, ScaleY := 77, Texture := option{VerseExposed.T_Mod}, Name := "T_Mod3", GroupName := "ArenaVoteScreen2", AnchorMinX := 0.5, AnchorMinY := 0.5, AnchorMaxX := 0.5, AnchorMaxY := 0.5}
                Color_Block{X:=37.0, Y := -295.0, ScaleX := 47, ScaleY := 51, Name := "C_ColorBlock", GroupName := "ArenaVoteScreen2", Color := color_block{DefaultColor := color{R := 1.0, G := 0.7605245046752924, B := 0.0}, DefaultOpacity := 1.0}, AnchorMinX := 0.5, AnchorMinY := 0.5, AnchorMaxX := 0.5, AnchorMaxY := 0.5 }
                Color_Block{X:=41.0, Y := -291.0, ScaleX := 39, ScaleY := 43, Name := "C_ColorBlock", GroupName := "ArenaVoteScreen2", Color := color_block{DefaultColor := color{R := 0.006995410187265387, G := 0.11953842798834562, B := 1.0}, DefaultOpacity := 1.0}, AnchorMinX := 0.5, AnchorMinY := 0.5, AnchorMaxX := 0.5, AnchorMaxY := 0.5 }
                Text_Block{X:=54.0, Y := -283.0, ScaleX := 20, ScaleY := 24, Text := "0", Name := "Txt_VoteAmmount", GroupName := "ArenaVoteScreen2", Color := color{R := 1.0, G := 1.0, B := 1.0}, Opacity := 1.0, AnchorMinX := 0.5, AnchorMinY := 0.5, AnchorMaxX := 0.5, AnchorMaxY := 0.5}
                Color_Block{X:=506.0, Y := -303.0, ScaleX := 447, ScaleY := 585, Name := "C_ColorBlock", GroupName := "ArenaVoteScreen3", Color := color_block{DefaultColor := color{R := 0.0, G := 0.0, B := 0.0}, DefaultOpacity := 0.30196078431372547}, AnchorMinX := 0.5, AnchorMinY := 0.5, AnchorMaxX := 0.5, AnchorMaxY := 0.5 }
                Color_Block{X:=506.0, Y := 91.0, ScaleX := 447, ScaleY := 191, Name := "C_ColorBlock", GroupName := "ArenaVoteScreen3", Color := color_block{DefaultColor := color{R := 0.0, G := 0.0, B := 0.0}, DefaultOpacity := 0.30196078431372547}, AnchorMinX := 0.5, AnchorMinY := 0.5, AnchorMaxX := 0.5, AnchorMaxY := 0.5 }
                Image_Block{X:=506.0, Y := -302.0, ScaleX := 448, ScaleY := 394, Texture := option{VerseExposed.T_Arena}, Name := "T_Arena", GroupName := "ArenaVoteScreen3", AnchorMinX := 0.5, AnchorMinY := 0.5, AnchorMaxX := 0.5, AnchorMaxY := 0.5}
                Image_Block{X:=883.0, Y := -301.0, ScaleX := 69, ScaleY := 76, Texture := option{VerseExposed.T_Mod}, Name := "T_Mod1", GroupName := "ArenaVoteScreen3", AnchorMinX := 0.5, AnchorMinY := 0.5, AnchorMaxX := 0.5, AnchorMaxY := 0.5}
                Image_Block{X:=883.0, Y := -232.0, ScaleX := 69, ScaleY := 76, Texture := option{VerseExposed.T_Mod}, Name := "T_Mod2", GroupName := "ArenaVoteScreen3", AnchorMinX := 0.5, AnchorMinY := 0.5, AnchorMaxX := 0.5, AnchorMaxY := 0.5}
                Image_Block{X:=883.0, Y := -163.0, ScaleX := 69, ScaleY := 77, Texture := option{VerseExposed.T_Mod}, Name := "T_Mod3", GroupName := "ArenaVoteScreen3", AnchorMinX := 0.5, AnchorMinY := 0.5, AnchorMaxX := 0.5, AnchorMaxY := 0.5}
                Color_Block{X:=513.0, Y := -295.0, ScaleX := 47, ScaleY := 51, Name := "C_ColorBlock", GroupName := "ArenaVoteScreen3", Color := color_block{DefaultColor := color{R := 1.0, G := 0.7605245046752924, B := 0.0}, DefaultOpacity := 1.0}, AnchorMinX := 0.5, AnchorMinY := 0.5, AnchorMaxX := 0.5, AnchorMaxY := 0.5 }
                Color_Block{X:=517.0, Y := -291.0, ScaleX := 39, ScaleY := 43, Name := "C_ColorBlock", GroupName := "ArenaVoteScreen3", Color := color_block{DefaultColor := color{R := 0.006995410187265387, G := 0.11953842798834562, B := 1.0}, DefaultOpacity := 1.0}, AnchorMinX := 0.5, AnchorMinY := 0.5, AnchorMaxX := 0.5, AnchorMaxY := 0.5 }
                Text_Block{X:=530.0, Y := -283.0, ScaleX := 20, ScaleY := 24, Text := "0", Name := "Txt_VoteAmmount", GroupName := "ArenaVoteScreen3", Color := color{R := 1.0, G := 1.0, B := 1.0}, Opacity := 1.0, AnchorMinX := 0.5, AnchorMinY := 0.5, AnchorMaxX := 0.5, AnchorMaxY := 0.5}
                Color_Block{X:=-2.0, Y := -306.0, ScaleX := 5, ScaleY := 590, Name := "C_ColorBlock", GroupName := "ArenaHighlight1", Color := color_block{DefaultColor := color{R := 1.0, G := 1.0, B := 1.0}, DefaultOpacity := 1.0}, AnchorMinX := 0.5, AnchorMinY := 0.5, AnchorMaxX := 0.5, AnchorMaxY := 0.5 }
                Color_Block{X:=-450.0, Y := -306.0, ScaleX := 5, ScaleY := 590, Name := "C_ColorBlock", GroupName := "ArenaHighlight1", Color := color_block{DefaultColor := color{R := 1.0, G := 1.0, B := 1.0}, DefaultOpacity := 1.0}, AnchorMinX := 0.5, AnchorMinY := 0.5, AnchorMaxX := 0.5, AnchorMaxY := 0.5 }
                Color_Block{X:=-446.0, Y := -306.0, ScaleX := 445, ScaleY := 4, Name := "C_ColorBlock", GroupName := "ArenaHighlight1", Color := color_block{DefaultColor := color{R := 1.0, G := 1.0, B := 1.0}, DefaultOpacity := 1.0}, AnchorMinX := 0.5, AnchorMinY := 0.5, AnchorMaxX := 0.5, AnchorMaxY := 0.5 }
                Color_Block{X:=-446.0, Y := 280.0, ScaleX := 445, ScaleY := 4, Name := "C_ColorBlock", GroupName := "ArenaHighlight1", Color := color_block{DefaultColor := color{R := 1.0, G := 1.0, B := 1.0}, DefaultOpacity := 1.0}, AnchorMinX := 0.5, AnchorMinY := 0.5, AnchorMaxX := 0.5, AnchorMaxY := 0.5 }
                Color_Block{X:=-394.0, Y := -295.0, ScaleX := 2, ScaleY := 51, Name := "C_ColorBlock", GroupName := "ArenaHighlight1", Color := color_block{DefaultColor := color{R := 1.0, G := 0.938685728457888, B := 0.47353149614800955}, DefaultOpacity := 1.0}, AnchorMinX := 0.5, AnchorMinY := 0.5, AnchorMaxX := 0.5, AnchorMaxY := 0.5 }
                Color_Block{X:=-439.0, Y := -295.0, ScaleX := 2, ScaleY := 51, Name := "C_ColorBlock", GroupName := "ArenaHighlight1", Color := color_block{DefaultColor := color{R := 1.0, G := 0.938685728457888, B := 0.47353149614800955}, DefaultOpacity := 1.0}, AnchorMinX := 0.5, AnchorMinY := 0.5, AnchorMaxX := 0.5, AnchorMaxY := 0.5 }
                Color_Block{X:=-439.0, Y := -295.0, ScaleX := 47, ScaleY := 2, Name := "C_ColorBlock", GroupName := "ArenaHighlight1", Color := color_block{DefaultColor := color{R := 1.0, G := 0.938685728457888, B := 0.47353149614800955}, DefaultOpacity := 1.0}, AnchorMinX := 0.5, AnchorMinY := 0.5, AnchorMaxX := 0.5, AnchorMaxY := 0.5 }
                Color_Block{X:=-439.0, Y := -246.0, ScaleX := 47, ScaleY := 2, Name := "C_ColorBlock", GroupName := "ArenaHighlight1", Color := color_block{DefaultColor := color{R := 1.0, G := 0.938685728457888, B := 0.47353149614800955}, DefaultOpacity := 1.0}, AnchorMinX := 0.5, AnchorMinY := 0.5, AnchorMaxX := 0.5, AnchorMaxY := 0.5 }
                Color_Block{X:=474.0, Y := -306.0, ScaleX := 5, ScaleY := 590, Name := "C_ColorBlock", GroupName := "ArenaHighlight2", Color := color_block{DefaultColor := color{R := 1.0, G := 1.0, B := 1.0}, DefaultOpacity := 1.0}, AnchorMinX := 0.5, AnchorMinY := 0.5, AnchorMaxX := 0.5, AnchorMaxY := 0.5 }
                Color_Block{X:=26.0, Y := -306.0, ScaleX := 5, ScaleY := 590, Name := "C_ColorBlock", GroupName := "ArenaHighlight2", Color := color_block{DefaultColor := color{R := 1.0, G := 1.0, B := 1.0}, DefaultOpacity := 1.0}, AnchorMinX := 0.5, AnchorMinY := 0.5, AnchorMaxX := 0.5, AnchorMaxY := 0.5 }
                Color_Block{X:=30.0, Y := -306.0, ScaleX := 445, ScaleY := 4, Name := "C_ColorBlock", GroupName := "ArenaHighlight2", Color := color_block{DefaultColor := color{R := 1.0, G := 1.0, B := 1.0}, DefaultOpacity := 1.0}, AnchorMinX := 0.5, AnchorMinY := 0.5, AnchorMaxX := 0.5, AnchorMaxY := 0.5 }
                Color_Block{X:=30.0, Y := 280.0, ScaleX := 445, ScaleY := 4, Name := "C_ColorBlock", GroupName := "ArenaHighlight2", Color := color_block{DefaultColor := color{R := 1.0, G := 1.0, B := 1.0}, DefaultOpacity := 1.0}, AnchorMinX := 0.5, AnchorMinY := 0.5, AnchorMaxX := 0.5, AnchorMaxY := 0.5 }
                Color_Block{X:=82.0, Y := -295.0, ScaleX := 2, ScaleY := 51, Name := "C_ColorBlock", GroupName := "ArenaHighlight2", Color := color_block{DefaultColor := color{R := 1.0, G := 0.938685728457888, B := 0.47353149614800955}, DefaultOpacity := 1.0}, AnchorMinX := 0.5, AnchorMinY := 0.5, AnchorMaxX := 0.5, AnchorMaxY := 0.5 }
                Color_Block{X:=37.0, Y := -295.0, ScaleX := 2, ScaleY := 51, Name := "C_ColorBlock", GroupName := "ArenaHighlight2", Color := color_block{DefaultColor := color{R := 1.0, G := 0.938685728457888, B := 0.47353149614800955}, DefaultOpacity := 1.0}, AnchorMinX := 0.5, AnchorMinY := 0.5, AnchorMaxX := 0.5, AnchorMaxY := 0.5 }
                Color_Block{X:=37.0, Y := -295.0, ScaleX := 47, ScaleY := 2, Name := "C_ColorBlock", GroupName := "ArenaHighlight2", Color := color_block{DefaultColor := color{R := 1.0, G := 0.938685728457888, B := 0.47353149614800955}, DefaultOpacity := 1.0}, AnchorMinX := 0.5, AnchorMinY := 0.5, AnchorMaxX := 0.5, AnchorMaxY := 0.5 }
                Color_Block{X:=37.0, Y := -246.0, ScaleX := 47, ScaleY := 2, Name := "C_ColorBlock", GroupName := "ArenaHighlight2", Color := color_block{DefaultColor := color{R := 1.0, G := 0.938685728457888, B := 0.47353149614800955}, DefaultOpacity := 1.0}, AnchorMinX := 0.5, AnchorMinY := 0.5, AnchorMaxX := 0.5, AnchorMaxY := 0.5 }
                Color_Block{X:=950.0, Y := -306.0, ScaleX := 5, ScaleY := 590, Name := "C_ColorBlock", GroupName := "ArenaHighlight3", Color := color_block{DefaultColor := color{R := 1.0, G := 1.0, B := 1.0}, DefaultOpacity := 1.0}, AnchorMinX := 0.5, AnchorMinY := 0.5, AnchorMaxX := 0.5, AnchorMaxY := 0.5 }
                Color_Block{X:=502.0, Y := -306.0, ScaleX := 5, ScaleY := 590, Name := "C_ColorBlock", GroupName := "ArenaHighlight3", Color := color_block{DefaultColor := color{R := 1.0, G := 1.0, B := 1.0}, DefaultOpacity := 1.0}, AnchorMinX := 0.5, AnchorMinY := 0.5, AnchorMaxX := 0.5, AnchorMaxY := 0.5 }
                Color_Block{X:=506.0, Y := -306.0, ScaleX := 445, ScaleY := 4, Name := "C_ColorBlock", GroupName := "ArenaHighlight3", Color := color_block{DefaultColor := color{R := 1.0, G := 1.0, B := 1.0}, DefaultOpacity := 1.0}, AnchorMinX := 0.5, AnchorMinY := 0.5, AnchorMaxX := 0.5, AnchorMaxY := 0.5 }
                Color_Block{X:=506.0, Y := 280.0, ScaleX := 445, ScaleY := 4, Name := "C_ColorBlock", GroupName := "ArenaHighlight3", Color := color_block{DefaultColor := color{R := 1.0, G := 1.0, B := 1.0}, DefaultOpacity := 1.0}, AnchorMinX := 0.5, AnchorMinY := 0.5, AnchorMaxX := 0.5, AnchorMaxY := 0.5 }
                Color_Block{X:=558.0, Y := -295.0, ScaleX := 2, ScaleY := 51, Name := "C_ColorBlock", GroupName := "ArenaHighlight3", Color := color_block{DefaultColor := color{R := 1.0, G := 0.938685728457888, B := 0.47353149614800955}, DefaultOpacity := 1.0}, AnchorMinX := 0.5, AnchorMinY := 0.5, AnchorMaxX := 0.5, AnchorMaxY := 0.5 }
                Color_Block{X:=513.0, Y := -295.0, ScaleX := 2, ScaleY := 51, Name := "C_ColorBlock", GroupName := "ArenaHighlight3", Color := color_block{DefaultColor := color{R := 1.0, G := 0.938685728457888, B := 0.47353149614800955}, DefaultOpacity := 1.0}, AnchorMinX := 0.5, AnchorMinY := 0.5, AnchorMaxX := 0.5, AnchorMaxY := 0.5 }
                Color_Block{X:=513.0, Y := -295.0, ScaleX := 47, ScaleY := 2, Name := "C_ColorBlock", GroupName := "ArenaHighlight3", Color := color_block{DefaultColor := color{R := 1.0, G := 0.938685728457888, B := 0.47353149614800955}, DefaultOpacity := 1.0}, AnchorMinX := 0.5, AnchorMinY := 0.5, AnchorMaxX := 0.5, AnchorMaxY := 0.5 }
                Color_Block{X:=513.0, Y := -246.0, ScaleX := 47, ScaleY := 2, Name := "C_ColorBlock", GroupName := "ArenaHighlight3", Color := color_block{DefaultColor := color{R := 1.0, G := 0.938685728457888, B := 0.47353149614800955}, DefaultOpacity := 1.0}, AnchorMinX := 0.5, AnchorMinY := 0.5, AnchorMaxX := 0.5, AnchorMaxY := 0.5 }
                Button_Quiet_Data{X:=661.0, Y := 152.0, ScaleX := 150, ScaleY := 68, Color := option{color_block{DefaultColor := color{R := 0.9686274509803922, G := 0.9607843137254902, B := 0.44313725490196076}, DefaultOpacity := 1.0} }, Text := "Vote", Name := "B_Vote_3", GroupName := "ArenaVoteButtons", ButtonType := E_ButtonType.Quiet, AnchorMinX := 0.5, AnchorMinY := 0.5, AnchorMaxX := 0.5, AnchorMaxY := 0.5}
                Button_Quiet_Data{X:=181.0, Y := 152.0, ScaleX := 150, ScaleY := 68, Color := option{color_block{DefaultColor := color{R := 0.9686274509803922, G := 0.9607843137254902, B := 0.44313725490196076}, DefaultOpacity := 1.0} }, Text := "Vote", Name := "B_Vote_2", GroupName := "ArenaVoteButtons", ButtonType := E_ButtonType.Quiet, AnchorMinX := 0.5, AnchorMinY := 0.5, AnchorMaxX := 0.5, AnchorMaxY := 0.5}
                Button_Quiet_Data{X:=-294.0, Y := 152.0, ScaleX := 150, ScaleY := 68, Color := option{color_block{DefaultColor := color{R := 0.9686274509803922, G := 0.9607843137254902, B := 0.44313725490196076}, DefaultOpacity := 1.0} }, Text := "Vote", Name := "B_Vote_1", GroupName := "ArenaVoteButtons", ButtonType := E_ButtonType.Quiet, AnchorMinX := 0.5, AnchorMinY := 0.5, AnchorMaxX := 0.5, AnchorMaxY := 0.5}
                Color_Block{X:=30.0, Y := -416.0, ScaleX := 447, ScaleY := 63, Name := "C_ColorBlock", GroupName := "ArenaSelect", Color := color_block{DefaultColor := color{R := 0.0, G := 0.0, B := 0.0}, DefaultOpacity := 0.30196078431372547}, AnchorMinX := 0.5, AnchorMinY := 0.5, AnchorMaxX := 0.5, AnchorMaxY := 0.5 }
                Text_Block{X:=359.0, Y := -399.0, ScaleX := 40, ScaleY := 24, Text := "00", Name := "Txt_Timer", GroupName := "ArenaSelect", Color := color{R := 1.0, G := 1.0, B := 1.0}, Opacity := 1.0, AnchorMinX := 0.5, AnchorMinY := 0.5, AnchorMaxX := 0.5, AnchorMaxY := 0.5}
                Text_Block{X:=104.0, Y := -399.0, ScaleX := 236, ScaleY := 24, Text := "Vote For Arena :", Name := "Txt_VoteForText", GroupName := "ArenaSelect", Color := color{R := 1.0, G := 1.0, B := 1.0}, Opacity := 1.0, AnchorMinX := 0.5, AnchorMinY := 0.5, AnchorMaxX := 0.5, AnchorMaxY := 0.5}
                Text_Block{X:=414.0, Y := -401.0, ScaleX := 33, ScaleY := 26, Text := "🕹", Name := "Txt_Console", GroupName := "ArenaSelect", Color := color{R := 0.807843137254902, G := 0.5686274509803921, B := 0.47058823529411764}, Opacity := 1.0, AnchorMinX := 0.5, AnchorMinY := 0.5, AnchorMaxX := 0.5, AnchorMaxY := 0.5}
                Text_Block{X:=62.0, Y := -401.0, ScaleX := 34, ScaleY := 26, Text := "🕹", Name := "Txt_Console", GroupName := "ArenaSelect", Color := color{R := 0.807843137254902, G := 0.5686274509803921, B := 0.47058823529411764}, Opacity := 1.0, AnchorMinX := 0.5, AnchorMinY := 0.5, AnchorMaxX := 0.5, AnchorMaxY := 0.5}
            ButtonActivateTriggers := ButtonTriggerData
            ShowGroups := ShowGroupData
            ExitButtons := ExitButtonData
E_VoteMenu_Button := enum:
    ArenaSelect_B_Exit
    ArenaVoteButtons_B_Vote_1
    ArenaVoteButtons_B_Vote_2
    ArenaVoteButtons_B_Vote_3
    GameSelect_B_Vote_1
    GameSelect_B_Vote_2
    GameSelect_B_Vote_3
    root_B_ButtonLoud
    root_B_ButtonRegular

(Button : E_VoteMenu_Button).GetName()<transacts>:tuple(string, string)=
    case(Button):
        E_VoteMenu_Button.ArenaSelect_B_Exit =>
            return ("ArenaSelect", "B_Exit")
        E_VoteMenu_Button.ArenaVoteButtons_B_Vote_1 =>
            return ("ArenaVoteButtons", "B_Vote_1")
        E_VoteMenu_Button.ArenaVoteButtons_B_Vote_2 =>
            return ("ArenaVoteButtons", "B_Vote_2")
        E_VoteMenu_Button.ArenaVoteButtons_B_Vote_3 =>
            return ("ArenaVoteButtons", "B_Vote_3")
        E_VoteMenu_Button.GameSelect_B_Vote_1 =>
            return ("GameSelect", "B_Vote_1")
        E_VoteMenu_Button.GameSelect_B_Vote_2 =>
            return ("GameSelect", "B_Vote_2")
        E_VoteMenu_Button.GameSelect_B_Vote_3 =>
            return ("GameSelect", "B_Vote_3")
        E_VoteMenu_Button.root_B_ButtonLoud =>
            return ("root", "B_ButtonLoud")
        E_VoteMenu_Button.root_B_ButtonRegular =>
            return ("root", "B_ButtonRegular")
E_VoteMenu_Group := enum:
    ArenaSelect
    ArenaVoteButtons
    ArenaVoteScreens
    GameSelect
    root
    ArenaVoteScreen1
    ArenaVoteScreen2
    ArenaVoteScreen3
    ArenaHighlight1
    ArenaHighlight2
    ArenaHighlight3
    ArenaHighlights

(Group : E_VoteMenu_Group).GetName()<transacts>:string=
    case(Group):
        E_VoteMenu_Group.ArenaSelect =>
            return "ArenaSelect"
        E_VoteMenu_Group.ArenaVoteButtons =>
            return "ArenaVoteButtons"
        E_VoteMenu_Group.ArenaVoteScreens =>
            return "ArenaVoteScreens"
        E_VoteMenu_Group.GameSelect =>
            return "GameSelect"
        E_VoteMenu_Group.root =>
            return "root"
        E_VoteMenu_Group.ArenaVoteScreen1 =>
            return "ArenaVoteScreen1"
        E_VoteMenu_Group.ArenaVoteScreen2 =>
            return "ArenaVoteScreen2"
        E_VoteMenu_Group.ArenaVoteScreen3 =>
            return "ArenaVoteScreen3"
        E_VoteMenu_Group.ArenaHighlight1 =>
            return "ArenaHighlight1"
        E_VoteMenu_Group.ArenaHighlight2 =>
            return "ArenaHighlight2"
        E_VoteMenu_Group.ArenaHighlight3 =>
            return "ArenaHighlight3"
        E_VoteMenu_Group.ArenaHighlights =>
            return "ArenaHighlights"
E_VoteMenu_Widget := enum:
        GameSelect_C_ColorBlock
        GameSelect_B_Vote_1
        GameSelect_B_Vote_2
        GameSelect_B_Vote_3
        ArenaSelect_T_SlidingWidget
        ArenaSelect_T_ChosenGame
        ArenaSelect_T_ChosenMiniGame
        ArenaSelect_B_Exit
        ArenaVoteScreen1_C_ColorBlock
        ArenaVoteScreen1_T_Arena
        ArenaVoteScreen1_T_Mod1
        ArenaVoteScreen1_T_Mod2
        ArenaVoteScreen1_T_Mod3
        ArenaVoteScreen1_Txt_VoteAmmount
        ArenaVoteScreen2_C_ColorBlock
        ArenaVoteScreen2_T_Arena
        ArenaVoteScreen2_T_Mod1
        ArenaVoteScreen2_T_Mod2
        ArenaVoteScreen2_T_Mod3
        ArenaVoteScreen2_Txt_VoteAmmount
        ArenaVoteScreen3_C_ColorBlock
        ArenaVoteScreen3_T_Arena
        ArenaVoteScreen3_T_Mod1
        ArenaVoteScreen3_T_Mod2
        ArenaVoteScreen3_T_Mod3
        ArenaVoteScreen3_Txt_VoteAmmount
        ArenaHighlight1_C_ColorBlock
        ArenaHighlight2_C_ColorBlock
        ArenaHighlight3_C_ColorBlock
        ArenaVoteButtons_B_Vote_3
        ArenaVoteButtons_B_Vote_2
        ArenaVoteButtons_B_Vote_1
        ArenaSelect_C_ColorBlock
        ArenaSelect_Txt_Timer
        ArenaSelect_Txt_VoteForText
        ArenaSelect_Txt_Console
(Widget : E_VoteMenu_Widget).GetName()<transacts>:tuple(string, string)=
    case(Widget):
        E_VoteMenu_Widget.GameSelect_C_ColorBlock =>
            return ("GameSelect", "C_ColorBlock")
        E_VoteMenu_Widget.GameSelect_B_Vote_1 =>
            return ("GameSelect", "B_Vote_1")
        E_VoteMenu_Widget.GameSelect_B_Vote_2 =>
            return ("GameSelect", "B_Vote_2")
        E_VoteMenu_Widget.GameSelect_B_Vote_3 =>
            return ("GameSelect", "B_Vote_3")
        E_VoteMenu_Widget.ArenaSelect_T_SlidingWidget =>
            return ("ArenaSelect", "T_SlidingWidget")
        E_VoteMenu_Widget.ArenaSelect_T_ChosenGame =>
            return ("ArenaSelect", "T_ChosenGame")
        E_VoteMenu_Widget.ArenaSelect_T_ChosenMiniGame =>
            return ("ArenaSelect", "T_ChosenMiniGame")
        E_VoteMenu_Widget.ArenaSelect_B_Exit =>
            return ("ArenaSelect", "B_Exit")
        E_VoteMenu_Widget.ArenaVoteScreen1_C_ColorBlock =>
            return ("ArenaVoteScreen1", "C_ColorBlock")
        E_VoteMenu_Widget.ArenaVoteScreen1_T_Arena =>
            return ("ArenaVoteScreen1", "T_Arena")
        E_VoteMenu_Widget.ArenaVoteScreen1_T_Mod1 =>
            return ("ArenaVoteScreen1", "T_Mod1")
        E_VoteMenu_Widget.ArenaVoteScreen1_T_Mod2 =>
            return ("ArenaVoteScreen1", "T_Mod2")
        E_VoteMenu_Widget.ArenaVoteScreen1_T_Mod3 =>
            return ("ArenaVoteScreen1", "T_Mod3")
        E_VoteMenu_Widget.ArenaVoteScreen1_Txt_VoteAmmount =>
            return ("ArenaVoteScreen1", "Txt_VoteAmmount")
        E_VoteMenu_Widget.ArenaVoteScreen2_C_ColorBlock =>
            return ("ArenaVoteScreen2", "C_ColorBlock")
        E_VoteMenu_Widget.ArenaVoteScreen2_T_Arena =>
            return ("ArenaVoteScreen2", "T_Arena")
        E_VoteMenu_Widget.ArenaVoteScreen2_T_Mod1 =>
            return ("ArenaVoteScreen2", "T_Mod1")
        E_VoteMenu_Widget.ArenaVoteScreen2_T_Mod2 =>
            return ("ArenaVoteScreen2", "T_Mod2")
        E_VoteMenu_Widget.ArenaVoteScreen2_T_Mod3 =>
            return ("ArenaVoteScreen2", "T_Mod3")
        E_VoteMenu_Widget.ArenaVoteScreen2_Txt_VoteAmmount =>
            return ("ArenaVoteScreen2", "Txt_VoteAmmount")
        E_VoteMenu_Widget.ArenaVoteScreen3_C_ColorBlock =>
            return ("ArenaVoteScreen3", "C_ColorBlock")
        E_VoteMenu_Widget.ArenaVoteScreen3_T_Arena =>
            return ("ArenaVoteScreen3", "T_Arena")
        E_VoteMenu_Widget.ArenaVoteScreen3_T_Mod1 =>
            return ("ArenaVoteScreen3", "T_Mod1")
        E_VoteMenu_Widget.ArenaVoteScreen3_T_Mod2 =>
            return ("ArenaVoteScreen3", "T_Mod2")
        E_VoteMenu_Widget.ArenaVoteScreen3_T_Mod3 =>
            return ("ArenaVoteScreen3", "T_Mod3")
        E_VoteMenu_Widget.ArenaVoteScreen3_Txt_VoteAmmount =>
            return ("ArenaVoteScreen3", "Txt_VoteAmmount")
        E_VoteMenu_Widget.ArenaHighlight1_C_ColorBlock =>
            return ("ArenaHighlight1", "C_ColorBlock")
        E_VoteMenu_Widget.ArenaHighlight2_C_ColorBlock =>
            return ("ArenaHighlight2", "C_ColorBlock")
        E_VoteMenu_Widget.ArenaHighlight3_C_ColorBlock =>
            return ("ArenaHighlight3", "C_ColorBlock")
        E_VoteMenu_Widget.ArenaVoteButtons_B_Vote_3 =>
            return ("ArenaVoteButtons", "B_Vote_3")
        E_VoteMenu_Widget.ArenaVoteButtons_B_Vote_2 =>
            return ("ArenaVoteButtons", "B_Vote_2")
        E_VoteMenu_Widget.ArenaVoteButtons_B_Vote_1 =>
            return ("ArenaVoteButtons", "B_Vote_1")
        E_VoteMenu_Widget.ArenaSelect_C_ColorBlock =>
            return ("ArenaSelect", "C_ColorBlock")
        E_VoteMenu_Widget.ArenaSelect_Txt_Timer =>
            return ("ArenaSelect", "Txt_Timer")
        E_VoteMenu_Widget.ArenaSelect_Txt_VoteForText =>
            return ("ArenaSelect", "Txt_VoteForText")
        E_VoteMenu_Widget.ArenaSelect_Txt_Console =>
            return ("ArenaSelect", "Txt_Console")
VoteMenu_BindToTriggers:= class<concrete>(Button_To_Trigger):
    @editable
    OnButtonPressed : ?E_VoteMenu_Button = false
    GetButtonName<override>():tuple(string,string)=
        if(Button := OnButtonPressed?):
            return Button.GetName()
        return ("", "")
MakeVoteMenu_BindToTriggers<constructor>(OldData : VoteMenu_BindToTriggers)<transacts> := VoteMenu_BindToTriggers:
    OnButtonPressed := OldData.OnButtonPressed
    Triggers := OldData.Triggers
VoteMenu_ShowGroupSettings:= class<concrete>(Show_Group_Settings):
    @editable
    Group : E_VoteMenu_Group = E_VoteMenu_Group.root
    GetGroupName<override>():string=
        set GroupName = Group.GetName()
        return GroupName
VoteMenu_ShowGroups:= class<concrete>(Show_Groups):
    @editable
    OnButtonPressed : ?E_VoteMenu_Button = false
    @editable
    GroupsSettings : []VoteMenu_ShowGroupSettings = array{}
    GetButtonName<override>():tuple(string,string)=
        if(Button := OnButtonPressed?):
            return Button.GetName()
        return ("", "")
    GetGroupSettings<override>():[]Show_Group_Settings=
        return GroupsSettings
MakeVoteMenu_ShowGroups<constructor>(OldData : VoteMenu_ShowGroups)<transacts> := VoteMenu_ShowGroups:
    OnButtonPressed := OldData.OnButtonPressed
    OnTrigger := OldData.OnTrigger
    GroupsSettings := OldData.GroupsSettings
VoteMenu_ExitButton:= class<concrete>(Exit_Button):
    @editable
    OnButtonPressed : ?E_VoteMenu_Button = false
    GetButtonName<override>():tuple(string,string)=
        if(Button := OnButtonPressed?):
            return Button.GetName()
        return ("", "")
MakeVoteMenu_ExitButton<constructor>(OldData : VoteMenu_ExitButton)<transacts> := VoteMenu_ExitButton:
    OnButtonPressed := OldData.OnButtonPressed
    OnTrigger := OldData.OnTrigger

