using { /UnrealEngine.com/Temporary/SpatialMath }
using { /UnrealEngine.com/Temporary/Diagnostics }
using { /Verse.org/Random }
using { /Verse.org/Colors }
using { /Verse.org/Simulation }
using { /Verse.org/Simulation/Tags }
using { /Fortnite.com/FortPlayerUtilities }
using { /Fortnite.com/Characters }
using { /Fortnite.com/Game }
using { /Fortnite.com/Devices }
using { /Verse.org/Concurrency }

using { /Fortnite.com/Devices }
using { /Fortnite.com/AI}
using { /Fortnite.com/Teams}
using { /Fortnite.com/UI }
using { /Verse.org/Simulation }
using { /Verse.org/Random }
using { /Verse.org/Colors }
using { /Verse.org/Assets }
using { /Fortnite.com/Characters }
using { /Fortnite.com/FortPlayerUtilities }
using { /Fortnite.com/Game }
using { /Fortnite.com/Teams }
using { /Verse.org/Simulation/Tags }
using { /UnrealEngine.com/Temporary/Diagnostics }
using { /UnrealEngine.com/Temporary/SpatialMath }
using { /UnrealEngine.com/Temporary/UI }


# slider:=type {Value:float where 0.0 <= Value, Value <= 1.0}

LOG := class(log_channel){}

# Print that transacts
PrintT(Message: string)<transacts>:void=
    Logger : log = log{Channel := LOG}
    Logger.Print(Message)

EaseOutExpo(X: float): float =
    return if(X = 1.0) then 1.0 else 1.0 - Pow(2.0, -10.0 * X)

EaseInExpo(X: float)<reads><allocates>: float =
    return if(X = 0.0) then 0.0 else Pow(2.0, 10.0 * X - 10.0)

EaseOutQuart(X: float): float =
    return 1.0 - Pow(1.0 - X, 4.0)

EaseOutSine(X: float)<reads><allocates>: float =
    return Sin((X * PiFloat) / 2.0)

Float(Number : int)<reads><allocates>:float=
    return Number * 1.0
    
operator'+'(I:int, F:float):float=
    return I * 1.0 + F

operator'+'(F:float, I:int):float=
    return F + I * 1.0

operator'-'(I:int, F:float):float=
    return I * 1.0 - F

operator'-'(F:float, I:int):float=
    return F - I * 1.0

operator'/'(I:int, F:float):float=
    return I * 1.0 / F

operator'/'(F:float, I:int):float=
    return F / (I * 1.0)

(Input:[]t where t:type).QuickSort(CompareFunction(Element1:t, Element2:t)<transacts>:int)<transacts>:[]t = 

    if (Input.Length <= 1):
        return Input

    var Less : []t = array {}
    var Equal : []t = array {}
    var Greater : []t = array {}

    if (Pivot := Input[Int[Input.Length * 0.5]]):
        for (Element : Input):
            Result := CompareFunction(Element, Pivot)
            ArrayElement := array { Element }
            if (Result < 0):
                set Less += ArrayElement
            else if (Result > 0):
                set Greater += ArrayElement
            else:
                set Equal += ArrayElement

    Less.QuickSort(CompareFunction) + Equal + Greater.QuickSort(CompareFunction)

SortIntDescending(Element1: int, Element2: int): int =
    if (Element1 < Element2):
        1
    else if (Element1 > Element2):
        -1
    else:
        0

GetRandomWeightedElement(ThresholdMap:[int]t, TotalWeight: int where t:type)<transacts><decides>:t=
    RandomInt := GetRandomInt(1, TotalWeight)
    var RandomElement : ?t = false
    loop:
        for(Threshold -> Element : ThresholdMap, RandomInt <= Threshold):
            set RandomElement = option{Element}
            break
        break
    RandomElement?

# Returns the same map with the provided index removed, unchanged map if Index not found
(Input:[t]u where t:subtype(comparable), u:type).WithRemovedKey(KeyToRemove:t)<transacts>:[t]u =
    if(not Input[KeyToRemove]):
        return Input
        
    var RetMap : [t]u = map{}
    for (Key -> Value : Input, Key <> KeyToRemove):
        option{set RetMap[Key] = Value}
    return RetMap

(Input:[t]u where t:subtype(comparable), u:type).GetRandomElement()<transacts><decides>:tuple(t, u) =
    var RetTuple : ?tuple(t, u) = false

    StopIndex := GetRandomInt(0, Input.Length - 1)
    var I : int = 0

    for(Index -> Element : Input):
        if(I = StopIndex):
            set RetTuple = option{(Index, Element)}
        set I += 1
    return RetTuple?

(Input:[t]u where t:subtype(comparable), u:type).ToTupleArray():[]tuple(t,u)=
    var RetArray : []tuple(t,u) = array{}
    for(Key -> Element : Input):
        set RetArray += array{(Key, Element)}
    return RetArray

(Input:[t]u where t:subtype(comparable), u:type).ToArray()<transacts>:[]u=
    var RetArray : []u = array{}
    for(Element : Input):
        set RetArray += array{Element}
    return RetArray

(Input:[]t where t:type).GetRandomElementFromArray()<transacts><decides>:tuple(t, int) =
    StopIndex := GetRandomInt(0, Input.Length - 1)
    return (Input[StopIndex], StopIndex)


(Input:[t]u where t:subtype(comparable), u:type).GetRandomKey()<transacts><decides>:t =
    return Input.GetRandomElement[](0)
    
(Input:[t]u where t:subtype(comparable), u:type).GetRandomValue()<transacts><decides>:u =
    return Input.GetRandomElement[](1)

(Array:[]t where t:subtype(comparable)).ToMapIndexes()<transacts>:[t]logic=
    var RetMap : [t]logic = map{}
    for(Value : Array, set RetMap[Value] = true):
    return RetMap

(Map:[t]u where t:subtype(comparable), u:type).Keys()<transacts>:[]t= for (Key -> __ : Map) { Key }
(Map:[t]u where t:subtype(comparable), u:type).Values()<transacts>:[]u= for (Value : Map) { Value }

GetRandomColor(?MinBrightness : float = 0.0):color= color:
    R := GetRandomFloat(MinBrightness, 1.0)
    G := GetRandomFloat(MinBrightness, 1.0)
    B := GetRandomFloat(MinBrightness, 1.0)

BuildPlayerMessage<localizes>(Prefix:string, Agent:agent, Suffix: string):message = "{Prefix}{Agent}{Suffix}"

FindCharIndex(SearchChar:char, Str:[]char)<transacts><decides>:int=
    var FoundIndex:int = 0
    var Found:logic = false
    for (X -> Char : Str, not Found?):
        if (Char = SearchChar):
            set FoundIndex = X
    FoundIndex

ToTruncatedString(Number:float, Decimals:int, ?ForceDisplayIfPointZero :logic = false)<transacts>:[]char=
    var Str:[]char = ToString(Number)

    if:
        DotIndex := FindCharIndex['.', Str]

        DecimalIndex := if (Decimals > 0) then Min(DotIndex+Decimals+1,Str.Length) else DotIndex

        if(not ForceDisplayIfPointZero?, Decimals > 0, NumbersAfterDot := Str.Slice[DotIndex + 1, DecimalIndex]):
            var CompareStr : string = ""
            for(I := 1..Decimals):
                set CompareStr += "0"

            if(NumbersAfterDot = CompareStr):
                _Str := Str.Slice[0, DotIndex]
                set Str = _Str
                
            else:
                _Str := Str.Slice[0, DecimalIndex]
                set Str = _Str
        else:
            _Str := Str.Slice[0, DecimalIndex]
            set Str = _Str
    Str

# Returns angle between -180 and 180
(NormalizedVector: vector2).GetWrappedAngle()<transacts>: float =
    # Don't ask me why
    Sign := if(NormalizedVector.Y < 0.0) then -1 else 1
    return Sign * RadiansToDegrees(ArcCos(DotProduct(NormalizedVector, vector2{X:=1.0,Y:=0.0}) / NormalizedVector.Length()))
    # return RadiansToDegrees(ArcTan(-NormalizedVector.Y, 1.0-NormalizedVector.X))

(Vector3: vector3).ToVector2()<transacts>: vector2 = vector2{X:=Vector3.X, Y:=Vector3.Y}

# Works for every characters?
(Character: fort_character).GetPlayerFeetTranslation()<transacts>:vector3 =
    return Character.GetTransform().Translation.PlayerFeetTranslated(Character)

(Vector3: vector3).PlayerFeetTranslated(Character: fort_character)<transacts>:vector3 =
    var HeightOffsetToPlayerFeet : float = 77.15
    if(Character.IsCrouching[]):
        set HeightOffsetToPlayerFeet -= 25.0

    return Character.GetTransform().Translation - vector3{Z:= HeightOffsetToPlayerFeet}

(Vector3: vector3).GetClosest(Elements: []t where t:subtype(positional))<transacts><decides>:t=
    var RetElement:?t = false
    var ClosestDistanceFound : float = 9999999.9

    for(Element : Elements):
        DistanceToElement := Distance(Vector3, Element.GetTransform().Translation)
        if(DistanceToElement < ClosestDistanceFound):
            set ClosestDistanceFound = DistanceToElement
            set RetElement = option{Element}

    return RetElement?

Clamp(Vector3: vector3, MinVector: vector3, MaxVector: vector3)<transacts>:vector3=
    return vector3{
        X := Clamp(Vector3.X, MinVector.X, MaxVector.X)
        Y := Clamp(Vector3.Y, MinVector.Y, MaxVector.Y)
        Z := Clamp(Vector3.Z, MinVector.Z, MaxVector.Z)
    }

DefaultFortniteGridSize : vector3 = vector3{X:=512.0,Y:=512.0,Z:=384.0}

# Calculates the distance between two rotation yaw values, taking into account the wrapping around of -180 to 180.
# Thanks ChatGTP! (I'm bad at math)
GetYawDistance(Yaw1: float, Yaw2: float)<transacts>: float =
    var Diff:float = Yaw2 - Yaw1
    if(Diff > 180.0):
        set Diff -= 360.0
    else if(Diff < -180.0):
        set Diff += 360.0
    return Abs(Diff)

GetHumanReadableLargeInt(Value: int): string =
    Thresholds : [int]string = map{
        1000000000 => "b",
        1000000 => "m",
        1000 => "k"
    }

    for(Threshold -> Unit : Thresholds):
        if(Value >= Threshold, Compacted := Float(Value) / Float(Threshold)):
            return "{ToTruncatedString(Compacted, 2)}{Unit}"

    return "{Value}"

GetHumanReadableWeight(Value: int): string =
    Thresholds : [int]string = map{
        1000000000000 => "Gt",
        1000000000 => "Mt",
        1000000 => "kt",
        1000 => "t",
        1 => "kg"
    }

    for(Threshold -> Unit : Thresholds):
        if(Value >= Threshold, Compacted := Float(Value) / Float(Threshold)):
            return "{ToTruncatedString(Compacted, 2)}{Unit}"

    return "{Value}"


# Generic class which you can set up to wait on multiple events
# and return the event that does occur and its payload.
# When you instantiate the class, you define the payload type you want. For example:
#     SharedWaiter:shared_waiter(int) = shared_waiter(int){}
shared_waiter<public> := class:
    # Single custom event to wait and signal on
    SharedEvent<internal>:event() = event(){}
        
    # Wait on the custom shared event. Result is the associated payload.
    Await()<suspends>:void=

        # Avoid looping indefinitly if we already Signaled the event upon calling this function
        branch:
            Sleep(0.0)
            block: 
                if(AwaitableCount = 0):
                    SharedEvent.Signal()

        SharedEvent.Await()

    var AwaitableCount : int = 0

    # Add the event to wait on and the payload associated with the event if it occurs
    AddAwaitable<public>(Awaitable:awaitable(any)):void=
        set AwaitableCount += 1
        spawn{AwaitAllSignals(Awaitable)}
    
    # Race between the multiple events. First event to occur signals the shared event
    # and cancels the wait for any remaining events that are still waiting to occur.
    AwaitAllSignals<internal>(Awaitable:awaitable(any))<suspends>:void=
        Awaitable.Await()

        set AwaitableCount -= 1
        if(AwaitableCount <= 0):
            SharedEvent.Signal()

# my_subscribable_player := class:
#     var Functions<private>:[]type {__(Player:player):void} = array{}
    
#     Subscribe(Function:type {__(Player:player):void}):void =
#         set Functions += array{Function}
    
#     Signal(Player: player):void =
#         for(Function : Functions):
#             Function(Player)

# my_subscribable_attached_item__character := class:
#     var Functions<private>:[]type {__(:player_attached_item, :fort_character):void} = array{}
    
#     Subscribe(Function:type {__(:player_attached_item, :fort_character):void}):void =
#         set Functions += array{Function}
    
#     Signal(AttachedItem: player_attached_item, Character: fort_character):void =
#         for(Function : Functions):
#             Function(AttachedItem, Character)
            

awaitable_cinematic_sequence_device := class<concrete>:
    @editable
    Sequence : cinematic_sequence_device = cinematic_sequence_device{}

    @editable
    Duration : float = 0.0

    # OnEnded : event() = event(){}

    Play(?ForceStopAfterDuration : logic = false):task(void)=
        return spawn{_Play(?ForceStopAfterDuration := ForceStopAfterDuration)}

    Play(Agent: agent, ?ForceStopAfterDuration : logic = false):task(void)=
        return spawn{_Play(?Agent := option{Agent}, ?ForceStopAfterDuration := ForceStopAfterDuration)}

    _Play(?Agent: ?agent = false, ?ForceStopAfterDuration : logic = false)<suspends>:void=
        if(Duration > 0.0):
            if(_Agent := Agent?):
                Sequence.Play(_Agent)
            else:
                Sequence.Play()

            Sleep(Duration)

            if(ForceStopAfterDuration?):
                if(_Agent := Agent?):
                    Sequence.Stop(_Agent)
                else:
                    Sequence.Stop()
                    
                Sleep(0.0)
        else:
            Sleep(0.0) # Or awaiting will last forever
        # OnEnded.Signal()

item_rarity := enum:
    Common
    Uncommon
    Rare
    Epic
    Legendary
    Mythic
    Exotic

ToString(Rarity: item_rarity)<transacts>:string=
    case(Rarity):
        item_rarity.Common => "Common"
        item_rarity.Uncommon => "Uncommon"
        item_rarity.Rare => "Rare"
        item_rarity.Epic => "Epic"
        item_rarity.Legendary => "Legendary"
        item_rarity.Mythic => "Mythic"
        item_rarity.Exotic => "Exotic"

<#GetUniqueCreativeObjectsWithTag(Tag:tag)<transacts>:[]creative_object_interface=
    var ParsedLocations : [string]logic = map{}
    var UniqueObjects : []creative_object_interface = array{}

    for(CreativeObject : GetCreativeObjectsWithTag(Tag)):
        Location := CreativeObject.GetTransform().Translation
        LocationIdentifier := "{Location.X}_{Location.Y}_{Location.Z}"

        if(not ParsedLocations[LocationIdentifier]):
            option{set ParsedLocations[LocationIdentifier] = true}
            set UniqueObjects += array{CreativeObject}
        else:
            PrintT("Found object with duplicate Verse Tag at {Location}")
    return UniqueObjects#>
    
# Returns the player who's in the same team as Agent, only works in FFA since it's limited to 1 player per team
(CreativeDevice:creative_device).GetInstigatorFFAPlayer(Agent:agent)<transacts><decides>:player=
    var RetPlayer : ?player = false
    if(Player := player[Agent]):
        set RetPlayer = option{Player}
    else:
        TeamCollection := CreativeDevice.GetPlayspace().GetTeamCollection()
        if(
            Team := TeamCollection.GetTeam[Agent],
            TeamAgents := CreativeDevice.GetPlayers(Team),
            TeamPlayers := for (TeamAgent : TeamAgents, Player := player[TeamAgent], Player.GetFortCharacter[], not Player.IsSpectator[]) { Player },
            TeamPlayers.Length = 1,
            FoundPlayer := TeamPlayers[0]
        ):
            set RetPlayer = option{FoundPlayer}
    RetPlayer?

(Teleporter: teleporter_device).CustomTeleport(Agent: agent, Translation: vector3, Rotation: rotation)<suspends>:void=
    race:
        Sleep(1.0)
        loop:
            if(Teleporter.TeleportTo[Translation, Rotation]):
                Teleporter.Teleport(Agent)
                break
            Sleep(0.0)

#TEAM MINIGAME CODE UPDATE
#Changed this to support teleporting teams not individual players
#UPDATE EXTRA: MODIFIED TO WORK WITH NEW CUSTOM TEAM LOGIC
(Positionals: []positional).BulkTeleport(PlayerTeam : custom_team, Agents: []agent, Teleporter: teleporter_device, ?TranslationOffset: vector3 = vector3{}, ?YawOffset : float = 0.0):void=
    #From Single Teleport
    Print("Positional Length = {Positionals.Length}")
    #We are using a single barrier. In this case for Find Closest Barrier.
    if(Positionals.Length = 1):
        if(TeleportTransform := Positionals[0]):
            for(Agent : Agents):
                Print("Teleporting to closest team barrier")
                FinalRotation := TeleportTransform.GetTransform().Rotation.ApplyYaw(DegreesToRadians(YawOffset))
                spawn{Teleporter.CustomTeleport(Agent, TeleportTransform.GetTransform().Translation + TranslationOffset, FinalRotation)}
    #We are going to use multiple barriers
    else:
        if(TeleportTransform := Positionals[PlayerTeam.TeamIndex]):
            for(Agent : Agents):
                Print("Teleporting to base teleporter")
                FinalRotation := TeleportTransform.GetTransform().Rotation.ApplyYaw(DegreesToRadians(YawOffset))
                spawn{Teleporter.CustomTeleport(Agent, TeleportTransform.GetTransform().Translation + TranslationOffset, FinalRotation)}

#TEAM MINIGAME CODE UPDATE - Will refine soon, built to only teleport 1 player from the team that matches the player we inputted
#TODO: REFINE CODE SO WE ONLY CHECK THE PLAYER, WHAT TEAM THEY ARE ON AND TELEPORT THEM TO THE CORRECT LOCATION OF THEIR TEAM BARRIER
(Positionals: []positional).SingleTeleport(PlayerTeam : custom_team, InPlayer : player, Teleporter: teleporter_device, ?TranslationOffset: vector3 = vector3{}, ?YawOffset : float = 0.0):void=
    Print("Positional Length = {Positionals.Length}")
    #We are using a single barrier. In this case for Find Closest Barrier.
    if(Positionals.Length = 1):
        if(TeleportTransform := Positionals[0]):
            Print("Teleporting to closest team barrier")
            FinalRotation := TeleportTransform.GetTransform().Rotation.ApplyYaw(DegreesToRadians(YawOffset))
            spawn{Teleporter.CustomTeleport(InPlayer, TeleportTransform.GetTransform().Translation + TranslationOffset, FinalRotation)}
    #We are going to use multiple barriers
    else:
        if(TeleportTransform := Positionals[PlayerTeam.TeamIndex]):
            Print("Teleporting to base teleporter")
            FinalRotation := TeleportTransform.GetTransform().Rotation.ApplyYaw(DegreesToRadians(YawOffset))
            spawn{Teleporter.CustomTeleport(InPlayer, TeleportTransform.GetTransform().Translation + TranslationOffset, FinalRotation)}
    Print("Teleport Player To Base {PlayerTeam.TeamIndex}")

(Positionals: []positional).BulkTeleportOld(Agents: []agent, Teleporter: teleporter_device, ?TranslationOffset: vector3 = vector3{}, ?YawOffset : float = 0.0):void=
    var PositionalIndex : int = 0
    var ShuffledPostionals : []positional = Shuffle(Positionals)
    for(
        Agent : Agents
        RandomPositional := ShuffledPostionals[PositionalIndex]
        not creative_prop[RandomPositional] or creative_prop[RandomPositional].IsValid[]
        Transform := RandomPositional.GetTransform()
    ):
        FinalRotation := Transform.Rotation.ApplyYaw(DegreesToRadians(YawOffset))

        spawn{Teleporter.CustomTeleport(Agent, Transform.Translation + TranslationOffset, FinalRotation)}
        set PositionalIndex = PositionalIndex + 1

        # Shuffle and continue if we used all available positionals
        if(PositionalIndex >= ShuffledPostionals.Length):
            set ShuffledPostionals = Shuffle(Positionals)
            set PositionalIndex = 0

(Positionals: []positional).CollectibleObjectTeleport(CollectibleObject: collectible_object_device, Teleporter: teleporter_device, ?TranslationOffset: vector3 = vector3{}, ?YawOffset : float = 0.0):void=
    var ShuffledPostionals : []positional = Shuffle(Positionals)
    if(RandomPositional := ShuffledPostionals[0], not creative_prop[RandomPositional] or creative_prop[RandomPositional].IsValid[]):
        Transform := RandomPositional.GetTransform()
        FinalRotation := Transform.Rotation.ApplyYaw(DegreesToRadians(YawOffset))
        Print("Final Rotation - {FinalRotation}")
        if(CollectibleObject.TeleportTo[Transform.Translation + TranslationOffset, FinalRotation]):
            Print("Teleported Collectible")

SpawnPropFromAsset(Asset: creative_prop_asset, Transform: transform)<suspends>:?creative_prop=
    var RetProp : ?creative_prop = false
    race:
        Sleep(2.0)
        loop:
            Result := SpawnProp(Asset, Transform)
            
            if(NewProp := Result(0)?):
                set RetProp = option{NewProp}
                break
            Sleep(0.0)
    RetProp

(Awaitables:[]awaitable(t) where t:type).Race<public>()<suspends>:?t=
    RaceAwaitables(Awaitables, 0)

RaceAwaitables<internal>(Awaitables:[]awaitable(t), Index:int where t:type)<suspends>:?t=
    if (Awaitable := Awaitables[Index]):
        race:
            RaceAwaitables(Awaitables, Index + 1)
            block:
                Result:t = Awaitable.Await()
                option{Result}
    else:
        if (Awaitables.Length > 0):
            Sleep(Inf)
        false

(Awaitables:[]awaitable(t) where t:type).Sync<public>()<suspends>:[]t=
    SyncAwaitables(Awaitables, 0)

SyncAwaitables<internal>(Awaitables:[]awaitable(t), Index:int where t:type)<suspends>:[]t=
    if (Awaitable := Awaitables[Index]):
        Result:tuple(t, []t) = sync:
            Awaitable.Await()
            SyncAwaitables(Awaitables, Index + 1)
        array{Result(0)} + Result(1)
    else:
        array{}

(ConditionalButton : conditional_button_device).CustomConsume(Agent:agent, ItemIndex:int, Amount:int):void=
    MaxConsumeAmount := 10000000 # Maximum amount a conditional button can consume at a time
    if(
        Amount > 0,
        MaxConsumeAmountIterations := Int[Float(Amount) / Float(MaxConsumeAmount)],
        Remainder := Mod[Amount, MaxConsumeAmount]
    ):

        if(MaxConsumeAmountIterations >= 1):
            ConditionalButton.SetItemCountRequired(ItemIndex, MaxConsumeAmount)
            for(__ := 1..MaxConsumeAmountIterations):
                ConditionalButton.Activate(Agent)
        
        ConditionalButton.SetItemCountRequired(ItemIndex, Remainder)
        ConditionalButton.Activate(Agent)

suspends_wrapper<internal>(t:type) := class:
    Callback<internal> : type{_(:t)<suspends>:void}

    Spawn<internal>(Payload:t):void=
        spawn {Callback(Payload)}

(Subscribable:subscribable(t) where t:type).SubscribeSuspends<public>(Callback:type{_(:t)<suspends>:void})<transacts>:cancelable=
    Wrapper:suspends_wrapper(t) = suspends_wrapper(t){Callback := Callback}
    Subscribable.Subscribe(Wrapper.Spawn)

(Awaitable:awaitable(t) where t:type).Debounce<public>(Callback(:t):void, Interval:float):void=
    spawn {AwaitDebounce(Awaitable, Callback, Interval)}

AwaitDebounce<internal>(Awaitable:awaitable(t), Callback(:t):void, Interval:float where t:type)<suspends>:void=
    var LastResult:?t = false
    loop:
        set LastResult = race:
            block:
                Result:t = Awaitable.Await()
                option{Result}
            block:
                if (not LastResult?):
                    Sleep(Inf)
                Sleep(Interval)
                if (Result := LastResult?):
                    Callback(Result)
                    set LastResult = false
                false

StringToMessage<localizes>(value:string) : message = "{value}"

# Fix for fort_team_collection.GetAgents[] not working properly (use playspace players instead of GetAgents to avoid iterating through a huge players array after a 4h session)
(CreativeDevice:creative_device).GetPlayers(Team: team)<transacts>:[]player=
    Playspace := CreativeDevice.GetPlayspace()
    TeamCollection := Playspace.GetTeamCollection()
    return for(Player : Playspace.GetPlayers(), TeamCollection.IsOnTeam[Player, Team]) { Player }

(CreativeDevice:creative_device).FindEmptyTeam(ExcludedTeams : []int)<transacts><decides>:team=
    var RetTeam : ?team = false

    TeamCollection := CreativeDevice.GetPlayspace().GetTeamCollection()

    loop:
        for(
            Index -> Team : TeamCollection.GetTeams(),
            not ExcludedTeams.Find[Index+1],
            Players := for(Player: CreativeDevice.GetPlayers(Team)) { Player },
            Players.Length <= 0
        ):
            set RetTeam = option{Team}
            break
        break

    RetTeam?  

(TeamCollection: fort_team_collection).ForceAddToTeam(InAgent:agent, InTeam:team)<suspends>:void=
    race:
        Sleep(1.0)
        loop:
            option{TeamCollection.AddToTeam[InAgent, InTeam]}
            if(TeamCollection.IsOnTeam[InAgent, InTeam]):
                break
            Print("AddToTeam failed")
            Sleep(0.0)

AsyncGetFortCharacter(Agent: agent, Timeout: float, ?Active : logic = false)<suspends>:?fort_character=
    var FoundCharacter : ?fort_character = false
    race:
        Sleep(Timeout)
        loop:
            if(Character := Agent.GetFortCharacter[], ((not Active?) or Character.IsActive[])):
                set FoundCharacter = option{Character}
                break
            Sleep(0.0)

    return FoundCharacter

(Vector3: vector3).ToComparable()<transacts>:string=
    if(
        IntX := Int[Vector3.X]
        IntY := Int[Vector3.Y]
        IntZ := Int[Vector3.Z]
    ):
        "{IntX}_{IntY}_{IntZ}"
    else:
        "{Vector3}"

var TrackerUpdatedEvents : weak_map(session, [string]custom_subscribable_player) = map{}
(Tracker: tracker_device).UpdatedEvent():custom_subscribable_player =
    Session := GetSession()
    # Trackers are not comparable by default
    Location := Tracker.GetTransform().Translation.ToComparable()
    if(not TrackerUpdatedEvents[Session]):
        option{set TrackerUpdatedEvents[Session] = map{}}

    if(not TrackerUpdatedEvents[Session][Location]):
        option{set TrackerUpdatedEvents[Session][Location] = custom_subscribable_player{}}

    if(Event := TrackerUpdatedEvents[Session][Location]):
        return Event
    return custom_subscribable_player{}

EmptyMessage<localizes>: message = ""
operator'+'<localizes>(A: message, B: message): message = "{A}{B}"
operator'+'<localizes>(A: message, B: string): message = "{A}{B}"
operator'+'<localizes>(A: string, B: message): message = "{A}{B}"
operator'+'<localizes>(A: message, B: int): message = "{A}{B}"
operator'+'<localizes>(A: message, B: float): message = "{A}{B}"
operator'+'<localizes>(A: message, B: agent): message = "{A}{B}"

# Add extra data to listenable with (agent) -> void
(Listenable : listenable(agent)).SubscribeAgent(OutputFunc : tuple(agent, t)->void, ExtraData : t where t:type) : cancelable =
    Wrapper := wrapper_agent(t){ExtraData := ExtraData, OutputFunc := OutputFunc}
    Listenable.Subscribe(Wrapper.InputFunc)

wrapper_agent(t : type) := class():
    ExtraData : t;
    OutputFunc : tuple(agent, t) -> void
    InputFunc(Agent : agent):void = OutputFunc(Agent, ExtraData)

# Add extra data to listenable with (?agent) -> void
(Listenable : listenable(?agent)).SubscribeQuestionAgent(OutputFunc : tuple(?agent, t)->void, ExtraData : t where t:type) : cancelable =
    Wrapper := wrapper_question_agent(t){ExtraData := ExtraData, OutputFunc := OutputFunc}
    Listenable.Subscribe(Wrapper.InputFunc)
    
wrapper_question_agent(t : type) := class():
    ExtraData : t;
    OutputFunc : tuple(?agent, t) -> void
    InputFunc(Agent : ?agent):void = OutputFunc(Agent, ExtraData)

(Listenable : listenable(tuple())).SubscribeEmptyTuple(OutputFunc : t -> void, ExtraData : t where t:type) : cancelable =
    Wrapper := wrapper_empty_tuple(t) {ExtraData := ExtraData, OutputFunc := OutputFunc}
    Listenable.Subscribe(Wrapper.InputFunc)

wrapper_empty_tuple(t : type) := class():
    ExtraData : t;
    OutputFunc : t -> void
    InputFunc():void = OutputFunc(ExtraData)