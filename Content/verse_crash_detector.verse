####################################################
# Copyright 2024 LAMA (https://x.com/lama_creator)
# Permission is hereby granted, free of charge, to any person obtaining a copy of this software and associated documentation files (the "Software"), to deal in the Software without restriction, including without limitation the rights to use, copy, modify, merge, publish, distribute, sublicense, and/or sell copies of the Software, and to permit persons to whom the Software is furnished to do so, subject to the following conditions:
# The above copyright notice and this permission notice shall be included in all copies or substantial portions of the Software.
# THE SOFTWARE IS PROVIDED "AS IS", WITHOUT WARRANTY OF ANY KIND, EXPRESS OR IMPLIED, INCLUDING BUT NOT LIMITED TO THE WARRANTIES OF MERCHANTABILITY, FITNESS FOR A PARTICULAR PURPOSE AND NONINFRINGEMENT. IN NO EVENT SHALL THE AUTHORS OR COPYRIGHT HOLDERS BE LIABLE FOR ANY CLAIM, DAMAGES OR OTHER LIABILITY, WHETHER IN AN ACTION OF CONTRACT, TORT OR OTHERWISE, ARISING FROM, OUT OF OR IN CONNECTION WITH THE SOFTWARE OR THE USE OR OTHER DEALINGS IN THE SOFTWARE.
####################################################

####################################################
# TODO: Allow script to dynamically register/unregister devices that gets loaded/unloaded from data layers
# TODO: Find a way to detect the device that holds the crashed function instead of the one that calls the crashed function
# TODO: Allow for more granular crash reports
# TODO: Add ability to not restart round (for certain scripts?) and to keep listening for crashes after a crash occurs
####################################################
# NOTE: Doesn't detect server crashes
# NOTE: The device that triggers the crash doesn't necessarily holds the crashing function (they called the crashing function)
####################################################

using { /Fortnite.com/Devices }
using { /Verse.org/Simulation }
using { /Verse.org/Random }
using { /Verse.org/Colors }

# Allow to print an error without crashing the whole device
PrintError(Message: string):void = Print("Error: {Message}", ?Color := NamedColors.Red)
PrintOutputLog(Message: string):void = Print("{Message}", ?Duration := 0.0)

var RegisteredCrashableDevices : weak_map(session, [crashable_device]crashable_device) = map{}
RegisterCrashableDevice(Device: crashable_device):void=
    Session := GetSession()
    if(not RegisteredCrashableDevices[Session]):
        option{set RegisteredCrashableDevices[Session] = map{}}

    if(not RegisteredCrashableDevices[Session][Device]):
        PrintOutputLog("Successfully registered crashable device {Device}")
        option{set RegisteredCrashableDevices[Session][Device] = Device}
    else:
        PrintError("Tried to register crashable device {Device} twice")

crashable_device := class<unique>(creative_device):

    @editable
    DeviceName : string = ""

    # Naming convention for the "Event Name" is crash_{device_name}
    @editable
    CrashAnalytics : ?analytics_device = false

    var IsAlive<private> : logic = true
    # var IsReady<private> : logic = false

    # ReadyEvent<private> : event() = event(){}
    PingEvent<private>  : event() = event(){}

    # Any (3568) error here means you need to rename all your (crashable_device).OnBegin() functions by OnCrashableDeviceBegin()
    OnBegin<override><final>()<suspends>:void=
        # This suspending thread has to be spawned by Self from Self in order for it to crash when something else crashes inside the device
        spawn{WatchPingEvents()}

        # Register before overridden OnBegin so we don't have to deal with OnBegin methods that never finish
        RegisterCrashableDevice(Self)

        OnCrashableDeviceBegin()

    OnCrashableDeviceBegin<public>()<suspends>:void={}

    # This thread will crash if Verse crashes
    WatchPingEvents<private>()<suspends>:void=
        loop:
            PingEvent.Await()
            set IsAlive = true

    # needs no_rollback :'(
    Ping<public>():logic=
        set IsAlive = false
        PingEvent.Signal()
        IsAlive

    # OVERRIDE ME: If you don't want to use the DeviceName field because you already have a way to dynamically generate a name
    GetDeviceName()<transacts><decides>:string = (__:?string=false)?

ToString(CrashableDevice: crashable_device)<transacts>:string =
    CrashableDevice.GetDeviceName[] or (CrashableDevice.DeviceName.Length > 0 and CrashableDevice.DeviceName or "UNKNOWN")

verse_crash_detector := class(creative_device):

    @editable
    Timer : timer_device = timer_device{}

    @editable
    CRASH_BUTTON : ?button_device = false

    var<private> WatchedDevices : [crashable_device]crashable_device = map{}

    OnBegin<override>()<suspends>:void=
        Timer.Disable() # Set default behaviour to "safe mode" (do not end round)

        # Listen to registered devices crashes
        spawn{WatchCrashes()}

        if(CrashButton := CRASH_BUTTON?):
            CrashButton.InteractedWithEvent.Await()
            Err("!!! CrashTest !!!") # Crashes this script, mainly for testing the 1.0 setup

    WatchCrashes<private>()<suspends>:void=
        # Allow for round to restart when a crash occurs
        Timer.Enable()
        Timer.Start()

        # Constantly look for any crashed device
        loop:
            for(Device : RegisteredCrashableDevices[GetSession()]):
                Pong := Device.Ping()
                if(not Pong?):

                    # Send analytics data if analytics is plugged
                    if(Analytics := Device.CrashAnalytics?, AnyAgent := GetPlayspace().GetPlayers()[0]):
                        Analytics.Submit(AnyAgent)

                    PrintError("verse_crash_detector: Device {Device} called a function that caused a crash")
                    break

            Timer.Reset()
            Timer.Start()
            Sleep(2.0)