
using { /Fortnite.com/Devices }
using { /Verse.org/Simulation }
using { /UnrealEngine.com/Temporary/Diagnostics }
using { /UnrealEngine.com/Temporary/UI }

# See https://dev.epicgames.com/documentation/en-us/uefn/create-your-own-device-in-verse for how to create a verse device.

Hat_Ref_Data := class<concrete>():

    @editable
    ShowCaseName : string = ""

    @editable
    VFX : visual_effect_powerup_device = visual_effect_powerup_device{}

# A Verse-authored creative device that can be placed in a level
V_HatManager := class(creative_device):
    @editable
    SpawnPads : []player_spawner_device = array{}

    @editable
    ItemShopHats : []visual_effect_powerup_device = array{}

    @editable
    bPlayerManager : ?V_PlayerManager = false

    var PlayerHatMap : [player]?visual_effect_powerup_device = map{}

    @editable
    UnEquipHat : visual_effect_powerup_device = visual_effect_powerup_device{}

    @editable
    UnEquipTrail : visual_effect_powerup_device = visual_effect_powerup_device{}

    # Delete these var after doing mass test and 1 live update. So can revert incase new way of equiping bugs out

    # @editable
    # Brand1Effects : []visual_effect_powerup_device = array{}

    # @editable
    # Brand2Effects : []visual_effect_powerup_device = array{}

    # @editable
    # BDHats : []visual_effect_powerup_device = array{}

    @editable
    Mutator : mutator_zone_device = mutator_zone_device{}

    @editable
    HatsVFX : []Hat_Ref_Data = array{}


    #var BrandHatVFXMap : [string]tuple([]visual_effect_powerup_device, int) = map{}

    var BrandHatVFXMap : [string]visual_effect_powerup_device = map{}
    
    # Runs when the device is started in a running game
    OnBegin<override>()<suspends>:void=
        #for(SpawnPad : SpawnPads):
            #SpawnPad.SpawnedEvent.Subscribe(OnSpawn)
        Mutator.AgentEntersEvent.Subscribe(TestMutator)
        for(Data : HatsVFX):
            if( set BrandHatVFXMap[Data.ShowCaseName] = Data.VFX):
        <#set BrandHatVFXMap = map:
            # Button's GroupName = (HatsVFX array, HatIndex from HatsVFX)
            "BoxShowcase" => (BDHats, 0)
            "ChickenShowcase" => (BDHats, 1)
            "CrownShowcase" => (BDHats, 2)
            "AstronautShowcase" => (BDHats,3)
            "SnailShowcase" => (BDHats, 4)
            "HeadPhonesShowcase" => (BDHats, 5)
            "CowBoyShowcase" => (BDHats, 6)
            "SuperSaiyanShowcase" => (BDHats, 7)
            "DonutPileShowcase" => (BDHats, 8)
            "SantaShowcase" => (BDHats, 9)
            
            "Brand1Item1Showcase" => (Brand1Effects, 0)
            "Brand1Item2Showcase" => (Brand1Effects, 1)
            "Brand1Item3Showcase" => (Brand1Effects, 2)#>

    TestMutator(Agent : agent):void=
        if (Player := player[Agent]):
            #OnBrandHatPurchase("Bu", "bu",Player, false)
            if(PlayerManager := bPlayerManager?):
                spawn:
                    PlayerManager.PrintAllSaveHatData(Player)

    OnShopUICreated(Player : player, UIData : UI_Data):void=
        block:
        #UIData.BindAnyButtonsWithName("B_Purchase", OnBrandHatPurchase)

    OnSpawn(InAgent : agent):void=
        if(Player := player[InAgent]):
            if (Hat := PlayerHatMap[Player]?):
                Hat.Pickup(Player)

    EquipHat(Player : player, bVFX : ?visual_effect_powerup_device):void=
        if (set PlayerHatMap[Player] = bVFX, VFX := bVFX?):
            VFX.Pickup(Player)

    EquipItemShopHat(Player : player, Index : int):void=
        if (Hat := ItemShopHats[Index]):
            EquipHat(Player, option{Hat})

    OnBrandHatEquip(Player : player, PlayerManager : V_PlayerManager, ShowCaseName : string, BrandName : string, ItemName : string, bUIData : ?UI_Data, Equip : logic):void=
        if(HatVFX := BrandHatVFXMap[ShowCaseName]):
            #HatArray := HatData(0)
            #HatArrayIndex := HatData(1)
            if (Equip = true):
                EquipHat(Player, option{HatVFX})
            #New Addition - Loops through to remove VFX from Index 0 and 1
            if (Equip = false):
                if(PurchaseData := PlayerManager.PurchaseMapBinds[ShowCaseName]):
                    # Get equip index of item that needs to be unequipped
                    EquipIndex := PurchaseData(3)
                    Print("Equip Index {EquipIndex}")
                    # If equip index is that of hat, call vfx device for unequipping hats
                    if (EquipIndex = 0):
                        EquipHat(Player, option{UnEquipHat})
                    # If equip index is that of trail, call vfx device for unequipping trails
                    else if(EquipIndex = 2):
                        Print("Unequip Trail")
                        EquipHat(Player, option{UnEquipTrail})
                