
using { /Fortnite.com/Devices }
using { /Verse.org/Simulation }
using { /UnrealEngine.com/Temporary/Diagnostics }
using { /UnrealEngine.com/Temporary/UI }
using { /Fortnite.com/UI }
using { /UnrealEngine.com/Temporary/SpatialMath}
using { /Verse.org/Colors }
using { /Verse.org/Assets }
using { /UnrealEngine.com/Temporary }

# See https://dev.epicgames.com/documentation/en-us/uefn/create-your-own-device-in-verse for how to create a verse device.

# A Verse-authored creative device that can be placed in a level
V_ItemShop := class(V_Krita):
	DefaultGroup <override>: string = "HomeMenu"
	@editable
	PlayerManager : V_PlayerManager = V_PlayerManager{}
	@editable
    RealTime : V_Real_Time_data = V_Real_Time_data{}


	PriceShowingButtonMenus : []string = array:
		"HatsMenu"

	#@editable
	#HatManager : V_HatManager = V_HatManager{}

	# Extra rules for when player opens menu groups # Overrides Default way of opening groups when menu buttons are clicked
	MenuSelectRules : [string]tuple(logic, []string) = map:
		# Group Name   AllDecendants,  Additional Groups To Show
		"HatsMenu" => (false, array{"ChickenShowcase"})

	_OnDeviceBegin<override>()<suspends>:void=
		block:
	
	# Called By each UIData instance 
	UITick<override>(UIData : UI_Data, Player : player):void=
		if(Group := UIData.GetGroup[E_Shop_Group.HomeMenu.GetName()], Group.Visibility = E_Widget_Visibility_Status.Visible, Group.HiddenPermanent = false):
			# Advert pages on home page flip every 5 seconds
			if(Advert_1 := UIData.GetGroup[E_Shop_Group.Advertpage_1.GetName()], Advert_2 := UIData.GetGroup[E_Shop_Group.Advertpage_2.GetName()]): 
				if(Advert_1.Visibility = E_Widget_Visibility_Status.Visible):
					UIData.HideGroup(E_Shop_Group.Advertpage_1.GetName(), true)
					UIData.ShowGroup(E_Shop_Group.Advertpage_2.GetName(), true)
				else:
					UIData.ShowGroup(E_Shop_Group.Advertpage_1.GetName(), true)
					UIData.HideGroup(E_Shop_Group.Advertpage_2.GetName(), true)


	_OnUIExit<override>(Player : player, UIData : UI_Data):void=
		block:
		#spawn:
			#PlayerManager.GrantXPDelayedOnShopClose(Player)

	# Update the reward collectable UI based on world time relative to players last collected save data time
	UpdateCollectable(Player : player):void=
		if(UIData := PlayerUI[Player],Group := UIData.GroupUIMap["TimedReward"]):
			if(Player.IsActive[], PlayerData := PlayerDataMap[Player]):
				<#TimeInSeconds := TotalSeconds(TimeDate)
				PlayerTimerInSeconds := PlayerData.TimeDateInSeconds
				TimeDiff := TimeInSeconds - PlayerTimerInSeconds
				HoursLeft := (TimeDiff * 1.0) / (SECONDS_PER_HOUR * 1.0)#>

				CanCollect := CalculateCurrentRewardStatus(Player)
				# Check if it has been over 24hrs since last reward collected
				if(CanCollect = true):
					# Reward is ready to collect
					UIData.HideWidget("TimedReward", "T_RewardNotReady", ?Permanent := true)
					UIData.HideWidget("TimedReward", "T_RewardCollected", ?Permanent := true)
					UIData.HideWidget("TimedReward", "B_CollectIn", ?Permanent := true)

					if(Group.Visibility = E_Widget_Visibility_Status.Visible, Group.HiddenPermanent = false):
						UIData.ShowWidget("TimedReward", "B_Collect", ?UnlockPermanent := true)
						UIData.ShowWidget("TimedReward", "T_RewardToCollect", ?UnlockPermanent := true)
					else:
						UIData.UnlockWidget("TimedReward", "B_Collect")
						UIData.UnlockWidget("TimedReward", "T_RewardToCollect")

					UIData.SetText("TimedReward", "Txt_Timer", "0")
					#UIData.SetButtonText("TimedReward", "B_Collect", "Collect")
				# It hasnt yet been past 24 hours so clock needs to keep ticking
				else:
					CountDown := UtcResetHoursRemaining()
					# Show Reward can be collected in: message
					if(Group.Visibility = E_Widget_Visibility_Status.Visible, Group.HiddenPermanent = false):
						UIData.ShowWidget("TimedReward", "T_RewardCollected", ?UnlockPermanent := true)
						UIData.ShowWidget("TimedReward", "B_CollectIn", ?UnlockPermanent := true)
					else:
						UIData.UnlockWidget("TimedReward", "T_RewardCollected")
						UIData.UnlockWidget("TimedReward", "B_CollectIn")

					UIData.HideWidget("TimedReward", "B_Collect", ?Permanent := true)
					UIData.HideWidget("TimedReward", "T_RewardNotReady", ?Permanent := true)
					UIData.HideWidget("TimedReward", "T_RewardToCollect", ?Permanent := true)

					UIData.SetText("TimedReward", "Txt_Timer", "{CountDown}")
					#UIData.SetButtonText("TimedReward", "B_Collect", "Collect In")
	
	# Called when UI is created, Update all purchase/equip text of UI
	UpdateAllUIPurchaseGroups(Player : player, UIData : UI_Data):void=
		for(Name -> Group : UIData.GroupUIMap):
			NameLength := Name.Length
			StartIndex := NameLength - 8
			if (SliceName := Name.Slice[StartIndex], SliceName = "Showcase"):
				UpdateUIPurchaseGroup(Name, option{UIData}, Player)
		
	# Determins the text of a purchasable item button and hides/unhides equip textures
	EquippedItem(Player : player, GroupName : string, ButtonName : string, bUIData : ?UI_Data, Equip : logic):void=
		if(UIData := bUIData?):
			if(Equip = true):
				# Set button text to unequip
				UIData.SetButtonText(GroupName, ButtonName, "UnEquip")
				# Display textures that have sufix Equip and are in showcase group
				if(GroupNameUnique := GroupName.RemoveSufix["Showcase"], Group := UIData.GetGroup[GroupName]):
					EquipTextureName := "T_" + GroupNameUnique + "Equip"
					if(ImageBlocks := Group.ImageBlocks[EquipTextureName]):
						if(Group.Visibility = E_Widget_Visibility_Status.Visible):
							# Show equipped textures if this showcase group is visible
							for(ImageBlock : ImageBlocks):
								ImageBlock._ShowWidget(?UnlockPermanent := true)
						else:
							# Unlock equipped textures if this showcase group is Invisible 
							# So next time the showcase is displayed on screen, the equipped texture is automatically visible
							for(ImageBlock : ImageBlocks):
								ImageBlock.UnlockWidget()
					# Display textures that have sufix Equip for pet style menus where the icon lives outside showcase group
					ParentGroupName := Group.ParentGroupName
					SmallWidgetName := "T_" + GroupNameUnique + "Equip"
					if(ParentGroup := UIData.GetGroup[ParentGroupName], ImageBlocks := ParentGroup.ImageBlocks[SmallWidgetName]):
						if(ParentGroup.Visibility = E_Widget_Visibility_Status.Visible):
							# Show equipped texture view icons if this showcase group is visible
							for(ImageBlock : ImageBlocks):
								ImageBlock._ShowWidget(?UnlockPermanent := true)
						else:
							# Unlock equipped texture view icons if this showcases parent group is Invisible 
							# So next time the parent group of showcase is displayed on screen, the small equipped texture for pet style menu is automatically visible
							for(ImageBlock : ImageBlocks):
								ImageBlock.UnlockWidget()
			else:
				# Set button text to Equip
				UIData.SetButtonText(GroupName, ButtonName, "Equip")
				# Hide textures that have sufix Equip
				HideEquipTexture(GroupName, bUIData)


						
	HideEquipTexture(GroupName : string, bUIData : ?UI_Data):void=
		# Hide textures that have sufix Equip
		if(UIData := bUIData?):
			if(GroupNameUnique := GroupName.RemoveSufix["Showcase"], Group := UIData.GetGroup[GroupName]):
				EquipTextureName := "T_" + GroupNameUnique + "Equip"
				if(ImageBlocks := Group.ImageBlocks[EquipTextureName]):
					for(ImageBlock : ImageBlocks):
						ImageBlock._HideWidget(?Permanent := true)

				# Hide textures that have sufix Equip for pet style menus where the icon lives outside showcase group
				ParentGroupName := Group.ParentGroupName
				SmallWidgetName := "T_" + GroupNameUnique + "Equip"
				if(ParentGroup := UIData.GetGroup[ParentGroupName], ImageBlocks := ParentGroup.ImageBlocks[SmallWidgetName]):
					# Hide equipped texture view icons for pet style menus
					for(ImageBlock : ImageBlocks):
						ImageBlock._HideWidget(?Permanent := true)


	
	# Update text for all player gold text widgets
	SetPlayerGold(Player : player, UIData : UI_Data):void=
		Gold := PlayerManager.GetPlayerGold(Player)
		for(Name -> Group : UIData.GroupUIMap):
			if (SliceName := Name.GetSufix["Menu"]):
				if(GoldTexts := Group.TextBlocks["Txt_Gold"]):
					UIData.SetText(Name, "Txt_Gold", "{Gold}")
			
		
	# Handles UI logic for purchasing, hides or unhides widgets based on player purchase save data
	# This Update function is called for all groups at the start of UI creation
	# This Update function is called for purchase group in V_PlayerManager every time player purchases an Item
	UpdateUIPurchaseGroup(GroupName : string, bUIData : ?UI_Data, Player : player):void=
		if(PurchaseData := PlayerManager.PurchaseMapBinds[GroupName], UIData := bUIData?, PlayerData := PlayerManager.GetPlayerData[Player]):
			Brand := PurchaseData(0)
			Item := PurchaseData(1)
			Cost := PurchaseData(2)
			EquipIndex := PurchaseData(3)
			var Purchased : logic = false
			if (Items := PlayerData.ItemsMap[Brand], Items.Find[Item]):
				set Purchased = true
			SetPlayerGold(Player , UIData)
			if(Group := UIData.GetGroup[GroupName]):
				for(Name -> WidgetArray : Group.ImageBlocks):
					NameLength := Name.Length
					StartIndex := NameLength - 5
					if (SliceName := Name.Slice[StartIndex], SliceName = "Color"):
						PurchasedWidgetName := Name
						if(NameToSearch := Name.Remove[StartIndex, NameLength]):
							UnPurchasedWidgetName := NameToSearch
							if (Purchased = true):
								if (Group.Visibility = E_Widget_Visibility_Status.Hidden):
									UIData.UnlockWidget(GroupName, PurchasedWidgetName)
									UIData.LockWidget(GroupName, UnPurchasedWidgetName)
								else:
									UIData.ShowWidget(GroupName, PurchasedWidgetName, ?UnlockPermanent := true)
									UIData.HideWidget(GroupName, UnPurchasedWidgetName, ?Permanent := true)
							else:
								if (Group.Visibility = E_Widget_Visibility_Status.Hidden):
									UIData.UnlockWidget(GroupName, UnPurchasedWidgetName)
									UIData.LockWidget(GroupName, PurchasedWidgetName)
								else:
									UIData.ShowWidget(GroupName, UnPurchasedWidgetName, ?UnlockPermanent := true)
									UIData.HideWidget(GroupName, PurchasedWidgetName, ?Permanent := true)
									
				if (Purchased = true):
					if(EquipData := PlayerManager.PlayerEquipMap[Player][EquipIndex]):
						if(EquipData(0) = GroupName):
							EquippedItem(Player, GroupName, "B_Purchase", bUIData, true)
						else:
							EquippedItem(Player, GroupName, "B_Purchase", bUIData, false)
					else:
						EquippedItem(Player, GroupName, "B_Purchase", bUIData, false)
					# Hide the shadow textures for pet style of menus
					ParentGroupName := Group.ParentGroupName
					if(GroupNameUnique := GroupName.RemoveSufix["Showcase"]):
						T := "T_"
						Small := "Small"
						SmallWidgetName := T + GroupNameUnique + Small
						UIData.HideWidget(ParentGroupName, SmallWidgetName, ?Permanent := true)
				else:
					HideEquipTexture(GroupName, bUIData)
					ParentGroupName := Group.ParentGroupName
					if(PriceShowingButtonMenus.Find[ParentGroupName]):
						UIData.SetButtonText(GroupName, "B_Purchase", "{Cost}")

									
	# Handels logic for selecting hats. Correct showcase page for hat will show when player clicks on a hat
	OnHatSelect(GroupName : string, ButtonName : string, Widget : widget_message, bUIData : ?UI_Data):void=
		Print("{ButtonName}", ?Duration := 5.0)
		if(NewButtonName := ButtonName.RemovePrefix["B_"], UIData := bUIData?):
			GroupNameToSearch := NewButtonName
			if(Group := UIData.GetGroup[GroupNameToSearch]):
				UIData.HideGroup(GroupName, true)
				UIData.ShowGroup(GroupName, false)
				UIData.ShowGroup(GroupNameToSearch, true)

	# Handles logic for all menu tab selection
	OnAnyMenuSelect(GroupName : string, ButtonName : string, Widget : widget_message, bUIData : ?UI_Data):void=
		if(SlicedName := ButtonName.Remove[0, 2], UIData := bUIData?):
			if(Group := UIData.GetGroup[SlicedName], Group.Visibility = E_Widget_Visibility_Status.Hidden, Group.HiddenPermanent = false):
				<#
				Print("{SlicedName}", ?Duration := 5.0)
				if(Rules := MenuSelectRules[SlicedName]):
					AllDecendants := Rules(0)
					AllUnlockGroups := Rules(1) + array{SlicedName}
					UIData.ShowOnlyGroups(AllUnlockGroups, AllDecendants)
				else:
					UIData.ShowOnlyGroups(array{Group.Name}, true)
				#>
				UIData.ShowOnlyGroups(array{Group.Name}, false)
				ChildNames := Group.ChildGroups
				var AdditionalUnlockShowcases : []string = array{}
				for(ChildName : ChildNames):
					if(ChildName.GetSufix["Showcase"], Group.ButtonsQuiet["B_{ChildName}"]):
					else:
						# set AdditionalUnlockShowcases += array{ChildName}
						UIData.ShowGroup(ChildName, true)
				# Trigger analytic device when player opens this menu
				if(AnalyticDevice := PlayerManager.AnalyticsMapMenuClick[ButtonName]?):
					AnalyticDevice.Submit(Widget.Player)
				

	<#UpdatePlayerGoldUI(UIData : UI_Data, UpdatedGold : int):void=
		UIData.SetText(E_Shop_Widget.HatsMenu_Txt_Gold.GetName(), "{UpdatedGold}")#>
	
	# Use this to inject your data on top of krita data. Called after krita data has been injected and before widgets are created
	_InjectUIData<override>(Player : player, UIData : UI_Data):void=
		block:
			<#PlayerGold := PlayerManager.GetPlayerGold(Player)
			UpdatePlayerGoldUI(UIData, PlayerGold)#>

	CollectPrize(GroupName : string, ButtonName : string, Widget : widget_message, bUIData : ?UI_Data):void=
		block:

	OnShopClicked(GroupName : string, ButtonName : string, Widget : widget_message, bUIData : ?UI_Data):void=
		if(AnalyticDevice := PlayerManager.AnalyticsMapMenuClick[ButtonName]?):
			AnalyticDevice.Submit(Widget.Player)
	
	# Buttons are ready to be bound. Called a second after UI is displayed on player screen this is a saftey feature for controller players
	_OnAllButtonSlotsAdded<override>(Player : player, UIData : UI_Data):void=
		#UIData.BindAllButtonsInGroup("HatsMenu", false, OnHatSelect)
		UIData.BindAnyButtonsEndsWithName("Showcase", OnHatSelect)
		#UIData.BindAnyButtonsWithNames(array{"B_HatsMenu", "B_HomeMenu", "B_Brand1Menu", "B_Brand2Menu"}, OnAnyMenuSelect)
		UIData.BindAnyButtonsEndsWithName("Menu", OnAnyMenuSelect)
		UIData.BindButtons("TimedReward", "B_Collect", false, CollectPrize)
		UIData.BindButtons("SpectateMenu", "B_ItemShop", false, OnShopClicked)
	
	# Called when UI widgets and canva are created but not yet displayed on screen
	_OnUICreated<override>(Player : player, UIData : UI_Data):void=
		PlayerManager.OnShopUICreated(Player, UIData)
		UpdateAllUIPurchaseGroups(Player, UIData)
		UpdateCollectable(Player)


	### AUTO GENERATED CODE FROM KRITA. Consult Sanjay before changing anything
	@editable
	ButtonTriggerBinds : []Shop_BindToTriggers = array{}
	@editable
	ShowGroups : []Shop_ShowGroups = array{}
	@editable
	ExitButtons : []Shop_ExitButton = array{}
	_MakeUIDataFromKrita<override>():UI_Data=
		var ExitButtonData : []Shop_ExitButton = array{}
		for(ExitButton:ExitButtons):
			set ExitButtonData += array{MakeShop_ExitButton(ExitButton)}
		var ShowGroupData : []Shop_ShowGroups = array{}
		for(ShowGroup:ShowGroups):
			set ShowGroupData += array{MakeShop_ShowGroups(ShowGroup)}
		var ButtonTriggerData : []Shop_BindToTriggers = array{}
		for(ButtonTriggerBind:ButtonTriggerBinds):
			set ButtonTriggerData += array{MakeShop_BindToTriggers(ButtonTriggerBind)}
		TempUIDATA :=UI_Data:
			GroupUIMap := map:
				"root" => Group_UI{Name := "root", ParentGroupName := "", ChildGroups := array{"Brand2Menu", "Brand1Menu", "HatsMenu", "HomeMenu", "SpectateMenu"}, AnchorMinX := 0.0, AnchorMinY := 0.0, AnchorMaxX := 0.0, AnchorMaxY := 0.0 }
				"Brand2Menu" => Group_UI{Name := "Brand2Menu", ParentGroupName := "root", ChildGroups := array{"Brand2Item3Showcase", "Brand2Item2Showcase", "Brand2Item1Showcase"}, AnchorMinX := 0.5, AnchorMinY := 0.5, AnchorMaxX := 0.5, AnchorMaxY := 0.5 }
				"Brand2Item3Showcase" => Group_UI{Name := "Brand2Item3Showcase", ParentGroupName := "Brand2Menu", ChildGroups := array{}, AnchorMinX := 0.5, AnchorMinY := 0.5, AnchorMaxX := 0.5, AnchorMaxY := 0.5 }
				"Brand2Item2Showcase" => Group_UI{Name := "Brand2Item2Showcase", ParentGroupName := "Brand2Menu", ChildGroups := array{}, AnchorMinX := 0.5, AnchorMinY := 0.5, AnchorMaxX := 0.5, AnchorMaxY := 0.5 }
				"Brand2Item1Showcase" => Group_UI{Name := "Brand2Item1Showcase", ParentGroupName := "Brand2Menu", ChildGroups := array{}, AnchorMinX := 0.5, AnchorMinY := 0.5, AnchorMaxX := 0.5, AnchorMaxY := 0.5 }
				"Brand1Menu" => Group_UI{Name := "Brand1Menu", ParentGroupName := "root", ChildGroups := array{"Brand1Item3Showcase", "Brand1Item2Showcase", "Brand1Item1Showcase"}, AnchorMinX := 0.5, AnchorMinY := 0.5, AnchorMaxX := 0.5, AnchorMaxY := 0.5 }
				"Brand1Item3Showcase" => Group_UI{Name := "Brand1Item3Showcase", ParentGroupName := "Brand1Menu", ChildGroups := array{}, AnchorMinX := 0.5, AnchorMinY := 0.5, AnchorMaxX := 0.5, AnchorMaxY := 0.5 }
				"Brand1Item2Showcase" => Group_UI{Name := "Brand1Item2Showcase", ParentGroupName := "Brand1Menu", ChildGroups := array{}, AnchorMinX := 0.5, AnchorMinY := 0.5, AnchorMaxX := 0.5, AnchorMaxY := 0.5 }
				"Brand1Item1Showcase" => Group_UI{Name := "Brand1Item1Showcase", ParentGroupName := "Brand1Menu", ChildGroups := array{}, AnchorMinX := 0.5, AnchorMinY := 0.5, AnchorMaxX := 0.5, AnchorMaxY := 0.5 }
				"HatsMenu" => Group_UI{Name := "HatsMenu", ParentGroupName := "root", ChildGroups := array{"SuperSaiyanShowcase", "CrownShowcase", "SnailShowcase", "CowBoyShowcase", "HeadPhonesShowcase", "DonutPileShowcase", "BoxShowcase", "ChickenShowcase"}, AnchorMinX := 0.5, AnchorMinY := 0.5, AnchorMaxX := 0.5, AnchorMaxY := 0.5 }
				"SuperSaiyanShowcase" => Group_UI{Name := "SuperSaiyanShowcase", ParentGroupName := "HatsMenu", ChildGroups := array{}, AnchorMinX := 0.5, AnchorMinY := 0.5, AnchorMaxX := 0.5, AnchorMaxY := 0.5 }
				"CrownShowcase" => Group_UI{Name := "CrownShowcase", ParentGroupName := "HatsMenu", ChildGroups := array{}, AnchorMinX := 0.5, AnchorMinY := 0.5, AnchorMaxX := 0.5, AnchorMaxY := 0.5 }
				"SnailShowcase" => Group_UI{Name := "SnailShowcase", ParentGroupName := "HatsMenu", ChildGroups := array{}, AnchorMinX := 0.5, AnchorMinY := 0.5, AnchorMaxX := 0.5, AnchorMaxY := 0.5 }
				"CowBoyShowcase" => Group_UI{Name := "CowBoyShowcase", ParentGroupName := "HatsMenu", ChildGroups := array{}, AnchorMinX := 0.5, AnchorMinY := 0.5, AnchorMaxX := 0.5, AnchorMaxY := 0.5 }
				"HeadPhonesShowcase" => Group_UI{Name := "HeadPhonesShowcase", ParentGroupName := "HatsMenu", ChildGroups := array{}, AnchorMinX := 0.5, AnchorMinY := 0.5, AnchorMaxX := 0.5, AnchorMaxY := 0.5 }
				"DonutPileShowcase" => Group_UI{Name := "DonutPileShowcase", ParentGroupName := "HatsMenu", ChildGroups := array{}, AnchorMinX := 0.5, AnchorMinY := 0.5, AnchorMaxX := 0.5, AnchorMaxY := 0.5 }
				"BoxShowcase" => Group_UI{Name := "BoxShowcase", ParentGroupName := "HatsMenu", ChildGroups := array{}, AnchorMinX := 0.5, AnchorMinY := 0.5, AnchorMaxX := 0.5, AnchorMaxY := 0.5 }
				"ChickenShowcase" => Group_UI{Name := "ChickenShowcase", ParentGroupName := "HatsMenu", ChildGroups := array{}, AnchorMinX := 0.5, AnchorMinY := 0.5, AnchorMaxX := 0.5, AnchorMaxY := 0.5 }
				"HomeMenu" => Group_UI{Name := "HomeMenu", ParentGroupName := "root", ChildGroups := array{"Advertpage_1", "Advertpage_2", "TimedReward", "GunLoadout1Showcase", "GunLoadout2Showcase", "GunLoadout3Showcase"}, AnchorMinX := 0.5, AnchorMinY := 0.5, AnchorMaxX := 0.5, AnchorMaxY := 0.5 }
				"Advertpage_1" => Group_UI{Name := "Advertpage_1", ParentGroupName := "HomeMenu", ChildGroups := array{}, AnchorMinX := 0.5, AnchorMinY := 0.5, AnchorMaxX := 0.5, AnchorMaxY := 0.5 }
				"Advertpage_2" => Group_UI{Name := "Advertpage_2", ParentGroupName := "HomeMenu", ChildGroups := array{}, AnchorMinX := 0.5, AnchorMinY := 0.5, AnchorMaxX := 0.5, AnchorMaxY := 0.5 }
				"TimedReward" => Group_UI{Name := "TimedReward", ParentGroupName := "HomeMenu", ChildGroups := array{}, AnchorMinX := 0.5, AnchorMinY := 0.5, AnchorMaxX := 0.5, AnchorMaxY := 0.5 }
				"GunLoadout1Showcase" => Group_UI{Name := "GunLoadout1Showcase", ParentGroupName := "HomeMenu", ChildGroups := array{}, AnchorMinX := 0.5, AnchorMinY := 0.5, AnchorMaxX := 0.5, AnchorMaxY := 0.5 }
				"GunLoadout2Showcase" => Group_UI{Name := "GunLoadout2Showcase", ParentGroupName := "HomeMenu", ChildGroups := array{}, AnchorMinX := 0.5, AnchorMinY := 0.5, AnchorMaxX := 0.5, AnchorMaxY := 0.5 }
				"GunLoadout3Showcase" => Group_UI{Name := "GunLoadout3Showcase", ParentGroupName := "HomeMenu", ChildGroups := array{}, AnchorMinX := 0.5, AnchorMinY := 0.5, AnchorMaxX := 0.5, AnchorMaxY := 0.5 }
				"SpectateMenu" => Group_UI{Name := "SpectateMenu", ParentGroupName := "root", ChildGroups := array{}, AnchorMinX := 0.5, AnchorMinY := 0.5, AnchorMaxX := 0.5, AnchorMaxY := 0.5 }
			OrderedWidgetData := array:
				Image_Block{X:=-885.0, Y := -497.0, ScaleX := 1777, ScaleY := 1009, Texture := option{VerseExposed.T_Brand2MenuBackground}, Name := "T_Brand2MenuBackground", GroupName := "Brand2Menu", AnchorMinX := 0.5, AnchorMinY := 0.5, AnchorMaxX := 0.5, AnchorMaxY := 0.5}
				Button_Quiet_Data{X:=418.0, Y := 238.0, ScaleX := 324, ScaleY := 86, Color := option{color_block{DefaultColor := color{R := 0.0, G := 0.0, B := 0.0}, DefaultOpacity := 1.0} }, Text := "BUY", Name := "B_Purchase", GroupName := "Brand2Item3Showcase", ButtonType := E_ButtonType.Loud, AnchorMinX := 0.5, AnchorMinY := 0.5, AnchorMaxX := 0.5, AnchorMaxY := 0.5}
				Image_Block{X:=289.0, Y := -318.0, ScaleX := 538, ScaleY := 541, Texture := option{VerseExposed.T_Brand2Item3Color}, Name := "T_Brand2Item3Color", GroupName := "Brand2Item3Showcase", AnchorMinX := 0.5, AnchorMinY := 0.5, AnchorMaxX := 0.5, AnchorMaxY := 0.5}
				Image_Block{X:=289.0, Y := -318.0, ScaleX := 538, ScaleY := 541, Texture := option{VerseExposed.T_Brand2Item3}, Name := "T_Brand2Item3", GroupName := "Brand2Item3Showcase", AnchorMinX := 0.5, AnchorMinY := 0.5, AnchorMaxX := 0.5, AnchorMaxY := 0.5}
				Button_Quiet_Data{X:=-137.0, Y := 238.0, ScaleX := 324, ScaleY := 86, Color := option{color_block{DefaultColor := color{R := 0.0, G := 0.0, B := 0.0}, DefaultOpacity := 1.0} }, Text := "BUY", Name := "B_Purchase", GroupName := "Brand2Item2Showcase", ButtonType := E_ButtonType.Loud, AnchorMinX := 0.5, AnchorMinY := 0.5, AnchorMaxX := 0.5, AnchorMaxY := 0.5}
				Image_Block{X:=-265.0, Y := -318.0, ScaleX := 538, ScaleY := 541, Texture := option{VerseExposed.T_Brand2Item2Color}, Name := "T_Brand2Item2Color", GroupName := "Brand2Item2Showcase", AnchorMinX := 0.5, AnchorMinY := 0.5, AnchorMaxX := 0.5, AnchorMaxY := 0.5}
				Image_Block{X:=-265.0, Y := -318.0, ScaleX := 538, ScaleY := 541, Texture := option{VerseExposed.T_Brand2Item2}, Name := "T_Brand2Item2", GroupName := "Brand2Item2Showcase", AnchorMinX := 0.5, AnchorMinY := 0.5, AnchorMaxX := 0.5, AnchorMaxY := 0.5}
				Button_Quiet_Data{X:=-692.0, Y := 238.0, ScaleX := 324, ScaleY := 86, Color := option{color_block{DefaultColor := color{R := 0.0, G := 0.0, B := 0.0}, DefaultOpacity := 1.0} }, Text := "BUY", Name := "B_Purchase", GroupName := "Brand2Item1Showcase", ButtonType := E_ButtonType.Loud, AnchorMinX := 0.5, AnchorMinY := 0.5, AnchorMaxX := 0.5, AnchorMaxY := 0.5}
				Image_Block{X:=-820.0, Y := -318.0, ScaleX := 538, ScaleY := 541, Texture := option{VerseExposed.T_Brand2Item1Color}, Name := "T_Brand2Item1Color", GroupName := "Brand2Item1Showcase", AnchorMinX := 0.5, AnchorMinY := 0.5, AnchorMaxX := 0.5, AnchorMaxY := 0.5}
				Image_Block{X:=-820.0, Y := -318.0, ScaleX := 538, ScaleY := 541, Texture := option{VerseExposed.T_Brand2Item1}, Name := "T_Brand2Item1", GroupName := "Brand2Item1Showcase", AnchorMinX := 0.5, AnchorMinY := 0.5, AnchorMaxX := 0.5, AnchorMaxY := 0.5}
				Text_Block{X:=652.0, Y := -457.0, ScaleX := 66, ScaleY := 28, Text := "9999", Name := "Txt_Gold", GroupName := "Brand2Menu", Color := color{R := 1.0, G := 1.0, B := 1.0}, Opacity := 1.0, AnchorMinX := 0.5, AnchorMinY := 0.5, AnchorMaxX := 0.5, AnchorMaxY := 0.5}
				Button_Quiet_Data{X:=-131.0, Y := 375.0, ScaleX := 325, ScaleY := 122, Color := option{color_block{DefaultColor := color{R := 1.0, G := 1.0, B := 1.0}, DefaultOpacity := 1.0} }, Text := "EXIT", Name := "B_Exit", GroupName := "Brand2Menu", ButtonType := E_ButtonType.Loud, AnchorMinX := 0.5, AnchorMinY := 0.5, AnchorMaxX := 0.5, AnchorMaxY := 0.5}
				Button_Quiet_Data{X:=-117.0, Y := -475.0, ScaleX := 343, ScaleY := 80, Color := option{color_block{DefaultColor := color{R := 0.0, G := 0.0, B := 0.0}, DefaultOpacity := 0.5215686274509804} }, Text := "AC MILAN", Name := "B_Brand2Menu", GroupName := "Brand2Menu", ButtonType := E_ButtonType.Loud, AnchorMinX := 0.5, AnchorMinY := 0.5, AnchorMaxX := 0.5, AnchorMaxY := 0.5}
				Button_Quiet_Data{X:=-564.0, Y := -472.0, ScaleX := 150, ScaleY := 62, Color := option{color_block{DefaultColor := color{R := 0.0, G := 0.0, B := 0.0}, DefaultOpacity := 1.0} }, Text := "ITEMS", Name := "B_HatsMenu", GroupName := "Brand2Menu", ButtonType := E_ButtonType.Loud, AnchorMinX := 0.5, AnchorMinY := 0.5, AnchorMaxX := 0.5, AnchorMaxY := 0.5}
				Button_Quiet_Data{X:=-740.0, Y := -472.0, ScaleX := 150, ScaleY := 62, Color := option{color_block{DefaultColor := color{R := 0.0, G := 0.0, B := 0.0}, DefaultOpacity := 1.0} }, Text := "HOME", Name := "B_HomeMenu", GroupName := "Brand2Menu", ButtonType := E_ButtonType.Loud, AnchorMinX := 0.5, AnchorMinY := 0.5, AnchorMaxX := 0.5, AnchorMaxY := 0.5}
				Button_Quiet_Data{X:=-386.0, Y := -472.0, ScaleX := 231, ScaleY := 62, Color := option{color_block{DefaultColor := color{R := 0.0, G := 0.0, B := 0.0}, DefaultOpacity := 1.0} }, Text := "HI-CHEW", Name := "B_Brand1Menu", GroupName := "Brand2Menu", ButtonType := E_ButtonType.Loud, AnchorMinX := 0.5, AnchorMinY := 0.5, AnchorMaxX := 0.5, AnchorMaxY := 0.5}
				Image_Block{X:=-885.0, Y := -497.0, ScaleX := 1777, ScaleY := 1009, Texture := option{VerseExposed.T_Brand1MenuBackground}, Name := "T_Brand1MenuBackground", GroupName := "Brand1Menu", AnchorMinX := 0.5, AnchorMinY := 0.5, AnchorMaxX := 0.5, AnchorMaxY := 0.5}
				Button_Quiet_Data{X:=418.0, Y := 238.0, ScaleX := 324, ScaleY := 86, Color := option{color_block{DefaultColor := color{R := 0.0, G := 0.0, B := 0.0}, DefaultOpacity := 1.0} }, Text := "BUY", Name := "B_Purchase", GroupName := "Brand1Item3Showcase", ButtonType := E_ButtonType.Loud, AnchorMinX := 0.5, AnchorMinY := 0.5, AnchorMaxX := 0.5, AnchorMaxY := 0.5}
				Image_Block{X:=289.0, Y := -318.0, ScaleX := 538, ScaleY := 541, Texture := option{VerseExposed.T_Brand1Item3Color}, Name := "T_Brand1Item3Color", GroupName := "Brand1Item3Showcase", AnchorMinX := 0.5, AnchorMinY := 0.5, AnchorMaxX := 0.5, AnchorMaxY := 0.5}
				Image_Block{X:=289.0, Y := -318.0, ScaleX := 538, ScaleY := 541, Texture := option{VerseExposed.T_Brand1Item3}, Name := "T_Brand1Item3", GroupName := "Brand1Item3Showcase", AnchorMinX := 0.5, AnchorMinY := 0.5, AnchorMaxX := 0.5, AnchorMaxY := 0.5}
				Button_Quiet_Data{X:=-137.0, Y := 238.0, ScaleX := 324, ScaleY := 86, Color := option{color_block{DefaultColor := color{R := 0.0, G := 0.0, B := 0.0}, DefaultOpacity := 1.0} }, Text := "BUY", Name := "B_Purchase", GroupName := "Brand1Item2Showcase", ButtonType := E_ButtonType.Loud, AnchorMinX := 0.5, AnchorMinY := 0.5, AnchorMaxX := 0.5, AnchorMaxY := 0.5}
				Image_Block{X:=-265.0, Y := -318.0, ScaleX := 538, ScaleY := 541, Texture := option{VerseExposed.T_Brand1Item2Color}, Name := "T_Brand1Item2Color", GroupName := "Brand1Item2Showcase", AnchorMinX := 0.5, AnchorMinY := 0.5, AnchorMaxX := 0.5, AnchorMaxY := 0.5}
				Image_Block{X:=-265.0, Y := -318.0, ScaleX := 538, ScaleY := 541, Texture := option{VerseExposed.T_Brand1Item2}, Name := "T_Brand1Item2", GroupName := "Brand1Item2Showcase", AnchorMinX := 0.5, AnchorMinY := 0.5, AnchorMaxX := 0.5, AnchorMaxY := 0.5}
				Button_Quiet_Data{X:=-692.0, Y := 238.0, ScaleX := 324, ScaleY := 86, Color := option{color_block{DefaultColor := color{R := 0.0, G := 0.0, B := 0.0}, DefaultOpacity := 1.0} }, Text := "BUY", Name := "B_Purchase", GroupName := "Brand1Item1Showcase", ButtonType := E_ButtonType.Loud, AnchorMinX := 0.5, AnchorMinY := 0.5, AnchorMaxX := 0.5, AnchorMaxY := 0.5}
				Image_Block{X:=-820.0, Y := -318.0, ScaleX := 538, ScaleY := 541, Texture := option{VerseExposed.T_Brand1Item1Color}, Name := "T_Brand1Item1Color", GroupName := "Brand1Item1Showcase", AnchorMinX := 0.5, AnchorMinY := 0.5, AnchorMaxX := 0.5, AnchorMaxY := 0.5}
				Image_Block{X:=-820.0, Y := -318.0, ScaleX := 538, ScaleY := 541, Texture := option{VerseExposed.T_Brand1Item1}, Name := "T_Brand1Item1", GroupName := "Brand1Item1Showcase", AnchorMinX := 0.5, AnchorMinY := 0.5, AnchorMaxX := 0.5, AnchorMaxY := 0.5}
				Text_Block{X:=652.0, Y := -457.0, ScaleX := 66, ScaleY := 28, Text := "9999", Name := "Txt_Gold", GroupName := "Brand1Menu", Color := color{R := 1.0, G := 1.0, B := 1.0}, Opacity := 1.0, AnchorMinX := 0.5, AnchorMinY := 0.5, AnchorMaxX := 0.5, AnchorMaxY := 0.5}
				Button_Quiet_Data{X:=-131.0, Y := 375.0, ScaleX := 325, ScaleY := 122, Color := option{color_block{DefaultColor := color{R := 1.0, G := 1.0, B := 1.0}, DefaultOpacity := 1.0} }, Text := "EXIT", Name := "B_Exit", GroupName := "Brand1Menu", ButtonType := E_ButtonType.Loud, AnchorMinX := 0.5, AnchorMinY := 0.5, AnchorMaxX := 0.5, AnchorMaxY := 0.5}
				Button_Quiet_Data{X:=-364.0, Y := -476.0, ScaleX := 270, ScaleY := 84, Color := option{color_block{DefaultColor := color{R := 0.0, G := 0.0, B := 0.0}, DefaultOpacity := 0.5215686274509804} }, Text := "HI-CHEW", Name := "B_Brand1Menu", GroupName := "Brand1Menu", ButtonType := E_ButtonType.Loud, AnchorMinX := 0.5, AnchorMinY := 0.5, AnchorMaxX := 0.5, AnchorMaxY := 0.5}
				Button_Quiet_Data{X:=-558.0, Y := -472.0, ScaleX := 161, ScaleY := 62, Color := option{color_block{DefaultColor := color{R := 0.0, G := 0.0, B := 0.0}, DefaultOpacity := 1.0} }, Text := "ITEMS", Name := "B_HatsMenu", GroupName := "Brand1Menu", ButtonType := E_ButtonType.Loud, AnchorMinX := 0.5, AnchorMinY := 0.5, AnchorMaxX := 0.5, AnchorMaxY := 0.5}
				Button_Quiet_Data{X:=-746.0, Y := -472.0, ScaleX := 161, ScaleY := 62, Color := option{color_block{DefaultColor := color{R := 0.0, G := 0.0, B := 0.0}, DefaultOpacity := 1.0} }, Text := "HOME", Name := "B_HomeMenu", GroupName := "Brand1Menu", ButtonType := E_ButtonType.Loud, AnchorMinX := 0.5, AnchorMinY := 0.5, AnchorMaxX := 0.5, AnchorMaxY := 0.5}
				Image_Block{X:=-885.0, Y := -497.0, ScaleX := 1777, ScaleY := 1009, Texture := option{VerseExposed.T_HatsMenuBackground}, Name := "T_HatsMenuBackground", GroupName := "HatsMenu", AnchorMinX := 0.5, AnchorMinY := 0.5, AnchorMaxX := 0.5, AnchorMaxY := 0.5}
				Image_Block{X:=365.0, Y := 3.0, ScaleX := 310, ScaleY := 346, Texture := option{VerseExposed.T_SuperSaiyanColor}, Name := "T_SuperSaiyanColor", GroupName := "SuperSaiyanShowcase", AnchorMinX := 0.5, AnchorMinY := 0.5, AnchorMaxX := 0.5, AnchorMaxY := 0.5}
				Image_Block{X:=365.0, Y := 3.0, ScaleX := 310, ScaleY := 347, Texture := option{VerseExposed.T_SuperSaiyan}, Name := "T_SuperSaiyan", GroupName := "SuperSaiyanShowcase", AnchorMinX := 0.5, AnchorMinY := 0.5, AnchorMaxX := 0.5, AnchorMaxY := 0.5}
				Image_Block{X:=366.0, Y := 3.0, ScaleX := 309, ScaleY := 346, Texture := option{VerseExposed.T_SuperSaiyanEquip}, Name := "T_SuperSaiyanEquip", GroupName := "SuperSaiyanShowcase", AnchorMinX := 0.5, AnchorMinY := 0.5, AnchorMaxX := 0.5, AnchorMaxY := 0.5}
				Button_Quiet_Data{X:=440.0, Y := 292.0, ScaleX := 228, ScaleY := 44, Color := option{color_block{DefaultColor := color{R := 0.0, G := 0.0, B := 0.0}, DefaultOpacity := 1.0} }, Text := "", Name := "B_Purchase", GroupName := "SuperSaiyanShowcase", ButtonType := E_ButtonType.Loud, AnchorMinX := 0.5, AnchorMinY := 0.5, AnchorMaxX := 0.5, AnchorMaxY := 0.5}
				Image_Block{X:=23.0, Y := 3.0, ScaleX := 309, ScaleY := 344, Texture := option{VerseExposed.T_CrownColor}, Name := "T_CrownColor", GroupName := "CrownShowcase", AnchorMinX := 0.5, AnchorMinY := 0.5, AnchorMaxX := 0.5, AnchorMaxY := 0.5}
				Image_Block{X:=22.0, Y := 3.0, ScaleX := 310, ScaleY := 344, Texture := option{VerseExposed.T_Crown}, Name := "T_Crown", GroupName := "CrownShowcase", AnchorMinX := 0.5, AnchorMinY := 0.5, AnchorMaxX := 0.5, AnchorMaxY := 0.5}
				Image_Block{X:=23.0, Y := 3.0, ScaleX := 309, ScaleY := 344, Texture := option{VerseExposed.T_CrownEquip}, Name := "T_CrownEquip", GroupName := "CrownShowcase", AnchorMinX := 0.5, AnchorMinY := 0.5, AnchorMaxX := 0.5, AnchorMaxY := 0.5}
				Button_Quiet_Data{X:=97.0, Y := 289.0, ScaleX := 228, ScaleY := 44, Color := option{color_block{DefaultColor := color{R := 0.0, G := 0.0, B := 0.0}, DefaultOpacity := 1.0} }, Text := "", Name := "B_Purchase", GroupName := "CrownShowcase", ButtonType := E_ButtonType.Loud, AnchorMinX := 0.5, AnchorMinY := 0.5, AnchorMaxX := 0.5, AnchorMaxY := 0.5}
				Image_Block{X:=-319.0, Y := 3.0, ScaleX := 311, ScaleY := 348, Texture := option{VerseExposed.T_SnailColor}, Name := "T_SnailColor", GroupName := "SnailShowcase", AnchorMinX := 0.5, AnchorMinY := 0.5, AnchorMaxX := 0.5, AnchorMaxY := 0.5}
				Image_Block{X:=-320.0, Y := 3.0, ScaleX := 312, ScaleY := 349, Texture := option{VerseExposed.T_Snail}, Name := "T_Snail", GroupName := "SnailShowcase", AnchorMinX := 0.5, AnchorMinY := 0.5, AnchorMaxX := 0.5, AnchorMaxY := 0.5}
				Image_Block{X:=-317.0, Y := 2.0, ScaleX := 309, ScaleY := 349, Texture := option{VerseExposed.T_SnailEquip}, Name := "T_SnailEquip", GroupName := "SnailShowcase", AnchorMinX := 0.5, AnchorMinY := 0.5, AnchorMaxX := 0.5, AnchorMaxY := 0.5}
				Button_Quiet_Data{X:=-246.0, Y := 294.0, ScaleX := 228, ScaleY := 44, Color := option{color_block{DefaultColor := color{R := 0.0, G := 0.0, B := 0.0}, DefaultOpacity := 1.0} }, Text := "", Name := "B_Purchase", GroupName := "SnailShowcase", ButtonType := E_ButtonType.Loud, AnchorMinX := 0.5, AnchorMinY := 0.5, AnchorMaxX := 0.5, AnchorMaxY := 0.5}
				Image_Block{X:=-659.0, Y := 3.0, ScaleX := 312, ScaleY := 346, Texture := option{VerseExposed.T_CowBoyColor}, Name := "T_CowBoyColor", GroupName := "CowBoyShowcase", AnchorMinX := 0.5, AnchorMinY := 0.5, AnchorMaxX := 0.5, AnchorMaxY := 0.5}
				Image_Block{X:=-659.0, Y := 3.0, ScaleX := 312, ScaleY := 347, Texture := option{VerseExposed.T_CowBoy}, Name := "T_CowBoy", GroupName := "CowBoyShowcase", AnchorMinX := 0.5, AnchorMinY := 0.5, AnchorMaxX := 0.5, AnchorMaxY := 0.5}
				Image_Block{X:=-661.0, Y := 3.0, ScaleX := 315, ScaleY := 346, Texture := option{VerseExposed.T_CowBoyEquip}, Name := "T_CowBoyEquip", GroupName := "CowBoyShowcase", AnchorMinX := 0.5, AnchorMinY := 0.5, AnchorMaxX := 0.5, AnchorMaxY := 0.5}
				Button_Quiet_Data{X:=-584.0, Y := 292.0, ScaleX := 228, ScaleY := 44, Color := option{color_block{DefaultColor := color{R := 0.0, G := 0.0, B := 0.0}, DefaultOpacity := 1.0} }, Text := "", Name := "B_Purchase", GroupName := "CowBoyShowcase", ButtonType := E_ButtonType.Loud, AnchorMinX := 0.5, AnchorMinY := 0.5, AnchorMaxX := 0.5, AnchorMaxY := 0.5}
				Image_Block{X:=358.0, Y := -360.0, ScaleX := 312, ScaleY := 344, Texture := option{VerseExposed.T_HeadPhonesColor}, Name := "T_HeadPhonesColor", GroupName := "HeadPhonesShowcase", AnchorMinX := 0.5, AnchorMinY := 0.5, AnchorMaxX := 0.5, AnchorMaxY := 0.5}
				Image_Block{X:=358.0, Y := -360.0, ScaleX := 312, ScaleY := 344, Texture := option{VerseExposed.T_HeadPhones}, Name := "T_HeadPhones", GroupName := "HeadPhonesShowcase", AnchorMinX := 0.5, AnchorMinY := 0.5, AnchorMaxX := 0.5, AnchorMaxY := 0.5}
				Image_Block{X:=359.0, Y := -360.0, ScaleX := 312, ScaleY := 344, Texture := option{VerseExposed.T_HeadPhonesEquip}, Name := "T_HeadPhonesEquip", GroupName := "HeadPhonesShowcase", AnchorMinX := 0.5, AnchorMinY := 0.5, AnchorMaxX := 0.5, AnchorMaxY := 0.5}
				Button_Quiet_Data{X:=434.0, Y := -72.0, ScaleX := 227, ScaleY := 44, Color := option{color_block{DefaultColor := color{R := 0.0, G := 0.0, B := 0.0}, DefaultOpacity := 1.0} }, Text := "", Name := "B_Purchase", GroupName := "HeadPhonesShowcase", ButtonType := E_ButtonType.Loud, AnchorMinX := 0.5, AnchorMinY := 0.5, AnchorMaxX := 0.5, AnchorMaxY := 0.5}
				Image_Block{X:=23.0, Y := -360.0, ScaleX := 309, ScaleY := 346, Texture := option{VerseExposed.T_DonutPileColor}, Name := "T_DonutPileColor", GroupName := "DonutPileShowcase", AnchorMinX := 0.5, AnchorMinY := 0.5, AnchorMaxX := 0.5, AnchorMaxY := 0.5}
				Image_Block{X:=23.0, Y := -360.0, ScaleX := 308, ScaleY := 347, Texture := option{VerseExposed.T_DonutPile}, Name := "T_DonutPile", GroupName := "DonutPileShowcase", AnchorMinX := 0.5, AnchorMinY := 0.5, AnchorMaxX := 0.5, AnchorMaxY := 0.5}
				Image_Block{X:=23.0, Y := -360.0, ScaleX := 309, ScaleY := 346, Texture := option{VerseExposed.T_DonutPileEquip}, Name := "T_DonutPileEquip", GroupName := "DonutPileShowcase", AnchorMinX := 0.5, AnchorMinY := 0.5, AnchorMaxX := 0.5, AnchorMaxY := 0.5}
				Button_Quiet_Data{X:=98.0, Y := -71.0, ScaleX := 228, ScaleY := 44, Color := option{color_block{DefaultColor := color{R := 0.0, G := 0.0, B := 0.0}, DefaultOpacity := 1.0} }, Text := "", Name := "B_Purchase", GroupName := "DonutPileShowcase", ButtonType := E_ButtonType.Loud, AnchorMinX := 0.5, AnchorMinY := 0.5, AnchorMaxX := 0.5, AnchorMaxY := 0.5}
				Image_Block{X:=-316.0, Y := -359.0, ScaleX := 308, ScaleY := 344, Texture := option{VerseExposed.T_BoxEquip}, Name := "T_BoxEquip", GroupName := "BoxShowcase", AnchorMinX := 0.5, AnchorMinY := 0.5, AnchorMaxX := 0.5, AnchorMaxY := 0.5}
				Image_Block{X:=-318.0, Y := -359.0, ScaleX := 310, ScaleY := 344, Texture := option{VerseExposed.T_BoxColor}, Name := "T_BoxColor", GroupName := "BoxShowcase", AnchorMinX := 0.5, AnchorMinY := 0.5, AnchorMaxX := 0.5, AnchorMaxY := 0.5}
				Image_Block{X:=-318.0, Y := -359.0, ScaleX := 310, ScaleY := 344, Texture := option{VerseExposed.T_Box}, Name := "T_Box", GroupName := "BoxShowcase", AnchorMinX := 0.5, AnchorMinY := 0.5, AnchorMaxX := 0.5, AnchorMaxY := 0.5}
				Button_Quiet_Data{X:=-247.0, Y := -72.0, ScaleX := 228, ScaleY := 44, Color := option{color_block{DefaultColor := color{R := 0.0, G := 0.0, B := 0.0}, DefaultOpacity := 1.0} }, Text := "", Name := "B_Purchase", GroupName := "BoxShowcase", ButtonType := E_ButtonType.Loud, AnchorMinX := 0.5, AnchorMinY := 0.5, AnchorMaxX := 0.5, AnchorMaxY := 0.5}
				Image_Block{X:=-659.0, Y := -360.0, ScaleX := 312, ScaleY := 344, Texture := option{VerseExposed.T_ChickenColor}, Name := "T_ChickenColor", GroupName := "ChickenShowcase", AnchorMinX := 0.5, AnchorMinY := 0.5, AnchorMaxX := 0.5, AnchorMaxY := 0.5}
				Image_Block{X:=-659.0, Y := -360.0, ScaleX := 312, ScaleY := 344, Texture := option{VerseExposed.T_Chicken}, Name := "T_Chicken", GroupName := "ChickenShowcase", AnchorMinX := 0.5, AnchorMinY := 0.5, AnchorMaxX := 0.5, AnchorMaxY := 0.5}
				Image_Block{X:=-659.0, Y := -360.0, ScaleX := 312, ScaleY := 344, Texture := option{VerseExposed.T_ChickenEquip}, Name := "T_ChickenEquip", GroupName := "ChickenShowcase", AnchorMinX := 0.5, AnchorMinY := 0.5, AnchorMaxX := 0.5, AnchorMaxY := 0.5}
				Button_Quiet_Data{X:=-585.0, Y := -75.0, ScaleX := 228, ScaleY := 44, Color := option{color_block{DefaultColor := color{R := 0.0, G := 0.0, B := 0.0}, DefaultOpacity := 1.0} }, Text := "", Name := "B_Purchase", GroupName := "ChickenShowcase", ButtonType := E_ButtonType.Loud, AnchorMinX := 0.5, AnchorMinY := 0.5, AnchorMaxX := 0.5, AnchorMaxY := 0.5}
				Button_Quiet_Data{X:=-743.0, Y := -472.0, ScaleX := 158, ScaleY := 62, Color := option{color_block{DefaultColor := color{R := 0.0, G := 0.0, B := 0.0}, DefaultOpacity := 1.0} }, Text := "HOME", Name := "B_HomeMenu", GroupName := "HatsMenu", ButtonType := E_ButtonType.Loud, AnchorMinX := 0.5, AnchorMinY := 0.5, AnchorMaxX := 0.5, AnchorMaxY := 0.5}
				Button_Quiet_Data{X:=-131.0, Y := 375.0, ScaleX := 325, ScaleY := 122, Color := option{color_block{DefaultColor := color{R := 1.0, G := 1.0, B := 1.0}, DefaultOpacity := 1.0} }, Text := "EXIT", Name := "B_Exit", GroupName := "HatsMenu", ButtonType := E_ButtonType.Loud, AnchorMinX := 0.5, AnchorMinY := 0.5, AnchorMaxX := 0.5, AnchorMaxY := 0.5}
				Button_Quiet_Data{X:=-551.0, Y := -477.0, ScaleX := 182, ScaleY := 77, Color := option{color_block{DefaultColor := color{R := 0.0, G := 0.0, B := 0.0}, DefaultOpacity := 0.4588235294117647} }, Text := "ITEMS", Name := "B_HatsMenu", GroupName := "HatsMenu", ButtonType := E_ButtonType.Loud, AnchorMinX := 0.5, AnchorMinY := 0.5, AnchorMaxX := 0.5, AnchorMaxY := 0.5}
				Text_Block{X:=652.0, Y := -457.0, ScaleX := 66, ScaleY := 28, Text := "9999", Name := "Txt_Gold", GroupName := "HatsMenu", Color := color{R := 1.0, G := 1.0, B := 1.0}, Opacity := 1.0, AnchorMinX := 0.5, AnchorMinY := 0.5, AnchorMaxX := 0.5, AnchorMaxY := 0.5}
				Image_Block{X:=-885.0, Y := -497.0, ScaleX := 1777, ScaleY := 1000, Texture := option{VerseExposed.T_HomeMenuBackground}, Name := "T_HomeMenuBackground", GroupName := "HomeMenu", AnchorMinX := 0.5, AnchorMinY := 0.5, AnchorMaxX := 0.5, AnchorMaxY := 0.5}
				Image_Block{X:=203.0, Y := -348.0, ScaleX := 560, ScaleY := 656, Texture := option{VerseExposed.T_Advert_1}, Name := "T_Advert_1", GroupName := "Advertpage_1", AnchorMinX := 0.5, AnchorMinY := 0.5, AnchorMaxX := 0.5, AnchorMaxY := 0.5}
				Image_Block{X:=203.0, Y := -348.0, ScaleX := 560, ScaleY := 656, Texture := option{VerseExposed.T_Advert_2}, Name := "T_Advert_2", GroupName := "Advertpage_2", AnchorMinX := 0.5, AnchorMinY := 0.5, AnchorMaxX := 0.5, AnchorMaxY := 0.5}
				Button_Quiet_Data{X:=-392.0, Y := 381.0, ScaleX := 847, ScaleY := 110, Color := option{color_block{DefaultColor := color{R := 0.0, G := 0.0, B := 0.0}, DefaultOpacity := 1.0} }, Text := "EXIT", Name := "B_Exit", GroupName := "HomeMenu", ButtonType := E_ButtonType.Loud, AnchorMinX := 0.5, AnchorMinY := 0.5, AnchorMaxX := 0.5, AnchorMaxY := 0.5}
				Button_Quiet_Data{X:=-788.0, Y := -477.0, ScaleX := 242, ScaleY := 81, Color := option{color_block{DefaultColor := color{R := 0.0, G := 0.0, B := 0.0}, DefaultOpacity := 0.4980392156862745} }, Text := "HOME", Name := "B_HomeMenu", GroupName := "HomeMenu", ButtonType := E_ButtonType.Loud, AnchorMinX := 0.5, AnchorMinY := 0.5, AnchorMaxX := 0.5, AnchorMaxY := 0.5}
				Button_Quiet_Data{X:=-508.0, Y := -472.0, ScaleX := 182, ScaleY := 62, Color := option{color_block{DefaultColor := color{R := 0.0, G := 0.0, B := 0.0}, DefaultOpacity := 1.0} }, Text := "ITEMS", Name := "B_HatsMenu", GroupName := "HomeMenu", ButtonType := E_ButtonType.Loud, AnchorMinX := 0.5, AnchorMinY := 0.5, AnchorMaxX := 0.5, AnchorMaxY := 0.5}
				Image_Block{X:=-697.0, Y := 102.0, ScaleX := 764, ScaleY := 249, Texture := option{VerseExposed.T_RewardNotReady}, Name := "T_RewardNotReady", GroupName := "TimedReward", AnchorMinX := 0.5, AnchorMinY := 0.5, AnchorMaxX := 0.5, AnchorMaxY := 0.5}
				Image_Block{X:=-697.0, Y := 102.0, ScaleX := 764, ScaleY := 249, Texture := option{VerseExposed.T_RewardToCollect}, Name := "T_RewardToCollect", GroupName := "TimedReward", AnchorMinX := 0.5, AnchorMinY := 0.5, AnchorMaxX := 0.5, AnchorMaxY := 0.5}
				Image_Block{X:=-711.0, Y := 66.0, ScaleX := 778, ScaleY := 330, Texture := option{VerseExposed.T_RewardCollected}, Name := "T_RewardCollected", GroupName := "TimedReward", AnchorMinX := 0.5, AnchorMinY := 0.5, AnchorMaxX := 0.5, AnchorMaxY := 0.5}
				Text_Block{X:=-56.0, Y := 215.0, ScaleX := 30, ScaleY := 27, Text := "10", Name := "Txt_Timer", GroupName := "TimedReward", Color := color{R := 0.4980392156862745, G := 0.0, B := 1.0}, Opacity := 1.0, AnchorMinX := 0.5, AnchorMinY := 0.5, AnchorMaxX := 0.5, AnchorMaxY := 0.5}
				Button_Quiet_Data{X:=-445.0, Y := 200.0, ScaleX := 277, ScaleY := 73, Color := option{color_block{DefaultColor := color{R := 0.8784313725490196, G := 0.7137254901960784, B := 0.13725490196078433}, DefaultOpacity := 1.0} }, Text := "", Name := "B_CollectIn", GroupName := "TimedReward", ButtonType := E_ButtonType.Quiet, AnchorMinX := 0.5, AnchorMinY := 0.5, AnchorMaxX := 0.5, AnchorMaxY := 0.5}
				Button_Quiet_Data{X:=-445.0, Y := 200.0, ScaleX := 277, ScaleY := 73, Color := option{color_block{DefaultColor := color{R := 0.8784313725490196, G := 0.7137254901960784, B := 0.13725490196078433}, DefaultOpacity := 1.0} }, Text := "COLLECT", Name := "B_Collect", GroupName := "TimedReward", ButtonType := E_ButtonType.Loud, AnchorMinX := 0.5, AnchorMinY := 0.5, AnchorMaxX := 0.5, AnchorMaxY := 0.5}
				Image_Block{X:=-773.0, Y := -330.0, ScaleX := 288, ScaleY := 280, Texture := option{VerseExposed.T_GunLoadout1}, Name := "T_GunLoadout1", GroupName := "GunLoadout1Showcase", AnchorMinX := 0.5, AnchorMinY := 0.5, AnchorMaxX := 0.5, AnchorMaxY := 0.5}
				Image_Block{X:=-773.0, Y := -330.0, ScaleX := 288, ScaleY := 280, Texture := option{VerseExposed.T_GunLoadout1Color}, Name := "T_GunLoadout1Color", GroupName := "GunLoadout1Showcase", AnchorMinX := 0.5, AnchorMinY := 0.5, AnchorMaxX := 0.5, AnchorMaxY := 0.5}
				Image_Block{X:=-773.0, Y := -330.0, ScaleX := 288, ScaleY := 280, Texture := option{VerseExposed.T_GunLoadout1Equip}, Name := "T_GunLoadout1Equip", GroupName := "GunLoadout1Showcase", AnchorMinX := 0.5, AnchorMinY := 0.5, AnchorMaxX := 0.5, AnchorMaxY := 0.5}
				Image_Block{X:=-456.0, Y := -330.0, ScaleX := 288, ScaleY := 280, Texture := option{VerseExposed.T_GunLoadout2}, Name := "T_GunLoadout2", GroupName := "GunLoadout2Showcase", AnchorMinX := 0.5, AnchorMinY := 0.5, AnchorMaxX := 0.5, AnchorMaxY := 0.5}
				Image_Block{X:=-456.0, Y := -330.0, ScaleX := 288, ScaleY := 280, Texture := option{VerseExposed.T_GunLoadout2Color}, Name := "T_GunLoadout2Color", GroupName := "GunLoadout2Showcase", AnchorMinX := 0.5, AnchorMinY := 0.5, AnchorMaxX := 0.5, AnchorMaxY := 0.5}
				Image_Block{X:=-456.0, Y := -330.0, ScaleX := 288, ScaleY := 280, Texture := option{VerseExposed.T_GunLoadout2Equip}, Name := "T_GunLoadout2Equip", GroupName := "GunLoadout2Showcase", AnchorMinX := 0.5, AnchorMinY := 0.5, AnchorMaxX := 0.5, AnchorMaxY := 0.5}
				Image_Block{X:=-139.0, Y := -330.0, ScaleX := 288, ScaleY := 280, Texture := option{VerseExposed.T_GunLoadout3}, Name := "T_GunLoadout3", GroupName := "GunLoadout3Showcase", AnchorMinX := 0.5, AnchorMinY := 0.5, AnchorMaxX := 0.5, AnchorMaxY := 0.5}
				Image_Block{X:=-139.0, Y := -330.0, ScaleX := 288, ScaleY := 280, Texture := option{VerseExposed.T_GunLoadout3Color}, Name := "T_GunLoadout3Color", GroupName := "GunLoadout3Showcase", AnchorMinX := 0.5, AnchorMinY := 0.5, AnchorMaxX := 0.5, AnchorMaxY := 0.5}
				Image_Block{X:=-139.0, Y := -330.0, ScaleX := 288, ScaleY := 280, Texture := option{VerseExposed.T_GunLoadout3Equip}, Name := "T_GunLoadout3Equip", GroupName := "GunLoadout3Showcase", AnchorMinX := 0.5, AnchorMinY := 0.5, AnchorMaxX := 0.5, AnchorMaxY := 0.5}
				Text_Block{X:=652.0, Y := -457.0, ScaleX := 66, ScaleY := 28, Text := "9999", Name := "Txt_Gold", GroupName := "HomeMenu", Color := color{R := 1.0, G := 1.0, B := 1.0}, Opacity := 1.0, AnchorMinX := 0.5, AnchorMinY := 0.5, AnchorMaxX := 0.5, AnchorMaxY := 0.5}
				Button_Quiet_Data{X:=-334.0, Y := 387.0, ScaleX := 340, ScaleY := 102, Color := option{color_block{DefaultColor := color{R := 0.0, G := 0.0, B := 0.0}, DefaultOpacity := 1.0} }, Text := "SPECTATE", Name := "B_Spectate", GroupName := "SpectateMenu", ButtonType := E_ButtonType.Loud, AnchorMinX := 0.5, AnchorMinY := 0.5, AnchorMaxX := 0.5, AnchorMaxY := 0.5}
				Button_Quiet_Data{X:=46.0, Y := 387.0, ScaleX := 340, ScaleY := 102, Color := option{color_block{DefaultColor := color{R := 0.0, G := 0.0, B := 0.0}, DefaultOpacity := 1.0} }, Text := "ITEM SHOP", Name := "B_ItemShop", GroupName := "SpectateMenu", ButtonType := E_ButtonType.Loud, AnchorMinX := 0.5, AnchorMinY := 0.5, AnchorMaxX := 0.5, AnchorMaxY := 0.5}
			ButtonActivateTriggers := ButtonTriggerData
			ShowGroups := ShowGroupData
			ExitButtons := ExitButtonData
E_Shop_Button := enum:
	AstronautShowcase_B_Purchase
	BoxShowcase_B_Purchase
	Brand1Item1Showcase_B_Purchase
	Brand1Item2Showcase_B_Purchase
	Brand1Item3Showcase_B_Purchase
	Brand1Menu_B_Brand1Menu
	Brand1Menu_B_Brand2Menu
	Brand1Menu_B_Exit
	Brand1Menu_B_HatsMenu
	Brand1Menu_B_HomeMenu
	Brand2Item1Showcase_B_Purchase
	Brand2Item2Showcase_B_Purchase
	Brand2Item3Showcase_B_Purchase
	Brand2Menu_B_Brand1Menu
	Brand2Menu_B_Brand2Menu
	Brand2Menu_B_Exit
	Brand2Menu_B_HatsMenu
	Brand2Menu_B_HomeMenu
	ChickenShowcase_B_Purchase
	CowBoyShowcase_B_Purchase
	CrownShowcase_B_Purchase
	GunLoadout1Showcase_B_Purchase
	GunLoadout2Showcase_B_Purchase
	GunLoadout3Showcase_B_Purchase
	HatsMenu_B_AstronautShowcase
	HatsMenu_B_BoxShowcase
	HatsMenu_B_Brand1Menu
	HatsMenu_B_Brand2Menu
	HatsMenu_B_ChickenShowcase
	HatsMenu_B_CowBoyShowcase
	HatsMenu_B_CrownShowcase
	HatsMenu_B_Exit
	HatsMenu_B_HatsMenu
	HatsMenu_B_HeadPhonesShowcase
	HatsMenu_B_HomeMenu
	HatsMenu_B_SnailShowcase
	HatsMenu_B_SuperSaiyanShowcase
	HeadPhonesShowcase_B_Purchase
	HomeMenu_B_Brand1Menu
	HomeMenu_B_Brand2Menu
	HomeMenu_B_Exit
	HomeMenu_B_HatsMenu
	HomeMenu_B_HomeMenu
	SnailShowcase_B_Purchase
	SpectateMenu_B_ItemShop
	SpectateMenu_B_Spectate
	SuperSaiyanShowcase_B_Purchase
	TimedReward_B_Collect
	DonutPileShowcase_B_Purchase
	HatsMenu_B_DonutPileShowcase
	TimedReward_B_CollectIn
	SantaShowcase_B_Purchase

(Button : E_Shop_Button).GetName()<transacts>:tuple(string, string)=
	case(Button):
		E_Shop_Button.AstronautShowcase_B_Purchase =>
			return ("AstronautShowcase", "B_Purchase")
		E_Shop_Button.BoxShowcase_B_Purchase =>
			return ("BoxShowcase", "B_Purchase")
		E_Shop_Button.Brand1Item1Showcase_B_Purchase =>
			return ("Brand1Item1Showcase", "B_Purchase")
		E_Shop_Button.Brand1Item2Showcase_B_Purchase =>
			return ("Brand1Item2Showcase", "B_Purchase")
		E_Shop_Button.Brand1Item3Showcase_B_Purchase =>
			return ("Brand1Item3Showcase", "B_Purchase")
		E_Shop_Button.Brand1Menu_B_Brand1Menu =>
			return ("Brand1Menu", "B_Brand1Menu")
		E_Shop_Button.Brand1Menu_B_Brand2Menu =>
			return ("Brand1Menu", "B_Brand2Menu")
		E_Shop_Button.Brand1Menu_B_Exit =>
			return ("Brand1Menu", "B_Exit")
		E_Shop_Button.Brand1Menu_B_HatsMenu =>
			return ("Brand1Menu", "B_HatsMenu")
		E_Shop_Button.Brand1Menu_B_HomeMenu =>
			return ("Brand1Menu", "B_HomeMenu")
		E_Shop_Button.Brand2Item1Showcase_B_Purchase =>
			return ("Brand2Item1Showcase", "B_Purchase")
		E_Shop_Button.Brand2Item2Showcase_B_Purchase =>
			return ("Brand2Item2Showcase", "B_Purchase")
		E_Shop_Button.Brand2Item3Showcase_B_Purchase =>
			return ("Brand2Item3Showcase", "B_Purchase")
		E_Shop_Button.Brand2Menu_B_Brand1Menu =>
			return ("Brand2Menu", "B_Brand1Menu")
		E_Shop_Button.Brand2Menu_B_Brand2Menu =>
			return ("Brand2Menu", "B_Brand2Menu")
		E_Shop_Button.Brand2Menu_B_Exit =>
			return ("Brand2Menu", "B_Exit")
		E_Shop_Button.Brand2Menu_B_HatsMenu =>
			return ("Brand2Menu", "B_HatsMenu")
		E_Shop_Button.Brand2Menu_B_HomeMenu =>
			return ("Brand2Menu", "B_HomeMenu")
		E_Shop_Button.ChickenShowcase_B_Purchase =>
			return ("ChickenShowcase", "B_Purchase")
		E_Shop_Button.CowBoyShowcase_B_Purchase =>
			return ("CowBoyShowcase", "B_Purchase")
		E_Shop_Button.CrownShowcase_B_Purchase =>
			return ("CrownShowcase", "B_Purchase")
		E_Shop_Button.GunLoadout1Showcase_B_Purchase =>
			return ("GunLoadout1Showcase", "B_Purchase")
		E_Shop_Button.GunLoadout2Showcase_B_Purchase =>
			return ("GunLoadout2Showcase", "B_Purchase")
		E_Shop_Button.GunLoadout3Showcase_B_Purchase =>
			return ("GunLoadout3Showcase", "B_Purchase")
		E_Shop_Button.HatsMenu_B_AstronautShowcase =>
			return ("HatsMenu", "B_AstronautShowcase")
		E_Shop_Button.HatsMenu_B_BoxShowcase =>
			return ("HatsMenu", "B_BoxShowcase")
		E_Shop_Button.HatsMenu_B_Brand1Menu =>
			return ("HatsMenu", "B_Brand1Menu")
		E_Shop_Button.HatsMenu_B_Brand2Menu =>
			return ("HatsMenu", "B_Brand2Menu")
		E_Shop_Button.HatsMenu_B_ChickenShowcase =>
			return ("HatsMenu", "B_ChickenShowcase")
		E_Shop_Button.HatsMenu_B_CowBoyShowcase =>
			return ("HatsMenu", "B_CowBoyShowcase")
		E_Shop_Button.HatsMenu_B_CrownShowcase =>
			return ("HatsMenu", "B_CrownShowcase")
		E_Shop_Button.HatsMenu_B_Exit =>
			return ("HatsMenu", "B_Exit")
		E_Shop_Button.HatsMenu_B_HatsMenu =>
			return ("HatsMenu", "B_HatsMenu")
		E_Shop_Button.HatsMenu_B_HeadPhonesShowcase =>
			return ("HatsMenu", "B_HeadPhonesShowcase")
		E_Shop_Button.HatsMenu_B_HomeMenu =>
			return ("HatsMenu", "B_HomeMenu")
		E_Shop_Button.HatsMenu_B_SnailShowcase =>
			return ("HatsMenu", "B_SnailShowcase")
		E_Shop_Button.HatsMenu_B_SuperSaiyanShowcase =>
			return ("HatsMenu", "B_SuperSaiyanShowcase")
		E_Shop_Button.HeadPhonesShowcase_B_Purchase =>
			return ("HeadPhonesShowcase", "B_Purchase")
		E_Shop_Button.HomeMenu_B_Brand1Menu =>
			return ("HomeMenu", "B_Brand1Menu")
		E_Shop_Button.HomeMenu_B_Brand2Menu =>
			return ("HomeMenu", "B_Brand2Menu")
		E_Shop_Button.HomeMenu_B_Exit =>
			return ("HomeMenu", "B_Exit")
		E_Shop_Button.HomeMenu_B_HatsMenu =>
			return ("HomeMenu", "B_HatsMenu")
		E_Shop_Button.HomeMenu_B_HomeMenu =>
			return ("HomeMenu", "B_HomeMenu")
		E_Shop_Button.SnailShowcase_B_Purchase =>
			return ("SnailShowcase", "B_Purchase")
		E_Shop_Button.SpectateMenu_B_ItemShop =>
			return ("SpectateMenu", "B_ItemShop")
		E_Shop_Button.SpectateMenu_B_Spectate =>
			return ("SpectateMenu", "B_Spectate")
		E_Shop_Button.SuperSaiyanShowcase_B_Purchase =>
			return ("SuperSaiyanShowcase", "B_Purchase")
		E_Shop_Button.TimedReward_B_Collect =>
			return ("TimedReward", "B_Collect")
		E_Shop_Button.DonutPileShowcase_B_Purchase =>
			return ("DonutPileShowcase", "B_Purchase")
		E_Shop_Button.HatsMenu_B_DonutPileShowcase =>
			return ("HatsMenu", "B_DonutPileShowcase")
		E_Shop_Button.TimedReward_B_CollectIn =>
			return ("TimedReward", "B_CollectIn")
		E_Shop_Button.SantaShowcase_B_Purchase =>
			return ("SantaShowcase", "B_Purchase")
E_Shop_Group := enum:
	Advertpage_1
	Advertpage_2
	AstronautShowcase
	BoxShowcase
	Brand1Item1Showcase
	Brand1Item2Showcase
	Brand1Item3Showcase
	Brand1Menu
	Brand2Item1Showcase
	Brand2Item2Showcase
	Brand2Item3Showcase
	Brand2Menu
	ChickenShowcase
	CowBoyShowcase
	CrownShowcase
	GunLoadout1Showcase
	GunLoadout2Showcase
	GunLoadout3Showcase
	HatsMenu
	HeadPhonesShowcase
	HomeMenu
	SnailShowcase
	SpectateMenu
	SuperSaiyanShowcase
	TimedReward
	root
	DonutPileShowcase
	SantaShowcase

(Group : E_Shop_Group).GetName()<transacts>:string=
	case(Group):
		E_Shop_Group.Advertpage_1 =>
			return "Advertpage_1"
		E_Shop_Group.Advertpage_2 =>
			return "Advertpage_2"
		E_Shop_Group.AstronautShowcase =>
			return "AstronautShowcase"
		E_Shop_Group.BoxShowcase =>
			return "BoxShowcase"
		E_Shop_Group.Brand1Item1Showcase =>
			return "Brand1Item1Showcase"
		E_Shop_Group.Brand1Item2Showcase =>
			return "Brand1Item2Showcase"
		E_Shop_Group.Brand1Item3Showcase =>
			return "Brand1Item3Showcase"
		E_Shop_Group.Brand1Menu =>
			return "Brand1Menu"
		E_Shop_Group.Brand2Item1Showcase =>
			return "Brand2Item1Showcase"
		E_Shop_Group.Brand2Item2Showcase =>
			return "Brand2Item2Showcase"
		E_Shop_Group.Brand2Item3Showcase =>
			return "Brand2Item3Showcase"
		E_Shop_Group.Brand2Menu =>
			return "Brand2Menu"
		E_Shop_Group.ChickenShowcase =>
			return "ChickenShowcase"
		E_Shop_Group.CowBoyShowcase =>
			return "CowBoyShowcase"
		E_Shop_Group.CrownShowcase =>
			return "CrownShowcase"
		E_Shop_Group.GunLoadout1Showcase =>
			return "GunLoadout1Showcase"
		E_Shop_Group.GunLoadout2Showcase =>
			return "GunLoadout2Showcase"
		E_Shop_Group.GunLoadout3Showcase =>
			return "GunLoadout3Showcase"
		E_Shop_Group.HatsMenu =>
			return "HatsMenu"
		E_Shop_Group.HeadPhonesShowcase =>
			return "HeadPhonesShowcase"
		E_Shop_Group.HomeMenu =>
			return "HomeMenu"
		E_Shop_Group.SnailShowcase =>
			return "SnailShowcase"
		E_Shop_Group.SpectateMenu =>
			return "SpectateMenu"
		E_Shop_Group.SuperSaiyanShowcase =>
			return "SuperSaiyanShowcase"
		E_Shop_Group.TimedReward =>
			return "TimedReward"
		E_Shop_Group.root =>
			return "root"
		E_Shop_Group.DonutPileShowcase =>
			return "DonutPileShowcase"
		E_Shop_Group.SantaShowcase =>
			return "SantaShowcase"
E_Shop_Widget := enum:
	 Brand2Menu_T_Brand2MenuBackground
	 Brand2Item3Showcase_B_Purchase
	 Brand2Item3Showcase_T_Brand2Item3Color
	 Brand2Item3Showcase_T_Brand2Item3
	 Brand2Item2Showcase_B_Purchase
	 Brand2Item2Showcase_T_Brand2Item2Color
	 Brand2Item2Showcase_T_Brand2Item2
	 Brand2Item1Showcase_B_Purchase
	 Brand2Item1Showcase_T_Brand2Item1Color
	 Brand2Item1Showcase_T_Brand2Item1
	 Brand2Menu_Txt_Gold
	 Brand2Menu_B_Exit
	 Brand2Menu_B_Brand2Menu
	 Brand2Menu_B_HatsMenu
	 Brand2Menu_B_HomeMenu
	 Brand2Menu_B_Brand1Menu
	 Brand1Menu_T_Brand1MenuBackground
	 Brand1Item3Showcase_B_Purchase
	 Brand1Item3Showcase_T_Brand1Item3Color
	 Brand1Item3Showcase_T_Brand1Item3
	 Brand1Item2Showcase_B_Purchase
	 Brand1Item2Showcase_T_Brand1Item2Color
	 Brand1Item2Showcase_T_Brand1Item2
	 Brand1Item1Showcase_B_Purchase
	 Brand1Item1Showcase_T_Brand1Item1Color
	 Brand1Item1Showcase_T_Brand1Item1
	 Brand1Menu_Txt_Gold
	 Brand1Menu_B_Exit
	 Brand1Menu_B_Brand1Menu
	 Brand1Menu_B_HatsMenu
	 Brand1Menu_B_HomeMenu
	 HatsMenu_T_HatsMenuBackground
	 SuperSaiyanShowcase_T_SuperSaiyanColor
	 SuperSaiyanShowcase_T_SuperSaiyan
	 SuperSaiyanShowcase_T_SuperSaiyanEquip
	 SuperSaiyanShowcase_B_Purchase
	 CrownShowcase_T_CrownColor
	 CrownShowcase_T_Crown
	 CrownShowcase_T_CrownEquip
	 CrownShowcase_B_Purchase
	 SnailShowcase_T_SnailColor
	 SnailShowcase_T_Snail
	 SnailShowcase_T_SnailEquip
	 SnailShowcase_B_Purchase
	 CowBoyShowcase_T_CowBoyColor
	 CowBoyShowcase_T_CowBoy
	 CowBoyShowcase_T_CowBoyEquip
	 CowBoyShowcase_B_Purchase
	 HeadPhonesShowcase_T_HeadPhonesColor
	 HeadPhonesShowcase_T_HeadPhones
	 HeadPhonesShowcase_T_HeadPhonesEquip
	 HeadPhonesShowcase_B_Purchase
	 DonutPileShowcase_T_DonutPileColor
	 DonutPileShowcase_T_DonutPile
	 DonutPileShowcase_T_DonutPileEquip
	 DonutPileShowcase_B_Purchase
	 BoxShowcase_T_BoxEquip
	 BoxShowcase_T_BoxColor
	 BoxShowcase_T_Box
	 BoxShowcase_B_Purchase
	 ChickenShowcase_T_ChickenColor
	 ChickenShowcase_T_Chicken
	 ChickenShowcase_T_ChickenEquip
	 ChickenShowcase_B_Purchase
	 HatsMenu_B_HomeMenu
	 HatsMenu_B_Exit
	 HatsMenu_B_HatsMenu
	 HatsMenu_Txt_Gold
	 HomeMenu_T_HomeMenuBackground
	 Advertpage_1_T_Advert_1
	 Advertpage_2_T_Advert_2
	 HomeMenu_B_Exit
	 HomeMenu_B_HomeMenu
	 HomeMenu_B_HatsMenu
	 TimedReward_T_RewardNotReady
	 TimedReward_T_RewardToCollect
	 TimedReward_T_RewardCollected
	 TimedReward_Txt_Timer
	 TimedReward_B_CollectIn
	 TimedReward_B_Collect
	 GunLoadout1Showcase_T_GunLoadout1
	 GunLoadout1Showcase_T_GunLoadout1Color
	 GunLoadout1Showcase_T_GunLoadout1Equip
	 GunLoadout2Showcase_T_GunLoadout2
	 GunLoadout2Showcase_T_GunLoadout2Color
	 GunLoadout2Showcase_T_GunLoadout2Equip
	 GunLoadout3Showcase_T_GunLoadout3
	 GunLoadout3Showcase_T_GunLoadout3Color
	 GunLoadout3Showcase_T_GunLoadout3Equip
	 HomeMenu_Txt_Gold
	 SpectateMenu_B_Spectate
	 SpectateMenu_B_ItemShop
(Widget : E_Shop_Widget).GetName()<transacts>:tuple(string, string)=
	case(Widget):
		E_Shop_Widget.Brand2Menu_T_Brand2MenuBackground =>
			return ("Brand2Menu", "T_Brand2MenuBackground")
		E_Shop_Widget.Brand2Item3Showcase_B_Purchase =>
			return ("Brand2Item3Showcase", "B_Purchase")
		E_Shop_Widget.Brand2Item3Showcase_T_Brand2Item3Color =>
			return ("Brand2Item3Showcase", "T_Brand2Item3Color")
		E_Shop_Widget.Brand2Item3Showcase_T_Brand2Item3 =>
			return ("Brand2Item3Showcase", "T_Brand2Item3")
		E_Shop_Widget.Brand2Item2Showcase_B_Purchase =>
			return ("Brand2Item2Showcase", "B_Purchase")
		E_Shop_Widget.Brand2Item2Showcase_T_Brand2Item2Color =>
			return ("Brand2Item2Showcase", "T_Brand2Item2Color")
		E_Shop_Widget.Brand2Item2Showcase_T_Brand2Item2 =>
			return ("Brand2Item2Showcase", "T_Brand2Item2")
		E_Shop_Widget.Brand2Item1Showcase_B_Purchase =>
			return ("Brand2Item1Showcase", "B_Purchase")
		E_Shop_Widget.Brand2Item1Showcase_T_Brand2Item1Color =>
			return ("Brand2Item1Showcase", "T_Brand2Item1Color")
		E_Shop_Widget.Brand2Item1Showcase_T_Brand2Item1 =>
			return ("Brand2Item1Showcase", "T_Brand2Item1")
		E_Shop_Widget.Brand2Menu_Txt_Gold =>
			return ("Brand2Menu", "Txt_Gold")
		E_Shop_Widget.Brand2Menu_B_Exit =>
			return ("Brand2Menu", "B_Exit")
		E_Shop_Widget.Brand2Menu_B_Brand2Menu =>
			return ("Brand2Menu", "B_Brand2Menu")
		E_Shop_Widget.Brand2Menu_B_HatsMenu =>
			return ("Brand2Menu", "B_HatsMenu")
		E_Shop_Widget.Brand2Menu_B_HomeMenu =>
			return ("Brand2Menu", "B_HomeMenu")
		E_Shop_Widget.Brand2Menu_B_Brand1Menu =>
			return ("Brand2Menu", "B_Brand1Menu")
		E_Shop_Widget.Brand1Menu_T_Brand1MenuBackground =>
			return ("Brand1Menu", "T_Brand1MenuBackground")
		E_Shop_Widget.Brand1Item3Showcase_B_Purchase =>
			return ("Brand1Item3Showcase", "B_Purchase")
		E_Shop_Widget.Brand1Item3Showcase_T_Brand1Item3Color =>
			return ("Brand1Item3Showcase", "T_Brand1Item3Color")
		E_Shop_Widget.Brand1Item3Showcase_T_Brand1Item3 =>
			return ("Brand1Item3Showcase", "T_Brand1Item3")
		E_Shop_Widget.Brand1Item2Showcase_B_Purchase =>
			return ("Brand1Item2Showcase", "B_Purchase")
		E_Shop_Widget.Brand1Item2Showcase_T_Brand1Item2Color =>
			return ("Brand1Item2Showcase", "T_Brand1Item2Color")
		E_Shop_Widget.Brand1Item2Showcase_T_Brand1Item2 =>
			return ("Brand1Item2Showcase", "T_Brand1Item2")
		E_Shop_Widget.Brand1Item1Showcase_B_Purchase =>
			return ("Brand1Item1Showcase", "B_Purchase")
		E_Shop_Widget.Brand1Item1Showcase_T_Brand1Item1Color =>
			return ("Brand1Item1Showcase", "T_Brand1Item1Color")
		E_Shop_Widget.Brand1Item1Showcase_T_Brand1Item1 =>
			return ("Brand1Item1Showcase", "T_Brand1Item1")
		E_Shop_Widget.Brand1Menu_Txt_Gold =>
			return ("Brand1Menu", "Txt_Gold")
		E_Shop_Widget.Brand1Menu_B_Exit =>
			return ("Brand1Menu", "B_Exit")
		E_Shop_Widget.Brand1Menu_B_Brand1Menu =>
			return ("Brand1Menu", "B_Brand1Menu")
		E_Shop_Widget.Brand1Menu_B_HatsMenu =>
			return ("Brand1Menu", "B_HatsMenu")
		E_Shop_Widget.Brand1Menu_B_HomeMenu =>
			return ("Brand1Menu", "B_HomeMenu")
		E_Shop_Widget.HatsMenu_T_HatsMenuBackground =>
			return ("HatsMenu", "T_HatsMenuBackground")
		E_Shop_Widget.SuperSaiyanShowcase_T_SuperSaiyanColor =>
			return ("SuperSaiyanShowcase", "T_SuperSaiyanColor")
		E_Shop_Widget.SuperSaiyanShowcase_T_SuperSaiyan =>
			return ("SuperSaiyanShowcase", "T_SuperSaiyan")
		E_Shop_Widget.SuperSaiyanShowcase_T_SuperSaiyanEquip =>
			return ("SuperSaiyanShowcase", "T_SuperSaiyanEquip")
		E_Shop_Widget.SuperSaiyanShowcase_B_Purchase =>
			return ("SuperSaiyanShowcase", "B_Purchase")
		E_Shop_Widget.CrownShowcase_T_CrownColor =>
			return ("CrownShowcase", "T_CrownColor")
		E_Shop_Widget.CrownShowcase_T_Crown =>
			return ("CrownShowcase", "T_Crown")
		E_Shop_Widget.CrownShowcase_T_CrownEquip =>
			return ("CrownShowcase", "T_CrownEquip")
		E_Shop_Widget.CrownShowcase_B_Purchase =>
			return ("CrownShowcase", "B_Purchase")
		E_Shop_Widget.SnailShowcase_T_SnailColor =>
			return ("SnailShowcase", "T_SnailColor")
		E_Shop_Widget.SnailShowcase_T_Snail =>
			return ("SnailShowcase", "T_Snail")
		E_Shop_Widget.SnailShowcase_T_SnailEquip =>
			return ("SnailShowcase", "T_SnailEquip")
		E_Shop_Widget.SnailShowcase_B_Purchase =>
			return ("SnailShowcase", "B_Purchase")
		E_Shop_Widget.CowBoyShowcase_T_CowBoyColor =>
			return ("CowBoyShowcase", "T_CowBoyColor")
		E_Shop_Widget.CowBoyShowcase_T_CowBoy =>
			return ("CowBoyShowcase", "T_CowBoy")
		E_Shop_Widget.CowBoyShowcase_T_CowBoyEquip =>
			return ("CowBoyShowcase", "T_CowBoyEquip")
		E_Shop_Widget.CowBoyShowcase_B_Purchase =>
			return ("CowBoyShowcase", "B_Purchase")
		E_Shop_Widget.HeadPhonesShowcase_T_HeadPhonesColor =>
			return ("HeadPhonesShowcase", "T_HeadPhonesColor")
		E_Shop_Widget.HeadPhonesShowcase_T_HeadPhones =>
			return ("HeadPhonesShowcase", "T_HeadPhones")
		E_Shop_Widget.HeadPhonesShowcase_T_HeadPhonesEquip =>
			return ("HeadPhonesShowcase", "T_HeadPhonesEquip")
		E_Shop_Widget.HeadPhonesShowcase_B_Purchase =>
			return ("HeadPhonesShowcase", "B_Purchase")
		E_Shop_Widget.DonutPileShowcase_T_DonutPileColor =>
			return ("DonutPileShowcase", "T_DonutPileColor")
		E_Shop_Widget.DonutPileShowcase_T_DonutPile =>
			return ("DonutPileShowcase", "T_DonutPile")
		E_Shop_Widget.DonutPileShowcase_T_DonutPileEquip =>
			return ("DonutPileShowcase", "T_DonutPileEquip")
		E_Shop_Widget.DonutPileShowcase_B_Purchase =>
			return ("DonutPileShowcase", "B_Purchase")
		E_Shop_Widget.BoxShowcase_T_BoxEquip =>
			return ("BoxShowcase", "T_BoxEquip")
		E_Shop_Widget.BoxShowcase_T_BoxColor =>
			return ("BoxShowcase", "T_BoxColor")
		E_Shop_Widget.BoxShowcase_T_Box =>
			return ("BoxShowcase", "T_Box")
		E_Shop_Widget.BoxShowcase_B_Purchase =>
			return ("BoxShowcase", "B_Purchase")
		E_Shop_Widget.ChickenShowcase_T_ChickenColor =>
			return ("ChickenShowcase", "T_ChickenColor")
		E_Shop_Widget.ChickenShowcase_T_Chicken =>
			return ("ChickenShowcase", "T_Chicken")
		E_Shop_Widget.ChickenShowcase_T_ChickenEquip =>
			return ("ChickenShowcase", "T_ChickenEquip")
		E_Shop_Widget.ChickenShowcase_B_Purchase =>
			return ("ChickenShowcase", "B_Purchase")
		E_Shop_Widget.HatsMenu_B_HomeMenu =>
			return ("HatsMenu", "B_HomeMenu")
		E_Shop_Widget.HatsMenu_B_Exit =>
			return ("HatsMenu", "B_Exit")
		E_Shop_Widget.HatsMenu_B_HatsMenu =>
			return ("HatsMenu", "B_HatsMenu")
		E_Shop_Widget.HatsMenu_Txt_Gold =>
			return ("HatsMenu", "Txt_Gold")
		E_Shop_Widget.HomeMenu_T_HomeMenuBackground =>
			return ("HomeMenu", "T_HomeMenuBackground")
		E_Shop_Widget.Advertpage_1_T_Advert_1 =>
			return ("Advertpage_1", "T_Advert_1")
		E_Shop_Widget.Advertpage_2_T_Advert_2 =>
			return ("Advertpage_2", "T_Advert_2")
		E_Shop_Widget.HomeMenu_B_Exit =>
			return ("HomeMenu", "B_Exit")
		E_Shop_Widget.HomeMenu_B_HomeMenu =>
			return ("HomeMenu", "B_HomeMenu")
		E_Shop_Widget.HomeMenu_B_HatsMenu =>
			return ("HomeMenu", "B_HatsMenu")
		E_Shop_Widget.TimedReward_T_RewardNotReady =>
			return ("TimedReward", "T_RewardNotReady")
		E_Shop_Widget.TimedReward_T_RewardToCollect =>
			return ("TimedReward", "T_RewardToCollect")
		E_Shop_Widget.TimedReward_T_RewardCollected =>
			return ("TimedReward", "T_RewardCollected")
		E_Shop_Widget.TimedReward_Txt_Timer =>
			return ("TimedReward", "Txt_Timer")
		E_Shop_Widget.TimedReward_B_CollectIn =>
			return ("TimedReward", "B_CollectIn")
		E_Shop_Widget.TimedReward_B_Collect =>
			return ("TimedReward", "B_Collect")
		E_Shop_Widget.GunLoadout1Showcase_T_GunLoadout1 =>
			return ("GunLoadout1Showcase", "T_GunLoadout1")
		E_Shop_Widget.GunLoadout1Showcase_T_GunLoadout1Color =>
			return ("GunLoadout1Showcase", "T_GunLoadout1Color")
		E_Shop_Widget.GunLoadout1Showcase_T_GunLoadout1Equip =>
			return ("GunLoadout1Showcase", "T_GunLoadout1Equip")
		E_Shop_Widget.GunLoadout2Showcase_T_GunLoadout2 =>
			return ("GunLoadout2Showcase", "T_GunLoadout2")
		E_Shop_Widget.GunLoadout2Showcase_T_GunLoadout2Color =>
			return ("GunLoadout2Showcase", "T_GunLoadout2Color")
		E_Shop_Widget.GunLoadout2Showcase_T_GunLoadout2Equip =>
			return ("GunLoadout2Showcase", "T_GunLoadout2Equip")
		E_Shop_Widget.GunLoadout3Showcase_T_GunLoadout3 =>
			return ("GunLoadout3Showcase", "T_GunLoadout3")
		E_Shop_Widget.GunLoadout3Showcase_T_GunLoadout3Color =>
			return ("GunLoadout3Showcase", "T_GunLoadout3Color")
		E_Shop_Widget.GunLoadout3Showcase_T_GunLoadout3Equip =>
			return ("GunLoadout3Showcase", "T_GunLoadout3Equip")
		E_Shop_Widget.HomeMenu_Txt_Gold =>
			return ("HomeMenu", "Txt_Gold")
		E_Shop_Widget.SpectateMenu_B_Spectate =>
			return ("SpectateMenu", "B_Spectate")
		E_Shop_Widget.SpectateMenu_B_ItemShop =>
			return ("SpectateMenu", "B_ItemShop")
Shop_BindToTriggers:= class<concrete>(Button_To_Trigger):
	@editable
	OnButtonPressed : ?E_Shop_Button = false
	GetButtonName<override>():tuple(string,string)=
		if(Button := OnButtonPressed?):
			return Button.GetName()
		return ("", "")
MakeShop_BindToTriggers<constructor>(OldData : Shop_BindToTriggers)<transacts> := Shop_BindToTriggers:
	OnButtonPressed := OldData.OnButtonPressed
	Triggers := OldData.Triggers
Shop_ShowGroupSettings:= class<concrete>(Show_Group_Settings):
	@editable
	Group : E_Shop_Group = E_Shop_Group.root
	GetGroupName<override>():string=
		set GroupName = Group.GetName()
		return GroupName
Shop_ShowGroups:= class<concrete>(Show_Groups):
	@editable
	OnButtonPressed : ?E_Shop_Button = false
	@editable
	GroupsSettings : []Shop_ShowGroupSettings = array{}
	GetButtonName<override>():tuple(string,string)=
		if(Button := OnButtonPressed?):
			return Button.GetName()
		return ("", "")
	GetGroupSettings<override>():[]Show_Group_Settings=
		return GroupsSettings
MakeShop_ShowGroups<constructor>(OldData : Shop_ShowGroups)<transacts> := Shop_ShowGroups:
	OnButtonPressed := OldData.OnButtonPressed
	OnTrigger := OldData.OnTrigger
	GroupsSettings := OldData.GroupsSettings
Shop_ExitButton:= class<concrete>(Exit_Button):
	@editable
	OnButtonPressed : ?E_Shop_Button = false
	GetButtonName<override>():tuple(string,string)=
		if(Button := OnButtonPressed?):
			return Button.GetName()
		return ("", "")
MakeShop_ExitButton<constructor>(OldData : Shop_ExitButton)<transacts> := Shop_ExitButton:
	OnButtonPressed := OldData.OnButtonPressed
	OnTrigger := OldData.OnTrigger

