using { /Fortnite.com/Devices }
using { /Verse.org/Simulation }
using { /UnrealEngine.com/Temporary/Diagnostics }
using { /UnrealEngine.com/Temporary/UI }
using { /Fortnite.com/UI }
using { /UnrealEngine.com/Temporary/SpatialMath}
using { /Verse.org/Colors }
using { /Verse.org/Assets }

# See https://dev.epicgames.com/documentation/en-us/uefn/create-your-own-device-in-verse for how to create a verse device.

# A Verse-authored creative device that can be placed in a level
V_Krita := class(creative_device):

    var PlayerUI : [player]UI_Data = map{}
    var TestUIData : UI_Data = UI_Data{}
    var ShowGroupConfig : []Show_Groups = array{}
    @editable
    Mutator : mutator_zone_device = mutator_zone_device{}
    @editable
    InteractNPC : character_device = character_device{}

    DefaultGroup : string = "SpectateMenu"
    # Example of condition based showing of widgets like effects system in GAS
    var Condition : []type{_():logic} = array{}
    FunFail():logic=
        false
    #@editable
    #Binds : test_Binds = test_Binds{}
    #T : texture = VerseExposed.T_Texture

    ######### Sorting algorithm lololol. Use this next time to sort things, recording my discovery here
    #SortBy(array{5, 10, 3}, Fun)
	#Fun(T1 : int, T2 : int)<computes><decides>:void=
		#T1 > T2
    

    # Runs when the device is started in a running game
    OnBegin<override><final>()<suspends>:void=
        # TODO: Replace this with your code
        Mutator.AgentEntersEvent.Subscribe(OnPlayerEnterZone)
        InteractNPC.InteractedWithEvent.Subscribe(OnPlayerEnterZone)
        _OnDeviceBegin()
        #Binds.GetBindData()

        <#set TestUIData = CreateUIData()
        TestUIData.CopydAllWidgetDataIntoGroupData()
        TestUIData.PrintALLGroupData()#>

    _OnDeviceBegin()<suspends>:void=
        block:
            <#AddShowGroupConfig
            (
                "n"
                "n"
                array:
                    Show_Group_Settings{GroupName := "n", Visibility := E_Show_Visibility_Option.Hide, ShowOnly := false, AllDescendants := true}
            )#>

    UITick(UIData : UI_Data, Player : player):void=
        block:
        

    AddShowGroupConfig(ButtonGroupName : string, ButtonName : string, ShowGroupSettings : []Show_Group_Settings):void=
        Group : Show_Groups = Show_Groups:
            ButtonName := ("{ButtonGroupName}","{ButtonName}")
            ShowGroupSettings := ShowGroupSettings
        set ShowGroupConfig += array{Group}
        

    AddShowGroupConfig(ButtonName : tuple(string, string), ShowGroupSettings : []Show_Group_Settings):void=
        Group : Show_Groups = Show_Groups:
            ButtonName := ButtonName
            ShowGroupSettings := ShowGroupSettings
        set ShowGroupConfig += array{Group}
            
    GetPlayerUIData(Player : player)<transacts><decides>:UI_Data=
        PlayerUI[Player]

    GetAllUIData():[]UI_Data=
        var TempUIData : []UI_Data = array{}
        for(Player -> UIData : PlayerUI):
            set TempUIData += array{UIData}
        return TempUIData

    CreateUIData<final>():UI_Data=
        UIData := _MakeUIDataFromKrita()
        return UIData
    
    # Override this function for krita data
    _MakeUIDataFromKrita():UI_Data=
        TempUIDATA := UI_Data{}
            #UIBinds := Binds.UIBinds
        return TempUIDATA

    _InjectUIData(Player : player, UIData : UI_Data):void=
        block:

    _OnUICreated(Player : player, UIData : UI_Data):void=
        block:

    _OnUIRendered(Player : player, UIData : UI_Data):void=
        block:

    _OnAllButtonSlotsAdded(Player : player, UIData : UI_Data):void=
        block:

    SetButtonTextAllPlayers(EnumName : tuple(string,string), Text : string):void=
        AllUIData := GetAllUIData()
        for(UIData : AllUIData):
            UIData.SetButtonText(EnumName, Text)
    SetButtonTextAllPlayers(GroupName : string, ButtonName : string, Text : string):void=
        AllUIData := GetAllUIData()
        for(UIData : AllUIData):
            UIData.SetButtonText(GroupName, ButtonName, Text)

    SetTextAllPlayers(EnumName : tuple(string,string), Text : string):void=
        AllUIData := GetAllUIData()
        for(UIData : AllUIData):
            UIData.SetText(EnumName, Text)
    SetTextAllPlayers(GroupName : string, TextName : string, Text : string):void=
        AllUIData := GetAllUIData()
        for(UIData : AllUIData):
            UIData.SetText(GroupName, TextName, Text)

    HideGroupUI(Player : player, GroupName : string, AllDecendants : logic, ?Permanent : logic = false):void=
        if(UIData := PlayerUI[Player]):
            UIData.HideGroup(GroupName, AllDecendants, ?Permanent := Permanent)
    ShowGroupUI(Player : player, GroupName : string, AllDecendants : logic, ?UnlockPermanent : logic = false):void=
        if(UIData := PlayerUI[Player]):
            UIData.ShowGroup(GroupName, AllDecendants, ?UnlockPermanent := UnlockPermanent)

    
    MakeUI<private>(Player : player, UIData : UI_Data, DefaultShowGroup : string):UI_Data=
        var JustB : []button_slot = array{}

        KritaWidgets := UIData.OrderedWidgetData
        var OverlaySlots : []overlay_slot = array{}
        var ButtonOverlaySlots : []overlay_slot = array{}
        var ButtonCanvasSlots : []canvas_slot = array{}
        var CanvasSlots : []canvas_slot = array{}
        var Canvases : []canvas = array{}
        # Copies data into data structures for access
        UIData.PreInit(Player, Self)
        # Users can inject custom data based on player data or external variables
        _InjectUIData(Player, UIData)
        for(i -> KritaWidget : KritaWidgets, Order := type{_X:int where 0 <= _X, _X <= 2147483647}[i]):
            set KritaWidget.Order = Order
            if(TextWidgetData := Text_Block[KritaWidget]):
                set TextWidgetData.TextBlock = text_block{DefaultText := (StringToMessage(TextWidgetData.Text)), DefaultTextColor :=  TextWidgetData.Color, DefaultTextOpacity := TextWidgetData.Opacity, DefaultJustification := text_justification.Center}#, DefaultShadowOffset := option{vector2{X := 0.0, Y := 10.0}},
                #DefaultShadowColor := color{R := 1.0, G := 0.0, B := 0.0}, DefaultShadowOpacity := 1.0}
                #set OverlaySlots += array{overlay_slot{Widget := TextWidgetData.TextBlock, Padding := margin{Left := TextWidgetData.X * 1.0, Top := TextWidgetData.Y * 1.0}, HorizontalAlignment := horizontal_alignment.Left, VerticalAlignment := vertical_alignment.Top}}
                #set CanvasSlots += array{MakeCanvasSlot(TextWidgetData.TextBlock, vector2{X:= TextWidgetData.X * 1.0, Y:= TextWidgetData.Y * 1.0}, ?Size := vector2{X :=  TextWidgetData.ScaleX * 1.0, Y :=  TextWidgetData.ScaleY * 1.0}, ?ZOrder := Order)}
                
                Slot :=
                    canvas_slot:
                        Anchors := anchors{Minimum := vector2{X:=TextWidgetData.AnchorMinX, Y:=TextWidgetData.AnchorMinY}, Maximum := vector2{X:=TextWidgetData.AnchorMaxX, Y:=TextWidgetData.AnchorMaxY}}
                        Alignment := vector2{X := 0.0, Y := 0.0}
                        Offsets := margin{Left := TextWidgetData.X - 3, Top := TextWidgetData.Y - 7, Right := TextWidgetData.ScaleX * 1.0, Bottom := TextWidgetData.ScaleY * 1.0}
                        SizeToContent := false
                        ZOrder := Order
                        Widget := TextWidgetData.TextBlock
                set CanvasSlots += array{Slot}
                set KritaWidget.bCanvasSlot = option{Slot}

            if(ColorWidgetData := Color_Block[KritaWidget]):
                #set OverlaySlots += array{overlay_slot{Widget := ColorWidgetData.Color, Padding := margin{Left := ColorWidgetData.X * 1.0, Top := ColorWidgetData.Y * 1.0}, HorizontalAlignment := horizontal_alignment.Left, VerticalAlignment := vertical_alignment.Top}}
                #set CanvasSlots += array{MakeCanvasSlot(ColorWidgetData.Color, vector2{X:= ColorWidgetData.X * 1.0, Y:= ColorWidgetData.Y * 1.0}, ?Size := vector2{X :=  ColorWidgetData.ScaleX * 1.0, Y :=  ColorWidgetData.ScaleY * 1.0}, ?ZOrder := Order)}
                
                Slot :=
                    canvas_slot:
                        Anchors := anchors{Minimum := vector2{X:=ColorWidgetData.AnchorMinX, Y:=ColorWidgetData.AnchorMinY}, Maximum := vector2{X:=ColorWidgetData.AnchorMaxX, Y:=ColorWidgetData.AnchorMaxY}}
                        Alignment := ColorWidgetData.Alignment
                        Offsets := margin{Left := ColorWidgetData.X, Top := ColorWidgetData.Y, Right := ColorWidgetData.ScaleX * 1.0, Bottom := ColorWidgetData.ScaleY * 1.0}
                        SizeToContent := ColorWidgetData.SizeToContent
                        ZOrder := Order
                        Widget := ColorWidgetData.Color
                set CanvasSlots += array{Slot}
                set KritaWidget.bCanvasSlot = option{Slot}

            if(ImageWidgetData := Image_Block[KritaWidget], Image := ImageWidgetData.Texture?):
                set ImageWidgetData.ImageTextureBlock = option{texture_block{DefaultImage := Image, DefaultDesiredSize := vector2{X := ImageWidgetData.ScaleX * 1.0, Y := ImageWidgetData.ScaleY * 1.0}}}
                if (ImageTextureBlock := ImageWidgetData.ImageTextureBlock?):
                    #set OverlaySlots += array{overlay_slot{Widget := ImageTextureBlock, Padding := margin{Left := ImageWidgetData.X * 1.0, Top := ImageWidgetData.Y * 1.0}, HorizontalAlignment := horizontal_alignment.Left, VerticalAlignment := vertical_alignment.Top}}
                    #set CanvasSlots += array{MakeCanvasSlot(ImageTextureBlock, vector2{X:= ImageWidgetData.X * 1.0, Y:= ImageWidgetData.Y * 1.0}, ?Size := vector2{X :=  ImageWidgetData.ScaleX * 1.0, Y :=  ImageWidgetData.ScaleY * 1.0}, ?ZOrder := Order)}
                    
                    Slot :=
                            canvas_slot:
                                Anchors := anchors{Minimum := vector2{X:=ImageWidgetData.AnchorMinX, Y:=ImageWidgetData.AnchorMinY}, Maximum := vector2{X:=ImageWidgetData.AnchorMaxX, Y:=ImageWidgetData.AnchorMaxY}}
                                Alignment := ImageWidgetData.Alignment
                                Offsets := margin{Left := ImageWidgetData.X, Top := ImageWidgetData.Y, Right := ImageWidgetData.ScaleX * 1.0, Bottom := ImageWidgetData.ScaleY * 1.0}
                                SizeToContent := ImageWidgetData.SizeToContent
                                ZOrder := Order
                                Widget := ImageTextureBlock
                    set CanvasSlots += array{Slot}
                    set KritaWidget.bCanvasSlot = option{Slot}

            if (ButtonWidgetData := Button_Quiet_Data[KritaWidget]):
                
                set KritaWidget.Order = 2147483646
                case(ButtonWidgetData.ButtonType):
                    E_ButtonType.Quiet =>
                        set ButtonWidgetData.Button = button_quiet{DefaultText := (StringToMessage(ButtonWidgetData.Text))}
                    E_ButtonType.Loud =>
                        set ButtonWidgetData.Button = button_loud{DefaultText := (StringToMessage(ButtonWidgetData.Text))}
                    E_ButtonType.Regular =>
                        set ButtonWidgetData.Button = button_regular{DefaultText := (StringToMessage(ButtonWidgetData.Text))}
                X := (ButtonWidgetData.X)
                Y := ButtonWidgetData.Y
                ScaleX := ButtonWidgetData.ScaleX
                ScaleY := ButtonWidgetData.ScaleY
                #set ButtonOverlaySlots += array{overlay_slot{Widget := ButtonWidgetData.Button, Padding := margin{Left := ButtonWidgetData.X * 1.0, Top := ButtonWidgetData.Y * 1.0}, HorizontalAlignment := horizontal_alignment.Left, VerticalAlignment := vertical_alignment.Top}}
                #set ButtonCanvasSlots += array{MakeCanvasSlot(ButtonWidgetData.Button, vector2{X:= ButtonWidgetData.X * 1.0, Y:= ButtonWidgetData.Y * 1.0}, ?Size := vector2{X :=  ButtonWidgetData.ScaleX * 1.0, Y :=  ButtonWidgetData.ScaleY * 1.0}, ?ZOrder := 2147483646)}
                
                Slot :=
                    canvas_slot:
                        Anchors := anchors{Minimum := vector2{X:=ButtonWidgetData.AnchorMinX, Y:=ButtonWidgetData.AnchorMinY}, Maximum := vector2{X:=ButtonWidgetData.AnchorMaxX, Y:=ButtonWidgetData.AnchorMaxY}}
                        Alignment := ButtonWidgetData.Alignment
                        Offsets := margin{Left := ButtonWidgetData.X, Top := ButtonWidgetData.Y, Right := ButtonWidgetData.ScaleX * 1.0, Bottom := ButtonWidgetData.ScaleY * 1.0}
                        SizeToContent := ButtonWidgetData.SizeToContent
                        ZOrder := 2147483646 #Order
                        Widget := ButtonWidgetData.Button
                set ButtonCanvasSlots += array{Slot}
                set KritaWidget.bCanvasSlot = option{Slot}

                if (Group := UIData.GroupUIMap[KritaWidget.GroupName], ButtonSlot := ButtonCanvasSlots[ButtonCanvasSlots.Length - 1]):
                    set Group.ButtonSlots += array{ButtonSlot}
        

                
        for(Slot : CanvasSlots):
            set Canvases += array{canvas{Slots:=array{Slot}}}
        for(CanvasWidget : Canvases):
            CanvasWidget.SetEnabled(true)
            set OverlaySlots += array{overlay_slot{Widget := CanvasWidget, HorizontalAlignment := horizontal_alignment.Fill, VerticalAlignment := vertical_alignment.Fill}}
        
        OverlayCanva := overlay:
            Slots := OverlaySlots# + ButtonOverlaySlots
                #overlay_slot{Widget := Canvas}
        #OverlayCanva.SetEnabled(false)
        CanvaOverlaySlot :=                                                   # MakeCanvasSlot(OverlayCanva, vector2{X:= 0.0, Y:= 0.0}, ?ZOrder := 0)
            canvas_slot:
                Anchors := anchors{Minimum := vector2{X:=0.5, Y:=0.5}, Maximum := vector2{X:=0.5, Y:=0.5}}
                Alignment := vector2{X:= 0.5, Y:= 0.5}
                Widget := OverlayCanva
        #var S : []canvas_slot = ButtonCanvasSlots
        #set S += array{TSlot}
        var SelectedButtons : []canvas_slot = array{}
        DefaultCghildGroupNames := UIData.GetAllDecendantGroupNames(DefaultShowGroup)
        DefaultGroupNames := DefaultCghildGroupNames + array{DefaultShowGroup}
            for(DefaultGroupName : DefaultGroupNames):
                if (Group := UIData.GetGroup[DefaultGroupName]):
                    set SelectedButtons += Group.ButtonSlots 

        Canvas := canvas:
            Slots :=  CanvasSlots + SelectedButtons #+ ButtonCanvasSlots #array{CanvaOverlaySlot} + ButtonCanvasSlots
        set UIData.Canvas = Canvas
        set UIData.DefaultShowGroup = DefaultShowGroup
        set UIData.ShowGroups += ShowGroupConfig
        UIData.PostInit(Player)
        UIData.ButtonActivateTriggers.Init(UIData, Player)
        UIData.ShowGroups.Init(UIData, Player)
        UIData.ExitButtons.Init(UIData, Player)
        UIData.ShowOnlyGroups(array{DefaultShowGroup}, true)
        _OnUICreated(Player, UIData)
        return UIData
        

    CreatePlayerUI(Player : player,  DefaultShowGroup : string):UI_Data=
        UIDATA := MakeUI(Player, CreateUIData(), DefaultShowGroup)
        if(OldUI := PlayerUI[Player]):
            OldUI.ExitUI()
        if(set PlayerUI[Player] = UIDATA){}
        return UIDATA

    RemovePlayerUI(Player:player):void=
        if (UIData := GetPlayerUIData[Player]):
            UIData.ExitUI()

    RemoveAllUI():void=
        for(Player -> UIData : PlayerUI):
            RemovePlayerUI(Player)

    OnPlayerEnterZone(Agent : agent):void=
        if (Player := player[Agent]):
            UIDataWithCanva := CreatePlayerUI(Player, DefaultGroup)
            spawn:
                RenderUIOnPlayerScreen(Player, UIDataWithCanva)
                
    RenderUIOnPlayerScreen(Player : player, UIDATA : UI_Data)<suspends>:void=
        if(PlayerUIDisplay := GetPlayerUI[Player]):
            #UIDATA.Canvas.SetEnabled(false)
            PlayerUIDisplay.AddWidget(UIDATA.Canvas, player_ui_slot{InputMode := ui_input_mode.All})
            _OnUIRendered(Player, UIDATA)
            Sleep(1.0)
            var SelectedButtons : []canvas_slot = array{}
            for(Group : UIDATA.GroupUIMap):
                if(not Group.Name = UIDATA.DefaultShowGroup):
                    set SelectedButtons += Group.ButtonSlots
            for(SelectedButton : SelectedButtons):
                UIDATA.Canvas.AddWidget(SelectedButton)
            _OnAllButtonSlotsAdded(Player, UIDATA)
               


                    #spawn{DisableButtons(UIDataWithCanva)}
    
    _OnUIExit(Player : player, UIData : UI_Data):void=
        block:
    
    <#DisableButtons(UIData : UI_Data)<suspends>:void=
        Sleep(5.0)
        for(Data : UIData.OrderedWidgetData):
            if (Bdata := Button_Quiet_Data[Data], Bdata.Name = "B_Button"):
                Bdata.Button.SetVisibility(widget_visibility.Hidden)#>


