using { /Fortnite.com/Devices }
using { /Verse.org/Simulation }
using { /Verse.org/Random }
using { /UnrealEngine.com/Temporary/Diagnostics }
using { /UnrealEngine.com/Temporary/SpatialMath }
using { /Fortnite.com/Characters}
using { /Fortnite.com/Devices/CreativeAnimation }

CustomLaserClass := class<concrete>():
    #This is the prop we control with player view
    @editable
    RaycastProp : creative_prop = creative_prop{}
    #The Raycast Hit Prop, the prop that moved to the hit location of the linetrace. We can use this to get the location and check things around them
    @editable
    RaycastHitProp : creative_prop = creative_prop{}
    @editable
    ExplosiveDevice : explosive_device = explosive_device{}
    #Input device to activate the ultimate
    #Was for Debugging
    #@editable
    #InputUltimate : input_trigger_device = input_trigger_device{}
    @editable
    VFXSpawner : vfx_spawner_device = vfx_spawner_device{}

V_LaserBeamUltimate := class(creative_device):
    @editable
    LaserAudio : audio_player_device = audio_player_device{}

    @editable
    LaserClassArray : []CustomLaserClass = array{}

    var CurrentIndex : int = 0

    var LaserPlayerMap : [player]CustomLaserClass = map{}

    @editable
    OrbitCameraActivateTrigger : trigger_device = trigger_device{}
    @editable
    OrbitCameraDeactivateTrigger : trigger_device = trigger_device{}

    #This is where the laser will teleport to while its not in use
    @editable
    LaserAFKLocation : creative_prop = creative_prop{}

    # Runs when the device is started in a running game
    #OnBegin<override>()<suspends>:void={}
        #Was for Debugging
        <#for(Player : GetPlayspace().GetPlayers()):
            RegisterPlayer(Player)#>

    RegisterPlayer(InPlayer : player) : void=
        Print("Array Length {LaserClassArray.Length}")
        if (LaserPlayerMap[InPlayer]):
            Print("Already Registered")
            return
        if(LaserClass := LaserClassArray[CurrentIndex]):
            Print("Registering - CurrentIndex -{CurrentIndex}")
            if(set LaserPlayerMap[InPlayer] = LaserClass):
                set CurrentIndex += 1
        Print("After Register")

    #This is done differently in minigame party, but this allows us to activate the laserbeam ultimate
    UltimateActive(InAgent : agent):void=
        spawn:
            LaserBeamUltimate(InAgent)

    LaserBeamUltimate(InAgent : agent)<suspends>:void=
        var HasFirstFrameTeleport : logic = false
        var bLaserClass : ?CustomLaserClass = false
        Print("Ult Called")
        if(Player := player[InAgent]):
            if(TempClass := LaserPlayerMap[Player]):
                set bLaserClass = option{TempClass}
                Print("Ult Found")
            else:
                Print("Ult Not Found")

        if(FortChar := InAgent.GetFortCharacter[], LaserClass := bLaserClass?):
            #Put the player in stasis
            FortChar.PutInStasis(stasis_args{AllowTurning := true, AllowFalling := true, AllowEmotes := true})
            FortChar.SetVulnerability(false)    
            OrbitCameraActivateTrigger.Trigger(InAgent)
            LaserClass.VFXSpawner.Enable()
            Print("Player In stasis", ?Duration := 8.0)
            LaserAudio.Play(InAgent)
            race:
                block:
                    loop:
                        #We use this to get the fort characters forward mesh rotation
                        ViewRotation := FortChar.GetViewRotation()
                        #We use this to get the cameras forward position so we can move the laser infront of the player
                        ViewDirection := ViewRotation.GetLocalForward()
                        #Location on the player the beam will appear and be cast from
                        PL_X := (ViewDirection.X * 60.0) + FortChar.GetTransform().Translation.X
                        PL_Y := (ViewDirection.Y * 60.0) + FortChar.GetTransform().Translation.Y
                        PL_Z := (ViewDirection.Z * 50.0) + FortChar.GetTransform().Translation.Z + 75.0

                        PlayerLocation : vector3 = vector3{X:= PL_X, Y:= PL_Y, Z:= PL_Z}

                        #Leaving this, it seems to fix long distance laser but we'll see if minigame needs it
                        #JitterLocation : vector3 = vector3{X:= PlayerLocation.X + GetRandomFloat(10.0, 20.0), Y:= PlayerLocation.Y + GetRandomFloat(10.0, 20.0), Z:= PlayerLocation.Z + GetRandomFloat(10.0, 20.0)}

                        #How we move the raycast prop, in this case we move player location "Once" but then the GetViewRotation is the key
                        if(HasFirstFrameTeleport = false):
                            set HasFirstFrameTeleport = true
                            if(LaserClass.RaycastProp.TeleportTo[PlayerLocation, FortChar.GetViewRotation()]):
                                Print("Lazering", ?Duration := 10.0)
                        else:
                            LaserClass.RaycastProp.MoveTo(PlayerLocation, FortChar.GetViewRotation(), 0.1)
                            
                        Sleep(0.0)
                block:
                    ExplodeOnImpactPoint(InAgent, LaserClass)
                block:
                    Sleep(6.0)
                    #Resets Player on end
                    LaserClass.VFXSpawner.Disable()
                    if(LaserClass.RaycastProp.TeleportTo[LaserAFKLocation.GetTransform()]):
                    if(LaserClass.ExplosiveDevice.TeleportTo[LaserAFKLocation.GetTransform()]):
            Print("laser ult got here")
            Sleep(1.0)
            OrbitCameraDeactivateTrigger.Trigger(InAgent)
            FortChar.ReleaseFromStasis()
            FortChar.SetVulnerability(true)

    #Damages Characters and Walls we detect around the impact location.
    ExplodeOnImpactPoint(InAgent : agent, InLaserClass : CustomLaserClass)<suspends> : void=
        loop:
            #Location we've hit
            ImpactLocation := InLaserClass.RaycastHitProp.GetTransform().Translation
            InLaserClass.ExplosiveDevice.Reset()
            if(InLaserClass.ExplosiveDevice.TeleportTo[ImpactLocation, MakeRotation(Axis:=vector3{X:=0.0, Y:=0.0, Z:=0.0}, AngleRadians:=0.0)]):
            Sleep(0.1)
            InLaserClass.ExplosiveDevice.Explode(InAgent)

    Reset() : void=
        set CurrentIndex = 0
        set LaserPlayerMap = map{}