using { /Fortnite.com/Devices }
using { /Verse.org/Simulation }
using { /UnrealEngine.com/Temporary/Diagnostics }

V_LobbyManager := class(creative_device):

    @editable
    LobbyType : int = 0
    @editable
    MascotType : int = 0

    @editable
    EnableOverlay : logic = false
    @editable
    OverlayType : int = 0

    @editable
    InteractButton : button_device = button_device{}
    @editable
    OptAudioPlayer : ?audio_player_device = false

    @editable
    MainLobbyTeleportProp : creative_prop = creative_prop{}
    @editable
    LobbyTeleportAwayProp : creative_prop = creative_prop{}

    @editable
    MascotTeleportProp : creative_prop = creative_prop{}
    @editable
    MascotTeleportAwayProp : creative_prop = creative_prop{}

    @editable
    LobbyParents : []creative_prop = array{}
    @editable
    MascotProps : []creative_prop = array{}
    @editable
    LobbyOverlayParents : []creative_prop = array{}

    OnBegin<override>()<suspends>:void=
        InteractButton.InteractedWithEvent.Subscribe(OnButtonInteract)

        if(not LobbyType = 0):
            #Teleport Main Lobby out
            if(LobbyParents[0].TeleportTo[LobbyTeleportAwayProp.GetTransform()]):
            #Teleport New Lobby in
            if(LobbyParents[LobbyType].TeleportTo[MainLobbyTeleportProp.GetTransform()]):

        if(not MascotType = 0):
            #Teleport Main Lobby out
            if(MascotProps[0].TeleportTo[MascotTeleportAwayProp.GetTransform()]):
            #Teleport New Lobby in
            if(MascotProps[MascotType].TeleportTo[MascotTeleportProp.GetTransform()]):

        if(EnableOverlay = true):
            #Teleport Overlaying Lobby
            if(LobbyOverlayParents[OverlayType].TeleportTo[MainLobbyTeleportProp.GetTransform()]):

    OnButtonInteract(InAgent : agent) : void=
        if(AudioPlayer := OptAudioPlayer?):
            AudioPlayer.Play()