
using { /Fortnite.com/Devices }
using { /Verse.org/Simulation }
using { /UnrealEngine.com/Temporary/Diagnostics }

Player_Data := class<final><persistable>:
    Version : int = 0
    Hat_1 : logic = false
    Hat_2 : logic = false
    Hat_3 : logic = false
    Hat_4 : logic = false
    Hat_5 : logic = false
    Hat_6 : logic = false
    Hat_7 : logic = false
    Hat_8 : logic = false
    
    WeaponLoadout_1 : int = 0
    WeaponLoadout_2 : int = 0
    WeaponLoadout_3 : int = 0
    
    # Depricated
    BrandHatMap : [string][]int = map{}

    ItemsMap : [string][]string = map{}

    # Stores showcase key data from player manager
    EquipedItems : []string = array{}

    TimeDateInSeconds : int = 0
    
    
    
#Creates a map of players and saved stats data. Stores player data in the map.
var PlayerDataMap : weak_map(player, Player_Data) = map{}

#Constructor used to make a new instance of player_stats_data to make updates
MakePlayerUpdate<constructor>(OldData : Player_Data)<transacts> := Player_Data:
    Version := OldData.Version
    Hat_1 := OldData.Hat_1
    Hat_2 := OldData.Hat_2
    Hat_3 := OldData.Hat_3
    Hat_4 := OldData.Hat_4
    Hat_5 := OldData.Hat_5
    Hat_6 := OldData.Hat_6
    Hat_7 := OldData.Hat_7
    Hat_8 := OldData.Hat_8
    
    WeaponLoadout_1 := OldData.WeaponLoadout_1
    WeaponLoadout_2 := OldData.WeaponLoadout_2
    WeaponLoadout_3 := OldData.WeaponLoadout_3

    BrandHatMap := OldData.BrandHatMap

    ItemsMap := OldData.ItemsMap
    EquipedItems := OldData.EquipedItems
    TimeDateInSeconds := OldData.TimeDateInSeconds
    

# A Verse-authored creative device that can be placed in a level
V_PlayerData := class(creative_device):

    # Runs when the device is started in a running game
    OnBegin<override>()<suspends>:void=
        # TODO: Replace this with your code
        Print("Hello, world!")
        Print("2 + 2 = {2 + 2}")