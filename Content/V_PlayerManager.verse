using { /Fortnite.com/Devices }
using { /Verse.org/Simulation }
using { /UnrealEngine.com/Temporary/Diagnostics }
using { /Fortnite.com/Playspaces }
using { /Fortnite.com/Characters }
using { /UnrealEngine.com/Temporary/UI }

UtcHour : int = 21

# True means you can collect
CalculateCurrentRewardStatus(Player : player):logic=
    PlayerTimeProvider := Date_Time_Provider{}
    CurrentTimeProvider := Date_Time_Provider{}
    if(Player.IsActive[]):
        if(OldData := PlayerDataMap[Player]):
            PlayerTimeInSec := OldData.TimeDateInSeconds
            CurrentTimeInSec := GetSecondsSinceEpoch()
            PlayerDate := PlayerTimeProvider.GetDateAndTime(PlayerTimeInSec * 1.0)
            CurrentDate := CurrentTimeProvider.GetDateAndTime()
            # Handle logic for player collecting today before or after UTC hour
            if:
                PlayerDate.year = CurrentDate.year
                PlayerDate.month = CurrentDate.month
                PlayerDate.day = CurrentDate.day
            then:
                if(CurrentDate.hour >= UtcHour):
                    if(PlayerDate.hour >= UtcHour):
                        return false
                    else:
                        return true
                else:
                    return false
            else:
                NumOfHoursSinceLastCollect := (CurrentTimeInSec - PlayerTimeInSec)/SECONDS_PER_HOUR
                if(NumOfHoursSinceLastCollect >= 24.0):
                    return true
                else if(CurrentDate.hour >= UtcHour):
                    return true
                else if(PlayerDate.hour < UtcHour):
                    return true

    return false

UtcResetHoursRemaining():int=
    CurrentTimeProvider := Date_Time_Provider{}
    CurrentTimeInSec := GetSecondsSinceEpoch()
    CurrentDate := CurrentTimeProvider.GetDateAndTime()
    if(CurrentDate.hour > UtcHour):
        return (24 - CurrentDate.hour) + UtcHour
    return UtcHour - CurrentDate.hour


# A Verse-authored creative device that can be placed in a level
V_PlayerManager := class(creative_device):
    @editable
    SpawnPads : []player_spawner_device = array{}

    @editable
    bGameManager : ?game_manager = false

    @editable
    bHatManager : ?V_HatManager = false

    @editable
    bUIShopManager : ?V_ItemShop = false

    @editable
    RealTime : V_Real_Time_data = V_Real_Time_data{}

    @editable
    XPReward : accolades_device = accolades_device{}

    @editable
    AnyHatPurchaseAnalytic : analytics_device = analytics_device{}

    ShopGoldReward : int = 100

    SaveVersion : int = 0

    @editable
    Analytics : []analytics_device = array{}

    @editable
    MenuAnalytics : []analytics_device = array{}
    
    # Map used for triggering Analytics when items are purchased or equipped
    var AnalyticsMap : [string]tuple(?analytics_device, ?analytics_device) = map{}

    # Map used for triggering Analytics evertytime menu tab buttons are clicked
    var AnalyticsMapMenuClick : [string]?analytics_device = map{}
        
    
    # A double hash map that stores what Item players have equipped on each equip index. Equip Index (0 = caps, 1 = guns, 2 = trail)
    # When a player spawns on spawnpad, this variable is updated by checking players equip save data and also updated everytime player equips or unequips an item
    # This var is used by other functions to check what player has currently have equipped under an index
    var PlayerEquipMap : [player][int]tuple(string, string, string, ?type{_(:player, :V_PlayerManager, :string, :string, :string, :?UI_Data, :logic):void}) = map{}

    # This stores Array of all items that belong to a brand. 
    # This is not storing player data, it gives ease of access to check what items belong to which brands
    # Its used to check with player data and remove outdated items/brands that migght still be in players save data
    # On begin this hash map is filled by coppying data from PurchaseMapBinds
    var BrandDataMap : [string][]string = map{}


    # Binding buttons to the purchasable items                    # This ?type is a function variable which only takes functions with parms of type Player, Player Manager, ButtonGroupName, etc.
    # On begin this var is filled with all purchase data          # Its used to send equip and unequip event function call for items 
    var PurchaseMapBinds : [string]tuple(string, string, int, int, ?type{_(Player:player, PlayerManager : V_PlayerManager, ButtonGroupName : string, BrandName : string, ItemName : string, bUIData : ?UI_Data, Equip : logic):void}) = map{}

    var PlayersToGrantXP : []player = array{}

    # Runs when the device is started in a running game
    OnBegin<override>()<suspends>:void=
        # Updating this var automates purchase data, eqip and unequip calls, save data and removing of outdated save data
        # How to fill this hash data?
        # Button's GroupName(Showcase Names from krita) = (BrandName, Item Name, Cost, EquipIndex, DispatchEvent)
        # Equip Index (0 = caps index, 1 = guns index, 2 = trails index)

        # !! Warning Make sure there are no spaces in any of these names, when coppying showcase names from krita, they might have spaces in names but its ok to have spaces in Krita but not in this hash map names

        set PurchaseMapBinds = map:
                                                                
            # Showcase name (Button's GroupName) is used as equip save data, changing ShowCase name when the same brand and item name are currently in play will corrupt equip data (but purchase save data is still safe)
            # Brand name and Item name is used as purchase save data
            # Hats data
            "BoxShowcase" => ("BD", "Box", 1000, 0, option{bHatManager?.OnBrandHatEquip})
            "ChickenShowcase" => ("BD", "Chicken", 1000, 0, option{bHatManager?.OnBrandHatEquip})
            "CrownShowcase" => ("BD", "Crown", 3000, 0, option{bHatManager?.OnBrandHatEquip})
            #"AstronautShowcase" => ("BD", "Astronaut", 0, 0, option{bHatManager?.OnBrandHatEquip})
            
            "SnailShowcase" => ("BD", "Snail", 1000, 0, option{bHatManager?.OnBrandHatEquip})
            "HeadPhonesShowcase" => ("BD", "HeadPhones", 1000, 0, option{bHatManager?.OnBrandHatEquip})
            "CowBoyShowcase" => ("BD", "CowBoy", 1000, 0, option{bHatManager?.OnBrandHatEquip})
            "SuperSaiyanShowcase" => ("BD", "Super", 9999, 0, option{bHatManager?.OnBrandHatEquip})
            "DonutPileShowcase" => ("BD", "Donut", 1000, 0, option{bHatManager?.OnBrandHatEquip})

            # Info about loadouts, they are internally considered purchasables but are free by default
            # If you ever want to make them into purchasables and have them versionable where player buys a new loadout every week, just change the Item name from '1' to '2' and so on
            "GunLoadout1Showcase" => ("Loadout1", "1", 0, 1, option{bGameManager?.OnDefaultWeaponEquip})
            "GunLoadout2Showcase" => ("Loadout2", "1", 0, 1, option{bGameManager?.OnDefaultWeaponEquip})
            "GunLoadout3Showcase" => ("Loadout3", "3", 0, 1, option{bGameManager?.OnDefaultWeaponEquip})

            # This is for updating brand 1 related items, Only change data in brackets, showcase name or add more showcase and items based on krita showcase names
            # Gave generic showcase names so that you only need to update brand, item and cost when updating for various brands
            "Brand1Item1Showcase" => ("HiChew", "Cap", 999, 0, option{bHatManager?.OnBrandHatEquip})
            "Brand1Item2Showcase" => ("HiChew", "Strawbery", 999, 2, option{bHatManager?.OnBrandHatEquip})
            "Brand1Item3Showcase" => ("HiChew", "FruitCap", 999, 0, option{bHatManager?.OnBrandHatEquip})

            # This is for updating brand 2 related items, Only change data in brackets, showcase name or add more showcase and items based on krita showcase names
            # Gave generic showcase names so that you only need to update brand, item and cost when updating for various brands
            "Brand2Item1Showcase" => ("ACMilan", "Choco", 0, 0, option{bHatManager?.OnBrandHatEquip})
            "Brand2Item2Showcase" => ("ACMilan", "Berry", 0, 2, option{bHatManager?.OnBrandHatEquip})
            "Brand2Item3Showcase" => ("ACMilan", "Vanilla", 0, 0, option{bHatManager?.OnBrandHatEquip})
        
        # Used fo Analytics related to purchasable item actions such as puchases, equip, and other actions 
        set AnalyticsMap = map:
            # Showcase Name =    Analytics fo puchase, Analytics fo equipping
            "Brand1Item1Showcase" => (option{Analytics[0]}, option{Analytics[1]})
            "Brand1Item2Showcase" => (option{Analytics[2]}, option{Analytics[3]})
            "Brand1Item3Showcase" => (option{Analytics[4]}, option{Analytics[5]})

            "BoxShowcase" => (option{AnyHatPurchaseAnalytic}, false)
            "ChickenShowcase" => (option{AnyHatPurchaseAnalytic}, false)
            "CrownShowcase" => (option{AnyHatPurchaseAnalytic}, false)
            "AstronautShowcase" => (option{AnyHatPurchaseAnalytic}, false)
            "SantaShowcase" => (option{AnyHatPurchaseAnalytic}, false)
            "SnailShowcase" => (option{AnyHatPurchaseAnalytic}, false)
            "HeadPhonesShowcase" => (option{AnyHatPurchaseAnalytic}, false)
            "CowBoyShowcase" => (option{AnyHatPurchaseAnalytic}, false)
            "SuperSaiyanShowcase" => (option{AnyHatPurchaseAnalytic}, false)
            "DonutPileShowcase" => (option{AnyHatPurchaseAnalytic}, false)

        # Used fo Analytics related to purchasable item actions such as puchases, equip, and other actions 
        set AnalyticsMapMenuClick = map:
            # Button Name =    Analytics for click
            "B_ItemShop" => option{MenuAnalytics[0]}
            "B_HatsMenu" => option{MenuAnalytics[1]}
 
        CopyDataIntoBrandMap()
        Sleep(0.2)
        AllPlayers := GetPlayspace().GetPlayers()
        spawn:
            Tick()
        InitializeAllPlayers(AllPlayers)
        for(SpawnPad : SpawnPads):
            SpawnPad.SpawnedEvent.Subscribe(OnSpawn)

    # Copies data from PurchaseMapBinds to BrandDataMap
    # This is done so its easy to check and validate brands and items within those brrands
    CopyDataIntoBrandMap():void=
        for(PurchaseName -> PurchaseData : PurchaseMapBinds):
            Brand := PurchaseData(0)
            Item := PurchaseData(1)
            if(BrandDataMap[Brand]):
                if(set BrandDataMap[Brand] += array{Item}):
            else:
                if(set BrandDataMap[Brand] = array{Item}):

    GetPlayerGold(Player : player)<transacts>:int=
        if(GameManager := bGameManager?):
            return GameManager.GoldTracker.GetValue(Player)
        return 0

    SetPlayerGold(Player : player, Value : int):void=
        if(GameManager := bGameManager?):
            return GameManager.GoldTracker.SetValue(Player, Value)


    OnSpawn(InAgent : agent):void=
        if(Player := player[InAgent]):
            #ResetPlayerUIOnRespawn(Player)
            spawn:
                SetUpPlayer(InAgent)

    SetUpPlayer(InAgent : agent)<suspends>:void=
        if(Player := player[InAgent], FortCharacter := InAgent.GetFortCharacter[]):
            #Print("PlayerManager - Setting up player")
            Sleep(0.5)
            InitializePlayer(Player)

    InitializeAllPlayers(Players:[]player):void=
        for(Player : Players):
            InitializePlayer(Player)

    InitializePlayer(Player : player):void=
        if(Player.IsActive[]):
            CreateSaveData(Player)
            if(OldData := PlayerDataMap[Player]):
                VersionLogic(Player) # versions and removes old brand data
                ProcessEquipData(Player, OldData)
        #spawn:
            #PrintAllSaveHatData(Player)
    
    PrintAllSaveHatData(Player : player)<suspends>:void=
        Sleep(5.0)
        if(OldData := PlayerDataMap[Player]): 
            HatMap := OldData.ItemsMap
            for(Brand -> HatIndex : HatMap):
                Print("Brand Name {Brand}", ?Duration := 10.0)
                for(Index : HatIndex):
                    Print("Index {Index}", ?Duration := 10.0)

    #Load Player Stats Map
    CreateSaveData(Player : player):logic=
        if:
            not PlayerDataMap[Player]
            set PlayerDataMap[Player] = Player_Data{}
        then:
            return true
        return false

    VersionLogic(Player : player):void=
        if(Player.IsActive[]):
            if(OldData := PlayerDataMap[Player]):
                RemoveSaveBrandsData(Player, OldData)
                if:
                    set PlayerDataMap[Player] = Player_Data:
                        MakePlayerUpdate<constructor>(OldData)
                        Version := SaveVersion

    RemoveSaveBrandsData(Player : player, OldData : Player_Data):void=
        if(Player.IsActive[]):
            var NewMap : [string][]string = OldData.ItemsMap
             for(BrandName -> SaveData : OldData.ItemsMap):
                if (Items := BrandDataMap[BrandName]):
                    var NewSaveItems : []string = array{}
                    for(SavedItem : SaveData, Items.Find[SavedItem]):
                        set NewSaveItems += array{SavedItem}
                    if (set NewMap[BrandName] = NewSaveItems):
                else:
                    set NewMap = RemoveKeyFromMap(NewMap, BrandName)
            if:
                set PlayerDataMap[Player] = Player_Data:
                    MakePlayerUpdate<constructor>(OldData)
                    ItemsMap := NewMap

    ProcessEquipData(Player : player, OldData : Player_Data):void=
        var EquipIndexMap : [int]tuple(string, string, string, ?type{_(:player, :V_PlayerManager, :string, :string, :string, :?UI_Data, :logic):void}) = map{}
        var ValidItemNames : []string = array{}
        ItemsSavedMap := OldData.ItemsMap
        for(SavedItemShowcase : OldData.EquipedItems):
            # Get current latest Item data
            if (ItemData := PurchaseMapBinds[SavedItemShowcase]):
                BrandName := ItemData(0)
                ItemName := ItemData(1)
                EquipIndex := ItemData(3)
                bEquipFunction := ItemData(4)
                # Make sure save data matches up with current Item data that might have been updated
                if (SavedItemsArray := ItemsSavedMap[BrandName], SavedItemsArray.Find[ItemName]):
                    if (set EquipIndexMap[EquipIndex] = (SavedItemShowcase, BrandName, ItemName, bEquipFunction)):
        for(Index -> ValidItem : EquipIndexMap):
            Showcase := ValidItem(0)
            Brand := ValidItem(1)
            Item := ValidItem(2)
            bEquipFunction := ValidItem(3)
            set ValidItemNames += array{Showcase}
            if (EquipFunction := bEquipFunction?):
                # Call equip functions
                EquipFunction(Player, Self, Showcase, Brand, Item, false, true)
        # This players equip data
        if(set PlayerEquipMap[Player] = EquipIndexMap):
        # Save the updated equiped showcase data data
        if:
            set PlayerDataMap[Player] = Player_Data:
                MakePlayerUpdate<constructor>(OldData)
                EquipedItems := ValidItemNames


    #See if the player has persistent data and return the stats
    GetPlayerData(Player : player)<decides><transacts>:Player_Data=
        var TempPlayerData : Player_Data = Player_Data{}
        if(Player.IsActive[]):
            if(PlayerData := PlayerDataMap[Player]):
                set TempPlayerData = PlayerData
        return TempPlayerData

    
    UpdatePurchaseSaveData(Player : player, BrandName : string, HatName : string):void=
        if(Player.IsActive[]):
            var NewItemMap : [string][]string = map{}
            if(OldData := PlayerDataMap[Player], set NewItemMap = OldData.ItemsMap): 
                if (NewItemMap[BrandName]):
                    # Brand exists, just add element index to array
                    if (set NewItemMap[BrandName] += array{HatName}):
                else :
                    # Brand does not exist, add brand as key and make array with this HatIndex as its first element
                    if(set NewItemMap[BrandName] = array{HatName}):
                if:
                    set PlayerDataMap[Player] = Player_Data:
                        MakePlayerUpdate<constructor>(OldData)
                        ItemsMap := NewItemMap

    # Removes an element from the given map and returns a new map without that element
    RemoveKeyFromMap(ExampleMap:[string][]string, ElementToRemove:string):[string][]string=
        var NewMap:[string][]string = map{}
        # Concatenate Keys from ExampleMap into NewMap, excluding ElementToRemove
        for (Key -> Value : ExampleMap, Key <> ElementToRemove):
            set NewMap = ConcatenateMaps(NewMap, map{Key => Value})
        return NewMap

    OnShopUICreated(Player : player, UIData : UI_Data):void=
        UIData.BindAnyButtonsWithName("B_Purchase", OnPurchase)
        UIData.BindButtons("TimedReward", "B_Collect", false, OnRewardCollect)

    RemoveEquipSaveData(Player : player, GroupName : string):void=
        if(OldData := PlayerDataMap[Player]):
            NewEquipArray := OldData.EquipedItems.RemoveAllElements(GroupName)
            if:
                set PlayerDataMap[Player] = Player_Data:
                    MakePlayerUpdate<constructor>(OldData)
                    EquipedItems := NewEquipArray

    AddEquipSaveData(Player : player, GroupName : string):void=
        if(OldData := PlayerDataMap[Player]):
            if(OldData.EquipedItems.Find[GroupName]):
                return
            NewEquipArray := OldData.EquipedItems + array{GroupName}
            if:
                set PlayerDataMap[Player] = Player_Data:
                    MakePlayerUpdate<constructor>(OldData)
                    EquipedItems := NewEquipArray

    Tick()<suspends>:void=
        loop:
            Sleep(1.1)
            for(Player : GetPlayspace().GetPlayers()):
                if(Player.IsActive[]):
                    if(UIShopManager := bUIShopManager?):
                        UIShopManager.UpdateCollectable(Player)
    

                            
    OnRewardCollect(GroupName : string, ButtonName : string, Widget : widget_message, bUIData : ?UI_Data):void=
        Player := Widget.Player    
        if(Player.IsActive[]):
            FTimeInSeconds := GetSecondsSinceEpoch()
            if(OldData := PlayerDataMap[Player], UIShopManager := bUIShopManager?, TimeInSeconds := Floor[FTimeInSeconds]):
                if:
                    set PlayerDataMap[Player] = Player_Data:
                        MakePlayerUpdate<constructor>(OldData)
                        TimeDateInSeconds := TimeInSeconds
                then:
                    set PlayersToGrantXP += array{Player}
                    UIShopManager.UpdateCollectable(Player)
                    if(GameManager := bGameManager?):
                        GameManager.GoldTracker.Add(Player, ShopGoldReward)
                    Print("Collected")

    GrantXPDelayedOnShopClose(Player : player)<suspends>:void=
        Sleep(0.5)
        if (PlayersToGrantXP.Find[Player]):
            set PlayersToGrantXP = PlayersToGrantXP.RemoveAllElements(Player)
            XPReward.Award(Player)


    OnPurchase(GroupName : string, ButtonName : string, Widget : widget_message, bUIData : ?UI_Data):void=
        Player := Widget.Player
        PlayerGold := GetPlayerGold(Player)
        # Find Brand hat data by buttons group name
        if (PurchaseData := PurchaseMapBinds[GroupName], PlayerData := GetPlayerData[Player], UIShopManager := bUIShopManager?, Player.IsActive[]):
            Cost := PurchaseData(2)
            bDispatchEvent := PurchaseData(4)
            EquipIndex := PurchaseData(3)
            ItemName := PurchaseData(1)
            BrandName := PurchaseData(0)
            # Check if player has brand name already as key       # Check if player already has name of hat in the array of names
            if (Brand := PlayerData.ItemsMap[BrandName], Brand.Find[ItemName]):
                # Unequip Previous Item and Equip new Item
                if (PreviousEquipedData := PlayerEquipMap[Player][EquipIndex]):
                    PreviousGroup := PreviousEquipedData(0)
                    if(PreviousEquipFun := PreviousEquipedData(3)?):
                        # Call Unequip on previous function
                        PreviousEquipFun(Player, Self, PreviousGroup, PreviousEquipedData(1), PreviousEquipedData(2), bUIData, false)
                    # UI show unequip
                    UIShopManager.EquippedItem(Player, PreviousGroup, ButtonName, bUIData, false)
                    RemoveEquipSaveData(Player, PreviousGroup)
                    if (set PlayerEquipMap[Player][EquipIndex] = ("", "", "", false)):
                    if (PreviousGroup = GroupName):
                        return
                AddEquipSaveData(Player, GroupName)
                # Send signal to equip analytics
                if(AnalyticsDevices := AnalyticsMap[GroupName], PuchaseAnalyticsDevice := AnalyticsDevices(1)?):
                    PuchaseAnalyticsDevice.Submit(Player)
                    Print("Equip signal fo {GroupName}")
                UIShopManager.EquippedItem(Player, GroupName, ButtonName, bUIData, true)
                if (set PlayerEquipMap[Player][EquipIndex] = (GroupName, BrandName, ItemName, bDispatchEvent)):
                if(DispatchEvent := bDispatchEvent?):
                    DispatchEvent(Player, Self, GroupName, BrandName, ItemName, bUIData, true)

            # Player doesnt have hat index or the brand it self, check if he has enough gold to buy hat
            else if(Cost <= PlayerGold):
                # Save the newly bought hat
                UpdatePurchaseSaveData(Player, BrandName, ItemName)
                # Send signal to puchase analytics
                if(AnalyticsDevices := AnalyticsMap[GroupName], PuchaseAnalyticsDevice := AnalyticsDevices(0)?):
                    PuchaseAnalyticsDevice.Submit(Player)
                    Print("Puchase signal fo {GroupName}")
                # Unequip Previous Item and Equip new Item
                if (PreviousEquipedData := PlayerEquipMap[Player][EquipIndex]):
                    PreviousGroup := PreviousEquipedData(0)
                    if(PreviousEquipFun := PreviousEquipedData(3)?):
                        # Call Unequip on previous function
                        PreviousEquipFun(Player, Self, PreviousGroup, PreviousEquipedData(1), PreviousEquipedData(2), bUIData, false)
                    # UI show unequip
                    UIShopManager.EquippedItem(Player, PreviousGroup, ButtonName, bUIData, false)
                    RemoveEquipSaveData(Player, PreviousGroup)
                    if (set PlayerEquipMap[Player][EquipIndex] = ("", "", "", false)):
                AddEquipSaveData(Player, GroupName)
                # Send signal to equip analytics
                if(AnalyticsDevices := AnalyticsMap[GroupName], PuchaseAnalyticsDevice := AnalyticsDevices(1)?):
                    PuchaseAnalyticsDevice.Submit(Player)
                    Print("Equip signal fo {GroupName}")
                if (set PlayerEquipMap[Player][EquipIndex] = (GroupName, BrandName, ItemName, bDispatchEvent)):
                if(DispatchEvent := bDispatchEvent?):
                    DispatchEvent(Player, Self, GroupName, BrandName, ItemName, bUIData, true)
                #EquipHat(Player, option{HatVFXArray[HatIndex]})
                SetPlayerGold(Player, PlayerGold - Cost)
                UIShopManager.UpdateUIPurchaseGroup(GroupName, bUIData, Player)