using { /Fortnite.com/Devices }
using { /Verse.org/Simulation }
using { /UnrealEngine.com/Temporary/Diagnostics }

#Custom Class that defines each scrap book sticker/part/icon
CustomMinigameIcon := class<unique><concrete>():
    @editable
    Name : string = ""
    @editable
    DebugTrigger : ?trigger_device = false
    @editable
    LevelSwitches : []switch_device = array{}

    #Subscribes to debug trigger and sets sequencer
    SetUp() : void=
        #Print("V_ScrapBook - Setup")
        if(Trigger := DebugTrigger?):
            Trigger.TriggeredEvent.Subscribe(OnDebugTriggerPressed)

    #Increases Level for each win/trigger pressed
    OnDebugTriggerPressed(InAgent : ?agent) : void=
        #Print("V_ScrapBook - Trigger Pressed")
        if(Agent := InAgent?):
            IncreaseLevel(Agent)

    #Increases Level for the player who won
    IncreaseLevel(InAgent : agent) : void=
        #Print("V_ScrapBook - Increase Level Called")
        for(Switch : LevelSwitches):
            if(Switch.GetCurrentState[InAgent]):
                #Code for multiple levels
                #Switch.TurnOff(InAgent)
                #Print("V_ScrapBook - Disable Current Switch")
            else:
                #Print("V_ScrapBook - Found Next Switch")
                Switch.TurnOn(InAgent)
                return

V_ScrapBook := class(creative_device):
    @editable
    MinigameScrapbookList : []CustomMinigameIcon = array{}

    #Sets up icons for players
    OnBegin<override>()<suspends>:void=
        for(Group : MinigameScrapbookList):
            Group.SetUp()

    #Checks which group should be updated and for which player
    IncreasePlayerLevel(InAgent : agent, InString : string) : void=
        for(Group : MinigameScrapbookList):
            #Print("V_ScrapBook - Checking Group")
            if(Group.Name = InString):
                #Print("V_ScrapBook - Is Correct Group")
                Group.IncreaseLevel(InAgent)
                return