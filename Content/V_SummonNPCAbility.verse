using { /Fortnite.com/Devices }
using { /Verse.org/Simulation }
using { /UnrealEngine.com/Temporary/Diagnostics }

#Added
using { /UnrealEngine.com/Temporary/SpatialMath }
using { /Fortnite.com/Characters }
using { /Fortnite.com/AI }
using { /Verse.org/Random }

# A Verse-authored creative device that can be placed in a level
V_SummonNPCAbility := class(crashable_device):

    #This will be the NPC the player can spawn
    @editable
    NPCToSpawn : npc_spawner_device = npc_spawner_device{}
    @editable
    TriggerBehaviour : trigger_device = trigger_device{}

    var AgentSpawnedLocation : vector3 = vector3{X:=0.0,Y:=0.0,Z:=0.0}

    var SpawningPlayerCharacter : []fort_character = array{}

    # Runs when the device is started in a running game
    OnCrashableDeviceBegin<override>()<suspends>:void=
        TriggerBehaviour.TriggeredEvent.Subscribe(SpawnNPC)

    #The InAgent is the player spawning the NPC
    SpawnNPC(InAgent : ?agent):void=
        #Note: I've set this up with a trigger, this way we could use this in other minigames if required instead of just tying it into the
        #super hero minigame
        if(bInAgent := InAgent?):
            if(FortChar := bInAgent.GetFortCharacter[]):
                #This array should always only have 1 entry
                set SpawningPlayerCharacter += array{FortChar}
                set AgentSpawnedLocation = FortChar.GetTransform().Translation
        #Spawn the NPC
        NPCToSpawn.Spawn()
        Print("Spawn NPCCCCC")
        #Get the spawned NPC
        NPCToSpawn.SpawnedEvent.Subscribe(NPCSpawned)

    #Handles teleporting the NPC to the player
    #The InAgent here is the Fortnite NPC we spawn
    NPCSpawned(InAgent : agent):void=
        #Here is where we can teleport the NPC to the player who summoned them
        if(FortChar := InAgent.GetFortCharacter[]):
            #Teleport the NPC to the player who spawned them
            if(FortChar.TeleportTo[AgentSpawnedLocation, FortChar.GetTransform().Rotation]):
                Print("Player Location {AgentSpawnedLocation}")
                Print("Teleport NPC")
        if(Character := InAgent.GetFortCharacter[]):
            if(FortCharacterWhoSpawnedNPC := SpawningPlayerCharacter[0]):
                spawn:
                    FocusOnTargetLoop(Character, FortCharacterWhoSpawnedNPC)
                #Reset the array
                set SpawningPlayerCharacter = array{}

    #I need to pass the player in here somehow
    FocusOnTargetLoop(InCharacter : fort_character, SpawningPlayer : fort_character)<suspends>:void=
        Sleep(1.0)
        NPCSpawnPoint := InCharacter.GetTransform().Translation
        # Get the Focus Interface, this allows you to tell it to look at something or somewhere.
        if(Focus := InCharacter.GetFocusInterface[]):
            sync:
                block:
                    loop:
                        #Currently getting all players then assigning a random one, need to change for registered players
                        var AllPlayers : []player = GetPlayspace().GetPlayers()
                        #Remove Element from the array
                        if(ElementToRemove := player[SpawningPlayer.GetAgent[]]):
                            for(i -> Player : AllPlayers):
                                Print("Looping through players")
                                if(Player = ElementToRemove):
                                    Print("Compared and found the casting player")
                                    if (NewArray := AllPlayers.RemoveElement[i]):
                                        set AllPlayers = NewArray
                                        Print("Removed Ultimate Caster From Array")
                        #Currently getting random players every 5 seconds and firing at them
                        PlayerIndex := GetRandomInt(0, AllPlayers.Length -1)
                        #Remove player from array
                        Print("Spawned Player Index = {PlayerIndex}")
                        #We get the player who is spawning the NPC
                        if(RandomTarget := AllPlayers[PlayerIndex]):
                            spawn:
                                Focus.MaintainFocus(RandomTarget)
                        Sleep(5.0)
                block:
                    #After a set time when the power is complete the AI gets eliminated
                    Sleep(30.0)
                    #Eliminate the NPC
                    InCharacter.Damage(9999.0)
                    return