
using { /Fortnite.com/Devices }
using { /Verse.org/Simulation }
using { /UnrealEngine.com/Temporary/Diagnostics }
using { /UnrealEngine.com/Temporary/UI }
using { /Fortnite.com/UI }
using { /UnrealEngine.com/Temporary/SpatialMath}
using { /Verse.org/Colors }
using { /Verse.org/Assets }

# See https://dev.epicgames.com/documentation/en-us/uefn/create-your-own-device-in-verse for how to create a verse device.
#_StringToMessage<localizes>(String : string) : message = "{String}"
#_EmptyMessage<localizes>() : message = ""
#StringToMessage(String : string) : message = if(String.Length > 0) then _StringToMessage(String) else _EmptyMessage()

(Name : string).RemoveSufix(Sufix : string)<transacts><decides>:string =
    SufixLength := Sufix.Length
    NameLength := Name.Length
    StartIndex := NameLength - SufixLength
    EndName := Name.Slice[StartIndex] 
    EndName = Sufix
    NewName := Name.Remove[StartIndex, NameLength]
    return NewName

(Name : string).RemovePrefix(Prefix : string)<transacts><decides>:string =
    PrefixLength := Prefix.Length
    StartingName := Name.Slice[0, PrefixLength] 
    StartingName = Prefix
    NewName := Name.Remove[0, PrefixLength]
    return NewName


(Name : string).GetSufix(Sufix : string)<transacts><decides>:string =
    SufixLength := Sufix.Length
    NameLength := Name.Length
    StartIndex := NameLength - SufixLength
    EndName := Name.Slice[StartIndex] 
    EndName = Sufix
    return EndName


E_ButtonType := enum:
    Quiet
    Loud
    Regular

E_Widget_Visibility_Status := enum:
    Visible
    Hidden

E_Show_Visibility_Option := enum:
    Show
    Hide
    HidePermanent
    ShowUnlock


NormalizeFloat(Value : float, MinValue : float, MaxVal : float):float= 
    if (MaxVal <= MinValue):
        return 0.0
    return (Value - MinValue) / (MaxVal - MinValue)

Event_Handle := class<unique>():
    #Bindable : ?Base_Bindable_ID = false
    #BindableContainer : ?Bindables_Container = false
    var HandlesContainer : Bindables_Container
    var HandleCancled : logic = false

    Dispatch():void=
        block:

    Cancel():void=
        HandlesContainer.RemoveEvent(Self)
        set HandleCancled = true

    Init()<transacts>:void=
        HandlesContainer.AddEvent(Self)

# Delegate function types
button_fun := type{_(GroupName : string, ButtonName :string, Widget : widget_message, bUIData : ?UI_Data):void}
Button_Event_Handle := class<unique>(Event_Handle):
    Function : button_fun

    Dispatch(GroupName : string, Name : string, Widget : widget_message, bUIData : ?UI_Data):void=
        if (HandleCancled = false):
            Function(GroupName, Name, Widget, bUIData)

# Container that stores map of bindables
Bindables_Container := class():
    var DynamicBindableMap : [comparable][]string = map{}
    var EventHandles : []Event_Handle = array{}

    AddEvent(EventHandle : Event_Handle)<transacts>:void=
        set EventHandles += array{EventHandle}

    RemoveEvent(EventHandle : Event_Handle):void=
        set EventHandles = EventHandles.RemoveAllElements(EventHandle)

    RemoveAllEvents():void=
        for(Handle : EventHandles):
            set Handle.HandleCancled = true
        set EventHandles = array{}



Parent_Widget_Data := class():
    Name : string = ""
    GroupName : string = ""
    GlobalIndex : int = -1
    var X : float = 0.0
    var Y : float = 0.0
    ScaleX : int = 0
    ScaleY : int = 0
    AnchorMinX : float = 0.0
    AnchorMinY : float = 0.0
    AnchorMaxX : float = 0.0
    AnchorMaxY : float = 0.0
    var Alignment : vector2 = vector2{X := 0.0, Y := 0.0}
    var SizeToContent : logic = false
    var Visibility : E_Widget_Visibility_Status = E_Widget_Visibility_Status.Visible
    var HiddenPermanent : logic = false
    var Order : type{_X:int where 0 <= _X, _X <= 2147483647} = 0
    var bCanvasSlot : ?canvas_slot = false
    var BindsContainer : Bindables_Container = Bindables_Container{}
    var EventHandles : []Event_Handle = array{}
    var bUIData : ?UI_Data = false
    _HideWidget(?Permanent : logic = false):void=
        set Visibility = E_Widget_Visibility_Status.Hidden
        if (Permanent = true):
            PermanentlyHideWidget(true)

    _ShowWidget(?UnlockPermanent : logic = false):void=
        if (UnlockPermanent = true):
            PermanentlyHideWidget(false)
            set Visibility = E_Widget_Visibility_Status.Visible
        else if(HiddenPermanent = false):
            set Visibility = E_Widget_Visibility_Status.Visible


    PermanentlyHideWidget(Permanent : logic):void=
        set HiddenPermanent = Permanent
    
    UnlockWidget():void=
        set HiddenPermanent = false

    LockWidget():void=
        set HiddenPermanent = true

    _ExitUI(UIData : UI_Data):void=
        BindsContainer.RemoveAllEvents()

Text_Block := class(Parent_Widget_Data):
    var TextBlock : text_block = text_block{}
    var Text : string = ""
    Color : color = color{}
    Opacity : type{_X:float where 0.0 <= _X, _X <= 1.0} = 1.0

    _HideWidget<override>(?Permanent : logic = false):void=
        (super :)_HideWidget(?Permanent := Permanent)
        #if (Visibility = E_Widget_Visibility_Status.Visible):
            TextBlock.SetVisibility(widget_visibility.Hidden)
    _ShowWidget<override>(?UnlockPermanent : logic = false):void=
        (super :)_ShowWidget(?UnlockPermanent := UnlockPermanent)
        if (HiddenPermanent = false):# , Visibility = E_Widget_Visibility_Status.Hidden
            TextBlock.SetVisibility(widget_visibility.Visible)
    SetText(InText : string):void=
        set Text = InText
        TextBlock.SetText(StringToMessage(InText))

Color_Block := class(Parent_Widget_Data):
    Color : color_block = color_block{}

    _HideWidget<override>(?Permanent : logic = false):void=
        (super :)_HideWidget(?Permanent := Permanent)
        #if (Visibility = E_Widget_Visibility_Status.Visible):
            Color.SetVisibility(widget_visibility.Hidden)

    _ShowWidget<override>(?UnlockPermanent : logic = false):void=
        (super :)_ShowWidget(?UnlockPermanent := UnlockPermanent)
        if (HiddenPermanent = false):# , Visibility = E_Widget_Visibility_Status.Hidden
            Color.SetVisibility(widget_visibility.Visible)

Image_Block := class(Parent_Widget_Data):
    var Texture : ?texture = false
    var ImageTextureBlock : ?texture_block = false

    _HideWidget<override>(?Permanent : logic = false):void=
        (super :)_HideWidget(?Permanent := Permanent)
        #if (Visibility = E_Widget_Visibility_Status.Visible):
            if (Image := ImageTextureBlock?):
                Image.SetVisibility(widget_visibility.Hidden)

    _ShowWidget<override>(?UnlockPermanent : logic = false):void=
        (super :)_ShowWidget(?UnlockPermanent := UnlockPermanent)
        if (HiddenPermanent = false):#, Visibility = E_Widget_Visibility_Status.Hidden
            if (Image := ImageTextureBlock?):
                Image.SetVisibility(widget_visibility.Visible)

    SetTexture(NewTexture : ?texture):void=
        set Texture = NewTexture

    SetImage(NewTexture : ?texture, bSize : ?vector2):void=
        if (ValidImageTextureBlock := ImageTextureBlock?):
            SetTexture(NewTexture)
            if(ValidTexture := NewTexture?):
                ValidImageTextureBlock.SetImage(ValidTexture)
            # else: dont know how to remove texture
            if(Size := bSize?):
                ValidImageTextureBlock.SetDesiredSize(Size)

Button_Data := class(Parent_Widget_Data):
    #PurchaseButton : button_quiet = button_quiet{}
    var Text : string = ""
    Color : ?color_block = false
    Texture : ?texture = false

    SetButtonText():void={}
        #PurchaseButton.SetText(StringToMessage(Text))

    BindButtonClick():void={}
        #PurchaseButton.OnClick().Subscribe(OnClick)

    OnClick(Message : widget_message):void={}
        <#if (ClickText = ""):
            return
        PurchaseButton.SetText(StringToMessage(ClickText))#>

(Buttons :[]Button_Quiet_Data).BindButtons(Function : button_fun)<transacts>:[]Button_Event_Handle=
    var ButtonHandles : []Button_Event_Handle = array{}
    for(Button : Buttons):
        set ButtonHandles += array{Button.BindButton(Function)}
    return ButtonHandles
        

Button_Quiet_Data := class(Button_Data):
    var Button : text_button_base = button_quiet{}
    ButtonType : E_ButtonType = E_ButtonType.Quiet
    var ButtonEventContainer : Bindables_Container = Bindables_Container{}

    MasterButtonBind_Internal():void=
        Button.OnClick().Subscribe(ButtonClicked_Internal)
        

    ButtonClicked_Internal(Widget : widget_message):void=
        Dispatch(GroupName, Name, Widget)
    
    # Called by user, to bind function to this function
    BindButton(Function : button_fun)<transacts>:Button_Event_Handle=
        Event := Button_Event_Handle:
            #BindableContainer := option{BindsContainer}
            HandlesContainer := ButtonEventContainer
            Function := Function
        Event.Init()
        #Print("Length offff {ButtonEventContainer.EventHandles.Length}", ?Duration := 12.0)
        return Event
    
    # Dispatches all player bound functions
    Dispatch(ButtonGroupName : string, ButtonName : string, Widget : widget_message):void=
        for(EventHandle : ButtonEventContainer.EventHandles, ButtonHandle := Button_Event_Handle[EventHandle]):
            ButtonHandle.Dispatch(ButtonGroupName, ButtonName, Widget, bUIData)



    _HideWidget<override>(?Permanent : logic = false):void=
        (super :)_HideWidget(?Permanent := Permanent)
        #Button.SetEnabled(false)
        #if (Visibility = E_Widget_Visibility_Status.Visible):
            Button.SetVisibility(widget_visibility.Hidden)
        

    _ShowWidget<override>(?UnlockPermanent : logic = false):void=
        (super :)_ShowWidget(?UnlockPermanent := UnlockPermanent)
        if (HiddenPermanent = false):#, Visibility = E_Widget_Visibility_Status.Hidden
            #Button.SetEnabled(true)
            Button.SetVisibility(widget_visibility.Visible)

    SetText(InText : string):void=
        set Text = InText
        Button.SetText(StringToMessage(InText))

    _ExitUI<override>(UIData : UI_Data):void=
        (super :)_ExitUI(UIData)
        ButtonEventContainer.RemoveAllEvents()


Group_UI := class(Parent_Widget_Data):
    #Name : string = ""
    var ButtonSlots : []canvas_slot = array{}
    var AddedButtons : logic = true
    var ParentGroupName : string = ""
    var ChildGroups : []string = array{}
    var OrderedWidgets : []Parent_Widget_Data = array{}
    var ButtonsQuiet : [string][]Button_Quiet_Data = map{}
    var ColorBlocks : [string][]Color_Block = map{}
    var ImageBlocks : [string][]Image_Block = map{}
    var TextBlocks : [string][]Text_Block = map{}

UI_Bind_Data := class():
    Groups : []string = array{}
    ShowOnly : logic = false
    HideOnly : logic = false
    ApplyToAllDecendents : logic = true
    OnButtonsPressed : []tuple(string, string) = array{}
    OnTriggered : ?trigger_device = false
    TriggeredOnButtonPressed : []trigger_device = array{}
    IsExit : logic = false
    var bUIData : ?UI_Data = false
    var bPlayer : ?player = false

    Init(UIData : UI_Data, Player : player):void=
        set bUIData = option{UIData}
        set bPlayer = option{Player}

    OnButtonPressed(Name : string, GroupName : string, Message : widget_message):void=
        for(Trigger : TriggeredOnButtonPressed):
            Trigger.Trigger(Message.Player)
        if (UIData := bUIData?):
            if (ShowOnly = true):
                UIData.ShowOnlyGroups(Groups, ApplyToAllDecendents)
        if (IsExit = true):
            if (UIData := bUIData?):
                UIData.ExitUI()
                

UI_Binds := class():
	var BindData : []UI_Bind_Data = array{}
	var bUIData : ?UI_Data = false
    var BindDataButtonMap : [string][]UI_Bind_Data = map{}
    var bPlayer : ?player = false

    Init(UIData : UI_Data, Player : player):void=
        set bUIData = option{UIData}
        set bPlayer = option{Player}
        for(Data:BindData):
            Data.Init(UIData, Player)
            var NoDupeButtonNames : []string = array{}
            for(TupleButtonName : Data.OnButtonsPressed):
                if(NoDupeButtonNames.Find[TupleButtonName(0)+TupleButtonName(1)]):
                else:
                    set NoDupeButtonNames += array{TupleButtonName(0)+TupleButtonName(1)}
                    # Map has elements in this key?
                    if(BindDataButtonMap[TupleButtonName(0)+TupleButtonName(1)]):
                        if(set BindDataButtonMap[TupleButtonName(0)+TupleButtonName(1)] += array{Data}):
                    else:
                        if(set BindDataButtonMap[TupleButtonName(0)+TupleButtonName(1)] = array{Data}):

    OnButtonPressed(Name : string, GroupName : string, Message : widget_message):void=
        if (DataBinds := BindDataButtonMap[GroupName+Name]):
            for(Data:DataBinds):
                Data.OnButtonPressed(Name, GroupName, Message)

Base_Bind_Feature := class():
    GetButtonName():tuple(string,string)=
        return ("", "")
    Init(UIData : UI_Data, Player : player):void=
        block:

(BaseBindFeatures : []Base_Bind_Feature).Init(UIData : UI_Data, Player : player):void=
    for(BaseBindFeature:BaseBindFeatures):
        BaseBindFeature.Init(UIData, Player)

Button_To_Trigger := class<concrete>(Base_Bind_Feature):
    var ButtonName : tuple(string,string) = ("", "")
    var bUIData : ?UI_Data = false
    @editable
    var Triggers : []trigger_device = array{}
    var bPlayer : ?player = false
    

    Init<override>(UIData : UI_Data, Player : player):void=
        set bUIData = option{UIData}
        set bPlayer = option{Player}
        if (UIDATA := bUIData?):
            # Bind to buttons under button enum that was selected
            set ButtonName = GetButtonName()
            if(Buttons := UIDATA.GetButtons[ButtonName(0), ButtonName(1)]):
                for(ButtonData:Buttons):
                    ButtonData.Button.OnClick().Subscribe(OnButtonClick)

    OnButtonClick(Message : widget_message):void=
        if (Player := bPlayer?):
            for(Trigger : Triggers):
                Trigger.Trigger(Player)


Show_Groups := class<concrete>(Base_Bind_Feature):
    var ButtonName : tuple(string,string) = ("", "")
    @editable
    var OnTrigger : ?trigger_device = false
    var ShowGroupSettings : []Show_Group_Settings = array{}
    var bUIData : ?UI_Data = false
    var bPlayer : ?player = false

    Init<override>(UIData : UI_Data, Player : player):void=
        set bUIData = option{UIData}
        set bPlayer = option{Player}
        Settings := GetGroupSettings()
        if (Settings.Length > 0):
            set ShowGroupSettings = Settings
        if (UIDATA := bUIData?):
            # Bind to buttons under button enum that was selected
            if (ButtonName = ("", "")):
                set ButtonName = GetButtonName()
            #Print("BUTTON NAME IS {ButtonName(0)} {ButtonName(1)}", ?Duration := 10.0)
            if(Buttons := UIDATA.GetButtons[ButtonName(0), ButtonName(1)]):
                for(ButtonData:Buttons):
                    ButtonData.Button.OnClick().Subscribe(OnButtonClick)

    OnButtonClick(Message : widget_message):void=
        if (Player := bPlayer?):
            var ShowOnlyGroupNames : []string = array{}
            for(ShowGroupSetting:ShowGroupSettings):
                var GroupName : string = ShowGroupSetting.GetGroupName()
                if (GroupName = ""):
                    set GroupName = ShowGroupSetting.GroupName
                bAllDescendants := ShowGroupSetting.AllDescendants
                if (UIData := bUIData?):
                    # Show only setting is specially handled
                    if (ShowGroupSetting.ShowOnly = true):
                        set ShowOnlyGroupNames += array{GroupName}
                        if (ShowGroupSetting.AllDescendants = true):
                            set ShowOnlyGroupNames += UIData.GetAllDecendantGroupNames(GroupName)
                    else:
                        case(ShowGroupSetting.Visibility):
                            # Show Group
                            E_Show_Visibility_Option.Show =>
                                UIData.ShowGroup(GroupName, bAllDescendants)

                            # Hide Group
                            E_Show_Visibility_Option.Hide =>
                                UIData.HideGroup(GroupName, bAllDescendants)
                            # Hide Group Permanently
                            E_Show_Visibility_Option.HidePermanent =>
                                UIData.HideGroup(GroupName, bAllDescendants, ?Permanent := true)

                            # Unlock Group if it has been permanently hidden then Show Group
                            E_Show_Visibility_Option.ShowUnlock =>
                                UIData.ShowGroup(GroupName, bAllDescendants, ?UnlockPermanent := true)

            if (UIData := bUIData?, ShowOnlyGroupNames.Length > 0):
                UIData.ShowOnlyGroups(ShowOnlyGroupNames, false)

    GetGroupSettings():[]Show_Group_Settings=
        return array{}

Show_Group_Settings := class<concrete>():
    var GroupName : string = ""
    @editable
    Visibility : E_Show_Visibility_Option = E_Show_Visibility_Option.Show
    @editable
    ShowOnly : logic = false
    @editable
    AllDescendants : logic = true

    GetGroupName():string=
        return ""

Exit_Button := class<concrete>(Base_Bind_Feature):
    var ButtonName : tuple(string,string) = ("", "")
    @editable
    var OnTrigger : ?trigger_device = false
    var bUIData : ?UI_Data = false
    var bPlayer : ?player = false

    Init<override>(UIData : UI_Data, Player : player):void=
        set bUIData = option{UIData}
        set bPlayer = option{Player}
        #Print("InitExit")
        if (UIDATA := bUIData?):
            # Bind to buttons under button enum that was selected
            set ButtonName = GetButtonName()
            
            if(Buttons := UIDATA.GetButtons[ButtonName(0), ButtonName(1)]):
                for(ButtonData:Buttons):
                    ButtonData.Button.OnClick().Subscribe(OnButtonClick)
                    

    OnButtonClick(Message : widget_message):void=
        if (Player := bPlayer?):
        if (UIData := bUIData?):
            UIData.ExitUI()
    

Hide_Groups := class<concrete>(Base_Bind_Feature):
    var BindData : []UI_Bind_Data = array{}
    var bUIData : ?UI_Data = false
    var BindDataButtonMap : [string][]UI_Bind_Data = map{}
    var bPlayer : ?player = false

    Init<override>(UIData : UI_Data, Player : player):void=
        set bUIData = option{UIData}
        set bPlayer = option{Player}


#HideAllWidgets()
Button_Bind := class():
    ButtonName : string = ""
    GroupName : string = ""
    bUIData : ?UI_Data = false

    ButtonPressed(Message : widget_message):void=
        if(UIData := bUIData?):
            UIData.OnButtonPressed(ButtonName, GroupName, Message)

UI_Data := class():
    #GroupUI : []Group_UI = array{}
    var GroupUIMap : [string]Group_UI = map{}
    var OrderedWidgetData : []Parent_Widget_Data = array{}
    var Canvas : canvas = canvas{}

    var bPlayer : ?player = false

    var ButtonBinds : []Button_Bind = array{}

    var bUIManager : ?V_Krita = false

    # These below are injected from krita
    var ButtonActivateTriggers : []Button_To_Trigger = array{}
    var ShowGroups : []Show_Groups = array{}
    var ExitButtons : []Exit_Button = array{}
    var DefaultShowGroup : string = ""

    var UIExited : logic = false

    ### ALL THE SHOW AND HIDE WIDGET/GROUP FUNCTIONS 

    ShowOnlyGroups(GroupNames : []string, AllDecendants : logic, ?UnlockPermanent : logic = false):void=
        var GroupsToShowNames : []string = GroupNames
        if (AllDecendants = true):
            for(GroupName:GroupNames):
                set GroupsToShowNames += GetAllDecendantGroupNames(GroupName)
        for(Name -> Group : GroupUIMap):
            if(GroupsToShowNames.Find[Name]):
                ShowGroup_Internal(Name, ?UnlockPermanent := UnlockPermanent)
            else:
                HideGroup_Internal(Name)

    HideGroup(GroupName : string, AllDecendants : logic, ?Permanent : logic = false):void=
        HideGroup_Internal(GroupName, ?Permanent := Permanent)
        if (AllDecendants = true):
            # This function only gets children names
            GroupNames := GetAllDecendantGroupNames(GroupName)
            for(Name : GroupNames):
                HideGroup_Internal(Name, ?Permanent := Permanent)

        
    HideGroup_Internal<private>(GroupName : string, ?Permanent : logic = false):void=
        if (Group := GroupUIMap[GroupName]):
            Group._HideWidget(?Permanent := Permanent)
            Widgets := Group.OrderedWidgets
            for(Widget : Widgets):
                if (ButtonWidget := Button_Quiet_Data[Widget]):
                    if(Group.AddedButtons = true):
                        Widget._HideWidget(?Permanent := Permanent)
                else:
                    Widget._HideWidget(?Permanent := Permanent)

    HideWidget(GroupName : string, WidgetName : string, ?Permanent : logic = false):void=
        HideWidget_Internal(GroupName, WidgetName, ?Permanent := Permanent)
    HideWidget(EnumName : tuple(string, string), ?Permanent : logic = false):void=
        HideWidget_Internal(EnumName(0), EnumName(1), ?Permanent := Permanent)


    HideWidget_Internal<private>(GroupName : string, WidgetName : string, ?Permanent : logic = false):void=
        if (Group := GroupUIMap[GroupName]):
            Widgets := Group.OrderedWidgets
            for(Widget : Widgets, Widget.Name = WidgetName):
                if (ButtonWidget := Button_Quiet_Data[Widget]):
                    if(Group.AddedButtons = true):
                        Widget._HideWidget(?Permanent := Permanent)
                    else:
                        Print("Cant Hide buttons")
                else:
                    Widget._HideWidget(?Permanent := Permanent)

    ShowWidget(GroupName : string, WidgetName : string, ?UnlockPermanent : logic = false):void=
        ShowWidget_Internal(GroupName, WidgetName, ?UnlockPermanent := UnlockPermanent)

    ShowWidget(EnumName : tuple(string, string), ?UnlockPermanent : logic = false):void=
        ShowWidget_Internal(EnumName(0), EnumName(1), ?UnlockPermanent := UnlockPermanent)

    ShowWidget_Internal<private>(GroupName : string, WidgetName : string, ?UnlockPermanent : logic = false):void=
        if (Group := GroupUIMap[GroupName]):
            Widgets := Group.OrderedWidgets
            for(Widget : Widgets, Widget.Name = WidgetName):
                if (ButtonWidget := Button_Quiet_Data[Widget]):
                    if(Group.AddedButtons = true):
                        Widget._ShowWidget(?UnlockPermanent := UnlockPermanent)
                    else:
                        Print("Cant Hide buttons")
                else:
                    Widget._ShowWidget(?UnlockPermanent := UnlockPermanent)

    UnlockWidget(GroupName : string, WidgetName : string):void=
        if (Group := GroupUIMap[GroupName]):
            Widgets := Group.OrderedWidgets
            for(Widget : Widgets, Widget.Name = WidgetName):
                Widget.UnlockWidget()

    LockWidget(GroupName : string, WidgetName : string):void=
        if (Group := GroupUIMap[GroupName]):
            Widgets := Group.OrderedWidgets
            for(Widget : Widgets, Widget.Name = WidgetName):
                Widget.LockWidget()

    ShowGroup(GroupName : string, AllDecendants : logic, ?UnlockPermanent : logic = false):void=
        ShowGroup_Internal(GroupName, ?UnlockPermanent := UnlockPermanent)
        if (AllDecendants = true):
            # This function only gets children names
            GroupNames := GetAllDecendantGroupNames(GroupName)
            for(Name : GroupNames):
                ShowGroup_Internal(Name, ?UnlockPermanent := UnlockPermanent)
        

    ShowGroup_Internal<private>(GroupName : string, ?UnlockPermanent : logic = false):void=
        if (Group := GroupUIMap[GroupName]):
            Group._ShowWidget(?UnlockPermanent := UnlockPermanent)
            Widgets := Group.OrderedWidgets
            for(Widget : Widgets):
                Widget._ShowWidget(?UnlockPermanent := UnlockPermanent)
            if (Group.AddedButtons = false): 
                set Group.AddedButtons = true
                Print("Adding buttons")
                for(Button : Group.ButtonSlots):
                    Canvas.AddWidget(Button)





    CopydWidgetDataIntoGroupData(WidgetData : Parent_Widget_Data):void=
        WidgetName := WidgetData.Name
        WidgetGroupName := WidgetData.GroupName
        if(GroupData := GroupUIMap[WidgetGroupName]):
            set GroupData.OrderedWidgets += array{WidgetData}
            set WidgetData.bUIData = option{Self}

            if (TextData := Text_Block[WidgetData]):
                if(GroupData.TextBlocks[WidgetName]):
                    if(set GroupData.TextBlocks[WidgetName] += array{TextData}){}
                else:
                    if(set GroupData.TextBlocks[WidgetName] = array{TextData}){}

            if (ImageData := Image_Block[WidgetData]):
                if(GroupData.ImageBlocks[WidgetName]):
                    if(set GroupData.ImageBlocks[WidgetName] += array{ImageData}){}
                else:
                    if(set GroupData.ImageBlocks[WidgetName] = array{ImageData}){}

            if (ColorData := Color_Block[WidgetData]):
                if(GroupData.ColorBlocks[WidgetName]):
                    if(set GroupData.ColorBlocks[WidgetName] += array{ColorData}){}
                else:
                    if(set GroupData.ColorBlocks[WidgetName] = array{ColorData}){}

            if (ButtonQuietData := Button_Quiet_Data[WidgetData]):
                if(GroupData.ButtonsQuiet[WidgetName]):
                    if(set GroupData.ButtonsQuiet[WidgetName] += array{ButtonQuietData}){}
                else:
                    if(set GroupData.ButtonsQuiet[WidgetName] = array{ButtonQuietData}){}
                
            
    PostInit(Player : player):void=
        set bPlayer = option{Player}
        BindAllButtons()
        

    PreInit(Player : player, UIManager : V_Krita):void=
        set bUIManager = option{UIManager}
        CopydAllWidgetDataIntoGroupData()
        set bPlayer = option{Player}
        spawn:
            Tick()

    Tick()<suspends>:void=
        loop:
            Sleep(5.0)
            if (UIExited = true):
                break
            if(UIManager := bUIManager?, Player := bPlayer?):
                UIManager.UITick(Self, Player)

    BindAllButtons():void=
        for(Widget : OrderedWidgetData):
            if (Buttonwidget := Button_Quiet_Data[Widget]):
                NewBind := Button_Bind{ButtonName := Buttonwidget.Name, GroupName := Buttonwidget.GroupName, bUIData := option{Self}}
                Buttonwidget.Button.OnClick().Subscribe(NewBind.ButtonPressed)
                set ButtonBinds += array{NewBind}
                Buttonwidget.MasterButtonBind_Internal()
    
    ExitUI():void=
        Canvas.SetVisibility(widget_visibility.Hidden)
        if (Player := bPlayer?):
            if(PlayerUIDisplay := GetPlayerUI[Player]):
                PlayerUIDisplay.RemoveWidget(Canvas)
        for(Widget : OrderedWidgetData):
            Widget._ExitUI(Self)
        set UIExited = true
    

    OnButtonPressed(Name : string, GroupName : string, Message : widget_message):void=
        block:
        #Print("BUTTTTTTTON PRESED = {Name}")
        <# Inv buttons
        if (Name = "B_InvButton_1"):
            ShowOnlyGroups(array{"WeaponsMenu"})
        if (Name = "B_InvButton_2"):
            ShowOnlyGroups(array{"ItemsMenu"})

        if (Name = "B_WeaponButton_1"):
            if(WeaponGroup := GroupUIMap["WeaponsMenu"], NewImgBlocks := WeaponGroup.ImageBlocks["T_Weapon_Great"], NewImgBlock := NewImgBlocks[0]):
                if(Group := GroupUIMap["MainMenu"], OldImgBlocks := Group.ImageBlocks["T_Weapon_1"], OldImgBlock := OldImgBlocks[0]):
                    if(WTexture := NewImgBlock.Texture?, WImg := NewImgBlock.ImageTextureBlock?):
                        if(MImg := OldImgBlock.ImageTextureBlock?, MTexture := OldImgBlock.Texture?):
                            MImg.SetImage(WTexture)
                            WImg.SetImage(MTexture)
                            set NewImgBlock.Texture = option{MTexture}
                            set OldImgBlock.Texture = option{WTexture}
            ShowOnlyGroups(array{"MainMenu"})

        if (Name = "B_Back"):
            ShowOnlyGroups(array{"MainMenu"})#>

    
    ### ALL THE GETTERS

    GetAllDecendantGroupNames(GroupName : string):[]string=
        var Children : []string = array{}
        if(Group := GroupUIMap[GroupName]):
            set Children += Group.ChildGroups 
            for(ChildName : Group.ChildGroups):
                set Children += GetAllDecendantGroupNames(ChildName)
        return Children
            
    GetGroup(Name : string)<transacts><decides>:Group_UI=
        return GroupUIMap[Name]

    GetGroup(Name : string, AllDecendants : logic):[]Group_UI=
        var GroupNames : []string = array{Name}
        var Groups : []Group_UI = array{}
        if (AllDecendants = true):
            set GroupNames += GetAllDecendantGroupNames(Name)
        for(GroupName : GroupNames, Group := GroupUIMap[GroupName]):
            set Groups += array{Group}
        return Groups

    GetGroupWidgets(Name : string, AllDecendants : logic):[]Parent_Widget_Data=
        var Widgets : []Parent_Widget_Data = array{}
        var GroupNames : []string = array{Name}
        if (AllDecendants = true):
            set GroupNames += GetAllDecendantGroupNames(Name)
        for(GroupName : GroupNames, Group := GetGroup[GroupName]):
            set Widgets += Group.OrderedWidgets
        return Widgets

    GetButtons(GroupName : string, ButtonName : string)<transacts><decides>:[]Button_Quiet_Data=
        GroupUIMap[GroupName].ButtonsQuiet[ButtonName]

    GetButtons(GroupNames : string, ButtonName : string, AllDecendants : logic):[]Button_Quiet_Data=
        return GetGroupsButtons(array{GroupNames}, ButtonName, AllDecendants)

    GetGroupsButtons(GroupNames : []string, ButtonName : string, AllDecendants : logic):[]Button_Quiet_Data=
        var AllGroups : []string = array{}
        for(GroupName : GroupNames, not AllGroups.Find[GroupName]):
            set AllGroups += array{GroupName}

        if (AllDecendants = true):
            var ChildGroups : []string = array{}
            for(GroupName : AllGroups):
                set ChildGroups += GetAllDecendantGroupNames(GroupName)
            for(GroupName : ChildGroups, not AllGroups.Find[GroupName]):
                set AllGroups += array{GroupName}

        var Buttons : []Button_Quiet_Data = array{}
        for(GroupName : AllGroups):
            if (TempButtons := GetButtons[GroupName, ButtonName]):
                set Buttons += TempButtons
        return Buttons

    GetTextBlocks(GroupName : string, TextName : string)<transacts><decides>:[]Text_Block=
        GroupUIMap[GroupName].TextBlocks[TextName]

    GetImageBlocks(GroupName : string, TextName : string)<transacts><decides>:[]Image_Block=
        GroupUIMap[GroupName].ImageBlocks[TextName]

    GetColorBlocks(GroupName : string, TextName : string)<transacts><decides>:[]Color_Block=
        GroupUIMap[GroupName].ColorBlocks[TextName] 


    ### ALL THE SETTERS

    SetButtonText(GroupName : string, ButtonName : string, Text : string):void=
        if(Buttons := GetButtons[GroupName, ButtonName]):
            for(Button : Buttons):
                Button.SetText(Text)
    SetButtonText(EnumName : tuple(string,string), Text : string):void=
        if(Buttons := GetButtons[EnumName(0), EnumName(1)]):
            for(Button : Buttons):
                Button.SetText(Text)
            

    SetText(GroupName : string, TextBlockName : string, Text : string):void=
        if(TextBlocks := GetTextBlocks[GroupName, TextBlockName]):
            for(TextBlock : TextBlocks):
                TextBlock.SetText(Text)
    SetText(EnumName : tuple(string,string), Text : string):void=
        if(TextBlocks := GetTextBlocks[EnumName(0), EnumName(1)]):
            for(TextBlock : TextBlocks):
                TextBlock.SetText(Text)

    SetTexture(GroupName : string, ImageBlockName : string, Texture : ?texture):void=
        if(ImageBlocks := GetImageBlocks[GroupName, ImageBlockName]):
            for(ImageBlock : ImageBlocks):
                ImageBlock.SetTexture(Texture)
    SetTexture(EnumName : tuple(string,string), Texture : ?texture):void=
        if(ImageBlocks := GetImageBlocks[EnumName(0), EnumName(1)]):
            for(ImageBlock : ImageBlocks):
                ImageBlock.SetTexture(Texture)

    CopydAllWidgetDataIntoGroupData<private>():void=
        for(WidgetData : OrderedWidgetData):
            CopydWidgetDataIntoGroupData(WidgetData)

    PrintALLGroupData():void=
        for(K -> V : GroupUIMap):
            Print("GroupName:{K}")
            Print("OrderedWidgets:")
            for(Widget:V.OrderedWidgets):
                Print("{Widget.Name}")

    ## ALL THE BINDS

    # Called by user, to bind function to all the buttons in this group that have this name
    BindButtons(GroupName : string, ButtonName : string, Function : button_fun)<transacts><decides>:[]Button_Event_Handle=
        GetButtons[GroupName, ButtonName].BindButtons(Function)

    # Called by user, to bind function to all the buttons in this group that have this name
    BindButtons(EnumName : tuple(string, string), Function : button_fun)<transacts><decides>:[]Button_Event_Handle=
        GetButtons[EnumName(0), EnumName(1)].BindButtons(Function)

    # Called by user, to bind function to the first button fond in this group by this name
    BindButton(GroupName : string, ButtonName : string, Function : button_fun, Index : int)<transacts><decides>:Button_Event_Handle=
        GetButtons[GroupName, ButtonName][Index].BindButton(Function)

    # Called by user, to bind function to the first button fond in this group by this name
    BindButton(EnumName : tuple(string, string), Function : button_fun, Index : int)<transacts><decides>:Button_Event_Handle=
        GetButtons[EnumName(0), EnumName(1)][Index].BindButton(Function)



    # Called by user, to bind function to buttons of multiple groups but same buton names
    BindGroupsButtons(GroupNames : []string, ButtonName : string, AllDecendants : logic, Function : button_fun):[]Button_Event_Handle=
        var ButtonHandles : []Button_Event_Handle = array{}
        Buttons := GetGroupsButtons(GroupNames, ButtonName, AllDecendants)
        for(Button : Buttons):
            set ButtonHandles += array{Button.BindButton(Function)}
        return ButtonHandles



    # Called by user, to bind function to buttons of group and their decendants but same buton names
    BindButtons(GroupName : string, ButtonName : string, AllDecendants : logic, Function : button_fun):[]Button_Event_Handle=
        var ButtonHandles : []Button_Event_Handle = array{}
        Buttons := GetGroupsButtons(array{GroupName}, ButtonName, AllDecendants)
        for(Button : Buttons):
            set ButtonHandles += array{Button.BindButton(Function)}
        return ButtonHandles

    # Called by user, to bind function to buttons of group and their decendants but same buton names
    BindButtons(EnumName : tuple(string, string), AllDecendants : logic, Function : button_fun):[]Button_Event_Handle=
        var ButtonHandles : []Button_Event_Handle = array{}
        Buttons := GetGroupsButtons(array{EnumName(0)}, EnumName(1), AllDecendants)
        for(Button : Buttons):
            set ButtonHandles += array{Button.BindButton(Function)}
        return ButtonHandles


    # Called by user, to bind function to all buttons inside this group or this group and their decendants
    BindAllButtonsInGroup(GroupName : string, AllDecendants : logic, Function : button_fun):[]Button_Event_Handle=
        var ButtonHandles : []Button_Event_Handle = array{}
        var Buttons : []Button_Quiet_Data = array{}
        Groups := GetGroup(GroupName, AllDecendants)
        for(Group : Groups):
            for(Name -> ButtonArray : Group.ButtonsQuiet):
                if (ButtonArray.Length > 0):
                    set Buttons += ButtonArray
        for(Button : Buttons):
            set ButtonHandles += array{Button.BindButton(Function)}
        return ButtonHandles


    # Called by user, to bind function to any button with this name
    BindAnyButtonsWithName(ButtonName : string, Function : button_fun):[]Button_Event_Handle=
        var ButtonHandles : []Button_Event_Handle = array{}
        for(Widget : OrderedWidgetData, ButtonWidget := Button_Quiet_Data[Widget], ButtonWidget.Name = ButtonName):
            set ButtonHandles += array{ButtonWidget.BindButton(Function)}
        return ButtonHandles

    # Called by user, to bind function to any button with this name
    BindAnyButtonsWithNames(ButtonNames : []string, Function : button_fun):[]Button_Event_Handle=
        var ButtonHandles : []Button_Event_Handle = array{}
        for(Widget : OrderedWidgetData, ButtonWidget := Button_Quiet_Data[Widget], ButtonNames.Find[ButtonWidget.Name]):
            set ButtonHandles += array{ButtonWidget.BindButton(Function)}
        return ButtonHandles

    # Called by user, to bind function to any button with this name
    BindAnyButtonsEndsWithName(ButtonEndName : string, Function : button_fun):[]Button_Event_Handle=
        var ButtonHandles : []Button_Event_Handle = array{}
        for(Widget : OrderedWidgetData, ButtonWidget := Button_Quiet_Data[Widget]):
            NameLength := ButtonWidget.Name.Length
            StartIndex := NameLength - ButtonEndName.Length
            if (EndName := ButtonWidget.Name.Slice[StartIndex], EndName = ButtonEndName):
                set ButtonHandles += array{ButtonWidget.BindButton(Function)}
        return ButtonHandles



    MoveGroupAdditive(GroupName : string, AllDecendants : logic, AddX : float, AddY : float, Speed : float, ?Seperate : logic = false, ?IgnoreButtons : logic = true)<suspends>:void=
        Widgets := GetGroupWidgets(GroupName, AllDecendants)
        if (Seperate = false):
            for(Widget : Widgets):
                if (Widget.Visibility = E_Widget_Visibility_Status.Visible):
                    spawn{MoveWidget(Widget, Widget.X + AddX, Widget.Y + AddY, Speed, ?IgnoreButton := IgnoreButtons)}
                else:
                    TeleportWidget(Widget, Widget.X + AddX, Widget.Y + AddY, ?IgnoreButton := IgnoreButtons)
        if (Seperate = true):
            for(Widget : Widgets):
                if (Widget.Visibility = E_Widget_Visibility_Status.Visible):
                    MoveWidget(Widget, Widget.X + AddX, Widget.Y + AddY, Speed, ?IgnoreButton := IgnoreButtons)
                else:
                    TeleportWidget(Widget, Widget.X + AddX, Widget.Y + AddY, ?IgnoreButton := IgnoreButtons)

    MoveWidget(Widget : Parent_Widget_Data, TargetX : float, TargetY : float, Speed : float, ?IgnoreButton : logic = true)<suspends>:void=
        if (IgnoreButton = true):
            if (Button_Data[Widget]):
                return 
        var Timer : float = 0.0
        Target :  vector2 = vector2{X := TargetX, Y := TargetY}

        var bCanvaWidget : ?widget = false
        if(TextWidgetData := Text_Block[Widget]):
            set bCanvaWidget = option{TextWidgetData.TextBlock}

        if(ColorWidgetData := Color_Block[Widget]):
            set bCanvaWidget = option{ColorWidgetData.Color}

        if(ImageWidgetData := Image_Block[Widget], Image := ImageWidgetData.Texture?):
            if (ImageTextureBlock := ImageWidgetData.ImageTextureBlock?):
                set bCanvaWidget = option{ImageTextureBlock}

        if(ButtonWidgetData := Button_Quiet_Data[Widget]):
            set bCanvaWidget = option{ButtonWidgetData.Button}
        loop:
            Current := vector2{X := Widget.X, Y := Widget.Y}
            set Timer = Min(Timer + Speed, 1.0)
            if (Current.X = Target.X, Current.Y = Target.Y):
                break
            NewLoc := Lerp(Current, Target, Timer)
            set Widget.X = NewLoc.X
            set Widget.Y = NewLoc.Y

            if(CanvaWidget := bCanvaWidget?):
                NewSlot := canvas_slot:
                            Anchors := anchors{Minimum := vector2{X:=Widget.AnchorMinX, Y:=Widget.AnchorMinY}, Maximum := vector2{X:=Widget.AnchorMaxX, Y:=Widget.AnchorMaxY}}
                            Alignment := Widget.Alignment
                            Offsets := margin{Left := Widget.X, Top := Widget.Y, Right := Widget.ScaleX * 1.0, Bottom := Widget.ScaleY * 1.0}
                            SizeToContent := Widget.SizeToContent
                            ZOrder := Widget.Order
                            Widget := CanvaWidget
                if(PreviousCanvaSlot := Widget.bCanvasSlot?):
                    PreviousCanvaSlot.Widget.SetVisibility(widget_visibility.Hidden)
                    Canvas.RemoveWidget(PreviousCanvaSlot.Widget)
                    PreviousCanvaSlot.Widget.SetVisibility(widget_visibility.Visible)
                    set Widget.bCanvasSlot = option{NewSlot}
                    Canvas.AddWidget(NewSlot)
                    Print("Adding widget{Widget.X}")
            Sleep(0.0)

    TeleportWidget(Widget : Parent_Widget_Data, TargetX : float, TargetY : float, ?IgnoreButton : logic = true):void=
        if (IgnoreButton = true):
            if (Button_Data[Widget]):
                return 

        set Widget.X = TargetX
        set Widget.Y = TargetY

        var bCanvaWidget : ?widget = false

        if(TextWidgetData := Text_Block[Widget]):
            set bCanvaWidget = option{TextWidgetData.TextBlock}

        if(ColorWidgetData := Color_Block[Widget]):
            set bCanvaWidget = option{ColorWidgetData.Color}

        if(ImageWidgetData := Image_Block[Widget], Image := ImageWidgetData.Texture?):
            if (ImageTextureBlock := ImageWidgetData.ImageTextureBlock?):
                set bCanvaWidget = option{ImageTextureBlock}

        if(ButtonWidgetData := Button_Quiet_Data[Widget]):
            set bCanvaWidget = option{ButtonWidgetData.Button}

        if(CanvaWidget := bCanvaWidget?):
            NewSlot := canvas_slot:
                        Anchors := anchors{Minimum := vector2{X:=Widget.AnchorMinX, Y:=Widget.AnchorMinY}, Maximum := vector2{X:=Widget.AnchorMaxX, Y:=Widget.AnchorMaxY}}
                        Alignment := Widget.Alignment
                        Offsets := margin{Left := Widget.X, Top := Widget.Y, Right := Widget.ScaleX * 1.0, Bottom := Widget.ScaleY * 1.0}
                        SizeToContent := Widget.SizeToContent
                        ZOrder := Widget.Order
                        Widget := CanvaWidget
            if(PreviousCanvaSlot := Widget.bCanvasSlot?):
                PreviousCanvaSlot.Widget.SetVisibility(widget_visibility.Hidden)
                Canvas.RemoveWidget(PreviousCanvaSlot.Widget)
                PreviousCanvaSlot.Widget.SetVisibility(widget_visibility.Visible)
                set Widget.bCanvasSlot = option{NewSlot}
                Canvas.AddWidget(NewSlot)


# A Verse-authored creative device that can be placed in a level
V_UIClasses := class(creative_device):

    # Runs when the device is started in a running game
    OnBegin<override>()<suspends>:void=
        # TODO: Replace this with your code
        Print("Hello, world!")
        Print("2 + 2 = {2 + 2}")